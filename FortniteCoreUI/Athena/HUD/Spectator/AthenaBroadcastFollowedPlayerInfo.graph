WidgetBlueprint /FortniteCoreUI/Athena/HUD/Spectator/AthenaBroadcastFollowedPlayerInfo
graphs [EventGraph {nodes [
   n1 = Event OnEliminatedStateChanged!() -> (OutputDelegate, then, PlayerData:Athena Remote Player View Data Object Reference, bIsEliminated:bool) {is=title K2Node_Event}
   > SetVisibility(
    self=Get LoadoutSB() -> (LoadoutSB:Size Box Object Reference)
    InVisibility=Select(Option 0=SelfHitTestInvisible, Option 1=Hidden, Index=@n1.bIsEliminated)
   )
   n2 = Event OnPipModeChanged!() -> (OutputDelegate, then, bPipModeEnabled:bool) {is=title K2Node_Event}
   > SetVisibility(
    self=Get PlayerWidget() -> (PlayerWidget:Athena Spectator Player Widget Base Object Reference)
    InVisibility=Select(Option 0=SelfHitTestInvisible, Option 1=Collapsed, Index=@n2.bPipModeEnabled)
   )
   n3 = Event OnVehicleStateChanged!() -> (OutputDelegate, then, NewVehicle:Actor Object Reference, OldVehicle:Actor Object Reference) {is=title K2Node_Event}
   > n4 = ShowVehicleHealthBarOnPlayerHUD!(execute:Exec, self:Object Reference=@n3.NewVehicle) -> (then:Exec, ReturnValue:bool) {is=title K2Node_Message}
   > n5 = Get All Passengers!(execute:Exec, self:Object Reference=@n3.NewVehicle) -> (then:Exec, ReturnValue:Array of Fort Player Pawn Object References) {is=title K2Node_Message}
   > n6 = GetFollowedPlayerData!(execute, self:Athena Broadcast Followed Player Info Base Object Reference) -> (then, ReturnValue:Athena Remote Player View Data Object Reference)
   > Is Valid(InputObject=@n6)
   > n7 = Branch(Condition=Contains Item(
     TargetArray:Array of Fort Player Pawn Object References=@n5
     ItemToFind:"Fort Player Pawn Object Reference (by ref)"=GetPlayerPawn!(self:Athena Remote Player View Data Object Reference=@n6) -> (ReturnValue:Fort Player Pawn Object Reference)
    ))
   > SetVisibility(
    self=Get VehicleHealthSB() -> (VehicleHealthSB:Size Box Object Reference)
    InVisibility=Select(Option 0=Collapsed, Option 1=SelfHitTestInvisible, Index=@n4)
   )
   SetVisibility(execute=@n7.else, self=Get VehicleHealthSB(), InVisibility=Collapsed)
  ]}]
variables []
widgets {
 info {
  inheritedWidgetCount=5
  namedSlotCount=2
  parentClass=/Script/FortniteUI.AthenaBroadcastFollowedPlayerInfoBase
  rootWidgetClass=&HorizontalBox
  widgetCount=29
 }
 widgets [
  HorizontalBox_1 {widgetClassPath=&HorizontalBox, widget=HorizontalBox_1}
  InventoryVert {widgetClassPath=&VerticalBox, parent=HorizontalBox_1, widget=InventoryVert, slot {class=HorizontalBoxSlot}}
  ShieldHealth+InvSize {
   widgetClassPath=&SizeBox
   parent=InventoryVert
   widget=ShieldHealth+InvSize
   properties {bOverride_WidthOverride=true, heightOverride=145, maxDesiredHeight=153, widthOverride=550}
   slot {class=VerticalBoxSlot, properties {horizontalAlignment=HAlign_Center, verticalAlignment=VAlign_Bottom}}
  }
  HorizontalBox_4 {widgetClassPath=&HorizontalBox, parent=ShieldHealth+InvSize, widget=HorizontalBox_4, slot {class=SizeBoxSlot}}
  VerticalBox_1 {
   widgetClassPath=&VerticalBox
   parent=HorizontalBox_4
   widget=VerticalBox_1
   slot {class=HorizontalBoxSlot, properties {size {sizeRule=Fill, value=1}}}
  }
  PlayerWidget {
   widgetClassPath=/FortniteCoreUI/Athena/HUD/Spectator/SquadStatus/AthenaBroadcastPlayerStatusTG
   bInherited=true
   bIsVariable=true
   parent=VerticalBox_1
   widget=PlayerWidget
   slot {class=VerticalBoxSlot}
  }
  ShieldHealth+BuildSize {
   widgetClassPath=&SizeBox
   parent=VerticalBox_1
   widget=ShieldHealth+BuildSize
   properties {heightOverride=62, maxDesiredHeight=62}
   slot {class=VerticalBoxSlot, properties {padding=<ltrb 0 0 0 6>}}
  }
  Overlay_1 {widgetClassPath=&Overlay, parent=ShieldHealth+BuildSize, widget=Overlay_1, slot {class=SizeBoxSlot}}
  ShieldHealth+BuildBkg {
   widgetClassPath=&Image
   parent=Overlay_1
   widget=ShieldHealth+BuildBkg
   properties {
    visibility=SelfHitTestInvisible
    brush {
     bIsDynamicallyLoaded=false
     drawAs=Image
     imageType=NoImage
     mirroring=NoMirror
     resourceName=None
     resourceObject=null
     tiling=NoTile
     imageSize=<32 32>
     margin=<ltrb 0 0 0 0>
     outlineSettings {
      bUseBrushTransparency=false
      roundingType=HalfHeightRadius
      width=0
      color {colorUseRule=UseColor_Specified, specifiedColor=<rgba 0 0 0 0>}
      cornerRadii=<0 0 0 0>
     }
     tintColor {colorUseRule=UseColor_Specified, specifiedColor=<rgba 0 0 0 0.3>}
     uVRegion {bIsValid=false, max=<0 0>, min=<0 0>}
    }
   }
   slot {class=OverlaySlot, properties {horizontalAlignment=HAlign_Fill, verticalAlignment=VAlign_Fill}}
  }
  ShieldHealth+BuildHoriz {
   widgetClassPath=&HorizontalBox
   parent=Overlay_1
   widget=ShieldHealth+BuildHoriz
   slot {class=OverlaySlot, properties {horizontalAlignment=HAlign_Fill}}
  }
  VerticalBox_0 {
   widgetClassPath=&VerticalBox
   parent=ShieldHealth+BuildHoriz
   widget=VerticalBox_0
   slot {class=HorizontalBoxSlot, properties {padding=<ltrb 5 0 0 0>, size {sizeRule=Fill, value=1}}}
  }
  VehicleHealthSB {
   widgetClassPath=&SizeBox
   bIsVariable=true
   parent=VerticalBox_0
   widget=VehicleHealthSB
   properties {bOverride_HeightOverride=true, bOverride_WidthOverride=true, heightOverride=24, widthOverride=420}
   slot {class=VerticalBoxSlot, properties {padding=<ltrb 0 5 0 0>, size {sizeRule=Fill, value=1}}}
  }
  HitPointBar_Vehicle {
   widgetClassPath=/Game/Athena/HUD/Spectator/AthenaSpectatorHitPointBar
   bInherited=true
   bIsVariable=true
   parent=VehicleHealthSB
   widget=HitPointBar_Vehicle
   properties {barType=VehicleHealth, showNumbers=true}
   slot {class=SizeBoxSlot}
  }
  ShieldSB {
   widgetClassPath=&SizeBox
   parent=VerticalBox_0
   widget=ShieldSB
   properties {bOverride_HeightOverride=true, bOverride_WidthOverride=true, heightOverride=24, widthOverride=420}
   slot {class=VerticalBoxSlot, properties {padding=<ltrb 0 5 0 0>, size {sizeRule=Fill, value=1}}}
  }
  HitPointBar_Shield {
   widgetClassPath=&AthenaSpectatorHitPointBar
   bInherited=true
   bIsVariable=true
   parent=ShieldSB
   widget=HitPointBar_Shield
   properties {barType=Shield, hideShieldOnDbno=false, showNumbers=true}
   slot {class=SizeBoxSlot}
  }
  HealthSB {
   widgetClassPath=&SizeBox
   parent=VerticalBox_0
   widget=HealthSB
   properties {bOverride_HeightOverride=true, bOverride_WidthOverride=true, heightOverride=24, widthOverride=420}
   slot {class=VerticalBoxSlot, properties {padding=<ltrb 0 5 0 5>, size {sizeRule=Fill, value=1}}}
  }
  HitPointBar_Health {
   widgetClassPath=&AthenaSpectatorHitPointBar
   bInherited=true
   bIsVariable=true
   parent=HealthSB
   widget=HitPointBar_Health
   properties {showNumbers=true}
   slot {class=SizeBoxSlot}
  }
  BuildCountWrapper {
   widgetClassPath=/Game/Athena/HUD/FortHUDElementWrapper
   parent=ShieldHealth+BuildHoriz
   widget=BuildCountWrapper
   properties {elementTag {tagName=Athena.Replay.Client}}
   slot {class=HorizontalBoxSlot}
  }
  AthenaSpectatorBuildCountWidget {
   widgetClassPath=/Game/Athena/HUD/Spectator/Inventory/AthenaSpectatorBuildCountWidget
   namedSlotHost=BuildCountWrapper
   widget=AthenaSpectatorBuildCountWidget
   properties {renderTransform {angle=0, scale=<1 1>, shear=<0 0>, translation=<-10 0>}}
  }
  LoadoutWrapper {
   widgetClassPath=&FortHUDElementWrapper
   parent=VerticalBox_1
   widget=LoadoutWrapper
   properties {elementTag {tagName=HUD.TopLeft.PlayerHitPoints}}
   slot {class=VerticalBoxSlot}
  }
  LoadoutSB {
   widgetClassPath=&SizeBox
   bIsVariable=true
   namedSlotHost=LoadoutWrapper
   widget=LoadoutSB
   properties {bOverride_HeightOverride=true, heightOverride=80}
  }
  HorizontalBox_0 {widgetClassPath=&HorizontalBox, parent=LoadoutSB, widget=HorizontalBox_0, slot {class=SizeBoxSlot}}
  SizeBox_0 {
   widgetClassPath=&SizeBox
   parent=HorizontalBox_0
   widget=SizeBox_0
   properties {bOverride_WidthOverride=true, widthOverride=425}
   slot {class=HorizontalBoxSlot, properties {verticalAlignment=VAlign_Center}}
  }
  ScaleBox_2 {widgetClassPath=&ScaleBox, parent=SizeBox_0, widget=ScaleBox_2, slot {class=SizeBoxSlot}}
  AthenaSpectatorWeaponLoadout {
   widgetClassPath=/FortniteCoreUI/Athena/HUD/Spectator/Inventory/AthenaSpectatorWeaponLoadout
   parent=ScaleBox_2
   widget=AthenaSpectatorWeaponLoadout
   slot {class=ScaleBoxSlot}
  }
  SizeBox_1 {
   widgetClassPath=&SizeBox
   parent=HorizontalBox_0
   widget=SizeBox_1
   properties {bOverride_WidthOverride=true, widthOverride=125}
   slot {class=HorizontalBoxSlot, properties {verticalAlignment=VAlign_Center}}
  }
  ScaleBox_0 {widgetClassPath=&ScaleBox, parent=SizeBox_1, widget=ScaleBox_0, slot {class=SizeBoxSlot}}
  AthenaSpectatorTrapsLoadout {
   widgetClassPath=/FortniteCoreUI/Athena/HUD/Spectator/Inventory/AthenaSpectatorTrapsLoadout
   parent=ScaleBox_0
   widget=AthenaSpectatorTrapsLoadout
   slot {class=ScaleBoxSlot}
  }
  TutorialHighlight {widgetClassPath=&TutorialHighlight}
 ]
}
