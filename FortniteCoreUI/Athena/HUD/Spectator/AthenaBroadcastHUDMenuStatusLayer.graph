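# Widget Blueprint layer for the Athena broadcast spectator HUD that hosts HUD menu status widgets as actor indicators on a FortActorCanvas.
# The AddHUDMenuStatus event from the native layer base adds the given status widget as an indicator and disables auto-removal on its slot.
# The RemoveHUDMenuStatus event removes that widget from the canvas, so indicator lifetime is driven only by these events.
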WidgetBlueprint /FortniteCoreUI/Athena/HUD/Spectator/AthenaBroadcastHUDMenuStatusLayer
graphs [EventGraph {nodes [
   n1 = Event AddHUDMenuStatus!() -> (OutputDelegate, then, HUDMenuStatusWidget:Athena Broadcast HUDMenu Status Base Object Reference) {is=title K2Node_Event}
   > n2 = AddActorIndicator!(
    execute
    self:Fort Actor Canvas Object Reference=Get HUDMenuStatusCanvas() -> (HUDMenuStatusCanvas:Fort Actor Canvas Object Reference)
    Indicator:Fort Actor Indicator Widget Object Reference=@n1.HUDMenuStatusWidget
   ) -> (then, ReturnValue:Fort Actor Canvas Slot Object Reference)
   > SetCanAutoRemove!(execute, self:Fort Actor Canvas Slot Object Reference=@n2, bAllowAutoRemove:bool=false) -> (then)
   n3 = Event RemoveHUDMenuStatus!() -> (OutputDelegate, then, HUDMenuStatusWidget:Athena Broadcast HUDMenu Status Base Object Reference) {is=title K2Node_Event}
   > RemoveChild(self=Get HUDMenuStatusCanvas(), Content=@n3.HUDMenuStatusWidget)
  ]}]
variables []
widgets {
 info {
  inheritedWidgetCount=1
  parentClass=/Script/FortniteUI.AthenaBroadcastHUDMenuStatusLayerBase
  rootWidgetClass=/Script/FortniteUI.FortActorCanvas
  widgetCount=2
 }
 widgets [
  HUDMenuStatusCanvas {widgetClassPath=&FortActorCanvas, bIsVariable=true, widget=HUDMenuStatusCanvas}
  TutorialHighlight {widgetClassPath=&TutorialHighlight}
 ]
}
