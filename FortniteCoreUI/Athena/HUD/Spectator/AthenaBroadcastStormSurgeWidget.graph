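# Spectator broadcast HUD widget for Storm Surge, showing whether the followed player is above or below the damage threshold.
# Base-class events set the title with an intro animation and switch Above/Equal, Below or None states, using open flags to avoid replaying animations.
# Opening plays alert sounds; Below also recolors the skewed-box material and drives the nested icon's animations.
# Followed-player changes call UpdateUI, and PreConstruct realigns the layout when isBacchus is set.
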WidgetBlueprint /FortniteCoreUI/Athena/HUD/Spectator/AthenaBroadcastStormSurgeWidget
animations [
 DMGLoad {
  binding_count=4
  display_label=""
  length_seconds=0.75
  object=&self:DMGLoad
  sequence {bindings [
    ThresholdText {
     id=#1
     possessed_class=CommonTextBlock
     tracks [{
       type=MovieScene2DTransformTrack
       sections [{
         type=MovieScene2DTransformSection
         end_frame=15
         start_frame=0
         channels [
          Translation.X {type=MovieSceneScriptingFloatChannel, const [0,1]}
          Translation.Y {type=MovieSceneScriptingFloatChannel, const [0,1]}
          Angle {type=MovieSceneScriptingFloatChannel, const [0,1]}
          Scale.X {type=MovieSceneScriptingFloatChannel, keys [[0,0],[9,0],[11,1]]}
          Scale.Y {type=MovieSceneScriptingFloatChannel, keys [[0,0],[9,0],[11,1]]}
          Shear.X {type=MovieSceneScriptingFloatChannel, const [0,1]}
          Shear.Y {type=MovieSceneScriptingFloatChannel, const [0,1]}
         ]
        }]
      }]
    }
    CommonBorder_0 {
     id=#2
     possessed_class=CommonBorder
     tracks [{
       type=MovieSceneWidgetMaterialTrack
       sections [{
         type=MovieSceneParameterSection
         end_frame=14
         start_frame=0
         channels [
          RGB_AR {type=MovieSceneScriptingFloatChannel, const [0.0444,1]}
          RGB_AG {type=MovieSceneScriptingFloatChannel, const [0.0795,1]}
          RGB_AB {type=MovieSceneScriptingFloatChannel, const [0.4062,1]}
          RGB_AA {type=MovieSceneScriptingFloatChannel, const [1,1]}
          RGB_BR {type=MovieSceneScriptingFloatChannel, const [0.2188,1]}
          RGB_BG {type=MovieSceneScriptingFloatChannel, const [0.5376,1]}
          RGB_BB {type=MovieSceneScriptingFloatChannel, const [0.6562,1]}
          RGB_BA {type=MovieSceneScriptingFloatChannel, const [1,1]}
         ]
        }]
      }]
    }
    SurgeIcon_SizeBox {
     id=#3
     display_name=StormSurgeIcon_NoBkg
     possessed_class=SizeBox
     tracks [{
       type=MovieScene2DTransformTrack
       sections [{
         type=MovieScene2DTransformSection
         end_frame=15
         start_frame=0
         channels [
          Translation.X {type=MovieSceneScriptingFloatChannel, const [0,1]}
          Translation.Y {type=MovieSceneScriptingFloatChannel, const [0,1]}
          Angle {type=MovieSceneScriptingFloatChannel, const [0,1]}
          Scale.X {type=MovieSceneScriptingFloatChannel, keys [[0,0],[5,1],[7,0.9],[9,1]]}
          Scale.Y {type=MovieSceneScriptingFloatChannel, keys [[0,0],[5,1],[7,0.9],[9,1]]}
          Shear.X {type=MovieSceneScriptingFloatChannel, const [0,1]}
          Shear.Y {type=MovieSceneScriptingFloatChannel, const [0,1]}
         ]
        }]
      }]
    }
    StormDMG_HorizontalBox {
     id=#4
     possessed_class=VerticalBox
     tracks [{
       type=MovieScene2DTransformTrack
       sections [{
         type=MovieScene2DTransformSection
         end_frame=15
         start_frame=0
         channels [
          Translation.X {type=MovieSceneScriptingFloatChannel, const [0,1]}
          Translation.Y {type=MovieSceneScriptingFloatChannel, const [0,1]}
          Angle {type=MovieSceneScriptingFloatChannel, const [0,1]}
          Scale.X {type=MovieSceneScriptingFloatChannel, keys [[0,0],[5,0],[7,1],[9,0.9],[11,1]]}
          Scale.Y {type=MovieSceneScriptingFloatChannel, const [1,1]}
          Shear.X {type=MovieSceneScriptingFloatChannel, const [0,1]}
          Shear.Y {type=MovieSceneScriptingFloatChannel, const [0,1]}
         ]
        }]
      }]
    }
   ]}
 }
 AboveDmg_Open {
  binding_count=2
  display_label=""
  length_seconds=0.75
  object=&self:AboveDmg_Open
  sequence {bindings [
    AboveDMG {
     id=#5
     possessed_class=CommonTextBlock
     tracks [
      {
       type=MovieScene2DTransformTrack
       sections [{
         type=MovieScene2DTransformSection
         end_frame=14
         start_frame=0
         channels [
          Translation.X {type=MovieSceneScriptingFloatChannel, const [0,1]}
          Translation.Y {type=MovieSceneScriptingFloatChannel, keys [[0,20],[6,0]]}
          Angle {type=MovieSceneScriptingFloatChannel, const [0,1]}
          Scale.X {type=MovieSceneScriptingFloatChannel, keys [[0,0],[6,1],[8,1.1],[10,1]]}
          Scale.Y {type=MovieSceneScriptingFloatChannel, keys [[0,0],[6,1],[8,1.1],[10,1]]}
          Shear.X {type=MovieSceneScriptingFloatChannel, const [0,1]}
          Shear.Y {type=MovieSceneScriptingFloatChannel, const [0,1]}
         ]
        }]
      }
      {
       type=MovieSceneFloatTrack
       display_name=Render Opacity
       sections [{type=MovieSceneFloatSection, end_frame=14, start_frame=0, channels [None {type=MovieSceneScriptingFloatChannel, keys [[0,0],[6,1]]}]}]
      }
     ]
    }
    CommonBorder_0 {
     id=#6
     possessed_class=CommonBorder
     tracks [{
       type=MovieSceneWidgetMaterialTrack
       sections [{
         type=MovieSceneParameterSection
         end_frame=14
         start_frame=0
         channels [
          RGB_AR {type=MovieSceneScriptingFloatChannel, const [0.0444,1]}
          RGB_AG {type=MovieSceneScriptingFloatChannel, const [0.0795,1]}
          RGB_AB {type=MovieSceneScriptingFloatChannel, const [0.4062,1]}
          RGB_AA {type=MovieSceneScriptingFloatChannel, const [1,1]}
          RGB_BR {type=MovieSceneScriptingFloatChannel, const [0.2188,1]}
          RGB_BG {type=MovieSceneScriptingFloatChannel, const [0.5376,1]}
          RGB_BB {type=MovieSceneScriptingFloatChannel, const [0.6562,1]}
          RGB_BA {type=MovieSceneScriptingFloatChannel, const [1,1]}
         ]
        }]
      }]
    }
   ]}
 }
 AboveDmg_Close {
  binding_count=2
  display_label=""
  length_seconds=0.75
  object=&self:AboveDmg_Close
  sequence {bindings [
    CommonBorder_0 {
     id=#7
     possessed_class=CommonBorder
     tracks [{
       type=MovieSceneWidgetMaterialTrack
       sections [{
         type=MovieSceneParameterSection
         end_frame=14
         start_frame=0
         channels [
          RGB_AR {type=MovieSceneScriptingFloatChannel, const [0.0444,1]}
          RGB_AG {type=MovieSceneScriptingFloatChannel, const [0.0795,1]}
          RGB_AB {type=MovieSceneScriptingFloatChannel, const [0.4062,1]}
          RGB_AA {type=MovieSceneScriptingFloatChannel, const [1,1]}
          RGB_BR {type=MovieSceneScriptingFloatChannel, const [0.2188,1]}
          RGB_BG {type=MovieSceneScriptingFloatChannel, const [0.5376,1]}
          RGB_BB {type=MovieSceneScriptingFloatChannel, const [0.6562,1]}
          RGB_BA {type=MovieSceneScriptingFloatChannel, const [1,1]}
         ]
        }]
      }]
    }
    AboveDMG {
     id=#8
     possessed_class=CommonTextBlock
     tracks [
      {
       type=MovieScene2DTransformTrack
       sections [{
         type=MovieScene2DTransformSection
         end_frame=15
         start_frame=0
         channels [
          Translation.X {type=MovieSceneScriptingFloatChannel, const [0,1]}
          Translation.Y {type=MovieSceneScriptingFloatChannel, keys [[0,0],[3,0],[7,20]]}
          Angle {type=MovieSceneScriptingFloatChannel, const [0,1]}
          Scale.X {type=MovieSceneScriptingFloatChannel, keys [[0,1],[7,0]]}
          Scale.Y {type=MovieSceneScriptingFloatChannel, keys [[0,1],[7,0]]}
          Shear.X {type=MovieSceneScriptingFloatChannel, const [0,1]}
          Shear.Y {type=MovieSceneScriptingFloatChannel, const [0,1]}
         ]
        }]
      }
      {
       type=MovieSceneFloatTrack
       display_name=Render Opacity
       sections [{type=MovieSceneFloatSection, end_frame=14, start_frame=0, channels [None {type=MovieSceneScriptingFloatChannel, keys [[0,1],[7,0]]}]}]
      }
     ]
    }
   ]}
 }
 BelowDmg_Open {
  binding_count=2
  display_label=""
  length_seconds=0.75
  object=&self:BelowDmg_Open
  sequence {bindings [
    BelowDMG {
     id=#9
     possessed_class=CommonTextBlock
     tracks [
      {
       type=MovieScene2DTransformTrack
       sections [{
         type=MovieScene2DTransformSection
         end_frame=15
         start_frame=0
         channels [
          Translation.X {type=MovieSceneScriptingFloatChannel, const [0,1]}
          Translation.Y {type=MovieSceneScriptingFloatChannel, keys [[0,-20],[6,0]]}
          Angle {type=MovieSceneScriptingFloatChannel, const [0,1]}
          Scale.X {type=MovieSceneScriptingFloatChannel, keys [[0,0],[6,1],[8,1.1],[10,1]]}
          Scale.Y {type=MovieSceneScriptingFloatChannel, keys [[0,0],[6,1],[8,1.1],[10,1]]}
          Shear.X {type=MovieSceneScriptingFloatChannel, const [0,1]}
          Shear.Y {type=MovieSceneScriptingFloatChannel, const [0,1]}
         ]
        }]
      }
      {
       type=MovieSceneFloatTrack
       display_name=Render Opacity
       sections [{type=MovieSceneFloatSection, end_frame=14, start_frame=0, channels [None {type=MovieSceneScriptingFloatChannel, keys [[0,0],[6,1]]}]}]
      }
     ]
    }
    CommonBorder_0 {
     id=#10
     possessed_class=CommonBorder
     tracks [{
       type=MovieSceneWidgetMaterialTrack
       sections [{
         type=MovieSceneParameterSection
         end_frame=14
         start_frame=0
         channels [
          RGB_AR {type=MovieSceneScriptingFloatChannel, keys [[0,0.0444],[4,0.0612]]}
          RGB_AG {type=MovieSceneScriptingFloatChannel, keys [[0,0.0795],[4,0.0242]]}
          RGB_AB {type=MovieSceneScriptingFloatChannel, keys [[0,0.4062],[4,0.0612]]}
          RGB_AA {type=MovieSceneScriptingFloatChannel, const [1,1]}
          RGB_BR {type=MovieSceneScriptingFloatChannel, keys [[0,0.2188],[4,0.7011]]}
          RGB_BG {type=MovieSceneScriptingFloatChannel, keys [[0,0.5376],[4,0.1812]]}
          RGB_BB {type=MovieSceneScriptingFloatChannel, keys [[0,0.6562],[4,0.8632]]}
          RGB_BA {type=MovieSceneScriptingFloatChannel, const [1,1]}
         ]
        }]
      }]
    }
   ]}
 }
 BelowDmg_Close {
  binding_count=2
  display_label=""
  length_seconds=0.75
  object=&self:BelowDmg_Close
  sequence {bindings [
    BelowDMG {
     id=#11
     possessed_class=CommonTextBlock
     tracks [
      {
       type=MovieScene2DTransformTrack
       sections [{
         type=MovieScene2DTransformSection
         end_frame=14
         start_frame=0
         channels [
          Translation.X {type=MovieSceneScriptingFloatChannel, const [0,1]}
          Translation.Y {type=MovieSceneScriptingFloatChannel, keys [[0,0],[6,-20]]}
          Angle {type=MovieSceneScriptingFloatChannel, const [0,1]}
          Scale.X {type=MovieSceneScriptingFloatChannel, keys [[0,1],[6,0]]}
          Scale.Y {type=MovieSceneScriptingFloatChannel, keys [[0,1],[6,0]]}
          Shear.X {type=MovieSceneScriptingFloatChannel, const [0,1]}
          Shear.Y {type=MovieSceneScriptingFloatChannel, const [0,1]}
         ]
        }]
      }
      {
       type=MovieSceneFloatTrack
       display_name=Render Opacity
       sections [{type=MovieSceneFloatSection, end_frame=14, start_frame=0, channels [None {type=MovieSceneScriptingFloatChannel, keys [[0,1],[6,0]]}]}]
      }
     ]
    }
    CommonBorder_0 {
     id=#12
     possessed_class=CommonBorder
     tracks [{
       type=MovieSceneWidgetMaterialTrack
       sections [{
         type=MovieSceneParameterSection
         end_frame=15
         start_frame=0
         channels [
          RGB_AR {type=MovieSceneScriptingFloatChannel, keys [[0,0.0612],[4,0.0444]]}
          RGB_AG {type=MovieSceneScriptingFloatChannel, keys [[0,0.0242],[4,0.0795]]}
          RGB_AB {type=MovieSceneScriptingFloatChannel, keys [[0,0.0666],[4,0.4062]]}
          RGB_AA {type=MovieSceneScriptingFloatChannel, const [1,1]}
          RGB_BR {type=MovieSceneScriptingFloatChannel, keys [[0,0.7011],[4,0.2188]]}
          RGB_BG {type=MovieSceneScriptingFloatChannel, keys [[0,0.1812],[4,0.5376]]}
          RGB_BB {type=MovieSceneScriptingFloatChannel, keys [[0,0.8632],[4,0.6562]]}
          RGB_BA {type=MovieSceneScriptingFloatChannel, const [1,1]}
         ]
        }]
      }]
    }
   ]}
 }
]
graphs [
 Anim_AboveDMGClose {nodes [
   Anim_AboveDMGClose() {is=title K2Node_FunctionEntry}
   > PlayAnimation(InAnimation=Get AboveDmg_Close() -> (AboveDmg_Close:Widget Animation Object Reference))
  ]}
 Anim_AboveDMGOpen {nodes [
   Anim_AboveDMGOpen() {is=title K2Node_FunctionEntry}
   > PlayAnimation(InAnimation=Get AboveDmg_Open() -> (AboveDmg_Open:Widget Animation Object Reference))
   > PlaySound2D(Sound=Get SoundAboveOpen() -> (SoundAboveOpen:Sound Base Object Reference))
  ]}
 Anim_BelowDMGClose {nodes [
   Anim_BelowDMGClose() {is=title K2Node_FunctionEntry}
   > PlayAnimation(InAnimation=Get BelowDmg_Close() -> (BelowDmg_Close:Widget Animation Object Reference))
   n1 = Get AthenaStormSurgeIconWidget() -> (AthenaStormSurgeIconWidget:Athena Storm Surge Icon Widget Object Reference)
   > PlayAnimationForward(
    self=@n1
    InAnimation=Get IconDMG_Off!(self:Athena Storm Surge Icon Widget Object Reference=@n1) -> (IconDMG_Off:Widget Animation Object Reference)
   )
  ]}
 Anim_BelowDMGOpen {nodes [
   Anim_BelowDMGOpen() {is=title K2Node_FunctionEntry}
   > PlayAnimation(InAnimation=Get BelowDmg_Open() -> (BelowDmg_Open:Widget Animation Object Reference))
   n1 = Get AthenaStormSurgeIconWidget()
   > PlayAnimationForward(
    self=@n1
    InAnimation=Get IconDMG_On!(self:Athena Storm Surge Icon Widget Object Reference=@n1) -> (IconDMG_On:Widget Animation Object Reference)
   )
   > PlaySound2D(Sound=Get SoundBelowOpen() -> (SoundBelowOpen:Sound Base Object Reference))
  ]}
 Anim_Load {nodes [
   Anim_Load() {is=title K2Node_FunctionEntry}
   > PlayAnimation(InAnimation=Get DMGLoad() -> (DMGLoad:Widget Animation Object Reference))
  ]}
 EventGraph {nodes [
   OnFollowedPlayerChanged!() -> (
    OutputDelegate
    then
    SpectatingPC:Fort Player Controller Spectating Object Reference
    FollowedPlayerState:Fort Player State Object Reference
   ) {is=title K2Node_CustomEvent}
   > UpdateUI(self:Athena Storm Surge Widget Base Object Reference)
   Event PreConstruct()
   > Branch(Condition=Get isBacchus() -> (isBacchus:bool))
   > SetHorizontalAlignment(self=SlotAsOverlaySlot(Widget=Get DmgThresh_SizeBox() -> (DmgThresh_SizeBox:Size Box Object Reference)), InHorizontalAlignment=HAlign_Left)
   n1 = Get SurgeIcon_SizeBox() -> (SurgeIcon_SizeBox:Size Box Object Reference)
   > SetHorizontalAlignment(self=SlotAsOverlaySlot(Widget=@n1), InHorizontalAlignment=HAlign_Right)
   > SetRenderTranslation(self=@n1, Translation=(X=10,Y=0))
   > SetHorizontalAlignment(
    self:Border Slot Object Reference=SlotAsBorderSlot!(Widget:Widget Object Reference=Get ThresholdText() -> (ThresholdText:Common Text Object Reference)) -> (ReturnValue:Border Slot Object Reference)
    InHorizontalAlignment=HAlign_Left
   )
   n2 = Get AboveDMG() -> (AboveDMG:Common Text Object Reference)
   n3 = SlotAsVerticalBoxSlot(Widget=@n2)
   > SetHorizontalAlignment(self:Vertical Box Slot Object Reference=@n3, InHorizontalAlignment=HAlign_Left)
   > SetPadding!(execute, self:Vertical Box Slot Object Reference=@n3, InPadding:Margin Structure=Make Margin(Left=25, Top=0, Right=0, Bottom=0)) -> (then)
   > SetRenderTranslation(self=@n2, Translation=(X=25,Y=0))
   n4 = Get BelowDMG() -> (BelowDMG:Common Text Object Reference)
   n5 = SlotAsVerticalBoxSlot(Widget=@n4)
   > SetHorizontalAlignment(self=@n5, InHorizontalAlignment=HAlign_Left)
   > SetPadding(self=@n5, InPadding=Make Margin(Left=25, Top=0, Right=0, Bottom=0))
   > SetRenderTranslation(self=@n4, Translation=(X=25,Y=0))
   n6 = Event OnStartDisplaying!() -> (OutputDelegate, then, TitleText:"Text (by ref)") {is=title K2Node_Event}
   > n7 = Branch(Condition=Get bIsDisplaying() -> (bIsDisplaying:bool))
   "SetText (Text)"(execute=[@n8.then, @n7.then], self=Get ThresholdText(), InText=@n6.TitleText)
   Set bIsDisplaying(execute=@n7.else, bIsDisplaying:bool=true) -> (Output_Get:bool)
   > n8 = Anim_Load() {is=title K2Node_CallFunction}
   Event OnStopDisplaying() {is=title K2Node_Event}
   > Set bIsDisplaying(bIsDisplaying=false)
   n9 = Event OnUpdateThresholdDisplay!() -> (OutputDelegate, then, ThresholdText:"Text (by ref)", ThresholdType:EStormSurgeThresholdType Enum=None) {is=title K2Node_Event}
   > n10 = Switch on EStormSurgeThresholdType!(execute, Selection:EStormSurgeThresholdType Enum=@n9.ThresholdType) -> (None:Exec, Above:Exec, Below:Exec, Equal:Exec) {is=title K2Node_SwitchEnum}
   SetVisibility(execute=[@n10.Above, @n10.Equal], self=Get AboveDMG(), InVisibility=HitTestInvisible)
   > SetVisibility(self=Get BelowDMG(), InVisibility=Hidden)
   > "SetText (Text)"(self=Get AboveDMG(), InText=TextToUpper!(InText:"Text (by ref)"=@n9.ThresholdText) -> (ReturnValue:Text))
   > n11 = Branch(Condition=Get bBelowIsOpen() -> (bBelowIsOpen:bool))
   > Set bBelowIsOpen(bBelowIsOpen:bool=false) -> (Output_Get:bool)
   > n12 = Anim_BelowDMGClose() {is=title K2Node_CallFunction}
   n13 = Branch(execute=[@n12.then, @n11.else], Condition=Get bAboveIsOpen() -> (bAboveIsOpen:bool))
   Set bAboveIsOpen(execute=@n13.else, bAboveIsOpen:bool=true) -> (Output_Get:bool)
   > Anim_AboveDMGOpen() {is=title K2Node_CallFunction}
   SetVisibility(execute=@n10.Below, self=Get AboveDMG(), InVisibility=Hidden)
   > SetVisibility(self=Get BelowDMG(), InVisibility=HitTestInvisible)
   > "SetText (Text)"(self=Get BelowDMG(), InText=TextToUpper(InText=@n9.ThresholdText))
   > n14 = Branch(Condition=Get bAboveIsOpen())
   > Set bAboveIsOpen(bAboveIsOpen=false)
   > n15 = Anim_AboveDMGClose() {is=title K2Node_CallFunction}
   n16 = Branch(execute=[@n15.then, @n14.else], Condition=Get bBelowIsOpen())
   Set bBelowIsOpen(execute=@n16.else, bBelowIsOpen=true)
   > Anim_BelowDMGOpen() {is=title K2Node_CallFunction}
   SetVisibility(execute=@n10.None, self=Get AboveDMG(), InVisibility=Hidden)
   > SetVisibility(self=Get BelowDMG(), InVisibility=Hidden)
   > n17 = Branch(Condition=Get bAboveIsOpen())
   > Set bAboveIsOpen(bAboveIsOpen=false)
   > n18 = Anim_AboveDMGClose()
   Branch(execute=[@n18.then, @n17.else], Condition=Get bBelowIsOpen())
   > Set bBelowIsOpen(bBelowIsOpen=false)
   > Anim_BelowDMGClose()
   Event Construct()
   > n19 = Cast To FortPlayerControllerSpectating!(execute, Object=GetOwningPlayer()) -> (then, CastFailed, AsFort Player Controller Spectating:Fort Player Controller Spectating Object Reference) {is=title K2Node_DynamicCast}
   > Bind Event to On Followed Player Changed(self:Fort Player Controller Spectating Object Reference=@n19, Delegate=Create Event()) {is=title K2Node_AddDelegate}
   SetRenderTransform!(
    execute
    self:Widget Object Reference=@n4
    InTransform:Widget Transform Structure=Make Widget Transform!(Translation:Vector 2D Structure=(X=25,Y=0), Scale:Vector 2D Structure=(X=1,Y=1), Shear:Vector 2D Structure=(X=0,Y=0), Angle:float=0) -> (WidgetTransform:Widget Transform Structure) {is=title K2Node_MakeStruct}
   ) -> (then)
  ]}
]
variables [
 bAboveIsOpen {type=boolean}
 bBelowIsOpen {type=boolean}
 bIsDisplaying {type=boolean}
 isBacchus {type=boolean}
 SoundAboveOpen {type=SoundBase, default=/Game/Athena/Sounds/LTMs/StormCap/StormCap_UIAlert_Above_Cue}
 SoundBelowOpen {type=SoundBase, default=/Game/Athena/Sounds/LTMs/StormCap/StormCap_UIAlert_Below_Cue}
]
widgets {
 info {inheritedWidgetCount=1, parentClass=/Script/FortniteUI.AthenaStormSurgeWidgetBase, rootWidgetClass=&SizeBox, widgetCount=13}
 widgets [
  SizeBox_5 {widgetClassPath=&SizeBox, widget=SizeBox_5, properties {bOverride_WidthOverride=true, widthOverride=290}}
  Overlay_1 {widgetClassPath=&Overlay, parent=SizeBox_5, widget=Overlay_1, slot {class=SizeBoxSlot}}
  DmgThresh_SizeBox {
   widgetClassPath=&SizeBox
   bIsVariable=true
   parent=Overlay_1
   widget=DmgThresh_SizeBox
   properties {bOverride_WidthOverride=true, widthOverride=300}
   slot {class=OverlaySlot, properties {horizontalAlignment=HAlign_Fill, verticalAlignment=VAlign_Center}}
  }
  StormDMG_HorizontalBox {
   widgetClassPath=&VerticalBox
   parent=DmgThresh_SizeBox
   widget=StormDMG_HorizontalBox
   properties {renderTransformPivot=<1 0.5>}
   slot {class=SizeBoxSlot}
  }
  AboveDMG {
   widgetClassPath=&CommonTextBlock
   bIsVariable=true
   parent=StormDMG_HorizontalBox
   widget=AboveDMG
   properties {
    justification=Center
    renderOpacity=0
    text=+ 265 Above
    visibility=Collapsed
    font {
     bForceMonospaced=false
     bMaterialIsStencil=false
     fontMaterial=null
     fontObject=&Burbank
     letterSpacing=0
     monospacedWidth=1
     size=24
     skewAmount=0
     typefaceFontName=Burbank-Black
     outlineSettings {
      bApplyOutlineToDropShadows=false
      bMiteredCorners=false
      bSeparateFillAlpha=false
      outlineMaterial=null
      outlineSize=1
      outlineColor=<rgba 0 0 0 0.2>
     }
    }
    shadowColorAndOpacity=<rgba 0 0 0 1>
   }
   slot {class=VerticalBoxSlot, properties {verticalAlignment=VAlign_Top, padding=<ltrb 30 3 0 0>}}
  }
  SizeBox_0 {
   widgetClassPath=&SizeBox
   parent=StormDMG_HorizontalBox
   widget=SizeBox_0
   properties {bOverride_HeightOverride=true, heightOverride=54}
   slot {class=VerticalBoxSlot, properties {verticalAlignment=VAlign_Center}}
  }
  CommonBorder_0 {
   widgetClassPath=&CommonBorder
   bIsVariable=true
   parent=SizeBox_0
   widget=CommonBorder_0
   properties {
    horizontalAlignment=HAlign_Center
    verticalAlignment=VAlign_Center
    background {
     bIsDynamicallyLoaded=false
     drawAs=Box
     imageType=NoImage
     mirroring=NoMirror
     resourceName=None
     resourceObject=/Game/Athena/HUD/Spectator/Art/M_UI_SkewBox_Broadcast_StormSurgeWidget
     tiling=NoTile
     imageSize=<32 32>
     margin=<ltrb 0 0 0 0>
     outlineSettings {
      bUseBrushTransparency=false
      roundingType=HalfHeightRadius
      width=0
      color {colorUseRule=UseColor_Specified, specifiedColor=<rgba 0 0 0 0>}
      cornerRadii=<0 0 0 0>
     }
     tintColor {colorUseRule=UseColor_Specified, specifiedColor=<rgba 1 1 1 1>}
     uVRegion {bIsValid=false, max=<0 0>, min=<0 0>}
    }
    padding=<ltrb 30 1 0 0>
   }
   slot {class=SizeBoxSlot}
  }
  ThresholdText {
   widgetClassPath=&CommonTextBlock
   bIsVariable=true
   parent=CommonBorder_0
   widget=ThresholdText
   properties {
    justification=Center
    scrollStyle=&TextScrollStyle-Base_C
    text=DAMAGE THRESHOLD
    font {
     bForceMonospaced=false
     bMaterialIsStencil=false
     fontMaterial=null
     fontObject=&BurbankSmall
     letterSpacing=0
     monospacedWidth=1
     size=16
     skewAmount=0
     typefaceFontName=Bold
     outlineSettings {
      bApplyOutlineToDropShadows=false
      bMiteredCorners=false
      bSeparateFillAlpha=false
      outlineMaterial=null
      outlineSize=1
      outlineColor=<rgba 0 0 0 0.2>
     }
    }
   }
   slot {class=BorderSlot, properties {horizontalAlignment=HAlign_Center, verticalAlignment=VAlign_Center, padding=<ltrb 30 1 0 0>}}
  }
  BelowDMG {
   widgetClassPath=&CommonTextBlock
   bIsVariable=true
   parent=StormDMG_HorizontalBox
   widget=BelowDMG
   properties {
    justification=Center
    renderOpacity=0
    text="-256 BELOW!"
    visibility=Collapsed
    font {
     bForceMonospaced=false
     bMaterialIsStencil=false
     fontMaterial=/Game/UI/InGame/HUD/Textures/M_StormAlertText
     fontObject=&Burbank
     letterSpacing=0
     monospacedWidth=1
     size=24
     skewAmount=0
     typefaceFontName=Burbank-Black
     outlineSettings {
      bApplyOutlineToDropShadows=true
      bMiteredCorners=false
      bSeparateFillAlpha=false
      outlineMaterial=null
      outlineSize=2
      outlineColor=<rgba 0 0 0 0.2>
     }
    }
    shadowColorAndOpacity=<rgba 0 0 0 1>
   }
   slot {class=VerticalBoxSlot, properties {verticalAlignment=VAlign_Center, padding=<ltrb 30 0 0 0>}}
  }
  SurgeIcon_SizeBox {
   widgetClassPath=&SizeBox
   bIsVariable=true
   parent=Overlay_1
   widget=SurgeIcon_SizeBox
   slot {class=OverlaySlot, properties {verticalAlignment=VAlign_Center, padding=<ltrb -10 0 0 0>}}
  }
  AthenaStormSurgeIconWidget {
   widgetClassPath=/Game/Athena/HUD/AthenaStormSurgeIconWidget
   bIsVariable=true
   parent=SurgeIcon_SizeBox
   widget=AthenaStormSurgeIconWidget
   slot {class=SizeBoxSlot, properties {horizontalAlignment=HAlign_Center, verticalAlignment=VAlign_Center}}
  }
  StormSurgeIcon_NoBkg {
   widgetClassPath=&Image
   bIsVariable=true
   parent=Overlay_1
   widget=StormSurgeIcon_NoBkg
   properties {brush {
     bIsDynamicallyLoaded=false
     drawAs=Image
     imageType=NoImage
     mirroring=NoMirror
     resourceName=None
     resourceObject=/Game/Athena/HUD/Spectator/Art/T-UI-Icon-StormSurge-64x64
     tiling=NoTile
     imageSize=<35 35>
     margin=<ltrb 0 0 0 0>
     outlineSettings {
      bUseBrushTransparency=false
      roundingType=HalfHeightRadius
      width=0
      color {colorUseRule=UseColor_Specified, specifiedColor=<rgba 0 0 0 0>}
      cornerRadii=<0 0 0 0>
     }
     tintColor {colorUseRule=UseColor_Specified, specifiedColor=<rgba 1 1 1 1>}
     uVRegion {bIsValid=false, max=<0 0>, min=<0 0>}
    }}
   slot {class=OverlaySlot, properties {verticalAlignment=VAlign_Center, padding=<ltrb 28 0 0 0>}}
  }
  TutorialHighlight {widgetClassPath=&TutorialHighlight}
 ]
}
