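# Widget Blueprint for the Athena spectator HUD layer that hosts nameplate indicators, derived from the native AthenaSpectatorNameplateLayerBase.
# Its root FortActorCanvas, Canvas_NameplateIndicators, is volatile, draws elements in order and is itself hit-test invisible.
# The tree also holds a TutorialHighlight widget.
# The event graph is empty and no variables are declared, so the Blueprint adds no logic to its parent.
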WidgetBlueprint /FortniteCoreUI/Athena/HUD/Spectator/AthenaSpectatorNameplateLayer
graphs [EventGraph {links [], nodes []}]
variables []
widgets {
 info {
  inheritedWidgetCount=2
  parentClass=/Script/FortniteUI.AthenaSpectatorNameplateLayerBase
  rootWidgetClass=/Script/FortniteUI.FortActorCanvas
  widgetCount=2
 }
 widgets [
  Canvas_NameplateIndicators {
   widgetClassPath=&FortActorCanvas
   bInherited=true
   bIsVariable=true
   widget=Canvas_NameplateIndicators
   properties {bDrawElementsInOrder=true, bIsVolatile=true, visibility=SelfHitTestInvisible}
  }
  TutorialHighlight {widgetClassPath=&TutorialHighlight}
 ]
}
