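# Broadcast spectator kill feed widget for Athena, derived from the native AthenaBroadcastKillFeedBase with no Blueprint logic of its own.
# An inherited vertical FortDynamicEntryBox root lists AthenaBroadcastKillFeedItem entries, capped at three, spaced and laid out last-to-first.
# A TutorialHighlight widget is also present.
# The EventGraph is empty, so feed behavior comes from the parent class.
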WidgetBlueprint /FortniteCoreUI/Athena/HUD/Spectator/KillFeed/AthenaBroadcastKillFeedWidget
graphs [EventGraph {links [], nodes []}]
variables []
widgets {
 info {
  inheritedWidgetCount=2
  parentClass=/Script/FortniteUI.AthenaBroadcastKillFeedBase
  rootWidgetClass=/Script/FortniteUI.FortDynamicEntryBox
  widgetCount=2
 }
 widgets [
  KillFeedEntryBox {
   widgetClassPath=&FortDynamicEntryBox
   bInherited=true
   bIsVariable=true
   widget=KillFeedEntryBox
   properties {
    entryBoxType=Vertical
    entryWidgetClass=&/AthenaBroadcastKillFeedItem.AthenaBroadcastKillFeedItem_C
    maximumEntries=3
    patternDirection=LastToFirst
    entrySpacing=<0 10>
   }
  }
  TutorialHighlight {widgetClassPath=&TutorialHighlight}
 ]
}
