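# Broadcast spectator HUD widget showing players and teams remaining out of their totals.
# Update events from AthenaSpectatorPlayersLeftBase set each count and shrink its font as digits grow; a count above its total raises the total, otherwise a scale pop plays.
# Max squad size picks the team icon from a map with a default fallback and collapses the team panel for solos.
# PreConstruct feeds preview values at design time and zeroes counters at runtime.
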WidgetBlueprint /FortniteCoreUI/Athena/HUD/Spectator/MatchStatus/AthenaBroadcastPlayersRemaining
animations [
 PlayersElimScale {
  binding_count=1
  length_seconds=0.25
  object=&self:PlayersElimScale
  sequence {bindings [PlayersLeft {
     id=#1
     possessed_class=CommonNumericTextBlock
     tracks [{
       type=MovieScene2DTransformTrack
       sections [{
         type=MovieScene2DTransformSection
         end_frame=5
         start_frame=0
         channels [
          Scale.X {type=MovieSceneScriptingFloatChannel, keys [[0,1],[1,1.5],[3,1.5],[4,1]]}
          Scale.Y {type=MovieSceneScriptingFloatChannel, keys [[0,1],[1,1.5],[3,1.5],[4,1]]}
         ]
        }]
      }]
    }]}
 }
 TeamsElimScale {
  binding_count=1
  length_seconds=0.25
  object=&self:TeamsElimScale
  sequence {bindings [PlayersLeft {
     id=#1
     display_name=TeamsLeft
     possessed_class=CommonNumericTextBlock
     tracks [{
       type=MovieScene2DTransformTrack
       sections [{
         type=MovieScene2DTransformSection
         end_frame=5
         start_frame=0
         channels [
          Scale.X {type=MovieSceneScriptingFloatChannel, keys [[0,1],[1,1.5],[3,1.5],[4,1]]}
          Scale.Y {type=MovieSceneScriptingFloatChannel, keys [[0,1],[1,1.5],[3,1.5],[4,1]]}
         ]
        }]
      }]
    }]}
 }
]
graphs [EventGraph {nodes [
   n1 = Event UpdatePlayersLeft!() -> (OutputDelegate, then, InPlayersLeft:int=0) {is=title K2Node_Event}
   n2 = "To Float (Integer)"(InInt=@n1.InPlayersLeft)
   > SetCurrentValue!(
    execute
    self:Common Numeric Text Block Object Reference=Get PlayersLeft() -> (PlayersLeft:Common Numeric Text Block Object Reference)
    NewValue:float=@n2
   ) -> (then)
   n3 = "integer > integer"(A=@n1.InPlayersLeft, B=9) {is=title K2Node_CallFunction}
   n4 = Get MediumFont() -> (MediumFont:Slate Font Info Structure)
   n5 = Get LargeFont() -> (LargeFont:Slate Font Info Structure)
   > SetFont(
    self=Get PlayersLeft()
    InFontInfo=Select(
     Option 0:Slate Font Info Structure=Select(Option 0=@n5, Option 1=@n4, Index=@n3)
     Option 1:Slate Font Info Structure=Get SmallFont() -> (SmallFont:Slate Font Info Structure)
     Index="integer > integer"(A=@n1.InPlayersLeft, B=99)
    ) -> (ReturnValue:Slate Font Info Structure)
   )
   n6 = Get TotalPlayers() -> (TotalPlayers:Common Numeric Text Block Object Reference)
   > n7 = Branch(Condition="float > float"(A=@n2, B=Get CurrentNumericValue!(self:Common Numeric Text Block Object Reference=@n6) -> (CurrentNumericValue:float)) {is=title K2Node_CallFunction})
   > SetCurrentValue(self=Get TotalPlayers(), NewValue=@n2)
   PlayAnimation(execute=@n7.else, InAnimation=Get PlayersElimScale() -> (PlayersElimScale:Widget Animation Object Reference))
   n8 = Event UpdateTotalPlayers!() -> (OutputDelegate, then, InTotalPlayers:int=0) {is=title K2Node_Event}
   > Branch(Condition="integer > integer"(A=@n8.InTotalPlayers, B=0))
   > SetCurrentValue(self=Get TotalPlayers(), NewValue="To Float (Integer)"(InInt=@n8.InTotalPlayers))
   n9 = Event UpdateTeamsLeft!() -> (OutputDelegate, then, InTeamsLeft:int=0) {is=title K2Node_Event}
   n10 = "To Float (Integer)"(InInt=@n9.InTeamsLeft)
   > SetCurrentValue(self=Get TeamsLeft() -> (TeamsLeft:Common Numeric Text Block Object Reference), NewValue=@n10)
   > SetFont(
    self=Get TeamsLeft()
    InFontInfo=Select(Option 0=Get LargeFont(), Option 1=Get MediumFont(), Index="integer > integer"(A=@n9.InTeamsLeft, B=9))
   )
   n11 = Get TotalTeams() -> (TotalTeams:Common Numeric Text Block Object Reference)
   > n12 = Branch(Condition="float > float"(A=@n10, B=Get CurrentNumericValue(self=@n11)))
   > SetCurrentValue(self=Get TotalTeams(), NewValue=@n10)
   PlayAnimation(execute=@n12.else, InAnimation=Get TeamsElimScale() -> (TeamsElimScale:Widget Animation Object Reference))
   n13 = Event PreConstruct()
   > n14 = Branch(Condition=@n13.IsDesignTime)
   > n15 = Sequence()
   > Set DesignTime_PlayersLeft(DesignTime_PlayersLeft:int="Min (Integer)"(A=Get DesignTime_PlayersLeft() -> (DesignTime_PlayersLeft:int=0), B=Get DesignTime_TotalPlayers() -> (DesignTime_TotalPlayers:int=0)) -> (ReturnValue=0)) -> (Output_Get:int=0)
   > Set DesignTime_TeamsLeft(DesignTime_TeamsLeft:int="Min (Integer)"(A=Get DesignTime_TeamsLeft() -> (DesignTime_TeamsLeft:int=0), B=Get DesignTime_TotalTeams() -> (DesignTime_TotalTeams:int=0)) -> (ReturnValue=0)) -> (Output_Get:int=0)
   UpdatePlayersLeft!(
    execute=@n15.then_1
    self:Athena Spectator Players Left Base Object Reference
    InPlayersLeft:int=Get DesignTime_PlayersLeft() -> (DesignTime_PlayersLeft=0)
   ) -> (then)
   > UpdateTotalPlayers!(
    execute
    self:Athena Spectator Players Left Base Object Reference
    InTotalPlayers:int=Get DesignTime_TotalPlayers() -> (DesignTime_TotalPlayers=0)
   ) -> (then)
   > UpdateTeamsLeft!(execute, self:Athena Spectator Players Left Base Object Reference, InTeamsLeft:int=Get DesignTime_TeamsLeft() -> (DesignTime_TeamsLeft=0)) -> (then)
   > UpdateTotalTeams!(execute, self:Athena Spectator Players Left Base Object Reference, InTotalTeams:int=Get DesignTime_TotalTeams() -> (DesignTime_TotalTeams=0)) -> (then)
   > UpdateMaxSquadSize!(
    execute
    self:Athena Spectator Players Left Base Object Reference
    MaxSquadSize:int=Get DesignTime_MaxSquadSize() -> (DesignTime_MaxSquadSize:int=0)
   ) -> (then)
   SetCurrentValue(execute=@n14.else, self=[Get TeamsLeft(), Get TotalTeams(), Get PlayersLeft(), Get TotalPlayers()], NewValue=0)
   n16 = Event UpdateTotalTeams!() -> (OutputDelegate, then, InTotalTeams:int=0) {is=title K2Node_Event}
   > Branch(Condition="integer > integer"(A=@n16.InTotalTeams, B=0))
   > SetCurrentValue(self=Get TotalTeams(), NewValue="To Float (Integer)"(InInt=@n16.InTotalTeams))
   n17 = Event UpdateMaxSquadSize!() -> (OutputDelegate, then, MaxSquadSize:int=0) {is=title K2Node_Event}
   > SetVisibility(
    self=Get TeamsSB() -> (TeamsSB:Size Box Object Reference)
    InVisibility=Select(
     Option 0:ESlateVisibility Enum=Collapsed
     Option 1:ESlateVisibility Enum=SelfHitTestInvisible
     Index="integer > integer"(A=@n17.MaxSquadSize, B=1)
    ) -> (ReturnValue:ESlateVisibility Enum)
   )
   n18 = Find(
    TargetMap:Map of Integers to Texture 2D Soft Object References=Get TeamIcons() -> (TeamIcons:Map of Integers to Texture 2D Soft Object References)
    Key:"Integer (by ref)"=@n17.MaxSquadSize
   ) -> (Value:Texture 2D Soft Object Reference)
   > SetBrushFromSoftTexture!(
    execute
    self:Image Object Reference=Get Team_Icon() -> (Team_Icon:Image Object Reference)
    SoftTexture:Texture 2D Soft Object Reference=Select(
     Option 0:Texture 2D Soft Object Reference=Get DefaultTeamIcon() -> (DefaultTeamIcon:Texture 2D Soft Object Reference)
     Option 1:Texture 2D Soft Object Reference=@n18
     Index=@n18.ReturnValue
    ) -> (ReturnValue:Texture 2D Soft Object Reference)
    bMatchSize:bool=false
   ) -> (then)
  ]}]
variables [
 DefaultTeamIcon {type=Texture2D, default=/Game/UI/Foundation/Textures/Icons/Social/T_Icon_Matchmaking_Squad_64}
 DesignTime_MaxSquadSize {type=integer, default=4}
 DesignTime_PlayersLeft {type=integer, default=100}
 DesignTime_TeamsLeft {type=integer, default=25}
 DesignTime_TotalPlayers {type=integer, default=100}
 DesignTime_TotalTeams {type=integer, default=25}
 LargeFont {
  type=SlateFontInfo
  default {
   bForceMonospaced=false
   bMaterialIsStencil=false
   fontMaterial=null
   fontObject=&BurbankSmall
   letterSpacing=0
   monospacedWidth=1
   size=34
   skewAmount=0
   typefaceFontName=Black
   outlineSettings {
    bApplyOutlineToDropShadows=false
    bMiteredCorners=false
    bSeparateFillAlpha=false
    outlineMaterial=null
    outlineSize=0
    outlineColor=<rgba 0 0 0 1>
   }
  }
 }
 MediumFont {
  type=SlateFontInfo
  default {
   bForceMonospaced=false
   bMaterialIsStencil=false
   fontMaterial=null
   fontObject=&BurbankSmall
   letterSpacing=0
   monospacedWidth=1
   size=28
   skewAmount=0
   typefaceFontName=Black
   outlineSettings {
    bApplyOutlineToDropShadows=false
    bMiteredCorners=false
    bSeparateFillAlpha=false
    outlineMaterial=null
    outlineSize=0
    outlineColor=<rgba 0 0 0 1>
   }
  }
 }
 SmallFont {
  type=SlateFontInfo
  default {
   bForceMonospaced=false
   bMaterialIsStencil=false
   fontMaterial=null
   fontObject=&BurbankSmall
   letterSpacing=0
   monospacedWidth=1
   size=24
   skewAmount=0
   typefaceFontName=Black
   outlineSettings {
    bApplyOutlineToDropShadows=false
    bMiteredCorners=false
    bSeparateFillAlpha=false
    outlineMaterial=null
    outlineSize=0
    outlineColor=<rgba 0 0 0 1>
   }
  }
 }
 TeamIcons {default {
   2=/Game/UI/Foundation/Textures/Icons/Social/T_Icon_Matchmaking_Duo_64
   3=/Game/UI/Foundation/Textures/Icons/Social/T_Icon_Matchmaking_Trios_64
  }}
]
widgets {
 info {inheritedWidgetCount=0, parentClass=/Script/FortniteUI.AthenaSpectatorPlayersLeftBase, rootWidgetClass=&SizeBox, widgetCount=20}
 widgets [
  SizeBox_1 {widgetClassPath=&SizeBox, widget=SizeBox_1, properties {bOverride_HeightOverride=true, heightOverride=54, maxDesiredHeight=54}}
  HorizontalBox_1 {widgetClassPath=&HorizontalBox, parent=SizeBox_1, widget=HorizontalBox_1, slot {class=SizeBoxSlot}}
  StatusSB {
   widgetClassPath=&SizeBox
   parent=HorizontalBox_1
   widget=StatusSB
   properties {bOverride_MinDesiredWidth=true, heightOverride=54, maxDesiredHeight=54, minDesiredHeight=54, minDesiredWidth=170, widthOverride=170}
   slot {class=HorizontalBoxSlot}
  }
  CommonBorder_0 {
   widgetClassPath=&CommonBorder
   parent=StatusSB
   widget=CommonBorder_0
   properties {
    background {
     bIsDynamicallyLoaded=false
     drawAs=Box
     imageType=NoImage
     mirroring=NoMirror
     resourceName=None
     resourceObject=/Game/Athena/HUD/Spectator/Art/M_UI_SkewBox_Broadcast_PlayersRemaining
     tiling=NoTile
     imageSize=<32 32>
     margin=<ltrb 0 0 0 0>
     outlineSettings {
      bUseBrushTransparency=false
      roundingType=HalfHeightRadius
      width=0
      color {colorUseRule=UseColor_Specified, specifiedColor=<rgba 0 0 0 0>}
      cornerRadii=<0 0 0 0>
     }
     tintColor {colorUseRule=UseColor_Specified, specifiedColor=<rgba 1 1 1 1>}
     uVRegion {bIsValid=false, max=<0 0>, min=<0 0>}
    }
    padding=<ltrb 0 0 0 0>
   }
   slot {class=SizeBoxSlot}
  }
  HorizontalBox {
   widgetClassPath=&HorizontalBox
   parent=CommonBorder_0
   widget=HorizontalBox
   slot {class=BorderSlot, properties {padding=<ltrb 0 0 0 0>}}
  }
  PlayerIcon {
   widgetClassPath=&Image
   parent=HorizontalBox
   widget=PlayerIcon
   properties {
    visibility=SelfHitTestInvisible
    brush {
     bIsDynamicallyLoaded=false
     drawAs=Image
     imageType=NoImage
     mirroring=NoMirror
     resourceName=None
     resourceObject=/Game/Athena/HUD/Spectator/Art/T-UI-Icon-PlayersRemaining-64x64
     tiling=NoTile
     imageSize=<36 36>
     margin=<ltrb 0 0 0 0>
     outlineSettings {
      bUseBrushTransparency=false
      roundingType=HalfHeightRadius
      width=0
      color {colorUseRule=UseColor_Specified, specifiedColor=<rgba 0 0 0 0>}
      cornerRadii=<0 0 0 0>
     }
     tintColor {colorUseRule=UseColor_Specified, specifiedColor=<rgba 1 1 1 1>}
     uVRegion {bIsValid=false, max=<0 0>, min=<0 0>}
    }
   }
   slot {class=HorizontalBoxSlot, properties {verticalAlignment=VAlign_Center, padding=<ltrb 10 0 0 0>}}
  }
  SizeBox_2 {
   widgetClassPath=&SizeBox
   parent=HorizontalBox
   widget=SizeBox_2
   properties {bOverride_MinDesiredWidth=true, minDesiredWidth=57}
   slot {
    class=HorizontalBoxSlot
    properties {horizontalAlignment=HAlign_Center, verticalAlignment=VAlign_Center, size {sizeRule=Automatic, value=0.5}}
   }
  }
  PlayersLeft {
   widgetClassPath=/Script/CommonUI.CommonNumericTextBlock
   bIsVariable=true
   parent=SizeBox_2
   widget=PlayersLeft
   properties {
    currentNumericValue=9
    justification=Center
    text=100
    font {
     bForceMonospaced=false
     bMaterialIsStencil=false
     fontMaterial=null
     fontObject=&BurbankSmall
     letterSpacing=0
     monospacedWidth=1
     size=24
     skewAmount=0
     typefaceFontName=Black
     outlineSettings {
      bApplyOutlineToDropShadows=false
      bMiteredCorners=false
      bSeparateFillAlpha=false
      outlineMaterial=null
      outlineSize=0
      outlineColor=<rgba 0 0 0 1>
     }
    }
    shadowOffset=<0 1>
   }
   slot {class=SizeBoxSlot, properties {horizontalAlignment=HAlign_Center, verticalAlignment=VAlign_Center}}
  }
  CommonTextBlock_0 {
   widgetClassPath=&CommonTextBlock
   parent=HorizontalBox
   widget=CommonTextBlock_0
   properties {
    text=/
    font {
     bForceMonospaced=false
     bMaterialIsStencil=false
     fontMaterial=null
     fontObject=&BurbankSmall
     letterSpacing=0
     monospacedWidth=1
     size=16
     skewAmount=0
     typefaceFontName=Bold
     outlineSettings {
      bApplyOutlineToDropShadows=false
      bMiteredCorners=false
      bSeparateFillAlpha=false
      outlineMaterial=null
      outlineSize=0
      outlineColor=<rgba 0 0 0 1>
     }
    }
    shadowOffset=<0 0>
   }
   slot {class=HorizontalBoxSlot, properties {verticalAlignment=VAlign_Center, padding=<ltrb 2 0 2 0>, size {sizeRule=Automatic, value=0.2}}}
  }
  SizeBox_4 {
   widgetClassPath=&SizeBox
   parent=HorizontalBox
   widget=SizeBox_4
   properties {bOverride_MinDesiredWidth=true, minDesiredWidth=37}
   slot {
    class=HorizontalBoxSlot
    properties {horizontalAlignment=HAlign_Center, verticalAlignment=VAlign_Center, padding=<ltrb 0 0.5 0 0>, size {sizeRule=Automatic, value=0.35}}
   }
  }
  TotalPlayers {
   widgetClassPath=&CommonNumericTextBlock
   bIsVariable=true
   parent=SizeBox_4
   widget=TotalPlayers
   properties {
    currentNumericValue=100
    text=99
    font {
     bForceMonospaced=false
     bMaterialIsStencil=false
     fontMaterial=null
     fontObject=&BurbankSmall
     letterSpacing=0
     monospacedWidth=1
     size=16
     skewAmount=0
     typefaceFontName=Bold
     outlineSettings {
      bApplyOutlineToDropShadows=false
      bMiteredCorners=false
      bSeparateFillAlpha=false
      outlineMaterial=null
      outlineSize=0
      outlineColor=<rgba 0 0 0 1>
     }
    }
    shadowOffset=<0 0>
   }
   slot {class=SizeBoxSlot, properties {horizontalAlignment=HAlign_Center, verticalAlignment=VAlign_Center}}
  }
  TeamsSB {
   widgetClassPath=&SizeBox
   bIsVariable=true
   parent=HorizontalBox_1
   widget=TeamsSB
   properties {bOverride_MinDesiredWidth=true, heightOverride=54, minDesiredWidth=158, widthOverride=158}
   slot {class=HorizontalBoxSlot}
  }
  CommonBorder_1 {
   widgetClassPath=&CommonBorder
   parent=TeamsSB
   widget=CommonBorder_1
   properties {
    background {
     bIsDynamicallyLoaded=false
     drawAs=Box
     imageType=NoImage
     mirroring=NoMirror
     resourceName=None
     resourceObject=/Game/Athena/HUD/Spectator/Art/M_UI_SkewBox_Broadcast_TeamRemaining
     tiling=NoTile
     imageSize=<32 32>
     margin=<ltrb 0 0 0 0>
     outlineSettings {
      bUseBrushTransparency=false
      roundingType=HalfHeightRadius
      width=0
      color {colorUseRule=UseColor_Specified, specifiedColor=<rgba 0 0 0 0>}
      cornerRadii=<0 0 0 0>
     }
     tintColor {colorUseRule=UseColor_Specified, specifiedColor=<rgba 1 1 1 1>}
     uVRegion {bIsValid=false, max=<0 0>, min=<0 0>}
    }
    padding=<ltrb 0 0 0 0>
   }
   slot {class=SizeBoxSlot}
  }
  HorizontalBox_0 {
   widgetClassPath=&HorizontalBox
   parent=CommonBorder_1
   widget=HorizontalBox_0
   slot {class=BorderSlot, properties {padding=<ltrb 0 0 0 0>}}
  }
  Team_Icon {
   widgetClassPath=&Image
   bIsVariable=true
   parent=HorizontalBox_0
   widget=Team_Icon
   properties {
    visibility=SelfHitTestInvisible
    brush {
     bIsDynamicallyLoaded=false
     drawAs=Image
     imageType=NoImage
     mirroring=NoMirror
     resourceName=None
     resourceObject=&T_Icon_Matchmaking_Squad_64
     tiling=NoTile
     imageSize=<36 36>
     margin=<ltrb 0 0 0 0>
     outlineSettings {
      bUseBrushTransparency=false
      roundingType=HalfHeightRadius
      width=0
      color {colorUseRule=UseColor_Specified, specifiedColor=<rgba 0 0 0 0>}
      cornerRadii=<0 0 0 0>
     }
     tintColor {colorUseRule=UseColor_Specified, specifiedColor=<rgba 1 1 1 1>}
     uVRegion {bIsValid=false, max=<0 0>, min=<0 0>}
    }
   }
   slot {class=HorizontalBoxSlot, properties {verticalAlignment=VAlign_Center, padding=<ltrb 18 0 5 0>}}
  }
  SizeBox_3 {
   widgetClassPath=&SizeBox
   parent=HorizontalBox_0
   widget=SizeBox_3
   properties {bOverride_MinDesiredWidth=true, minDesiredWidth=48}
   slot {
    class=HorizontalBoxSlot
    properties {horizontalAlignment=HAlign_Center, verticalAlignment=VAlign_Center, size {sizeRule=Automatic, value=0.7}}
   }
  }
  TeamsLeft {
   widgetClassPath=&CommonNumericTextBlock
   bIsVariable=true
   parent=SizeBox_3
   widget=TeamsLeft
   properties {
    currentNumericValue=50
    justification=Center
    text=888
    font {
     bForceMonospaced=false
     bMaterialIsStencil=false
     fontMaterial=null
     fontObject=&BurbankSmall
     letterSpacing=0
     monospacedWidth=1
     size=24
     skewAmount=0
     typefaceFontName=Black
     outlineSettings {
      bApplyOutlineToDropShadows=false
      bMiteredCorners=false
      bSeparateFillAlpha=false
      outlineMaterial=null
      outlineSize=0
      outlineColor=<rgba 0 0 0 1>
     }
    }
    shadowOffset=<0 1>
   }
   slot {class=SizeBoxSlot, properties {horizontalAlignment=HAlign_Center, verticalAlignment=VAlign_Center}}
  }
  CommonTextBlock_1 {
   widgetClassPath=&CommonTextBlock
   parent=HorizontalBox_0
   widget=CommonTextBlock_1
   properties {
    text=/
    font {
     bForceMonospaced=false
     bMaterialIsStencil=false
     fontMaterial=null
     fontObject=&BurbankSmall
     letterSpacing=0
     monospacedWidth=1
     size=16
     skewAmount=0
     typefaceFontName=Bold
     outlineSettings {
      bApplyOutlineToDropShadows=false
      bMiteredCorners=false
      bSeparateFillAlpha=false
      outlineMaterial=null
      outlineSize=0
      outlineColor=<rgba 0 0 0 1>
     }
    }
    shadowOffset=<0 0>
   }
   slot {class=HorizontalBoxSlot, properties {verticalAlignment=VAlign_Center, padding=<ltrb 2 0 2 0>, size {sizeRule=Automatic, value=0.3}}}
  }
  SizeBox_5 {
   widgetClassPath=&SizeBox
   parent=HorizontalBox_0
   widget=SizeBox_5
   properties {bOverride_MinDesiredWidth=true, minDesiredWidth=28}
   slot {
    class=HorizontalBoxSlot
    properties {horizontalAlignment=HAlign_Center, verticalAlignment=VAlign_Center, padding=<ltrb 0 0.5 0 0>, size {sizeRule=Automatic, value=0.25}}
   }
  }
  TotalTeams {
   widgetClassPath=&CommonNumericTextBlock
   bIsVariable=true
   parent=SizeBox_5
   widget=TotalTeams
   properties {
    currentNumericValue=88
    text=50
    font {
     bForceMonospaced=false
     bMaterialIsStencil=false
     fontMaterial=null
     fontObject=&BurbankSmall
     letterSpacing=0
     monospacedWidth=1
     size=16
     skewAmount=0
     typefaceFontName=Bold
     outlineSettings {
      bApplyOutlineToDropShadows=false
      bMiteredCorners=false
      bSeparateFillAlpha=false
      outlineMaterial=null
      outlineSize=0
      outlineColor=<rgba 0 0 0 1>
     }
    }
    shadowOffset=<0 1>
   }
   slot {class=SizeBoxSlot, properties {horizontalAlignment=HAlign_Center}}
  }
 ]
}
