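# Widget Blueprint for the broadcast spectator HUD's match status area that displays the session ID.
# It derives from the native AthenaBroadcastSessionId class and adds no graph logic or variables of its own.
# The inherited Text_SessionId text block is styled as small, centered, semi-transparent white bold Noto Sans, with "SESSION ID" as its default text.
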WidgetBlueprint /FortniteCoreUI/Athena/HUD/Spectator/MatchStatus/AthenaBroadcastSessionId
graphs [EventGraph {links [], nodes []}]
variables []
widgets {
 info {inheritedWidgetCount=2, parentClass=/Script/FortniteUI.AthenaBroadcastSessionId, rootWidgetClass=&TextBlock, widgetCount=2}
 widgets [
  Text_SessionId {
   widgetClassPath=&TextBlock
   bInherited=true
   bIsVariable=true
   widget=Text_SessionId
   properties {
    justification=Center
    text=SESSION ID
    colorAndOpacity {colorUseRule=UseColor_Specified, specifiedColor=<rgba 1 1 1 0.4>}
    font {
     bForceMonospaced=false
     bMaterialIsStencil=false
     fontMaterial=null
     fontObject=&NotoSans
     letterSpacing=0
     monospacedWidth=1
     size=10
     skewAmount=0
     typefaceFontName=NotoSans-Bold
     outlineSettings {
      bApplyOutlineToDropShadows=false
      bMiteredCorners=false
      bSeparateFillAlpha=false
      outlineMaterial=null
      outlineSize=0
      outlineColor=<rgba 0 0 0 1>
     }
    }
   }
  }
  TutorialHighlight {widgetClassPath=&TutorialHighlight}
 ]
}
