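# Broadcast spectator HUD widget, built on AthenaBroadcastMatchStatusWidget, that shows the storm phase beside the game timer.
# Storm phase and phase-count events set the current number and a "/N" total next to a storm icon.
# A SetMinigameOverride custom event forwards a minigame to the embedded AthenaBroadcastGameTimeWidget.
# The two panels use skewed-box materials, with negative padding overlapping the timer onto the phase box.
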WidgetBlueprint /FortniteCoreUI/Athena/HUD/Spectator/MatchStatus/AthenaBroadcastStormStatus
graphs [EventGraph {nodes [
   n1 = SetMinigameOverride!() -> (OutputDelegate, then, MinigameOverride:Fort Minigame Object Reference) {is=title K2Node_CustomEvent}
   > SetMinigameOverride!(
    execute
    self:Athena Broadcast Game Phase Widget Object Reference=Get AthenaBroadcastGameTimeWidget() -> (AthenaBroadcastGameTimeWidget:Athena Broadcast Game Time Widget Object Reference)
    Minigame:Fort Minigame Object Reference=@n1.MinigameOverride
   ) -> (then) {is=title K2Node_CallFunction}
   n2 = Event OnStormPhaseChanged!() -> (OutputDelegate, then, StormPhase:int=0) {is=title K2Node_Event}
   > "SetText (Text)"(self=Get CurrentStormPhaseText() -> (CurrentStormPhaseText:Text Object Reference), InText="To Text (Integer)"(Value=@n2.StormPhase))
   n3 = Event OnStormPhaseCountChanged!() -> (OutputDelegate, then, StormPhaseCount:int=0) {is=title K2Node_Event}
   > "SetText (Text)"(
    self=Get MaxStormPhaseText() -> (MaxStormPhaseText:Text Object Reference)
    InText=Format Text!(Format="LOCTEXT(\"/{PhaseCount}\")", PhaseCount:Text="To Text (Integer)"(Value=@n3.StormPhaseCount)) -> (Result)
   )
   Event PreConstruct()
   Event Construct()
   Event Tick()
  ]}]
variables []
widgets {
 info {inheritedWidgetCount=1, parentClass=/Script/FortniteUI.AthenaBroadcastMatchStatusWidget, rootWidgetClass=&SizeBox, widgetCount=13}
 widgets [
  StatusSB {widgetClassPath=&SizeBox, widget=StatusSB, properties {bOverride_HeightOverride=true, heightOverride=55, widthOverride=300}}
  HorizontalBox {widgetClassPath=&HorizontalBox, parent=StatusSB, widget=HorizontalBox, slot {class=SizeBoxSlot}}
  StormPhaseSB {
   widgetClassPath=&SizeBox
   bIsVariable=true
   parent=HorizontalBox
   widget=StormPhaseSB
   properties {bOverride_WidthOverride=true, widthOverride=155}
   slot {class=HorizontalBoxSlot}
  }
  StormPhaseBorder {
   widgetClassPath=&CommonBorder
   parent=StormPhaseSB
   widget=StormPhaseBorder
   properties {
    horizontalAlignment=HAlign_Left
    verticalAlignment=VAlign_Center
    background {
     bIsDynamicallyLoaded=false
     drawAs=Box
     imageType=NoImage
     mirroring=NoMirror
     resourceName=None
     resourceObject=/Game/Athena/HUD/Spectator/Art/M_UI_SkewBox_Broadcast_MatchStatus
     tiling=NoTile
     imageSize=<32 32>
     margin=<ltrb 0 0 0 0>
     outlineSettings {
      bUseBrushTransparency=false
      roundingType=HalfHeightRadius
      width=0
      color {colorUseRule=UseColor_Specified, specifiedColor=<rgba 0 0 0 0>}
      cornerRadii=<0 0 0 0>
     }
     tintColor {colorUseRule=UseColor_Specified, specifiedColor=<rgba 1 1 1 1>}
     uVRegion {bIsValid=false, max=<0 0>, min=<0 0>}
    }
    padding=<ltrb 0 0 0 0>
   }
   slot {class=SizeBoxSlot}
  }
  StormPhaseBox {
   widgetClassPath=&HorizontalBox
   parent=StormPhaseBorder
   widget=StormPhaseBox
   slot {class=BorderSlot, properties {horizontalAlignment=HAlign_Left, verticalAlignment=VAlign_Center, padding=<ltrb 0 0 0 0>}}
  }
  StormPhaseImage {
   widgetClassPath=&Image
   parent=StormPhaseBox
   widget=StormPhaseImage
   properties {brush {
     bIsDynamicallyLoaded=false
     drawAs=Image
     imageType=NoImage
     mirroring=NoMirror
     resourceName=None
     resourceObject=/Game/UI/Foundation/Textures/Icons/Mission/T-Icon-StormZone-128
     tiling=NoTile
     imageSize=<40 40>
     margin=<ltrb 0 0 0 0>
     outlineSettings {
      bUseBrushTransparency=false
      roundingType=HalfHeightRadius
      width=0
      color {colorUseRule=UseColor_Specified, specifiedColor=<rgba 0 0 0 0>}
      cornerRadii=<0 0 0 0>
     }
     tintColor {colorUseRule=UseColor_Specified, specifiedColor=<rgba 1 1 1 1>}
     uVRegion {bIsValid=false, max=<0 0>, min=<0 0>}
    }}
   slot {class=HorizontalBoxSlot, properties {horizontalAlignment=HAlign_Left, verticalAlignment=VAlign_Center, padding=<ltrb 4 0 4 0>}}
  }
  StormPhaseTextBox {
   widgetClassPath=&HorizontalBox
   parent=StormPhaseBox
   widget=StormPhaseTextBox
   slot {class=HorizontalBoxSlot, properties {verticalAlignment=VAlign_Center}}
  }
  CurrentStormPhaseText {
   widgetClassPath=&TextBlock
   bIsVariable=true
   parent=StormPhaseTextBox
   widget=CurrentStormPhaseText
   properties {
    text=0
    font {
     bForceMonospaced=false
     bMaterialIsStencil=false
     fontMaterial=null
     fontObject=&BurbankSmall
     letterSpacing=0
     monospacedWidth=1
     size=30
     skewAmount=0
     typefaceFontName=Black
     outlineSettings {
      bApplyOutlineToDropShadows=false
      bMiteredCorners=false
      bSeparateFillAlpha=false
      outlineMaterial=null
      outlineSize=0
      outlineColor=<rgba 0 0 0 1>
     }
    }
   }
   slot {class=HorizontalBoxSlot, properties {padding=<ltrb 0 0 1 0>}}
  }
  MaxStormPhaseText {
   widgetClassPath=&TextBlock
   bIsVariable=true
   parent=StormPhaseTextBox
   widget=MaxStormPhaseText
   properties {
    text=/12
    font {
     bForceMonospaced=false
     bMaterialIsStencil=false
     fontMaterial=null
     fontObject=&BurbankSmall
     letterSpacing=0
     monospacedWidth=1
     size=16
     skewAmount=0
     typefaceFontName=Bold
     outlineSettings {
      bApplyOutlineToDropShadows=false
      bMiteredCorners=false
      bSeparateFillAlpha=false
      outlineMaterial=null
      outlineSize=0
      outlineColor=<rgba 0 0 0 1>
     }
    }
   }
   slot {class=HorizontalBoxSlot, properties {verticalAlignment=VAlign_Center, padding=<ltrb 0 1 0 0>}}
  }
  GameTimeSB {
   widgetClassPath=&SizeBox
   parent=HorizontalBox
   widget=GameTimeSB
   properties {bOverride_WidthOverride=true, widthOverride=162}
   slot {class=HorizontalBoxSlot, properties {padding=<ltrb -17 0 0 0>}}
  }
  GameTimeBorder {
   widgetClassPath=&CommonBorder
   parent=GameTimeSB
   widget=GameTimeBorder
   properties {
    horizontalAlignment=HAlign_Right
    verticalAlignment=VAlign_Center
    background {
     bIsDynamicallyLoaded=false
     drawAs=Box
     imageType=NoImage
     mirroring=NoMirror
     resourceName=None
     resourceObject=/Game/Athena/HUD/Spectator/Art/M_UI_SkewBox_Broadcast_StormTimer
     tiling=NoTile
     imageSize=<32 32>
     margin=<ltrb 0 0 0 0>
     outlineSettings {
      bUseBrushTransparency=false
      roundingType=HalfHeightRadius
      width=0
      color {colorUseRule=UseColor_Specified, specifiedColor=<rgba 0 0 0 0>}
      cornerRadii=<0 0 0 0>
     }
     tintColor {colorUseRule=UseColor_Specified, specifiedColor=<rgba 1 1 1 1>}
     uVRegion {bIsValid=false, max=<0 0>, min=<0 0>}
    }
    padding=<ltrb 10 0 10 0>
   }
   slot {class=SizeBoxSlot}
  }
  AthenaBroadcastGameTimeWidget {
   widgetClassPath=&/AthenaBroadcastGameTimeWidget
   bIsVariable=true
   parent=GameTimeBorder
   widget=AthenaBroadcastGameTimeWidget
   slot {class=BorderSlot, properties {horizontalAlignment=HAlign_Right, verticalAlignment=VAlign_Center, padding=<ltrb 10 0 10 0>}}
  }
  TutorialHighlight {widgetClassPath=&TutorialHighlight}
 ]
}
