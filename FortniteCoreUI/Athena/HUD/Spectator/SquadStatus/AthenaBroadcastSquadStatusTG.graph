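# Spectator broadcast HUD widget showing the followed player's squad: player status rows plus a team tag.
# Player additions and follow changes rebuild the list from squad view data, routing the followed player to a dedicated row and others to dynamic entries.
# PIP mode shows that row, removes top padding and hides the tag, which also hides without squadmates.
# Minigame team ID and session rank changes propagate to child widgets.
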WidgetBlueprint /FortniteCoreUI/Athena/HUD/Spectator/SquadStatus/AthenaBroadcastSquadStatusTG
graphs [
 EventGraph {nodes [
   n1 = UpdateFromMinigameTeamId!() -> (OutputDelegate, then, MinigameTeamId:Fort Minigame Team Identification Object Reference) {is=title K2Node_CustomEvent}
   > UpdateFromMinigameTeamId!(
    execute
    self:Athena Broadcast Player Status TG Object Reference=Get FollowedPlayerStatus() -> (FollowedPlayerStatus:Athena Broadcast Player Status TG Object Reference)
    MinigameTeamId:Fort Minigame Team Identification Object Reference=@n1.MinigameTeamId
   ) -> (then) {is=title K2Node_CallFunction}
   > n2 = For Each Loop(Array:Array of User Widget Object References=GetAllEntries!(self:Dynamic Entry Box Base Object Reference=Get SquadMateList() -> (SquadMateList:Fort Dynamic Entry Box Object Reference)) -> (ReturnValue:Array of User Widget Object References)) -> (Array Element:User Widget Object Reference)
   > n3 = Cast To AthenaBroadcastPlayerStatusTG!(execute, Object=@n2) -> (then, CastFailed, AsAthena Broadcast Player Status TG:Athena Broadcast Player Status TG Object Reference) {is=title K2Node_DynamicCast}
   > UpdateFromMinigameTeamId(self=@n3, MinigameTeamId=@n1.MinigameTeamId)
   n4 = Event OnPipModeChanged!() -> (OutputDelegate, then, bPipModeEnabled:bool) {is=title K2Node_Event}
   > Set bPipModeEnabled(bPipModeEnabled:bool=@n4.bPipModeEnabled) -> (Output_Get:bool)
   > SetVisibility(self=Get FollowedPlayerStatus(), InVisibility=Select(Option 0=Collapsed, Option 1=SelfHitTestInvisible, Index=@n4.bPipModeEnabled))
   > SetPadding!(
    execute
    self:Overlay Slot Object Reference=SlotAsOverlaySlot(Widget=Get StatusVB() -> (StatusVB:Vertical Box Object Reference))
    InPadding_Left:float=0
    InPadding_Top:float=Select(Option 0:double=45, Option 1:double=0, Index=@n4.bPipModeEnabled) -> (ReturnValue:double)
    InPadding_Right:float=0
    InPadding_Bottom:float=0
   ) -> (then)
   > UpdateTeamTagVisibility()
   n5 = Event PreConstruct()
   > n6 = Branch(Condition=@n5.IsDesignTime)
   n7 = Get SquadMateList()
   > n8 = For Each Loop(Array=GetAllEntries(self=@n7))
   n9 = Get bPipModeEnabled() -> (bPipModeEnabled:bool)
   OnPipModeChanged!(execute=@n8.Completed, self:Athena Broadcast Squad Status Widget Object Reference, bPipModeEnabled:bool=@n9) -> (then)
   n10 = Cast To AthenaBroadcastPlayerStatusTG(execute=@n8.LoopBody, Object=@n8)
   > SetWidgetMode!(execute, self:Athena Broadcast Player Status TG Object Reference=@n10, bIsFollowedPlayer:bool=false, bPipMode:bool=@n9) -> (then)
   Reset!(execute=@n6.else, self:Dynamic Entry Box Object Reference=@n7, bDeleteWidgets:bool=true) -> (then)
   n11 = Event OnSquadSessionRankChanged!() -> (OutputDelegate, then, InSquadData:Athena Remote Squad View Data Object Reference, InSquadSessionRank:int=0) {is=title K2Node_Event}
   > UpdateSessionRank!(
    execute
    self:Athena Broadcast Team Tag Object Reference=Get AthenaBroadcastTeamTag() -> (AthenaBroadcastTeamTag:Athena Broadcast Team Tag Object Reference)
    InSessionRank:int=@n11.InSquadSessionRank
   ) -> (then)
   n12 = Event HandleDisplaySessionRankChanged!() -> (OutputDelegate, then, bDisplay:bool) {is=title K2Node_Event}
   > ToggleAllowSessionRank!(execute, self:Athena Broadcast Team Tag Object Reference=Get AthenaBroadcastTeamTag(), Allow:bool=@n12.bDisplay) -> (then)
   n13 = Event OnMinigameTeamIdChanged!() -> (OutputDelegate, then, MinigameTeamId:Fort Minigame Team Identification Object Reference) {is=title K2Node_Event}
   > UpdateFromMinigameTeamId(self:Self Object Reference, MinigameTeamId=@n13.MinigameTeamId)
   n14 = Event OnPlayerAdded!() -> (
    OutputDelegate
    then
    InSquadData:Athena Remote Squad View Data Object Reference
    InPlayerData:Athena Remote Player View Data Object Reference
   ) {is=title K2Node_Event}
   RebuildSquad(execute=[@n14.then, @n15.then])
   > UpdateTeamTagVisibility()
   n15 = Event OnFollowedPlayerChanged() {is=title K2Node_Event}
  ]}
 RebuildSquad {nodes [
   RebuildSquad() {is=title K2Node_FunctionEntry}
   > Reset(self=Get SquadMateList(), bDeleteWidgets=false)
   > n1 = Get!(execute:Exec) -> (then:Exec, else:Exec, SquadData:Athena Remote Squad View Data Weak object)
   > n2 = For Each Loop(Array:Array of Athena Remote Player View Data Object References=GetPlayerDatas!(self:Athena Remote Squad View Data Object Reference=@n1) -> (ReturnValue:Array of Athena Remote Player View Data Object References)) -> (Array Element:Athena Remote Player View Data Object Reference)
   UpdateFromMinigameTeamId(
    execute=@n2.Completed
    MinigameTeamId=GetMinigameTeamId!(self:Athena Remote Squad View Data Object Reference=@n1) -> (ReturnValue:Fort Minigame Team Identification Object Reference)
   )
   n3 = GetPlayerState!(self:Athena Remote Player View Data Object Reference=@n2) -> (ReturnValue:Fort Player State Athena Object Reference)
   Is Valid(exec=@n2.LoopBody, InputObject=@n3)
   > n4 = Branch(Condition="Equal (UniqueIdRepl)"!(
     A:"Unique Net Id Structure (by ref)"=GetPlayerUniqueNetID!(self:Fort Player State Object Reference=@n3) -> (ReturnValue:Unique Net Id Structure)
     B:"Unique Net Id Structure (by ref)"=Get FollowedPlayerUID() -> (FollowedPlayerUID:Unique Net Id Structure)
    ) -> (ReturnValue:bool))
   > SetPlayerData!(
    execute
    self:Athena Spectator Player Widget Base Object Reference=Get FollowedPlayerStatus()
    InPlayerData:Athena Remote Player View Data Object Reference=@n2
    bForceUpdate:bool=false
   ) -> (then)
   > Branch()
   n5 = Create Entry!(execute=@n4.else, self:Dynamic Entry Box Object Reference=Get SquadMateList()) -> (then, ReturnValue:User Widget Object Reference)
   > n6 = Cast To AthenaBroadcastPlayerStatusTG(Object=@n5)
   > SetPlayerData(self=@n6, InPlayerData=@n2, bForceUpdate=true)
   > SetWidgetMode(self=@n6, bIsFollowedPlayer=false, bPipMode=true)
  ]}
 UpdateTeamTagVisibility {nodes [
   UpdateTeamTagVisibility() {is=title K2Node_FunctionEntry}
   n1 = "Equal (Integer)"(A=GetNumEntries!(self:Dynamic Entry Box Base Object Reference=Get SquadMateList()) -> (ReturnValue:int=0), B=0) {is=title K2Node_CallFunction}
   n2 = Get bPipModeEnabled()
   > SetVisibility(
    self=Get AthenaBroadcastTeamTag()
    InVisibility=Select(Option 0:ESlateVisibility Enum=SelfHitTestInvisible, Option 1:ESlateVisibility Enum=Collapsed, Index=OR Boolean(A=@n2, B=@n1)) -> (ReturnValue:ESlateVisibility Enum)
   )
  ]}
]
variables [bPipModeEnabled {type=boolean}]
widgets {
 info {inheritedWidgetCount=1, parentClass=/Script/FortniteUI.AthenaBroadcastSquadStatusWidget, rootWidgetClass=&Overlay, widgetCount=6}
 widgets [
  Overlay_1 {widgetClassPath=&Overlay, widget=Overlay_1}
  StatusVB {
   widgetClassPath=&VerticalBox
   bIsVariable=true
   parent=Overlay_1
   widget=StatusVB
   slot {class=OverlaySlot, properties {horizontalAlignment=HAlign_Fill, verticalAlignment=VAlign_Fill, padding=<ltrb 0 45 0 0>}}
  }
  FollowedPlayerStatus {
   widgetClassPath=&/AthenaBroadcastPlayerStatusTG
   bIsVariable=true
   parent=StatusVB
   widget=FollowedPlayerStatus
   properties {bPipMode=true}
   slot {class=VerticalBoxSlot}
  }
  SquadMateList {
   widgetClassPath=/Script/FortniteUI.FortDynamicEntryBox
   bIsVariable=true
   parent=StatusVB
   widget=SquadMateList
   properties {
    entryBoxType=Vertical
    entryWidgetClass=&/AthenaBroadcastPlayerStatusTG.AthenaBroadcastPlayerStatusTG_C
    numDesignerPreviewEntries=2
   }
   slot {class=VerticalBoxSlot, properties {size {sizeRule=Fill, value=1}}}
  }
  AthenaBroadcastTeamTag {
   widgetClassPath=/FortniteCoreUI/Athena/HUD/Spectator/AthenaBroadcastTeamTag
   bIsVariable=true
   parent=Overlay_1
   widget=AthenaBroadcastTeamTag
   slot {class=OverlaySlot, properties {padding=<ltrb 10 0 0 0>}}
  }
  TutorialHighlight {widgetClassPath=&TutorialHighlight}
 ]
}
