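# Widget Blueprint for the Athena replay spectator action feed panel, which tells viewers to click a player's name to jump to that elimination on the timeline.
# A vertical stack holds a titled header, a body with the hint and a list of AthenaReplayActionFeedRow entries, and a footer Close button bound to the Close input action.
# The Blueprint is layout only: its graph is empty, and the list and button are inherited from AthenaReplayActionFeedScreenBase.
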WidgetBlueprint /FortniteCoreUI/Athena/UI/Spectator/ActionFeed/AthenaReplayActionFeedScreen
graphs [EventGraph {links [], nodes []}]
variables []
widgets {
 info {
  inheritedWidgetCount=2
  parentClass=/Script/FortniteUI.AthenaReplayActionFeedScreenBase
  rootWidgetClass=&VerticalBox
  widgetCount=9
 }
 widgets [
  VerticalBox_0 {widgetClassPath=&VerticalBox, widget=VerticalBox_0}
  HeaderBorder {
   widgetClassPath=&Border
   parent=VerticalBox_0
   widget=HeaderBorder
   properties {horizontalAlignment=HAlign_Left, visibility=SelfHitTestInvisible, brushColor=<rgba 0.0157 0.0196 0.0314 1>, padding=<ltrb 15 2 0 2>}
   slot {class=VerticalBoxSlot}
  }
  PlayerList {
   widgetClassPath=&CommonTextBlock
   parent=HeaderBorder
   widget=PlayerList
   properties {
    style=/FortUILibrary/Styles/Typography/Foundation/TextStyle-Header-M.TextStyle-Header-M_C
    text=ACTION FEED
    font {
     bForceMonospaced=false
     bMaterialIsStencil=false
     fontMaterial=null
     fontObject=&Burbank
     letterSpacing=0
     monospacedWidth=1
     size=32
     skewAmount=0
     typefaceFontName=Burbank-Black
     outlineSettings {
      bApplyOutlineToDropShadows=false
      bMiteredCorners=false
      bSeparateFillAlpha=false
      outlineMaterial=null
      outlineSize=0
      outlineColor=<rgba 0 0 0 1>
     }
    }
    shadowOffset=<0 0>
   }
   slot {class=BorderSlot, properties {horizontalAlignment=HAlign_Left, padding=<ltrb 15 2 0 2>}}
  }
  DescriptionBorder {
   widgetClassPath=&CommonBorder
   parent=VerticalBox_0
   widget=DescriptionBorder
   properties {
    background {
     bIsDynamicallyLoaded=false
     drawAs=Box
     imageType=NoImage
     mirroring=NoMirror
     resourceName=None
     resourceObject=null
     tiling=NoTile
     imageSize=<32 32>
     margin=<ltrb 0 0 0 0>
     outlineSettings {
      bUseBrushTransparency=false
      roundingType=HalfHeightRadius
      width=0
      color {colorUseRule=UseColor_Specified, specifiedColor=<rgba 0 0 0 0>}
      cornerRadii=<0 0 0 0>
     }
     tintColor {colorUseRule=UseColor_Specified, specifiedColor=<rgba 1 1 1 1>}
     uVRegion {bIsValid=false, max=<0 0>, min=<0 0>}
    }
    brushColor=<rgba 0.0157 0.0196 0.0314 0.702>
    padding=<ltrb 0 0 0 0>
   }
   slot {class=VerticalBoxSlot, properties {size {sizeRule=Fill, value=1}}}
  }
  VerticalBox_1 {
   widgetClassPath=&VerticalBox
   parent=DescriptionBorder
   widget=VerticalBox_1
   slot {class=BorderSlot, properties {padding=<ltrb 0 0 0 0>}}
  }
  TextBlock_88 {
   widgetClassPath=&TextBlock
   parent=VerticalBox_1
   widget=TextBlock_88
   properties {
    text=Click on a player's name to be taken to their elimination on the timeline.
    visibility=SelfHitTestInvisible
    wrapTextAt=570
    font {
     bForceMonospaced=false
     bMaterialIsStencil=false
     fontMaterial=null
     fontObject=&NotoSans
     letterSpacing=0
     monospacedWidth=1
     size=16
     skewAmount=0
     typefaceFontName=NotoSans-Regular
     outlineSettings {
      bApplyOutlineToDropShadows=false
      bMiteredCorners=false
      bSeparateFillAlpha=false
      outlineMaterial=null
      outlineSize=0
      outlineColor=<rgba 0 0 0 1>
     }
    }
   }
   slot {class=VerticalBoxSlot, properties {padding=<ltrb 15 15 15 15>}}
  }
  ListView_Entries {
   widgetClassPath=&CommonListView
   bInherited=true
   parent=VerticalBox_1
   widget=ListView_Entries
   properties {
    entryWidgetClass=&/AthenaReplayActionFeedRow.AthenaReplayActionFeedRow_C
    horizontalEntrySpacing=2
    numDesignerPreviewEntries=20
    selectionMode=None
    verticalEntrySpacing=2
    visibility=SelfHitTestInvisible
   }
   slot {class=VerticalBoxSlot, properties {padding=<ltrb 5 0 5 0>, size {sizeRule=Fill, value=1}}}
  }
  FooterBorder {
   widgetClassPath=&CommonBorder
   parent=VerticalBox_0
   widget=FooterBorder
   properties {
    horizontalAlignment=HAlign_Right
    background {
     bIsDynamicallyLoaded=false
     drawAs=Box
     imageType=NoImage
     mirroring=NoMirror
     resourceName=None
     resourceObject=null
     tiling=NoTile
     imageSize=<32 32>
     margin=<ltrb 0 0 0 0>
     outlineSettings {
      bUseBrushTransparency=false
      roundingType=HalfHeightRadius
      width=0
      color {colorUseRule=UseColor_Specified, specifiedColor=<rgba 0 0 0 0>}
      cornerRadii=<0 0 0 0>
     }
     tintColor {colorUseRule=UseColor_Specified, specifiedColor=<rgba 1 1 1 1>}
     uVRegion {bIsValid=false, max=<0 0>, min=<0 0>}
    }
    brushColor=<rgba 0.0157 0.0196 0.0314 1>
   }
   slot {class=VerticalBoxSlot}
  }
  Button_Close {
   widgetClassPath=&IconTextButton
   bInherited=true
   parent=FooterBorder
   widget=Button_Close
   properties {
    bHideInputActionWithKeyboard=true
    bIsFocusable=false
    buttonText=Close
    contentAlignment=HAlign_Center
    onButtonBaseClicked=(&self:WidgetTree.Button_Close.HandleOnButtonClicked)
    onButtonBaseDoubleClicked=(&self:WidgetTree.Button_Close.HandleOnButtonDoubleClicked)
    onButtonBaseHovered=(&self:WidgetTree.Button_Close.HandleOnButtonHovered)
    onButtonBaseUnhovered=(&self:WidgetTree.Button_Close.HandleOnButtonUnhovered)
    onSelectedChangedBase=(&self:WidgetTree.Button_Close.HandleOnSelectedChanged)
    style=/FortUILibrary/Foundation/Buttons/ButtonStyles/ButtonStyle-TransparentWithCues-XS-B-White.ButtonStyle-TransparentWithCues-XS-B-White_C
    pressedSlateSoundOverride {resourceObject=/Game/Sounds/Fort_UI_Sounds/Replay/DVR_CamOptions_Close_Cue}
    triggeringInputAction {dataTable=&InputActionDataTable, rowName=Close}
   }
   slot {class=BorderSlot, properties {horizontalAlignment=HAlign_Right}}
  }
 ]
}
