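# Widget Blueprint for the spectator event-match scoreboard screen, based on AthenaEventMatchScoreboardScreenBase.
# A scale-to-fit panel shows a header row (player, health, materials, eliminations, place) above the inherited ScoreboardListView of squad entry widgets.
# Fixed-width size boxes define the header columns, with a spacer reserved for the scrollbar.
# SetSquadIdsVisible_BP toggles the unlabeled squad ID column between SelfHitTestInvisible and Collapsed.
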WidgetBlueprint /FortniteCoreUI/Athena/UI/Spectator/EventMatchInfo/AthenaEventMatchScoreboardScreen
graphs [EventGraph {nodes [
   n1 = Event SetSquadIdsVisible_BP!() -> (OutputDelegate, then, bSquadIdsVisible:bool) {is=title K2Node_Event}
   > SetVisibility(
    self=Get SquadIdSB() -> (SquadIdSB:Size Box Object Reference)
    InVisibility=Select(Option 0=Collapsed, Option 1=SelfHitTestInvisible, Index=@n1.bSquadIdsVisible)
   )
  ]}]
variables []
widgets {
 info {
  inheritedWidgetCount=1
  parentClass=/Script/FortniteUI.AthenaEventMatchScoreboardScreenBase
  rootWidgetClass=&CommonBorder
  widgetCount=21
 }
 widgets [
  CommonBorder-BG {
   widgetClassPath=&CommonBorder
   widget=CommonBorder-BG
   properties {
    background {
     bIsDynamicallyLoaded=false
     drawAs=Image
     imageType=NoImage
     mirroring=NoMirror
     resourceName=None
     resourceObject=/Game/UI/Foundation/Materials/M_UI_GradientPanel_MatchStatus
     tiling=NoTile
     imageSize=<1920 1080>
     margin=<ltrb 0 0 0 0>
     outlineSettings {
      bUseBrushTransparency=false
      roundingType=HalfHeightRadius
      width=0
      color {colorUseRule=UseColor_Specified, specifiedColor=<rgba 0 0 0 0>}
      cornerRadii=<0 0 0 0>
     }
     tintColor {colorUseRule=UseColor_Specified, specifiedColor=<rgba 1 1 1 1>}
     uVRegion {bIsValid=false, max=<0 0>, min=<0 0>}
    }
    padding=<ltrb 0 0 0 0>
   }
  }
  Overlay_1 {widgetClassPath=&Overlay, parent=CommonBorder-BG, widget=Overlay_1, slot {class=BorderSlot, properties {padding=<ltrb 0 0 0 0>}}}
  VerticalBox_0 {
   widgetClassPath=&VerticalBox
   parent=Overlay_1
   widget=VerticalBox_0
   slot {class=OverlaySlot, properties {horizontalAlignment=HAlign_Fill, verticalAlignment=VAlign_Fill, padding=<ltrb 50 12 50 70>}}
  }
  CommonTextBlock_0 {
   widgetClassPath=&CommonTextBlock
   parent=VerticalBox_0
   widget=CommonTextBlock_0
   properties {
    scrollStyle=&TextScrollStyle-Base_C
    text=SCOREBOARD
    font {
     bForceMonospaced=false
     bMaterialIsStencil=false
     fontMaterial=null
     fontObject=&Burbank
     letterSpacing=0
     monospacedWidth=1
     size=48
     skewAmount=0
     typefaceFontName=BurbankRegular-Black
     outlineSettings {
      bApplyOutlineToDropShadows=false
      bMiteredCorners=false
      bSeparateFillAlpha=false
      outlineMaterial=null
      outlineSize=0
      outlineColor=<rgba 0 0 0 1>
     }
    }
   }
   slot {class=VerticalBoxSlot, properties {horizontalAlignment=HAlign_Center, padding=<ltrb 0 40 0 0>}}
  }
  ScaleBox_0 {
   widgetClassPath=&ScaleBox
   parent=VerticalBox_0
   widget=ScaleBox_0
   properties {stretch=ScaleToFitX}
   slot {class=VerticalBoxSlot, properties {verticalAlignment=VAlign_Top, padding=<ltrb 0 15 0 0>, size {sizeRule=Fill, value=1}}}
  }
  VerticalBox_1 {
   widgetClassPath=&VerticalBox
   parent=ScaleBox_0
   widget=VerticalBox_1
   slot {class=ScaleBoxSlot, properties {horizontalAlignment=HAlign_Fill, verticalAlignment=VAlign_Fill}}
  }
  SizeBox_0 {
   widgetClassPath=&SizeBox
   parent=VerticalBox_1
   widget=SizeBox_0
   properties {bOverride_HeightOverride=true, heightOverride=40}
   slot {class=VerticalBoxSlot, properties {horizontalAlignment=HAlign_Left}}
  }
  HorizontalBox_0 {widgetClassPath=&HorizontalBox, parent=SizeBox_0, widget=HorizontalBox_0, slot {class=SizeBoxSlot}}
  SquadIdSB {
   widgetClassPath=&SizeBox
   bIsVariable=true
   parent=HorizontalBox_0
   widget=SquadIdSB
   properties {bOverride_WidthOverride=true, widthOverride=50}
   slot {class=HorizontalBoxSlot}
  }
  PlayerNameSB {
   widgetClassPath=&SizeBox
   bIsVariable=true
   parent=HorizontalBox_0
   widget=PlayerNameSB
   properties {bOverride_WidthOverride=true, widthOverride=330}
   slot {class=HorizontalBoxSlot}
  }
  PlayerText {
   widgetClassPath=&CommonTextBlock
   parent=PlayerNameSB
   widget=PlayerText
   properties {
    justification=Center
    scrollStyle=&TextScrollStyle-Base_C
    text=PLAYER
    colorAndOpacity {colorUseRule=UseColor_Specified, specifiedColor=<rgba 1 1 1 0.7>}
    font {
     bForceMonospaced=false
     bMaterialIsStencil=false
     fontMaterial=null
     fontObject=&Burbank
     letterSpacing=0
     monospacedWidth=1
     size=16
     skewAmount=0
     typefaceFontName=BurbankRegular-Black
     outlineSettings {
      bApplyOutlineToDropShadows=false
      bMiteredCorners=false
      bSeparateFillAlpha=false
      outlineMaterial=null
      outlineSize=0
      outlineColor=<rgba 0 0 0 1>
     }
    }
   }
   slot {class=SizeBoxSlot, properties {horizontalAlignment=HAlign_Center, verticalAlignment=VAlign_Bottom, padding=<ltrb 4 0 4 0>}}
  }
  HealthSB {
   widgetClassPath=&SizeBox
   bIsVariable=true
   parent=HorizontalBox_0
   widget=HealthSB
   properties {bOverride_WidthOverride=true, widthOverride=260}
   slot {class=HorizontalBoxSlot}
  }
  Health {
   widgetClassPath=&CommonTextBlock
   parent=HealthSB
   widget=Health
   properties {
    scrollStyle=&TextScrollStyle-Base_C
    text=HEALTH
    colorAndOpacity {colorUseRule=UseColor_Specified, specifiedColor=<rgba 1 1 1 0.7>}
    font {
     bForceMonospaced=false
     bMaterialIsStencil=false
     fontMaterial=null
     fontObject=&Burbank
     letterSpacing=0
     monospacedWidth=1
     size=16
     skewAmount=0
     typefaceFontName=BurbankRegular-Black
     outlineSettings {
      bApplyOutlineToDropShadows=false
      bMiteredCorners=false
      bSeparateFillAlpha=false
      outlineMaterial=null
      outlineSize=0
      outlineColor=<rgba 0 0 0 1>
     }
    }
   }
   slot {class=SizeBoxSlot, properties {horizontalAlignment=HAlign_Center, verticalAlignment=VAlign_Bottom}}
  }
  MaterialsSB {
   widgetClassPath=&SizeBox
   bIsVariable=true
   parent=HorizontalBox_0
   widget=MaterialsSB
   properties {bOverride_WidthOverride=true, widthOverride=150}
   slot {class=HorizontalBoxSlot}
  }
  Materials {
   widgetClassPath=&CommonTextBlock
   parent=MaterialsSB
   widget=Materials
   properties {
    scrollStyle=&TextScrollStyle-Base_C
    text=MATERIALS
    colorAndOpacity {colorUseRule=UseColor_Specified, specifiedColor=<rgba 1 1 1 0.7>}
    font {
     bForceMonospaced=false
     bMaterialIsStencil=false
     fontMaterial=null
     fontObject=&Burbank
     letterSpacing=0
     monospacedWidth=1
     size=16
     skewAmount=0
     typefaceFontName=BurbankRegular-Black
     outlineSettings {
      bApplyOutlineToDropShadows=false
      bMiteredCorners=false
      bSeparateFillAlpha=false
      outlineMaterial=null
      outlineSize=0
      outlineColor=<rgba 0 0 0 1>
     }
    }
   }
   slot {class=SizeBoxSlot, properties {horizontalAlignment=HAlign_Center, verticalAlignment=VAlign_Bottom}}
  }
  EliminationsSB {
   widgetClassPath=&SizeBox
   bIsVariable=true
   parent=HorizontalBox_0
   widget=EliminationsSB
   properties {bOverride_WidthOverride=true, widthOverride=150}
   slot {class=HorizontalBoxSlot}
  }
  Eliminations {
   widgetClassPath=&CommonTextBlock
   parent=EliminationsSB
   widget=Eliminations
   properties {
    scrollStyle=&TextScrollStyle-Base_C
    text=ELIMINATIONS
    colorAndOpacity {colorUseRule=UseColor_Specified, specifiedColor=<rgba 1 1 1 0.7>}
    font {
     bForceMonospaced=false
     bMaterialIsStencil=false
     fontMaterial=null
     fontObject=&Burbank
     letterSpacing=0
     monospacedWidth=1
     size=16
     skewAmount=0
     typefaceFontName=BurbankRegular-Black
     outlineSettings {
      bApplyOutlineToDropShadows=false
      bMiteredCorners=false
      bSeparateFillAlpha=false
      outlineMaterial=null
      outlineSize=0
      outlineColor=<rgba 0 0 0 1>
     }
    }
   }
   slot {class=SizeBoxSlot, properties {horizontalAlignment=HAlign_Center, verticalAlignment=VAlign_Bottom}}
  }
  PlaceSB {
   widgetClassPath=&SizeBox
   bIsVariable=true
   parent=HorizontalBox_0
   widget=PlaceSB
   properties {bOverride_WidthOverride=true, widthOverride=110}
   slot {class=HorizontalBoxSlot}
  }
  Place {
   widgetClassPath=&CommonTextBlock
   parent=PlaceSB
   widget=Place
   properties {
    scrollStyle=&TextScrollStyle-Base_C
    text=PLACE
    colorAndOpacity {colorUseRule=UseColor_Specified, specifiedColor=<rgba 1 1 1 0.7>}
    font {
     bForceMonospaced=false
     bMaterialIsStencil=false
     fontMaterial=null
     fontObject=&Burbank
     letterSpacing=0
     monospacedWidth=1
     size=16
     skewAmount=0
     typefaceFontName=BurbankRegular-Black
     outlineSettings {
      bApplyOutlineToDropShadows=false
      bMiteredCorners=false
      bSeparateFillAlpha=false
      outlineMaterial=null
      outlineSize=0
      outlineColor=<rgba 0 0 0 1>
     }
    }
   }
   slot {class=SizeBoxSlot, properties {horizontalAlignment=HAlign_Center, verticalAlignment=VAlign_Bottom}}
  }
  ScrollbarSB {
   widgetClassPath=&SizeBox
   parent=HorizontalBox_0
   widget=ScrollbarSB
   properties {bOverride_WidthOverride=true, widthOverride=28}
   slot {class=HorizontalBoxSlot}
  }
  ScoreboardListView {
   widgetClassPath=&CommonListView
   bInherited=true
   bIsVariable=true
   parent=VerticalBox_1
   widget=ScoreboardListView
   properties {
    entryWidgetClass=&/AthenaEventMatchScoreboardSquadWidget.AthenaEventMatchScoreboardSquadWidget_C
    horizontalEntrySpacing=10
    numDesignerPreviewEntries=7
    selectionMode=None
    verticalEntrySpacing=10
   }
   slot {class=VerticalBoxSlot, properties {size {sizeRule=Fill, value=1}}}
  }
 ]
}
