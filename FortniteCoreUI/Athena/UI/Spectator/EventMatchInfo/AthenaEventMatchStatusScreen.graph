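# Match status screen for event spectating that shows match time, players remaining, storm state, match totals and elimination leaders.
# Parent-class update events push text into named widgets, with helper functions playing pulse or new-top animations on highlighted stats.
# Construct fills fields with placeholder text, and a player-count change also pulses time since last elimination.
# When a player's dead state changes, leader entries showing that player's name play a TopPlayerDead animation.
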WidgetBlueprint /FortniteCoreUI/Athena/UI/Spectator/EventMatchInfo/AthenaEventMatchStatusScreen
graphs [
 EventGraph {nodes [
   n1 = For Each Loop(
    Exec=@n23.Is Valid
    Array:Array of Athena Match Status Text Player Object References=Make Array!(
     "[0]":Athena Match Status Text Player Object Reference=Get FastestEliminationPlayer() -> (FastestEliminationPlayer:Athena Match Status Text Player Object Reference)
     "[1]":Athena Match Status Text Player Object Reference=Get LongestEliminationPlayer() -> (LongestEliminationPlayer:Athena Match Status Text Player Object Reference)
     "[2]":Athena Match Status Text Player Object Reference=Get MostEliminationsPlayer() -> (MostEliminationsPlayer:Athena Match Status Text Player Object Reference)
    ) -> (Array:Array of Athena Match Status Text Player Object References)
   ) -> (Array Element:Athena Match Status Text Player Object Reference)
   StopAllAnimations(self:User Widget Object Reference=@n1)
   > PlayAnimation(
    self=@n1
    InAnimation=Get TopPlayerDead!(self:Athena Match Status Text Player Object Reference=@n1) -> (TopPlayerDead:Widget Animation Object Reference)
   )
   n2 = Event UpdateGamePhaseTimeText!() -> (OutputDelegate, then, GamePhaseTimeText:"Text (by ref)") {is=title K2Node_Event}
   > "SetText (Text)"(self=Get GamePhaseTime() -> (GamePhaseTime:Common Text Object Reference), InText=@n2.GamePhaseTimeText)
   n3 = Event UpdateGamePhaseText!() -> (OutputDelegate, then, GamePhaseText:"Text (by ref)") {is=title K2Node_Event}
   > NewStatUpdate3Pulse!(
    execute
    self
    TargetWidget:Athena Event Match Status Text Object Reference=Get StormPhase() -> (StormPhase:Athena Event Match Status Text Object Reference)
    InText:Text=@n3.GamePhaseText
    SkipTextUpdate:bool=false
   ) -> (then)
   n4 = Event UpdateShotsFiredText!() -> (OutputDelegate, then, ShotsFiredText:"Text (by ref)") {is=title K2Node_Event}
   > "SetText (Text)"(self=Get ShotsFired() -> (ShotsFired:Common Text Object Reference), InText=@n4.ShotsFiredText)
   n5 = Event UpdatePiecesDestroyedText!() -> (OutputDelegate, then, PiecesDestroyedText:"Text (by ref)") {is=title K2Node_Event}
   > "SetText (Text)"(self=Get PiecesDestroyed() -> (PiecesDestroyed:Common Text Object Reference), InText=@n5.PiecesDestroyedText)
   n6 = Event UpdatePiecesBuiltText!() -> (OutputDelegate, then, PiecesBuiltText:"Text (by ref)") {is=title K2Node_Event}
   > "SetText (Text)"(self=Get PiecesBuilt() -> (PiecesBuilt:Common Text Object Reference), InText=@n6.PiecesBuiltText)
   n7 = Event UpdateLongestEliminationText!() -> (OutputDelegate, then, LongestEliminationText:"Text (by ref)") {is=title K2Node_Event}
   > NewTopPlayer!(
    execute
    self
    TargetWidget:Athena Match Status Text Player Object Reference=Get LongestEliminationPlayer()
    InText:Text=@n7.LongestEliminationText
   ) -> (then)
   n8 = Event UpdateDamageTakenText!() -> (OutputDelegate, then, DamageTakenText:"Text (by ref)") {is=title K2Node_Event}
   > "SetText (Text)"(self=Get DamageTakenPlayer() -> (DamageTakenPlayer:Common Text Object Reference), InText=@n8.DamageTakenText)
   n9 = Event UpdateDamageHealedText!() -> (OutputDelegate, then, DamageHealedText:"Text (by ref)") {is=title K2Node_Event}
   > "SetText (Text)"(self=Get DamageHealedPlayer() -> (DamageHealedPlayer:Common Text Object Reference), InText=@n9.DamageHealedText)
   n10 = Event UpdateDamageInflictedText!() -> (OutputDelegate, then, DamageInflictedText:"Text (by ref)") {is=title K2Node_Event}
   > "SetText (Text)"(self=Get DamageInflictedPlayer() -> (DamageInflictedPlayer:Common Text Object Reference), InText=@n10.DamageInflictedText)
   n11 = Event UpdateChestsLootedText!() -> (OutputDelegate, then, ChestsLootedText:"Text (by ref)") {is=title K2Node_Event}
   > "SetText (Text)"(self=Get ChestsLooted() -> (ChestsLooted:Common Text Object Reference), InText=@n11.ChestsLootedText)
   n12 = Event UpdateMostKillsScore!() -> (OutputDelegate, then, MostKillsScore:"Text (by ref)") {is=title K2Node_Event}
   > NewTopScore!(
    execute
    self
    TargetWidget:Athena Event Match Status Text Object Reference=Get MostEliminations() -> (MostEliminations:Athena Event Match Status Text Object Reference)
    InText:Text=@n12.MostKillsScore
   ) -> (then)
   n13 = Event UpdateLongestEliminationScore!() -> (OutputDelegate, then, LongestEliminationScore:"Text (by ref)") {is=title K2Node_Event}
   > NewTopScore(
    TargetWidget=Get LongestElimination() -> (LongestElimination:Athena Event Match Status Text Object Reference)
    InText=@n13.LongestEliminationScore
   )
   n14 = Event UpdateMatchTimeText!() -> (OutputDelegate, then, MatchTimeText:"Text (by ref)") {is=title K2Node_Event}
   > "SetText (Text)"(self=Get MatchTime() -> (MatchTime:Common Text Object Reference), InText=@n14.MatchTimeText)
   n15 = Event UpdateTimeSinceLastKillText!() -> (OutputDelegate, then, TimeSinceLastKillText:"Text (by ref)") {is=title K2Node_Event}
   > SetText!(
    execute
    self:Athena Event Match Status Text Object Reference=Get TimeSinceLastKill() -> (TimeSinceLastKill:Athena Event Match Status Text Object Reference)
    InText:Text=@n15.TimeSinceLastKillText
    ApplyImmediately:bool=true
   ) -> (then)
   n16 = Event UpdateFastestKillText!() -> (OutputDelegate, then, FastestKillText:"Text (by ref)", FastestKillTime:"Text (by ref)") {is=title K2Node_Event}
   > n17 = Sequence()
   > NewTopPlayer(TargetWidget=Get FastestEliminationPlayer(), InText=@n16.FastestKillText)
   NewTopScore(
    execute=@n17.then_1
    TargetWidget=Get FastestEliminationTime() -> (FastestEliminationTime:Athena Event Match Status Text Object Reference)
    InText=@n16.FastestKillTime
   )
   n18 = Event UpdatePlayersRemainingText!() -> (OutputDelegate, then, PlayersRemainingText:"Text (by ref)") {is=title K2Node_Event}
   > NewStatUpdate1Pulse!(
    execute
    self
    TargetWidget:Athena Event Match Status Text Object Reference=Get PlayersRemaining() -> (PlayersRemaining:Athena Event Match Status Text Object Reference)
    InText:Text=@n18.PlayersRemainingText
    SkipTextUpdate:bool=false
   ) -> (then)
   > NewStatUpdate1Pulse(TargetWidget=Get TimeSinceLastKill(), SkipTextUpdate=true)
   n19 = Event UpdateGamePhaseTimeTitleText!() -> (OutputDelegate, then, GamePhaseTimeTitleText:"Text (by ref)") {is=title K2Node_Event}
   > "SetText (Text)"(self=Get GamePhaseTitleText() -> (GamePhaseTitleText:Common Text Object Reference), InText=@n19.GamePhaseTimeTitleText)
   n20 = Event UpdateCircleRadiusText!() -> (OutputDelegate, then, CircleRadiusText:"Text (by ref)") {is=title K2Node_Event}
   > "SetText (Text)"(self=Get CircleRadius() -> (CircleRadius:Common Text Object Reference), InText=@n20.CircleRadiusText)
   n21 = Event UpdateMostKillsText!() -> (OutputDelegate, then, MostKillsText:"Text (by ref)") {is=title K2Node_Event}
   > NewTopPlayer(TargetWidget=Get MostEliminationsPlayer(), InText=@n21.MostKillsText)
   n22 = Event On Player Dead State Changed!() -> (OutputDelegate, then, InPlayerData:Athena Remote Player View Data Object Reference, bInIsDead:bool) {is=title K2Node_Event}
   > n23 = Is Valid(InputObject=@n22.InPlayerData)
   Event Construct()
   > n24 = Sequence!(execute) -> (then_0, then_1, then_2:Exec)
   n25 = For Each Loop(
    Exec=@n24.then_1
    Array:Array of Athena Event Match Status Text Object References=Make Array!(
     "[0]":Athena Event Match Status Text Object Reference=Get FastestEliminationTime()
     "[1]":Athena Event Match Status Text Object Reference=Get MostEliminations()
     "[2]":Athena Event Match Status Text Object Reference=Get LongestElimination()
     "[3]":Athena Event Match Status Text Object Reference=Get TimeSinceLastKill()
     "[4]":Athena Event Match Status Text Object Reference=Get StormPhase()
     "[5]":Athena Event Match Status Text Object Reference=Get PlayersRemaining()
    ) -> (Array:Array of Athena Event Match Status Text Object References)
   ) -> (Array Element:Athena Event Match Status Text Object Reference)
   n26 = MakeLiteralText(Value="LOCTEXT(\"N/A\")")
   > SetText(self=@n25, InText=@n26, ApplyImmediately=true)
   n27 = For Each Loop(
    Exec=@n25.Completed
    Array:Array of Athena Match Status Text Player Object References=Make Array!(
     "[0]":Athena Match Status Text Player Object Reference=Get FastestEliminationPlayer()
     "[1]":Athena Match Status Text Player Object Reference=Get MostEliminationsPlayer()
     "[2]":Athena Match Status Text Player Object Reference=Get LongestEliminationPlayer()
    ) -> (Array:Array of Athena Match Status Text Player Object References)
   ) -> (Array Element:Athena Match Status Text Player Object Reference)
   > SetText(self:Athena Match Status Text Player Object Reference=@n27, InText=@n26, ApplyImmediately=true)
   UpdateCircleRadiusText!(execute=@n27.Completed, self:Athena Event Match Status Screen Base Object Reference, CircleRadiusText:"Text (by ref)"=@n26) -> (then)
   > UpdateGamePhaseTimeText!(execute, self:Athena Event Match Status Screen Base Object Reference, GamePhaseTimeText:"Text (by ref)"=@n26) -> (then)
   n28 = MakeLiteralText()
   UpdateMatchTimeText!(execute=@n24.then_0, self:Athena Event Match Status Screen Base Object Reference, MatchTimeText:"Text (by ref)"=@n28) -> (then)
   > UpdateSupplyDropsLootedText!(execute, self:Athena Event Match Status Screen Base Object Reference, SupplyDropsLootedText:"Text (by ref)"=@n28) -> (then)
   > UpdatePiecesBuiltText!(execute, self:Athena Event Match Status Screen Base Object Reference, PiecesBuiltText:"Text (by ref)"=@n28) -> (then)
   > UpdatePiecesDestroyedText!(execute, self:Athena Event Match Status Screen Base Object Reference, PiecesDestroyedText:"Text (by ref)"=@n28) -> (then)
   > UpdateShotsFiredText!(execute, self:Athena Event Match Status Screen Base Object Reference, ShotsFiredText:"Text (by ref)"=@n28) -> (then)
   > UpdateChestsLootedText!(execute, self:Athena Event Match Status Screen Base Object Reference, ChestsLootedText:"Text (by ref)"=@n28) -> (then)
   > UpdateDamageHealedText!(execute, self:Athena Event Match Status Screen Base Object Reference, DamageHealedText:"Text (by ref)"=@n28) -> (then)
   > UpdateDamageInflictedText!(execute, self:Athena Event Match Status Screen Base Object Reference, DamageInflictedText:"Text (by ref)"=@n28) -> (then)
   > UpdateDamageTakenText!(execute, self:Athena Event Match Status Screen Base Object Reference, DamageTakenText:"Text (by ref)"=@n28) -> (then)
   > UpdateGamePhaseTimeTitleText!(execute, self:Athena Event Match Status Screen Base Object Reference, GamePhaseTimeTitleText:"Text (by ref)"=@n28) -> (then)
   n29 = Event UpdateSupplyDropsLootedText!() -> (OutputDelegate, then, SupplyDropsLootedText:"Text (by ref)") {is=title K2Node_Event}
   > "SetText (Text)"(self=Get SupplyDropsLooted() -> (SupplyDropsLooted:Common Text Object Reference), InText=@n29.SupplyDropsLootedText)
   n30 = Get TextToDisplay!(self:Athena Match Status Text Player Object Reference=@n1) -> (TextToDisplay:Text)
   SetText(execute=@n32.then, self=@n1, InText=@n30, ApplyImmediately=true)
   SetVisibility(
    self=[Get ChestsLooted(), Get ChestsLootedTitleText() -> (ChestsLootedTitleText:Common Text Object Reference)]
    InVisibility=Select(Option 0=SelfHitTestInvisible, Option 1=Collapsed, Index=TextIsEmpty(InText=@n11.ChestsLootedText))
   )
   n31 = "Equal Exactly (Text)"!(
    A:"Text (by ref)"=@n30
    B:"Text (by ref)"="To Text (String)"(InString=GetPlayerName!(self:Athena Remote Player View Data Object Reference=@n22.InPlayerData) -> (ReturnValue:String))
   ) -> (ReturnValue:bool)
   n32 = Branch(execute=@n1.LoopBody, Condition=@n31)
  ]}
 NewStatUpdate1Pulse {nodes [
   n1 = NewStatUpdate1Pulse!() -> (then, TargetWidget:Athena Event Match Status Text Object Reference, InText:Text, SkipTextUpdate:bool=~) {is=title K2Node_FunctionEntry}
   > n2 = Branch(Condition=@n1.SkipTextUpdate)
   SetText(execute=@n2.else, self:Athena Event Match Status Text Object Reference=@n1, InText=@n1.InText, ApplyImmediately=true)
   Reroute Node()
   n3 = Get HighlightStat1!(self:Athena Event Match Status Text Object Reference=@n1) -> (HighlightStat1:Widget Animation Object Reference)
   > n4 = Branch(Condition=IsAnimationPlaying(self=@n1, InAnimation=@n3))
   PlayAnimation(execute=@n4.else, self=@n1, InAnimation=@n3)
  ]}
 NewStatUpdate3Pulse {nodes [
   n1 = NewStatUpdate3Pulse!() -> (then, TargetWidget:Athena Event Match Status Text Object Reference, InText:Text, SkipTextUpdate:bool=~) {is=title K2Node_FunctionEntry}
   > n2 = Branch(Condition=@n1.SkipTextUpdate)
   SetText(execute=@n2.else, self=@n1, InText=@n1.InText, ApplyImmediately=true)
   Reroute Node()
   n3 = Get HighlightStat3!(self:Athena Event Match Status Text Object Reference=@n1) -> (HighlightStat3:Widget Animation Object Reference)
   > n4 = Branch(Condition=IsAnimationPlaying(self=@n1, InAnimation=@n3))
   PlayAnimation(execute=@n4.else, self=@n1, InAnimation=@n3)
  ]}
 NewTopPlayer {nodes [
   n1 = NewTopPlayer!() -> (then, TargetWidget:Athena Match Status Text Player Object Reference, InText:Text) {is=title K2Node_FunctionEntry}
   StopAllAnimations(self=@n1)
   > PlayAnimation(
    self=@n1
    InAnimation=Get NewTopPlayer!(self:Athena Match Status Text Player Object Reference=@n1) -> (NewTopPlayer:Widget Animation Object Reference)
   )
   SetText(execute=@n1.then, self:Athena Match Status Text Player Object Reference=@n1, InText=@n1.InText, ApplyImmediately=true)
  ]}
 NewTopScore {nodes [
   n1 = NewTopScore!() -> (then, TargetWidget:Athena Event Match Status Text Object Reference, InText:Text) {is=title K2Node_FunctionEntry}
   > SetText(self:Athena Event Match Status Text Object Reference=@n1, InText=@n1.InText, ApplyImmediately=true)
   n2 = Get NewTopScore!(self:Athena Event Match Status Text Object Reference=@n1) -> (NewTopScore:Widget Animation Object Reference)
   n3 = Branch(Condition=IsAnimationPlaying(self=@n1, InAnimation=@n2))
   > n4 = StopAnimation(self=@n1, InAnimation=@n2)
   PlayAnimation(execute=[@n4.then, @n3.else], self=@n1, InAnimation=@n2)
  ]}
]
variables []
widgets {
 info {
  inheritedWidgetCount=0
  parentClass=/Script/FortniteUI.AthenaEventMatchStatusScreenBase
  rootWidgetClass=&CommonBorder
  widgetCount=130
 }
 widgets [
  CommonBorder_0 {
   widgetClassPath=&CommonBorder
   widget=CommonBorder_0
   properties {
    background {
     bIsDynamicallyLoaded=false
     drawAs=Box
     imageType=NoImage
     mirroring=NoMirror
     resourceName=None
     resourceObject=/Game/UI/Foundation/Materials/M_UI_GradientPanel_MatchStatus
     tiling=NoTile
     imageSize=<32 32>
     margin=<ltrb 0 0 0 0>
     outlineSettings {
      bUseBrushTransparency=false
      roundingType=HalfHeightRadius
      width=0
      color {colorUseRule=UseColor_Specified, specifiedColor=<rgba 0 0 0 0>}
      cornerRadii=<0 0 0 0>
     }
     tintColor {colorUseRule=UseColor_Specified, specifiedColor=<rgba 1 1 1 1>}
     uVRegion {bIsValid=false, max=<0 0>, min=<0 0>}
    }
    padding=<ltrb 0 0 0 0>
   }
  }
  VerticalBox_0 {
   widgetClassPath=&VerticalBox
   parent=CommonBorder_0
   widget=VerticalBox_0
   slot {class=BorderSlot, properties {padding=<ltrb 0 0 0 0>}}
  }
  MatchStatusTitleText {
   widgetClassPath=&CommonTextBlock
   parent=VerticalBox_0
   widget=MatchStatusTitleText
   properties {
    scrollStyle=&TextScrollStyle-Base_C
    text=MATCH STATUS
    font {
     bForceMonospaced=false
     bMaterialIsStencil=false
     fontMaterial=null
     fontObject=&Burbank
     letterSpacing=0
     monospacedWidth=1
     size=64
     skewAmount=0
     typefaceFontName=BurbankRegular-Black
     outlineSettings {
      bApplyOutlineToDropShadows=false
      bMiteredCorners=false
      bSeparateFillAlpha=false
      outlineMaterial=null
      outlineSize=0
      outlineColor=<rgba 0 0 0 1>
     }
    }
   }
   slot {class=VerticalBoxSlot, properties {horizontalAlignment=HAlign_Center, verticalAlignment=VAlign_Center, padding=<ltrb 0 70 0 30>}}
  }
  HorizontalBox_0 {
   widgetClassPath=&HorizontalBox
   parent=VerticalBox_0
   widget=HorizontalBox_0
   slot {class=VerticalBoxSlot, properties {padding=<ltrb 220 30 220 30>}}
  }
  VerticalBox_1 {
   widgetClassPath=&VerticalBox
   parent=HorizontalBox_0
   widget=VerticalBox_1
   slot {class=HorizontalBoxSlot, properties {horizontalAlignment=HAlign_Center, size {sizeRule=Fill, value=1}}}
  }
  HorizontalBox_2 {widgetClassPath=&HorizontalBox, parent=VerticalBox_1, widget=HorizontalBox_2, slot {class=VerticalBoxSlot}}
  VerticalBox_8 {widgetClassPath=&VerticalBox, parent=HorizontalBox_2, widget=VerticalBox_8, slot {class=HorizontalBoxSlot}}
  CommonBorder_10 {
   widgetClassPath=&CommonBorder
   parent=VerticalBox_8
   widget=CommonBorder_10
   properties {style=/FortUILibrary/Styles/Border/Border_MatchStatus_Blue.Border_MatchStatus_Blue_C, padding=<ltrb 0 0 0 0>}
   slot {class=VerticalBoxSlot, properties {padding=<ltrb 0 20 0 20>}}
  }
  SizeBox_0 {
   widgetClassPath=&SizeBox
   parent=CommonBorder_10
   widget=SizeBox_0
   properties {bOverride_HeightOverride=true, bOverride_WidthOverride=true, heightOverride=70, widthOverride=650}
   slot {class=BorderSlot, properties {padding=<ltrb 0 0 0 0>}}
  }
  HorizontalBox_3 {widgetClassPath=&HorizontalBox, parent=SizeBox_0, widget=HorizontalBox_3, slot {class=SizeBoxSlot}}
  MatchTimeTitleText {
   widgetClassPath=&CommonTextBlock
   parent=HorizontalBox_3
   widget=MatchTimeTitleText
   properties {
    scrollStyle=&TextScrollStyle-Base_C
    text=MATCH TIME
    font {
     bForceMonospaced=false
     bMaterialIsStencil=false
     fontMaterial=null
     fontObject=&Burbank
     letterSpacing=0
     monospacedWidth=1
     size=26
     skewAmount=0
     typefaceFontName=BurbankRegular-Black
     outlineSettings {
      bApplyOutlineToDropShadows=false
      bMiteredCorners=false
      bSeparateFillAlpha=false
      outlineMaterial=null
      outlineSize=0
      outlineColor=<rgba 0 0 0 1>
     }
    }
   }
   slot {class=HorizontalBoxSlot, properties {verticalAlignment=VAlign_Center, padding=<ltrb 30 5 0 0>}}
  }
  SizeBox_12 {
   widgetClassPath=&SizeBox
   parent=HorizontalBox_3
   widget=SizeBox_12
   properties {bOverride_WidthOverride=true, widthOverride=190}
   slot {class=HorizontalBoxSlot, properties {horizontalAlignment=HAlign_Right, size {sizeRule=Fill, value=1}}}
  }
  CommonBorder_2 {
   widgetClassPath=&CommonBorder
   parent=SizeBox_12
   widget=CommonBorder_2
   properties {
    horizontalAlignment=HAlign_Center
    verticalAlignment=VAlign_Center
    background {
     bIsDynamicallyLoaded=false
     drawAs=Image
     imageType=NoImage
     mirroring=NoMirror
     resourceName=None
     resourceObject=/Game/UI/Foundation/Materials/MI_MatchStatusBG_Yellow
     tiling=NoTile
     imageSize=<32 32>
     margin=<ltrb 0 0 0 0>
     outlineSettings {
      bUseBrushTransparency=false
      roundingType=HalfHeightRadius
      width=0
      color {colorUseRule=UseColor_Specified, specifiedColor=<rgba 0 0 0 0>}
      cornerRadii=<0 0 0 0>
     }
     tintColor {colorUseRule=UseColor_Specified, specifiedColor=<rgba 1 1 1 1>}
     uVRegion {bIsValid=false, max=<0 0>, min=<0 0>}
    }
    padding=<ltrb 0 5 0 0>
   }
   slot {class=SizeBoxSlot}
  }
  MatchTime {
   widgetClassPath=&CommonTextBlock
   bIsVariable=true
   parent=CommonBorder_2
   widget=MatchTime
   properties {
    scrollStyle=&TextScrollStyle-Base_C
    text=88:88
    colorAndOpacity {colorUseRule=UseColor_Specified, specifiedColor=<rgba 0.0091 0.1559 0.5583 1>}
    font {
     bForceMonospaced=false
     bMaterialIsStencil=false
     fontMaterial=null
     fontObject=&Burbank
     letterSpacing=0
     monospacedWidth=1
     size=26
     skewAmount=0
     typefaceFontName=BurbankRegular-Black
     outlineSettings {
      bApplyOutlineToDropShadows=false
      bMiteredCorners=false
      bSeparateFillAlpha=false
      outlineMaterial=null
      outlineSize=0
      outlineColor=<rgba 0 0 0 1>
     }
    }
   }
   slot {class=BorderSlot, properties {horizontalAlignment=HAlign_Center, verticalAlignment=VAlign_Center, padding=<ltrb 0 5 0 0>}}
  }
  CommonBorder_15 {
   widgetClassPath=&CommonBorder
   parent=VerticalBox_8
   widget=CommonBorder_15
   properties {style=&Border_MatchStatus_Blue_C, padding=<ltrb 0 0 0 0>}
   slot {class=VerticalBoxSlot, properties {padding=<ltrb 0 20 0 20>}}
  }
  SizeBox_11 {
   widgetClassPath=&SizeBox
   parent=CommonBorder_15
   widget=SizeBox_11
   properties {bOverride_HeightOverride=true, bOverride_WidthOverride=true, heightOverride=70, widthOverride=650}
   slot {class=BorderSlot, properties {padding=<ltrb 0 0 0 0>}}
  }
  HorizontalBox_7 {widgetClassPath=&HorizontalBox, parent=SizeBox_11, widget=HorizontalBox_7, slot {class=SizeBoxSlot}}
  PlayersRemainingTitleText {
   widgetClassPath=&CommonTextBlock
   parent=HorizontalBox_7
   widget=PlayersRemainingTitleText
   properties {
    scrollStyle=&TextScrollStyle-Base_C
    text=PLAYERS REMAINING
    font {
     bForceMonospaced=false
     bMaterialIsStencil=false
     fontMaterial=null
     fontObject=&Burbank
     letterSpacing=0
     monospacedWidth=1
     size=26
     skewAmount=0
     typefaceFontName=BurbankRegular-Black
     outlineSettings {
      bApplyOutlineToDropShadows=false
      bMiteredCorners=false
      bSeparateFillAlpha=false
      outlineMaterial=null
      outlineSize=0
      outlineColor=<rgba 0 0 0 1>
     }
    }
   }
   slot {class=HorizontalBoxSlot, properties {verticalAlignment=VAlign_Center, padding=<ltrb 30 5 0 0>}}
  }
  SizeBox_13 {
   widgetClassPath=&SizeBox
   parent=HorizontalBox_7
   widget=SizeBox_13
   properties {bOverride_WidthOverride=true, widthOverride=190}
   slot {class=HorizontalBoxSlot, properties {horizontalAlignment=HAlign_Right, size {sizeRule=Fill, value=1}}}
  }
  CommonBorder_16 {
   widgetClassPath=&CommonBorder
   parent=SizeBox_13
   widget=CommonBorder_16
   properties {
    horizontalAlignment=HAlign_Center
    verticalAlignment=VAlign_Center
    background {
     bIsDynamicallyLoaded=false
     drawAs=Image
     imageType=NoImage
     mirroring=NoMirror
     resourceName=None
     resourceObject=&MI_MatchStatusBG_Yellow
     tiling=NoTile
     imageSize=<32 32>
     margin=<ltrb 0 0 0 0>
     outlineSettings {
      bUseBrushTransparency=false
      roundingType=HalfHeightRadius
      width=0
      color {colorUseRule=UseColor_Specified, specifiedColor=<rgba 0 0 0 0>}
      cornerRadii=<0 0 0 0>
     }
     tintColor {colorUseRule=UseColor_Specified, specifiedColor=<rgba 1 1 1 1>}
     uVRegion {bIsValid=false, max=<0 0>, min=<0 0>}
    }
    padding=<ltrb 0 5 0 0>
   }
   slot {class=SizeBoxSlot}
  }
  PlayersRemaining {
   widgetClassPath=&/AthenaEventMatchStatusText
   bIsVariable=true
   parent=CommonBorder_16
   widget=PlayersRemaining
   properties {
    textToDisplay=888
    color {colorUseRule=UseColor_Specified, specifiedColor=<rgba 0.0091 0.1559 0.5583 1>}
    font {
     bForceMonospaced=false
     bMaterialIsStencil=false
     fontMaterial=null
     fontObject=&Burbank
     letterSpacing=0
     monospacedWidth=1
     size=26
     skewAmount=0
     typefaceFontName=BurbankRegular-Black
     outlineSettings {
      bApplyOutlineToDropShadows=false
      bMiteredCorners=false
      bSeparateFillAlpha=false
      outlineMaterial=null
      outlineSize=0
      outlineColor=<rgba 0 0 0 1>
     }
    }
   }
   slot {class=BorderSlot, properties {horizontalAlignment=HAlign_Center, verticalAlignment=VAlign_Center, padding=<ltrb 0 5 0 0>}}
  }
  CommonBorder_17 {
   widgetClassPath=&CommonBorder
   parent=VerticalBox_8
   widget=CommonBorder_17
   properties {style=&Border_MatchStatus_Blue_C, padding=<ltrb 0 0 0 0>}
   slot {class=VerticalBoxSlot, properties {padding=<ltrb 0 20 0 20>}}
  }
  SizeBox_14 {
   widgetClassPath=&SizeBox
   parent=CommonBorder_17
   widget=SizeBox_14
   properties {bOverride_HeightOverride=true, bOverride_WidthOverride=true, heightOverride=70, widthOverride=650}
   slot {class=BorderSlot, properties {padding=<ltrb 0 0 0 0>}}
  }
  HorizontalBox_8 {widgetClassPath=&HorizontalBox, parent=SizeBox_14, widget=HorizontalBox_8, slot {class=SizeBoxSlot}}
  TimeSinceLastElimTitleText {
   widgetClassPath=&CommonTextBlock
   parent=HorizontalBox_8
   widget=TimeSinceLastElimTitleText
   properties {
    scrollStyle=&TextScrollStyle-Base_C
    text=TIME SINCE LAST ELIM.
    font {
     bForceMonospaced=false
     bMaterialIsStencil=false
     fontMaterial=null
     fontObject=&Burbank
     letterSpacing=0
     monospacedWidth=1
     size=26
     skewAmount=0
     typefaceFontName=BurbankRegular-Black
     outlineSettings {
      bApplyOutlineToDropShadows=false
      bMiteredCorners=false
      bSeparateFillAlpha=false
      outlineMaterial=null
      outlineSize=0
      outlineColor=<rgba 0 0 0 1>
     }
    }
   }
   slot {class=HorizontalBoxSlot, properties {verticalAlignment=VAlign_Center, padding=<ltrb 30 5 0 0>}}
  }
  SizeBox_15 {
   widgetClassPath=&SizeBox
   parent=HorizontalBox_8
   widget=SizeBox_15
   properties {bOverride_WidthOverride=true, widthOverride=190}
   slot {class=HorizontalBoxSlot, properties {horizontalAlignment=HAlign_Right, size {sizeRule=Fill, value=1}}}
  }
  CommonBorder_18 {
   widgetClassPath=&CommonBorder
   parent=SizeBox_15
   widget=CommonBorder_18
   properties {
    horizontalAlignment=HAlign_Center
    verticalAlignment=VAlign_Center
    background {
     bIsDynamicallyLoaded=false
     drawAs=Image
     imageType=NoImage
     mirroring=NoMirror
     resourceName=None
     resourceObject=&MI_MatchStatusBG_Yellow
     tiling=NoTile
     imageSize=<32 32>
     margin=<ltrb 0 0 0 0>
     outlineSettings {
      bUseBrushTransparency=false
      roundingType=HalfHeightRadius
      width=0
      color {colorUseRule=UseColor_Specified, specifiedColor=<rgba 0 0 0 0>}
      cornerRadii=<0 0 0 0>
     }
     tintColor {colorUseRule=UseColor_Specified, specifiedColor=<rgba 1 1 1 1>}
     uVRegion {bIsValid=false, max=<0 0>, min=<0 0>}
    }
    padding=<ltrb 0 5 0 0>
   }
   slot {class=SizeBoxSlot}
  }
  TimeSinceLastKill {
   widgetClassPath=&AthenaEventMatchStatusText
   bIsVariable=true
   parent=CommonBorder_18
   widget=TimeSinceLastKill
   properties {
    textToDisplay=88:88
    color {colorUseRule=UseColor_Specified, specifiedColor=<rgba 0.0091 0.1559 0.5583 1>}
    font {
     bForceMonospaced=false
     bMaterialIsStencil=false
     fontMaterial=null
     fontObject=&Burbank
     letterSpacing=0
     monospacedWidth=1
     size=26
     skewAmount=0
     typefaceFontName=BurbankRegular-Black
     outlineSettings {
      bApplyOutlineToDropShadows=false
      bMiteredCorners=false
      bSeparateFillAlpha=false
      outlineMaterial=null
      outlineSize=0
      outlineColor=<rgba 0 0 0 1>
     }
    }
   }
   slot {class=BorderSlot, properties {horizontalAlignment=HAlign_Center, verticalAlignment=VAlign_Center, padding=<ltrb 0 5 0 0>}}
  }
  VerticalBox_2 {
   widgetClassPath=&VerticalBox
   parent=HorizontalBox_0
   widget=VerticalBox_2
   slot {class=HorizontalBoxSlot, properties {horizontalAlignment=HAlign_Center, size {sizeRule=Fill, value=1}}}
  }
  HorizontalBox_11 {widgetClassPath=&HorizontalBox, parent=VerticalBox_2, widget=HorizontalBox_11, slot {class=VerticalBoxSlot}}
  VerticalBox_4 {widgetClassPath=&VerticalBox, parent=HorizontalBox_11, widget=VerticalBox_4, slot {class=HorizontalBoxSlot}}
  CommonBorder_21 {
   widgetClassPath=&CommonBorder
   parent=VerticalBox_4
   widget=CommonBorder_21
   properties {style=&Border_MatchStatus_Blue_C, padding=<ltrb 0 0 0 0>}
   slot {class=VerticalBoxSlot, properties {padding=<ltrb 0 20 0 20>}}
  }
  SizeBox_18 {
   widgetClassPath=&SizeBox
   parent=CommonBorder_21
   widget=SizeBox_18
   properties {bOverride_HeightOverride=true, bOverride_WidthOverride=true, heightOverride=70, minDesiredHeight=70, widthOverride=650}
   slot {class=BorderSlot, properties {padding=<ltrb 0 0 0 0>}}
  }
  HorizontalBox_10 {widgetClassPath=&HorizontalBox, parent=SizeBox_18, widget=HorizontalBox_10, slot {class=SizeBoxSlot}}
  GamePhaseTitleText {
   widgetClassPath=&CommonTextBlock
   bIsVariable=true
   parent=HorizontalBox_10
   widget=GamePhaseTitleText
   properties {
    scrollStyle=&TextScrollStyle-Base_C
    text=GAME PHASE
    font {
     bForceMonospaced=false
     bMaterialIsStencil=false
     fontMaterial=null
     fontObject=&Burbank
     letterSpacing=0
     monospacedWidth=1
     size=26
     skewAmount=0
     typefaceFontName=BurbankRegular-Black
     outlineSettings {
      bApplyOutlineToDropShadows=false
      bMiteredCorners=false
      bSeparateFillAlpha=false
      outlineMaterial=null
      outlineSize=0
      outlineColor=<rgba 0 0 0 1>
     }
    }
   }
   slot {class=HorizontalBoxSlot, properties {verticalAlignment=VAlign_Center, padding=<ltrb 30 5 0 0>}}
  }
  SizeBox_19 {
   widgetClassPath=&SizeBox
   parent=HorizontalBox_10
   widget=SizeBox_19
   properties {bOverride_WidthOverride=true, widthOverride=190}
   slot {class=HorizontalBoxSlot, properties {horizontalAlignment=HAlign_Right, size {sizeRule=Fill, value=1}}}
  }
  CommonBorder_22 {
   widgetClassPath=&CommonBorder
   parent=SizeBox_19
   widget=CommonBorder_22
   properties {
    horizontalAlignment=HAlign_Center
    verticalAlignment=VAlign_Center
    background {
     bIsDynamicallyLoaded=false
     drawAs=Image
     imageType=NoImage
     mirroring=NoMirror
     resourceName=None
     resourceObject=&MI_MatchStatusBG_Yellow
     tiling=NoTile
     imageSize=<32 32>
     margin=<ltrb 0 0 0 0>
     outlineSettings {
      bUseBrushTransparency=false
      roundingType=HalfHeightRadius
      width=0
      color {colorUseRule=UseColor_Specified, specifiedColor=<rgba 0 0 0 0>}
      cornerRadii=<0 0 0 0>
     }
     tintColor {colorUseRule=UseColor_Specified, specifiedColor=<rgba 1 1 1 1>}
     uVRegion {bIsValid=false, max=<0 0>, min=<0 0>}
    }
    padding=<ltrb 0 5 0 0>
   }
   slot {class=SizeBoxSlot}
  }
  GamePhaseTime {
   widgetClassPath=&CommonTextBlock
   bIsVariable=true
   parent=CommonBorder_22
   widget=GamePhaseTime
   properties {
    scrollStyle=&TextScrollStyle-Base_C
    text=88:88
    colorAndOpacity {colorUseRule=UseColor_Specified, specifiedColor=<rgba 0.0086 0.1559 0.5583 1>}
    font {
     bForceMonospaced=false
     bMaterialIsStencil=false
     fontMaterial=null
     fontObject=&Burbank
     letterSpacing=0
     monospacedWidth=1
     size=26
     skewAmount=0
     typefaceFontName=BurbankRegular-Black
     outlineSettings {
      bApplyOutlineToDropShadows=false
      bMiteredCorners=false
      bSeparateFillAlpha=false
      outlineMaterial=null
      outlineSize=0
      outlineColor=<rgba 0 0 0 1>
     }
    }
   }
   slot {class=BorderSlot, properties {horizontalAlignment=HAlign_Center, verticalAlignment=VAlign_Center, padding=<ltrb 0 5 0 0>}}
  }
  CommonBorder_3 {
   widgetClassPath=&CommonBorder
   parent=VerticalBox_4
   widget=CommonBorder_3
   properties {style=&Border_MatchStatus_Blue_C, padding=<ltrb 0 0 0 0>}
   slot {class=VerticalBoxSlot, properties {padding=<ltrb 0 20 0 20>}}
  }
  SizeBox_1 {
   widgetClassPath=&SizeBox
   parent=CommonBorder_3
   widget=SizeBox_1
   properties {bOverride_HeightOverride=true, bOverride_WidthOverride=true, heightOverride=70, minDesiredHeight=70, widthOverride=650}
   slot {class=BorderSlot, properties {padding=<ltrb 0 0 0 0>}}
  }
  HorizontalBox_4 {widgetClassPath=&HorizontalBox, parent=SizeBox_1, widget=HorizontalBox_4, slot {class=SizeBoxSlot}}
  StormPhaseTitleText {
   widgetClassPath=&CommonTextBlock
   parent=HorizontalBox_4
   widget=StormPhaseTitleText
   properties {
    scrollStyle=&TextScrollStyle-Base_C
    text=STORM PHASE
    font {
     bForceMonospaced=false
     bMaterialIsStencil=false
     fontMaterial=null
     fontObject=&Burbank
     letterSpacing=0
     monospacedWidth=1
     size=26
     skewAmount=0
     typefaceFontName=BurbankRegular-Black
     outlineSettings {
      bApplyOutlineToDropShadows=false
      bMiteredCorners=false
      bSeparateFillAlpha=false
      outlineMaterial=null
      outlineSize=0
      outlineColor=<rgba 0 0 0 1>
     }
    }
   }
   slot {class=HorizontalBoxSlot, properties {verticalAlignment=VAlign_Center, padding=<ltrb 30 5 0 0>}}
  }
  SizeBox_2 {
   widgetClassPath=&SizeBox
   parent=HorizontalBox_4
   widget=SizeBox_2
   properties {bOverride_WidthOverride=true, widthOverride=190}
   slot {class=HorizontalBoxSlot, properties {horizontalAlignment=HAlign_Right, size {sizeRule=Fill, value=1}}}
  }
  CommonBorder_13 {
   widgetClassPath=&CommonBorder
   parent=SizeBox_2
   widget=CommonBorder_13
   properties {
    horizontalAlignment=HAlign_Center
    verticalAlignment=VAlign_Center
    background {
     bIsDynamicallyLoaded=false
     drawAs=Image
     imageType=NoImage
     mirroring=NoMirror
     resourceName=None
     resourceObject=&MI_MatchStatusBG_Yellow
     tiling=NoTile
     imageSize=<32 32>
     margin=<ltrb 0 0 0 0>
     outlineSettings {
      bUseBrushTransparency=false
      roundingType=HalfHeightRadius
      width=0
      color {colorUseRule=UseColor_Specified, specifiedColor=<rgba 0 0 0 0>}
      cornerRadii=<0 0 0 0>
     }
     tintColor {colorUseRule=UseColor_Specified, specifiedColor=<rgba 1 1 1 1>}
     uVRegion {bIsValid=false, max=<0 0>, min=<0 0>}
    }
    padding=<ltrb 0 5 0 0>
   }
   slot {class=SizeBoxSlot}
  }
  StormPhase {
   widgetClassPath=&AthenaEventMatchStatusText
   bIsVariable=true
   parent=CommonBorder_13
   widget=StormPhase
   properties {
    textToDisplay=888
    color {colorUseRule=UseColor_Specified, specifiedColor=<rgba 0.0091 0.1559 0.5583 1>}
    font {
     bForceMonospaced=false
     bMaterialIsStencil=false
     fontMaterial=null
     fontObject=&Burbank
     letterSpacing=0
     monospacedWidth=1
     size=26
     skewAmount=0
     typefaceFontName=BurbankRegular-Black
     outlineSettings {
      bApplyOutlineToDropShadows=false
      bMiteredCorners=false
      bSeparateFillAlpha=false
      outlineMaterial=null
      outlineSize=0
      outlineColor=<rgba 0 0 0 1>
     }
    }
   }
   slot {class=BorderSlot, properties {horizontalAlignment=HAlign_Center, verticalAlignment=VAlign_Center, padding=<ltrb 0 5 0 0>}}
  }
  CommonBorder_19 {
   widgetClassPath=&CommonBorder
   parent=VerticalBox_4
   widget=CommonBorder_19
   properties {style=&Border_MatchStatus_Blue_C, padding=<ltrb 0 0 0 0>}
   slot {class=VerticalBoxSlot, properties {padding=<ltrb 0 20 0 20>}}
  }
  SizeBox_16 {
   widgetClassPath=&SizeBox
   parent=CommonBorder_19
   widget=SizeBox_16
   properties {bOverride_HeightOverride=true, bOverride_WidthOverride=true, heightOverride=70, minDesiredHeight=70, widthOverride=650}
   slot {class=BorderSlot, properties {padding=<ltrb 0 0 0 0>}}
  }
  HorizontalBox_9 {widgetClassPath=&HorizontalBox, parent=SizeBox_16, widget=HorizontalBox_9, slot {class=SizeBoxSlot}}
  StormRadiusTitleText {
   widgetClassPath=&CommonTextBlock
   parent=HorizontalBox_9
   widget=StormRadiusTitleText
   properties {
    text=STORM RADIUS
    font {
     bForceMonospaced=false
     bMaterialIsStencil=false
     fontMaterial=null
     fontObject=&Burbank
     letterSpacing=0
     monospacedWidth=1
     size=26
     skewAmount=0
     typefaceFontName=BurbankRegular-Black
     outlineSettings {
      bApplyOutlineToDropShadows=false
      bMiteredCorners=false
      bSeparateFillAlpha=false
      outlineMaterial=null
      outlineSize=0
      outlineColor=<rgba 0 0 0 1>
     }
    }
   }
   slot {class=HorizontalBoxSlot, properties {verticalAlignment=VAlign_Center, padding=<ltrb 30 5 0 0>}}
  }
  SizeBox_17 {
   widgetClassPath=&SizeBox
   parent=HorizontalBox_9
   widget=SizeBox_17
   properties {bOverride_WidthOverride=true, widthOverride=190}
   slot {class=HorizontalBoxSlot, properties {horizontalAlignment=HAlign_Right, size {sizeRule=Fill, value=1}}}
  }
  CommonBorder_20 {
   widgetClassPath=&CommonBorder
   parent=SizeBox_17
   widget=CommonBorder_20
   properties {
    horizontalAlignment=HAlign_Center
    verticalAlignment=VAlign_Center
    background {
     bIsDynamicallyLoaded=false
     drawAs=Image
     imageType=NoImage
     mirroring=NoMirror
     resourceName=None
     resourceObject=&MI_MatchStatusBG_Yellow
     tiling=NoTile
     imageSize=<32 32>
     margin=<ltrb 0 0 0 0>
     outlineSettings {
      bUseBrushTransparency=false
      roundingType=HalfHeightRadius
      width=0
      color {colorUseRule=UseColor_Specified, specifiedColor=<rgba 0 0 0 0>}
      cornerRadii=<0 0 0 0>
     }
     tintColor {colorUseRule=UseColor_Specified, specifiedColor=<rgba 1 1 1 1>}
     uVRegion {bIsValid=false, max=<0 0>, min=<0 0>}
    }
    padding=<ltrb 0 5 0 0>
   }
   slot {class=SizeBoxSlot}
  }
  CircleRadius {
   widgetClassPath=&CommonTextBlock
   bIsVariable=true
   parent=CommonBorder_20
   widget=CircleRadius
   properties {
    scrollStyle=&TextScrollStyle-Base_C
    text=888 m
    colorAndOpacity {colorUseRule=UseColor_Specified, specifiedColor=<rgba 0.0086 0.1559 0.5583 1>}
    font {
     bForceMonospaced=false
     bMaterialIsStencil=false
     fontMaterial=null
     fontObject=&Burbank
     letterSpacing=0
     monospacedWidth=1
     size=26
     skewAmount=0
     typefaceFontName=BurbankRegular-Black
     outlineSettings {
      bApplyOutlineToDropShadows=false
      bMiteredCorners=false
      bSeparateFillAlpha=false
      outlineMaterial=null
      outlineSize=0
      outlineColor=<rgba 0 0 0 1>
     }
    }
   }
   slot {class=BorderSlot, properties {horizontalAlignment=HAlign_Center, verticalAlignment=VAlign_Center, padding=<ltrb 0 5 0 0>}}
  }
  CommonBorder_23 {
   widgetClassPath=&CommonBorder
   parent=VerticalBox_4
   widget=CommonBorder_23
   properties {style=&Border_MatchStatus_Blue_C, visibility=Collapsed, padding=<ltrb 0 0 0 0>}
   slot {class=VerticalBoxSlot, properties {padding=<ltrb 0 6 0 6>}}
  }
  SizeBox_21 {
   widgetClassPath=&SizeBox
   parent=CommonBorder_23
   widget=SizeBox_21
   properties {bOverride_HeightOverride=true, bOverride_WidthOverride=true, heightOverride=70, widthOverride=620}
   slot {class=BorderSlot, properties {padding=<ltrb 0 0 0 0>}}
  }
  HorizontalBox_12 {widgetClassPath=&HorizontalBox, parent=SizeBox_21, widget=HorizontalBox_12, slot {class=SizeBoxSlot}}
  ShotsFiredTitleText {
   widgetClassPath=&CommonTextBlock
   parent=HorizontalBox_12
   widget=ShotsFiredTitleText
   properties {
    text=SHOTS FIRED
    font {
     bForceMonospaced=false
     bMaterialIsStencil=false
     fontMaterial=null
     fontObject=&Burbank
     letterSpacing=0
     monospacedWidth=1
     size=26
     skewAmount=0
     typefaceFontName=BurbankRegular-Black
     outlineSettings {
      bApplyOutlineToDropShadows=false
      bMiteredCorners=false
      bSeparateFillAlpha=false
      outlineMaterial=null
      outlineSize=0
      outlineColor=<rgba 0 0 0 1>
     }
    }
   }
   slot {class=HorizontalBoxSlot, properties {verticalAlignment=VAlign_Center, padding=<ltrb 30 5 0 0>}}
  }
  SizeBox_22 {
   widgetClassPath=&SizeBox
   parent=HorizontalBox_12
   widget=SizeBox_22
   properties {bOverride_WidthOverride=true, widthOverride=190}
   slot {class=HorizontalBoxSlot, properties {horizontalAlignment=HAlign_Right, size {sizeRule=Fill, value=1}}}
  }
  CommonBorder_24 {
   widgetClassPath=&CommonBorder
   parent=SizeBox_22
   widget=CommonBorder_24
   properties {
    horizontalAlignment=HAlign_Center
    verticalAlignment=VAlign_Center
    background {
     bIsDynamicallyLoaded=false
     drawAs=Image
     imageType=NoImage
     mirroring=NoMirror
     resourceName=None
     resourceObject=&MI_MatchStatusBG_Yellow
     tiling=NoTile
     imageSize=<32 32>
     margin=<ltrb 0 0 0 0>
     outlineSettings {
      bUseBrushTransparency=false
      roundingType=HalfHeightRadius
      width=0
      color {colorUseRule=UseColor_Specified, specifiedColor=<rgba 0 0 0 0>}
      cornerRadii=<0 0 0 0>
     }
     tintColor {colorUseRule=UseColor_Specified, specifiedColor=<rgba 1 1 1 1>}
     uVRegion {bIsValid=false, max=<0 0>, min=<0 0>}
    }
    padding=<ltrb 0 5 0 0>
   }
   slot {class=SizeBoxSlot}
  }
  ShotsFired {
   widgetClassPath=&CommonTextBlock
   bIsVariable=true
   parent=CommonBorder_24
   widget=ShotsFired
   properties {
    text=8888888
    colorAndOpacity {colorUseRule=UseColor_Specified, specifiedColor=<rgba 0.0086 0.1559 0.5583 1>}
    font {
     bForceMonospaced=false
     bMaterialIsStencil=false
     fontMaterial=null
     fontObject=&Burbank
     letterSpacing=0
     monospacedWidth=1
     size=26
     skewAmount=0
     typefaceFontName=BurbankRegular-Black
     outlineSettings {
      bApplyOutlineToDropShadows=false
      bMiteredCorners=false
      bSeparateFillAlpha=false
      outlineMaterial=null
      outlineSize=0
      outlineColor=<rgba 0 0 0 1>
     }
    }
   }
   slot {class=BorderSlot, properties {horizontalAlignment=HAlign_Center, verticalAlignment=VAlign_Center, padding=<ltrb 0 5 0 0>}}
  }
  CommonBorder_25 {
   widgetClassPath=&CommonBorder
   parent=VerticalBox_4
   widget=CommonBorder_25
   properties {style=&Border_MatchStatus_Blue_C, visibility=Collapsed, padding=<ltrb 0 0 0 0>}
   slot {class=VerticalBoxSlot, properties {padding=<ltrb 0 6 0 6>}}
  }
  SizeBox_23 {
   widgetClassPath=&SizeBox
   parent=CommonBorder_25
   widget=SizeBox_23
   properties {bOverride_HeightOverride=true, bOverride_WidthOverride=true, heightOverride=70, widthOverride=620}
   slot {class=BorderSlot, properties {padding=<ltrb 0 0 0 0>}}
  }
  HorizontalBox_13 {widgetClassPath=&HorizontalBox, parent=SizeBox_23, widget=HorizontalBox_13, slot {class=SizeBoxSlot}}
  PiecesBuiltTitleText {
   widgetClassPath=&CommonTextBlock
   parent=HorizontalBox_13
   widget=PiecesBuiltTitleText
   properties {
    text=PIECES BUILT
    font {
     bForceMonospaced=false
     bMaterialIsStencil=false
     fontMaterial=null
     fontObject=&Burbank
     letterSpacing=0
     monospacedWidth=1
     size=26
     skewAmount=0
     typefaceFontName=BurbankRegular-Black
     outlineSettings {
      bApplyOutlineToDropShadows=false
      bMiteredCorners=false
      bSeparateFillAlpha=false
      outlineMaterial=null
      outlineSize=0
      outlineColor=<rgba 0 0 0 1>
     }
    }
   }
   slot {class=HorizontalBoxSlot, properties {verticalAlignment=VAlign_Center, padding=<ltrb 30 5 0 0>}}
  }
  SizeBox_24 {
   widgetClassPath=&SizeBox
   parent=HorizontalBox_13
   widget=SizeBox_24
   properties {bOverride_WidthOverride=true, widthOverride=190}
   slot {class=HorizontalBoxSlot, properties {horizontalAlignment=HAlign_Right, size {sizeRule=Fill, value=1}}}
  }
  CommonBorder_26 {
   widgetClassPath=&CommonBorder
   parent=SizeBox_24
   widget=CommonBorder_26
   properties {
    horizontalAlignment=HAlign_Center
    verticalAlignment=VAlign_Center
    background {
     bIsDynamicallyLoaded=false
     drawAs=Image
     imageType=NoImage
     mirroring=NoMirror
     resourceName=None
     resourceObject=&MI_MatchStatusBG_Yellow
     tiling=NoTile
     imageSize=<32 32>
     margin=<ltrb 0 0 0 0>
     outlineSettings {
      bUseBrushTransparency=false
      roundingType=HalfHeightRadius
      width=0
      color {colorUseRule=UseColor_Specified, specifiedColor=<rgba 0 0 0 0>}
      cornerRadii=<0 0 0 0>
     }
     tintColor {colorUseRule=UseColor_Specified, specifiedColor=<rgba 1 1 1 1>}
     uVRegion {bIsValid=false, max=<0 0>, min=<0 0>}
    }
    padding=<ltrb 0 5 0 0>
   }
   slot {class=SizeBoxSlot}
  }
  PiecesBuilt {
   widgetClassPath=&CommonTextBlock
   bIsVariable=true
   parent=CommonBorder_26
   widget=PiecesBuilt
   properties {
    text=8888888
    colorAndOpacity {colorUseRule=UseColor_Specified, specifiedColor=<rgba 0.0086 0.1559 0.5583 1>}
    font {
     bForceMonospaced=false
     bMaterialIsStencil=false
     fontMaterial=null
     fontObject=&Burbank
     letterSpacing=0
     monospacedWidth=1
     size=26
     skewAmount=0
     typefaceFontName=BurbankRegular-Black
     outlineSettings {
      bApplyOutlineToDropShadows=false
      bMiteredCorners=false
      bSeparateFillAlpha=false
      outlineMaterial=null
      outlineSize=0
      outlineColor=<rgba 0 0 0 1>
     }
    }
   }
   slot {class=BorderSlot, properties {horizontalAlignment=HAlign_Center, verticalAlignment=VAlign_Center, padding=<ltrb 0 5 0 0>}}
  }
  CommonBorder_27 {
   widgetClassPath=&CommonBorder
   parent=VerticalBox_4
   widget=CommonBorder_27
   properties {style=&Border_MatchStatus_Blue_C, visibility=Collapsed, padding=<ltrb 0 0 0 0>}
   slot {class=VerticalBoxSlot, properties {padding=<ltrb 0 6 0 6>}}
  }
  SizeBox_25 {
   widgetClassPath=&SizeBox
   parent=CommonBorder_27
   widget=SizeBox_25
   properties {bOverride_HeightOverride=true, bOverride_WidthOverride=true, heightOverride=70, widthOverride=620}
   slot {class=BorderSlot, properties {padding=<ltrb 0 0 0 0>}}
  }
  HorizontalBox_14 {widgetClassPath=&HorizontalBox, parent=SizeBox_25, widget=HorizontalBox_14, slot {class=SizeBoxSlot}}
  PiecesDestroyedTitleText {
   widgetClassPath=&CommonTextBlock
   parent=HorizontalBox_14
   widget=PiecesDestroyedTitleText
   properties {
    text=PIECES DESTROYED
    font {
     bForceMonospaced=false
     bMaterialIsStencil=false
     fontMaterial=null
     fontObject=&Burbank
     letterSpacing=0
     monospacedWidth=1
     size=26
     skewAmount=0
     typefaceFontName=BurbankRegular-Black
     outlineSettings {
      bApplyOutlineToDropShadows=false
      bMiteredCorners=false
      bSeparateFillAlpha=false
      outlineMaterial=null
      outlineSize=0
      outlineColor=<rgba 0 0 0 1>
     }
    }
   }
   slot {class=HorizontalBoxSlot, properties {verticalAlignment=VAlign_Center, padding=<ltrb 30 5 0 0>}}
  }
  SizeBox_26 {
   widgetClassPath=&SizeBox
   parent=HorizontalBox_14
   widget=SizeBox_26
   properties {bOverride_WidthOverride=true, widthOverride=190}
   slot {class=HorizontalBoxSlot, properties {horizontalAlignment=HAlign_Right, size {sizeRule=Fill, value=1}}}
  }
  CommonBorder_28 {
   widgetClassPath=&CommonBorder
   parent=SizeBox_26
   widget=CommonBorder_28
   properties {
    horizontalAlignment=HAlign_Center
    verticalAlignment=VAlign_Center
    background {
     bIsDynamicallyLoaded=false
     drawAs=Image
     imageType=NoImage
     mirroring=NoMirror
     resourceName=None
     resourceObject=&MI_MatchStatusBG_Yellow
     tiling=NoTile
     imageSize=<32 32>
     margin=<ltrb 0 0 0 0>
     outlineSettings {
      bUseBrushTransparency=false
      roundingType=HalfHeightRadius
      width=0
      color {colorUseRule=UseColor_Specified, specifiedColor=<rgba 0 0 0 0>}
      cornerRadii=<0 0 0 0>
     }
     tintColor {colorUseRule=UseColor_Specified, specifiedColor=<rgba 1 1 1 1>}
     uVRegion {bIsValid=false, max=<0 0>, min=<0 0>}
    }
    padding=<ltrb 0 5 0 0>
   }
   slot {class=SizeBoxSlot}
  }
  PiecesDestroyed {
   widgetClassPath=&CommonTextBlock
   bIsVariable=true
   parent=CommonBorder_28
   widget=PiecesDestroyed
   properties {
    text=8888888
    colorAndOpacity {colorUseRule=UseColor_Specified, specifiedColor=<rgba 0.0086 0.1559 0.5583 1>}
    font {
     bForceMonospaced=false
     bMaterialIsStencil=false
     fontMaterial=null
     fontObject=&Burbank
     letterSpacing=0
     monospacedWidth=1
     size=26
     skewAmount=0
     typefaceFontName=BurbankRegular-Black
     outlineSettings {
      bApplyOutlineToDropShadows=false
      bMiteredCorners=false
      bSeparateFillAlpha=false
      outlineMaterial=null
      outlineSize=0
      outlineColor=<rgba 0 0 0 1>
     }
    }
   }
   slot {class=BorderSlot, properties {horizontalAlignment=HAlign_Center, verticalAlignment=VAlign_Center, padding=<ltrb 0 5 0 0>}}
  }
  CommonBorder_29 {
   widgetClassPath=&CommonBorder
   parent=VerticalBox_4
   widget=CommonBorder_29
   properties {style=&Border_MatchStatus_Blue_C, visibility=Collapsed, padding=<ltrb 0 0 0 0>}
   slot {class=VerticalBoxSlot, properties {padding=<ltrb 0 6 0 6>}}
  }
  SizeBox_27 {
   widgetClassPath=&SizeBox
   parent=CommonBorder_29
   widget=SizeBox_27
   properties {bOverride_HeightOverride=true, bOverride_WidthOverride=true, heightOverride=70, widthOverride=620}
   slot {class=BorderSlot, properties {padding=<ltrb 0 0 0 0>}}
  }
  HorizontalBox_15 {widgetClassPath=&HorizontalBox, parent=SizeBox_27, widget=HorizontalBox_15, slot {class=SizeBoxSlot}}
  ChestsLootedTitleText {
   widgetClassPath=&CommonTextBlock
   parent=HorizontalBox_15
   widget=ChestsLootedTitleText
   properties {
    text=CHESTS LOOTED
    font {
     bForceMonospaced=false
     bMaterialIsStencil=false
     fontMaterial=null
     fontObject=&Burbank
     letterSpacing=0
     monospacedWidth=1
     size=26
     skewAmount=0
     typefaceFontName=BurbankRegular-Black
     outlineSettings {
      bApplyOutlineToDropShadows=false
      bMiteredCorners=false
      bSeparateFillAlpha=false
      outlineMaterial=null
      outlineSize=0
      outlineColor=<rgba 0 0 0 1>
     }
    }
   }
   slot {class=HorizontalBoxSlot, properties {verticalAlignment=VAlign_Center, padding=<ltrb 30 5 0 0>}}
  }
  SizeBox_28 {
   widgetClassPath=&SizeBox
   parent=HorizontalBox_15
   widget=SizeBox_28
   properties {bOverride_WidthOverride=true, widthOverride=190}
   slot {class=HorizontalBoxSlot, properties {horizontalAlignment=HAlign_Right, size {sizeRule=Fill, value=1}}}
  }
  CommonBorder_30 {
   widgetClassPath=&CommonBorder
   parent=SizeBox_28
   widget=CommonBorder_30
   properties {
    horizontalAlignment=HAlign_Center
    verticalAlignment=VAlign_Center
    background {
     bIsDynamicallyLoaded=false
     drawAs=Image
     imageType=NoImage
     mirroring=NoMirror
     resourceName=None
     resourceObject=&MI_MatchStatusBG_Yellow
     tiling=NoTile
     imageSize=<32 32>
     margin=<ltrb 0 0 0 0>
     outlineSettings {
      bUseBrushTransparency=false
      roundingType=HalfHeightRadius
      width=0
      color {colorUseRule=UseColor_Specified, specifiedColor=<rgba 0 0 0 0>}
      cornerRadii=<0 0 0 0>
     }
     tintColor {colorUseRule=UseColor_Specified, specifiedColor=<rgba 1 1 1 1>}
     uVRegion {bIsValid=false, max=<0 0>, min=<0 0>}
    }
    padding=<ltrb 0 5 0 0>
   }
   slot {class=SizeBoxSlot}
  }
  ChestsLooted {
   widgetClassPath=&CommonTextBlock
   bIsVariable=true
   parent=CommonBorder_30
   widget=ChestsLooted
   properties {
    text=8888888
    colorAndOpacity {colorUseRule=UseColor_Specified, specifiedColor=<rgba 0.0086 0.1559 0.5583 1>}
    font {
     bForceMonospaced=false
     bMaterialIsStencil=false
     fontMaterial=null
     fontObject=&Burbank
     letterSpacing=0
     monospacedWidth=1
     size=26
     skewAmount=0
     typefaceFontName=BurbankRegular-Black
     outlineSettings {
      bApplyOutlineToDropShadows=false
      bMiteredCorners=false
      bSeparateFillAlpha=false
      outlineMaterial=null
      outlineSize=0
      outlineColor=<rgba 0 0 0 1>
     }
    }
   }
   slot {class=BorderSlot, properties {horizontalAlignment=HAlign_Center, verticalAlignment=VAlign_Center, padding=<ltrb 0 5 0 0>}}
  }
  CommonBorder_31 {
   widgetClassPath=&CommonBorder
   parent=VerticalBox_4
   widget=CommonBorder_31
   properties {style=&Border_MatchStatus_Blue_C, visibility=Collapsed, padding=<ltrb 0 0 0 0>}
   slot {class=VerticalBoxSlot, properties {padding=<ltrb 0 6 0 6>}}
  }
  SizeBox_29 {
   widgetClassPath=&SizeBox
   parent=CommonBorder_31
   widget=SizeBox_29
   properties {bOverride_HeightOverride=true, bOverride_WidthOverride=true, heightOverride=70, widthOverride=620}
   slot {class=BorderSlot, properties {padding=<ltrb 0 0 0 0>}}
  }
  HorizontalBox_16 {widgetClassPath=&HorizontalBox, parent=SizeBox_29, widget=HorizontalBox_16, slot {class=SizeBoxSlot}}
  SupplyDropsLootedTitleText {
   widgetClassPath=&CommonTextBlock
   parent=HorizontalBox_16
   widget=SupplyDropsLootedTitleText
   properties {
    text=SUPPLY DROPS LOOTED
    font {
     bForceMonospaced=false
     bMaterialIsStencil=false
     fontMaterial=null
     fontObject=&Burbank
     letterSpacing=0
     monospacedWidth=1
     size=26
     skewAmount=0
     typefaceFontName=BurbankRegular-Black
     outlineSettings {
      bApplyOutlineToDropShadows=false
      bMiteredCorners=false
      bSeparateFillAlpha=false
      outlineMaterial=null
      outlineSize=0
      outlineColor=<rgba 0 0 0 1>
     }
    }
   }
   slot {class=HorizontalBoxSlot, properties {verticalAlignment=VAlign_Center, padding=<ltrb 30 5 0 0>}}
  }
  SizeBox_30 {
   widgetClassPath=&SizeBox
   parent=HorizontalBox_16
   widget=SizeBox_30
   properties {bOverride_WidthOverride=true, widthOverride=190}
   slot {class=HorizontalBoxSlot, properties {horizontalAlignment=HAlign_Right, size {sizeRule=Fill, value=1}}}
  }
  CommonBorder_32 {
   widgetClassPath=&CommonBorder
   parent=SizeBox_30
   widget=CommonBorder_32
   properties {
    horizontalAlignment=HAlign_Center
    verticalAlignment=VAlign_Center
    background {
     bIsDynamicallyLoaded=false
     drawAs=Image
     imageType=NoImage
     mirroring=NoMirror
     resourceName=None
     resourceObject=&MI_MatchStatusBG_Yellow
     tiling=NoTile
     imageSize=<32 32>
     margin=<ltrb 0 0 0 0>
     outlineSettings {
      bUseBrushTransparency=false
      roundingType=HalfHeightRadius
      width=0
      color {colorUseRule=UseColor_Specified, specifiedColor=<rgba 0 0 0 0>}
      cornerRadii=<0 0 0 0>
     }
     tintColor {colorUseRule=UseColor_Specified, specifiedColor=<rgba 1 1 1 1>}
     uVRegion {bIsValid=false, max=<0 0>, min=<0 0>}
    }
    padding=<ltrb 0 5 0 0>
   }
   slot {class=SizeBoxSlot}
  }
  SupplyDropsLooted {
   widgetClassPath=&CommonTextBlock
   bIsVariable=true
   parent=CommonBorder_32
   widget=SupplyDropsLooted
   properties {
    text=8888888
    colorAndOpacity {colorUseRule=UseColor_Specified, specifiedColor=<rgba 0.0086 0.1559 0.5583 1>}
    font {
     bForceMonospaced=false
     bMaterialIsStencil=false
     fontMaterial=null
     fontObject=&Burbank
     letterSpacing=0
     monospacedWidth=1
     size=26
     skewAmount=0
     typefaceFontName=BurbankRegular-Black
     outlineSettings {
      bApplyOutlineToDropShadows=false
      bMiteredCorners=false
      bSeparateFillAlpha=false
      outlineMaterial=null
      outlineSize=0
      outlineColor=<rgba 0 0 0 1>
     }
    }
   }
   slot {class=BorderSlot, properties {horizontalAlignment=HAlign_Center, verticalAlignment=VAlign_Center, padding=<ltrb 0 5 0 0>}}
  }
  HorizontalBox_1 {
   widgetClassPath=&HorizontalBox
   parent=VerticalBox_0
   widget=HorizontalBox_1
   slot {class=VerticalBoxSlot, properties {padding=<ltrb 110 50 110 0>}}
  }
  CommonBorder_1 {
   widgetClassPath=&CommonBorder
   parent=HorizontalBox_1
   widget=CommonBorder_1
   properties {style=/FortUILibrary/Styles/Border/Border_MatchStatus_StatsLeft_Blue.Border_MatchStatus_StatsLeft_Blue_C, padding=<ltrb 0 0 0 0>}
   slot {class=HorizontalBoxSlot, properties {padding=<ltrb 20 0 0 0>, size {sizeRule=Fill, value=1}}}
  }
  VerticalBox_3 {
   widgetClassPath=&VerticalBox
   parent=CommonBorder_1
   widget=VerticalBox_3
   slot {class=BorderSlot, properties {padding=<ltrb 0 0 0 0>}}
  }
  VerticalBox_18 {
   widgetClassPath=&VerticalBox
   parent=VerticalBox_3
   widget=VerticalBox_18
   slot {class=VerticalBoxSlot, properties {horizontalAlignment=HAlign_Center, padding=<ltrb 0 50 0 50>}}
  }
  FastestEliminationTitleText {
   widgetClassPath=&CommonTextBlock
   parent=VerticalBox_18
   widget=FastestEliminationTitleText
   properties {
    justification=Center
    scrollStyle=&TextScrollStyle-Base_C
    text=FASTEST ELIMINATION
    colorAndOpacity {colorUseRule=UseColor_Specified, specifiedColor=<rgba 0.147 0.5089 0.8388 1>}
    font {
     bForceMonospaced=false
     bMaterialIsStencil=false
     fontMaterial=null
     fontObject=&Burbank
     letterSpacing=0
     monospacedWidth=1
     size=24
     skewAmount=0
     typefaceFontName=BurbankRegular-Black
     outlineSettings {
      bApplyOutlineToDropShadows=false
      bMiteredCorners=false
      bSeparateFillAlpha=false
      outlineMaterial=null
      outlineSize=0
      outlineColor=<rgba 0 0 0 1>
     }
    }
   }
   slot {class=VerticalBoxSlot, properties {horizontalAlignment=HAlign_Center, padding=<ltrb 0 0 0 20>}}
  }
  FastestEliminationPlayer {
   widgetClassPath=&/AthenaMatchStatusTextPlayer
   bIsVariable=true
   parent=VerticalBox_18
   widget=FastestEliminationPlayer
   properties {textToDisplay=PLAYER NAME}
   slot {class=VerticalBoxSlot, properties {horizontalAlignment=HAlign_Center, verticalAlignment=VAlign_Center}}
  }
  SizeBox_3 {
   widgetClassPath=&SizeBox
   parent=VerticalBox_18
   widget=SizeBox_3
   properties {bOverride_MinDesiredHeight=true, bOverride_MinDesiredWidth=true, minDesiredHeight=60, minDesiredWidth=300}
   slot {class=VerticalBoxSlot, properties {horizontalAlignment=HAlign_Center, padding=<ltrb 15 15 15 15>}}
  }
  CommonBorder_4 {
   widgetClassPath=&CommonBorder
   parent=SizeBox_3
   widget=CommonBorder_4
   properties {
    horizontalAlignment=HAlign_Center
    verticalAlignment=VAlign_Center
    background {
     bIsDynamicallyLoaded=false
     drawAs=Box
     imageType=NoImage
     mirroring=NoMirror
     resourceName=None
     resourceObject=/Game/UI/Foundation/Materials/MI_MatchStatusBG_Stats_Yellow
     tiling=NoTile
     imageSize=<32 32>
     margin=<ltrb 0 0 0 0>
     outlineSettings {
      bUseBrushTransparency=false
      roundingType=HalfHeightRadius
      width=0
      color {colorUseRule=UseColor_Specified, specifiedColor=<rgba 0 0 0 0>}
      cornerRadii=<0 0 0 0>
     }
     tintColor {colorUseRule=UseColor_Specified, specifiedColor=<rgba 1 1 1 1>}
     uVRegion {bIsValid=false, max=<0 0>, min=<0 0>}
    }
    padding=<ltrb 0 0 0 0>
   }
   slot {class=SizeBoxSlot}
  }
  FastestEliminationTime {
   widgetClassPath=&AthenaEventMatchStatusText
   bIsVariable=true
   parent=CommonBorder_4
   widget=FastestEliminationTime
   properties {
    textToDisplay=SCORE
    color {colorUseRule=UseColor_Specified, specifiedColor=<rgba 0.0086 0.1559 0.5583 1>}
    font {
     bForceMonospaced=false
     bMaterialIsStencil=false
     fontMaterial=null
     fontObject=&Burbank
     letterSpacing=0
     monospacedWidth=1
     size=24
     skewAmount=0
     typefaceFontName=BurbankRegular-Black
     outlineSettings {
      bApplyOutlineToDropShadows=false
      bMiteredCorners=false
      bSeparateFillAlpha=false
      outlineMaterial=null
      outlineSize=0
      outlineColor=<rgba 0 0 0 1>
     }
    }
   }
   slot {class=BorderSlot, properties {horizontalAlignment=HAlign_Center, verticalAlignment=VAlign_Center, padding=<ltrb 0 0 0 0>}}
  }
  VerticalBox_19 {
   widgetClassPath=&VerticalBox
   parent=VerticalBox_3
   widget=VerticalBox_19
   properties {visibility=Collapsed}
   slot {class=VerticalBoxSlot, properties {horizontalAlignment=HAlign_Center, padding=<ltrb 0 20 0 25>}}
  }
  DamageTakenTitleText {
   widgetClassPath=&CommonTextBlock
   bIsVariable=true
   parent=VerticalBox_19
   widget=DamageTakenTitleText
   properties {
    justification=Center
    scrollStyle=&TextScrollStyle-Base_C
    text=DAMAGE TAKEN
    colorAndOpacity {colorUseRule=UseColor_Specified, specifiedColor=<rgba 0.147 0.5089 0.8388 1>}
    font {
     bForceMonospaced=false
     bMaterialIsStencil=false
     fontMaterial=null
     fontObject=&Burbank
     letterSpacing=0
     monospacedWidth=1
     size=24
     skewAmount=0
     typefaceFontName=BurbankRegular-Black
     outlineSettings {
      bApplyOutlineToDropShadows=false
      bMiteredCorners=false
      bSeparateFillAlpha=false
      outlineMaterial=null
      outlineSize=0
      outlineColor=<rgba 0 0 0 1>
     }
    }
   }
   slot {class=VerticalBoxSlot, properties {padding=<ltrb 0 0 0 10>}}
  }
  DamageTakenPlayer {
   widgetClassPath=&CommonTextBlock
   bIsVariable=true
   parent=VerticalBox_19
   widget=DamageTakenPlayer
   properties {
    scrollStyle=&TextScrollStyle-Base_C
    text=PLAYER NAME
    font {
     bForceMonospaced=false
     bMaterialIsStencil=false
     fontMaterial=null
     fontObject=&Burbank
     letterSpacing=0
     monospacedWidth=1
     size=42
     skewAmount=0
     typefaceFontName=BurbankRegular-Black
     outlineSettings {
      bApplyOutlineToDropShadows=false
      bMiteredCorners=false
      bSeparateFillAlpha=false
      outlineMaterial=null
      outlineSize=0
      outlineColor=<rgba 0 0 0 1>
     }
    }
   }
   slot {class=VerticalBoxSlot, properties {horizontalAlignment=HAlign_Center, padding=<ltrb 20 0 20 5>}}
  }
  SizeBox_4 {
   widgetClassPath=&SizeBox
   parent=VerticalBox_19
   widget=SizeBox_4
   properties {bOverride_MinDesiredHeight=true, bOverride_MinDesiredWidth=true, minDesiredHeight=60, minDesiredWidth=300}
   slot {class=VerticalBoxSlot, properties {horizontalAlignment=HAlign_Center}}
  }
  CommonBorder_5 {
   widgetClassPath=&CommonBorder
   parent=SizeBox_4
   widget=CommonBorder_5
   properties {
    horizontalAlignment=HAlign_Center
    verticalAlignment=VAlign_Center
    background {
     bIsDynamicallyLoaded=false
     drawAs=Box
     imageType=NoImage
     mirroring=NoMirror
     resourceName=None
     resourceObject=&MI_MatchStatusBG_Stats_Yellow
     tiling=NoTile
     imageSize=<32 40>
     margin=<ltrb 0.0708 0 0.9027 0.0708>
     outlineSettings {
      bUseBrushTransparency=false
      roundingType=HalfHeightRadius
      width=0
      color {colorUseRule=UseColor_Specified, specifiedColor=<rgba 0 0 0 0>}
      cornerRadii=<0 0 0 0>
     }
     tintColor {colorUseRule=UseColor_Specified, specifiedColor=<rgba 1 1 1 1>}
     uVRegion {bIsValid=false, max=<0 0>, min=<0 0>}
    }
    padding=<ltrb 20 10 20 10>
   }
   slot {class=SizeBoxSlot}
  }
  DamageTaken {
   widgetClassPath=&CommonTextBlock
   bIsVariable=true
   parent=CommonBorder_5
   widget=DamageTaken
   properties {
    scrollStyle=&TextScrollStyle-Base_C
    text=SCORE
    colorAndOpacity {colorUseRule=UseColor_Specified, specifiedColor=<rgba 0.0086 0.1559 0.5583 1>}
    font {
     bForceMonospaced=false
     bMaterialIsStencil=false
     fontMaterial=null
     fontObject=&Burbank
     letterSpacing=0
     monospacedWidth=1
     size=24
     skewAmount=0
     typefaceFontName=BurbankRegular-Black
     outlineSettings {
      bApplyOutlineToDropShadows=false
      bMiteredCorners=false
      bSeparateFillAlpha=false
      outlineMaterial=null
      outlineSize=0
      outlineColor=<rgba 0 0 0 1>
     }
    }
   }
   slot {class=BorderSlot, properties {horizontalAlignment=HAlign_Center, verticalAlignment=VAlign_Center, padding=<ltrb 20 10 20 10>}}
  }
  CommonBorder_11 {
   widgetClassPath=&CommonBorder
   parent=HorizontalBox_1
   widget=CommonBorder_11
   properties {style=&Border_MatchStatus_Blue_C, padding=<ltrb 0 0 0 0>}
   slot {class=HorizontalBoxSlot, properties {size {sizeRule=Fill, value=1}}}
  }
  VerticalBox_16 {
   widgetClassPath=&VerticalBox
   parent=CommonBorder_11
   widget=VerticalBox_16
   slot {class=BorderSlot, properties {padding=<ltrb 0 0 0 0>}}
  }
  VerticalBox_20 {
   widgetClassPath=&VerticalBox
   parent=VerticalBox_16
   widget=VerticalBox_20
   slot {class=VerticalBoxSlot, properties {horizontalAlignment=HAlign_Center, padding=<ltrb 0 50 0 50>}}
  }
  MostEliminationsTitleText {
   widgetClassPath=&CommonTextBlock
   parent=VerticalBox_20
   widget=MostEliminationsTitleText
   properties {
    justification=Center
    scrollStyle=&TextScrollStyle-Base_C
    text=MOST ELIMINATIONS
    colorAndOpacity {colorUseRule=UseColor_Specified, specifiedColor=<rgba 0.147 0.5089 0.8388 1>}
    font {
     bForceMonospaced=false
     bMaterialIsStencil=false
     fontMaterial=null
     fontObject=&Burbank
     letterSpacing=0
     monospacedWidth=1
     size=24
     skewAmount=0
     typefaceFontName=BurbankRegular-Black
     outlineSettings {
      bApplyOutlineToDropShadows=false
      bMiteredCorners=false
      bSeparateFillAlpha=false
      outlineMaterial=null
      outlineSize=0
      outlineColor=<rgba 0 0 0 1>
     }
    }
   }
   slot {class=VerticalBoxSlot, properties {horizontalAlignment=HAlign_Center, padding=<ltrb 0 0 0 20>}}
  }
  MostEliminationsPlayer {
   widgetClassPath=&AthenaMatchStatusTextPlayer
   bIsVariable=true
   parent=VerticalBox_20
   widget=MostEliminationsPlayer
   properties {textToDisplay=PLAYER NAME}
   slot {class=VerticalBoxSlot, properties {horizontalAlignment=HAlign_Center, verticalAlignment=VAlign_Center}}
  }
  SizeBox_5 {
   widgetClassPath=&SizeBox
   parent=VerticalBox_20
   widget=SizeBox_5
   properties {bOverride_MinDesiredHeight=true, bOverride_MinDesiredWidth=true, minDesiredHeight=60, minDesiredWidth=300}
   slot {class=VerticalBoxSlot, properties {horizontalAlignment=HAlign_Center, padding=<ltrb 15 15 15 15>}}
  }
  CommonBorder_6 {
   widgetClassPath=&CommonBorder
   parent=SizeBox_5
   widget=CommonBorder_6
   properties {
    horizontalAlignment=HAlign_Center
    verticalAlignment=VAlign_Center
    background {
     bIsDynamicallyLoaded=false
     drawAs=Box
     imageType=NoImage
     mirroring=NoMirror
     resourceName=None
     resourceObject=&MI_MatchStatusBG_Stats_Yellow
     tiling=NoTile
     imageSize=<32 40>
     margin=<ltrb 0.0708 0 0.9027 0.0708>
     outlineSettings {
      bUseBrushTransparency=false
      roundingType=HalfHeightRadius
      width=0
      color {colorUseRule=UseColor_Specified, specifiedColor=<rgba 0 0 0 0>}
      cornerRadii=<0 0 0 0>
     }
     tintColor {colorUseRule=UseColor_Specified, specifiedColor=<rgba 1 1 1 1>}
     uVRegion {bIsValid=false, max=<0 0>, min=<0 0>}
    }
    padding=<ltrb 0 0 0 0>
   }
   slot {class=SizeBoxSlot}
  }
  MostEliminations {
   widgetClassPath=&AthenaEventMatchStatusText
   bIsVariable=true
   parent=CommonBorder_6
   widget=MostEliminations
   properties {
    textToDisplay=SCORE
    color {colorUseRule=UseColor_Specified, specifiedColor=<rgba 0.0086 0.1559 0.5583 1>}
    font {
     bForceMonospaced=false
     bMaterialIsStencil=false
     fontMaterial=null
     fontObject=&Burbank
     letterSpacing=0
     monospacedWidth=1
     size=24
     skewAmount=0
     typefaceFontName=BurbankRegular-Black
     outlineSettings {
      bApplyOutlineToDropShadows=false
      bMiteredCorners=false
      bSeparateFillAlpha=false
      outlineMaterial=null
      outlineSize=0
      outlineColor=<rgba 0 0 0 1>
     }
    }
   }
   slot {class=BorderSlot, properties {horizontalAlignment=HAlign_Center, verticalAlignment=VAlign_Center, padding=<ltrb 0 0 0 0>}}
  }
  VerticalBox_21 {
   widgetClassPath=&VerticalBox
   parent=VerticalBox_16
   widget=VerticalBox_21
   properties {visibility=Collapsed}
   slot {class=VerticalBoxSlot, properties {horizontalAlignment=HAlign_Center, padding=<ltrb 0 20 0 25>}}
  }
  DamageInflictedTitleText {
   widgetClassPath=&CommonTextBlock
   parent=VerticalBox_21
   widget=DamageInflictedTitleText
   properties {
    justification=Center
    scrollStyle=&TextScrollStyle-Base_C
    text=DAMAGE INFLICTED
    colorAndOpacity {colorUseRule=UseColor_Specified, specifiedColor=<rgba 0.147 0.5089 0.8388 1>}
    font {
     bForceMonospaced=false
     bMaterialIsStencil=false
     fontMaterial=null
     fontObject=&Burbank
     letterSpacing=0
     monospacedWidth=1
     size=24
     skewAmount=0
     typefaceFontName=BurbankRegular-Black
     outlineSettings {
      bApplyOutlineToDropShadows=false
      bMiteredCorners=false
      bSeparateFillAlpha=false
      outlineMaterial=null
      outlineSize=0
      outlineColor=<rgba 0 0 0 1>
     }
    }
   }
   slot {class=VerticalBoxSlot, properties {padding=<ltrb 0 5 0 10>}}
  }
  DamageInflictedPlayer {
   widgetClassPath=&CommonTextBlock
   bIsVariable=true
   parent=VerticalBox_21
   widget=DamageInflictedPlayer
   properties {
    scrollStyle=&TextScrollStyle-Base_C
    text=PLAYER NAME
    font {
     bForceMonospaced=false
     bMaterialIsStencil=false
     fontMaterial=null
     fontObject=&Burbank
     letterSpacing=0
     monospacedWidth=1
     size=42
     skewAmount=0
     typefaceFontName=BurbankRegular-Black
     outlineSettings {
      bApplyOutlineToDropShadows=false
      bMiteredCorners=false
      bSeparateFillAlpha=false
      outlineMaterial=null
      outlineSize=0
      outlineColor=<rgba 0 0 0 1>
     }
    }
   }
   slot {class=VerticalBoxSlot, properties {horizontalAlignment=HAlign_Center, padding=<ltrb 20 0 20 5>}}
  }
  SizeBox_6 {
   widgetClassPath=&SizeBox
   parent=VerticalBox_21
   widget=SizeBox_6
   properties {bOverride_MinDesiredHeight=true, bOverride_MinDesiredWidth=true, minDesiredHeight=60, minDesiredWidth=300}
   slot {class=VerticalBoxSlot, properties {horizontalAlignment=HAlign_Center}}
  }
  CommonBorder_7 {
   widgetClassPath=&CommonBorder
   parent=SizeBox_6
   widget=CommonBorder_7
   properties {
    horizontalAlignment=HAlign_Center
    verticalAlignment=VAlign_Center
    background {
     bIsDynamicallyLoaded=false
     drawAs=Box
     imageType=NoImage
     mirroring=NoMirror
     resourceName=None
     resourceObject=&MI_MatchStatusBG_Stats_Yellow
     tiling=NoTile
     imageSize=<32 40>
     margin=<ltrb 0.0708 0 0.9027 0.0708>
     outlineSettings {
      bUseBrushTransparency=false
      roundingType=HalfHeightRadius
      width=0
      color {colorUseRule=UseColor_Specified, specifiedColor=<rgba 0 0 0 0>}
      cornerRadii=<0 0 0 0>
     }
     tintColor {colorUseRule=UseColor_Specified, specifiedColor=<rgba 1 1 1 1>}
     uVRegion {bIsValid=false, max=<0 0>, min=<0 0>}
    }
    padding=<ltrb 20 10 20 10>
   }
   slot {class=SizeBoxSlot}
  }
  DamageInflicted {
   widgetClassPath=&CommonTextBlock
   bIsVariable=true
   parent=CommonBorder_7
   widget=DamageInflicted
   properties {
    scrollStyle=&TextScrollStyle-Base_C
    text=SCORE
    colorAndOpacity {colorUseRule=UseColor_Specified, specifiedColor=<rgba 0.0086 0.1559 0.5583 1>}
    font {
     bForceMonospaced=false
     bMaterialIsStencil=false
     fontMaterial=null
     fontObject=&Burbank
     letterSpacing=0
     monospacedWidth=1
     size=24
     skewAmount=0
     typefaceFontName=BurbankRegular-Black
     outlineSettings {
      bApplyOutlineToDropShadows=false
      bMiteredCorners=false
      bSeparateFillAlpha=false
      outlineMaterial=null
      outlineSize=0
      outlineColor=<rgba 0 0 0 1>
     }
    }
   }
   slot {class=BorderSlot, properties {horizontalAlignment=HAlign_Center, verticalAlignment=VAlign_Center, padding=<ltrb 20 10 20 10>}}
  }
  CommonBorder_12 {
   widgetClassPath=&CommonBorder
   parent=HorizontalBox_1
   widget=CommonBorder_12
   properties {style=/FortUILibrary/Styles/Border/Border_MatchStatus_StatsRight_Blue.Border_MatchStatus_StatsRight_Blue_C, padding=<ltrb 0 0 0 0>}
   slot {class=HorizontalBoxSlot, properties {padding=<ltrb 0 0 20 0>, size {sizeRule=Fill, value=1}}}
  }
  VerticalBox_17 {
   widgetClassPath=&VerticalBox
   parent=CommonBorder_12
   widget=VerticalBox_17
   slot {class=BorderSlot, properties {padding=<ltrb 0 0 0 0>}}
  }
  VerticalBox_22 {
   widgetClassPath=&VerticalBox
   parent=VerticalBox_17
   widget=VerticalBox_22
   slot {class=VerticalBoxSlot, properties {horizontalAlignment=HAlign_Center, padding=<ltrb 0 50 0 50>}}
  }
  LongestEliminationTitleText {
   widgetClassPath=&CommonTextBlock
   bIsVariable=true
   parent=VerticalBox_22
   widget=LongestEliminationTitleText
   properties {
    justification=Center
    scrollStyle=&TextScrollStyle-Base_C
    text=LONGEST ELIMINATION
    colorAndOpacity {colorUseRule=UseColor_Specified, specifiedColor=<rgba 0.147 0.5089 0.8388 1>}
    font {
     bForceMonospaced=false
     bMaterialIsStencil=false
     fontMaterial=null
     fontObject=&Burbank
     letterSpacing=0
     monospacedWidth=1
     size=24
     skewAmount=0
     typefaceFontName=BurbankRegular-Black
     outlineSettings {
      bApplyOutlineToDropShadows=false
      bMiteredCorners=false
      bSeparateFillAlpha=false
      outlineMaterial=null
      outlineSize=0
      outlineColor=<rgba 0 0 0 1>
     }
    }
   }
   slot {class=VerticalBoxSlot, properties {horizontalAlignment=HAlign_Center, padding=<ltrb 0 0 0 20>}}
  }
  LongestEliminationPlayer {
   widgetClassPath=&AthenaMatchStatusTextPlayer
   bIsVariable=true
   parent=VerticalBox_22
   widget=LongestEliminationPlayer
   properties {textToDisplay=PLAYER NAME}
   slot {class=VerticalBoxSlot, properties {horizontalAlignment=HAlign_Center, verticalAlignment=VAlign_Center}}
  }
  SizeBox_7 {
   widgetClassPath=&SizeBox
   parent=VerticalBox_22
   widget=SizeBox_7
   properties {bOverride_MinDesiredHeight=true, bOverride_MinDesiredWidth=true, minDesiredHeight=60, minDesiredWidth=300}
   slot {class=VerticalBoxSlot, properties {horizontalAlignment=HAlign_Center, padding=<ltrb 15 15 15 15>}}
  }
  CommonBorder_9 {
   widgetClassPath=&CommonBorder
   parent=SizeBox_7
   widget=CommonBorder_9
   properties {
    horizontalAlignment=HAlign_Center
    verticalAlignment=VAlign_Center
    background {
     bIsDynamicallyLoaded=false
     drawAs=Box
     imageType=NoImage
     mirroring=NoMirror
     resourceName=None
     resourceObject=&MI_MatchStatusBG_Stats_Yellow
     tiling=NoTile
     imageSize=<32 40>
     margin=<ltrb 0.0708 0 0.9027 0.0708>
     outlineSettings {
      bUseBrushTransparency=false
      roundingType=HalfHeightRadius
      width=0
      color {colorUseRule=UseColor_Specified, specifiedColor=<rgba 0 0 0 0>}
      cornerRadii=<0 0 0 0>
     }
     tintColor {colorUseRule=UseColor_Specified, specifiedColor=<rgba 1 1 1 1>}
     uVRegion {bIsValid=false, max=<0 0>, min=<0 0>}
    }
    padding=<ltrb 20 10 20 10>
   }
   slot {class=SizeBoxSlot}
  }
  LongestElimination {
   widgetClassPath=&AthenaEventMatchStatusText
   bIsVariable=true
   parent=CommonBorder_9
   widget=LongestElimination
   properties {
    textToDisplay=SCORE
    color {colorUseRule=UseColor_Specified, specifiedColor=<rgba 0.0086 0.1559 0.5583 1>}
    font {
     bForceMonospaced=false
     bMaterialIsStencil=false
     fontMaterial=null
     fontObject=&Burbank
     letterSpacing=0
     monospacedWidth=1
     size=24
     skewAmount=0
     typefaceFontName=BurbankRegular-Black
     outlineSettings {
      bApplyOutlineToDropShadows=false
      bMiteredCorners=false
      bSeparateFillAlpha=false
      outlineMaterial=null
      outlineSize=0
      outlineColor=<rgba 0 0 0 1>
     }
    }
   }
   slot {class=BorderSlot, properties {horizontalAlignment=HAlign_Center, verticalAlignment=VAlign_Center, padding=<ltrb 20 10 20 10>}}
  }
  VerticalBox_23 {
   widgetClassPath=&VerticalBox
   parent=VerticalBox_17
   widget=VerticalBox_23
   properties {visibility=Collapsed}
   slot {class=VerticalBoxSlot, properties {horizontalAlignment=HAlign_Center, padding=<ltrb 0 20 0 25>}}
  }
  DamageHealedTitleText {
   widgetClassPath=&CommonTextBlock
   bIsVariable=true
   parent=VerticalBox_23
   widget=DamageHealedTitleText
   properties {
    justification=Center
    scrollStyle=&TextScrollStyle-Base_C
    text=DAMAGE HEALED
    colorAndOpacity {colorUseRule=UseColor_Specified, specifiedColor=<rgba 0.147 0.5089 0.8388 1>}
    font {
     bForceMonospaced=false
     bMaterialIsStencil=false
     fontMaterial=null
     fontObject=&Burbank
     letterSpacing=0
     monospacedWidth=1
     size=24
     skewAmount=0
     typefaceFontName=BurbankRegular-Black
     outlineSettings {
      bApplyOutlineToDropShadows=false
      bMiteredCorners=false
      bSeparateFillAlpha=false
      outlineMaterial=null
      outlineSize=0
      outlineColor=<rgba 0 0 0 1>
     }
    }
   }
   slot {class=VerticalBoxSlot, properties {padding=<ltrb 0 0 0 10>}}
  }
  DamageHealedPlayer {
   widgetClassPath=&CommonTextBlock
   bIsVariable=true
   parent=VerticalBox_23
   widget=DamageHealedPlayer
   properties {
    scrollStyle=&TextScrollStyle-Base_C
    text=PLAYER NAME
    font {
     bForceMonospaced=false
     bMaterialIsStencil=false
     fontMaterial=null
     fontObject=&Burbank
     letterSpacing=0
     monospacedWidth=1
     size=42
     skewAmount=0
     typefaceFontName=BurbankRegular-Black
     outlineSettings {
      bApplyOutlineToDropShadows=false
      bMiteredCorners=false
      bSeparateFillAlpha=false
      outlineMaterial=null
      outlineSize=0
      outlineColor=<rgba 0 0 0 1>
     }
    }
   }
   slot {class=VerticalBoxSlot, properties {horizontalAlignment=HAlign_Center, padding=<ltrb 20 0 20 5>}}
  }
  SizeBox_8 {
   widgetClassPath=&SizeBox
   parent=VerticalBox_23
   widget=SizeBox_8
   properties {bOverride_MinDesiredHeight=true, bOverride_MinDesiredWidth=true, minDesiredHeight=60, minDesiredWidth=300}
   slot {class=VerticalBoxSlot, properties {horizontalAlignment=HAlign_Center}}
  }
  CommonBorder_8 {
   widgetClassPath=&CommonBorder
   parent=SizeBox_8
   widget=CommonBorder_8
   properties {
    horizontalAlignment=HAlign_Center
    verticalAlignment=VAlign_Center
    background {
     bIsDynamicallyLoaded=false
     drawAs=Box
     imageType=NoImage
     mirroring=NoMirror
     resourceName=None
     resourceObject=&MI_MatchStatusBG_Stats_Yellow
     tiling=NoTile
     imageSize=<32 40>
     margin=<ltrb 0.0708 0 0.9027 0.0708>
     outlineSettings {
      bUseBrushTransparency=false
      roundingType=HalfHeightRadius
      width=0
      color {colorUseRule=UseColor_Specified, specifiedColor=<rgba 0 0 0 0>}
      cornerRadii=<0 0 0 0>
     }
     tintColor {colorUseRule=UseColor_Specified, specifiedColor=<rgba 1 1 1 1>}
     uVRegion {bIsValid=false, max=<0 0>, min=<0 0>}
    }
    padding=<ltrb 20 10 20 10>
   }
   slot {class=SizeBoxSlot}
  }
  DamageHealed {
   widgetClassPath=&CommonTextBlock
   bIsVariable=true
   parent=CommonBorder_8
   widget=DamageHealed
   properties {
    scrollStyle=&TextScrollStyle-Base_C
    text=SCORE
    colorAndOpacity {colorUseRule=UseColor_Specified, specifiedColor=<rgba 0.0086 0.1559 0.5583 1>}
    font {
     bForceMonospaced=false
     bMaterialIsStencil=false
     fontMaterial=null
     fontObject=&Burbank
     letterSpacing=0
     monospacedWidth=1
     size=24
     skewAmount=0
     typefaceFontName=BurbankRegular-Black
     outlineSettings {
      bApplyOutlineToDropShadows=false
      bMiteredCorners=false
      bSeparateFillAlpha=false
      outlineMaterial=null
      outlineSize=0
      outlineColor=<rgba 0 0 0 1>
     }
    }
   }
   slot {class=BorderSlot, properties {horizontalAlignment=HAlign_Center, verticalAlignment=VAlign_Center, padding=<ltrb 20 10 20 10>}}
  }
 ]
}
