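# Broadcast spectator map panel widget, built on AthenaBroadcastMapPanelBase, showing the map with a switchable left sidebar, game phase and players-left readouts.
# LHSSwitcher holds a player list and an empty spacer, beside a cycle button hidden for gamepad and touch.
# On switcher changes, the graph points the button's arrow right when the spacer is active, left otherwise.
# A bottom bar sits below, hidden for mouse/keyboard, touch and mobile.
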WidgetBlueprint /FortniteCoreUI/HUD/Spectator/AthenaBroadcastMapPanel
graphs [EventGraph {nodes [
   n1 = "On Active Widget Changed (LHSSwitcher)"!() -> (OutputDelegate, then, ActiveWidget:Widget Object Reference, ActiveWidgetIndex:int=0) {is=title K2Node_ComponentBoundEvent}
   > n2 = Cast To IconTextButton!(execute, Object=Get Button_CycleSideBar() -> (Button_CycleSideBar:Common Button Legacy Object Reference)) -> (then, CastFailed, AsIcon Text Button:Icon Text Button Object Reference) {is=title K2Node_DynamicCast}
   n3 = Get LhsSwitcherEmpty() -> (LhsSwitcherEmpty:Spacer Object Reference)
   > Set Icon!(
    execute
    self:Icon Text Button Object Reference=@n2
    IconBrush:Slate Brush Structure=Select(
     Option 0:Slate Brush Structure=Get LeftArrowBrush() -> (LeftArrowBrush:Slate Brush Structure)
     Option 1:Slate Brush Structure=Get RightArrowBrush() -> (RightArrowBrush:Slate Brush Structure)
     Index="Equal (Object)"(A=@n3, B=@n1.ActiveWidget) {is=title K2Node_CallFunction}
    ) -> (ReturnValue:Slate Brush Structure)
   ) -> (then) {is=title K2Node_CallFunction}
  ]}]
variables [
 LeftArrowBrush {
  type=SlateBrush
  default {
   bIsDynamicallyLoaded=false
   drawAs=Image
   imageType=NoImage
   mirroring=NoMirror
   resourceName=None
   resourceObject=/Game/UI/Style/Button-BasicArrowLeft-_normal_
   tiling=NoTile
   imageSize=<24 48>
   margin=<ltrb 0 0 0 0>
   outlineSettings {
    bUseBrushTransparency=false
    roundingType=HalfHeightRadius
    width=0
    color {colorUseRule=UseColor_Specified, specifiedColor=<rgba 0 0 0 0>}
    cornerRadii=<0 0 0 1>
   }
   tintColor {colorUseRule=UseColor_Specified, specifiedColor=<rgba 1 1 1 1>}
   uVRegion {bIsValid=false, max=<0 0>, min=<0 0>}
  }
 }
 RightArrowBrush {
  type=SlateBrush
  default {
   bIsDynamicallyLoaded=false
   drawAs=Image
   imageType=NoImage
   mirroring=NoMirror
   resourceName=None
   resourceObject=/Game/UI/Style/Button-BasicArrowRight-_normal_
   tiling=NoTile
   imageSize=<24 48>
   margin=<ltrb 0 0 0 0>
   outlineSettings {
    bUseBrushTransparency=false
    roundingType=HalfHeightRadius
    width=0
    color {colorUseRule=UseColor_Specified, specifiedColor=<rgba 0 0 0 0>}
    cornerRadii=<0 0 0 1>
   }
   tintColor {colorUseRule=UseColor_Specified, specifiedColor=<rgba 1 1 1 1>}
   uVRegion {bIsValid=false, max=<0 0>, min=<0 0>}
  }
 }
]
widgets {
 info {inheritedWidgetCount=4, parentClass=/Script/FortniteUI.AthenaBroadcastMapPanelBase, rootWidgetClass=&VerticalBox, widgetCount=18}
 widgets [
  VerticalBox_0 {widgetClassPath=&VerticalBox, widget=VerticalBox_0}
  CommonBorder_2 {
   widgetClassPath=&CommonBorder
   parent=VerticalBox_0
   widget=CommonBorder_2
   properties {style=/FortUILibrary/Styles/Border/Border_Navy_VGrad.Border_Navy_VGrad_C, padding=<ltrb 0 0 0 0>}
   slot {class=VerticalBoxSlot, properties {size {sizeRule=Fill, value=1}}}
  }
  Overlay_1 {widgetClassPath=&Overlay, parent=CommonBorder_2, widget=Overlay_1, slot {class=BorderSlot, properties {padding=<ltrb 0 0 0 0>}}}
  HorizontalBox_0 {
   widgetClassPath=&HorizontalBox
   parent=Overlay_1
   widget=HorizontalBox_0
   slot {class=OverlaySlot, properties {horizontalAlignment=HAlign_Fill, verticalAlignment=VAlign_Fill}}
  }
  CommonVisibilityWidget_1 {
   widgetClassPath=&CommonVisibilityWidgetLegacy
   parent=HorizontalBox_0
   widget=CommonVisibilityWidget_1
   properties {bShowForGamepad=false, bShowForTouch=false, padding=<ltrb 0 0 0 0>}
   slot {class=HorizontalBoxSlot}
  }
  SizeBox_2 {
   widgetClassPath=&SizeBox
   parent=CommonVisibilityWidget_1
   widget=SizeBox_2
   properties {bOverride_WidthOverride=true, widthOverride=50}
   slot {class=BorderSlot, properties {padding=<ltrb 0 0 0 0>}}
  }
  Button_CycleSideBar {
   widgetClassPath=&IconTextButton
   bInherited=true
   bIsVariable=true
   parent=SizeBox_2
   widget=Button_CycleSideBar
   properties {
    bDisplayInputActionWhenNotInteractable=false
    bHideInputAction=true
    bHideInputActionWithKeyboard=true
    bShouldUseFallbackDefaultInputAction=false
    buttonText=""
    contentAlignment=HAlign_Center
    onButtonBaseClicked=(&self:WidgetTree.Button_CycleSideBar.HandleOnButtonClicked)
    onButtonBaseDoubleClicked=(&self:WidgetTree.Button_CycleSideBar.HandleOnButtonDoubleClicked)
    onButtonBaseHovered=(&self:WidgetTree.Button_CycleSideBar.HandleOnButtonHovered)
    onButtonBaseUnhovered=(&self:WidgetTree.Button_CycleSideBar.HandleOnButtonUnhovered)
    onSelectedChangedBase=(&self:WidgetTree.Button_CycleSideBar.HandleOnSelectedChanged)
    style=/FortUILibrary/Foundation/Buttons/ButtonStyles/TopBar/ButtonStyle-Outline-MediumBlue.ButtonStyle-Outline-MediumBlue_C
    iconBrush {
     bIsDynamicallyLoaded=false
     drawAs=Image
     imageType=NoImage
     mirroring=NoMirror
     resourceName=None
     resourceObject=&Button-BasicArrowLeft-_normal_
     tiling=NoTile
     imageSize=<24 48>
     margin=<ltrb 0 0 0 0>
     outlineSettings {
      bUseBrushTransparency=false
      roundingType=HalfHeightRadius
      width=0
      color {colorUseRule=UseColor_Specified, specifiedColor=<rgba 0 0 0 0>}
      cornerRadii=<0 0 0 1>
     }
     tintColor {colorUseRule=UseColor_Specified, specifiedColor=<rgba 1 1 1 1>}
     uVRegion {bIsValid=false, max=<0 0>, min=<0 0>}
    }
   }
   slot {class=SizeBoxSlot, properties {horizontalAlignment=HAlign_Center}}
  }
  LHSSwitcher {
   widgetClassPath=&CommonWidgetSwitcherLegacy
   bInherited=true
   bIsVariable=true
   parent=HorizontalBox_0
   widget=LHSSwitcher
   properties {transitionDuration=0}
   slot {class=HorizontalBoxSlot, properties {size {sizeRule=Automatic, value=0.35}}}
  }
  LhsSwitcherEmpty {
   widgetClassPath=&Spacer
   bIsVariable=true
   parent=LHSSwitcher
   widget=LhsSwitcherEmpty
   properties {size=<0 0>}
   slot {class=WidgetSwitcherSlot}
  }
  PlayerList {
   widgetClassPath=&/AthenaSpectatorMapPlayerList
   bInherited=true
   parent=LHSSwitcher
   widget=PlayerList
   properties {bDesignTime_ShowEventScores=false}
   slot {class=WidgetSwitcherSlot}
  }
  ScaleBox_0 {
   widgetClassPath=&ScaleBox
   parent=HorizontalBox_0
   widget=ScaleBox_0
   properties {stretch=ScaleToFitY}
   slot {class=HorizontalBoxSlot, properties {size {sizeRule=Fill, value=1}}}
  }
  AthenaMapLayer {
   widgetClassPath=/Script/FortniteUI.AthenaMapLayer
   bInherited=true
   parent=ScaleBox_0
   widget=AthenaMapLayer
   slot {class=ScaleBoxSlot, properties {horizontalAlignment=HAlign_Fill, verticalAlignment=VAlign_Fill}}
  }
  HorizontalBox_1 {
   widgetClassPath=&HorizontalBox
   parent=Overlay_1
   widget=HorizontalBox_1
   slot {class=OverlaySlot, properties {horizontalAlignment=HAlign_Right, padding=<ltrb 0 8 8 0>}}
  }
  AthenaGamePhaseWidget {
   widgetClassPath=/Game/Athena/HUD/Phase/AthenaGamePhaseWidget
   parent=HorizontalBox_1
   widget=AthenaGamePhaseWidget
   slot {class=HorizontalBoxSlot}
  }
  AthenaPlayersLeftWidget {
   widgetClassPath=/Game/Athena/HUD/AthenaPlayersLeftWidget
   parent=HorizontalBox_1
   widget=AthenaPlayersLeftWidget
   slot {class=HorizontalBoxSlot}
  }
  CommonBorder_0 {
   widgetClassPath=&CommonBorder
   parent=VerticalBox_0
   widget=CommonBorder_0
   properties {
    style=&Border_Navy_VGrad_C
    background {
     bIsDynamicallyLoaded=false
     drawAs=Image
     imageType=NoImage
     mirroring=NoMirror
     resourceName=None
     resourceObject=/FortUILibrary/Styles/Border/Materials/M_Border_DarkNavy_VGrad
     tiling=NoTile
     imageSize=<32 32>
     margin=<ltrb 0 0 0 0>
     outlineSettings {
      bUseBrushTransparency=false
      roundingType=HalfHeightRadius
      width=0
      color {colorUseRule=UseColor_Specified, specifiedColor=<rgba 0 0 0 0>}
      cornerRadii=<0 0 0 0>
     }
     tintColor {colorUseRule=UseColor_Specified, specifiedColor=<rgba 1 1 1 1>}
     uVRegion {bIsValid=false, max=<0 0>, min=<0 0>}
    }
   }
   slot {class=VerticalBoxSlot}
  }
  CommonVisibilityWidget_0 {
   widgetClassPath=&CommonVisibilityWidgetLegacy
   parent=CommonBorder_0
   widget=CommonVisibilityWidget_0
   properties {bShowForAndroid=false, bShowForIOS=false, bShowForMouseAndKeyboard=false, bShowForTouch=false, padding=<ltrb 0 0 0 0>}
   slot {class=BorderSlot}
  }
  AthenaBottomBarWidget {
   widgetClassPath=/Game/UI/Frontend/BottomBar/WBP_AthenaBottomBarWidget
   parent=CommonVisibilityWidget_0
   widget=AthenaBottomBarWidget
   slot {class=BorderSlot, properties {padding=<ltrb 0 0 0 0>}}
  }
 ]
}
