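# Tabbed full-screen Battle Royale inventory screen widget, with a tab list, activatable tab switcher and touch-only close button.
# Activation plays intro, vignette fade and sound; header backing follows the active tab's interface on tab changes.
# Creative photo mode deactivates the screen and reopens the inventory when photo mode ends.
# Drops of inventory drag items onto the screen discard droppable items through the inventory context, either the whole stack or half.
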WidgetBlueprint /Game/Athena/HUD/Inventory/AthenaInventoryScreenContainer
animations [
 IntroPanel {
  binding_count=1
  length_seconds=0.2
  object=&self:IntroPanel
  sequence {bindings [AthenaInventoryPanelContainer {
     id=#1
     possessed_class=AthenaInventoryPanelContainer_C
     tracks [
      {
       type=MovieSceneFloatTrack
       display_name=Render Opacity
       sections [{type=MovieSceneFloatSection, channels [None {type=MovieSceneScriptingFloatChannel, keys [[0,0],[4,1]]}]}]
      }
      {
       type=MovieScene2DTransformTrack
       sections [{
         type=MovieScene2DTransformSection
         channels [
          Translation.X {type=MovieSceneScriptingFloatChannel, keys [[0,128],[4,0]]}
          Scale.X {type=MovieSceneScriptingFloatChannel, keys [[0,0.975],[4,1]]}
          Scale.Y {type=MovieSceneScriptingFloatChannel, keys [[0,0.975],[4,1]]}
         ]
        }]
      }
     ]
    }]}
 }
 FadeScreenFill {
  binding_count=1
  length_seconds=1
  object=&self:FadeScreenFill
  sequence {bindings [Image_ScreenFIll {
     id=#2
     possessed_class=Image
     tracks [{
       type=MovieSceneWidgetMaterialTrack
       sections [{type=MovieSceneParameterSection, channels [Animate_Alpha {type=MovieSceneScriptingFloatChannel, keys [[0,0],[20,1]]}]}]
      }]
    }]}
 }
]
graphs [
 EventGraph {nodes [
   "On Tab Selected (TabList_Inventory)"!() -> (OutputDelegate, then, TabId:Name=None) {is=title K2Node_ComponentBoundEvent}
   > Branch(Condition=IsActivated!(self:Common Activatable Widget Object Reference) -> (ReturnValue:bool))
   > Update Header Style()
   Event On Deactivated()
   > StopAnimation(InAnimation=Get FadeScreenFill() -> (FadeScreenFill:Widget Animation Object Reference))
   > StopAnimation(InAnimation=Get IntroPanel() -> (IntroPanel:Widget Animation Object Reference))
   > n1 = Cast To FortPlayerControllerGameplay!(execute, Object=GetOwningPlayer()) -> (then, CastFailed, AsFort Player Controller Gameplay:Fort Player Controller Gameplay Object Reference) {is=title K2Node_DynamicCast}
   > Unbind Event from On Creative Photo Mode Activated(self:Fort Player Controller Gameplay Object Reference=@n1, Delegate=Create Event()) {is=title K2Node_RemoveDelegate}
   > PlaySound2D(Sound=/Game/Sounds/Fort_UI_Sounds/Gameplay/Inventory/UI_Inventory_Close_Cue)
   Event OnInitialized()
   > Branch(Condition=IsUsingTouch(PlayerController=GetOwningPlayer()))
   > SetVisibility(self=Get Image_ScreenFIll() -> (Image_ScreenFIll:Image Object Reference), InVisibility=SelfHitTestInvisible)
   > SetVisibility(InVisibility=SelfHitTestInvisible)
   Event Destruct()
   > n2 = Cast To FortPlayerControllerGameplay(Object=GetOwningPlayer())
   > Unbind Event from On Creative Photo Mode Deactivated(self:Fort Player Controller Gameplay Object Reference=@n2, Delegate=Create Event()) {is=title K2Node_RemoveDelegate}
   Event On Activated()
   > HandleIntroAnim(self:Athena Inventory Panel Container Object Reference=Get AthenaInventoryPanelContainer() -> (AthenaInventoryPanelContainer:Athena Inventory Panel Container Object Reference))
   > Update Header Style()
   > PlayAnimation(InAnimation=Get FadeScreenFill(), PlaybackSpeed=3)
   > PlayAnimation(InAnimation=Get IntroPanel(), PlaybackSpeed=1.25)
   > n3 = Branch(Condition=IsUsingTouch(PlayerController=GetOwningPlayer()))
   > n4 = SetVisibility(InVisibility=SelfHitTestInvisible)
   n5 = Cast To FortPlayerControllerGameplay(execute=[@n4.then, @n3.else], Object=GetOwningPlayer())
   > Bind Event to On Creative Photo Mode Activated(self:Fort Player Controller Gameplay Object Reference=@n5, Delegate=Create Event()) {is=title K2Node_AddDelegate}
   > PlaySound2D(Sound=/Game/Sounds/Fort_UI_Sounds/Gameplay/Inventory/UI_Inventory_Open_Cue)
  ]}
 HandleCreativePhotoModeActivated {nodes [
   HandleCreativePhotoModeActivated() {is=title K2Node_FunctionEntry}
   > DeactivateWidget()
   > n1 = Cast To FortPlayerControllerGameplay(Object=GetOwningPlayer())
   > Bind Event to On Creative Photo Mode Deactivated(self:Fort Player Controller Gameplay Object Reference=@n1, Delegate=Create Event()) {is=title K2Node_AddDelegate}
  ]}
 HandleCreativePhotoModeDeactivated {nodes [
   HandleCreativePhotoModeDeactivated() {is=title K2Node_FunctionEntry}
   > n1 = Cast To FortPlayerControllerGameplay(Object=GetOwningPlayer())
   > Unbind Event from On Creative Photo Mode Deactivated(self=@n1, Delegate=Create Event())
   > OpenInventory!(execute, self:Fort Player Controller Object Reference=@n1, InventoryTabId:Name=None) -> (then)
  ]}
 IsDragItemDroppable {nodes [
   n1 = IsDragItemDroppable!() -> (then, DragDropOperation:Drag Drop Operation Object Reference) {is=title K2Node_FunctionEntry}
   > n2 = Cast To AthenaInventoryDragDrop!(execute, Object=@n1) -> (then, CastFailed, AsAthena Inventory Drag Drop:Athena Inventory Drag Drop Object Reference) {is=title K2Node_DynamicCast}
   > n3 = Cast To FortWorldItem!(execute, Object=Get Payload!(self:Drag Drop Operation Object Reference=@n2) -> (Payload:Object Reference)) -> (then, CastFailed, AsFort World Item:Fort World Item Object Reference) {is=title K2Node_DynamicCast}
   > Return Node!(execute, bIsDroppable:bool=CanBeDropped!(self:Fort Item Object Reference=@n3) -> (ReturnValue:bool))
   Return Node(execute=@n3.CastFailed, bIsDroppable=false)
   Return Node(execute=@n2.CastFailed, bIsDroppable=false)
  ]}
 OnDragOver {nodes [
   n1 = OnDragOver!() -> (then, MyGeometry:Geometry Structure, PointerEvent:Pointer Event Structure, Operation:Drag Drop Operation Object Reference) {is=title K2Node_FunctionEntry}
   > n2 = Cast To AthenaInventoryDragDrop(Object=@n1.Operation)
   > ShowDropIcon!(
    execute
    self:Athena Inventory Drag Drop Object Reference=@n2
    Drop Action:AthenaDragDropAction Enum=Select(
     Option 0:AthenaDragDropAction Enum=NewEnumerator3
     Option 1:AthenaDragDropAction Enum=NewEnumerator1
     Index=IsDragItemDroppable!(DragDropOperation:Drag Drop Operation Object Reference=@n2) -> (bIsDroppable:bool) {is=title K2Node_CallFunction}
    ) -> (ReturnValue:AthenaDragDropAction Enum)
   ) -> (then)
   > Return Node!(execute, ReturnValue:bool=true)
   Return Node(execute=@n2.CastFailed, ReturnValue=false)
  ]}
 OnDrop {nodes [
   n1 = OnDrop!() -> (then, MyGeometry:Geometry Structure, PointerEvent:Pointer Event Structure, Operation:Drag Drop Operation Object Reference) {is=title K2Node_FunctionEntry}
   > n2 = Branch(Condition=IsDragItemDroppable(DragDropOperation=@n1.Operation))
   > n3 = Cast To AthenaInventoryDragDrop(Object=@n1.Operation)
   > n4 = Cast To FortWorldItem(Object=Get Payload(self=@n3))
   n5 = GetNumInStack!(self:Fort Item Object Reference=@n4) -> (ReturnValue:int=0)
   n6 = int / int!(A:int=@n5, B:int=2) -> (ReturnValue:int=0) {is=title K2Node_CallFunction}
   > DropItem!(
    execute
    self:Fort Inventory Context Object Reference=Get FortInventoryContext!() -> (ReturnValue:Fort Inventory Context Object Reference) {is=title K2Node_GetBlueprintContextEx}
    ItemBeingDropped:Fort World Item Object Reference=@n4
    Quantity:int=Select(
     Option 0:int=@n5
     Option 1:int="Max (Integer)"(A=@n6, B=1)
     Index=Get SplitTheStack!(self:Athena Inventory Drag Drop Object Reference=@n3) -> (SplitTheStack:bool)
    ) -> (ReturnValue:int)
   ) -> (then, ReturnValue:bool)
   Return Node(execute=[@n4.CastFailed, @n3.CastFailed, @n2.else], ReturnValue=false)
  ]}
 Update Header Style {nodes [
   Update Header Style() {is=title K2Node_FunctionEntry}
   > n1 = Sequence()
   > n2 = Cast To AthenaInventoryScreenContainerTabInterface!(
    execute
    Object=GetActiveWidget!(self:Widget Switcher Object Reference=Get Switcher_InventoryTabs() -> (Switcher_InventoryTabs:Common Activatable Widget Switcher Object Reference)) -> (ReturnValue:Widget Object Reference)
   ) -> (then, CastFailed, AsAthena Inventory Screen Container Tab Interface:Athena Inventory Screen Container Tab Interface Interface) {is=title K2Node_DynamicCast}
   > n3 = ShowHeaderBackground!(execute, self:Athena Inventory Screen Container Tab Interface Interface=@n2) -> (then, bShowHeaderBackground:bool)
   > Set bLocal_ShowHeaderBackground(bLocal_ShowHeaderBackground:bool=@n3) -> (Output_Get:bool)
   > n4 = ShowBackgroundBlurHeader!(execute, self:Athena Inventory Screen Container Tab Interface Interface=@n2) -> (then, bShowBackgroundBlurHeader:bool)
   > Set bLocal_ShowBackgroundBlurHeader(bLocal_ShowBackgroundBlurHeader:bool=@n4) -> (Output_Get:bool)
   SetVisibility(
    execute=@n1.then_1
    self=[
     Get Overlay_HeaderBacking() -> (Overlay_HeaderBacking:Overlay Object Reference)
     Get Image_FarBackPlate() -> (Image_FarBackPlate:Image Object Reference)
     Get HeaderBackground() -> (HeaderBackground:Image Object Reference)
    ]
    InVisibility=Select(
     Option 0:ESlateVisibility Enum=Collapsed
     Option 1:ESlateVisibility Enum=SelfHitTestInvisible
     Index=Get bLocal_ShowHeaderBackground() -> (bLocal_ShowHeaderBackground:bool)
    ) -> (ReturnValue:ESlateVisibility Enum)
   )
  ]}
]
variables []
widgets {
 info {inheritedWidgetCount=4, parentClass=/Script/FortniteUI.AthenaInventoryScreenContainer, rootWidgetClass=&ScaleBox, widgetCount=17}
 widgets [
  ScaleBox_Leto {widgetClassPath=&ScaleBox, bInherited=true, widget=ScaleBox_Leto, properties {stretch=None}}
  Overlay_2 {
   widgetClassPath=&Overlay
   parent=ScaleBox_Leto
   widget=Overlay_2
   slot {class=ScaleBoxSlot, properties {horizontalAlignment=HAlign_Fill, verticalAlignment=VAlign_Fill}}
  }
  Image_ScreenFIll {
   widgetClassPath=&Image
   bIsVariable=true
   parent=Overlay_2
   widget=Image_ScreenFIll
   properties {
    visibility=SelfHitTestInvisible
    brush {
     bIsDynamicallyLoaded=false
     drawAs=Image
     imageType=NoImage
     mirroring=NoMirror
     resourceName=None
     resourceObject=/Game/Athena/HUD/Inventory/Art/M_UI_HUD_Vignette_Inventory
     tiling=NoTile
     imageSize=<32 32>
     margin=<ltrb 0 0 0 0>
     outlineSettings {
      bUseBrushTransparency=false
      roundingType=HalfHeightRadius
      width=0
      color {colorUseRule=UseColor_Specified, specifiedColor=<rgba 0 0 0 0>}
      cornerRadii=<0 0 0 0>
     }
     tintColor {colorUseRule=UseColor_Specified, specifiedColor=<rgba 1 1 1 1>}
     uVRegion {bIsValid=false, max=<0 0>, min=<0 0>}
    }
   }
   slot {class=OverlaySlot, properties {horizontalAlignment=HAlign_Fill, verticalAlignment=VAlign_Fill}}
  }
  SafeZone_0 {
   widgetClassPath=&SafeZone
   parent=Overlay_2
   widget=SafeZone_0
   properties {padBottom=false, padTop=false}
   slot {class=OverlaySlot, properties {horizontalAlignment=HAlign_Fill, verticalAlignment=VAlign_Fill}}
  }
  ScaleBox_2 {widgetClassPath=&ScaleBox, parent=SafeZone_0, widget=ScaleBox_2, properties {stretch=ScaleBySafeZone}, slot {class=SafeZoneSlot}}
  Overlay_0 {
   widgetClassPath=&Overlay
   parent=ScaleBox_2
   widget=Overlay_0
   slot {class=ScaleBoxSlot, properties {horizontalAlignment=HAlign_Fill, verticalAlignment=VAlign_Fill}}
  }
  Overlay_HeaderBacking {
   widgetClassPath=&Overlay
   bIsVariable=true
   parent=Overlay_0
   widget=Overlay_HeaderBacking
   slot {class=OverlaySlot, properties {horizontalAlignment=HAlign_Fill}}
  }
  Image_FarBackPlate {
   widgetClassPath=&Image
   bIsVariable=true
   parent=Overlay_HeaderBacking
   widget=Image_FarBackPlate
   properties {
    renderOpacity=2
    visibility=Collapsed
    brush {
     bIsDynamicallyLoaded=false
     drawAs=Image
     imageType=NoImage
     mirroring=NoMirror
     resourceName=None
     resourceObject=null
     tiling=NoTile
     imageSize=<32 96>
     margin=<ltrb 0 0 0 0>
     outlineSettings {
      bUseBrushTransparency=false
      roundingType=HalfHeightRadius
      width=0
      color {colorUseRule=UseColor_Specified, specifiedColor=<rgba 0 0 0 0>}
      cornerRadii=<0 0 0 0>
     }
     tintColor {colorUseRule=UseColor_Specified, specifiedColor=<rgba 0 0.0392 0.2078 1>}
     uVRegion {bIsValid=false, max=<0 0>, min=<0 0>}
    }
   }
   slot {class=OverlaySlot, properties {horizontalAlignment=HAlign_Fill}}
  }
  HeaderBackground {
   widgetClassPath=&Image
   bIsVariable=true
   parent=Overlay_HeaderBacking
   widget=HeaderBackground
   properties {
    renderOpacity=2
    visibility=Collapsed
    brush {
     bIsDynamicallyLoaded=false
     drawAs=Image
     imageType=NoImage
     mirroring=NoMirror
     resourceName=None
     resourceObject=null
     tiling=NoTile
     imageSize=<32 96>
     margin=<ltrb 0 0 0 0>
     outlineSettings {
      bUseBrushTransparency=false
      roundingType=HalfHeightRadius
      width=0
      color {colorUseRule=UseColor_Specified, specifiedColor=<rgba 0 0 0 0>}
      cornerRadii=<0 0 0 0>
     }
     tintColor {colorUseRule=UseColor_Specified, specifiedColor=<rgba 1 1 1 1>}
     uVRegion {bIsValid=false, max=<0 0>, min=<0 0>}
    }
    colorAndOpacity=<rgba 0.0003 0.0762 0.4125 1>
   }
   slot {class=OverlaySlot, properties {horizontalAlignment=HAlign_Fill}}
  }
  VerticalBox_1 {
   widgetClassPath=&VerticalBox
   parent=Overlay_0
   widget=VerticalBox_1
   slot {class=OverlaySlot, properties {horizontalAlignment=HAlign_Fill, verticalAlignment=VAlign_Fill}}
  }
  SizeBox_TabList {
   widgetClassPath=&SizeBox
   parent=VerticalBox_1
   widget=SizeBox_TabList
   properties {bOverride_HeightOverride=true, heightOverride=84}
   slot {class=VerticalBoxSlot}
  }
  Overlay_1 {widgetClassPath=&Overlay, parent=SizeBox_TabList, widget=Overlay_1, slot {class=SizeBoxSlot}}
  TabList_Inventory {
   widgetClassPath=/Game/UI/Foundation/TabbedView/HorizontalTabList_Leto
   bInherited=true
   parent=Overlay_1
   widget=TabList_Inventory
   properties {
    bAutoListenForInput=true
    bHideSingleTab=true
    nextTabInputActionData {dataTable=&InputActionDataTable, rowName=MainTabRight}
    previousTabInputActionData {dataTable=&InputActionDataTable, rowName=MainTabLeft}
   }
   slot {class=OverlaySlot, properties {horizontalAlignment=HAlign_Center, verticalAlignment=VAlign_Center}}
  }
  Switcher_InventoryTabs {
   widgetClassPath=&CommonActivatableWidgetSwitcher
   bInherited=true
   parent=VerticalBox_1
   widget=Switcher_InventoryTabs
   properties {activeWidgetIndex=2, transitionDuration=0}
   slot {class=VerticalBoxSlot, properties {size {sizeRule=Fill, value=1}}}
  }
  AthenaInventoryPanelContainer {
   widgetClassPath=&/AthenaInventoryPanelContainer
   bIsVariable=true
   parent=Switcher_InventoryTabs
   widget=AthenaInventoryPanelContainer
   properties {bIsFocusable=true, bSupportsActivationFocus=true}
   slot {class=WidgetSwitcherSlot, properties {padding=<ltrb 0 0 48 0>}}
  }
  VisibilityBorder_Touch {
   widgetClassPath=&CommonHardwareVisibilityBorder
   parent=Overlay_0
   widget=VisibilityBorder_Touch
   properties {
    horizontalAlignment=HAlign_Right
    verticalAlignment=VAlign_Top
    padding=<ltrb 0 32 32 0>
    visibilityQuery {
     autoDescription=" ANY( Input.Touch )"
     tokenStreamVersion=0
     userDescription=if using touch
     queryTokenStream [0,1,1,1,0]
     tagDictionary [{tagName=Input.Touch}]
    }
   }
   slot {class=OverlaySlot, properties {horizontalAlignment=HAlign_Right}}
  }
  Button_TouchClose {
   widgetClassPath=/FortUILibrary/Foundation/Buttons/CloseButton
   bInherited=true
   parent=VisibilityBorder_Touch
   widget=Button_TouchClose
   properties {
    onButtonBaseClicked=(&self:WidgetTree.Button_TouchClose.HandleOnButtonClicked)
    onButtonBaseDoubleClicked=(&self:WidgetTree.Button_TouchClose.HandleOnButtonDoubleClicked)
    onButtonBaseHovered=(&self:WidgetTree.Button_TouchClose.HandleOnButtonHovered)
    onButtonBaseUnhovered=(&self:WidgetTree.Button_TouchClose.HandleOnButtonUnhovered)
    onSelectedChangedBase=(&self:WidgetTree.Button_TouchClose.HandleOnSelectedChanged)
   }
   slot {class=BorderSlot, properties {horizontalAlignment=HAlign_Right, verticalAlignment=VAlign_Top, padding=<ltrb 0 32 32 0>}}
  }
 ]
}
