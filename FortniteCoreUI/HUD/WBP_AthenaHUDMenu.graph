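# Athena HUD menu widget, derived from AthenaHUDMenu, layering primary content stack, inventory screen, collapsed shutdown-timer and content-on-demand containers, and bottom bars.
# The bottom bars sit on a separate non-scaling layer inside a visibility switcher.
# SetBottomBarVisibility sets switcher visibility and, for visible states, activates the legacy or menu bar if not already active.
# Events realign the legacy bar's slot, centering it horizontally when game over shows and filling otherwise.
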WidgetBlueprint /FortniteCoreUI/HUD/WBP_AthenaHUDMenu
graphs [
 EventGraph {nodes [
   n1 = Event SetBottomBarAlignments!() -> (
    OutputDelegate
    then
    VerticalAligment:"EVerticalAlignment Enum (by ref)"=VAlign_Fill
    HorizontalAligment:"EHorizontalAlignment Enum (by ref)"=HAlign_Fill
   ) {is=title K2Node_Event}
   n2 = Get LegacyBottomBar() -> (LegacyBottomBar:WBP Athena Bottom Bar Widget Object Reference)
   > Is Valid(InputObject=@n2)
   > n3 = Cast To CommonVisibilitySwitcherSlot!(execute, Object=Get Slot!(self:Widget Object Reference=@n2) -> (Slot:Panel Slot Object Reference)) -> (then, CastFailed, AsCommon Visibility Switcher Slot:Common Visibility Switcher Slot Object Reference) {is=title K2Node_DynamicCast}
   > SetVerticalAlignment(self=@n3, InVerticalAlignment=@n1.VerticalAligment)
   > SetHorizontalAlignment(self=@n3, InHorizontalAlignment=@n1.HorizontalAligment)
   n4 = Event On Game Over Screen!() -> (OutputDelegate, then, bShown:bool) {is=title K2Node_Event}
   > n5 = Cast To CommonVisibilitySwitcherSlot(Object=Get Slot(self=Get LegacyBottomBar()))
   > SetHorizontalAlignment(
    self=@n5
    InHorizontalAlignment=Select(Option 0:EHorizontalAlignment Enum=HAlign_Fill, Option 1:EHorizontalAlignment Enum=HAlign_Center, Index=@n4.bShown) -> (ReturnValue:EHorizontalAlignment Enum)
   )
  ]}
 SetBottomBarVisibility {nodes [
   n1 = SetBottomBarVisibility!() -> (then, NewVisibility:ESlateVisibility Enum=Visible, bIsUsingLegacyBottomBar:bool) {is=title K2Node_FunctionEntry}
   > SetVisibility(self=Get BottomBar() -> (BottomBar:Common Visibility Switcher Object Reference), InVisibility=@n1)
   > n2 = Switch on ESlateVisibility!(execute, Selection:ESlateVisibility Enum=@n1) -> (Visible:Exec, Collapsed:Exec, Hidden:Exec, HitTestInvisible:Exec, SelfHitTestInvisible:Exec) {is=title K2Node_SwitchEnum}
   n3 = Get BottomBar()
   n4 = Select(
    Option 0:Widget Object Reference=Get MenuBottomBar() -> (MenuBottomBar:WBP Athena Bottom Bar Widget Object Reference)
    Option 1:Widget Object Reference=Get LegacyBottomBar()
    Index=@n1.bIsUsingLegacyBottomBar
   ) -> (ReturnValue:Widget Object Reference)
   n5 = Branch(
    execute=[@n2.HitTestInvisible, @n2.SelfHitTestInvisible, @n2.Visible]
    Condition="Equal (Object)"(A=GetActiveWidget!(self:Common Visibility Switcher Object Reference=@n3) -> (ReturnValue:Widget Object Reference), B=@n4)
   )
   SetActiveWidget(execute=@n5.else, self=Get BottomBar(), Widget=@n4)
  ]}
]
variables []
widgets {
 info {inheritedWidgetCount=6, namedSlotCount=2, parentClass=/Script/FortniteUI.AthenaHUDMenu, rootWidgetClass=&Overlay, widgetCount=15}
 widgets [
  OverlayRoot {widgetClassPath=&Overlay, widget=OverlayRoot}
  ScaleBox_LetoSplitscreen {
   widgetClassPath=&ScaleBox
   bInherited=true
   parent=OverlayRoot
   widget=ScaleBox_LetoSplitscreen
   properties {stretch=None}
   slot {class=OverlaySlot, properties {horizontalAlignment=HAlign_Fill, verticalAlignment=VAlign_Fill}}
  }
  Overlay_AllContent {
   widgetClassPath=&Overlay
   bInherited=true
   parent=ScaleBox_LetoSplitscreen
   widget=Overlay_AllContent
   slot {class=ScaleBoxSlot, properties {horizontalAlignment=HAlign_Fill, verticalAlignment=VAlign_Fill}}
  }
  WidgetStack_PrimaryContent {
   widgetClassPath=/Script/CommonUI.CommonActivatableWidgetStack
   bInherited=true
   parent=Overlay_AllContent
   widget=WidgetStack_PrimaryContent
   properties {transitionDuration=0}
   slot {class=OverlaySlot, properties {horizontalAlignment=HAlign_Fill, verticalAlignment=VAlign_Fill}}
  }
  VerticalBox_0 {
   widgetClassPath=&VerticalBox
   parent=Overlay_AllContent
   widget=VerticalBox_0
   slot {class=OverlaySlot, properties {horizontalAlignment=HAlign_Center, padding=<ltrb 0 20 0 0>}}
  }
  Container_ServerShutdownTimer {
   widgetClassPath=&CommonBorder
   bInherited=true
   parent=VerticalBox_0
   widget=Container_ServerShutdownTimer
   properties {visibility=Collapsed}
   slot {class=VerticalBoxSlot}
  }
  AthenaServerShutdownWidget {
   widgetClassPath=/Game/Athena/HUD/EndGame/IGM/AthenaServerShutdownWidget
   bIsVariable=true
   parent=Container_ServerShutdownTimer
   widget=AthenaServerShutdownWidget
   slot {class=BorderSlot}
  }
  Container_ContentOnDemand {
   widgetClassPath=&CommonBorder
   parent=VerticalBox_0
   widget=Container_ContentOnDemand
   properties {horizontalAlignment=HAlign_Center, verticalAlignment=VAlign_Top, visibility=Collapsed}
   slot {class=VerticalBoxSlot, properties {horizontalAlignment=HAlign_Center, verticalAlignment=VAlign_Top}}
  }
  InventoryScreenContainer {
   widgetClassPath=/Game/Athena/HUD/Inventory/AthenaInventoryScreenContainer
   bInherited=true
   parent=Overlay_AllContent
   widget=InventoryScreenContainer
   properties {activatedVisibility=Visible, bSetVisibilityOnActivated=true, bSetVisibilityOnDeactivated=true}
   slot {class=OverlaySlot, properties {horizontalAlignment=HAlign_Fill, verticalAlignment=VAlign_Fill}}
  }
  NonScaling_Layer {
   widgetClassPath=&Overlay
   bInherited=true
   parent=OverlayRoot
   widget=NonScaling_Layer
   slot {class=OverlaySlot, properties {horizontalAlignment=HAlign_Fill, verticalAlignment=VAlign_Fill}}
  }
  BottomBar {
   widgetClassPath=&CommonVisibilitySwitcher
   bIsVariable=true
   parent=NonScaling_Layer
   widget=BottomBar
   slot {class=OverlaySlot, properties {horizontalAlignment=HAlign_Fill, verticalAlignment=VAlign_Fill}}
  }
  LegacyBottomBar {
   widgetClassPath=/Game/UI/Frontend/BottomBar/WBP_AthenaBottomBarWidget
   bIsVariable=true
   parent=BottomBar
   widget=LegacyBottomBar
   slot {class=CommonVisibilitySwitcherSlot, properties {horizontalAlignment=HAlign_Right, verticalAlignment=VAlign_Bottom}}
  }
  AthenaBoundActionBar_Legacy {
   widgetClassPath=/Game/UI/Frontend/BottomBar/AthenaBoundActionBar_BottomBar
   namedSlotHost=LegacyBottomBar
   widget=AthenaBoundActionBar_Legacy
   properties {actionButtonClass=/Game/UI/Frontend/BottomBar/AthenaHUDBoundActionButton.AthenaHUDBoundActionButton_C}
  }
  MenuBottomBar {
   widgetClassPath=&WBP_AthenaBottomBarWidget
   bIsVariable=true
   parent=BottomBar
   widget=MenuBottomBar
   slot {class=CommonVisibilitySwitcherSlot, properties {horizontalAlignment=HAlign_Right, verticalAlignment=VAlign_Bottom}}
  }
  WBP_CaptureForPostBufferUpdate {
   widgetClassPath=&WBP_CaptureForPostBufferUpdate
   bIsVariable=true
   namedSlotHost=MenuBottomBar
   slot=~
  }
 ]
}
