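# FortBroadcastViewSettings is the tabbed view settings panel for replay and spectator camera controls, with collapse, reset, info coach mark and close actions.
# ReplayTabs drives a widget switcher of settings pages, and widget-host gameplay tags let extra pages register as tabs at runtime.
# Console variables, build type and current camera type gate which tabs are shown or enabled.
# Input type swaps footer buttons for gamepad prompts.
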WidgetBlueprint /FortniteCoreUI/UI/Broadcast/FortBroadcastViewSettings
graphs [
 CameraTypeToText {nodes [
   n1 = CameraTypeToText!() -> (then, CameraType:ESpectatorCameraType Enum=ThirdPerson) {is=title K2Node_FunctionEntry}
   > Return Node!(
    execute
    Text:Text=Select!(
     ThirdPerson:Text="LOCTEXT(\"Third Person Cam\")"
     DroneFree:Text="LOCTEXT(\"Drone Free Cam\")"
     Gameplay:Text="LOCTEXT(\"Gameplay Cam\")"
     ReverseShot:Text="LOCTEXT(\"Reverse Shot Cam\")"
     FollowShot:Text="LOCTEXT(\"Follow Shot Cam\")"
     DroneFollow:Text="LOCTEXT(\"Drone Follow Cam\")"
     DroneAttach:Text="LOCTEXT(\"Drone Attach Cam\")"
     BattleMap:Text="LOCTEXT(\"Battle Map Cam\")"
     ARDrone:Text
     DroneFreeDynamic:Text
     Sequencer:Text
     Index:ESpectatorCameraType Enum=@n1
    ) -> (ReturnValue:Text)
   )
  ]}
 CollapseSettingsPanel {nodes [
   n1 = CollapseSettingsPanel!() -> (then, bCollapse:bool=~) {is=title K2Node_FunctionEntry}
   > n2 = Set bCollapsed(bCollapsed:bool=@n1) -> (Output_Get:bool)
   > n3 = Branch(Condition=@n2)
   > SetVisibility(
    self=[Get CoachMark() -> (CoachMark:WBP Replay Coach Mark Object Reference), Get Overlay_Settings() -> (Overlay_Settings:Overlay Object Reference)]
    InVisibility=Collapsed
   )
   > SetIcon!(
    execute
    self:WBP UIKit Button CTA Base Object Reference=Get Button_Collapse() -> (Button_Collapse:WBP Replay Button Object Reference)
    Icon:Object Reference=/FortUILibrary/Textures/Icons/ToBeDeprecated/T_UI_Icon_Global_Chevron_Right
   ) -> (then)
   > Set bInfoSettingTurnedOn(bInfoSettingTurnedOn:bool=false) -> (Output_Get:bool)
   n4 = Get CommonInputSubsystem!() -> (ReturnValue:Common Input Object Reference) {is=title K2Node_GetBlueprintContextEx}
   > Branch(Condition="Equal (Enum)"(A=GetCurrentInputType(self=@n4), B=Gamepad))
   > Set Input Mode Game And UI!(
    execute
    PlayerController:Player Controller Object Reference=GetPlayerController!(PlayerIndex:int=0) -> (ReturnValue:Player Controller Object Reference)
    InWidgetToFocus:Widget Object Reference
    InMouseLockMode:EMouseLockMode Enum=DoNotLock
    bHideCursorDuringCapture:bool=true
    bFlushInput:bool=true
   ) -> (then) {is=title K2Node_CallFunction}
   > SetVisibility(
    self=[
     Get Button_Reset_Hidden() -> (Button_Reset_Hidden:WBP Replay Button Object Reference)
     Get Button_Exit_Hidden() -> (Button_Exit_Hidden:WBP Replay Button Object Reference)
    ]
    InVisibility=Collapsed
   )
   > SetTabCollapsed!(
    execute
    self:Fort Replay Context Object Reference=Get FortReplayContext!() -> (ReturnValue:Fort Replay Context Object Reference) {is=title K2Node_GetBlueprintContextEx}
    bIsCollapsed:bool=true
   ) -> (then)
   SetVisibility(execute=@n3.else, self=Get Overlay_Settings(), InVisibility=SelfHitTestInvisible)
   > SetIcon(self=Get Button_Collapse(), Icon=/FortUILibrary/Textures/Icons/ToBeDeprecated/T_UI_Icon_Global_Chevron_Left)
   > SetPanelCollapsed!(execute, self:Fort Replay Context Object Reference=Get FortReplayContext(), bIsCollapsed:bool=true) -> (then)
   n5 = Get CommonInputSubsystem()
   > Branch(Condition="Equal (Enum)"(A=GetCurrentInputType(self=@n5), B=Gamepad))
   > SetVisibility(
    self=[Get Button_Reset_Hidden(), Get Button_Exit_Hidden(), Get BoxPrompts() -> (BoxPrompts:Overlay Object Reference)]
    InVisibility=SelfHitTestInvisible
   )
  ]}
 EventGraph {nodes [
   n1 = IsDesktopPlatform!(
    execute
    self:Fort Global UIContext Object Reference=Get FortGlobalUIContext!() -> (ReturnValue:Fort Global UIContext Object Reference) {is=title K2Node_GetBlueprintContextEx}
   ) -> (then, ReturnValue:bool)
   > Branch(Condition=@n1)
   > n2 = GetConsoleVariableBoolValue!(execute, VariableName:String=camera.EnableCameraFXTab) -> (then, ReturnValue:bool)
   > SetTabVisibility!(
    execute
    self:Common Tab List Widget Base Object Reference=Get Tabs!(self:WBP Replay Tabs Object Reference=Get ReplayTabs() -> (ReplayTabs:WBP Replay Tabs Object Reference)) -> (Tabs:WBP UIKit Tab Group Object Reference)
    TabNameID:Name=EFFECTS
    NewVisibility:ESlateVisibility Enum=Select!(Option 0:ESlateVisibility Enum=Collapsed, Option 1:ESlateVisibility Enum=SelfHitTestInvisible, Index:bool=@n2) -> (ReturnValue:ESlateVisibility Enum)
   ) -> (then)
   > n3 = GetConsoleVariableBoolValue(VariableName=replay.EnableCustomizationSettingsTab)
   > n4 = IsFollowCam!(
    execute
    self
    CamType:ESpectatorCameraType Enum=Get CurrentCameraType!(self:Fort Player Controller Spectating Object Reference=Get SpectatorController() -> (SpectatorController:Fort Player Controller Spectating Object Reference)) -> (CurrentCameraType:ESpectatorCameraType Enum=ThirdPerson)
   ) -> (then, IsFollowCam:bool)
   n5 = GetRuntimeOptions()
   n6 = AND Boolean!(
    A=@n4
    B=@n3
    C:bool=IsPlayingReplay!() -> (ReturnValue:bool)
    D:bool=NOT Boolean(A=IsShippingBuild!(self:Fort Runtime Options Object Reference=@n5) -> (ReturnValue:bool))
   ) -> (ReturnValue)
   > SetTabEnabled!(execute, self:Common Tab List Widget Base Object Reference=Get Tabs(self=Get ReplayTabs()), TabNameID:Name=LOCKER, bEnable:bool=@n6) -> (then)
   > SetTabVisibility(TabNameID=LOCKER, NewVisibility=Select(Option 0=Collapsed, Option 1=Visible, Index=@n6))
   n7 = IsShippingBuild(self=GetRuntimeOptions())
   > SetTabVisibility(
    self=Get Tabs(self=Get ReplayTabs())
    TabNameID=DEVELOPMENT
    NewVisibility=Select(Option 0=Collapsed, Option 1=Visible, Index=NOT Boolean(A=@n7))
   )
   > n8 = GetConsoleVariableBoolValue(VariableName=replay.EnableAnimeSettingsTab)
   > SetTabVisibility(self=Get Tabs(self=Get ReplayTabs()), TabNameID=ANIME, NewVisibility=Select(Option 0=Collapsed, Option 1=SelfHitTestInvisible, Index=@n8))
   > n9 = GetConsoleVariableBoolValue(VariableName=replay.EnableFogSettingsTab)
   > SetTabVisibility(self=Get Tabs(self=Get ReplayTabs()), TabNameID=FOG, NewVisibility=Select(Option 0=Collapsed, Option 1=SelfHitTestInvisible, Index=@n9))
   "On Unhover (ReplayTabs)"() {is=title K2Node_ComponentBoundEvent}
   > Branch(Condition=Get bUseHover() -> (bUseHover:bool))
   > SetTabCollapsed(self=Get FortReplayContext(), bIsCollapsed=true)
   n10 = "Tab Selected (ReplayTabs)"!() -> (OutputDelegate, then, TabID:Name=None) {is=title K2Node_ComponentBoundEvent}
   > n11 = For Each Loop(Array:Array of Names=Get TabNames() -> (TabNames:Array of Names)) -> (Array Element:Name)
   > Branch(Condition="Equal (Name)"(A=@n11, B=@n10.TabID))
   > SetActiveWidgetIndex(self=Get ViewSettingsWidgetSwitcher() -> (ViewSettingsWidgetSwitcher:Common Widget Switcher Legacy Object Reference), Index=@n11.Array Index)
   CollapseSettingsPanel!(execute=@n11.Completed, self, bCollapse:bool=false) -> (then) {is=title K2Node_CallFunction}
   > Branch(Condition=Get bUseHover())
   > SetTabCollapsed(self=Get FortReplayContext(), bIsCollapsed=false)
   "Tab Clicked (ReplayTabs)"() {is=title K2Node_ComponentBoundEvent}
   > CollapseSettingsPanel(bCollapse=false)
   n12 = "On Hold Action Progress (Button_ResetHold)"!() -> (OutputDelegate, then, HeldPercent:double) {is=title K2Node_ComponentBoundEvent}
   > Set bIsBlockedHoldEnd(bIsBlockedHoldEnd:bool=false) -> (Output_Get:bool)
   > Branch(Condition="Equal (Float)"(A=@n12.HeldPercent, B=1))
   > n13 = Branch(Condition=Get bIsBlockedHoldEnd() -> (bIsBlockedHoldEnd:bool))
   n14 = Cast To FortReplayViewSettingsTabBase!(
    execute=@n13.else
    Object=GetActiveWidget!(self:Widget Switcher Object Reference=Get ViewSettingsWidgetSwitcher()) -> (ReturnValue:Widget Object Reference)
   ) -> (then, CastFailed, AsFort Replay View Settings Tab Base:Fort Replay View Settings Tab Base Object Reference) {is=title K2Node_DynamicCast}
   > ResetFocusedElement(self:Fort Replay View Settings Tab Base Object Reference=@n14)
   n15 = "On Clicked (Button_Info)"!() -> (OutputDelegate, then, Button:Common Button Base Object Reference) {is=title K2Node_ComponentBoundEvent}
   n16 = Branch(execute=[@n15.then, @n25.then], Condition=Get bCollapsed() -> (bCollapsed:bool))
   n17 = Branch(execute=@n16.else, Condition=IsActivated!(self:Common Activatable Widget Object Reference=Get CoachMark()) -> (ReturnValue:bool))
   > n18 = Branch(Condition=Get bInfoSettingTurnedOn() -> (bInfoSettingTurnedOn:bool))
   n19 = Get CoachMark()
   > n20 = Play Animation with Finished event!(
    execute:Exec
    Widget:User Widget Object Reference=@n19
    InAnimation:Widget Animation Object Reference=Get Anim_Outro!(self:WBP UIKit Coach Mark Base Object Reference=@n19) -> (Anim_Outro:Widget Animation Object Reference)
    StartAtTime:float=0
    NumLoopsToPlay:int=1
    PlayMode:EUMGSequencePlayMode Enum
    PlaybackSpeed:float=1
   ) -> (then:Exec, Result:Widget Animation Handle Structure, Finished:Exec) {is=title K2Node_PlayAnimation2}
   SetVisibility(execute=@n20.Finished, self=Get CoachMark(), InVisibility=Collapsed)
   > Set bInfoSettingTurnedOn(bInfoSettingTurnedOn=false)
   SetVisibility(execute=@n18.else, self=Get CoachMark(), InVisibility=SelfHitTestInvisible)
   > Set bInfoSettingTurnedOn(bInfoSettingTurnedOn=true)
   n21 = Get CoachMark()
   > Play Animation with Finished event(Widget=@n21, InAnimation=Get Anim_Outro(self=@n21), StartAtTime=0, NumLoopsToPlay=1, PlayMode=Reverse, PlaybackSpeed=1)
   ActivateWidget(execute=@n17.else, self=Get CoachMark())
   > Set bInfoSettingTurnedOn(bInfoSettingTurnedOn=true)
   "On Clicked (Button_Reset)"!() -> (OutputDelegate, then, Button:Common Button Base Object Reference) {is=title K2Node_ComponentBoundEvent}
   > ResetCamera(self:Fort Player Controller Spectating Object Reference=Get SpectatorController())
   > n22 = Cast To FortReplayViewSettingsTabBase(Object=GetActiveWidget(self=Get ViewSettingsWidgetSwitcher()))
   > ResetToDefault(self:Fort Replay View Settings Tab Base Object Reference=@n22)
   n23 = "On Clicked (Button_Close)"!() -> (OutputDelegate, then, Button:Common Button Base Object Reference) {is=title K2Node_ComponentBoundEvent}
   ToggleViewSettingsVisibility(execute=[@n23.then, @n28.then], self:Fort Replay Context Object Reference=Get FortReplayContext())
   "On Clicked (Button_Collapse)"!() -> (OutputDelegate, then, Button:Common Button Base Object Reference) {is=title K2Node_ComponentBoundEvent}
   n24 = "Equal (Enum)"(
    A:ESlateVisibility Enum=GetVisibility!(self:Widget Object Reference=Get Overlay_Settings()) -> (ReturnValue:ESlateVisibility Enum=Visible)
    B:ESlateVisibility Enum=SelfHitTestInvisible
   )
   > CollapseSettingsPanel(bCollapse=@n24)
   > SetPanelCollapsed(self=Get FortReplayContext(), bIsCollapsed=@n24)
   > SetIsEnabled(self=Get Button_Info_Hidden() -> (Button_Info_Hidden:WBP Replay Button Object Reference), bInIsEnabled=@n24)
   n25 = "On Clicked (Button_Info_Hidden)"!() -> (OutputDelegate, then, Button:Common Button Base Object Reference) {is=title K2Node_ComponentBoundEvent}
   n26 = "On Clicked (Button_Reset_Hidden)"!() -> (OutputDelegate, then, Button:Common Button Base Object Reference) {is=title K2Node_ComponentBoundEvent}
   ResetCamera(execute=[@n29.then, @n26.then], self=Get SpectatorController())
   > n27 = Cast To FortReplayViewSettingsTabBase(Object=GetActiveWidget(self=Get ViewSettingsWidgetSwitcher()))
   > ResetToDefault(self=@n27)
   n28 = "On Clicked (Button_Exit_Hidden)"!() -> (OutputDelegate, then, Button:Common Button Base Object Reference) {is=title K2Node_ComponentBoundEvent}
   "On Hover (ReplayTabs)"() {is=title K2Node_ComponentBoundEvent}
   > Branch(Condition=Get bUseHover())
   > SetTabCollapsed(self=Get FortReplayContext(), bIsCollapsed=false)
   "On Hold Completed (Button_ResetHold)"() {is=title K2Node_ComponentBoundEvent}
   > n29 = Set bIsBlockedHoldEnd(bIsBlockedHoldEnd=true)
   n30 = OnKeybindsChanged() {is=title K2Node_CustomEvent}
   n31 = Get CommonInputSubsystem()
   > Is Valid(InputObject=@n31)
   n32 = GetCurrentInputType(self=@n31)
   > n33 = Branch(Condition="Equal (Enum)"(A:ECommonInputType Enum=@n32, B:ECommonInputType Enum=Gamepad))
   > SetVisibility(self=Get FooterBoxButtons() -> (FooterBoxButtons:Horizontal Box Object Reference), InVisibility=Collapsed)
   > n34 = Branch(Condition=Get bCollapsed())
   SetVisibility(execute=@n34.else, self=[Get Button_Exit_Hidden(), Get BoxPrompts(), Get Button_Reset_Hidden()], InVisibility=SelfHitTestInvisible)
   Branch(execute=@n33.else, Condition="Equal (Enum)"(A=@n32, B=MouseAndKeyboard))
   > SetVisibility(self=Get FooterBoxButtons(), InVisibility=SelfHitTestInvisible)
   > SetVisibility(self=[Get BoxPrompts(), Get Button_Reset_Hidden(), Get Button_Exit_Hidden()], InVisibility=Collapsed)
   n35 = Reopen!() -> (OutputDelegate, then, IsVisible:EHudVisibilityState Enum=FullyVisible) {is=title K2Node_CustomEvent}
   > SetVisibility(self=Get Overlay_Settings(), InVisibility=SelfHitTestInvisible)
   > SetIcon(self=Get Button_Collapse(), Icon=&T_UI_Icon_Global_Chevron_Left)
   > SetVisibility(self=[Get Button_Reset_Hidden(), Get Button_Exit_Hidden(), Get BoxPrompts()], InVisibility=SelfHitTestInvisible)
   > n36 = Switch on EHudVisibilityState!(execute, Selection:EHudVisibilityState Enum=@n35.IsVisible) -> (FullyVisible:Exec, FullyHidden:Exec, GameOnly:Exec, ReplayOnly:Exec) {is=title K2Node_SwitchEnum}
   SetVisibility(execute=@n36.FullyHidden, self=Get CoachMark(), InVisibility=Collapsed)
   > Set bInfoSettingTurnedOn(bInfoSettingTurnedOn=false)
   > Set bCollapsed(bCollapsed=false)
   n37 = HandleOnShowDesc!() -> (OutputDelegate, then, Header:"Text (by ref)", Description:"Text (by ref)") {is=title K2Node_CustomEvent}
   n38 = TextIsEmpty(InText=@n37.Description)
   n39 = InvertBool!(bValue:bool=@n38) -> (ReturnValue:bool)
   > SetIsEnabled(self=Get Button_Info_Hidden(), bInIsEnabled=@n39)
   > SetVisibility(
    self=Get Prompt_Info() -> (Prompt_Info:WBP Replay Key Binding Object Reference)
    InVisibility=Select(Option 0=Collapsed, Option 1=SelfHitTestInvisible, Index=@n39)
   )
   > n40 = Branch(Condition=@n38)
   > SetVisibility(self=Get CoachMark(), InVisibility=Collapsed)
   UpdateCoachMark!(execute=@n40.else, self, Header:Text=@n37.Header, Description:Text=@n37.Description) -> (then)
   Event Construct()
   > n41 = Sequence()
   > n42 = Cast To FortPlayerControllerSpectating!(execute, Object=GetPlayerController(PlayerIndex=0)) -> (then, CastFailed, AsFort Player Controller Spectating:Fort Player Controller Spectating Object Reference) {is=title K2Node_DynamicCast}
   > Set SpectatorController(SpectatorController:Fort Player Controller Spectating Object Reference=@n42) -> (Output_Get:Fort Player Controller Spectating Object Reference)
   > Bind Event to On Show Setting Description(self:Fort Replay Context Object Reference=Get FortReplayContext(), Delegate=@n37) {is=title K2Node_AddDelegate}
   > Bind Event to On Camera Type Changed(self:Fort Player Controller Spectating Object Reference=Get SpectatorController(), Delegate=Create Event()) {is=title K2Node_AddDelegate}
   n43 = Get SpectatorController()
   > OnCameraTypeChanged!(
    execute
    self
    PlayerController:Fort Player Controller Spectating Object Reference=@n43
    CameraType:ESpectatorCameraType Enum=Get CurrentCameraType(self=@n43) -> (CurrentCameraType=ThirdPerson)
   ) -> (then)
   > n44 = GetConsoleVariableBoolValue(VariableName=replay.AllowEditingTimeOfDayInReplays)
   > n45 = GetConsoleVariableBoolValue(VariableName=spectator.CheatSpectatorEnabled)
   n46 = Cast To FortReplaySpectatorAthena!(Object=GetOwningPlayer()) -> (AsFort Replay Spectator Athena:Fort Replay Spectator Athena Object Reference, bSuccess:bool=~) {is=title K2Node_DynamicCast}
   > SetTabEnabled(self=Get Tabs(self=Get ReplayTabs()), TabNameID=WORLD, bEnable=OR Boolean(A=@n45, B=AND Boolean!(A=@n44, B=@n46.bSuccess) -> (ReturnValue)))
   n47 = Get ReplaySettingTabSlotTag() -> (ReplaySettingTabSlotTag:Gameplay Tag Structure)
   RegisterWidgetHost!(execute=@n41.then_1, WidgetHost:Widget Object Reference, WidgetHostTag:"Gameplay Tag Structure (by ref)"=@n47, IsPlayerSpecific:bool=true) -> (then)
   n48 = Get ReplayTestingSettingTabSlotTag() -> (ReplayTestingSettingTabSlotTag:Gameplay Tag Structure)
   > RegisterWidgetHost(WidgetHostTag=@n48, IsPlayerSpecific=true)
   Event PreConstruct()
   > SetCollapseInputAction()
   n49 = Event Add Content To Host Widget!() -> (
    OutputDelegate
    then
    WidgetHostTag:"Gameplay Tag Structure (by ref)"
    ContentWidget:Widget Object Reference
    Padding:"Margin Structure (by ref)"
    bFill:bool
   ) {is=title K2Node_Event}
   n50 = Get ReplayTestingSettingTabSlotTag()
   n51 = Get ReplaySettingTabSlotTag()
   > Branch(Condition=OR Boolean(A="Equal (GameplayTag)"(A=@n49.WidgetHostTag, B=@n51), B="Equal (GameplayTag)"(A=@n49.WidgetHostTag, B=@n50)))
   > n52 = Cast To FortReplayViewSettingsTabBase(Object=@n49.ContentWidget)
   n53 = Get TabNameId!(self:Fort Replay View Settings Tab Base Object Reference=@n52) -> (TabNameId:Name=None)
   > n54 = Add(TargetArray:Array of Names=Get TabNames(), NewItem:"Name (by ref)"=@n53) {is=title K2Node_CallArrayFunction}
   n55 = GetTabViewModel!(
    self:UIKit Tab Group Object Reference=Get ReplayTabs_Extension() -> (ReplayTabs_Extension:WBP UIKit Tab Group Object Reference)
    TabID:Name=@n53
   ) -> (ReturnValue:MVVM Base Viewmodel Object Reference)
   n56 = IsValid(Object=@n55)
   > n57 = Branch(Condition=@n56)
   n58 = Construct Object from Class(execute=@n57.else, Class=/Script/FortniteUI.UIKitSelectableButtonViewModel) -> (ReturnValue:UI Kit Selectable Button View Model Object Reference)
   AddTab!(
    execute=[@n57.then, @n64.then]
    self:WBP UIKit Tab Group Base Object Reference=Get Tabs(self=Get ReplayTabs())
    Tab VM:MVVM Base Viewmodel Object Reference=Select(Option 0:MVVM Base Viewmodel Object Reference=@n58, Option 1:MVVM Base Viewmodel Object Reference=@n55, Index=@n56) -> (ReturnValue:MVVM Base Viewmodel Object Reference)
    Tab ID:Name=@n53
    Index:int=@n54
   ) -> (then)
   > AddChild(self=Get ViewSettingsWidgetSwitcher(), Content=@n52)
   > Branch(Condition="integer > integer"(A=@n54, B=6))
   > Do Once()
   n59 = Get Tabs(self=Get ReplayTabs())
   > Set DividerSpacing!(execute, DividerSpacing:double=30, self:WBP UIKit Tab Group Base Object Reference=@n59) -> (then, Output_Get:double)
   > AddDivider!(execute, self:WBP UIKit Tab Group Base Object Reference=@n59, TabID:Name=LIGHTING, AddAfter:bool=true) -> (then)
   > Set bUseHover(bUseHover:bool=false) -> (Output_Get:bool)
   Event Destruct()
   > UnregisterWidgetHost!(execute, WidgetHost:Widget Object Reference, WidgetHostTag:"Gameplay Tag Structure (by ref)"=@n47) -> (then)
   > UnregisterWidgetHost(WidgetHostTag=@n48)
   Event On Activated()
   n60 = Get ScrollBox_Tabs!(self:WBP UIKit Tab Group Base Object Reference=Get Tabs(self=Get ReplayTabs())) -> (ScrollBox_Tabs:Scroll Box Object Reference)
   > SetConsumeMouseWheel!(execute, self:Scroll Box Object Reference=@n60, NewConsumeMouseWheel:EConsumeMouseWheel Enum=Always) -> (then)
   > SetScrollbarThickness!(execute, self:Scroll Box Object Reference=@n60, NewScrollbarThickness_X:double=15, NewScrollbarThickness_Y:double=15) -> (then)
   > SetScrollbarPadding!(
    execute
    self:Scroll Box Object Reference=@n60
    NewScrollbarPadding_Left:float=-5
    NewScrollbarPadding_Top:float=25
    NewScrollbarPadding_Right:float=15
    NewScrollbarPadding_Bottom:float=25
   ) -> (then)
   > SetAnimateWheelScrolling!(execute, self:Scroll Box Object Reference=@n60, bShouldAnimateWheelScrolling:bool=true) -> (then)
   > SetAllowOverscroll!(execute, self:Scroll Box Object Reference=@n60, NewAllowOverscroll:bool=false) -> (then)
   > SwitchToValidTab()
   > n61 = GetConsoleVariableBoolValue(VariableName=spectator.EnableResetHoldButtonSettings)
   > SetVisibility(
    self=Get Button_ResetHold() -> (Button_ResetHold:WBP Replay Button Object Reference)
    InVisibility=Select(Option 0:ESlateVisibility Enum=Collapsed, Option 1:ESlateVisibility Enum=SelfHitTestInvisible, Index=@n61) -> (ReturnValue:ESlateVisibility Enum)
   )
   > SetVisibility(
    self=Get Button_Reset() -> (Button_Reset:WBP Replay Button Object Reference)
    InVisibility=Select(Option 0=SelfHitTestInvisible, Option 1=Collapsed, Index=@n61)
   )
   > OnKeybindsChanged() {is=title K2Node_CallFunction}
   > Bind Event to On Keybinds Changed(self:Fort Global UIContext Object Reference=Get FortGlobalUIContext(), Delegate=@n30) {is=title K2Node_AddDelegate}
   > Bind Event to On Replay Hud Visibility Changed(self:Fort Replay Context Object Reference=Get FortReplayContext(), Delegate=@n35) {is=title K2Node_AddDelegate}
   n62 = Get Image!(self:Fort Replay View Settings Tab Base Object Reference=@n52) -> (Image:Texture 2D Object Reference)
   n63 = Get DisplayName!(self:Fort Replay View Settings Tab Base Object Reference=@n52) -> (DisplayName:Text)
   n64 = Set Label!(execute=@n65.then, Label:Text=@n63, self:UI Kit Selectable Button View Model Object Reference=@n58) -> (then, Output_Get:Text)
   n65 = Set Icon!(execute=@n58.then, Icon:Texture 2D Object Reference=@n62, self:UI Kit Selectable Button View Model Object Reference=@n58) -> (then, Output_Get:Texture 2D Object Reference)
  ]}
 GetRemainingAnimTime {nodes [
   GetRemainingAnimTime() {is=title K2Node_FunctionEntry}
   > Return Node!(execute, FunctionIsImplemented:bool=false, RemainingTime:double=0, TotalTime:double=0)
  ]}
 GetSettingsTabs {nodes [
   GetSettingsTabs() {is=title K2Node_FunctionEntry}
   > n1 = For Each Loop(Array:Array of Widget Object References=GetAllChildren(self=Get ViewSettingsWidgetSwitcher())) -> (Array Element:Widget Object Reference)
   Return Node!(
    execute=@n1.Completed
    ReturnValue:Array of Fort Replay View Settings Tab Base Object References=Get Tabs!() -> (Tabs:Array of Fort Replay View Settings Tab Base Object References)
   )
   > n2 = Cast To FortReplayViewSettingsTabBase(Object=@n1)
   > Add(
    TargetArray:Array of Fort Replay View Settings Tab Base Object References=Get Tabs()
    NewItem:"Fort Replay View Settings Tab Base Object Reference (by ref)"=@n2
   )
  ]}
 IsDroneCam {nodes [
   n1 = IsDroneCam!() -> (then, CamType:ESpectatorCameraType Enum=ThirdPerson) {is=title K2Node_FunctionEntry}
   > n2 = Switch on ESpectatorCameraType!(execute, Selection:ESpectatorCameraType Enum=@n1) -> (
    ThirdPerson:Exec
    DroneFree:Exec
    Gameplay:Exec
    ReverseShot:Exec
    FollowShot:Exec
    DroneFollow:Exec
    DroneAttach:Exec
    BattleMap:Exec
    ARDrone:Exec
    DroneFreeDynamic:Exec
    Sequencer:Exec
   ) {is=title K2Node_SwitchEnum}
   n3 = Set IsDroneType(execute=[@n2.ARDrone, @n2.BattleMap, @n2.FollowShot, @n2.Gameplay, @n2.ReverseShot, @n2.ThirdPerson], IsDroneType:bool=false) -> (Output_Get:bool)
   Return Node!(execute=[@n4.then, @n3.then], IsDroneCam:bool=Get IsDroneType() -> (IsDroneType:bool))
   n4 = Set IsDroneType(execute=[@n2.DroneAttach, @n2.DroneFollow, @n2.DroneFree, @n2.DroneFreeDynamic], IsDroneType=true)
  ]}
 IsFollowCam {nodes [
   n1 = IsFollowCam!() -> (then, CamType:ESpectatorCameraType Enum=ThirdPerson) {is=title K2Node_FunctionEntry}
   > n2 = Switch on ESpectatorCameraType(Selection=@n1)
   n3 = Set IsFollowCamera(execute=[@n2.ARDrone, @n2.BattleMap, @n2.DroneFree], IsFollowCamera:bool=false) -> (Output_Get:bool)
   Return Node!(execute=[@n4.then, @n3.then], IsFollowCam:bool=Get IsFollowCamera() -> (IsFollowCamera:bool))
   n4 = Set IsFollowCamera(execute=[@n2.DroneAttach, @n2.DroneFollow, @n2.FollowShot, @n2.Gameplay, @n2.ReverseShot, @n2.ThirdPerson], IsFollowCamera=true)
  ]}
 OnCameraTypeChanged {nodes [
   n1 = OnCameraTypeChanged!() -> (then, PlayerController:Fort Player Controller Spectating Object Reference, CameraType:ESpectatorCameraType Enum=ThirdPerson) {is=title K2Node_FunctionEntry}
   > Set Loc_CamType(Loc_CamType:ESpectatorCameraType Enum=@n1.CameraType) -> (Output_Get:ESpectatorCameraType Enum=ThirdPerson)
   > n2 = Sequence!(execute) -> (then_0, then_1, then_2:Exec, then_3:Exec)
   > n3 = IsDroneCam!(execute, self, CamType:ESpectatorCameraType Enum=Get Loc_CamType() -> (Loc_CamType:ESpectatorCameraType Enum=ThirdPerson)) -> (then, IsDroneCam:bool) {is=title K2Node_CallFunction}
   > SetTabEnabled(
    self=Get Tabs!(self:WBP Replay Tabs Object Reference=Get ReplayTabs()) -> (Tabs:WBP UIKit Tab Group Object Reference)
    TabNameID=DRONE
    bEnable=@n3
   )
   > n4 = Branch(Condition=@n3)
   n5 = Cast To CommonActivatablePanelLegacy!(execute=@n4.else, Object=GetActiveWidget(self=Get ViewSettingsWidgetSwitcher())) -> (then, CastFailed, AsCommon Activatable Panel Legacy:Common Activatable Panel Legacy Object Reference) {is=title K2Node_DynamicCast}
   n6 = GetActiveTab!(self:Common Tab List Widget Base Object Reference=Get Tabs(self=Get ReplayTabs())) -> (ReturnValue:Name=None)
   > Branch(Condition=AND Boolean(A=IsActivated(self=@n5), B="Equal (Name)"!(A:Name=@n6, B:Name=DRONE) -> (ReturnValue:bool) {is=title K2Node_CallFunction}))
   > SwitchToValidTab()
   n7 = "Not Equal (Enum)"!(A:ESpectatorCameraType Enum=Get Loc_CamType() -> (Loc_CamType=ThirdPerson), B:ESpectatorCameraType Enum=Gameplay) -> (ReturnValue:bool)
   SetTabEnabled(execute=@n2.then_1, self=Get Tabs(self=Get ReplayTabs()), TabNameID=LENS, bEnable=@n7)
   > n8 = Branch(Condition=@n7)
   n9 = Cast To CommonActivatablePanelLegacy(execute=@n8.else, Object=GetActiveWidget(self=Get ViewSettingsWidgetSwitcher()))
   n10 = GetActiveTab(self=Get Tabs(self=Get ReplayTabs())) -> (ReturnValue=None)
   > Branch(Condition=AND Boolean(A=IsActivated(self=@n9), B="Equal (Name)"(A=@n10, B=LENS)))
   > SelectTabByID!(
    execute
    self:Common Tab List Widget Base Object Reference=Get Tabs(self=Get ReplayTabs())
    TabNameID:Name=GAME
    bSuppressClickFeedback:bool=true
   ) -> (then, ReturnValue:bool)
   n11 = Cast To CommonActivatablePanelLegacy(execute=@n2.then_2, Object=GetActiveWidget(self=Get ViewSettingsWidgetSwitcher()))
   n12 = GetActiveTab(self=Get Tabs(self=Get ReplayTabs())) -> (ReturnValue=None)
   n13 = "Equal (Enum)"(A:ESpectatorCameraType Enum=Get Loc_CamType() -> (Loc_CamType=ThirdPerson), B:ESpectatorCameraType Enum=Gameplay)
   n14 = "Equal (Enum)"(A=Get Loc_CamType() -> (Loc_CamType=ThirdPerson), B=ThirdPerson)
   > Branch(Condition=AND Boolean!(A=IsActivated(self=@n11), B="Equal (Name)"(A=@n12, B=DRONE), C:bool=OR Boolean(A=@n14, B=@n13)) -> (ReturnValue))
   > SelectTabByID(self=Get Tabs(self=Get ReplayTabs()), TabNameID=GAME, bSuppressClickFeedback=true)
   SetTabEnabled(
    execute=@n2.then_3
    self=Get Tabs(self=Get ReplayTabs())
    TabNameID=DYNAMIC_CAMERA
    bEnable="Equal (Enum)"(A=Get Loc_CamType() -> (Loc_CamType=ThirdPerson), B=DroneFreeDynamic)
   )
  ]}
 RemoveContentFromWidgetHost {nodes [
   Remove Content From Host Widget!() -> (
    then
    WidgetHostTag:"Gameplay Tag Structure (by ref)"
    ContentWidget:Widget Object Reference
    RemovalType:EWidgetSlottingRemovalType Enum=Unparent
   ) {is=title K2Node_FunctionEntry}
   > Return Node!(execute, ReturnValue:bool=false)
  ]}
 SetCollapseInputAction {nodes [
   SetCollapseInputAction() {is=title K2Node_FunctionEntry}
   n1 = Get Prompt!(self:WBP UIKit Button CTA Base Object Reference=Get Button_Collapse()) -> (Prompt:Grid Panel Object Reference)
   > SetGridSlotColumnAndRow!(execute, Widget:Widget Object Reference=@n1, Row:int=2, Row Span:int=0, Column:int=3, Column Span:int=0) -> (then)
   > SetGridSlotAlignment!(
    execute
    Widget:Widget Object Reference=@n1
    Horizontal:EHorizontalAlignment Enum=HAlign_Center
    Vertical:EVerticalAlignment Enum=VAlign_Center
   ) -> (then)
   > SetGridSlotPadding!(execute, Widget:Widget Object Reference=@n1, Padding_Left:float=0, Padding_Top:float=0, Padding_Right:float=0, Padding_Bottom:float=0) -> (then)
   > SetGridSlotNudge!(execute, Widget:Widget Object Reference=@n1, Nudge_X:double=0, Nudge_Y:double=-200) -> (then)
   n2 = Get Texts!(self:WBP UIKit Button CTA Base Object Reference=Get Button_Collapse()) -> (Texts:Vertical Box Object Reference)
   > SetGridSlotColumnAndRow(Widget=@n2, Row=1, Row Span=0, Column=3, Column Span=0)
   > SetGridSlotAlignment(Widget=@n2, Horizontal=HAlign_Center, Vertical=VAlign_Center)
  ]}
 SwitchToValidTab {nodes [
   SwitchToValidTab() {is=title K2Node_FunctionEntry}
   n1 = GetActiveTab(self=Get Tabs(self=Get ReplayTabs())) -> (ReturnValue=None)
   n2 = Get CurrentCameraType(self=Get SpectatorController()) -> (CurrentCameraType=ThirdPerson)
   > n3 = Branch(Condition=AND Boolean!(A="Equal (Enum)"(A=@n2, B=Gameplay), B="Equal (Name)"(A=@n1, B=LENS)) -> (ReturnValue))
   > SelectTabByID(self=Get Tabs(self=Get ReplayTabs()), TabNameID=GAME, bSuppressClickFeedback=true)
   ActivateWidget(execute=@n3.else, self:Common Widget Switcher Legacy Object Reference=Get ViewSettingsWidgetSwitcher())
  ]}
 UpdateCoachMark {nodes [
   n1 = UpdateCoachMark!() -> (then, Header:Text, Description:Text) {is=title K2Node_FunctionEntry}
   n2 = Get CoachMark()
   > SetTextHeader!(execute, self:WBP UIKit Coach Mark Base Object Reference=@n2, Text:Text=@n1) -> (then, Empty:bool)
   > SetTextDescription!(execute, self:WBP UIKit Coach Mark Base Object Reference=@n2, Text:Text=@n1.Description) -> (then, Empty:bool)
   > Branch(Condition=Get bInfoSettingTurnedOn())
   > SetVisibility(self=Get CoachMark(), InVisibility=SelfHitTestInvisible)
  ]}
]
variables [
 bCollapsed {type=boolean}
 bCurrentDescriptionEmpty {type=boolean}
 bInfoSettingTurnedOn {type=boolean}
 bIsBlockedHoldEnd {type=boolean}
 bUseHover {type=boolean, default=true}
 ReplaySettingTabSlotTag {type=GameplayTag, default {tagName=DynamicUI.HUD.Fort.Athena.ReplaySettingTab.Slot}}
 ReplayTestingSettingTabSlotTag {type=GameplayTag, default {tagName=DynamicUI.HUD.Fort.Athena.ReplayTestingSettingTab.Slot}}
 SpectatorController {type=FortPlayerControllerSpectating, default=null}
 TabNames {type="array<string>", default [LENS,DRONE,WORLD,GAME,INTERFACE]}
]
widgets {
 info {inheritedWidgetCount=0, parentClass=/Script/FortniteUI.FortBroadcastViewSettings, rootWidgetClass=&Overlay, widgetCount=36}
 widgets [
  Overlay_1 {widgetClassPath=&Overlay, widget=Overlay_1, properties {bOverride_Cursor=true}}
  HeightOverride {
   widgetClassPath=&SizeBox
   parent=Overlay_1
   widget=HeightOverride
   properties {bOverride_HeightOverride=true, heightOverride=500, maxDesiredHeight=500}
   slot {class=OverlaySlot, properties {horizontalAlignment=HAlign_Fill, verticalAlignment=VAlign_Fill}}
  }
  HorizontalBox_2 {widgetClassPath=&HorizontalBox, parent=HeightOverride, widget=HorizontalBox_2, slot {class=SizeBoxSlot}}
  ReplayTabs {
   widgetClassPath=/FortniteCoreUI/UI/Replay/WBP_Replay_Tabs
   bIsVariable=true
   parent=HorizontalBox_2
   widget=ReplayTabs
   slot {class=HorizontalBoxSlot, properties {padding=<ltrb 0 0 10 0>}}
  }
  ReplayTabs_Extension {
   widgetClassPath=/FortUILibrary/UIKit/Tabs/WBP_UIKit_TabGroup
   bIsVariable=true
   parent=HorizontalBox_2
   widget=ReplayTabs_Extension
   properties {
    animateReservedPromptSpace=false
    arrowButtonsAllowWrap=false
    bIsFocusable=true
    backplateBrightness=Dark
    backplateClass=null
    buttonClass=/FortniteCoreUI/UI/Replay/WBP_Replay_TabButton.WBP_Replay_TabButton_C
    hidePrompts=true
    initialSelectedTabId=0
    initialSelectedTabIndex=0
    isNaturalNavigation=true
    isVertical=true
    promptsBackplateClass=null
    reservedPromptSpace=65
    tabGroupHorizontalAlignment=HAlign_Center
    tabSpacing=21
    timeToHidePrompts=0
    timeToShowPrompts=0
    useBackplate=false
    useInitialTabs=true
    usePromptBackplate=false
    visibility=Collapsed
    buttonPadding=<ltrb 5 0 5 0>
    iconSize=<36 36>
    initialTabsData [
     {
      bEnableTabId=true
      tabId=LIGHTING
      tabViewModel=&self:WidgetTree.ReplayTabs_Extension.UIKitSelectableButtonViewModel_0
     }
     {
      bEnableTabId=true
      tabId=DEVELOPMENT
      tabViewModel=&self:WidgetTree.ReplayTabs_Extension.UIKitSelectableButtonViewModel_4
     }
     {
      bEnableTabId=true
      tabId=FOG SHEETS
      tabViewModel=&self:WidgetTree.ReplayTabs_Extension.UIKitSelectableButtonViewModel_2
     }
     {
      bEnableTabId=true
      tabId=ANIME MATERIAL EDITOR
      tabViewModel=&self:WidgetTree.ReplayTabs_Extension.UIKitSelectableButtonViewModel_1
     }
     {
      bEnableTabId=true
      tabId=CAMERA SHAKE
      tabViewModel=&self:WidgetTree.ReplayTabs_Extension.UIKitSelectableButtonViewModel_3
     }
     {
      bEnableTabId=true
      tabId=FACIAL ANIMATION
      tabViewModel=&self:WidgetTree.ReplayTabs_Extension.UIKitSelectableButtonViewModel_5
     }
     {
      bEnableTabId=true
      tabId=CHARACTER CUSTOMIZATION
      tabViewModel=&self:WidgetTree.ReplayTabs_Extension.UIKitSelectableButtonViewModel_6
     }
     {
      bEnableTabId=true
      tabId=STYLE
      tabViewModel=&self:WidgetTree.ReplayTabs_Extension.UIKitSelectableButtonViewModel_7
     }
     {
      bEnableTabId=true
      tabId=DYNAMIC_CAMERA
      tabViewModel=&self:WidgetTree.ReplayTabs_Extension.UIKitSelectableButtonViewModel_8
     }
     {
      bEnableTabId=true
      tabId=COLORCORRECT
      tabViewModel=&self:WidgetTree.ReplayTabs_Extension.UIKitSelectableButtonViewModel_9
     }
    ]
    minButtonSize=<58 58>
    tabGroupPadding=<ltrb 15 21 15 21>
   }
   slot {class=HorizontalBoxSlot}
  }
  Overlay_Settings {
   widgetClassPath=&Overlay
   bIsVariable=true
   parent=HorizontalBox_2
   widget=Overlay_Settings
   slot {class=HorizontalBoxSlot, properties {padding=<ltrb 0 0 10 0>}}
  }
  VerticalBox_3 {
   widgetClassPath=&VerticalBox
   parent=Overlay_Settings
   widget=VerticalBox_3
   slot {class=OverlaySlot, properties {horizontalAlignment=HAlign_Fill, verticalAlignment=VAlign_Fill}}
  }
  Overlay_0 {widgetClassPath=&Overlay, parent=VerticalBox_3, widget=Overlay_0, slot {class=VerticalBoxSlot}}
  Settings_Backplate {
   widgetClassPath=&WBP_UIKit_Backplate
   parent=Overlay_0
   widget=Settings_Backplate
   properties {backplateMaterial=/FortniteCoreUI/UI/Replay/Materials/M_Replay_Backplate_Inst, cornerRadius=Tab}
   slot {class=OverlaySlot, properties {horizontalAlignment=HAlign_Fill, verticalAlignment=VAlign_Fill}}
  }
  ViewSettingsWidgetSwitcher {
   widgetClassPath=&CommonWidgetSwitcherLegacy
   bIsVariable=true
   parent=Overlay_0
   widget=ViewSettingsWidgetSwitcher
   properties {
    bOutroPanelBelow=false
    navigation=&self:WidgetTree.ViewSettingsWidgetSwitcher.WidgetNavigation_0
    transitionDuration=0
   }
   slot {class=OverlaySlot, properties {horizontalAlignment=HAlign_Fill, verticalAlignment=VAlign_Fill, padding=<ltrb 10 15 10 15>}}
  }
  LensSettingsTab {
   widgetClassPath=/FortniteCoreUI/UI/Replay/FortReplayLensSettingsTab
   bIsVariable=true
   parent=ViewSettingsWidgetSwitcher
   widget=LensSettingsTab
   properties {tabNameId=LENS}
   slot {class=WidgetSwitcherSlot}
  }
  DroneSettingsTab {
   widgetClassPath=&/FortBroadcastAdvancedSettingsTab
   bIsVariable=true
   parent=ViewSettingsWidgetSwitcher
   widget=DroneSettingsTab
   properties {tabNameId=DRONE}
   slot {class=WidgetSwitcherSlot}
  }
  WorldSettingsTab {
   widgetClassPath=&/FortBroadcastAmbientSettingsTab
   bIsVariable=true
   parent=ViewSettingsWidgetSwitcher
   widget=WorldSettingsTab
   properties {tabNameId=WORLD}
   slot {class=WidgetSwitcherSlot}
  }
  GameSettingsTab {
   widgetClassPath=&/FortBroadcastCameraSettingsTab
   bIsVariable=true
   parent=ViewSettingsWidgetSwitcher
   widget=GameSettingsTab
   properties {tabNameId=GAME}
   slot {class=WidgetSwitcherSlot}
  }
  InterfaceSettingsTab {
   widgetClassPath=&/FortBroadcastNameplateSettingsTab
   bIsVariable=true
   parent=ViewSettingsWidgetSwitcher
   widget=InterfaceSettingsTab
   properties {tabNameId=INTERFACE}
   slot {class=WidgetSwitcherSlot}
  }
  FooterBoxButtons {
   widgetClassPath=&HorizontalBox
   bIsVariable=true
   parent=VerticalBox_3
   widget=FooterBoxButtons
   slot {class=VerticalBoxSlot, properties {horizontalAlignment=HAlign_Right, padding=<ltrb 0 10 0 0>}}
  }
  Button_ResetHold {
   widgetClassPath=/FortniteCoreUI/UI/Replay/WBP_Replay_Button
   bIsVariable=true
   parent=FooterBoxButtons
   widget=Button_ResetHold
   properties {
    bRequiresHold=true
    text="Reset (Hold)"
    largePrimaryFont {
     bForceMonospaced=false
     bMaterialIsStencil=false
     fontMaterial=null
     fontObject=&HeadingNow
     letterSpacing=10
     monospacedWidth=1
     size=22.5
     skewAmount=0
     typefaceFontName=76 Bold
     outlineSettings {
      bApplyOutlineToDropShadows=false
      bMiteredCorners=false
      bSeparateFillAlpha=false
      outlineMaterial=null
      outlineSize=0
      outlineColor=<rgba 0 0 0 1>
     }
    }
    triggeringInputAction {dataTable=/FortUILibrary/Input/ReplayInputActionDataTable, rowName=ResetViewSettings}
   }
   slot {class=HorizontalBoxSlot}
  }
  Spacer_3 {widgetClassPath=&Spacer, parent=FooterBoxButtons, widget=Spacer_3, properties {size=<18 0>}, slot {class=HorizontalBoxSlot}}
  Button_Info {
   widgetClassPath=&WBP_Replay_Button
   bIsVariable=true
   parent=FooterBoxButtons
   widget=Button_Info
   properties {
    text=Info
    largePrimaryFont {
     bForceMonospaced=false
     bMaterialIsStencil=false
     fontMaterial=null
     fontObject=&HeadingNow
     letterSpacing=10
     monospacedWidth=1
     size=22.5
     skewAmount=0
     typefaceFontName=76 Bold
     outlineSettings {
      bApplyOutlineToDropShadows=false
      bMiteredCorners=false
      bSeparateFillAlpha=false
      outlineMaterial=null
      outlineSize=0
      outlineColor=<rgba 0 0 0 1>
     }
    }
    triggeringInputAction {dataTable=&ReplayInputActionDataTable, rowName=ToggleSettingInfo}
   }
   slot {class=HorizontalBoxSlot}
  }
  Spacer_1 {widgetClassPath=&Spacer, parent=FooterBoxButtons, widget=Spacer_1, properties {size=<18 0>}, slot {class=HorizontalBoxSlot}}
  Button_Reset {
   widgetClassPath=&WBP_Replay_Button
   bIsVariable=true
   parent=FooterBoxButtons
   widget=Button_Reset
   properties {
    bRequiresHold=true
    text=Reset
    largePrimaryFont {
     bForceMonospaced=false
     bMaterialIsStencil=false
     fontMaterial=null
     fontObject=&HeadingNow
     letterSpacing=10
     monospacedWidth=1
     size=22.5
     skewAmount=0
     typefaceFontName=76 Bold
     outlineSettings {
      bApplyOutlineToDropShadows=false
      bMiteredCorners=false
      bSeparateFillAlpha=false
      outlineMaterial=null
      outlineSize=0
      outlineColor=<rgba 0 0 0 1>
     }
    }
    triggeringInputAction {dataTable=&ReplayInputActionDataTable, rowName=ResetViewSettings}
   }
   slot {class=HorizontalBoxSlot}
  }
  Spacer_93 {widgetClassPath=&Spacer, parent=FooterBoxButtons, widget=Spacer_93, properties {size=<18 0>}, slot {class=HorizontalBoxSlot}}
  Button_Close {
   widgetClassPath=&WBP_Replay_Button
   bIsVariable=true
   parent=FooterBoxButtons
   widget=Button_Close
   properties {
    text=Close
    largePrimaryFont {
     bForceMonospaced=false
     bMaterialIsStencil=false
     fontMaterial=null
     fontObject=&HeadingNow
     letterSpacing=10
     monospacedWidth=1
     size=22.5
     skewAmount=0
     typefaceFontName=76 Bold
     outlineSettings {
      bApplyOutlineToDropShadows=false
      bMiteredCorners=false
      bSeparateFillAlpha=false
      outlineMaterial=null
      outlineSize=0
      outlineColor=<rgba 0 0 0 1>
     }
    }
    triggeringInputAction {dataTable=&ReplayInputActionDataTable, rowName=ToggleViewSettings}
   }
   slot {class=HorizontalBoxSlot}
  }
  BoxPrompts {
   widgetClassPath=&Overlay
   bIsVariable=true
   parent=VerticalBox_3
   widget=BoxPrompts
   slot {class=VerticalBoxSlot, properties {horizontalAlignment=HAlign_Right, padding=<ltrb 0 10 0 0>}}
  }
  PromptBacking {
   widgetClassPath=&WBP_UIKit_Backplate
   bIsVariable=true
   parent=BoxPrompts
   widget=PromptBacking
   properties {backplateMaterial=&M_Replay_Backplate_Inst, brightness=Dark, cornerRadius=Circle}
   slot {class=OverlaySlot, properties {horizontalAlignment=HAlign_Fill, verticalAlignment=VAlign_Fill}}
  }
  FooterBoxPrompts {
   widgetClassPath=&HorizontalBox
   bIsVariable=true
   parent=BoxPrompts
   widget=FooterBoxPrompts
   slot {class=OverlaySlot, properties {horizontalAlignment=HAlign_Center, verticalAlignment=VAlign_Fill}}
  }
  Prompt_NavigateTabs {
   widgetClassPath=/FortniteCoreUI/UI/Replay/WBP_Replay_KeyBinding
   bIsVariable=true
   parent=FooterBoxPrompts
   widget=Prompt_NavigateTabs
   properties {
    axisScale01=1
    axisScale02=-1
    boundActionName01=SpectorRackOrDutchOrZoomAdjust
    boundActionName02=SpectorRackOrDutchOrZoomAdjust
    keybindsVisible=2
    title=Switch Tabs
   }
   slot {class=HorizontalBoxSlot}
  }
  Prompt_Info {
   widgetClassPath=&WBP_Replay_KeyBinding
   bIsVariable=true
   parent=FooterBoxPrompts
   widget=Prompt_Info
   properties {
    actionWidgetBoundAction=PlayPause
    axisScale01=1
    axisScale02=-1
    inputMode=Action Widget
    keybindsVisible=0
    title=Info
    visibility=Collapsed
   }
   slot {class=HorizontalBoxSlot}
  }
  Prompt_Reset {
   widgetClassPath=&WBP_Replay_KeyBinding
   bIsVariable=true
   parent=FooterBoxPrompts
   widget=Prompt_Reset
   properties {boundActionName01=HoverDrone_DecreaseSpeedOrDutchAngle, title="Reset (Hold)"}
   slot {class=HorizontalBoxSlot}
  }
  Prompt_Exit {
   widgetClassPath=&WBP_Replay_KeyBinding
   bIsVariable=true
   parent=FooterBoxPrompts
   widget=Prompt_Exit
   properties {boundActionName01=HoverDrone_IncreaseSpeedOrDutchAngle, title=Exit}
   slot {class=HorizontalBoxSlot}
  }
  Button_Collapse {
   widgetClassPath=&WBP_Replay_Button
   bIsVariable=true
   parent=HorizontalBox_2
   widget=Button_Collapse
   properties {
    buttonSize=Medium
    icon=&T_UI_Icon_Global_Chevron_Left
    iconMaterial=/FortUILibrary/UIKit/Buttons/MaterialInstances/GenericButtons/MI_UIKit_GenericIcon_Quiet
    text=""
    largePrimaryFont {
     bForceMonospaced=false
     bMaterialIsStencil=false
     fontMaterial=null
     fontObject=&HeadingNow
     letterSpacing=10
     monospacedWidth=1
     size=22.5
     skewAmount=0
     typefaceFontName=76 Bold
     outlineSettings {
      bApplyOutlineToDropShadows=false
      bMiteredCorners=false
      bSeparateFillAlpha=false
      outlineMaterial=null
      outlineSize=0
      outlineColor=<rgba 0 0 0 1>
     }
    }
    triggeringInputAction {dataTable=&ReplayInputActionDataTable, rowName=CollapseSettings}
   }
   slot {class=HorizontalBoxSlot}
  }
  SizeBox_1 {
   widgetClassPath=&SizeBox
   parent=HorizontalBox_2
   widget=SizeBox_1
   properties {bOverride_WidthOverride=true, heightOverride=280, maxDesiredHeight=300, minDesiredHeight=280, widthOverride=525}
   slot {
    class=HorizontalBoxSlot
    properties {horizontalAlignment=HAlign_Left, verticalAlignment=VAlign_Top, padding=<ltrb 25 15 0 0>, size {sizeRule=Fill, value=1}}
   }
  }
  CoachMark {
   widgetClassPath=/FortniteCoreUI/UI/Replay/WBP_Replay_CoachMark
   bIsVariable=true
   parent=SizeBox_1
   widget=CoachMark
   properties {
    textDescription="When enabled, the camera's settings will automatically adjust to optimize the brightness of the image."
    textHeader=Auto Exposure
   }
   slot {class=SizeBoxSlot}
  }
  Button_Info_Hidden {
   widgetClassPath=&WBP_Replay_Button
   bIsVariable=true
   parent=HorizontalBox_2
   widget=Button_Info_Hidden
   properties {
    buttonSize=Medium
    renderOpacity=0
    text=""
    largePrimaryFont {
     bForceMonospaced=false
     bMaterialIsStencil=false
     fontMaterial=null
     fontObject=&HeadingNow
     letterSpacing=10
     monospacedWidth=1
     size=22.5
     skewAmount=0
     typefaceFontName=76 Bold
     outlineSettings {
      bApplyOutlineToDropShadows=false
      bMiteredCorners=false
      bSeparateFillAlpha=false
      outlineMaterial=null
      outlineSize=0
      outlineColor=<rgba 0 0 0 1>
     }
    }
    triggeringInputAction {dataTable=&ReplayInputActionDataTable, rowName=ToggleSettingInfo}
   }
   slot {class=HorizontalBoxSlot}
  }
  Button_Reset_Hidden {
   widgetClassPath=&WBP_Replay_Button
   bIsVariable=true
   parent=HorizontalBox_2
   widget=Button_Reset_Hidden
   properties {
    bRequiresHold=true
    buttonSize=Medium
    renderOpacity=0
    text=""
    largePrimaryFont {
     bForceMonospaced=false
     bMaterialIsStencil=false
     fontMaterial=null
     fontObject=&HeadingNow
     letterSpacing=10
     monospacedWidth=1
     size=22.5
     skewAmount=0
     typefaceFontName=76 Bold
     outlineSettings {
      bApplyOutlineToDropShadows=false
      bMiteredCorners=false
      bSeparateFillAlpha=false
      outlineMaterial=null
      outlineSize=0
      outlineColor=<rgba 0 0 0 1>
     }
    }
    triggeringInputAction {dataTable=&ReplayInputActionDataTable, rowName=ResetViewSettings}
   }
   slot {class=HorizontalBoxSlot}
  }
  Button_Exit_Hidden {
   widgetClassPath=&WBP_Replay_Button
   bIsVariable=true
   parent=HorizontalBox_2
   widget=Button_Exit_Hidden
   properties {
    buttonSize=Medium
    renderOpacity=0
    text=""
    largePrimaryFont {
     bForceMonospaced=false
     bMaterialIsStencil=false
     fontMaterial=null
     fontObject=&HeadingNow
     letterSpacing=10
     monospacedWidth=1
     size=22.5
     skewAmount=0
     typefaceFontName=76 Bold
     outlineSettings {
      bApplyOutlineToDropShadows=false
      bMiteredCorners=false
      bSeparateFillAlpha=false
      outlineMaterial=null
      outlineSize=0
      outlineColor=<rgba 0 0 0 1>
     }
    }
    triggeringInputAction {dataTable=&ReplayInputActionDataTable, rowName=CloseViewSettings}
   }
   slot {class=HorizontalBoxSlot}
  }
 ]
}
