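# Replay spectator notification widget that shows a short text message on a backplate, then fades it out.
# Construct binds to spectating controller notification delegates, and show calls set the text and display time from settings or a default.
# Tick counts down, then interpolates border opacity to zero.
# HUD-visibility messages and override settings can appear while the widget is hidden, and visibility changes wait in PendingVisibility until those messages finish fading.
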WidgetBlueprint /FortniteCoreUI/UI/Replay/FortReplayNotification
graphs [
 EventGraph {nodes [
   n1 = OnShowNotificationsEnabledChanged_Event_0!() -> (OutputDelegate, then, bShouldShowNotifications:bool=~) {is=title K2Node_CustomEvent}
   > n2 = Branch(Condition=@n1.bShouldShowNotifications)
   Set RemainingTime(execute=@n2.else, RemainingTime:double=0) -> (Output_Get:double)
   n3 = Event Tick()
   > TickRemainingTime!(execute, self, DeltaTime:double=@n3.InDeltaTime) -> (then)
   > TickFade!(execute, self, DeltaTime:double=@n3.InDeltaTime) -> (then)
   Event Construct()
   > n4 = Cast To FortPlayerControllerSpectating!(execute, Object=GetPlayerController!(PlayerIndex:int=0) -> (ReturnValue:Player Controller Object Reference)) -> (then, CastFailed, AsFort Player Controller Spectating:Fort Player Controller Spectating Object Reference) {is=title K2Node_DynamicCast}
   > Bind Event to On Show Notification(self:Fort Player Controller Spectating Object Reference=@n4, Delegate=Create Event()) {is=title K2Node_AddDelegate}
   > Assign On Show Notifications Enabled Changed!(execute:Exec, self:Fort Player Controller Spectating Object Reference=@n4, Delegate:"Delegate (by ref)"=@n1) -> (then:Exec) {is=title K2Node_AssignDelegate}
  ]}
 GetVisibilityText {nodes [
   n1 = GetVisibilityText!() -> (then, Visibility:EHudVisibilityState Enum=FullyVisible) {is=title K2Node_FunctionEntry}
   > Return Node!(
    execute
    Text:Text=Select!(
     FullyVisible:Text=Get FullyVisible() -> (FullyVisible:Text)
     FullyHidden:Text=Get FullyHidden() -> (FullyHidden:Text)
     GameOnly:Text=Get GameOnly() -> (GameOnly:Text)
     ReplayOnly:Text=Get ReplayOnly() -> (ReplayOnly:Text)
     Index:EHudVisibilityState Enum=@n1
    ) -> (ReturnValue:Text)
   )
  ]}
 IsVisibilityNotification {nodes [
   n1 = IsVisibilityNotification!() -> (then, Notification:Text) {is=title K2Node_FunctionEntry}
   > n2 = For Each Loop(Array:Array of Texts=Make Array!("[0]":Text=Get FullyVisible(), "[1]":Text=Get FullyHidden(), "[2]":Text=Get GameOnly(), "[3]":Text=Get ReplayOnly()) -> (Array:Array of Texts)) -> (Array Element:Text)
   Return Node!(execute=@n2.Completed, IsVisibilityNotification:bool=false)
   > Branch(Condition="Equal Exactly (Text)"!(A:"Text (by ref)"=@n2, B:"Text (by ref)"=@n1) -> (ReturnValue:bool))
   > Return Node(IsVisibilityNotification=true)
  ]}
 SetNotificationVisibility {nodes [
   n1 = SetNotificationVisibility!() -> (then, Visibility:ESlateVisibility Enum=Visible) {is=title K2Node_FunctionEntry}
   > Set PendingVisibility(PendingVisibility:ESlateVisibility Enum=@n1) -> (Output_Get:ESlateVisibility Enum=Visible)
   n2 = IsVisibilityNotification!(Notification:Text="GetText (Text)"!(self:Text Object Reference=Get Notification() -> (Notification:Common Text Object Reference)) -> (ReturnValue:Text)) -> (IsVisibilityNotification:bool) {is=title K2Node_CallFunction}
   n3 = GetVisibility!(self:Widget Object Reference) -> (ReturnValue:ESlateVisibility Enum=Visible)
   > Branch(Condition=AND Boolean(A="Not Equal (Enum)"!(A:ESlateVisibility Enum=@n1, B:ESlateVisibility Enum=@n3) -> (ReturnValue:bool), B=NOT Boolean(A=@n2)))
   > SetVisibility(InVisibility=@n1)
   > Set RemainingTime(RemainingTime=0)
   > Set IsFading(IsFading:bool=false) -> (Output_Get:bool)
   > SetVisibility(self=Get Border() -> (Border:Overlay Object Reference), InVisibility=Hidden)
  ]}
 ShowNotification {nodes [
   n1 = ShowNotification!() -> (then, NotificationText:"Text (by ref)") {is=title K2Node_FunctionEntry}
   > ShowNotificationWIthSettings!(
    execute
    self
    NotificationText:"Text (by ref)"=@n1
    NotificationSettings:"Notification UISettings Structure (by ref)"=Make Notification UISettings!() -> (NotificationUISettings:Notification UISettings Structure) {is=title K2Node_MakeStruct}
   ) -> (then)
  ]}
 ShowNotificationWIthSettings {nodes [
   n1 = ShowNotificationWIthSettings!() -> (then, NotificationText:"Text (by ref)", NotificationSettings:"Notification UISettings Structure (by ref)") {is=title K2Node_FunctionEntry}
   n2 = IsVisibilityNotification(Notification="GetText (Text)"(self=Get Notification()))
   > n3 = Branch(Condition=AND Boolean(A=IsVisible(), B=NOT Boolean(A=@n2)))
   "SetText (Text)"(execute=[@n7.then, @n3.then], self=Get Notification(), InText=@n1)
   n4 = Get Border()
   > SetRenderOpacity(self=@n4, InOpacity=1)
   > SetVisibility(self=@n4, InVisibility=Visible)
   > Set IsFading(IsFading=false)
   n5 = Break Notification UISettings!(NotificationUISettings:"Notification UISettings Structure (by ref)"=Get NotificationSettings() -> (NotificationSettings:"Notification UISettings Structure (by ref)")) -> (DisplayTime:float=~, bShouldOverrideVisibilitySettings:bool=~) {is=title K2Node_BreakStruct}
   > n6 = Branch(Condition="float < float"!(A:double=@n5, B:double=0) -> (ReturnValue:bool) {is=title K2Node_CallFunction})
   > Set RemainingTime(RemainingTime=Get TimeToDisplay() -> (TimeToDisplay:double))
   Set RemainingTime(execute=@n6.else, RemainingTime=@n5)
   Branch(
    execute=@n3.else
    Condition=OR Boolean(
     A=IsVisibilityNotification(Notification=@n1)
     B=Break Notification UISettings(NotificationUISettings=@n1.NotificationSettings) -> (DisplayTime=~, bShouldOverrideVisibilitySettings=~).bShouldOverrideVisibilitySettings
    )
   )
   > n7 = SetVisibility(InVisibility=SelfHitTestInvisible)
  ]}
 StartFading {nodes [
   StartFading() {is=title K2Node_FunctionEntry}
   > Set IsFading(IsFading=true)
  ]}
 TickFade {nodes [
   n1 = TickFade!() -> (then, DeltaTime:double=~) {is=title K2Node_FunctionEntry}
   > Branch(Condition=Get IsFading() -> (IsFading:bool))
   n2 = Get Border()
   n3 = FInterpTo(
    Current=GetRenderOpacity!(self:Widget Object Reference=@n2) -> (ReturnValue:float)
    Target=0
    DeltaTime=@n1
    InterpSpeed=Get FadeInterpSpeed() -> (FadeInterpSpeed:double)
   )
   n4 = "float < float"(A=@n3, B=0.025)
   > SetRenderOpacity(self=@n2, InOpacity=SelectFloat(A=0, B=@n3, bPickA=@n4))
   > Branch(Condition=@n4)
   > SetVisibility(self=@n2, InVisibility=Hidden)
   > Set IsFading(IsFading=false)
   n5 = Get Notification()
   > Branch(Condition=IsVisibilityNotification(Notification="GetText (Text)"(self=@n5)))
   > "SetText (Text)"(self=Get Notification())
   > SetNotificationVisibility!(execute, self, Visibility:ESlateVisibility Enum=Get PendingVisibility() -> (PendingVisibility:ESlateVisibility Enum=Visible)) -> (then) {is=title K2Node_CallFunction}
  ]}
 TickRemainingTime {nodes [
   n1 = TickRemainingTime!() -> (then, DeltaTime:double=~) {is=title K2Node_FunctionEntry}
   > Branch(Condition="float > float"(A=Get RemainingTime() -> (RemainingTime:double), B=0) {is=title K2Node_CallFunction})
   > n2 = Set RemainingTime(RemainingTime=float - float(A=Get RemainingTime(), B=@n1) {is=title K2Node_CallFunction})
   > Branch(Condition="float <= float"!(A:double=@n2, B:double=0) -> (ReturnValue:bool) {is=title K2Node_CallFunction})
   > Set IsFading(IsFading=true)
  ]}
]
variables [
 FadeInterpSpeed {type=number, default=5}
 FullyHidden {type=string, default=Fully Hidden HUD}
 FullyVisible {type=string, default=Fully Visible HUD}
 GameOnly {type=string, default=Game Only HUD}
 IsFading {type=boolean}
 PendingVisibility {type=ESlateVisibility, default=Hidden}
 RemainingTime {type=number}
 ReplayOnly {type=string, default=Replay Only HUD}
 TimeToDisplay {type=number, default=2}
]
widgets {
 info {inheritedWidgetCount=0, parentClass=&CommonUserWidget, rootWidgetClass=&SizeBox, widgetCount=4}
 widgets [
  SizeBox_0 {
   widgetClassPath=&SizeBox
   widget=SizeBox_0
   properties {bOverride_MaxDesiredWidth=true, maxDesiredWidth=350, minDesiredWidth=195.3264, widthOverride=350}
  }
  Border {
   widgetClassPath=&Overlay
   bIsVariable=true
   parent=SizeBox_0
   widget=Border
   properties {bOverride_Cursor=true, visibility=Hidden}
   slot {class=SizeBoxSlot}
  }
  WBP_UIKit_Backplate {
   widgetClassPath=&WBP_UIKit_Backplate
   bIsVariable=true
   parent=Border
   widget=WBP_UIKit_Backplate
   properties {backplateMaterial=/FortniteCoreUI/UI/Replay/Materials/M_Replay_Backplate_Inst, brightness=Dark}
   slot {class=OverlaySlot, properties {horizontalAlignment=HAlign_Fill, verticalAlignment=VAlign_Fill}}
  }
  Notification {
   widgetClassPath=&CommonTextBlock
   bIsVariable=true
   parent=Border
   widget=Notification
   properties {
    scrollStyle=&TextScrollStyle-Base_C
    text=Notification
    font {
     bForceMonospaced=false
     bMaterialIsStencil=false
     fontMaterial=null
     fontObject=&HeadingNow
     letterSpacing=0
     monospacedWidth=1
     size=15.75
     skewAmount=0
     typefaceFontName=74 Regular
     outlineSettings {
      bApplyOutlineToDropShadows=false
      bMiteredCorners=false
      bSeparateFillAlpha=false
      outlineMaterial=null
      outlineSize=0
      outlineColor=<rgba 0 0 0 1>
     }
    }
    shadowOffset=<0 0>
   }
   slot {class=OverlaySlot, properties {verticalAlignment=VAlign_Center, padding=<ltrb 12 8 12 8>}}
  }
 ]
}
