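# Widget Blueprint for a replay shot title screen, built on the native FortShotTitleScreen class, that shows a line of text over a backdrop.
# An overlay layers a default lobby background image behind the inherited TextBlock_MainText, centered in large wrapped Burbank Black type.
# The elimination-count sentence is static default text in the widget.
# The event graph holds only unconnected PreConstruct, Construct and Tick events, so this asset adds layout and styling only.
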WidgetBlueprint /FortniteCoreUI/UI/Replay/FortShotTitleScreen
graphs [EventGraph {nodes [
   Event PreConstruct()
   Event Construct()
   Event Tick()
  ]}]
variables []
widgets {
 info {inheritedWidgetCount=1, parentClass=/Script/FortniteUI.FortShotTitleScreen, rootWidgetClass=&Overlay, widgetCount=3}
 widgets [
  Overlay_34 {widgetClassPath=&Overlay, widget=Overlay_34}
  Image_88 {
   widgetClassPath=&Image
   bIsVariable=true
   parent=Overlay_34
   widget=Image_88
   properties {brush {
     bIsDynamicallyLoaded=false
     drawAs=Image
     imageType=NoImage
     mirroring=NoMirror
     resourceName=None
     resourceObject=/Game/Environments/FrontEnd/BackPlates/Textures/T_Background_LobbyDefault_SF
     tiling=NoTile
     imageSize=<4096 2048>
     margin=<ltrb 0 0 0 0>
     outlineSettings {
      bUseBrushTransparency=false
      roundingType=HalfHeightRadius
      width=0
      color {colorUseRule=UseColor_Specified, specifiedColor=<rgba 0 0 0 0>}
      cornerRadii=<0 0 0 0>
     }
     tintColor {colorUseRule=UseColor_Specified, specifiedColor=<rgba 1 1 1 1>}
     uVRegion {bIsValid=false, max=<0 0>, min=<0 0>}
    }}
   slot {class=OverlaySlot}
  }
  TextBlock_MainText {
   widgetClassPath=&CommonTextBlock
   bInherited=true
   bIsVariable=true
   parent=Overlay_34
   widget=TextBlock_MainText
   properties {
    justification=Center
    text="You got 32 eliminations this season!"
    wrapTextAt=1478.0941
    font {
     bForceMonospaced=false
     bMaterialIsStencil=false
     fontMaterial=null
     fontObject=&Burbank
     letterSpacing=0
     monospacedWidth=1
     size=58
     skewAmount=0
     typefaceFontName=Burbank-Black
     outlineSettings {
      bApplyOutlineToDropShadows=false
      bMiteredCorners=false
      bSeparateFillAlpha=false
      outlineMaterial=null
      outlineSize=0
      outlineColor=<rgba 0 0 0 1>
     }
    }
    shadowOffset=<0 0>
   }
   slot {class=OverlaySlot, properties {horizontalAlignment=HAlign_Center, verticalAlignment=VAlign_Center, padding=<ltrb 0 0 0 10>}}
  }
 ]
}
