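# Widget Blueprint for the replay "Go to Time" dialog, where a player types a time in seconds to jump to on the timeline.
# It only supplies layout and styling: a backplate, header, range-hint text, right-aligned editable seconds field, and Confirm/Close buttons bound to input actions.
# The event graph is empty, with the hint text, field and both buttons inherited from the native parent FortReplayGoToTimeDialogue.
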WidgetBlueprint /FortniteCoreUI/UI/Replay/GotoTimeDialog
graphs [EventGraph {links [], nodes []}]
variables []
widgets {
 info {inheritedWidgetCount=4, parentClass=/Script/FortniteUI.FortReplayGoToTimeDialogue, rootWidgetClass=&VerticalBox, widgetCount=17}
 widgets [
  VerticalBox_1 {widgetClassPath=&VerticalBox, widget=VerticalBox_1}
  Overlay_0 {widgetClassPath=&Overlay, parent=VerticalBox_1, widget=Overlay_0, slot {class=VerticalBoxSlot}}
  Settings_Backplate {
   widgetClassPath=&WBP_UIKit_Backplate
   parent=Overlay_0
   widget=Settings_Backplate
   properties {backplateMaterial=/FortniteCoreUI/UI/Replay/Materials/M_Replay_Backplate_Inst, cornerRadius=Tab}
   slot {class=OverlaySlot, properties {horizontalAlignment=HAlign_Fill, verticalAlignment=VAlign_Fill}}
  }
  VerticalBox_2 {
   widgetClassPath=&VerticalBox
   parent=Overlay_0
   widget=VerticalBox_2
   slot {class=OverlaySlot, properties {horizontalAlignment=HAlign_Fill, verticalAlignment=VAlign_Fill}}
  }
  HeaderText {
   widgetClassPath=&CommonTextBlock
   bIsVariable=true
   parent=VerticalBox_2
   widget=HeaderText
   properties {
    text=GO TO TIME
    font {
     bForceMonospaced=false
     bMaterialIsStencil=false
     fontMaterial=null
     fontObject=&HeadingNow
     letterSpacing=0
     monospacedWidth=1
     size=32
     skewAmount=0
     typefaceFontName=76 Bold Italic
     outlineSettings {
      bApplyOutlineToDropShadows=false
      bMiteredCorners=false
      bSeparateFillAlpha=false
      outlineMaterial=null
      outlineSize=0
      outlineColor=<rgba 0 0 0 1>
     }
    }
    shadowOffset=<0 0>
   }
   slot {class=VerticalBoxSlot, properties {horizontalAlignment=HAlign_Left, verticalAlignment=VAlign_Top, padding=<ltrb 15 5 0 5>}}
  }
  DescriptionBorder {
   widgetClassPath=&CommonBorder
   parent=VerticalBox_2
   widget=DescriptionBorder
   properties {
    background {
     bIsDynamicallyLoaded=false
     drawAs=Box
     imageType=NoImage
     mirroring=NoMirror
     resourceName=None
     resourceObject=null
     tiling=NoTile
     imageSize=<32 32>
     margin=<ltrb 0 0 0 0>
     outlineSettings {
      bUseBrushTransparency=false
      roundingType=HalfHeightRadius
      width=0
      color {colorUseRule=UseColor_Specified, specifiedColor=<rgba 0 0 0 0>}
      cornerRadii=<0 0 0 0>
     }
     tintColor {colorUseRule=UseColor_Specified, specifiedColor=<rgba 1 1 1 1>}
     uVRegion {bIsValid=false, max=<0 0>, min=<0 0>}
    }
    brushColor=<rgba 0.0157 0.0196 0.0314 0>
    padding=<ltrb 15 20 15 20>
   }
   slot {class=VerticalBoxSlot}
  }
  VerticalBox_0 {
   widgetClassPath=&VerticalBox
   parent=DescriptionBorder
   widget=VerticalBox_0
   slot {class=BorderSlot, properties {padding=<ltrb 15 20 15 20>}}
  }
  Text_HowTo {
   widgetClassPath=&TextBlock
   bInherited=true
   bIsVariable=true
   parent=VerticalBox_0
   widget=Text_HowTo
   properties {
    text="Enter a time (between {0} and {1} seconds) to be taken to that point on the timeline."
    visibility=SelfHitTestInvisible
    wrapTextAt=600
    font {
     bForceMonospaced=false
     bMaterialIsStencil=false
     fontMaterial=null
     fontObject=&HeadingNow
     letterSpacing=0
     monospacedWidth=1
     size=16
     skewAmount=0
     typefaceFontName=74 Regular
     outlineSettings {
      bApplyOutlineToDropShadows=false
      bMiteredCorners=false
      bSeparateFillAlpha=false
      outlineMaterial=null
      outlineSize=0
      outlineColor=<rgba 0 0 0 1>
     }
    }
   }
   slot {class=VerticalBoxSlot, properties {padding=<ltrb 0 0 0 20>}}
  }
  HorizontalBox_2 {
   widgetClassPath=&HorizontalBox
   parent=VerticalBox_0
   widget=HorizontalBox_2
   slot {class=VerticalBoxSlot, properties {horizontalAlignment=HAlign_Center, verticalAlignment=VAlign_Center}}
  }
  SizeBox_0 {
   widgetClassPath=&SizeBox
   parent=HorizontalBox_2
   widget=SizeBox_0
   properties {bOverride_WidthOverride=true, maxDesiredWidth=150, minDesiredWidth=250, widthOverride=150}
   slot {class=HorizontalBoxSlot}
  }
  WhiteStrokeBorder {
   widgetClassPath=&CommonBorder
   parent=SizeBox_0
   widget=WhiteStrokeBorder
   properties {
    style=/FortUILibrary/Styles/Border/Border-1pxOutline.Border-1pxOutline_C
    background {
     bIsDynamicallyLoaded=false
     drawAs=Box
     imageType=NoImage
     mirroring=NoMirror
     resourceName=None
     resourceObject=/FortUILibrary/Styles/Border/Materials/M_Outline_DkBlue
     tiling=NoTile
     imageSize=<16 16>
     margin=<ltrb 0.5 0.5 0.5 0.5>
     outlineSettings {
      bUseBrushTransparency=false
      roundingType=HalfHeightRadius
      width=0
      color {colorUseRule=UseColor_Specified, specifiedColor=<rgba 0 0 0 0>}
      cornerRadii=<0 0 0 0>
     }
     tintColor {colorUseRule=UseColor_Specified, specifiedColor=<rgba 1 1 1 1>}
     uVRegion {bIsValid=false, max=<0 0>, min=<0 0>}
    }
    brushColor=<rgba 1 1 1 0.7>
    padding=<ltrb 15 20 15 20>
   }
   slot {class=SizeBoxSlot}
  }
  EditableText_GoToSeconds {
   widgetClassPath=/Script/UMG.EditableText
   bInherited=true
   parent=WhiteStrokeBorder
   widget=EditableText_GoToSeconds
   properties {
    hintText=0
    justification=Right
    widgetStyle {
     backgroundImageComposing {
      bIsDynamicallyLoaded=false
      drawAs=Border
      imageType=FullColor
      mirroring=NoMirror
      resourceName=../../../Engine/Content/Slate/Old/HyperlinkDotted.png
      resourceObject=null
      tiling=Both
      imageSize=<0 0>
      margin=<ltrb 0 0 0 0.1875>
      outlineSettings {
       bUseBrushTransparency=false
       roundingType=HalfHeightRadius
       width=0
       color {colorUseRule=UseColor_Specified, specifiedColor=<rgba 0 0 0 0>}
       cornerRadii=<0 0 0 0>
      }
      tintColor {colorUseRule=UseColor_Specified, specifiedColor=<rgba 1 1 1 1>}
      uVRegion {bIsValid=false, max=<0 0>, min=<0 0>}
     }
     backgroundImageSelected {
      bIsDynamicallyLoaded=false
      drawAs=Box
      imageType=FullColor
      mirroring=NoMirror
      resourceName=../../../Engine/Content/Slate/Common/EditableTextSelectionBackground.png
      resourceObject=null
      tiling=NoTile
      imageSize=<0 0>
      margin=<ltrb 0.25 0.25 0.25 0.25>
      outlineSettings {
       bUseBrushTransparency=false
       roundingType=HalfHeightRadius
       width=0
       color {colorUseRule=UseColor_Specified, specifiedColor=<rgba 0 0 0 0>}
       cornerRadii=<0 0 0 0>
      }
      tintColor {colorUseRule=UseColor_Specified, specifiedColor=<rgba 1 1 1 1>}
      uVRegion {bIsValid=false, max=<0 0>, min=<0 0>}
     }
     caretImage {
      bIsDynamicallyLoaded=false
      drawAs=Image
      imageType=FullColor
      mirroring=NoMirror
      resourceName=../../../Engine/Content/Slate/Old/White.png
      resourceObject=null
      tiling=NoTile
      imageSize=<16 16>
      margin=<ltrb 0 0 0 0>
      outlineSettings {
       bUseBrushTransparency=false
       roundingType=HalfHeightRadius
       width=0
       color {colorUseRule=UseColor_Specified, specifiedColor=<rgba 0 0 0 0>}
       cornerRadii=<0 0 0 0>
      }
      tintColor {colorUseRule=UseColor_Specified, specifiedColor=<rgba 1 1 1 1>}
      uVRegion {bIsValid=false, max=<0 0>, min=<0 0>}
     }
     colorAndOpacity {colorUseRule=UseColor_Specified, specifiedColor=<rgba 1 1 1 1>}
     font {
      bForceMonospaced=false
      bMaterialIsStencil=false
      fontMaterial=null
      fontObject=&HeadingNow
      letterSpacing=0
      monospacedWidth=1
      size=20
      skewAmount=0
      typefaceFontName=75 Medium
      outlineSettings {
       bApplyOutlineToDropShadows=false
       bMiteredCorners=false
       bSeparateFillAlpha=false
       outlineMaterial=null
       outlineSize=0
       outlineColor=<rgba 0 0 0 1>
      }
     }
     hintColorAndOpacity {colorUseRule=UseColor_Foreground, specifiedColor=<rgba 1 0 1 1>}
     selectedBackgroundColor {colorUseRule=UseColor_Foreground, specifiedColor=<rgba 1 0 1 1>}
    }
   }
   slot {class=BorderSlot, properties {padding=<ltrb 15 20 15 20>}}
  }
  NumberUnit {
   widgetClassPath=&TextBlock
   parent=HorizontalBox_2
   widget=NumberUnit
   properties {
    text=seconds
    visibility=SelfHitTestInvisible
    font {
     bForceMonospaced=false
     bMaterialIsStencil=false
     fontMaterial=null
     fontObject=&HeadingNow
     letterSpacing=0
     monospacedWidth=1
     size=20
     skewAmount=0
     typefaceFontName=74 Regular
     outlineSettings {
      bApplyOutlineToDropShadows=false
      bMiteredCorners=false
      bSeparateFillAlpha=false
      outlineMaterial=null
      outlineSize=0
      outlineColor=<rgba 0 0 0 1>
     }
    }
   }
   slot {
    class=HorizontalBoxSlot
    properties {horizontalAlignment=HAlign_Left, verticalAlignment=VAlign_Center, padding=<ltrb 10 0 0 0>, size {sizeRule=Automatic, value=0.6}}
   }
  }
  FooterBorder {
   widgetClassPath=&CommonBorder
   parent=VerticalBox_2
   widget=FooterBorder
   properties {
    horizontalAlignment=HAlign_Right
    background {
     bIsDynamicallyLoaded=false
     drawAs=Box
     imageType=NoImage
     mirroring=NoMirror
     resourceName=None
     resourceObject=null
     tiling=NoTile
     imageSize=<32 32>
     margin=<ltrb 0 0 0 0>
     outlineSettings {
      bUseBrushTransparency=false
      roundingType=HalfHeightRadius
      width=0
      color {colorUseRule=UseColor_Specified, specifiedColor=<rgba 0 0 0 0>}
      cornerRadii=<0 0 0 0>
     }
     tintColor {colorUseRule=UseColor_Specified, specifiedColor=<rgba 1 1 1 1>}
     uVRegion {bIsValid=false, max=<0 0>, min=<0 0>}
    }
    brushColor=<rgba 0.0157 0.0196 0.0314 0>
   }
   slot {class=VerticalBoxSlot}
  }
  ButtonBorder {
   widgetClassPath=&HorizontalBox
   parent=FooterBorder
   widget=ButtonBorder
   slot {class=BorderSlot, properties {horizontalAlignment=HAlign_Right}}
  }
  Button_Confirm {
   widgetClassPath=&IconTextButton
   bInherited=true
   parent=ButtonBorder
   widget=Button_Confirm
   properties {
    bHideInputActionWithKeyboard=true
    buttonText=Confirm
    contentAlignment=HAlign_Center
    onButtonBaseClicked=(&self:WidgetTree.Button_Confirm.HandleOnButtonClicked)
    onButtonBaseDoubleClicked=(&self:WidgetTree.Button_Confirm.HandleOnButtonDoubleClicked)
    onButtonBaseHovered=(&self:WidgetTree.Button_Confirm.HandleOnButtonHovered)
    onButtonBaseUnhovered=(&self:WidgetTree.Button_Confirm.HandleOnButtonUnhovered)
    onSelectedChangedBase=(&self:WidgetTree.Button_Confirm.HandleOnSelectedChanged)
    style=/FortUILibrary/Foundation/Buttons/ButtonStyles/ButtonStyle-TransparentWithCues-XS-B-White.ButtonStyle-TransparentWithCues-XS-B-White_C
    font {
     bForceMonospaced=false
     bMaterialIsStencil=false
     fontMaterial=null
     fontObject=&HeadingNow
     letterSpacing=0
     monospacedWidth=1
     size=18
     skewAmount=0
     typefaceFontName=76 Bold
     outlineSettings {
      bApplyOutlineToDropShadows=false
      bMiteredCorners=false
      bSeparateFillAlpha=false
      outlineMaterial=null
      outlineSize=0
      outlineColor=<rgba 0 0 0 1>
     }
    }
    pressedSlateSoundOverride {resourceObject=/Game/Sounds/Fort_UI_Sounds/Replay/DVR_CamOptions_Close_Cue}
    triggeringInputAction {dataTable=&InputActionDataTable, rowName=Confirm}
   }
   slot {class=HorizontalBoxSlot}
  }
  Button_Close {
   widgetClassPath=&IconTextButton
   bInherited=true
   parent=ButtonBorder
   widget=Button_Close
   properties {
    bHideInputActionWithKeyboard=true
    buttonText=Close
    contentAlignment=HAlign_Center
    onButtonBaseClicked=(&self:WidgetTree.Button_Close.HandleOnButtonClicked)
    onButtonBaseDoubleClicked=(&self:WidgetTree.Button_Close.HandleOnButtonDoubleClicked)
    onButtonBaseHovered=(&self:WidgetTree.Button_Close.HandleOnButtonHovered)
    onButtonBaseUnhovered=(&self:WidgetTree.Button_Close.HandleOnButtonUnhovered)
    onSelectedChangedBase=(&self:WidgetTree.Button_Close.HandleOnSelectedChanged)
    style=&ButtonStyle-TransparentWithCues-XS-B-White_C
    font {
     bForceMonospaced=false
     bMaterialIsStencil=false
     fontMaterial=null
     fontObject=&HeadingNow
     letterSpacing=0
     monospacedWidth=1
     size=18
     skewAmount=0
     typefaceFontName=76 Bold
     outlineSettings {
      bApplyOutlineToDropShadows=false
      bMiteredCorners=false
      bSeparateFillAlpha=false
      outlineMaterial=null
      outlineSize=0
      outlineColor=<rgba 0 0 0 1>
     }
    }
    pressedSlateSoundOverride {resourceObject=&DVR_CamOptions_Close_Cue}
    triggeringInputAction {dataTable=&InputActionDataTable, rowName=Close}
   }
   slot {class=HorizontalBoxSlot, properties {padding=<ltrb 10 0 0 0>}}
  }
 ]
}
