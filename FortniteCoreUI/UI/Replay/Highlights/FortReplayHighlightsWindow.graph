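# Replay highlights window widget, based on the native FortReplayHighlightsWindow class, that frames a list of replay highlights.
# A header with icon and HIGHLIGHTS title sits above an inherited VerticalBox_ItemContainer and a footer Close button bound to the CloseViewSettings input action.
# Clicking Close calls ToggleHighlightsListVisibility on the FortReplayContext, the only wired graph logic.
# No item-filling logic appears here; PreConstruct, Construct, Tick and On Activated events are empty.
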WidgetBlueprint /FortniteCoreUI/UI/Replay/Highlights/FortReplayHighlightsWindow
graphs [EventGraph {nodes [
   "On Button Clicked (CloseButton)"!() -> (OutputDelegate, then, Button:Common Button Legacy Object Reference) {is=title K2Node_ComponentBoundEvent}
   > ToggleHighlightsListVisibility(self:Fort Replay Context Object Reference=Get FortReplayContext!() -> (ReturnValue:Fort Replay Context Object Reference) {is=title K2Node_GetBlueprintContextEx})
   Event PreConstruct()
   Event Construct()
   Event Tick()
   Event On Activated()
  ]}]
variables []
widgets {
 info {inheritedWidgetCount=1, parentClass=/Script/FortniteUI.FortReplayHighlightsWindow, rootWidgetClass=&SizeBox, widgetCount=12}
 widgets [
  SizeBox_0 {widgetClassPath=&SizeBox, widget=SizeBox_0, properties {widthOverride=720}}
  VerticalBox_0 {widgetClassPath=&VerticalBox, parent=SizeBox_0, widget=VerticalBox_0, slot {class=SizeBoxSlot}}
  HeaderBorder {
   widgetClassPath=&Border
   parent=VerticalBox_0
   widget=HeaderBorder
   properties {brushColor=<rgba 0.0157 0.0196 0.0314 1>, padding=<ltrb 0 0 0 0>}
   slot {class=VerticalBoxSlot}
  }
  HorizontalBox_1 {
   widgetClassPath=&HorizontalBox
   parent=HeaderBorder
   widget=HorizontalBox_1
   slot {class=BorderSlot, properties {padding=<ltrb 0 0 0 0>}}
  }
  HighlightsIcon {
   widgetClassPath=&Image
   parent=HorizontalBox_1
   widget=HighlightsIcon
   properties {
    visibility=SelfHitTestInvisible
    brush {
     bIsDynamicallyLoaded=false
     drawAs=Image
     imageType=NoImage
     mirroring=NoMirror
     resourceName=None
     resourceObject=/Game/UI/Foundation/Textures/Icons/Replay/T-Icon-Replay-CurrentCamera-L
     tiling=NoTile
     imageSize=<40 40>
     margin=<ltrb 0 0 0 0>
     outlineSettings {
      bUseBrushTransparency=false
      roundingType=HalfHeightRadius
      width=0
      color {colorUseRule=UseColor_Specified, specifiedColor=<rgba 0 0 0 0>}
      cornerRadii=<0 0 0 0>
     }
     tintColor {colorUseRule=UseColor_Specified, specifiedColor=<rgba 1 1 1 1>}
     uVRegion {bIsValid=false, max=<0 0>, min=<0 0>}
    }
   }
   slot {class=HorizontalBoxSlot, properties {horizontalAlignment=HAlign_Center, verticalAlignment=VAlign_Center, padding=<ltrb 12 0 12 0>}}
  }
  Title {
   widgetClassPath=&CommonTextBlock
   parent=HorizontalBox_1
   widget=Title
   properties {
    style=/FortUILibrary/Styles/Typography/Foundation/TextStyle-Header-M.TextStyle-Header-M_C
    text=HIGHLIGHTS
    font {
     bForceMonospaced=false
     bMaterialIsStencil=false
     fontMaterial=null
     fontObject=&Burbank
     letterSpacing=0
     monospacedWidth=1
     size=32
     skewAmount=0
     typefaceFontName=Burbank-Black
     outlineSettings {
      bApplyOutlineToDropShadows=false
      bMiteredCorners=false
      bSeparateFillAlpha=false
      outlineMaterial=null
      outlineSize=0
      outlineColor=<rgba 0 0 0 1>
     }
    }
    shadowOffset=<0 0>
   }
   slot {class=HorizontalBoxSlot}
  }
  SizeBox_1 {
   widgetClassPath=&SizeBox
   parent=VerticalBox_0
   widget=SizeBox_1
   properties {bOverride_WidthOverride=true, widthOverride=650}
   slot {class=VerticalBoxSlot}
  }
  ItemContainerBorderBorder {
   widgetClassPath=&CommonBorder
   parent=SizeBox_1
   widget=ItemContainerBorderBorder
   properties {
    horizontalAlignment=HAlign_Left
    background {
     bIsDynamicallyLoaded=false
     drawAs=Box
     imageType=NoImage
     mirroring=NoMirror
     resourceName=None
     resourceObject=null
     tiling=NoTile
     imageSize=<32 32>
     margin=<ltrb 0 0 0 0>
     outlineSettings {
      bUseBrushTransparency=false
      roundingType=HalfHeightRadius
      width=0
      color {colorUseRule=UseColor_Specified, specifiedColor=<rgba 0 0 0 0>}
      cornerRadii=<0 0 0 0>
     }
     tintColor {colorUseRule=UseColor_Specified, specifiedColor=<rgba 1 1 1 1>}
     uVRegion {bIsValid=false, max=<0 0>, min=<0 0>}
    }
    brushColor=<rgba 0.0157 0.0196 0.0314 0.702>
    padding=<ltrb 4 0 4 0>
   }
   slot {class=SizeBoxSlot}
  }
  VerticalBox_ItemContainer {
   widgetClassPath=&VerticalBox
   bInherited=true
   bIsVariable=true
   parent=ItemContainerBorderBorder
   widget=VerticalBox_ItemContainer
   properties {navigation=&self:WidgetTree.VerticalBox_ItemContainer.WidgetNavigation_0}
   slot {class=BorderSlot, properties {horizontalAlignment=HAlign_Left, padding=<ltrb 4 0 4 0>}}
  }
  FooterBorder {
   widgetClassPath=&CommonBorder
   parent=VerticalBox_0
   widget=FooterBorder
   properties {
    background {
     bIsDynamicallyLoaded=false
     drawAs=Image
     imageType=NoImage
     mirroring=NoMirror
     resourceName=None
     resourceObject=null
     tiling=NoTile
     imageSize=<32 32>
     margin=<ltrb 0 0 0 0>
     outlineSettings {
      bUseBrushTransparency=false
      roundingType=HalfHeightRadius
      width=0
      color {colorUseRule=UseColor_Specified, specifiedColor=<rgba 0 0 0 0>}
      cornerRadii=<0 0 0 0>
     }
     tintColor {colorUseRule=UseColor_Specified, specifiedColor=<rgba 1 1 1 1>}
     uVRegion {bIsValid=false, max=<0 0>, min=<0 0>}
    }
    brushColor=<rgba 0.0157 0.0196 0.0314 1>
    padding=<ltrb 0 0 0 0>
   }
   slot {class=VerticalBoxSlot, properties {size {sizeRule=Automatic, value=0.1}}}
  }
  HorizontalBox_0 {
   widgetClassPath=&HorizontalBox
   parent=FooterBorder
   widget=HorizontalBox_0
   slot {class=BorderSlot, properties {padding=<ltrb 0 0 0 0>}}
  }
  CloseButton {
   widgetClassPath=&IconTextButton
   bIsVariable=true
   parent=HorizontalBox_0
   widget=CloseButton
   properties {
    buttonText=Close
    contentAlignment=HAlign_Center
    onButtonBaseClicked=(&self:WidgetTree.CloseButton.HandleOnButtonClicked)
    onButtonBaseDoubleClicked=(&self:WidgetTree.CloseButton.HandleOnButtonDoubleClicked)
    onButtonBaseHovered=(&self:WidgetTree.CloseButton.HandleOnButtonHovered)
    onButtonBaseUnhovered=(&self:WidgetTree.CloseButton.HandleOnButtonUnhovered)
    onSelectedChangedBase=(&self:WidgetTree.CloseButton.HandleOnSelectedChanged)
    style=/FortUILibrary/Foundation/Buttons/ButtonStyles/ButtonStyle-TransparentWithCues-XS-B-White.ButtonStyle-TransparentWithCues-XS-B-White_C
    pressedSlateSoundOverride {resourceObject=/Game/Sounds/Fort_UI_Sounds/FrontEnd/New_FrontEnd/UI_BlankForOverride_Cue}
    triggeringInputAction {dataTable=/FortUILibrary/Input/ReplayInputActionDataTable, rowName=CloseViewSettings}
   }
   slot {class=HorizontalBoxSlot, properties {horizontalAlignment=HAlign_Right, size {sizeRule=Fill, value=1}}}
  }
 ]
}
