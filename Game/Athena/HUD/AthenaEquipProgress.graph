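# HUD widget showing a radial progress indicator with centered countdown text for a timed equip or consumable action, built on FortHUDEquipProgressBase.
# It starts hidden, appears when progress starts, hides when progress ends, and collapses once time remaining reaches zero.
# Start and time-change events call UpdateTextScale, which shrinks the text at one second or less, setting render scale only when the value differs.
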WidgetBlueprint /Game/Athena/HUD/AthenaEquipProgress
graphs [
 EventGraph {nodes [
   Event Construct()
   > SetVisibility(self=Get ContentRoot() -> (ContentRoot:Size Box Object Reference), InVisibility=Hidden)
   n1 = Event OnProgressStarted!() -> (OutputDelegate, then, Duration:float) {is=title K2Node_Event}
   > UpdateTextScale!(execute, self, TimeRemaining:double=@n1.Duration) -> (then)
   > SetVisibility(self=Get ContentRoot(), InVisibility=SelfHitTestInvisible)
   Event OnProgressEnded() {is=title K2Node_Event}
   > SetVisibility(self=Get ContentRoot(), InVisibility=Hidden)
   n2 = Event OnTimeRemainingChanged!() -> (OutputDelegate, then, TimeRemaining:float) {is=title K2Node_Event}
   > UpdateTextScale(TimeRemaining=@n2.TimeRemaining)
   > Branch(Condition="float <= float"(A=@n2.TimeRemaining) -> (ReturnValue=~))
   > SetVisibility(self=Get ContentRoot(), InVisibility=Collapsed)
  ]}
 UpdateTextScale {nodes [
   n1 = UpdateTextScale!() -> (then, TimeRemaining:double=~) {is=title K2Node_FunctionEntry}
   n2 = Select(
    Option 0:double=Get TextScaleUnderOneSecond() -> (TextScaleUnderOneSecond:double)
    Option 1:double=Get TextScaleOverOneSecond() -> (TextScaleOverOneSecond:double)
    Index="float > float"(A=@n1, B=1) {is=title K2Node_CallFunction}
   ) -> (ReturnValue:double)
   n3 = Get Text_TimeRemaining() -> (Text_TimeRemaining:Common Text Object Reference)
   > Branch(Condition="Not Equal (Float)"!(
     A:double=Get!(self:Widget Object Reference=@n3) -> (
      RenderTransform_Translation:Vector 2D Structure
      RenderTransform_Scale_X:double
      RenderTransform_Scale_Y:double
      RenderTransform_Shear:Vector 2D Structure
      RenderTransform_Angle:float=~
     ).RenderTransform_Scale_X
     B:double=@n2
    ) -> (ReturnValue:bool) {is=title K2Node_CallFunction})
   > SetRenderScale!(execute, self:Widget Object Reference=@n3, Scale_X:double=@n2, Scale_Y:double=@n2) -> (then)
  ]}
]
variables [
 TextScaleOverOneSecond {type=number, default=1}
 TextScaleUnderOneSecond {type=number, default=0.7}
]
widgets {
 info {inheritedWidgetCount=3, parentClass=/Script/FortniteUI.FortHUDEquipProgressBase, rootWidgetClass=&SizeBox, widgetCount=5}
 widgets [
  ContentRoot {
   widgetClassPath=&SizeBox
   bIsVariable=true
   widget=ContentRoot
   properties {bOverride_MaxAspectRatio=true, bOverride_MinAspectRatio=true}
  }
  OverlayTimer {widgetClassPath=&Overlay, parent=ContentRoot, widget=OverlayTimer, slot {class=SizeBoxSlot}}
  Image_ProgressBar {
   widgetClassPath=&Image
   bInherited=true
   parent=OverlayTimer
   widget=Image_ProgressBar
   properties {
    visibility=SelfHitTestInvisible
    brush {
     bIsDynamicallyLoaded=false
     drawAs=Image
     imageType=NoImage
     mirroring=NoMirror
     resourceName=None
     resourceObject=/Game/Athena/HUD/Materials/M_UI_RadialProgress_Consumable
     tiling=NoTile
     imageSize=<96 96>
     margin=<ltrb 0 0 0 0>
     outlineSettings {
      bUseBrushTransparency=false
      roundingType=HalfHeightRadius
      width=0
      color {colorUseRule=UseColor_Specified, specifiedColor=<rgba 0 0 0 0>}
      cornerRadii=<0 0 0 0>
     }
     tintColor {colorUseRule=UseColor_Specified, specifiedColor=<rgba 1 1 1 1>}
     uVRegion {bIsValid=false, max=<0 0>, min=<0 0>}
    }
   }
   slot {class=OverlaySlot, properties {horizontalAlignment=HAlign_Center, verticalAlignment=VAlign_Center}}
  }
  Text_TimeRemaining {
   widgetClassPath=&CommonTextBlock
   bInherited=true
   parent=OverlayTimer
   widget=Text_TimeRemaining
   properties {
    justification=Center
    text=4
    font {
     bForceMonospaced=false
     bMaterialIsStencil=false
     fontMaterial=null
     fontObject=&Burbank
     letterSpacing=0
     monospacedWidth=1
     size=40
     skewAmount=0
     typefaceFontName=Burbank-Black
     outlineSettings {
      bApplyOutlineToDropShadows=false
      bMiteredCorners=false
      bSeparateFillAlpha=true
      outlineMaterial=/Game/UI/Foundation/Materials/M-FontShadow
      outlineSize=2
      outlineColor=<rgba 0 0 0 0.5>
     }
    }
    renderTransform {angle=0, scale=<1 1>, shear=<-12 0>, translation=<-1 1>}
    shadowOffset=<0 0>
   }
   slot {class=OverlaySlot, properties {horizontalAlignment=HAlign_Center, verticalAlignment=VAlign_Center, padding=<ltrb 0 1 0 0>}}
  }
  TutorialHighlight {widgetClassPath=&TutorialHighlight}
 ]
}
