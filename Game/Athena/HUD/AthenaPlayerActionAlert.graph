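# Athena HUD widget that shows brief player action alerts for knockdowns, eliminations, entering the storm and skydive leader.
# Show Alert picks a switcher panel by type, sets its text and plays shared and per-type animations, skipping squad-wipe and crit-kill alerts.
# Knockdowns and eliminations can show a rank icon from override data, and extension widgets in the kill row receive the other player's state.
# Finishing ShowAlert requests the next alert and hides the widget.
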WidgetBlueprint /Game/Athena/HUD/AthenaPlayerActionAlert
animations [
 FlashRun {
  binding_count=1
  display_label=""
  length_seconds=0.7
  object=&self:FlashRun
  sequence {bindings [RunText {
     id=#1
     possessed_class=CommonTextBlock
     tracks [{
       type=MovieScene2DTransformTrack
       sections [{
         type=MovieScene2DTransformSection
         end_frame=14
         start_frame=0
         channels [
          Translation.X {type=MovieSceneScriptingFloatChannel}
          Translation.Y {type=MovieSceneScriptingFloatChannel}
          Angle {type=MovieSceneScriptingFloatChannel}
          Scale.X {type=MovieSceneScriptingFloatChannel, keys [[0,1],[6,1],[10,2],[13.9997,1]]}
          Scale.Y {type=MovieSceneScriptingFloatChannel, keys [[0,1],[6,1],[10,2],[13.9997,1]]}
          Shear.X {type=MovieSceneScriptingFloatChannel}
          Shear.Y {type=MovieSceneScriptingFloatChannel}
         ]
        }]
      }]
    }]}
 }
 ShowAlert {
  binding_count=1
  length_seconds=2.75
  object=&self:ShowAlert
  sequence {bindings [AthenaPlayerActionAlert_C_1 {
     id=#2
     possessed_class=AthenaPlayerActionAlert_C
     tracks [
      {
       type=MovieScene2DTransformTrack
       sections [{
         type=MovieScene2DTransformSection
         end_frame=55
         start_frame=0
         channels [
          Translation.X {type=MovieSceneScriptingFloatChannel, const [0,2]}
          Translation.Y {type=MovieSceneScriptingFloatChannel, keys [[50,0],[55,100]]}
          Scale.X {type=MovieSceneScriptingFloatChannel, keys [[47,1],[50,1.2504],[55,0.5]]}
          Scale.Y {type=MovieSceneScriptingFloatChannel, keys [[47,1],[50,1.2504],[55,0.5]]}
         ]
        }]
      }
      {
       type=MovieSceneFloatTrack
       display_name=Render Opacity
       sections [{
         type=MovieSceneFloatSection
         end_frame=55
         start_frame=0
         channels [None {type=MovieSceneScriptingFloatChannel, keys [[0,0],[2,1],[50,1],[55,0]]}]
        }]
      }
     ]
    }]}
 }
 Elimination {
  binding_count=3
  length_seconds=0.3407
  object=&self:Elimination
  sequence {bindings [
    HorizontalBox_KillPlayerInfo {
     id=#3
     possessed_class=HorizontalBox
     tracks [
      {
       type=MovieSceneFloatTrack
       display_name=Render Opacity
       sections [{type=MovieSceneFloatSection, channels [None {type=MovieSceneScriptingFloatChannel, keys [[4,0],[5,1]]}]}]
      }
      {
       type=MovieScene2DTransformTrack
       sections [{
         type=MovieScene2DTransformSection
         channels [
          Translation.X {type=MovieSceneScriptingFloatChannel}
          Translation.Y {type=MovieSceneScriptingFloatChannel}
          Angle {type=MovieSceneScriptingFloatChannel, keys [[4,2.7356],[6,-1.4207],[8.1903,0.4896],[9.6193,-0.094],[10.813,0]]}
          Scale.X {type=MovieSceneScriptingFloatChannel, keys [[4,3.7296],[6,0.7512],[8,1.1351],[9,0.9371],[10.807,1]]}
          Scale.Y {type=MovieSceneScriptingFloatChannel, keys [[4,3.7296],[6,0.7512],[8,1.1351],[9,0.9371],[10.807,1]]}
          Shear.X {type=MovieSceneScriptingFloatChannel}
          Shear.Y {type=MovieSceneScriptingFloatChannel}
         ]
        }]
      }
     ]
    }
    EliminatedMessage {
     id=#4
     possessed_class=CommonTextBlock
     tracks [
      {
       type=MovieScene2DTransformTrack
       sections [{
         type=MovieScene2DTransformSection
         channels [
          Translation.X {type=MovieSceneScriptingFloatChannel, keys [[4,-116.8107],[6.5983,0],[7.9717,-0.6574],[10,0]]}
          Translation.Y {type=MovieSceneScriptingFloatChannel}
          Angle {type=MovieSceneScriptingFloatChannel}
          Scale.X {type=MovieSceneScriptingFloatChannel}
          Scale.Y {type=MovieSceneScriptingFloatChannel}
          Shear.X {type=MovieSceneScriptingFloatChannel}
          Shear.Y {type=MovieSceneScriptingFloatChannel}
         ]
        }]
      }
      {
       type=MovieSceneFloatTrack
       display_name=Render Opacity
       sections [{type=MovieSceneFloatSection, channels [None {type=MovieSceneScriptingFloatChannel, keys [[4,0],[5.5983,1]]}]}]
      }
     ]
    }
   ]}
 }
 CrawlDown {
  binding_count=3
  length_seconds=0.3407
  object=&self:CrawlDown
  sequence {bindings [
    HorizontalBox_DownPlayerInfo {
     id=#5
     possessed_class=HorizontalBox
     tracks [
      {
       type=MovieSceneFloatTrack
       display_name=Render Opacity
       sections [{type=MovieSceneFloatSection, channels [None {type=MovieSceneScriptingFloatChannel, keys [[4,0],[5,1]]}]}]
      }
      {
       type=MovieScene2DTransformTrack
       sections [{
         type=MovieScene2DTransformSection
         channels [
          Translation.X {type=MovieSceneScriptingFloatChannel}
          Translation.Y {type=MovieSceneScriptingFloatChannel}
          Angle {type=MovieSceneScriptingFloatChannel, keys [[4,2.7356],[6,-1.4207],[8.1903,0.4896],[9.6193,-0.094],[10.813,0]]}
          Scale.X {type=MovieSceneScriptingFloatChannel, keys [[4,3.7296],[6,0.7512],[8,1.1351],[9,0.9371],[10.807,1]]}
          Scale.Y {type=MovieSceneScriptingFloatChannel, keys [[4,3.7296],[6,0.7512],[8,1.1351],[9,0.9371],[10.807,1]]}
          Shear.X {type=MovieSceneScriptingFloatChannel}
          Shear.Y {type=MovieSceneScriptingFloatChannel}
         ]
        }]
      }
     ]
    }
    EliminatedMessage {
     id=#4
     display_name=DownMessage
     possessed_class=CommonTextBlock
     tracks [
      {
       type=MovieScene2DTransformTrack
       sections [{
         type=MovieScene2DTransformSection
         channels [
          Translation.X {type=MovieSceneScriptingFloatChannel, keys [[4,-116.8107],[6.5983,0],[7.9717,-0.6574],[10,0]]}
          Translation.Y {type=MovieSceneScriptingFloatChannel}
          Angle {type=MovieSceneScriptingFloatChannel}
          Scale.X {type=MovieSceneScriptingFloatChannel}
          Scale.Y {type=MovieSceneScriptingFloatChannel}
          Shear.X {type=MovieSceneScriptingFloatChannel}
          Shear.Y {type=MovieSceneScriptingFloatChannel}
         ]
        }]
      }
      {
       type=MovieSceneFloatTrack
       display_name=Render Opacity
       sections [{type=MovieSceneFloatSection, channels [None {type=MovieSceneScriptingFloatChannel, keys [[4,0],[5.5983,1]]}]}]
      }
     ]
    }
   ]}
 }
]
graphs [
 EventGraph {nodes [
   Event On Alerts Cleared() {is=title K2Node_Event}
   > SetVisibility(InVisibility=Hidden)
   > SetVisibility(
    self=[
     Get LazyImage_Habanero_DownIcon() -> (LazyImage_Habanero_DownIcon:Fort Lazy Image Object Reference)
     Get LazyImage_Habanero_KillIcon() -> (LazyImage_Habanero_KillIcon:Fort Lazy Image Object Reference)
    ]
    InVisibility=Collapsed
   )
   n1 = Event Show Alert!() -> (OutputDelegate, then, Alert:Athena Player Action Alert Structure) {is=title K2Node_Event}
   n2 = Break Athena Player Action Alert!(AthenaPlayerActionAlert:"Athena Player Action Alert Structure (by ref)"=@n1.Alert) -> (
    AlertType:EAthenaPlayerActionAlert Enum
    DetailText:Text
    RankedProgress:Ranked Progress Replicated Data Structure
    OtherPlayerState:Fort Player State Athena Weak object
    NumKills:int
   ) {is=title K2Node_BreakStruct}
   > n3 = Branch(Condition=OR Boolean(
     A="Equal (Enum)"(A:EAthenaPlayerActionAlert Enum=@n2, B:EAthenaPlayerActionAlert Enum=PlayerKillSquadWipe)
     B="Equal (Enum)"(A=@n2, B=PlayerKillCritKill)
    ))
   > RequestNextAlert(self:Athena HUDPlayer Action Alert Base Object Reference)
   > SetVisibility(InVisibility=Hidden)
   > SetVisibility(self=[Get LazyImage_Habanero_DownIcon(), Get LazyImage_Habanero_KillIcon()], InVisibility=Collapsed)
   n4 = Sequence(execute=@n3.else)
   > SetVisibility(InVisibility=HitTestInvisible)
   > SetActiveWidgetIndex(
    self:Common Visibility Switcher Object Reference=Get MessageSwitcher() -> (MessageSwitcher:Common Visibility Switcher Object Reference)
    Index=Select!(
     PlayerDown:int=0
     PlayerKill:int=1
     EnteredStorm:int=2
     NewZebulonDrone:int=3
     NewZebulonDroneYou:int=4
     PlayerKillSquadWipe:int=0
     PlayerKillCritKill:int=0
     Index:EAthenaPlayerActionAlert Enum=@n2
    ) -> (ReturnValue:int)
   )
   > PlayAnimation(InAnimation=Get ShowAlert() -> (ShowAlert:Widget Animation Object Reference))
   > n5 = Sequence()
   n6 = Get PlayerDownText() -> (PlayerDownText:Common Text Object Reference)
   n7 = Select(
    PlayerDown:Common Text Object Reference=Get PlayerDownText()
    PlayerKill:Common Text Object Reference=Get PlayerEliminatedText() -> (PlayerEliminatedText:Common Text Object Reference)
    EnteredStorm:Common Text Object Reference
    NewZebulonDrone:Common Text Object Reference=Get JumpLeaderText() -> (JumpLeaderText:Common Text Object Reference)
    NewZebulonDroneYou:Common Text Object Reference
    PlayerKillSquadWipe:Common Text Object Reference=@n6
    PlayerKillCritKill:Common Text Object Reference=@n6
    Index=@n2
   ) -> (ReturnValue:Common Text Object Reference)
   > Is Valid(InputObject=@n7)
   > "SetText (Text)"(self=@n7, InText=@n2.DetailText)
   n8 = Select(
    PlayerDown:Widget Animation Object Reference=Get CrawlDown() -> (CrawlDown:Widget Animation Object Reference)
    PlayerKill:Widget Animation Object Reference=Get Elimination() -> (Elimination:Widget Animation Object Reference)
    EnteredStorm:Widget Animation Object Reference=Get FlashRun() -> (FlashRun:Widget Animation Object Reference)
    NewZebulonDrone:Widget Animation Object Reference
    NewZebulonDroneYou:Widget Animation Object Reference
    PlayerKillSquadWipe:Widget Animation Object Reference
    PlayerKillCritKill:Widget Animation Object Reference
    Index=@n2
   ) -> (ReturnValue:Widget Animation Object Reference)
   Is Valid(exec=@n5.then_1, InputObject=@n8)
   > PlayAnimation(InAnimation=@n8)
   SetRankedProgress!(
    execute=@n4.then_1
    self
    RankedProgress:Ranked Progress Replicated Data Structure=@n2.RankedProgress
    HabaneroIconWidget:Fort Lazy Image Object Reference=Select(
     PlayerDown:Fort Lazy Image Object Reference=Get LazyImage_Habanero_DownIcon()
     PlayerKill:Fort Lazy Image Object Reference=Get LazyImage_Habanero_KillIcon()
     EnteredStorm:Fort Lazy Image Object Reference
     NewZebulonDrone:Fort Lazy Image Object Reference
     NewZebulonDroneYou:Fort Lazy Image Object Reference
     PlayerKillSquadWipe:Fort Lazy Image Object Reference
     PlayerKillCritKill:Fort Lazy Image Object Reference
     Index=@n2
    ) -> (ReturnValue:Fort Lazy Image Object Reference)
   ) -> (then)
   > n9 = For Each Loop(Array:Array of Widget Object References=GetAllChildren(self=Get HorizontalBox_KillPlayerInfoExtension() -> (HorizontalBox_KillPlayerInfoExtension:Horizontal Box Object Reference))) -> (Array Element:Widget Object Reference)
   > n10 = Cast To AthenaPlayerActionAlertExtensionBase!(execute, Object=@n9) -> (then, CastFailed, AsAthena Player Action Alert Extension Base:Athena Player Action Alert Extension Base Object Reference) {is=title K2Node_DynamicCast}
   > SetPlayerState!(
    execute
    self:Athena Player Action Alert Extension Base Object Reference=@n10
    InPlayerState:Player State Object Reference=@n2.OtherPlayerState
   ) -> (then)
   "Animation Finished (ShowAlert)"!() -> (OutputDelegate:Delegate, then:Exec) {is=title K2Node_WidgetAnimationEvent}
   > RequestNextAlert()
   > SetVisibility(InVisibility=Hidden)
   > SetVisibility(self=[Get LazyImage_Habanero_DownIcon(), Get LazyImage_Habanero_KillIcon()], InVisibility=Collapsed)
  ]}
 OnExtensionWidgetAdded {nodes [
   n1 = OnExtensionWidgetAdded!() -> (then, ExtensionSlot:EUIExtensionSlot Enum=Invalid, ExtensionWidget:User Widget Object Reference) {is=title K2Node_FunctionEntry}
   > Branch(Condition="Equal (Enum)"(
     A:EUIExtensionSlot Enum=@n1
     B:EUIExtensionSlot Enum=Get HUDPLayerActionAlertKillPlayerInfoExtension() -> (HUDPLayerActionAlertKillPlayerInfoExtension:EUIExtensionSlot Enum=Invalid)
    ))
   > n2 = Cast To AthenaPlayerActionAlertExtensionBase(Object=Get ExtensionWidget() -> (ExtensionWidget:User Widget Object Reference))
   > AddChild(self=Get HorizontalBox_KillPlayerInfoExtension(), Content=@n2)
  ]}
 SetRankedProgress {nodes [
   n1 = SetRankedProgress!() -> (then, RankedProgress:Ranked Progress Replicated Data Structure, HabaneroIconWidget:Fort Lazy Image Object Reference) {is=title K2Node_FunctionEntry}
   > Is Valid(InputObject=@n1.HabaneroIconWidget)
   > SetVisibility(self=@n1.HabaneroIconWidget, InVisibility=Collapsed)
   n2 = Break Ranked Progress Replicated Data!(RankedProgressReplicatedData:"Ranked Progress Replicated Data Structure (by ref)"=@n1) -> (RankType:String, Rank:int, HighestRank:int, CurrentMMR:int, CurrentPosition:int) {is=title K2Node_BreakStruct}
   n3 = GetHabaneroDisplayDataOverride!(
    InDisplayData:Fort Habanero Display Data Object Reference=Get RankData() -> (RankData:Fort Habanero Display Data Object Reference)
    InOverrideData:Fort Compete UIHabanero Tier Override List Object Reference=Get RankInfoOverrides() -> (RankInfoOverrides:Fort Compete UIHabanero Tier Override List Object Reference)
    InKey:"Fort Compete UIHabanero Tier Override Key Structure (by ref)"=Make Fort Compete UIHabanero Tier Override Key!(
     RankedTrack:Name=String To Name(InString=@n2)
     Season:String
     Rank:int=@n2.Rank
     MatchmakingRating:int=@n2.CurrentMMR
     CurrentPosition:int=@n2.CurrentPosition
    ) -> (FortCompeteUIHabaneroTierOverrideKey:Fort Compete UIHabanero Tier Override Key Structure) {is=title K2Node_MakeStruct}
   ) -> (OutData:Fort Habanero Tier Override Structure, ReturnValue:bool)
   > Branch(Condition=@n3.ReturnValue)
   > SetBrushFromSoftTexture!(
    execute
    self:Image Object Reference=@n1.HabaneroIconWidget
    SoftTexture:Texture 2D Soft Object Reference=Break Fort Habanero Tier!(FortHabaneroTier:"Fort Habanero Tier Structure (by ref)"=@n3) -> (
     ForegroundIcon:Material Instance Soft Object Reference
     ForegroundIconSmall:Material Instance Soft Object Reference
     BackgroundIcon:Material Instance Soft Object Reference
     BackgroundIconSmall:Material Instance Soft Object Reference
     WhiteIcon:Texture 2D Soft Object Reference
     SimpleIconSmall:Texture 2D Soft Object Reference
     Name:Text
     TierType:Fort Habanero Tier Type Data Object Reference
     AdditionalLeftIcon:Material Instance Soft Object Reference
     AdditionalRightIcon:Material Instance Soft Object Reference
    ) {is=title K2Node_BreakStruct}.SimpleIconSmall
    bMatchSize:bool=false
   ) -> (then)
   > SetVisibility(self=@n1.HabaneroIconWidget, InVisibility=HitTestInvisible)
  ]}
]
variables [
 HUDPLayerActionAlertKillPlayerInfoExtension {type=EUIExtensionSlot, default=HUDPlayerActionAlertKillPlayerInfoExtension}
 RankData {type=FortHabaneroDisplayData, default=/Game/UI/Habanero/HabaneroDisplayData}
 RankInfoOverrides {type=FortCompeteUIHabaneroTierOverrideList, default=/Game/UI/Habanero/DA_RankDisplayDataOverride}
]
widgets {
 info {
  inheritedWidgetCount=1
  namedSlotCount=1
  parentClass=/Script/FortniteUI.AthenaHUDPlayerActionAlertBase
  rootWidgetClass=&CommonVisibilitySwitcher
  widgetCount=24
 }
 widgets [
  MessageSwitcher {widgetClassPath=&CommonVisibilitySwitcher, bIsVariable=true, widget=MessageSwitcher}
  DownFeedback {widgetClassPath=&VerticalBox, parent=MessageSwitcher, widget=DownFeedback, slot {class=CommonVisibilitySwitcherSlot}}
  HorizontalBox_DownFeedback {widgetClassPath=&HorizontalBox, parent=DownFeedback, widget=HorizontalBox_DownFeedback, slot {class=VerticalBoxSlot}}
  DownMessage {
   widgetClassPath=&CommonTextBlock
   parent=HorizontalBox_DownFeedback
   widget=DownMessage
   properties {
    text=KNOCKED DOWN
    textTransformPolicy=ToUpper
    font {
     bForceMonospaced=false
     bMaterialIsStencil=false
     fontMaterial=null
     fontObject=&Burbank
     letterSpacing=20
     monospacedWidth=1
     size=18
     skewAmount=0.25
     typefaceFontName=BurbankRegular-Black
     outlineSettings {
      bApplyOutlineToDropShadows=false
      bMiteredCorners=false
      bSeparateFillAlpha=false
      outlineMaterial=null
      outlineSize=0
      outlineColor=<rgba 0 0 0 0>
     }
    }
    shadowColorAndOpacity=<rgba 0 0 0 0.5>
    shadowOffset=<1.5 1.5>
    strikeBrush {
     bIsDynamicallyLoaded=false
     drawAs=Image
     imageType=NoImage
     mirroring=NoMirror
     resourceName=None
     resourceObject=null
     tiling=NoTile
     imageSize=<32 32>
     margin=<ltrb 0 0 0 0>
     outlineSettings {
      bUseBrushTransparency=false
      roundingType=HalfHeightRadius
      width=0
      color {colorUseRule=UseColor_Specified, specifiedColor=<rgba 0 0 0 0>}
      cornerRadii=<0 0 0 0>
     }
     tintColor {colorUseRule=UseColor_Specified, specifiedColor=<rgba 0.8 0.8 0.8 1>}
     uVRegion {bIsValid=false, max=<0 0>, min=<0 0>}
    }
   }
   slot {class=HorizontalBoxSlot, properties {verticalAlignment=VAlign_Center}}
  }
  HorizontalBox_DownPlayerInfo {
   widgetClassPath=&HorizontalBox
   parent=HorizontalBox_DownFeedback
   widget=HorizontalBox_DownPlayerInfo
   slot {class=HorizontalBoxSlot, properties {padding=<ltrb 6 0 0 0>}}
  }
  LazyImage_Habanero_DownIcon {
   widgetClassPath=&FortLazyImage
   bIsVariable=true
   parent=HorizontalBox_DownPlayerInfo
   widget=LazyImage_Habanero_DownIcon
   properties {brush {
     bIsDynamicallyLoaded=false
     drawAs=Image
     imageType=NoImage
     mirroring=NoMirror
     resourceName=None
     resourceObject=null
     tiling=NoTile
     imageSize=<64 64>
     margin=<ltrb 0 0 0 0>
     outlineSettings {
      bUseBrushTransparency=false
      roundingType=HalfHeightRadius
      width=0
      color {colorUseRule=UseColor_Specified, specifiedColor=<rgba 0 0 0 0>}
      cornerRadii=<0 0 0 0>
     }
     tintColor {colorUseRule=UseColor_Specified, specifiedColor=<rgba 1 1 1 1>}
     uVRegion {bIsValid=false, max=<0 0>, min=<0 0>}
    }}
   slot {class=HorizontalBoxSlot, properties {horizontalAlignment=HAlign_Center, verticalAlignment=VAlign_Center, padding=<ltrb 0 0 -2 0>}}
  }
  PlayerDownText {
   widgetClassPath=&CommonTextBlock
   bIsVariable=true
   parent=HorizontalBox_DownPlayerInfo
   widget=PlayerDownText
   properties {
    text="{Player}"
    colorAndOpacity {colorUseRule=UseColor_Specified, specifiedColor=<rgba 1 0.5833 0 1>}
    font {
     bForceMonospaced=false
     bMaterialIsStencil=false
     fontMaterial=null
     fontObject=&Burbank
     letterSpacing=20
     monospacedWidth=1
     size=24
     skewAmount=0.25
     typefaceFontName=Burbank-Black
     outlineSettings {
      bApplyOutlineToDropShadows=false
      bMiteredCorners=false
      bSeparateFillAlpha=false
      outlineMaterial=null
      outlineSize=0
      outlineColor=<rgba 0 0 0 0>
     }
    }
    shadowColorAndOpacity=<rgba 0 0 0 0.666>
    shadowOffset=<1.5 1.5>
    strikeBrush {
     bIsDynamicallyLoaded=false
     drawAs=Image
     imageType=NoImage
     mirroring=NoMirror
     resourceName=None
     resourceObject=null
     tiling=NoTile
     imageSize=<32 32>
     margin=<ltrb 0 0 0 0>
     outlineSettings {
      bUseBrushTransparency=false
      roundingType=HalfHeightRadius
      width=0
      color {colorUseRule=UseColor_Specified, specifiedColor=<rgba 0 0 0 0>}
      cornerRadii=<0 0 0 0>
     }
     tintColor {colorUseRule=UseColor_Specified, specifiedColor=<rgba 0.8 0.8 0.8 1>}
     uVRegion {bIsValid=false, max=<0 0>, min=<0 0>}
    }
   }
   slot {class=HorizontalBoxSlot, properties {verticalAlignment=VAlign_Center}}
  }
  KillFeedback {widgetClassPath=&HorizontalBox, parent=MessageSwitcher, widget=KillFeedback, slot {class=CommonVisibilitySwitcherSlot}}
  EliminatedMessage {
   widgetClassPath=&CommonTextBlock
   parent=KillFeedback
   widget=EliminatedMessage
   properties {
    text=ELIMINATED
    textTransformPolicy=ToUpper
    colorAndOpacity {colorUseRule=UseColor_Specified, specifiedColor=<rgba 0.95 0.95 0.95 1>}
    font {
     bForceMonospaced=false
     bMaterialIsStencil=false
     fontMaterial=null
     fontObject=&Burbank
     letterSpacing=80
     monospacedWidth=1
     size=19.5
     skewAmount=0.25
     typefaceFontName=Burbank-Black
     outlineSettings {
      bApplyOutlineToDropShadows=false
      bMiteredCorners=false
      bSeparateFillAlpha=false
      outlineMaterial=null
      outlineSize=0
      outlineColor=<rgba 0 0 0 0>
     }
    }
    shadowColorAndOpacity=<rgba 0 0 0 0.7>
    shadowOffset=<1.6 1.6>
    strikeBrush {
     bIsDynamicallyLoaded=false
     drawAs=Image
     imageType=NoImage
     mirroring=NoMirror
     resourceName=None
     resourceObject=null
     tiling=NoTile
     imageSize=<32 32>
     margin=<ltrb 0 0 0 0>
     outlineSettings {
      bUseBrushTransparency=false
      roundingType=HalfHeightRadius
      width=0
      color {colorUseRule=UseColor_Specified, specifiedColor=<rgba 0 0 0 0>}
      cornerRadii=<0 0 0 0>
     }
     tintColor {colorUseRule=UseColor_Specified, specifiedColor=<rgba 0.9 0.9 0.9 1>}
     uVRegion {bIsValid=false, max=<0 0>, min=<0 0>}
    }
   }
   slot {class=HorizontalBoxSlot, properties {verticalAlignment=VAlign_Center}}
  }
  HorizontalBox_KillPlayerInfo {
   widgetClassPath=&HorizontalBox
   parent=KillFeedback
   widget=HorizontalBox_KillPlayerInfo
   slot {class=HorizontalBoxSlot, properties {padding=<ltrb 6 0 0 0>}}
  }
  LazyImage_Habanero_KillIcon {
   widgetClassPath=&FortLazyImage
   bIsVariable=true
   parent=HorizontalBox_KillPlayerInfo
   widget=LazyImage_Habanero_KillIcon
   properties {
    brush {
     bIsDynamicallyLoaded=false
     drawAs=Image
     imageType=NoImage
     mirroring=NoMirror
     resourceName=None
     resourceObject=null
     tiling=NoTile
     imageSize=<64 64>
     margin=<ltrb 0 0 0 0>
     outlineSettings {
      bUseBrushTransparency=false
      roundingType=HalfHeightRadius
      width=0
      color {colorUseRule=UseColor_Specified, specifiedColor=<rgba 0 0 0 0>}
      cornerRadii=<0 0 0 0>
     }
     tintColor {colorUseRule=UseColor_Specified, specifiedColor=<rgba 1 1 1 1>}
     uVRegion {bIsValid=false, max=<0 0>, min=<0 0>}
    }
    loadingBackgroundBrush {
     bIsDynamicallyLoaded=false
     drawAs=NoDrawType
     imageType=NoImage
     mirroring=NoMirror
     resourceName=None
     resourceObject=null
     tiling=NoTile
     imageSize=<64 64>
     margin=<ltrb 0 0 0 0>
     outlineSettings {
      bUseBrushTransparency=false
      roundingType=HalfHeightRadius
      width=0
      color {colorUseRule=UseColor_Specified, specifiedColor=<rgba 0 0 0 0>}
      cornerRadii=<0 0 0 0>
     }
     tintColor {colorUseRule=UseColor_Specified, specifiedColor=<rgba 1 1 1 1>}
     uVRegion {bIsValid=false, max=<0 0>, min=<0 0>}
    }
   }
   slot {class=HorizontalBoxSlot, properties {horizontalAlignment=HAlign_Center, verticalAlignment=VAlign_Center}}
  }
  PlayerEliminatedText {
   widgetClassPath=&CommonTextBlock
   bIsVariable=true
   parent=HorizontalBox_KillPlayerInfo
   widget=PlayerEliminatedText
   properties {
    text="{Player}"
    colorAndOpacity {colorUseRule=UseColor_Specified, specifiedColor=<rgba 0.8 0.012 0.012 1>}
    font {
     bForceMonospaced=false
     bMaterialIsStencil=false
     fontMaterial=null
     fontObject=&Burbank
     letterSpacing=50
     monospacedWidth=1
     size=24
     skewAmount=0.25
     typefaceFontName=Burbank-Black
     outlineSettings {
      bApplyOutlineToDropShadows=false
      bMiteredCorners=false
      bSeparateFillAlpha=false
      outlineMaterial=null
      outlineSize=0
      outlineColor=<rgba 0 0 0 0>
     }
    }
    shadowColorAndOpacity=<rgba 0 0 0 0.7>
    shadowOffset=<1.6 1.6>
   }
   slot {class=HorizontalBoxSlot, properties {verticalAlignment=VAlign_Center, padding=<ltrb 0 0 0 2>}}
  }
  HorizontalBox_KillPlayerInfoExtension {
   widgetClassPath=&HorizontalBox
   bIsVariable=true
   parent=HorizontalBox_KillPlayerInfo
   widget=HorizontalBox_KillPlayerInfoExtension
   slot {class=HorizontalBoxSlot}
  }
  FortHUDElementWrapper {
   widgetClassPath=&/FortHUDElementWrapper
   bIsVariable=true
   parent=MessageSwitcher
   widget=FortHUDElementWrapper
   properties {
    elementTag {tagName=HUD.Athena.PlayerActions.StormFeedback}
    hUDElementTag {gameplayTags [
      {tagName=Athena.HideForCapture}
      {tagName=HUD.LTM.HideForTryHard}
      {tagName=HUD.Event.HideForSpecialEvent}
      {tagName=HUD.Athena.HideForPostGameSpectate}
      {tagName=HUD.Athena.HideForPostGame}
      {tagName=HUD.Athena.HidePlayerAlert}
     ]}
   }
   slot {class=CommonVisibilitySwitcherSlot}
  }
  StormFeeedback {widgetClassPath=&HorizontalBox, namedSlotHost=FortHUDElementWrapper, widget=StormFeeedback}
  StormEnterMessage {
   widgetClassPath=&CommonTextBlock
   parent=StormFeeedback
   widget=StormEnterMessage
   properties {
    text=YOU ARE IN THE STORM
    textTransformPolicy=ToUpper
    font {
     bForceMonospaced=false
     bMaterialIsStencil=false
     fontMaterial=null
     fontObject=&Burbank
     letterSpacing=0
     monospacedWidth=1
     size=18
     skewAmount=0.25
     typefaceFontName=BurbankRegular-Black
     outlineSettings {
      bApplyOutlineToDropShadows=false
      bMiteredCorners=false
      bSeparateFillAlpha=false
      outlineMaterial=null
      outlineSize=0
      outlineColor=<rgba 0 0 0 0>
     }
    }
    shadowColorAndOpacity=<rgba 0 0 0 0.5>
    shadowOffset=<2 3>
   }
   slot {class=HorizontalBoxSlot, properties {verticalAlignment=VAlign_Center}}
  }
  RunText {
   widgetClassPath=&CommonTextBlock
   parent=StormFeeedback
   widget=RunText
   properties {
    text="Run!"
    textTransformPolicy=ToUpper
    colorAndOpacity {colorUseRule=UseColor_Specified, specifiedColor=<rgba 1 0.25 0 1>}
    font {
     bForceMonospaced=false
     bMaterialIsStencil=false
     fontMaterial=null
     fontObject=&Burbank
     letterSpacing=0
     monospacedWidth=1
     size=24
     skewAmount=0.25
     typefaceFontName=Burbank-Black
     outlineSettings {
      bApplyOutlineToDropShadows=false
      bMiteredCorners=false
      bSeparateFillAlpha=false
      outlineMaterial=null
      outlineSize=0
      outlineColor=<rgba 0 0 0 0>
     }
    }
    shadowColorAndOpacity=<rgba 0 0 0 0.666>
    shadowOffset=<3 4>
   }
   slot {class=HorizontalBoxSlot, properties {horizontalAlignment=HAlign_Center, verticalAlignment=VAlign_Center, padding=<ltrb 4 0 0 0>}}
  }
  JumpLeaderFeedback {widgetClassPath=&HorizontalBox, parent=MessageSwitcher, widget=JumpLeaderFeedback, slot {class=CommonVisibilitySwitcherSlot}}
  JumpLeaderText {
   widgetClassPath=&CommonTextBlock
   bIsVariable=true
   parent=JumpLeaderFeedback
   widget=JumpLeaderText
   properties {
    text="{Player}"
    colorAndOpacity {colorUseRule=UseColor_Specified, specifiedColor=<rgba 1 0.5833 0 1>}
    font {
     bForceMonospaced=false
     bMaterialIsStencil=false
     fontMaterial=null
     fontObject=&Burbank
     letterSpacing=0
     monospacedWidth=1
     size=24
     skewAmount=0.25
     typefaceFontName=Burbank-Black
     outlineSettings {
      bApplyOutlineToDropShadows=false
      bMiteredCorners=false
      bSeparateFillAlpha=false
      outlineMaterial=null
      outlineSize=0
      outlineColor=<rgba 0 0 0 0>
     }
    }
    shadowColorAndOpacity=<rgba 0 0 0 0.666>
    shadowOffset=<3 4>
   }
   slot {class=HorizontalBoxSlot, properties {horizontalAlignment=HAlign_Center, verticalAlignment=VAlign_Center, padding=<ltrb 0 0 4 0>}}
  }
  CommonTextBlock_0 {
   widgetClassPath=&CommonTextBlock
   parent=JumpLeaderFeedback
   widget=CommonTextBlock_0
   properties {
    text=IS THE SKYDIVE LEADER
    textTransformPolicy=ToUpper
    font {
     bForceMonospaced=false
     bMaterialIsStencil=false
     fontMaterial=null
     fontObject=&Burbank
     letterSpacing=0
     monospacedWidth=1
     size=18
     skewAmount=0.25
     typefaceFontName=BurbankRegular-Black
     outlineSettings {
      bApplyOutlineToDropShadows=false
      bMiteredCorners=false
      bSeparateFillAlpha=false
      outlineMaterial=null
      outlineSize=0
      outlineColor=<rgba 0 0 0 0>
     }
    }
    shadowColorAndOpacity=<rgba 0 0 0 0.5>
    shadowOffset=<2 3>
   }
   slot {class=HorizontalBoxSlot, properties {verticalAlignment=VAlign_Center, padding=<ltrb 4 0 0 0>}}
  }
  JumpLeaderYouFeedback {widgetClassPath=&HorizontalBox, parent=MessageSwitcher, widget=JumpLeaderYouFeedback, slot {class=CommonVisibilitySwitcherSlot}}
  CommonTextBlock_1 {
   widgetClassPath=&CommonTextBlock
   parent=JumpLeaderYouFeedback
   widget=CommonTextBlock_1
   properties {
    text=YOU
    colorAndOpacity {colorUseRule=UseColor_Specified, specifiedColor=<rgba 1 0.8417 0.05 1>}
    font {
     bForceMonospaced=false
     bMaterialIsStencil=false
     fontMaterial=null
     fontObject=&Burbank
     letterSpacing=0
     monospacedWidth=1
     size=24
     skewAmount=0.25
     typefaceFontName=Burbank-Black
     outlineSettings {
      bApplyOutlineToDropShadows=false
      bMiteredCorners=false
      bSeparateFillAlpha=false
      outlineMaterial=null
      outlineSize=0
      outlineColor=<rgba 0 0 0 0>
     }
    }
    shadowColorAndOpacity=<rgba 0 0 0 0.666>
    shadowOffset=<3 4>
   }
   slot {class=HorizontalBoxSlot, properties {horizontalAlignment=HAlign_Center, verticalAlignment=VAlign_Center}}
  }
  CommonTextBlock_2 {
   widgetClassPath=&CommonTextBlock
   parent=JumpLeaderYouFeedback
   widget=CommonTextBlock_2
   properties {
    text=ARE THE SKYDIVE LEADER
    textTransformPolicy=ToUpper
    font {
     bForceMonospaced=false
     bMaterialIsStencil=false
     fontMaterial=null
     fontObject=&Burbank
     letterSpacing=0
     monospacedWidth=1
     size=18
     skewAmount=0.25
     typefaceFontName=BurbankRegular-Black
     outlineSettings {
      bApplyOutlineToDropShadows=false
      bMiteredCorners=false
      bSeparateFillAlpha=false
      outlineMaterial=null
      outlineSize=0
      outlineColor=<rgba 0 0 0 0>
     }
    }
    shadowColorAndOpacity=<rgba 0 0 0 0.5>
    shadowOffset=<2 3>
   }
   slot {class=HorizontalBoxSlot, properties {verticalAlignment=VAlign_Center, padding=<ltrb 4 0 0 0>}}
  }
  TutorialHighlight {widgetClassPath=&TutorialHighlight}
 ]
}
