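# Athena respawn countdown widget: a "RESPAWNING IN" label with a circular progress image and time text, built on native AthenaRespawnTimer.
# Each tick shows it, except for touch players in BlastBerry, plays the TickDown scale/fade animation and conditionally plays a countdown sound.
# Time end hides it and plays the respawn sound; extension widgets append below.
# A camera audio-fade delegate binds on ticks and unbinds later, but its handler is empty.
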WidgetBlueprint /Game/Athena/HUD/AthenaRespawnTimer
animations [TickDown {
  binding_count=1
  display_label=""
  length_seconds=1.0018
  object=&self:TickDown
  sequence {bindings [TimeText {
     id=#1
     possessed_class=CommonTextBlock
     tracks [
      {
       type=MovieScene2DTransformTrack
       sections [{
         type=MovieScene2DTransformSection
         end_frame=20
         start_frame=0
         channels [
          Translation.X {type=MovieSceneScriptingFloatChannel}
          Translation.Y {type=MovieSceneScriptingFloatChannel}
          Angle {type=MovieSceneScriptingFloatChannel}
          Scale.X {type=MovieSceneScriptingFloatChannel, keys [[0,2],[3,0.95],[5,1],[16,1],[20.036,0.8]]}
          Scale.Y {type=MovieSceneScriptingFloatChannel, keys [[0,2],[3,0.95],[5,1],[16,1],[20.036,0.8]]}
          Shear.X {type=MovieSceneScriptingFloatChannel}
          Shear.Y {type=MovieSceneScriptingFloatChannel}
         ]
        }]
      }
      {
       type=MovieSceneFloatTrack
       display_name=Render Opacity
       sections [{
         type=MovieSceneFloatSection
         end_frame=20
         start_frame=0
         channels [None {type=MovieSceneScriptingFloatChannel, keys [[0,0],[1,1],[11.971,1],[20.036,0]]}]
        }]
      }
     ]
    }]}
 }]
graphs [EventGraph {nodes [
   Event OnTickDown() {is=title K2Node_Event}
   n1 = GetGameStateAthena!() -> (ReturnValue:Fort Game State Athena Object Reference)
   > Is Valid(InputObject=@n1)
   n2 = GetOwningPlayer()
   n3 = GetAthenaPlaylistContextTags!(self:Fort Game State Athena Object Reference=@n1) -> (ReturnValue:Gameplay Tag Container Structure)
   > n4 = Branch(Condition=AND Boolean(
     A=IsUsingTouch(PlayerController=@n2)
     B=HasTag!(
      TagContainer:"Gameplay Tag Container Structure (by ref)"=@n3
      Tag:Gameplay Tag Structure=(TagName="Athena.Playlist.BlastBerry")
      bExactMatch:bool=false
     ) -> (ReturnValue:bool)
    ))
   SetVisibility(execute=@n4.else, self=Get Content() -> (Content:Vertical Box Object Reference), InVisibility=SelfHitTestInvisible)
   > PlayAnimation(InAnimation=Get TickDown() -> (TickDown:Widget Animation Object Reference))
   n5 = GetGameStateAthena()
   > n6 = Is Valid(InputObject=@n5)
   > n7 = Branch(Condition=GetShouldPlayRespawnTimerSFX!(self:Fort Game State Athena Object Reference=@n5) -> (ReturnValue:bool))
   PlaySound2D(execute=[@n7.then, @n6.Is Not Valid], Sound=/Game/Athena/Sounds/LTMs/Respawn/LTM_Respawn_Countdown_Cue)
   n8 = GetPlayerCameraManager!(PlayerIndex:int=0) -> (ReturnValue:Player Camera Manager Object Reference)
   > Is Valid(InputObject=@n8)
   > Bind Event to On Audio Fade Change Event(self:Player Camera Manager Object Reference=@n8, Delegate=Create Event()) {is=title K2Node_AddDelegate}
   Event OnTimeEnded() {is=title K2Node_Event}
   > SetVisibility(self=Get Content(), InVisibility=Hidden)
   > PlaySound2D(Sound=/Game/Athena/Sounds/LTMs/Respawn/LTM_Respawn_Cue)
   > n9 = Delay(Duration=3)
   Unbind Event from On Audio Fade Change Event(execute=[@n9.then, @n12.then], self:Player Camera Manager Object Reference=GetPlayerCameraManager(PlayerIndex=0), Delegate=Create Event()) {is=title K2Node_RemoveDelegate}
   n10 = Event Add Extension Widget!() -> (OutputDelegate, then, Widget:Widget Object Reference) {is=title K2Node_Event}
   > n11 = AddChildToVerticalBox(self=Get Content(), Content=@n10.Widget)
   > SetVerticalAlignment(self:Vertical Box Slot Object Reference=@n11, InVerticalAlignment=VAlign_Center)
   > SetHorizontalAlignment(self:Vertical Box Slot Object Reference=@n11, InHorizontalAlignment=HAlign_Center)
   Event Construct()
   > SetVisibility(self=Get Content(), InVisibility=Hidden)
   n12 = Event Destruct()
   On Audio Fade Change Event!() -> (OutputDelegate, then, bFadeOut:bool=~, FadeTime:float=~) {is=title K2Node_CustomEvent}
  ]}]
variables []
widgets {
 info {inheritedWidgetCount=2, parentClass=/Script/FortniteUI.AthenaRespawnTimer, rootWidgetClass=&VerticalBox, widgetCount=7}
 widgets [
  Content {widgetClassPath=&VerticalBox, bIsVariable=true, widget=Content}
  HBox {
   widgetClassPath=&HorizontalBox
   parent=Content
   widget=HBox
   slot {class=VerticalBoxSlot, properties {horizontalAlignment=HAlign_Center}}
  }
  Message {
   widgetClassPath=&CommonTextBlock
   parent=HBox
   widget=Message
   properties {
    text=RESPAWNING IN
    textTransformPolicy=ToUpper
    font {
     bForceMonospaced=false
     bMaterialIsStencil=false
     fontMaterial=null
     fontObject=&Burbank
     letterSpacing=0
     monospacedWidth=1
     size=18
     skewAmount=0
     typefaceFontName=BurbankRegular-Black
     outlineSettings {
      bApplyOutlineToDropShadows=false
      bMiteredCorners=false
      bSeparateFillAlpha=false
      outlineMaterial=/Game/UI/Foundation/Materials/M-FontShadow
      outlineSize=2
      outlineColor=<rgba 0 0 0 1>
     }
    }
    renderTransform {angle=0, scale=<1 1>, shear=<-12 0>, translation=<0 0>}
    shadowOffset=<0 0>
   }
   slot {class=HorizontalBoxSlot, properties {verticalAlignment=VAlign_Center, padding=<ltrb 0 0 4 -2>}}
  }
  Overlay_0 {
   widgetClassPath=&Overlay
   parent=HBox
   widget=Overlay_0
   slot {class=HorizontalBoxSlot, properties {horizontalAlignment=HAlign_Center, verticalAlignment=VAlign_Center}}
  }
  Image_ProgressBar {
   widgetClassPath=&Image
   parent=Overlay_0
   widget=Image_ProgressBar
   properties {
    visibility=SelfHitTestInvisible
    brush {
     bIsDynamicallyLoaded=false
     drawAs=Image
     imageType=NoImage
     mirroring=NoMirror
     resourceName=None
     resourceObject=/Game/UI/Materials/Athena_HollowCircularProgressBar_Respawn
     tiling=NoTile
     imageSize=<56 56>
     margin=<ltrb 0 0 0 0>
     outlineSettings {
      bUseBrushTransparency=false
      roundingType=HalfHeightRadius
      width=0
      color {colorUseRule=UseColor_Specified, specifiedColor=<rgba 0 0 0 0>}
      cornerRadii=<0 0 0 0>
     }
     tintColor {colorUseRule=UseColor_Specified, specifiedColor=<rgba 1 1 1 1>}
     uVRegion {bIsValid=false, max=<0 0>, min=<0 0>}
    }
   }
   slot {class=OverlaySlot, properties {horizontalAlignment=HAlign_Center, verticalAlignment=VAlign_Center, padding=<ltrb 0 -20 0 -20>}}
  }
  Text_Time {
   widgetClassPath=&CommonTextBlock
   bInherited=true
   parent=Overlay_0
   widget=Text_Time
   properties {
    justification=Center
    text=5
    font {
     bForceMonospaced=false
     bMaterialIsStencil=false
     fontMaterial=null
     fontObject=&Burbank
     letterSpacing=0
     monospacedWidth=1
     size=32
     skewAmount=0
     typefaceFontName=Burbank-Black
     outlineSettings {
      bApplyOutlineToDropShadows=false
      bMiteredCorners=false
      bSeparateFillAlpha=false
      outlineMaterial=null
      outlineSize=2
      outlineColor=<rgba 0 1 1 0.125>
     }
    }
    renderTransform {angle=0, scale=<0.8 0.8>, shear=<0 0>, translation=<0 0>}
    shadowColorAndOpacity=<rgba 0 0.0625 0.125 1>
    shadowOffset=<0 0>
   }
   slot {class=OverlaySlot, properties {horizontalAlignment=HAlign_Center, verticalAlignment=VAlign_Center, padding=<ltrb 0 2 0 0>}}
  }
  TutorialHighlight {widgetClassPath=&TutorialHighlight}
 ]
}
