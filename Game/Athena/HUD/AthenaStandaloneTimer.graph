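# Standalone HUD timer widget on AthenaTimeWidgetBase, showing time as centered rich text over a translucent skewed background.
# UpdateDisplay wraps the incoming text in a style tag, falling back to DefaultStyle (Heist) when no style is given.
# IsTextStyleValid looks for the style name among rows of the AthenaStandaloneTimerStyles table.
# Minigame start shifts the widget down and minigame end resets it, both calling the parent handlers.
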WidgetBlueprint /Game/Athena/HUD/AthenaStandaloneTimer
graphs [
 EventGraph {nodes [
   n1 = Event UpdateDisplay!() -> (OutputDelegate, then, DisplayText:"Text (by ref)", DisplayTextStyle:"Name (by ref)"=None) {is=title K2Node_Event}
   n2 = Select(
    Option 0:Name=Get DefaultStyle() -> (DefaultStyle:Name=None)
    Option 1:Name=@n1.DisplayTextStyle
    Index="Not Equal (Name)"!(A:Name=@n1.DisplayTextStyle, B:Name=None) -> (ReturnValue:bool) {is=title K2Node_CallFunction}
   ) -> (ReturnValue:Name)
   > SetText(
    self=Get TimeRichText() -> (TimeRichText:Rich Text Block Object Reference)
    InText=Format Text!(
     Format="LOCTEXT(\"<{Style}>{Text}</>\")"
     Style:Text="To Text (Name)"!(InName:Name=@n2) -> (ReturnValue:Text)
     Text:"Text (by ref)"=@n1.DisplayText
    ) -> (Result)
   )
   Event On Minigame Started() {is=title K2Node_Event}
   > SetRenderTranslation(Translation=(X=0,Y=80))
   > Parent: On Minigame Started() {is=title K2Node_CallParentFunction}
   Event On Minigame Ended() {is=title K2Node_Event}
   > SetRenderTranslation(Translation=(X=0,Y=0))
   > Parent: On Minigame Ended() {is=title K2Node_CallParentFunction}
  ]}
 IsTextStyleValid {nodes [
   IsTextStyleValid!() -> (then, DisplayTextStyle:"Name (by ref)"=None) {is=title K2Node_FunctionEntry}
   > n1 = GetDataTableRowNames(Table=/Game/Athena/HUD/TextStyles/AthenaStandaloneTimerStyles)
   > n2 = For Each Loop(Array:Array of Names=@n1) -> (Array Element:Name)
   Return Node!(execute=@n2.Completed, ReturnValue:bool=false)
   > Branch(Condition="Equal (Name)"!(A:Name=@n2, B:Name=Get DisplayTextStyle() -> (DisplayTextStyle:"Name (by ref)"=None)) -> (ReturnValue:bool) {is=title K2Node_CallFunction})
   > Return Node(ReturnValue=true)
  ]}
]
variables [DefaultStyle {type=string, default=Heist}]
widgets {
 info {inheritedWidgetCount=1, parentClass=/Script/FortniteUI.AthenaTimeWidgetBase, rootWidgetClass=&SizeBox, widgetCount=5}
 widgets [
  SizeBox_0 {widgetClassPath=&SizeBox, widget=SizeBox_0, properties {bOverride_MinDesiredWidth=true, minDesiredWidth=106}}
  Overlay_0 {
   widgetClassPath=&Overlay
   parent=SizeBox_0
   widget=Overlay_0
   slot {class=SizeBoxSlot, properties {horizontalAlignment=HAlign_Center, verticalAlignment=VAlign_Center}}
  }
  Image_background {
   widgetClassPath=&Image
   parent=Overlay_0
   widget=Image_background
   properties {
    brush {
     bIsDynamicallyLoaded=false
     drawAs=Box
     imageType=NoImage
     mirroring=NoMirror
     resourceName=None
     resourceObject=/FortUILibrary/Foundation/Buttons/Materials/M_SkewButtonShape_DF
     tiling=NoTile
     imageSize=<88.8474 46.9781>
     margin=<ltrb 0.05 0.05 0.05 0.05>
     outlineSettings {
      bUseBrushTransparency=false
      roundingType=HalfHeightRadius
      width=0
      color {colorUseRule=UseColor_Specified, specifiedColor=<rgba 0 0 0 0>}
      cornerRadii=<0 0 0 0>
     }
     tintColor {colorUseRule=UseColor_Specified, specifiedColor=<rgba 1 1 1 1>}
     uVRegion {bIsValid=false, max=<0 0>, min=<0 0>}
    }
    colorAndOpacity=<rgba 0.0052 0.0052 0.0056 0.15>
   }
   slot {class=OverlaySlot, properties {horizontalAlignment=HAlign_Fill, verticalAlignment=VAlign_Fill, padding=<ltrb 0 -5 0 -3>}}
  }
  TimeRichText {
   widgetClassPath=/Script/UMG.RichTextBlock
   bIsVariable=true
   parent=Overlay_0
   widget=TimeRichText
   properties {justification=Center, text="<Heist>88:88</>", textStyleSet=&AthenaStandaloneTimerStyles}
   slot {class=OverlaySlot, properties {horizontalAlignment=HAlign_Center, padding=<ltrb 12 0 10 0>}}
  }
  TutorialHighlight {widgetClassPath=&TutorialHighlight}
 ]
}
