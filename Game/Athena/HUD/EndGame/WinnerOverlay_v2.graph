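# WinnerOverlay_v2 is the end-game HUD widget that presents the Victory Royale celebration.
# Victory_v2 punches in win text over a glow and vignette, then swaps it for a badge with scaling Victory Royale text, number-one emblem and flares.
# Victory_v2_b starts at the badge reveal and adds a lightshaft; Victory_v2_b2 is another variant, and SpinningSpikeball slowly rotates one image.
# Animated material parameters drive the backplate reveal, text shine, white-out and glow pulse.
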WidgetBlueprint /Game/Athena/HUD/EndGame/WinnerOverlay_v2
animations [
 Victory_v2 {
  binding_count=40
  display_label=""
  length_seconds=5.0298
  object=&self:Victory_v2
  sequence {
   bindings [
    Image_50 {
     id=#1
     possessed_class=Image
     tracks [{
       type=MovieScene2DTransformTrack
       sections [{
         type=MovieScene2DTransformSection
         end_frame=67
         start_frame=0
         channels [
          Translation.X {type=MovieSceneScriptingFloatChannel}
          Translation.Y {type=MovieSceneScriptingFloatChannel}
          Angle {type=MovieSceneScriptingFloatChannel}
          Scale.X {type=MovieSceneScriptingFloatChannel, keys [[0,0],[10,0.004],[16,1.074],[29,0.9],[37,0.9],[38,1.408],[56,0.939]]}
          Scale.Y {type=MovieSceneScriptingFloatChannel, keys [[0,0],[10,0.004],[16,1.074],[29,0.9],[37,0.9],[38,1.408],[56,0.939]]}
          Shear.X {type=MovieSceneScriptingFloatChannel}
          Shear.Y {type=MovieSceneScriptingFloatChannel}
         ]
        }]
      }]
    }
    Image_52 {
     id=#2
     possessed_class=Image
     tracks [{
       type=MovieScene2DTransformTrack
       sections [{
         type=MovieScene2DTransformSection
         end_frame=67
         start_frame=0
         channels [
          Translation.X {type=MovieSceneScriptingFloatChannel}
          Translation.Y {type=MovieSceneScriptingFloatChannel}
          Angle {type=MovieSceneScriptingFloatChannel}
          Scale.X {type=MovieSceneScriptingFloatChannel, keys [[0,0],[10,0.004],[16,1.074],[29,0.9],[37,0.9],[38,1.408],[56,0.939]]}
          Scale.Y {type=MovieSceneScriptingFloatChannel, keys [[0,0],[10,0.004],[16,1.074],[29,0.9],[37,0.9],[38,1.408],[56,0.939]]}
          Shear.X {type=MovieSceneScriptingFloatChannel}
          Shear.Y {type=MovieSceneScriptingFloatChannel}
         ]
        }]
      }]
    }
    Image_107 {
     id=#3
     possessed_class=Image
     tracks [
      {
       type=MovieSceneFloatTrack
       display_name=Render Opacity
       sections [{
         type=MovieSceneFloatSection
         end_frame=273
         start_frame=0
         channels [None {type=MovieSceneScriptingFloatChannel, keys [[185.156,0],[200.615,0.8097],[217.421,1],[255.585,0]]}]
        }]
      }
      {
       type=MovieScene2DTransformTrack
       sections [{
         type=MovieScene2DTransformSection
         end_frame=274
         start_frame=59
         channels [
          Translation.X {type=MovieSceneScriptingFloatChannel, keys [[184.875,163.574],[256.302,362]]}
          Translation.Y {type=MovieSceneScriptingFloatChannel, keys [[184.875,-227.803],[256.302,-252]]}
          Angle {type=MovieSceneScriptingFloatChannel}
          Scale.X {type=MovieSceneScriptingFloatChannel}
          Scale.Y {type=MovieSceneScriptingFloatChannel}
          Shear.X {type=MovieSceneScriptingFloatChannel}
          Shear.Y {type=MovieSceneScriptingFloatChannel}
         ]
        }]
      }
     ]
    }
    ImageCross {
     id=#4
     possessed_class=Image
     tracks [{
       type=MovieSceneFloatTrack
       display_name=Render Opacity
       sections [{
         type=MovieSceneFloatSection
         end_frame=268
         start_frame=0
         channels [None {type=MovieSceneScriptingFloatChannel, keys [[187.427,0],[195.622,0.8097],[212.428,1],[250.592,0]]}]
        }]
      }]
    }
    Image_text1flarebg {
     id=#5
     possessed_class=Image
     tracks [
      {
       type=MovieSceneFloatTrack
       display_name=Render Opacity
       sections [{
         type=MovieSceneFloatSection
         end_frame=221
         start_frame=0
         channels [None {type=MovieSceneScriptingFloatChannel, keys [[164.491,0],[172.157,0.8097],[188.963,1],[203.067,0]]}]
        }]
      }
      {
       type=MovieScene2DTransformTrack
       sections [{
         type=MovieScene2DTransformSection
         end_frame=237
         start_frame=9
         channels [
          Translation.X {type=MovieSceneScriptingFloatChannel}
          Translation.Y {type=MovieSceneScriptingFloatChannel}
          Angle {type=MovieSceneScriptingFloatChannel}
          Scale.X {type=MovieSceneScriptingFloatChannel, keys [[164.728,1],[173.042,1.2],[219.339,1]]}
          Scale.Y {type=MovieSceneScriptingFloatChannel, keys [[164.728,0.14],[173.042,1.2],[219.339,1]]}
          Shear.X {type=MovieSceneScriptingFloatChannel}
          Shear.Y {type=MovieSceneScriptingFloatChannel}
         ]
        }]
      }
     ]
    }
    CommonTextBlock_5 {
     id=#6
     possessed_class=CommonTextBlock
     tracks [
      {
       type=MovieScene2DTransformTrack
       sections [{
         type=MovieScene2DTransformSection
         end_frame=188
         start_frame=0
         channels [
          Translation.X {type=MovieSceneScriptingFloatChannel}
          Translation.Y {type=MovieSceneScriptingFloatChannel}
          Angle {type=MovieSceneScriptingFloatChannel}
          Scale.X {type=MovieSceneScriptingFloatChannel, keys [[142.059,0],[151,0.001],[160.385,1.222],[163.769,1],[170.096,1]]}
          Scale.Y {type=MovieSceneScriptingFloatChannel, keys [[142.059,0],[151,0.001],[160.385,1.222],[163.769,1],[170.096,1]]}
          Shear.X {type=MovieSceneScriptingFloatChannel}
          Shear.Y {type=MovieSceneScriptingFloatChannel}
         ]
        }]
      }
      {
       type=MovieSceneWidgetMaterialTrack
       display_name=Font Material
       sections [{
         type=MovieSceneParameterSection
         end_frame=290
         start_frame=100
         channels [ShineOffset {type=MovieSceneScriptingFloatChannel, keys [[231.819,0.6],[272.819,-0.7262]]}]
        }]
      }
     ]
    }
    BorderNumber1Text {
     id=#7
     possessed_class=Border
     tracks [{
       type=MovieScene2DTransformTrack
       sections [{
         type=MovieScene2DTransformSection
         end_frame=204
         start_frame=0
         channels [
          Translation.X {type=MovieSceneScriptingFloatChannel}
          Translation.Y {type=MovieSceneScriptingFloatChannel}
          Angle {type=MovieSceneScriptingFloatChannel}
          Scale.X {type=MovieSceneScriptingFloatChannel, keys [[166.224,0],[168.985,0],[176.606,0.6],[180.525,0.999],[182.525,0.6],[186.347,0.6]]}
          Scale.Y {type=MovieSceneScriptingFloatChannel, keys [[166.224,0],[168.985,0],[176.606,0.6],[180.525,0.999],[182.525,0.6],[186.347,0.6]]}
          Shear.X {type=MovieSceneScriptingFloatChannel}
          Shear.Y {type=MovieSceneScriptingFloatChannel}
         ]
        }]
      }]
    }
    Image_text1flare1 {
     id=#8
     possessed_class=Image
     tracks [{
       type=MovieSceneFloatTrack
       display_name=Render Opacity
       sections [{
         type=MovieSceneFloatSection
         end_frame=249
         start_frame=0
         channels [None {type=MovieSceneScriptingFloatChannel, keys [[164.491,0],[172.157,0.8097],[188.963,1],[231.819,0]]}]
        }]
      }]
    }
    CommonTextYOU {
     id=#9
     possessed_class=CommonTextBlock
     tracks [
      {
       type=MovieScene2DTransformTrack
       sections [{
         type=MovieScene2DTransformSection
         end_frame=151
         start_frame=0
         channels [
          Translation.X {type=MovieSceneScriptingFloatChannel}
          Translation.Y {type=MovieSceneScriptingFloatChannel}
          Angle {type=MovieSceneScriptingFloatChannel}
          Scale.X {type=MovieSceneScriptingFloatChannel, thinned_from=6, keys [[0,1],[15,0.4],[109.435,0.4],[122.227,0.055]]}
          Scale.Y {type=MovieSceneScriptingFloatChannel, thinned_from=6, keys [[0,1],[15,0.4],[109.435,0.4],[122.227,0.055]]}
          Shear.X {type=MovieSceneScriptingFloatChannel}
          Shear.Y {type=MovieSceneScriptingFloatChannel}
         ]
        }]
      }
      {
       type=MovieSceneWidgetMaterialTrack
       display_name=Font Material
       sections [{
         type=MovieSceneParameterSection
         end_frame=195
         start_frame=0
         channels [
          WhiteOut {type=MovieSceneScriptingFloatChannel, keys [[0,1],[15,0]]}
          ShineOffset {type=MovieSceneScriptingFloatChannel, keys [[25,0.4],[48.411,-0.545]]}
         ]
        }]
      }
      {
       type=MovieSceneFloatTrack
       display_name=Render Opacity
       sections [{type=MovieSceneFloatSection, channels [None {type=MovieSceneScriptingFloatChannel, keys [[0,0],[2,1]]}]}]
      }
     ]
    }
    OverlayRoyaleBadge {
     id=#10
     possessed_class=Overlay
     tracks [{
       type=MovieSceneFloatTrack
       display_name=Render Opacity
       sections [{
         type=MovieSceneFloatSection
         end_frame=195
         start_frame=0
         channels [None {type=MovieSceneScriptingFloatChannel, keys [[55.525,0],[133.834,0],[146.333,1]]}]
        }]
      }]
    }
    Image_47 {
     id=#11
     possessed_class=Image
     tracks [{
       type=MovieScene2DTransformTrack
       sections [{
         type=MovieScene2DTransformSection
         end_frame=67
         start_frame=0
         channels [
          Translation.X {type=MovieSceneScriptingFloatChannel}
          Translation.Y {type=MovieSceneScriptingFloatChannel}
          Angle {type=MovieSceneScriptingFloatChannel}
          Scale.X {type=MovieSceneScriptingFloatChannel, keys [[0,0],[10,0.004],[16,1.074],[29,0.9],[37,0.9],[38,1.408],[56,0.939]]}
          Scale.Y {type=MovieSceneScriptingFloatChannel, keys [[0,0],[10,0.004],[16,1.074],[29,0.9],[37,0.9],[38,1.408],[56,0.939]]}
          Shear.X {type=MovieSceneScriptingFloatChannel}
          Shear.Y {type=MovieSceneScriptingFloatChannel}
         ]
        }]
      }]
    }
    Image_49 {
     id=#12
     possessed_class=Image
     tracks [{
       type=MovieScene2DTransformTrack
       sections [{
         type=MovieScene2DTransformSection
         end_frame=67
         start_frame=0
         channels [
          Translation.X {type=MovieSceneScriptingFloatChannel}
          Translation.Y {type=MovieSceneScriptingFloatChannel}
          Angle {type=MovieSceneScriptingFloatChannel}
          Scale.X {type=MovieSceneScriptingFloatChannel, keys [[0,0],[10,0.004],[16,1.074],[29,0.9],[37,0.9],[38,1.408],[56,0.939]]}
          Scale.Y {type=MovieSceneScriptingFloatChannel, keys [[0,0],[10,0.004],[16,1.074],[29,0.9],[37,0.9],[38,1.408],[56,0.939]]}
          Shear.X {type=MovieSceneScriptingFloatChannel}
          Shear.Y {type=MovieSceneScriptingFloatChannel}
         ]
        }]
      }]
    }
    Image_text2flarebg {
     id=#13
     possessed_class=Image
     tracks [
      {
       type=MovieSceneFloatTrack
       display_name=Render Opacity
       sections [{
         type=MovieSceneFloatSection
         end_frame=237
         start_frame=0
         channels [None {type=MovieSceneScriptingFloatChannel, keys [[164.728,0],[178.194,0.8097],[195,1],[219.873,0]]}]
        }]
      }
      {
       type=MovieScene2DTransformTrack
       sections [{
         type=MovieScene2DTransformSection
         end_frame=254
         start_frame=4
         channels [
          Translation.X {type=MovieSceneScriptingFloatChannel}
          Translation.Y {type=MovieSceneScriptingFloatChannel}
          Angle {type=MovieSceneScriptingFloatChannel}
          Scale.X {type=MovieSceneScriptingFloatChannel, keys [[164.965,1],[173.042,1.2],[236.145,1]]}
          Scale.Y {type=MovieSceneScriptingFloatChannel, keys [[164.965,0.37],[173.042,1.2],[236.145,1]]}
          Shear.X {type=MovieSceneScriptingFloatChannel}
          Shear.Y {type=MovieSceneScriptingFloatChannel}
         ]
        }]
      }
     ]
    }
    OverlayWinText {
     id=#14
     possessed_class=Overlay
     tracks [{
       type=MovieSceneFloatTrack
       display_name=Render Opacity
       sections [{type=MovieSceneFloatSection, channels [None {type=MovieSceneScriptingFloatChannel, keys [[104.804,1],[125.575,0]]}]}]
      }]
    }
    Imagetext2flare1 {
     id=#15
     possessed_class=Image
     tracks [{
       type=MovieSceneFloatTrack
       display_name=Render Opacity
       sections [{
         type=MovieSceneFloatSection
         end_frame=266
         start_frame=0
         channels [None {type=MovieSceneScriptingFloatChannel, keys [[164.728,0],[188.963,0.8097],[205.769,1],[248.625,0]]}]
        }]
      }]
    }
    Image_46 {
     id=#16
     possessed_class=Image
     tracks [{
       type=MovieScene2DTransformTrack
       sections [{
         type=MovieScene2DTransformSection
         end_frame=67
         start_frame=0
         channels [
          Translation.X {type=MovieSceneScriptingFloatChannel}
          Translation.Y {type=MovieSceneScriptingFloatChannel}
          Angle {type=MovieSceneScriptingFloatChannel}
          Scale.X {type=MovieSceneScriptingFloatChannel, keys [[0,0],[10,0.004],[16,1.074],[29,0.9],[37,0.9],[38,1.408],[56,0.939]]}
          Scale.Y {type=MovieSceneScriptingFloatChannel, keys [[0,0],[10,0.004],[16,1.074],[29,0.9],[37,0.9],[38,1.408],[56,0.939]]}
          Shear.X {type=MovieSceneScriptingFloatChannel}
          Shear.Y {type=MovieSceneScriptingFloatChannel}
         ]
        }]
      }]
    }
    ImageVignette {
     id=#17
     possessed_class=Image
     tracks [{
       type=MovieSceneFloatTrack
       display_name=Render Opacity
       sections [{type=MovieSceneFloatSection, end_frame=195, start_frame=0, channels [None {type=MovieSceneScriptingFloatChannel, keys [[0,0],[8.935,1]]}]}]
      }]
    }
    ImageBackplate {
     id=#18
     possessed_class=Image
     tracks [{
       type=MovieSceneWidgetMaterialTrack
       display_name=Brush Material
       sections [{
         type=MovieSceneParameterSection
         end_frame=268
         start_frame=0
         channels [
          RevealDensity {type=MovieSceneScriptingFloatChannel, keys [[142.06,0.0632],[176.031,2.4457]]}
          RevealRadius {type=MovieSceneScriptingFloatChannel, keys [[142.06,0],[176.031,1.3973]]}
          ShineOffset {type=MovieSceneScriptingFloatChannel, keys [[181.706,0.6576],[250.526,-1.0328]]}
         ]
        }]
      }]
    }
    Imagespikeball2 {
     id=#19
     possessed_class=Image
     tracks [{
       type=MovieSceneFloatTrack
       display_name=Render Opacity
       sections [{
         type=MovieSceneFloatSection
         end_frame=294
         start_frame=0
         channels [None {type=MovieSceneScriptingFloatChannel, keys [[213.787,0],[221.982,0.8097],[238.788,1],[276.952,0]]}]
        }]
      }]
    }
    ImageSpinningSpikeball1 {
     id=#20
     possessed_class=Image
     tracks [{
       type=MovieSceneFloatTrack
       display_name=Render Opacity
       sections [{
         type=MovieSceneFloatSection
         end_frame=273
         start_frame=0
         channels [None {type=MovieSceneScriptingFloatChannel, keys [[230.584,0],[238.224,0],[238.779,0.8097],[255.585,1]]}]
        }]
      }]
    }
    Image_53 {
     id=#21
     possessed_class=Image
     tracks [{
       type=MovieScene2DTransformTrack
       sections [{
         type=MovieScene2DTransformSection
         end_frame=67
         start_frame=0
         channels [
          Translation.X {type=MovieSceneScriptingFloatChannel}
          Translation.Y {type=MovieSceneScriptingFloatChannel}
          Angle {type=MovieSceneScriptingFloatChannel}
          Scale.X {type=MovieSceneScriptingFloatChannel, keys [[0,0],[10,0.004],[16,1.074],[29,0.9],[37,0.9],[38,1.408],[56,0.939]]}
          Scale.Y {type=MovieSceneScriptingFloatChannel, keys [[0,0],[10,0.004],[16,1.074],[29,0.9],[37,0.9],[38,1.408],[56,0.939]]}
          Shear.X {type=MovieSceneScriptingFloatChannel}
          Shear.Y {type=MovieSceneScriptingFloatChannel}
         ]
        }]
      }]
    }
    Image_51 {
     id=#22
     possessed_class=Image
     tracks [{
       type=MovieScene2DTransformTrack
       sections [{
         type=MovieScene2DTransformSection
         end_frame=67
         start_frame=0
         channels [
          Translation.X {type=MovieSceneScriptingFloatChannel}
          Translation.Y {type=MovieSceneScriptingFloatChannel}
          Angle {type=MovieSceneScriptingFloatChannel}
          Scale.X {type=MovieSceneScriptingFloatChannel, keys [[0,0],[10,0.004],[16,1.074],[29,0.9],[37,0.9],[38,1.408],[56,0.939]]}
          Scale.Y {type=MovieSceneScriptingFloatChannel, keys [[0,0],[10,0.004],[16,1.074],[29,0.9],[37,0.9],[38,1.408],[56,0.939]]}
          Shear.X {type=MovieSceneScriptingFloatChannel}
          Shear.Y {type=MovieSceneScriptingFloatChannel}
         ]
        }]
      }]
    }
    Image_48 {
     id=#23
     possessed_class=Image
     tracks [{
       type=MovieScene2DTransformTrack
       sections [{
         type=MovieScene2DTransformSection
         end_frame=67
         start_frame=0
         channels [
          Translation.X {type=MovieSceneScriptingFloatChannel}
          Translation.Y {type=MovieSceneScriptingFloatChannel}
          Angle {type=MovieSceneScriptingFloatChannel}
          Scale.X {type=MovieSceneScriptingFloatChannel, keys [[0,0],[10,0.004],[16,1.074],[29,0.9],[37,0.9],[38,1.408],[56,0.939]]}
          Scale.Y {type=MovieSceneScriptingFloatChannel, keys [[0,0],[10,0.004],[16,1.074],[29,0.9],[37,0.9],[38,1.408],[56,0.939]]}
          Shear.X {type=MovieSceneScriptingFloatChannel}
          Shear.Y {type=MovieSceneScriptingFloatChannel}
         ]
        }]
      }]
    }
    CommonTextBlock_1 {
     id=#24
     possessed_class=CommonTextBlock
     tracks [
      {
       type=MovieScene2DTransformTrack
       sections [{
         type=MovieScene2DTransformSection
         end_frame=204
         start_frame=0
         channels [
          Translation.X {type=MovieSceneScriptingFloatChannel}
          Translation.Y {type=MovieSceneScriptingFloatChannel}
          Angle {type=MovieSceneScriptingFloatChannel}
          Scale.X {type=MovieSceneScriptingFloatChannel, thinned_from=6, keys [[22,1],[37,0.4],[112.783,0.4],[125.575,0.055]]}
          Scale.Y {type=MovieSceneScriptingFloatChannel, thinned_from=6, keys [[22,1],[37,0.4],[112.783,0.4],[125.575,0.055]]}
          Shear.X {type=MovieSceneScriptingFloatChannel}
          Shear.Y {type=MovieSceneScriptingFloatChannel}
         ]
        }]
      }
      {
       type=MovieSceneWidgetMaterialTrack
       display_name=Font Material
       sections [{
         type=MovieSceneParameterSection
         end_frame=247
         start_frame=0
         channels [
          WhiteOut {type=MovieSceneScriptingFloatChannel, keys [[22,1],[37,0]]}
          ShineOffset {type=MovieSceneScriptingFloatChannel, keys [[46,0.4],[69.411,-0.545]]}
         ]
        }]
      }
      {
       type=MovieSceneFloatTrack
       display_name=Render Opacity
       sections [{type=MovieSceneFloatSection, channels [None {type=MovieSceneScriptingFloatChannel, keys [[22,0],[24,1]]}]}]
      }
     ]
    }
    Image_108 {
     id=#25
     possessed_class=Image
     tracks [{
       type=MovieSceneFloatTrack
       display_name=Render Opacity
       sections [{
         type=MovieSceneFloatSection
         end_frame=270
         start_frame=0
         channels [None {type=MovieSceneScriptingFloatChannel, keys [[189.242,0],[197.437,0.8097],[214.243,1],[252.407,0]]}]
        }]
      }]
    }
    Image_text2flare2 {
     id=#26
     possessed_class=Image
     tracks [{
       type=MovieSceneFloatTrack
       display_name=Render Opacity
       sections [{
         type=MovieSceneFloatSection
         end_frame=266
         start_frame=0
         channels [None {type=MovieSceneScriptingFloatChannel, keys [[164.728,0],[188.963,0.8097],[205.769,1],[248.625,0]]}]
        }]
      }]
    }
    CommonTextRoyale {
     id=#27
     possessed_class=CommonTextBlock
     tracks [{
       type=MovieScene2DTransformTrack
       sections [{
         type=MovieScene2DTransformSection
         end_frame=188
         start_frame=0
         channels [
          Translation.X {type=MovieSceneScriptingFloatChannel}
          Translation.Y {type=MovieSceneScriptingFloatChannel}
          Angle {type=MovieSceneScriptingFloatChannel}
          Scale.X {type=MovieSceneScriptingFloatChannel, keys [[142.059,0],[151,0.001],[160.385,1.222],[163.769,1],[170.096,1]]}
          Scale.Y {type=MovieSceneScriptingFloatChannel, keys [[142.059,0],[151,0.001],[160.385,1.222],[163.769,1],[170.096,1]]}
          Shear.X {type=MovieSceneScriptingFloatChannel}
          Shear.Y {type=MovieSceneScriptingFloatChannel}
         ]
        }]
      }]
    }
    CommonTextBlock_4 {
     id=#28
     possessed_class=CommonTextBlock
     tracks [
      {
       type=MovieScene2DTransformTrack
       sections [{
         type=MovieScene2DTransformSection
         end_frame=283
         start_frame=0
         channels [
          Translation.X {type=MovieSceneScriptingFloatChannel}
          Translation.Y {type=MovieSceneScriptingFloatChannel}
          Angle {type=MovieSceneScriptingFloatChannel}
          Scale.X {type=MovieSceneScriptingFloatChannel, keys [[142.059,0],[151,0.001],[160.385,1.222],[163.769,1],[170.096,1]]}
          Scale.Y {type=MovieSceneScriptingFloatChannel, keys [[142.059,0],[151,0.001],[160.385,1.222],[163.769,1],[170.096,1]]}
          Shear.X {type=MovieSceneScriptingFloatChannel}
          Shear.Y {type=MovieSceneScriptingFloatChannel}
         ]
        }]
      }
      {
       type=MovieSceneWidgetMaterialTrack
       display_name=Font Material
       sections [{
         type=MovieSceneParameterSection
         end_frame=283
         start_frame=104
         channels [ShineOffset {type=MovieSceneScriptingFloatChannel, keys [[224.525,0.59],[265.525,-0.7]]}]
        }]
      }
     ]
    }
    ImageGlow {
     id=#29
     possessed_class=Image
     tracks [
      {
       type=MovieSceneFloatTrack
       display_name=Render Opacity
       sections [{type=MovieSceneFloatSection, channels [None {type=MovieSceneScriptingFloatChannel, keys [[0,0],[14.597,0],[15.526,1]]}]}]
      }
      {
       type=MovieSceneWidgetMaterialTrack
       display_name=Brush Material
       sections [{
         type=MovieSceneParameterSection
         end_frame=195
         start_frame=0
         channels [Radius {type=MovieSceneScriptingFloatChannel, keys [[14.133,0.4697],[15,0.2777],[19,0.564],[27,0.4],[37,0.4],[38,0.571],[48,0.47]]}]
        }]
      }
     ]
    }
    Image_text1flare2 {
     id=#30
     possessed_class=Image
     tracks [{
       type=MovieSceneFloatTrack
       display_name=Render Opacity
       sections [{
         type=MovieSceneFloatSection
         end_frame=249
         start_frame=0
         channels [None {type=MovieSceneScriptingFloatChannel, keys [[164.491,0],[172.157,0.8097],[188.963,1],[231.819,0]]}]
        }]
      }]
    }
    CommonTextBlock_6 {
     id=#31
     possessed_class=CommonTextBlock
     tracks [{
       type=MovieScene2DTransformTrack
       sections [{
         type=MovieScene2DTransformSection
         end_frame=188
         start_frame=0
         channels [
          Translation.X {type=MovieSceneScriptingFloatChannel}
          Translation.Y {type=MovieSceneScriptingFloatChannel}
          Angle {type=MovieSceneScriptingFloatChannel}
          Scale.X {type=MovieSceneScriptingFloatChannel, keys [[142.059,0],[151,0.001],[160.385,1.222],[163.769,1],[170.096,1]]}
          Scale.Y {type=MovieSceneScriptingFloatChannel, keys [[142.059,0],[151,0.001],[160.385,1.222],[163.769,1],[170.096,1]]}
          Shear.X {type=MovieSceneScriptingFloatChannel}
          Shear.Y {type=MovieSceneScriptingFloatChannel}
         ]
        }]
      }]
    }
    Image1Cross {
     id=#32
     possessed_class=Image
     tracks [{
       type=MovieSceneFloatTrack
       display_name=Render Opacity
       sections [{
         type=MovieSceneFloatSection
         end_frame=301
         start_frame=0
         channels [None {type=MovieSceneScriptingFloatChannel, keys [[220.597,0],[228.792,0.8097],[245.598,1],[283.762,0]]}]
        }]
      }]
    }
   ]
   display_rate {numerator=60}
  }
 }
 Victory_v2_b {
  binding_count=29
  display_label=""
  length_seconds=2.8333
  object=&self:Victory_v2_b
  sequence {
   bindings [
    Image_107 {
     id=#3
     possessed_class=Image
     tracks [
      {
       type=MovieSceneFloatTrack
       display_name=Render Opacity
       sections [{
         type=MovieSceneFloatSection
         end_frame=312
         start_frame=0
         channels [None {type=MovieSceneScriptingFloatChannel, keys [[86.312,0],[117.23,0.8097],[150.842,1],[227.17,0]]}]
        }]
      }
      {
       type=MovieScene2DTransformTrack
       sections [{
         type=MovieScene2DTransformSection
         end_frame=314
         start_frame=0
         channels [
          Translation.X {type=MovieSceneScriptingFloatChannel, keys [[85.75,-12.0191],[228.604,186.407]]}
          Translation.Y {type=MovieSceneScriptingFloatChannel, keys [[85.75,-181.193],[228.604,-205.39]]}
          Angle {type=MovieSceneScriptingFloatChannel, keys [[85.75,0],[228.604,20.494]]}
          Scale.X {type=MovieSceneScriptingFloatChannel}
          Scale.Y {type=MovieSceneScriptingFloatChannel}
          Shear.X {type=MovieSceneScriptingFloatChannel}
          Shear.Y {type=MovieSceneScriptingFloatChannel}
         ]
        }]
      }
     ]
    }
    ImageCross {
     id=#4
     possessed_class=Image
     tracks [{
       type=MovieSceneFloatTrack
       display_name=Render Opacity
       sections [{
         type=MovieSceneFloatSection
         end_frame=302
         start_frame=0
         channels [None {type=MovieSceneScriptingFloatChannel, keys [[90.854,0],[107.244,0.8097],[140.856,1],[217.184,0]]}]
        }]
      }]
    }
    Image_Lightshaft {
     id=#33
     possessed_class=Image
     tracks [
      {
       type=MovieSceneWidgetMaterialTrack
       display_name=Brush Material
       sections [{
         type=MovieSceneParameterSection
         end_frame=340
         start_frame=0
         channels [
          ShaftMax {type=MovieSceneScriptingFloatChannel, const [1.0018,2]}
          ShaftMin {type=MovieSceneScriptingFloatChannel, keys [[44,0.9713],[60,-2.7486],[88,0.9713]]}
         ]
        }]
      }
      {
       type=MovieSceneFloatTrack
       display_name=Render Opacity
       sections [{
         type=MovieSceneFloatSection
         end_frame=340
         start_frame=0
         channels [None {type=MovieSceneScriptingFloatChannel, keys [[44.002,0],[48,1],[88,0]]}]
        }]
      }
     ]
    }
    Image_text1flarebg {
     id=#5
     possessed_class=Image
     tracks [
      {
       type=MovieSceneFloatTrack
       display_name=Render Opacity
       sections [{
         type=MovieSceneFloatSection
         end_frame=257
         start_frame=0
         channels [None {type=MovieSceneScriptingFloatChannel, keys [[44.982,0],[60.314,0.8097],[93.926,1],[122.134,0]]}]
        }]
      }
      {
       type=MovieScene2DTransformTrack
       sections [{
         type=MovieScene2DTransformSection
         end_frame=232
         start_frame=0
         channels [
          Translation.X {type=MovieSceneScriptingFloatChannel}
          Translation.Y {type=MovieSceneScriptingFloatChannel}
          Angle {type=MovieSceneScriptingFloatChannel}
          Scale.X {type=MovieSceneScriptingFloatChannel, keys [[45.456,1],[62.084,1.2],[154.678,1]]}
          Scale.Y {type=MovieSceneScriptingFloatChannel, keys [[45.456,0.14],[62.084,1.2],[154.678,1]]}
          Shear.X {type=MovieSceneScriptingFloatChannel}
          Shear.Y {type=MovieSceneScriptingFloatChannel}
         ]
        }]
      }
     ]
    }
    CommonTextBlock_5 {
     id=#6
     display_name=RichTextBlock_Royale
     possessed_class=CommonTextBlock
     tracks [
      {
       type=MovieScene2DTransformTrack
       sections [{
         type=MovieScene2DTransformSection
         end_frame=223
         start_frame=0
         channels [
          Translation.X {type=MovieSceneScriptingFloatChannel}
          Translation.Y {type=MovieSceneScriptingFloatChannel}
          Angle {type=MovieSceneScriptingFloatChannel}
          Scale.X {type=MovieSceneScriptingFloatChannel, keys [[0.118,0],[18,0.001],[36.77,1.222],[43.538,1],[56.192,1]]}
          Scale.Y {type=MovieSceneScriptingFloatChannel, keys [[0.118,0],[18,0.001],[36.77,1.222],[43.538,1],[56.192,1]]}
          Shear.X {type=MovieSceneScriptingFloatChannel}
          Shear.Y {type=MovieSceneScriptingFloatChannel}
         ]
        }]
      }
      {
       type=MovieSceneWidgetMaterialTrack
       display_name=Font Material
       sections [{
         type=MovieSceneParameterSection
         end_frame=319
         start_frame=0
         channels [ShineOffset {type=MovieSceneScriptingFloatChannel, keys [[179.638,0.6],[261.638,-0.7262]]}]
        }]
      }
     ]
    }
    BorderNumber1Text {
     id=#7
     possessed_class=Border
     tracks [
      {
       type=MovieScene2DTransformTrack
       sections [{
         type=MovieScene2DTransformSection
         end_frame=340
         start_frame=0
         channels [
          Translation.X {type=MovieSceneScriptingFloatChannel}
          Translation.Y {type=MovieSceneScriptingFloatChannel}
          Angle {type=MovieSceneScriptingFloatChannel}
          Scale.X {type=MovieSceneScriptingFloatChannel, keys [[53.97,0.3925],[69.212,0.9925],[77.05,1.3915],[81.05,0.9925],[88.694,0.9925]]}
          Scale.Y {type=MovieSceneScriptingFloatChannel, keys [[53.97,0.3925],[69.212,0.9925],[77.05,1.3915],[81.05,0.9925],[88.694,0.9925]]}
          Shear.X {type=MovieSceneScriptingFloatChannel}
          Shear.Y {type=MovieSceneScriptingFloatChannel}
         ]
        }]
      }
      {
       type=MovieSceneFloatTrack
       display_name=Render Opacity
       sections [{
         type=MovieSceneFloatSection
         end_frame=316
         start_frame=0
         channels [None {type=MovieSceneScriptingFloatChannel, keys [[53.97,0],[66.364,1]]}]
        }]
      }
     ]
    }
    Image_text1flare1 {
     id=#8
     possessed_class=Image
     tracks [{
       type=MovieSceneFloatTrack
       display_name=Render Opacity
       sections [{
         type=MovieSceneFloatSection
         end_frame=257
         start_frame=0
         channels [None {type=MovieSceneScriptingFloatChannel, keys [[44.982,0],[60.314,0.8097],[93.926,1],[179.638,0]]}]
        }]
      }]
    }
    OverlayRoyaleBadge {
     id=#10
     possessed_class=Overlay
     tracks [{
       type=MovieSceneFloatTrack
       display_name=Render Opacity
       sections [{
         type=MovieSceneFloatSection
         end_frame=308
         start_frame=0
         channels [None {type=MovieSceneScriptingFloatChannel, keys [[0.972,0],[25.97,1]]}]
        }]
      }]
    }
    Image_text2flarebg {
     id=#13
     possessed_class=Image
     tracks [
      {
       type=MovieSceneFloatTrack
       display_name=Render Opacity
       sections [{
         type=MovieSceneFloatSection
         end_frame=233
         start_frame=0
         channels [None {type=MovieSceneScriptingFloatChannel, keys [[45.456,0],[72.388,0.8097],[106,1],[155.746,0]]}]
        }]
      }
      {
       type=MovieScene2DTransformTrack
       sections [{
         type=MovieScene2DTransformSection
         end_frame=265
         start_frame=0
         channels [
          Translation.X {type=MovieSceneScriptingFloatChannel}
          Translation.Y {type=MovieSceneScriptingFloatChannel}
          Angle {type=MovieSceneScriptingFloatChannel}
          Scale.X {type=MovieSceneScriptingFloatChannel, keys [[45.93,1],[62.084,1.2],[188.29,1]]}
          Scale.Y {type=MovieSceneScriptingFloatChannel, keys [[45.93,0.37],[62.084,1.2],[188.29,1]]}
          Shear.X {type=MovieSceneScriptingFloatChannel}
          Shear.Y {type=MovieSceneScriptingFloatChannel}
         ]
        }]
      }
     ]
    }
    Imagetext2flare1 {
     id=#15
     possessed_class=Image
     tracks [{
       type=MovieSceneFloatTrack
       display_name=Render Opacity
       sections [{
         type=MovieSceneFloatSection
         end_frame=290
         start_frame=0
         channels [None {type=MovieSceneScriptingFloatChannel, keys [[45.456,0],[93.926,0.8097],[127.538,1],[213.25,0]]}]
        }]
      }]
    }
    ImageBackplate {
     id=#18
     possessed_class=Image
     tracks [{
       type=MovieSceneWidgetMaterialTrack
       display_name=Brush Material
       sections [{
         type=MovieSceneParameterSection
         end_frame=311
         start_frame=0
         channels [
          RevealDensity {type=MovieSceneScriptingFloatChannel, keys [[0.12,0.0632],[68.062,2.4457],[70,10]]}
          RevealRadius {type=MovieSceneScriptingFloatChannel, keys [[0.12,0],[68.062,1.3973],[70,3.2234]]}
          ShineOffset {type=MovieSceneScriptingFloatChannel, keys [[79.412,0.6576],[217.052,-1.0328]]}
         ]
        }]
      }]
    }
    ImageBackplateShadow {
     id=#34
     possessed_class=Image
     tracks [{
       type=MovieSceneWidgetMaterialTrack
       display_name=Brush Material
       sections [{
         type=MovieSceneParameterSection
         end_frame=311
         start_frame=0
         channels [
          RevealDensity {type=MovieSceneScriptingFloatChannel, thinned_from=5, keys [[0,0.0632],[0.12,0.0632],[68.062,2.4457],[70,10]]}
          RevealRadius {type=MovieSceneScriptingFloatChannel, keys [[0.12,0],[68.062,1.3973],[70,3.2234]]}
          ShineOffset {type=MovieSceneScriptingFloatChannel}
         ]
        }]
      }]
    }
    Imagespikeball2 {
     id=#19
     possessed_class=Image
     tracks [{
       type=MovieSceneFloatTrack
       display_name=Render Opacity
       sections [{
         type=MovieSceneFloatSection
         end_frame=330
         start_frame=0
         channels [None {type=MovieSceneScriptingFloatChannel, keys [[143.574,0],[159.964,0.8097],[193.576,1],[269.904,0]]}]
        }]
      }]
    }
    ImageSpinningSpikeball1 {
     id=#20
     possessed_class=Image
     tracks [{
       type=MovieSceneFloatTrack
       display_name=Render Opacity
       sections [{
         type=MovieSceneFloatSection
         end_frame=302
         start_frame=0
         channels [None {type=MovieSceneScriptingFloatChannel, keys [[177.452,0],[193.558,0.8097],[227.17,1]]}]
        }]
      }]
    }
    Image_108 {
     id=#25
     possessed_class=Image
     tracks [{
       type=MovieSceneFloatTrack
       display_name=Render Opacity
       sections [{
         type=MovieSceneFloatSection
         end_frame=319
         start_frame=0
         channels [None {type=MovieSceneScriptingFloatChannel, keys [[94.484,0],[110.874,0.8097],[144.486,1],[220.814,0]]}]
        }]
      }]
    }
    Image_text2flare2 {
     id=#26
     possessed_class=Image
     tracks [{
       type=MovieSceneFloatTrack
       display_name=Render Opacity
       sections [{
         type=MovieSceneFloatSection
         end_frame=290
         start_frame=0
         channels [None {type=MovieSceneScriptingFloatChannel, keys [[45.456,0],[93.926,0.8097],[127.538,1],[213.25,0]]}]
        }]
      }]
    }
    CommonTextRoyale {
     id=#27
     display_name=RichTextBlock_RoyaleShadow
     possessed_class=CommonTextBlock
     tracks [{
       type=MovieScene2DTransformTrack
       sections [{
         type=MovieScene2DTransformSection
         end_frame=299
         start_frame=0
         channels [
          Translation.X {type=MovieSceneScriptingFloatChannel}
          Translation.Y {type=MovieSceneScriptingFloatChannel}
          Angle {type=MovieSceneScriptingFloatChannel}
          Scale.X {type=MovieSceneScriptingFloatChannel, keys [[0.118,0],[18,0.001],[36.77,1.222],[43.538,1],[56.192,1]]}
          Scale.Y {type=MovieSceneScriptingFloatChannel, keys [[0.118,0],[18,0.001],[36.77,1.222],[43.538,1],[56.192,1]]}
          Shear.X {type=MovieSceneScriptingFloatChannel}
          Shear.Y {type=MovieSceneScriptingFloatChannel}
         ]
        }]
      }]
    }
    CommonTextBlock_4 {
     id=#28
     display_name=RichTextBlock_Victory
     possessed_class=CommonTextBlock
     tracks [
      {
       type=MovieScene2DTransformTrack
       sections [{
         type=MovieScene2DTransformSection
         end_frame=319
         start_frame=0
         channels [
          Translation.X {type=MovieSceneScriptingFloatChannel}
          Translation.Y {type=MovieSceneScriptingFloatChannel}
          Angle {type=MovieSceneScriptingFloatChannel}
          Scale.X {type=MovieSceneScriptingFloatChannel, keys [[0.118,0],[18,0.001],[36.77,1.222],[43.538,1],[56.192,1]]}
          Scale.Y {type=MovieSceneScriptingFloatChannel, keys [[0.118,0],[18,0.001],[36.77,1.222],[43.538,1],[56.192,1]]}
          Shear.X {type=MovieSceneScriptingFloatChannel}
          Shear.Y {type=MovieSceneScriptingFloatChannel}
         ]
        }]
      }
      {
       type=MovieSceneWidgetMaterialTrack
       display_name=Font Material
       sections [{
         type=MovieSceneParameterSection
         end_frame=319
         start_frame=36
         channels [ShineOffset {type=MovieSceneScriptingFloatChannel, keys [[165.05,0.59],[247.05,-0.7]]}]
        }]
      }
     ]
    }
    Image_text1flare2 {
     id=#30
     possessed_class=Image
     tracks [{
       type=MovieSceneFloatTrack
       display_name=Render Opacity
       sections [{
         type=MovieSceneFloatSection
         end_frame=257
         start_frame=0
         channels [None {type=MovieSceneScriptingFloatChannel, keys [[44.982,0],[60.314,0.8097],[93.926,1],[179.638,0]]}]
        }]
      }]
    }
    CommonTextBlock_6 {
     id=#31
     display_name=RichTextBlock_VictoryShadow
     possessed_class=CommonTextBlock
     tracks [{
       type=MovieScene2DTransformTrack
       sections [{
         type=MovieScene2DTransformSection
         end_frame=301
         start_frame=0
         channels [
          Translation.X {type=MovieSceneScriptingFloatChannel}
          Translation.Y {type=MovieSceneScriptingFloatChannel}
          Angle {type=MovieSceneScriptingFloatChannel}
          Scale.X {type=MovieSceneScriptingFloatChannel, keys [[0.118,0],[18,0.001],[36.77,1.222],[43.538,1],[56.192,1]]}
          Scale.Y {type=MovieSceneScriptingFloatChannel, keys [[0.118,0],[18,0.001],[36.77,1.222],[43.538,1],[56.192,1]]}
          Shear.X {type=MovieSceneScriptingFloatChannel}
          Shear.Y {type=MovieSceneScriptingFloatChannel}
         ]
        }]
      }]
    }
    Image1Cross {
     id=#32
     possessed_class=Image
     tracks [{
       type=MovieSceneFloatTrack
       display_name=Render Opacity
       sections [{
         type=MovieSceneFloatSection
         end_frame=335
         start_frame=0
         channels [None {type=MovieSceneScriptingFloatChannel, keys [[179.98,0],[207.196,1],[283.524,0]]}]
        }]
      }]
    }
   ]
   display_rate {numerator=120}
  }
 }
 SpinningSpikeball {
  binding_count=1
  display_label=""
  length_seconds=35.7931
  object=&self:SpinningSpikeball
  sequence {bindings [ImageSpinningSpikeball1 {
     id=#35
     possessed_class=Image
     tracks [{
       type=MovieScene2DTransformTrack
       sections [{
         type=MovieScene2DTransformSection
         end_frame=715
         start_frame=0
         channels [
          Translation.X {type=MovieSceneScriptingFloatChannel}
          Translation.Y {type=MovieSceneScriptingFloatChannel}
          Angle {type=MovieSceneScriptingFloatChannel, keys [[0,0],[715.8623,360]]}
          Scale.X {type=MovieSceneScriptingFloatChannel}
          Scale.Y {type=MovieSceneScriptingFloatChannel}
          Shear.X {type=MovieSceneScriptingFloatChannel}
          Shear.Y {type=MovieSceneScriptingFloatChannel}
         ]
        }]
      }]
    }]}
 }
 Victory_v2_b2 {
  binding_count=29
  display_label=""
  length_seconds=3.3348
  object=&self:Victory_v2_b2
  sequence {
   bindings [
    Image_107 {
     id=#3
     possessed_class=Image
     tracks [
      {
       type=MovieSceneFloatTrack
       display_name=Render Opacity
       sections [{
         type=MovieSceneFloatSection
         end_frame=312
         start_frame=0
         channels [None {type=MovieSceneScriptingFloatChannel, keys [[86.312,0],[117.23,0.8097],[150.842,1],[227.17,0]]}]
        }]
      }
      {
       type=MovieScene2DTransformTrack
       sections [{
         type=MovieScene2DTransformSection
         end_frame=314
         start_frame=0
         channels [
          Translation.X {type=MovieSceneScriptingFloatChannel, keys [[85.75,-12.0191],[228.604,186.407]]}
          Translation.Y {type=MovieSceneScriptingFloatChannel, keys [[85.75,-181.193],[228.604,-205.39]]}
          Angle {type=MovieSceneScriptingFloatChannel, keys [[85.75,0],[228.604,20.494]]}
          Scale.X {type=MovieSceneScriptingFloatChannel}
          Scale.Y {type=MovieSceneScriptingFloatChannel}
          Shear.X {type=MovieSceneScriptingFloatChannel}
          Shear.Y {type=MovieSceneScriptingFloatChannel}
         ]
        }]
      }
     ]
    }
    ImageCross {
     id=#4
     possessed_class=Image
     tracks [{
       type=MovieSceneFloatTrack
       display_name=Render Opacity
       sections [{
         type=MovieSceneFloatSection
         end_frame=302
         start_frame=0
         channels [None {type=MovieSceneScriptingFloatChannel, keys [[90.854,0],[107.244,0.8097],[140.856,1],[217.184,0]]}]
        }]
      }]
    }
    Image_Lightshaft {
     id=#33
     possessed_class=Image
     tracks [
      {
       type=MovieSceneWidgetMaterialTrack
       display_name=Brush Material
       sections [{
         type=MovieSceneParameterSection
         end_frame=340
         start_frame=0
         channels [
          ShaftMax {type=MovieSceneScriptingFloatChannel, const [1.0018,2]}
          ShaftMin {type=MovieSceneScriptingFloatChannel, keys [[44,0.9713],[60,-2.7486],[88,0.9713]]}
         ]
        }]
      }
      {
       type=MovieSceneFloatTrack
       display_name=Render Opacity
       sections [{
         type=MovieSceneFloatSection
         end_frame=340
         start_frame=0
         channels [None {type=MovieSceneScriptingFloatChannel, keys [[44.002,0],[48,1],[88,0]]}]
        }]
      }
     ]
    }
    Image_text1flarebg {
     id=#5
     possessed_class=Image
     tracks [
      {
       type=MovieSceneFloatTrack
       display_name=Render Opacity
       sections [{
         type=MovieSceneFloatSection
         end_frame=257
         start_frame=0
         channels [None {type=MovieSceneScriptingFloatChannel, keys [[44.982,0],[60.314,0.8097],[93.926,1],[122.134,0]]}]
        }]
      }
      {
       type=MovieScene2DTransformTrack
       sections [{
         type=MovieScene2DTransformSection
         end_frame=232
         start_frame=0
         channels [
          Translation.X {type=MovieSceneScriptingFloatChannel}
          Translation.Y {type=MovieSceneScriptingFloatChannel}
          Angle {type=MovieSceneScriptingFloatChannel}
          Scale.X {type=MovieSceneScriptingFloatChannel, keys [[45.456,1],[62.084,1.2],[154.678,1]]}
          Scale.Y {type=MovieSceneScriptingFloatChannel, keys [[45.456,0.14],[62.084,1.2],[154.678,1]]}
          Shear.X {type=MovieSceneScriptingFloatChannel}
          Shear.Y {type=MovieSceneScriptingFloatChannel}
         ]
        }]
      }
     ]
    }
    CommonTextBlock_5 {
     id=#6
     display_name=RichTextBlock_Royale
     possessed_class=CommonTextBlock
     tracks [
      {
       type=MovieScene2DTransformTrack
       sections [{
         type=MovieScene2DTransformSection
         end_frame=223
         start_frame=0
         channels [
          Translation.X {type=MovieSceneScriptingFloatChannel}
          Translation.Y {type=MovieSceneScriptingFloatChannel}
          Angle {type=MovieSceneScriptingFloatChannel}
          Scale.X {type=MovieSceneScriptingFloatChannel, keys [[0.118,0],[18,0.001],[36.77,1.222],[43.538,1],[56.192,1]]}
          Scale.Y {type=MovieSceneScriptingFloatChannel, keys [[0.118,0],[18,0.001],[36.77,1.222],[43.538,1],[56.192,1]]}
          Shear.X {type=MovieSceneScriptingFloatChannel}
          Shear.Y {type=MovieSceneScriptingFloatChannel}
         ]
        }]
      }
      {
       type=MovieSceneWidgetMaterialTrack
       display_name=Font Material
       sections [{
         type=MovieSceneParameterSection
         end_frame=319
         start_frame=0
         channels [ShineOffset {type=MovieSceneScriptingFloatChannel, keys [[179.638,0.6],[261.638,-0.7262]]}]
        }]
      }
     ]
    }
    BorderNumber1Text {
     id=#7
     possessed_class=Border
     tracks [
      {
       type=MovieScene2DTransformTrack
       sections [{
         type=MovieScene2DTransformSection
         end_frame=340
         start_frame=0
         channels [
          Translation.X {type=MovieSceneScriptingFloatChannel}
          Translation.Y {type=MovieSceneScriptingFloatChannel}
          Angle {type=MovieSceneScriptingFloatChannel}
          Scale.X {type=MovieSceneScriptingFloatChannel, keys [[53.97,0.3925],[69.212,0.9925],[77.05,1.3915],[81.05,0.9925],[88.694,0.9925]]}
          Scale.Y {type=MovieSceneScriptingFloatChannel, keys [[53.97,0.3925],[69.212,0.9925],[77.05,1.3915],[81.05,0.9925],[88.694,0.9925]]}
          Shear.X {type=MovieSceneScriptingFloatChannel}
          Shear.Y {type=MovieSceneScriptingFloatChannel}
         ]
        }]
      }
      {
       type=MovieSceneFloatTrack
       display_name=Render Opacity
       sections [{
         type=MovieSceneFloatSection
         end_frame=316
         start_frame=0
         channels [None {type=MovieSceneScriptingFloatChannel, keys [[53.97,0],[66.364,1]]}]
        }]
      }
     ]
    }
    Image_text1flare1 {
     id=#8
     possessed_class=Image
     tracks [{
       type=MovieSceneFloatTrack
       display_name=Render Opacity
       sections [{
         type=MovieSceneFloatSection
         end_frame=257
         start_frame=0
         channels [None {type=MovieSceneScriptingFloatChannel, keys [[44.982,0],[60.314,0.8097],[93.926,1],[179.638,0]]}]
        }]
      }]
    }
    OverlayRoyaleBadge {
     id=#10
     possessed_class=Overlay
     tracks [{
       type=MovieSceneFloatTrack
       display_name=Render Opacity
       sections [{
         type=MovieSceneFloatSection
         end_frame=308
         start_frame=-61
         channels [None {type=MovieSceneScriptingFloatChannel, keys [[-60.176,0],[0.972,0],[25.97,1]]}]
        }]
      }]
    }
    Image_text2flarebg {
     id=#13
     possessed_class=Image
     tracks [
      {
       type=MovieSceneFloatTrack
       display_name=Render Opacity
       sections [{
         type=MovieSceneFloatSection
         end_frame=233
         start_frame=0
         channels [None {type=MovieSceneScriptingFloatChannel, keys [[45.456,0],[72.388,0.8097],[106,1],[155.746,0]]}]
        }]
      }
      {
       type=MovieScene2DTransformTrack
       sections [{
         type=MovieScene2DTransformSection
         end_frame=265
         start_frame=0
         channels [
          Translation.X {type=MovieSceneScriptingFloatChannel}
          Translation.Y {type=MovieSceneScriptingFloatChannel}
          Angle {type=MovieSceneScriptingFloatChannel}
          Scale.X {type=MovieSceneScriptingFloatChannel, keys [[45.93,1],[62.084,1.2],[188.29,1]]}
          Scale.Y {type=MovieSceneScriptingFloatChannel, keys [[45.93,0.37],[62.084,1.2],[188.29,1]]}
          Shear.X {type=MovieSceneScriptingFloatChannel}
          Shear.Y {type=MovieSceneScriptingFloatChannel}
         ]
        }]
      }
     ]
    }
    Imagetext2flare1 {
     id=#15
     possessed_class=Image
     tracks [{
       type=MovieSceneFloatTrack
       display_name=Render Opacity
       sections [{
         type=MovieSceneFloatSection
         end_frame=290
         start_frame=0
         channels [None {type=MovieSceneScriptingFloatChannel, keys [[45.456,0],[93.926,0.8097],[127.538,1],[213.25,0]]}]
        }]
      }]
    }
    ImageBackplate {
     id=#18
     possessed_class=Image
     tracks [{
       type=MovieSceneWidgetMaterialTrack
       display_name=Brush Material
       sections [{
         type=MovieSceneParameterSection
         end_frame=311
         start_frame=0
         channels [
          RevealDensity {type=MovieSceneScriptingFloatChannel, keys [[0.12,0.0632],[40.944,1.3382],[42.882,10]]}
          RevealRadius {type=MovieSceneScriptingFloatChannel, keys [[0.12,0],[40.944,1.3973],[42.882,3.2234]]}
          ShineOffset {type=MovieSceneScriptingFloatChannel, keys [[79.412,0.6576],[170.558,-1.0328]]}
          PatternRep {type=MovieSceneScriptingFloatChannel, const [0.9,1]}
         ]
        }]
      }]
    }
    ImageBackplateShadow {
     id=#34
     possessed_class=Image
     tracks [{
       type=MovieSceneWidgetMaterialTrack
       display_name=Brush Material
       sections [{
         type=MovieSceneParameterSection
         end_frame=311
         start_frame=0
         channels [
          RevealDensity {type=MovieSceneScriptingFloatChannel, keys [[0.12,0.0632],[40.944,1.3382],[42.882,10]]}
          RevealRadius {type=MovieSceneScriptingFloatChannel, keys [[0.12,0],[40.944,1.3973],[42.882,3.2234]]}
          ShineOffset {type=MovieSceneScriptingFloatChannel}
          PatternRep {type=MovieSceneScriptingFloatChannel, const [0.9,1]}
         ]
        }]
      }]
    }
    Imagespikeball2 {
     id=#19
     possessed_class=Image
     tracks [{
       type=MovieSceneFloatTrack
       display_name=Render Opacity
       sections [{
         type=MovieSceneFloatSection
         end_frame=330
         start_frame=0
         channels [None {type=MovieSceneScriptingFloatChannel, keys [[143.574,0],[159.964,0.8097],[193.576,1],[269.904,0]]}]
        }]
      }]
    }
    ImageSpinningSpikeball1 {
     id=#20
     possessed_class=Image
     tracks [{
       type=MovieSceneFloatTrack
       display_name=Render Opacity
       sections [{
         type=MovieSceneFloatSection
         end_frame=302
         start_frame=0
         channels [None {type=MovieSceneScriptingFloatChannel, keys [[177.452,0],[193.558,0.8097],[227.17,1]]}]
        }]
      }]
    }
    Image_108 {
     id=#25
     possessed_class=Image
     tracks [{
       type=MovieSceneFloatTrack
       display_name=Render Opacity
       sections [{
         type=MovieSceneFloatSection
         end_frame=319
         start_frame=0
         channels [None {type=MovieSceneScriptingFloatChannel, keys [[94.484,0],[110.874,0.8097],[144.486,1],[220.814,0]]}]
        }]
      }]
    }
    Image_text2flare2 {
     id=#26
     possessed_class=Image
     tracks [{
       type=MovieSceneFloatTrack
       display_name=Render Opacity
       sections [{
         type=MovieSceneFloatSection
         end_frame=290
         start_frame=0
         channels [None {type=MovieSceneScriptingFloatChannel, keys [[45.456,0],[93.926,0.8097],[127.538,1],[213.25,0]]}]
        }]
      }]
    }
    CommonTextRoyale {
     id=#27
     display_name=RichTextBlock_RoyaleShadow
     possessed_class=CommonTextBlock
     tracks [{
       type=MovieScene2DTransformTrack
       sections [{
         type=MovieScene2DTransformSection
         end_frame=299
         start_frame=0
         channels [
          Translation.X {type=MovieSceneScriptingFloatChannel}
          Translation.Y {type=MovieSceneScriptingFloatChannel}
          Angle {type=MovieSceneScriptingFloatChannel}
          Scale.X {type=MovieSceneScriptingFloatChannel, keys [[0.118,0],[18,0.001],[36.77,1.222],[43.538,1],[56.192,1]]}
          Scale.Y {type=MovieSceneScriptingFloatChannel, keys [[0.118,0],[18,0.001],[36.77,1.222],[43.538,1],[56.192,1]]}
          Shear.X {type=MovieSceneScriptingFloatChannel}
          Shear.Y {type=MovieSceneScriptingFloatChannel}
         ]
        }]
      }]
    }
    CommonTextBlock_4 {
     id=#28
     display_name=RichTextBlock_Victory
     possessed_class=CommonTextBlock
     tracks [
      {
       type=MovieScene2DTransformTrack
       sections [{
         type=MovieScene2DTransformSection
         end_frame=319
         start_frame=0
         channels [
          Translation.X {type=MovieSceneScriptingFloatChannel}
          Translation.Y {type=MovieSceneScriptingFloatChannel}
          Angle {type=MovieSceneScriptingFloatChannel}
          Scale.X {type=MovieSceneScriptingFloatChannel, keys [[0.118,0],[18,0.001],[36.77,1.222],[43.538,1],[56.192,1]]}
          Scale.Y {type=MovieSceneScriptingFloatChannel, keys [[0.118,0],[18,0.001],[36.77,1.222],[43.538,1],[56.192,1]]}
          Shear.X {type=MovieSceneScriptingFloatChannel}
          Shear.Y {type=MovieSceneScriptingFloatChannel}
         ]
        }]
      }
      {
       type=MovieSceneWidgetMaterialTrack
       display_name=Font Material
       sections [{
         type=MovieSceneParameterSection
         end_frame=319
         start_frame=36
         channels [ShineOffset {type=MovieSceneScriptingFloatChannel, keys [[165.05,0.59],[247.05,-0.7]]}]
        }]
      }
     ]
    }
    Image_text1flare2 {
     id=#30
     possessed_class=Image
     tracks [{
       type=MovieSceneFloatTrack
       display_name=Render Opacity
       sections [{
         type=MovieSceneFloatSection
         end_frame=257
         start_frame=0
         channels [None {type=MovieSceneScriptingFloatChannel, keys [[44.982,0],[60.314,0.8097],[93.926,1],[179.638,0]]}]
        }]
      }]
    }
    CommonTextBlock_6 {
     id=#31
     display_name=RichTextBlock_VictoryShadow
     possessed_class=CommonTextBlock
     tracks [{
       type=MovieScene2DTransformTrack
       sections [{
         type=MovieScene2DTransformSection
         end_frame=301
         start_frame=0
         channels [
          Translation.X {type=MovieSceneScriptingFloatChannel}
          Translation.Y {type=MovieSceneScriptingFloatChannel}
          Angle {type=MovieSceneScriptingFloatChannel}
          Scale.X {type=MovieSceneScriptingFloatChannel, keys [[0.118,0],[18,0.001],[36.77,1.222],[43.538,1],[56.192,1]]}
          Scale.Y {type=MovieSceneScriptingFloatChannel, keys [[0.118,0],[18,0.001],[36.77,1.222],[43.538,1],[56.192,1]]}
          Shear.X {type=MovieSceneScriptingFloatChannel}
          Shear.Y {type=MovieSceneScriptingFloatChannel}
         ]
        }]
      }]
    }
    Image1Cross {
     id=#32
     possessed_class=Image
     tracks [{
       type=MovieSceneFloatTrack
       display_name=Render Opacity
       sections [{
         type=MovieSceneFloatSection
         end_frame=335
         start_frame=0
         channels [None {type=MovieSceneScriptingFloatChannel, keys [[179.98,0],[207.196,1],[283.524,0]]}]
        }]
      }]
    }
   ]
   display_rate {numerator=120}
  }
 }
]
graphs [EventGraph {nodes [
   SetVisibility(
    self=Get CommonWidgetSwitcher_winanim() -> (CommonWidgetSwitcher_winanim:Common Widget Switcher Legacy Object Reference)
    InVisibility=Collapsed
   )
   > n1 = Create Widget!(execute:Exec, Class:User Widget Class Reference, OwningPlayer:Player Controller Object Reference) -> (then:Exec, ReturnValue:User Widget Object Reference)
   > n2 = Bad cast node!(execute, Object=@n1) -> (then, CastFailed, AsFort LTMVictory Widget Base:Object Reference) {is=title K2Node_DynamicCast}
   > n3 = AddChildToOverlay!(
    execute
    self:Overlay Object Reference=Get Overlay_SpecialEndingSocket() -> (Overlay_SpecialEndingSocket:Overlay Object Reference)
    Content:Widget Object Reference=@n2
   ) -> (then, ReturnValue:Overlay Slot Object Reference)
   > SetHorizontalAlignment(self=@n3, InHorizontalAlignment=HAlign_Fill)
   > SetVerticalAlignment(self=@n3, InVerticalAlignment=VAlign_Fill)
   > Start Intro Sequence(self:Object Reference=@n2)
   PlayVictoryAnim() {is=title K2Node_CustomEvent}
   > n4 = Sequence()
   > SetVisibility(InVisibility=HitTestInvisible)
   n5 = Get Victory_v2_b() -> (Victory_v2_b:Widget Animation Object Reference)
   > PlayAnimation(InAnimation=@n5, StartAtTime=GetEndTime(self=@n5), PlayMode=Reverse, PlaybackSpeed=10)
   > PlayAnimation(InAnimation=Get Victory_v2_b2() -> (Victory_v2_b2:Widget Animation Object Reference), PlaybackSpeed=0.5)
   > PlayAnimation(InAnimation=Get SpinningSpikeball() -> (SpinningSpikeball:Widget Animation Object Reference), NumLoopsToPlay=0)
   Branch(
    execute=@n4.then_1
    Condition=ShouldVictoryUIPushEmoteInput!(self:Athena HUDContext Object Reference=Get AthenaHUDContext!() -> (ReturnValue:Athena HUDContext Object Reference) {is=title K2Node_GetBlueprintContextEx}) -> (ReturnValue:bool)
   )
   > PushGameOverEmoteInput(self:Athena HUDContext Object Reference=Get AthenaHUDContext())
   n6 = Event PreConstruct()
   n7 = SlotAsVerticalBoxSlot(Widget=Get RichTextBlock_Royale() -> (RichTextBlock_Royale:Rich Text Block Object Reference))
   n8 = Break Margin!(Margin:"Margin Structure (by ref)"=Get Padding!(self:Vertical Box Slot Object Reference=@n7) -> (Padding:Margin Structure)) -> (Left:float=~, Top:float=~, Right:float=~, Bottom:float=~) {is=title K2Node_BreakStruct}
   n9 = GetCurrentLanguage!() -> (ReturnValue:String)
   n10 = Select(
    Option 0:double=0
    Option 1:double=30
    Index=OR Boolean!(
     A=Contains(SearchIn=@n9, Substring=ko-KR)
     B=Contains(SearchIn=@n9, Substring=zh-hant)
     C:bool=Get DEVTestLocSpace() -> (DEVTestLocSpace:bool)
    ) -> (ReturnValue)
   ) -> (ReturnValue:double)
   SetPadding!(
    execute=[@n11.then, @n6.then]
    self:Vertical Box Slot Object Reference=[SlotAsVerticalBoxSlot(Widget=Get RichTextBlock_RoyaleShadow() -> (RichTextBlock_RoyaleShadow:Rich Text Block Object Reference)), @n7]
    InPadding:Margin Structure=Make Margin(Left=@n8, Top=float + float(A=@n8.Top, B=@n10) {is=title K2Node_CommutativeAssociativeBinaryOperator}, Right=@n8.Right, Bottom=@n8.Bottom)
   ) -> (then)
   Event Construct()
   > n11 = Set DEVTestLocSpace(DEVTestLocSpace:bool=false) -> (Output_Get:bool)
   K2Node_Event_4 = BP_OnActivated_1() {is=title K2Node_CustomEvent}
  ]}]
variables [DEVTestLocSpace {type=boolean}]
widgets {
 info {inheritedWidgetCount=1, parentClass=&FortHUDElementWidget, rootWidgetClass=&Overlay, widgetCount=46}
 widgets [
  Overlay_1 {widgetClassPath=&Overlay, widget=Overlay_1}
  CommonWidgetSwitcher_winanim {
   widgetClassPath=&CommonWidgetSwitcherLegacy
   bIsVariable=true
   parent=Overlay_1
   widget=CommonWidgetSwitcher_winanim
   slot {class=OverlaySlot, properties {horizontalAlignment=HAlign_Fill, verticalAlignment=VAlign_Fill}}
  }
  OverlayWinV2 {widgetClassPath=&Overlay, parent=CommonWidgetSwitcher_winanim, widget=OverlayWinV2, slot {class=WidgetSwitcherSlot}}
  OverlayRoyaleBadge {
   widgetClassPath=&Overlay
   parent=OverlayWinV2
   widget=OverlayRoyaleBadge
   properties {renderTransform {angle=0, scale=<1 1>, shear=<0 0>, translation=<0 -243>}}
   slot {class=OverlaySlot, properties {horizontalAlignment=HAlign_Fill, verticalAlignment=VAlign_Fill}}
  }
  Image_text1flarebg {
   widgetClassPath=&Image
   parent=OverlayRoyaleBadge
   widget=Image_text1flarebg
   properties {
    brush {
     bIsDynamicallyLoaded=false
     drawAs=Image
     imageType=NoImage
     mirroring=NoMirror
     resourceName=None
     resourceObject=/Game/Athena/HUD/Materials/Victory/M_UI_LightFlare
     tiling=NoTile
     imageSize=<1371.1171 946.7446>
     margin=<ltrb 0 0 0 0>
     outlineSettings {
      bUseBrushTransparency=false
      roundingType=HalfHeightRadius
      width=0
      color {colorUseRule=UseColor_Specified, specifiedColor=<rgba 0 0 0 0>}
      cornerRadii=<0 0 0 0>
     }
     tintColor {colorUseRule=UseColor_Specified, specifiedColor=<rgba 1 1 1 1>}
     uVRegion {bIsValid=false, max=<0 0>, min=<0 0>}
    }
    renderTransform {angle=0, scale=<1.192 1.192>, shear=<0 0>, translation=<37 -86>}
   }
   slot {class=OverlaySlot, properties {horizontalAlignment=HAlign_Center, verticalAlignment=VAlign_Center}}
  }
  OverlayBody {
   widgetClassPath=&Overlay
   parent=OverlayRoyaleBadge
   widget=OverlayBody
   properties {renderTransform {angle=0, scale=<0.9 0.9>, shear=<0 0>, translation=<0 0>}}
   slot {class=OverlaySlot, properties {horizontalAlignment=HAlign_Center, verticalAlignment=VAlign_Center}}
  }
  Image_Lightshaft {
   widgetClassPath=&Image
   parent=OverlayBody
   widget=Image_Lightshaft
   properties {
    renderOpacity=0
    brush {
     bIsDynamicallyLoaded=false
     drawAs=Image
     imageType=NoImage
     mirroring=NoMirror
     resourceName=None
     resourceObject=/Game/Athena/HUD/Materials/Victory/M_Victory_LightShafts_Inst
     tiling=NoTile
     imageSize=<1210.3491 446.3852>
     margin=<ltrb 0 0 0 0>
     outlineSettings {
      bUseBrushTransparency=false
      roundingType=HalfHeightRadius
      width=0
      color {colorUseRule=UseColor_Specified, specifiedColor=<rgba 0 0 0 0>}
      cornerRadii=<0 0 0 0>
     }
     tintColor {colorUseRule=UseColor_Specified, specifiedColor=<rgba 1 1 1 1>}
     uVRegion {bIsValid=false, max=<0 0>, min=<0 0>}
    }
    renderTransform {angle=0, scale=<1 1>, shear=<0 0>, translation=<0 -82>}
   }
   slot {class=OverlaySlot, properties {horizontalAlignment=HAlign_Center, verticalAlignment=VAlign_Center, padding=<ltrb -1000 -1000 -1000 -1000>}}
  }
  Overlay_Backplate {
   widgetClassPath=&Overlay
   parent=OverlayBody
   widget=Overlay_Backplate
   properties {renderTransform {angle=-4.11, scale=<1 1>, shear=<-36 0>, translation=<-10 -90>}}
   slot {class=OverlaySlot, properties {horizontalAlignment=HAlign_Center}}
  }
  ImageBackplateShadow {
   widgetClassPath=&Image
   parent=Overlay_Backplate
   widget=ImageBackplateShadow
   properties {
    brush {
     bIsDynamicallyLoaded=false
     drawAs=Image
     imageType=NoImage
     mirroring=NoMirror
     resourceName=None
     resourceObject=/Game/Athena/HUD/Materials/Victory/M_VictoryBG2_Inst
     tiling=NoTile
     imageSize=<740 270>
     margin=<ltrb 0 0 0 0>
     outlineSettings {
      bUseBrushTransparency=false
      roundingType=HalfHeightRadius
      width=0
      color {colorUseRule=UseColor_Specified, specifiedColor=<rgba 0 0 0 0>}
      cornerRadii=<0 0 0 0>
     }
     tintColor {colorUseRule=UseColor_Specified, specifiedColor=<rgba 1 1 1 1>}
     uVRegion {bIsValid=false, max=<0 0>, min=<0 0>}
    }
    colorAndOpacity=<rgba 0 0.0103 0.1301 0.8>
    renderTransform {angle=0, scale=<1 1>, shear=<0 0>, translation=<12 10>}
   }
   slot {class=OverlaySlot}
  }
  ImageBackplate {
   widgetClassPath=&Image
   parent=Overlay_Backplate
   widget=ImageBackplate
   properties {brush {
     bIsDynamicallyLoaded=false
     drawAs=Image
     imageType=NoImage
     mirroring=NoMirror
     resourceName=None
     resourceObject=&M_VictoryBG2_Inst
     tiling=NoTile
     imageSize=<740 270>
     margin=<ltrb 0 0 0 0>
     outlineSettings {
      bUseBrushTransparency=false
      roundingType=HalfHeightRadius
      width=0
      color {colorUseRule=UseColor_Specified, specifiedColor=<rgba 0 0 0 0>}
      cornerRadii=<0 0 0 0>
     }
     tintColor {colorUseRule=UseColor_Specified, specifiedColor=<rgba 1 1 1 1>}
     uVRegion {bIsValid=false, max=<0 0>, min=<0 0>}
    }}
   slot {class=OverlaySlot}
  }
  SizeBox_1 {
   widgetClassPath=&SizeBox
   parent=OverlayBody
   widget=SizeBox_1
   properties {bOverride_WidthOverride=true, maxDesiredWidth=420, widthOverride=420}
   slot {class=OverlaySlot, properties {horizontalAlignment=HAlign_Center, verticalAlignment=VAlign_Center}}
  }
  ScaleBox_1 {widgetClassPath=&ScaleBox, parent=SizeBox_1, widget=ScaleBox_1, properties {stretchDirection=DownOnly}, slot {class=SizeBoxSlot}}
  OverlayVictoryText {
   widgetClassPath=&Overlay
   bIsVariable=true
   parent=ScaleBox_1
   widget=OverlayVictoryText
   properties {renderTransform {angle=-4.11, scale=<1.05 1>, shear=<0 0>, translation=<19 -99>}}
   slot {class=ScaleBoxSlot}
  }
  Border_1 {
   widgetClassPath=&Border
   parent=OverlayVictoryText
   widget=Border_1
   properties {
    horizontalAlignment=HAlign_Left
    background {
     bIsDynamicallyLoaded=false
     drawAs=NoDrawType
     imageType=NoImage
     mirroring=NoMirror
     resourceName=None
     resourceObject=null
     tiling=NoTile
     imageSize=<32 32>
     margin=<ltrb 0 0 0 0>
     outlineSettings {
      bUseBrushTransparency=false
      roundingType=HalfHeightRadius
      width=0
      color {colorUseRule=UseColor_Specified, specifiedColor=<rgba 0 0 0 0>}
      cornerRadii=<0 0 0 0>
     }
     tintColor {colorUseRule=UseColor_Specified, specifiedColor=<rgba 1 1 1 1>}
     uVRegion {bIsValid=false, max=<0 0>, min=<0 0>}
    }
    contentColorAndOpacity=<rgba 0.0307 0.0307 0.1812 0.9>
    padding=<ltrb 0 0 0 0>
   }
   slot {class=OverlaySlot}
  }
  VerticalBox_2 {
   widgetClassPath=&VerticalBox
   parent=Border_1
   widget=VerticalBox_2
   properties {renderTransform {angle=0, scale=<1 1>, shear=<-4 0>, translation=<0 0>}}
   slot {class=BorderSlot, properties {horizontalAlignment=HAlign_Left, padding=<ltrb 0 0 0 0>}}
  }
  RichTextBlock_VictoryShadow {
   widgetClassPath=/Script/UMG.RichTextBlock
   bIsVariable=true
   parent=VerticalBox_2
   widget=RichTextBlock_VictoryShadow
   properties {text="<cap>V</>ICTORY", textStyleSet=&/VictoryRichTextStyles}
   slot {class=VerticalBoxSlot, properties {horizontalAlignment=HAlign_Left}}
  }
  RichTextBlock_RoyaleShadow {
   widgetClassPath=&RichTextBlock
   bIsVariable=true
   parent=VerticalBox_2
   widget=RichTextBlock_RoyaleShadow
   properties {
    text="<cap>R</>OYALE"
    textStyleSet=&VictoryRichTextStyles
    decoratorClasses [/Script/UMG.RichTextBlockImageDecorator]
   }
   slot {class=VerticalBoxSlot, properties {padding=<ltrb 85 -65 0 0>}}
  }
  CommonTextBlock_6 {
   widgetClassPath=&CommonTextBlock
   parent=VerticalBox_2
   widget=CommonTextBlock_6
   properties {
    text=Victory
    visibility=Collapsed
    colorAndOpacity {colorUseRule=UseColor_Specified, specifiedColor=<rgba 0.0239 0.0239 0.141 0.9>}
    font {
     bForceMonospaced=false
     bMaterialIsStencil=false
     fontMaterial=null
     fontObject=&Burbank
     letterSpacing=0
     monospacedWidth=1
     size=150
     skewAmount=0
     typefaceFontName=Burbank-Black
     outlineSettings {
      bApplyOutlineToDropShadows=true
      bMiteredCorners=false
      bSeparateFillAlpha=false
      outlineMaterial=null
      outlineSize=3
      outlineColor=<rgba 1 1 1 0>
     }
    }
    renderTransformPivot=<0.8 0.8>
   }
   slot {class=VerticalBoxSlot, properties {horizontalAlignment=HAlign_Left}}
  }
  CommonTextRoyale {
   widgetClassPath=&CommonTextBlock
   parent=VerticalBox_2
   widget=CommonTextRoyale
   properties {
    text=Royale
    visibility=Collapsed
    colorAndOpacity {colorUseRule=UseColor_Specified, specifiedColor=<rgba 0.0239 0.0239 0.141 0.9>}
    font {
     bForceMonospaced=false
     bMaterialIsStencil=false
     fontMaterial=null
     fontObject=&Burbank
     letterSpacing=0
     monospacedWidth=1
     size=150
     skewAmount=0
     typefaceFontName=Burbank-Black
     outlineSettings {
      bApplyOutlineToDropShadows=true
      bMiteredCorners=false
      bSeparateFillAlpha=false
      outlineMaterial=null
      outlineSize=3
      outlineColor=<rgba 1 1 1 0>
     }
    }
    renderTransformPivot=<0.5 0.2>
   }
   slot {class=VerticalBoxSlot, properties {horizontalAlignment=HAlign_Right, padding=<ltrb 0 -120 0 0>}}
  }
  VerticalBox_1 {
   widgetClassPath=&VerticalBox
   parent=OverlayVictoryText
   widget=VerticalBox_1
   properties {renderTransform {angle=0, scale=<1 1>, shear=<-4 0>, translation=<-5 -8>}}
   slot {class=OverlaySlot}
  }
  RichTextBlock_Victory {
   widgetClassPath=&RichTextBlock
   bIsVariable=true
   parent=VerticalBox_1
   widget=RichTextBlock_Victory
   properties {text="<cap>V</>ICTORY", textStyleSet=&VictoryRichTextStyles}
   slot {class=VerticalBoxSlot, properties {horizontalAlignment=HAlign_Left}}
  }
  RichTextBlock_Royale {
   widgetClassPath=&RichTextBlock
   bIsVariable=true
   parent=VerticalBox_1
   widget=RichTextBlock_Royale
   properties {
    text="<cap>R</>OYALE"
    textStyleSet=&VictoryRichTextStyles
    decoratorClasses [&RichTextBlockImageDecorator]
   }
   slot {class=VerticalBoxSlot, properties {padding=<ltrb 85 -65 0 0>}}
  }
  CommonTextBlock_4 {
   widgetClassPath=&CommonTextBlock
   parent=VerticalBox_1
   widget=CommonTextBlock_4
   properties {
    text=Victory
    visibility=Collapsed
    font {
     bForceMonospaced=false
     bMaterialIsStencil=false
     fontMaterial=/Game/Athena/HUD/Materials/Victory/M_WinText_Chrome
     fontObject=&Burbank
     letterSpacing=0
     monospacedWidth=1
     size=150
     skewAmount=0
     typefaceFontName=Burbank-Black
     outlineSettings {
      bApplyOutlineToDropShadows=true
      bMiteredCorners=false
      bSeparateFillAlpha=false
      outlineMaterial=/Game/Athena/HUD/Materials/Victory/M_VictoryText_Outline
      outlineSize=3
      outlineColor=<rgba 1 1 1 0.25>
     }
    }
    renderTransformPivot=<0.8 0.8>
   }
   slot {class=VerticalBoxSlot, properties {horizontalAlignment=HAlign_Left}}
  }
  CommonTextBlock_5 {
   widgetClassPath=&CommonTextBlock
   parent=VerticalBox_1
   widget=CommonTextBlock_5
   properties {
    text=Royale
    visibility=Collapsed
    font {
     bForceMonospaced=false
     bMaterialIsStencil=false
     fontMaterial=&M_WinText_Chrome
     fontObject=&Burbank
     letterSpacing=0
     monospacedWidth=1
     size=150
     skewAmount=0
     typefaceFontName=Burbank-Black
     outlineSettings {
      bApplyOutlineToDropShadows=true
      bMiteredCorners=false
      bSeparateFillAlpha=false
      outlineMaterial=null
      outlineSize=3
      outlineColor=<rgba 1 1 1 0.25>
     }
    }
    renderTransformPivot=<0.5 0.2>
   }
   slot {class=VerticalBoxSlot, properties {horizontalAlignment=HAlign_Right, padding=<ltrb 0 -120 0 0>}}
  }
  BorderNumber1Text {
   widgetClassPath=&Border
   parent=OverlayBody
   widget=BorderNumber1Text
   properties {
    horizontalAlignment=HAlign_Right
    verticalAlignment=VAlign_Center
    background {
     bIsDynamicallyLoaded=false
     drawAs=NoDrawType
     imageType=NoImage
     mirroring=NoMirror
     resourceName=None
     resourceObject=null
     tiling=NoTile
     imageSize=<32 32>
     margin=<ltrb 0 0 0 0>
     outlineSettings {
      bUseBrushTransparency=false
      roundingType=HalfHeightRadius
      width=0
      color {colorUseRule=UseColor_Specified, specifiedColor=<rgba 0 0 0 0>}
      cornerRadii=<0 0 0 0>
     }
     tintColor {colorUseRule=UseColor_Specified, specifiedColor=<rgba 1 1 1 1>}
     uVRegion {bIsValid=false, max=<0 0>, min=<0 0>}
    }
    padding=<ltrb 0 0 0 0>
    renderTransform {angle=0, scale=<0.9925 0.9925>, shear=<0 0>, translation=<-565 -66>}
    renderTransformPivot=<1 0.7>
   }
   slot {class=OverlaySlot, properties {horizontalAlignment=HAlign_Right, verticalAlignment=VAlign_Center, padding=<ltrb 0 -45 0 -45>}}
  }
  Overlay_5 {
   widgetClassPath=&Overlay
   parent=BorderNumber1Text
   widget=Overlay_5
   properties {renderTransform {angle=-15, scale=<1 1>, shear=<-10 0>, translation=<0 0>}}
   slot {class=BorderSlot, properties {horizontalAlignment=HAlign_Right, verticalAlignment=VAlign_Center, padding=<ltrb 0 0 0 0>}}
  }
  Border_2 {
   widgetClassPath=&Border
   parent=Overlay_5
   widget=Border_2
   properties {
    background {
     bIsDynamicallyLoaded=false
     drawAs=NoDrawType
     imageType=NoImage
     mirroring=NoMirror
     resourceName=None
     resourceObject=null
     tiling=NoTile
     imageSize=<32 32>
     margin=<ltrb 0 0 0 0>
     outlineSettings {
      bUseBrushTransparency=false
      roundingType=HalfHeightRadius
      width=0
      color {colorUseRule=UseColor_Specified, specifiedColor=<rgba 0 0 0 0>}
      cornerRadii=<0 0 0 0>
     }
     tintColor {colorUseRule=UseColor_Specified, specifiedColor=<rgba 1 1 1 1>}
     uVRegion {bIsValid=false, max=<0 0>, min=<0 0>}
    }
    contentColorAndOpacity=<rgba 0.0123 0.0144 0.0395 1>
    padding=<ltrb 0 0 0 0>
   }
   slot {class=OverlaySlot, properties {horizontalAlignment=HAlign_Right}}
  }
  RichTextBlock_PlacementShadow {
   widgetClassPath=&RichTextBlock
   parent=Border_2
   widget=RichTextBlock_PlacementShadow
   properties {text="<PlacementNumberSymbol>##</><PlacementValue>1</>", textStyleSet=&VictoryRichTextStyles}
   slot {class=BorderSlot, properties {padding=<ltrb 0 0 0 0>}}
  }
  RichTextBlock_Placement {
   widgetClassPath=&RichTextBlock
   parent=Overlay_5
   widget=RichTextBlock_Placement
   properties {
    justification=Right
    text="<PlacementNumberSymbol>##</><PlacementValue>1</>"
    textStyleSet=&VictoryRichTextStyles
    renderTransform {angle=0, scale=<1 1>, shear=<0 0>, translation=<-8 -9>}
   }
   slot {class=OverlaySlot, properties {horizontalAlignment=HAlign_Right}}
  }
  CommonTextBlock_2 {
   widgetClassPath=&CommonTextBlock
   parent=Overlay_5
   widget=CommonTextBlock_2
   properties {
    text="##1"
    visibility=Collapsed
    colorAndOpacity {colorUseRule=UseColor_Specified, specifiedColor=<rgba 0.0123 0.0144 0.0395 1>}
    font {
     bForceMonospaced=false
     bMaterialIsStencil=false
     fontMaterial=null
     fontObject=&Burbank
     letterSpacing=0
     monospacedWidth=1
     size=294
     skewAmount=0
     typefaceFontName=Burbank-Black
     outlineSettings {
      bApplyOutlineToDropShadows=true
      bMiteredCorners=false
      bSeparateFillAlpha=false
      outlineMaterial=null
      outlineSize=5
      outlineColor=<rgba 0 0 0 0>
     }
    }
   }
   slot {class=OverlaySlot, properties {horizontalAlignment=HAlign_Right}}
  }
  CommonTextBlock_3 {
   widgetClassPath=&CommonTextBlock
   parent=Overlay_5
   widget=CommonTextBlock_3
   properties {
    text="##1"
    visibility=Collapsed
    font {
     bForceMonospaced=false
     bMaterialIsStencil=false
     fontMaterial=/Game/Athena/HUD/Materials/Victory/M_WinText_Gold
     fontObject=&Burbank
     letterSpacing=0
     monospacedWidth=1
     size=294
     skewAmount=0
     typefaceFontName=Burbank-Black
     outlineSettings {
      bApplyOutlineToDropShadows=true
      bMiteredCorners=false
      bSeparateFillAlpha=false
      outlineMaterial=null
      outlineSize=5
      outlineColor=<rgba 1 0.6376 0.15 0.2>
     }
    }
    renderTransform {angle=0, scale=<1 1>, shear=<0 0>, translation=<-23 -26>}
    shadowColorAndOpacity=<rgba 0.0123 0.0144 0.0395 1>
   }
   slot {class=OverlaySlot}
  }
  OverlayFlares {
   widgetClassPath=&Overlay
   parent=OverlayRoyaleBadge
   widget=OverlayFlares
   properties {visibility=Collapsed}
   slot {class=OverlaySlot, properties {horizontalAlignment=HAlign_Center, verticalAlignment=VAlign_Center, padding=<ltrb -500 -500 -500 -500>}}
  }
  Image_text1flare1 {
   widgetClassPath=&Image
   parent=OverlayFlares
   widget=Image_text1flare1
   properties {
    visibility=HitTestInvisible
    brush {
     bIsDynamicallyLoaded=false
     drawAs=Image
     imageType=NoImage
     mirroring=NoMirror
     resourceName=None
     resourceObject=/Game/UI/ItemReceived/Materials/M_EFX_CameraFlare_01
     tiling=NoTile
     imageSize=<1587.8158 155.1329>
     margin=<ltrb 0 0 0 0>
     outlineSettings {
      bUseBrushTransparency=false
      roundingType=HalfHeightRadius
      width=0
      color {colorUseRule=UseColor_Specified, specifiedColor=<rgba 0 0 0 0>}
      cornerRadii=<0 0 0 0>
     }
     tintColor {colorUseRule=UseColor_Specified, specifiedColor=<rgba 1 1 1 1>}
     uVRegion {bIsValid=false, max=<0 0>, min=<0 0>}
    }
    colorAndOpacity=<rgba 0.4538 0.6823 1 1>
    renderTransform {angle=0, scale=<1 1>, shear=<0 0>, translation=<-6 -134>}
   }
   slot {class=OverlaySlot, properties {horizontalAlignment=HAlign_Center, verticalAlignment=VAlign_Center}}
  }
  Image_text1flare2 {
   widgetClassPath=&Image
   parent=OverlayFlares
   widget=Image_text1flare2
   properties {
    visibility=HitTestInvisible
    brush {
     bIsDynamicallyLoaded=false
     drawAs=Image
     imageType=NoImage
     mirroring=NoMirror
     resourceName=None
     resourceObject=&M_EFX_CameraFlare_01
     tiling=NoTile
     imageSize=<1776.848 162.9608>
     margin=<ltrb 0 0 0 0>
     outlineSettings {
      bUseBrushTransparency=false
      roundingType=HalfHeightRadius
      width=0
      color {colorUseRule=UseColor_Specified, specifiedColor=<rgba 0 0 0 0>}
      cornerRadii=<0 0 0 0>
     }
     tintColor {colorUseRule=UseColor_Specified, specifiedColor=<rgba 1 1 1 1>}
     uVRegion {bIsValid=false, max=<0 0>, min=<0 0>}
    }
    colorAndOpacity=<rgba 0.4538 0.6823 1 1>
    renderTransform {angle=0, scale=<1 1>, shear=<0 0>, translation=<-6 -141>}
   }
   slot {class=OverlaySlot, properties {horizontalAlignment=HAlign_Center, verticalAlignment=VAlign_Center}}
  }
  Imagetext2flare1 {
   widgetClassPath=&Image
   parent=OverlayFlares
   widget=Imagetext2flare1
   properties {
    visibility=HitTestInvisible
    brush {
     bIsDynamicallyLoaded=false
     drawAs=Image
     imageType=NoImage
     mirroring=NoMirror
     resourceName=None
     resourceObject=&M_EFX_CameraFlare_01
     tiling=NoTile
     imageSize=<1293.0994 155.1329>
     margin=<ltrb 0 0 0 0>
     outlineSettings {
      bUseBrushTransparency=false
      roundingType=HalfHeightRadius
      width=0
      color {colorUseRule=UseColor_Specified, specifiedColor=<rgba 0 0 0 0>}
      cornerRadii=<0 0 0 0>
     }
     tintColor {colorUseRule=UseColor_Specified, specifiedColor=<rgba 1 1 1 1>}
     uVRegion {bIsValid=false, max=<0 0>, min=<0 0>}
    }
    colorAndOpacity=<rgba 0.4538 0.6823 1 1>
    renderTransform {angle=0, scale=<1 1>, shear=<0 0>, translation=<80 -50>}
   }
   slot {class=OverlaySlot, properties {horizontalAlignment=HAlign_Center, verticalAlignment=VAlign_Center}}
  }
  Image_text2flare2 {
   widgetClassPath=&Image
   parent=OverlayFlares
   widget=Image_text2flare2
   properties {
    visibility=HitTestInvisible
    brush {
     bIsDynamicallyLoaded=false
     drawAs=Image
     imageType=NoImage
     mirroring=NoMirror
     resourceName=None
     resourceObject=&M_EFX_CameraFlare_01
     tiling=NoTile
     imageSize=<1293.0994 126.8842>
     margin=<ltrb 0 0 0 0>
     outlineSettings {
      bUseBrushTransparency=false
      roundingType=HalfHeightRadius
      width=0
      color {colorUseRule=UseColor_Specified, specifiedColor=<rgba 0 0 0 0>}
      cornerRadii=<0 0 0 0>
     }
     tintColor {colorUseRule=UseColor_Specified, specifiedColor=<rgba 1 1 1 1>}
     uVRegion {bIsValid=false, max=<0 0>, min=<0 0>}
    }
    colorAndOpacity=<rgba 0.4538 0.6823 1 1>
    renderTransform {angle=0, scale=<1 1>, shear=<0 0>, translation=<80 -36>}
   }
   slot {class=OverlaySlot, properties {horizontalAlignment=HAlign_Center, verticalAlignment=VAlign_Center}}
  }
  Image_text2flarebg {
   widgetClassPath=&Image
   parent=OverlayFlares
   widget=Image_text2flarebg
   properties {
    visibility=Collapsed
    brush {
     bIsDynamicallyLoaded=false
     drawAs=Image
     imageType=NoImage
     mirroring=NoMirror
     resourceName=None
     resourceObject=&M_EFX_CameraFlare_01
     tiling=NoTile
     imageSize=<1770.9321 596.5913>
     margin=<ltrb 0 0 0 0>
     outlineSettings {
      bUseBrushTransparency=false
      roundingType=HalfHeightRadius
      width=0
      color {colorUseRule=UseColor_Specified, specifiedColor=<rgba 0 0 0 0>}
      cornerRadii=<0 0 0 0>
     }
     tintColor {colorUseRule=UseColor_Specified, specifiedColor=<rgba 1 1 1 1>}
     uVRegion {bIsValid=false, max=<0 0>, min=<0 0>}
    }
    colorAndOpacity=<rgba 0.4538 0.6823 1 1>
    renderTransform {angle=0, scale=<1 1.1>, shear=<0 0>, translation=<80 -50>}
   }
   slot {class=OverlaySlot, properties {horizontalAlignment=HAlign_Center, verticalAlignment=VAlign_Center}}
  }
  ImageCross {
   widgetClassPath=&Image
   parent=OverlayFlares
   widget=ImageCross
   properties {
    visibility=Collapsed
    brush {
     bIsDynamicallyLoaded=false
     drawAs=Image
     imageType=NoImage
     mirroring=NoMirror
     resourceName=None
     resourceObject=/Game/Athena/HUD/Materials/Victory/M_Cross_Flare
     tiling=NoTile
     imageSize=<128 128>
     margin=<ltrb 0 0 0 0>
     outlineSettings {
      bUseBrushTransparency=false
      roundingType=HalfHeightRadius
      width=0
      color {colorUseRule=UseColor_Specified, specifiedColor=<rgba 0 0 0 0>}
      cornerRadii=<0 0 0 0>
     }
     tintColor {colorUseRule=UseColor_Specified, specifiedColor=<rgba 1 1 1 1>}
     uVRegion {bIsValid=false, max=<0 0>, min=<0 0>}
    }
    renderTransform {angle=0, scale=<1 1>, shear=<0 0>, translation=<278 -57>}
   }
   slot {class=OverlaySlot, properties {horizontalAlignment=HAlign_Center, verticalAlignment=VAlign_Center}}
  }
  Image_107 {
   widgetClassPath=&Image
   parent=OverlayFlares
   widget=Image_107
   properties {
    visibility=Collapsed
    brush {
     bIsDynamicallyLoaded=false
     drawAs=Image
     imageType=NoImage
     mirroring=NoMirror
     resourceName=None
     resourceObject=/Game/Athena/HUD/Materials/Victory/M_Flare_Spikeball
     tiling=NoTile
     imageSize=<180 180>
     margin=<ltrb 0 0 0 0>
     outlineSettings {
      bUseBrushTransparency=false
      roundingType=HalfHeightRadius
      width=0
      color {colorUseRule=UseColor_Specified, specifiedColor=<rgba 0 0 0 0>}
      cornerRadii=<0 0 0 0>
     }
     tintColor {colorUseRule=UseColor_Specified, specifiedColor=<rgba 1 1 1 1>}
     uVRegion {bIsValid=false, max=<0 0>, min=<0 0>}
    }
    renderTransform {angle=0, scale=<1 1>, shear=<0 0>, translation=<195.3717 -206>}
   }
   slot {class=OverlaySlot, properties {horizontalAlignment=HAlign_Center, verticalAlignment=VAlign_Center}}
  }
  Imagespikeball2 {
   widgetClassPath=&Image
   parent=OverlayRoyaleBadge
   widget=Imagespikeball2
   properties {
    brush {
     bIsDynamicallyLoaded=false
     drawAs=Image
     imageType=NoImage
     mirroring=NoMirror
     resourceName=None
     resourceObject=&M_Flare_Spikeball
     tiling=NoTile
     imageSize=<90 90>
     margin=<ltrb 0 0 0 0>
     outlineSettings {
      bUseBrushTransparency=false
      roundingType=HalfHeightRadius
      width=0
      color {colorUseRule=UseColor_Specified, specifiedColor=<rgba 0 0 0 0>}
      cornerRadii=<0 0 0 0>
     }
     tintColor {colorUseRule=UseColor_Specified, specifiedColor=<rgba 1 1 1 1>}
     uVRegion {bIsValid=false, max=<0 0>, min=<0 0>}
    }
    renderTransform {angle=0, scale=<1 1>, shear=<0 0>, translation=<-311 -91>}
   }
   slot {class=OverlaySlot, properties {horizontalAlignment=HAlign_Center, verticalAlignment=VAlign_Center}}
  }
  Image1Cross {
   widgetClassPath=&Image
   parent=OverlayRoyaleBadge
   widget=Image1Cross
   properties {
    brush {
     bIsDynamicallyLoaded=false
     drawAs=Image
     imageType=NoImage
     mirroring=NoMirror
     resourceName=None
     resourceObject=&M_Cross_Flare
     tiling=NoTile
     imageSize=<128 128>
     margin=<ltrb 0 0 0 0>
     outlineSettings {
      bUseBrushTransparency=false
      roundingType=HalfHeightRadius
      width=0
      color {colorUseRule=UseColor_Specified, specifiedColor=<rgba 0 0 0 0>}
      cornerRadii=<0 0 0 0>
     }
     tintColor {colorUseRule=UseColor_Specified, specifiedColor=<rgba 1 1 1 1>}
     uVRegion {bIsValid=false, max=<0 0>, min=<0 0>}
    }
    renderTransform {angle=0, scale=<1 1>, shear=<0 0>, translation=<-214 -122>}
   }
   slot {class=OverlaySlot, properties {horizontalAlignment=HAlign_Center, verticalAlignment=VAlign_Center}}
  }
  ImageSpinningSpikeball1 {
   widgetClassPath=&Image
   parent=OverlayRoyaleBadge
   widget=ImageSpinningSpikeball1
   properties {
    brush {
     bIsDynamicallyLoaded=false
     drawAs=Image
     imageType=NoImage
     mirroring=NoMirror
     resourceName=None
     resourceObject=&M_Flare_Spikeball
     tiling=NoTile
     imageSize=<256 256>
     margin=<ltrb 0 0 0 0>
     outlineSettings {
      bUseBrushTransparency=false
      roundingType=HalfHeightRadius
      width=0
      color {colorUseRule=UseColor_Specified, specifiedColor=<rgba 0 0 0 0>}
      cornerRadii=<0 0 0 0>
     }
     tintColor {colorUseRule=UseColor_Specified, specifiedColor=<rgba 1 1 1 1>}
     uVRegion {bIsValid=false, max=<0 0>, min=<0 0>}
    }
    renderTransform {angle=0, scale=<1 1>, shear=<0 0>, translation=<-214 -119>}
   }
   slot {class=OverlaySlot, properties {horizontalAlignment=HAlign_Center, verticalAlignment=VAlign_Center}}
  }
  Image_108 {
   widgetClassPath=&Image
   parent=OverlayRoyaleBadge
   widget=Image_108
   properties {
    brush {
     bIsDynamicallyLoaded=false
     drawAs=Image
     imageType=NoImage
     mirroring=NoMirror
     resourceName=None
     resourceObject=&M_Flare_Spikeball
     tiling=NoTile
     imageSize=<180 180>
     margin=<ltrb 0 0 0 0>
     outlineSettings {
      bUseBrushTransparency=false
      roundingType=HalfHeightRadius
      width=0
      color {colorUseRule=UseColor_Specified, specifiedColor=<rgba 0 0 0 0>}
      cornerRadii=<0 0 0 0>
     }
     tintColor {colorUseRule=UseColor_Specified, specifiedColor=<rgba 1 1 1 1>}
     uVRegion {bIsValid=false, max=<0 0>, min=<0 0>}
    }
    renderTransform {angle=0, scale=<1 1>, shear=<0 0>, translation=<-205 -138>}
   }
   slot {class=OverlaySlot, properties {horizontalAlignment=HAlign_Center, verticalAlignment=VAlign_Center}}
  }
  Overlay_SpecialEndingSocket {
   widgetClassPath=&Overlay
   bIsVariable=true
   parent=Overlay_1
   widget=Overlay_SpecialEndingSocket
   slot {class=OverlaySlot, properties {horizontalAlignment=HAlign_Fill, verticalAlignment=VAlign_Fill}}
  }
  AthenaWinnerList {
   widgetClassPath=&/AthenaWinnerList
   bIsVariable=true
   parent=Overlay_1
   widget=AthenaWinnerList
   properties {isWinner=true, visibility=Collapsed}
   slot {class=OverlaySlot, properties {horizontalAlignment=HAlign_Center, verticalAlignment=VAlign_Center, padding=<ltrb 0 200 0 200>}}
  }
  TutorialHighlight {widgetClassPath=&TutorialHighlight}
 ]
}
