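# Broadcast HUD control bar for Athena replay playback and live spectating.
# Button events drive the replay context (playback, skips, speed) and spectating controllers (player following, camera shots), while live speed buttons issue slomo cheats in non-shipping builds.
# Replay context delegates update play/pause and speed displays, and a looping timer polls for timeline markers outside shipping builds.
# Hover events store the last focused widget so input-type refreshes can restore focus.
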WidgetBlueprint /Game/Athena/HUD/Spectator/AthenaBroadcastHUDControls
animations [Fade {
  binding_count=4
  length_seconds=0.25
  object=&self:Fade
  sequence {bindings [
    SpectatorSettingsSize {
     id=#1
     possessed_class=SizeBox
     tracks [{
       type=MovieSceneFloatTrack
       display_name=Render Opacity
       sections [{type=MovieSceneFloatSection, end_frame=5, start_frame=0, channels [None {type=MovieSceneScriptingFloatChannel, keys [[0,1],[5,0.5]]}]}]
      }]
    }
    ReplayTimeline {
     id=#2
     possessed_class=FortReplayTimeline_C
     tracks [{
       type=MovieSceneFloatTrack
       display_name=Render Opacity
       sections [{type=MovieSceneFloatSection, end_frame=5, start_frame=0, channels [None {type=MovieSceneScriptingFloatChannel, keys [[0,1],[5,0.5]]}]}]
      }]
    }
    ControlsWrapper {
     id=#3
     possessed_class=FortHUDElementWrapper_C
     tracks [{
       type=MovieSceneFloatTrack
       display_name=Render Opacity
       sections [{type=MovieSceneFloatSection, end_frame=5, start_frame=0, channels [None {type=MovieSceneScriptingFloatChannel, keys [[0,1],[5,0.5]]}]}]
      }]
    }
    ReplaySettingsWrapper {
     id=#4
     possessed_class=FortHUDElementWrapper_C
     tracks [{
       type=MovieSceneFloatTrack
       display_name=Render Opacity
       sections [{type=MovieSceneFloatSection, end_frame=5, start_frame=0, channels [None {type=MovieSceneScriptingFloatChannel, keys [[0,1],[5,0.5]]}]}]
      }]
    }
   ]}
 }]
graphs [
 BindReplayContext {nodes [
   BindReplayContext() {is=title K2Node_FunctionEntry}
   n1 = Get FortReplayContext!() -> (ReturnValue:Fort Replay Context Object Reference) {is=title K2Node_GetBlueprintContextEx}
   > Bind Event to On Pause State Changed(self:Fort Replay Context Object Reference=@n1, Delegate=Create Event()) {is=title K2Node_AddDelegate}
   > OnPauseStateChanged!(
    execute
    self
    bPaused:bool=Break Fort Replay Playback State!(FortReplayPlaybackState:"Fort Replay Playback State Structure (by ref)"=Get PlaybackState!(self:Fort Replay Context Object Reference=@n1) -> (PlaybackState:Fort Replay Playback State Structure)) -> (
     StartTime:float=~
     EndTime:float=~
     TimeNow:float=~
     bIsPaused:bool=~
     PlaybackSpeedMultiplier:float=~
     HUDVisibility:EHudVisibilityState Enum
     bLevelStreaming:bool=~
     bHasRelevancyZone:bool=~
     bHasPlayInterval:bool=~
     PlayIntervalStartTime:float=~
     PlayIntervalEndTime:float=~
     bLoopInterval:bool=~
     IntervalLoopBufferTime:float=~
    ) {is=title K2Node_BreakStruct}.bIsPaused
   ) -> (then)
   > Bind Event to On Replay Playback Multiplier Changed(self:Fort Replay Context Object Reference=@n1, Delegate=Create Event()) {is=title K2Node_AddDelegate}
   > OnReplayPlaybackMultiplierChanged!(
    execute
    self
    Multiplier:float=Break Fort Replay Playback State(FortReplayPlaybackState=Get PlaybackState(self=@n1)) -> (
     StartTime=~
     EndTime=~
     TimeNow=~
     bIsPaused=~
     PlaybackSpeedMultiplier=~
     bLevelStreaming=~
     bHasRelevancyZone=~
     bHasPlayInterval=~
     PlayIntervalStartTime=~
     PlayIntervalEndTime=~
     bLoopInterval=~
     IntervalLoopBufferTime=~
    ).PlaybackSpeedMultiplier
   ) -> (then)
  ]}
 CenterControl {nodes [
   CenterControl() {is=title K2Node_FunctionEntry}
   > n1 = IsOnBattleMapTimeline!(execute, self) -> (then, IsOnBattleMapTimeline:bool)
   > n2 = Branch(Condition=@n1)
   n3 = Get LastHoveredWidget() -> (LastHoveredWidget:User Widget Object Reference)
   n4 = Is Valid(exec=@n2.else, InputObject=@n3)
   > SetFocus(self=@n3)
   SetFocus(
    execute=@n4.Is Not Valid
    self=Select(
     Option 0:Icon Text Button Replay Object Reference=Get ViewSettingsButton_Live() -> (ViewSettingsButton_Live:Icon Text Button Replay Object Reference)
     Option 1:Icon Text Button Replay Object Reference=Get ViewSettingsButton() -> (ViewSettingsButton:Icon Text Button Replay Object Reference)
     Index=Get bIsPlayingReplay() -> (bIsPlayingReplay:bool)
    ) -> (ReturnValue:Icon Text Button Replay Object Reference)
   )
  ]}
 ChangeGameSpeed {nodes [
   n1 = ChangeGameSpeed!() -> (then, Command:String, InText:Text) {is=title K2Node_FunctionEntry}
   > ExecuteConsoleCommand!(execute, Command=@n1, SpecificPlayer=GetOwningPlayer()) -> (then)
   > "SetText (Text)"(self=Get TextTimeDilation_Live() -> (TextTimeDilation_Live:Text Object Reference), InText=@n1.InText)
  ]}
 CollectReplayMarkers {nodes [
   CollectReplayMarkers() {is=title K2Node_FunctionEntry}
   > n1 = Branch(Condition=IsShippingBuild!(self:Fort Runtime Options Object Reference=GetRuntimeOptions()) -> (ReturnValue:bool))
   n2 = For Each Loop(
    Exec=@n1.else
    Array:Array of Fort Replay Timeline Marker Object References=GetReplayTimelineMarkersByType!(self:Athena Spectator HUDControls Base Object Reference, EventType:EFortReplayEventType Enum=Elimination) -> (ReturnValue:Array of Fort Replay Timeline Marker Object References)
   ) -> (Array Element:Fort Replay Timeline Marker Object Reference)
   n3 = For Each Loop(Exec=@n2.Completed, Array=GetReplayTimelineMarkersByType(EventType=Eliminated))
   n4 = For Each Loop(Exec=@n3.Completed, Array=GetReplayTimelineMarkersByType(EventType=DBNO))
   > FilterTimelineMarker!(
    execute:Exec
    self:Object Reference=GetPlayerController!(PlayerIndex:int=0) -> (ReturnValue:Player Controller Object Reference)
    Marker:Widget Object Reference=@n4
   ) -> (then:Exec) {is=title K2Node_Message}
   GetUserPlacedMarkers!(
    execute:Exec=@n4.Completed
    self:Object Reference=GetPlayerController(PlayerIndex=0)
    Markers:Array of Fort Replay Timeline Marker Object References=GetReplayTimelineMarkersByType(EventType=UserPlaced)
   ) -> (then:Exec) {is=title K2Node_Message}
   FilterTimelineMarker(execute=@n3.LoopBody, self=GetPlayerController(PlayerIndex=0), Marker=@n3)
   FilterTimelineMarker(execute=@n2.LoopBody, self=GetPlayerController(PlayerIndex=0), Marker=@n2)
  ]}
 DIsablePlaybackButtons {nodes [
   DIsablePlaybackButtons() {is=title K2Node_FunctionEntry}
   > Set bIsEnabled!(execute, bIsEnabled:bool=false, self:Widget Object Reference=Get Slower() -> (Slower:Icon Text Button Replay Object Reference)) -> (then, Output_Get:bool)
   > Set bIsEnabled(bIsEnabled=true, self=Get Faster() -> (Faster:Icon Text Button Replay Object Reference))
  ]}
 EventGraph {nodes [
   "On Button Clicked (JumptoStart)"!() -> (OutputDelegate, then, Button:Common Button Legacy Object Reference) {is=title K2Node_ComponentBoundEvent}
   > JumpToPreviousEvent(self:Fort Replay Context Object Reference=Get FortReplayContext())
   > ResetFocusToPlayPauseIfNeeded(self:Athena Spectator HUDControls Base Object Reference)
   "On Button Clicked (Back)"!() -> (OutputDelegate, then, Button:Common Button Legacy Object Reference) {is=title K2Node_ComponentBoundEvent}
   > SkipBack!(execute, self:Fort Replay Context Object Reference=Get FortReplayContext(), TimeInSeconds:float=Get TimeToSkip() -> (TimeToSkip:double)) -> (then)
   > ResetFocusToPlayPauseIfNeeded()
   n1 = "On Button Clicked (ViewSettingsButton)"!() -> (OutputDelegate, then, Button:Common Button Legacy Object Reference) {is=title K2Node_ComponentBoundEvent}
   ToggleViewSettingsVisibility(execute=[@n1.then, @n11.then], self:Fort Replay Context Object Reference=Get FortReplayContext())
   > ResetFocusToPlayPauseIfNeeded()
   n2 = "On Button Clicked (PreviousPlayerButton)"!() -> (OutputDelegate, then, Button:Common Button Legacy Object Reference) {is=title K2Node_ComponentBoundEvent}
   n3 = Cast To FortPlayerControllerSpectating!(execute=[@n2.then, @n9.then], Object=GetOwningPlayer()) -> (then, CastFailed, AsFort Player Controller Spectating:Fort Player Controller Spectating Object Reference) {is=title K2Node_DynamicCast}
   > FollowPlayer!(execute, self:Fort Player Controller Gameplay Object Reference=@n3, Follow:EPlayerControllerFollow Enum=PreviousPlayer) -> (then, ReturnValue:bool)
   > ResetFocusToPlayPauseIfNeeded()
   n4 = "On Button Clicked (NextPlayerButton)"!() -> (OutputDelegate, then, Button:Common Button Legacy Object Reference) {is=title K2Node_ComponentBoundEvent}
   n5 = Cast To FortPlayerControllerSpectating(execute=[@n4.then, @n10.then], Object=GetOwningPlayer())
   > FollowPlayer(self=@n5, Follow=NextPlayer)
   > ResetFocusToPlayPauseIfNeeded()
   n6 = "On Hovered (KeyframeButton)"!() -> (OutputDelegate, then, Button:Common Button Base Object Reference) {is=title K2Node_ComponentBoundEvent}
   > Set LastHoveredWidget(LastHoveredWidget:User Widget Object Reference=@n6.Button) -> (Output_Get:User Widget Object Reference)
   "On Button Clicked (IconTextButtonReplay_PlayHighlights)"!() -> (OutputDelegate, then, Button:Common Button Legacy Object Reference) {is=title K2Node_ComponentBoundEvent}
   > ToggleHighlightsListVisibility(self:Fort Replay Context Object Reference=Get FortReplayContext())
   "On Button Clicked (BattleMapTimelineModeButton)"!() -> (OutputDelegate, then, Button:Common Button Legacy Object Reference) {is=title K2Node_ComponentBoundEvent}
   > n7 = Cast To FortPlayerControllerSpectating(Object=GetOwningPlayer())
   > n8 = Cast To BP_BattleMapPawnReplay!(execute, Object=GetSpectatorPawn!(self:Player Controller Object Reference=@n7) -> (ReturnValue:Spectator Pawn Object Reference)) -> (then, CastFailed, AsBP Battle Map Pawn Replay:BP Battle Map Pawn Replay Object Reference) {is=title K2Node_DynamicCast}
   > BP_SetTimelineMode!(
    execute
    self:Battle Map Pawn Replay Object Reference=@n8
    bEnabled:bool=NOT Boolean(A=BP_IsOnTimelineMode!(self:Battle Map Pawn Replay Object Reference=@n8) -> (ReturnValue:bool))
   ) -> (then, ReturnValue:bool)
   n9 = "On Button Clicked (PreviousPlayerButton_Live)"!() -> (OutputDelegate, then, Button:Common Button Legacy Object Reference) {is=title K2Node_ComponentBoundEvent}
   n10 = "On Button Clicked (NextPlayerButton_Live)"!() -> (OutputDelegate, then, Button:Common Button Legacy Object Reference) {is=title K2Node_ComponentBoundEvent}
   n11 = "On Button Clicked (ViewSettingsButton_Live)"!() -> (OutputDelegate, then, Button:Common Button Legacy Object Reference) {is=title K2Node_ComponentBoundEvent}
   "On Unhovered (CameraTypeSpectatorContextMenu)"() {is=title K2Node_ComponentBoundEvent}
   > SetVisibility(self=Get CameraTypeSpectatorInputAction() -> (CameraTypeSpectatorInputAction:Common Action Widget Object Reference), InVisibility=Hidden)
   n12 = "On Button Clicked (PlayPause)"!() -> (OutputDelegate, then, Button:Common Button Legacy Object Reference) {is=title K2Node_ComponentBoundEvent}
   TogglePlayPause(execute=[@n136.then, @n12.then], self:Fort Replay Context Object Reference=Get FortReplayContext())
   > ResetFocusToPlayPauseIfNeeded()
   n13 = "On Hovered (JumpToEnd)"!() -> (OutputDelegate, then, Button:Common Button Base Object Reference) {is=title K2Node_ComponentBoundEvent}
   > Set LastHoveredWidget(LastHoveredWidget=@n13.Button)
   n14 = "On Hovered (Button_FocusTimeline)"!() -> (OutputDelegate, then, Button:Common Button Base Object Reference) {is=title K2Node_ComponentBoundEvent}
   > Set LastHoveredWidget(LastHoveredWidget=@n14.Button)
   n15 = "On Hovered (ViewSettingsButton)"!() -> (OutputDelegate, then, Button:Common Button Base Object Reference) {is=title K2Node_ComponentBoundEvent}
   > Set LastHoveredWidget(LastHoveredWidget=@n15.Button)
   n16 = "On Hovered (JumptoStart)"!() -> (OutputDelegate, then, Button:Common Button Base Object Reference) {is=title K2Node_ComponentBoundEvent}
   > Set LastHoveredWidget(LastHoveredWidget=@n16.Button)
   n17 = "On Hovered (Back)"!() -> (OutputDelegate, then, Button:Common Button Base Object Reference) {is=title K2Node_ComponentBoundEvent}
   > Set LastHoveredWidget(LastHoveredWidget=@n17.Button)
   "On Clicked (FreeDroneTeleportToPawn)"!() -> (OutputDelegate, then, Button:Common Button Base Object Reference) {is=title K2Node_ComponentBoundEvent}
   > n18 = Cast To BP_ReplayPC_Athena!(execute, Object=GetOwningPlayer()) -> (then, CastFailed, AsBP Replay PC Athena:BP Replay PC Athena Object Reference) {is=title K2Node_DynamicCast}
   > TeleportToPawn(self:BP Replay PC Athena Object Reference=@n18)
   n19 = "On Clicked (FreeDroneTargetHidingToolSelect)"!() -> (OutputDelegate, then, Button:Common Button Base Object Reference) {is=title K2Node_ComponentBoundEvent}
   n20 = Get HidingToolComponent() -> (HidingToolComponent:Fort Replay Hiding Tool Component Object Reference)
   Is Valid(exec=[@n19.then, @n33.then], InputObject=@n20)
   > HideActor(self:Fort Replay Hiding Tool Component Object Reference=@n20)
   n21 = "On Clicked (FreeDroneTargetHidingToolUndo)"!() -> (OutputDelegate, then, Button:Common Button Base Object Reference) {is=title K2Node_ComponentBoundEvent}
   n22 = Get HidingToolComponent()
   Is Valid(exec=[@n21.then, @n37.then], InputObject=@n22)
   > UnHideActor(self:Fort Replay Hiding Tool Component Object Reference=@n22)
   n23 = "On Hovered (Forward)"!() -> (OutputDelegate, then, Button:Common Button Base Object Reference) {is=title K2Node_ComponentBoundEvent}
   > Set LastHoveredWidget(LastHoveredWidget=@n23.Button)
   "On Hovered (PlayerSelectorContextMenu)"() {is=title K2Node_ComponentBoundEvent}
   > n24 = Branch(Condition=IsShippingBuild(self=GetRuntimeOptions()))
   Set LastHoveredWidget(
    execute=@n24.else
    LastHoveredWidget=Get PlayerSelectorContextMenu() -> (PlayerSelectorContextMenu:WBP UIKit Context Menu Anchor Button Object Reference)
   )
   > Branch(Condition=IsInputType!(InputType:ECommonInputType Enum=Gamepad, IsNot:bool=false) -> (Result:bool))
   n25 = Get PlayerSelectorInputAction() -> (PlayerSelectorInputAction:Common Action Widget Object Reference)
   > SetRenderOpacity(self=@n25, InOpacity=1)
   > SetVisibility(self=@n25, InVisibility=Visible)
   "On Button Clicked (Forward)"!() -> (OutputDelegate, then, Button:Common Button Legacy Object Reference) {is=title K2Node_ComponentBoundEvent}
   > SkipForward!(execute, self:Fort Replay Context Object Reference=Get FortReplayContext(), TimeInSeconds:float=Get TimeToSkip()) -> (then)
   > ResetFocusToPlayPauseIfNeeded()
   n26 = "On Hovered (BlurSettingsButton)"!() -> (OutputDelegate, then, Button:Common Button Base Object Reference) {is=title K2Node_ComponentBoundEvent}
   > Set LastHoveredWidget(LastHoveredWidget=@n26.Button)
   > n27 = GetBlurAmountTooltipText!(execute, self:Athena Spectator HUDControls Base Object Reference) -> (then, ReturnValue:Text)
   > SetToolTipText!(execute, self:Widget Object Reference=@n26.Button, InToolTipText:"Text (by ref)"=@n27) -> (then)
   "On Clicked (SpectatorEjectButton_Live)"!() -> (OutputDelegate, then, Button:Common Button Base Object Reference) {is=title K2Node_ComponentBoundEvent}
   > SetLightingToolVisibility!(execute, self:Fort Replay Context Object Reference=Get FortReplayContext(), IsVisible:bool=false) -> (then)
   > n28 = Cast To FortCaptureSpectatorController!(execute, Object=GetOwningPlayer()) -> (then, CastFailed, AsFort Capture Spectator Controller:Fort Capture Spectator Controller Object Reference) {is=title K2Node_DynamicCast}
   > ExecuteConsoleCommand(Command=cheat SpawnCaptureSpectator 0, SpecificPlayer=@n28)
   ExecuteConsoleCommand(execute=@n28.CastFailed, Command=spectator.EnableSpectatorMode 0)
   "On Clicked (TimecodeButton_Live)"!() -> (OutputDelegate, then, Button:Common Button Base Object Reference) {is=title K2Node_ComponentBoundEvent}
   > n29 = Flip Flop!(None:Exec) -> (A:Exec, B:Exec, IsA:bool=~) {is=title K2Node_MacroInstance}
   > ExecuteConsoleCommand(Command=DisplayAll GameState ElapsedTime, SpecificPlayer=GetOwningPlayer())
   ExecuteConsoleCommand(execute=@n29.B, Command=DisplayClear GameState ElapsedTime, SpecificPlayer=GetOwningPlayer())
   n30 = "On Hovered (FPSSettingsButton_Live)"!() -> (OutputDelegate, then, Button:Common Button Base Object Reference) {is=title K2Node_ComponentBoundEvent}
   > Set LastHoveredWidget(LastHoveredWidget=@n30.Button)
   n31 = "On Hovered (FreeDroneTargetHidingToolUndo)"!() -> (OutputDelegate, then, Button:Common Button Base Object Reference) {is=title K2Node_ComponentBoundEvent}
   > Set LastHoveredWidget(LastHoveredWidget=@n31.Button)
   n32 = "On Hovered (FreeDroneTargetHidingToolSelect)"!() -> (OutputDelegate, then, Button:Common Button Base Object Reference) {is=title K2Node_ComponentBoundEvent}
   > Set LastHoveredWidget(LastHoveredWidget=@n32.Button)
   n33 = "On Clicked (FreeDroneTargetHidingToolSelect_Live)"!() -> (OutputDelegate, then, Button:Common Button Base Object Reference) {is=title K2Node_ComponentBoundEvent}
   n34 = "On Hovered (SkipOneFrame)"!() -> (OutputDelegate, then, Button:Common Button Base Object Reference) {is=title K2Node_ComponentBoundEvent}
   > Set LastHoveredWidget(LastHoveredWidget=@n34.Button)
   "On Unhovered (PlayerSelectorContextMenu)"() {is=title K2Node_ComponentBoundEvent}
   > n35 = Branch(Condition=IsShippingBuild(self=GetRuntimeOptions()))
   n36 = Get PlayerSelectorInputAction()
   SetRenderOpacity(execute=@n35.else, self=@n36, InOpacity=0)
   > SetVisibility(self=@n36, InVisibility=Hidden)
   n37 = "On Clicked (FreeDroneTargetHidingToolUndo_Live)"!() -> (OutputDelegate, then, Button:Common Button Base Object Reference) {is=title K2Node_ComponentBoundEvent}
   "On Button Clicked (JumpToEnd)"!() -> (OutputDelegate, then, Button:Common Button Legacy Object Reference) {is=title K2Node_ComponentBoundEvent}
   > JumpToNextEvent(self:Fort Replay Context Object Reference=Get FortReplayContext())
   > ResetFocusToPlayPauseIfNeeded()
   n38 = "On Hovered (FreeDroneTeleportToPawn)"!() -> (OutputDelegate, then, Button:Common Button Base Object Reference) {is=title K2Node_ComponentBoundEvent}
   > Set LastHoveredWidget(LastHoveredWidget=@n38.Button)
   n39 = "On Hovered (PreviousPlayerButton)"!() -> (OutputDelegate, then, Button:Common Button Base Object Reference) {is=title K2Node_ComponentBoundEvent}
   > Set LastHoveredWidget(LastHoveredWidget=@n39.Button)
   n40 = "On Hovered (NextPlayerButton)"!() -> (OutputDelegate, then, Button:Common Button Base Object Reference) {is=title K2Node_ComponentBoundEvent}
   > Set LastHoveredWidget(LastHoveredWidget=@n40.Button)
   n41 = "On Hovered (ViewSettingsButton_Live)"!() -> (OutputDelegate, then, Button:Common Button Base Object Reference) {is=title K2Node_ComponentBoundEvent}
   > Set LastHoveredWidget(LastHoveredWidget=@n41.Button)
   n42 = "On Hovered (SpectatorEjectButton_Live)"!() -> (OutputDelegate, then, Button:Common Button Base Object Reference) {is=title K2Node_ComponentBoundEvent}
   > Set LastHoveredWidget(LastHoveredWidget=@n42.Button)
   n43 = "On Hovered (TimecodeButton_Live)"!() -> (OutputDelegate, then, Button:Common Button Base Object Reference) {is=title K2Node_ComponentBoundEvent}
   > Set LastHoveredWidget(LastHoveredWidget=@n43.Button)
   n44 = "On Hovered (FreeDroneTargetHidingToolUndo_Live)"!() -> (OutputDelegate, then, Button:Common Button Base Object Reference) {is=title K2Node_ComponentBoundEvent}
   > Set LastHoveredWidget(LastHoveredWidget=@n44.Button)
   "On Hovered (PlayerSelectorSpectatorContextMenu)"() {is=title K2Node_ComponentBoundEvent}
   > n45 = Branch(Condition=IsShippingBuild(self=GetRuntimeOptions()))
   Set LastHoveredWidget(
    execute=@n45.else
    LastHoveredWidget=Get PlayerSelectorSpectatorContextMenu() -> (PlayerSelectorSpectatorContextMenu:WBP UIKit Context Menu Anchor Button Object Reference)
   )
   > Branch(Condition=IsInputType(InputType=Gamepad, IsNot=false))
   n46 = Get PlayerSelectorSpectatorInputAction() -> (PlayerSelectorSpectatorInputAction:Common Action Widget Object Reference)
   > SetRenderOpacity(self=@n46, InOpacity=1)
   > SetVisibility(self=@n46, InVisibility=Visible)
   n47 = "On Hovered (FreeDroneTargetHidingToolSelect_Live)"!() -> (OutputDelegate, then, Button:Common Button Base Object Reference) {is=title K2Node_ComponentBoundEvent}
   > Set LastHoveredWidget(LastHoveredWidget=@n47.Button)
   "On Button Clicked (SkipOneFrame)"!() -> (OutputDelegate, then, Button:Common Button Legacy Object Reference) {is=title K2Node_ComponentBoundEvent}
   > StepForwardOneFrame(self:Fort Replay Context Object Reference=Get FortReplayContext())
   "On Button Clicked (Slower)"!() -> (OutputDelegate, then, Button:Common Button Legacy Object Reference) {is=title K2Node_ComponentBoundEvent}
   > DecreasePlaybackMultiplier(self:Fort Replay Context Object Reference=Get FortReplayContext())
   > ResetFocusToPlayPauseIfNeeded()
   n48 = "On Hovered (PreviousPlayerButton_Live)"!() -> (OutputDelegate, then, Button:Common Button Base Object Reference) {is=title K2Node_ComponentBoundEvent}
   > Set LastHoveredWidget(LastHoveredWidget=@n48.Button)
   n49 = "On Hovered (NextPlayerButton_Live)"!() -> (OutputDelegate, then, Button:Common Button Base Object Reference) {is=title K2Node_ComponentBoundEvent}
   > Set LastHoveredWidget(LastHoveredWidget=@n49.Button)
   "On Hovered (CameraTypeContextMenu)"() {is=title K2Node_ComponentBoundEvent}
   > Set LastHoveredWidget(LastHoveredWidget=Get CameraTypeContextMenu() -> (CameraTypeContextMenu:WBP UIKit Context Menu Anchor Button Object Reference))
   > Branch(Condition=IsInputType(InputType=Gamepad, IsNot=false))
   > SetRenderOpacity(self=Get CameraTypeInputAction() -> (CameraTypeInputAction:Common Action Widget Object Reference), InOpacity=1)
   "On Unhovered (CameraTypeContextMenu)"() {is=title K2Node_ComponentBoundEvent}
   > SetRenderOpacity(self=Get CameraTypeInputAction(), InOpacity=0)
   "On Unhovered (PlayerSelectorSpectatorContextMenu)"() {is=title K2Node_ComponentBoundEvent}
   > n50 = Branch(Condition=IsShippingBuild(self=GetRuntimeOptions()))
   n51 = Get PlayerSelectorSpectatorInputAction()
   SetRenderOpacity(execute=@n50.else, self=@n51, InOpacity=0)
   > SetVisibility(self=@n51, InVisibility=Hidden)
   n52 = "On Hovered (BlurSettingsButton_Live)"!() -> (OutputDelegate, then, Button:Common Button Base Object Reference) {is=title K2Node_ComponentBoundEvent}
   > Set LastHoveredWidget()
   > n53 = GetBlurAmountTooltipText()
   > SetToolTipText(self=@n52.Button, InToolTipText=@n53)
   "On Clicked (SaveShot)"!() -> (OutputDelegate, then, Button:Common Button Base Object Reference) {is=title K2Node_ComponentBoundEvent}
   n54 = GetOwningPlayer()
   > n55 = Cast To FortReplaySpectator!(execute, Object=@n54) -> (then, CastFailed, AsFort Replay Spectator:Fort Replay Spectator Object Reference) {is=title K2Node_DynamicCast}
   n56 = Get CurrentShot() -> (CurrentShot:int=0)
   > SetCameraShot!(execute, self:Fort Replay Spectator Object Reference=@n55, HotKeyNum:int=@n56) -> (then)
   n57 = GetOwningPlayer()
   > n58 = Cast To FortReplaySpectator(Object=@n57)
   > ShowNotification!(execute, self:Fort Player Controller Spectating Object Reference=@n58, NotificationText:Text="LOCTEXT(\"Camera Saved\")") -> (then)
   "On Clicked (SaveShotFull)"!() -> (OutputDelegate, then, Button:Common Button Base Object Reference) {is=title K2Node_ComponentBoundEvent}
   > n59 = Cast To FortReplaySpectator(Object=@n54)
   > SetCameraShotWithTime!(execute, self:Fort Replay Spectator Object Reference=@n59, HotKeyNum:int=@n56) -> (then)
   > n60 = Cast To FortReplaySpectator(Object=@n57)
   > ShowNotification(self=@n60, NotificationText="LOCTEXT(\"Camera and Time Saved\")")
   "On Hovered (ShotNumberContextMenu)"() {is=title K2Node_ComponentBoundEvent}
   > Set LastHoveredWidget(LastHoveredWidget=Get ShotNumberContextMenu() -> (ShotNumberContextMenu:WBP UIKit Context Menu Anchor Button Object Reference))
   > Branch(Condition=IsInputType(InputType=Gamepad, IsNot=false))
   > SetRenderOpacity(self=Get ShotNumberInputAction() -> (ShotNumberInputAction:Common Action Widget Object Reference), InOpacity=1)
   "On Button Clicked (Faster)"!() -> (OutputDelegate, then, Button:Common Button Legacy Object Reference) {is=title K2Node_ComponentBoundEvent}
   > IncreasePlaybackMultiplier(self:Fort Replay Context Object Reference=Get FortReplayContext())
   > ResetFocusToPlayPauseIfNeeded()
   "On Unhovered (ShotNumberContextMenu)"() {is=title K2Node_ComponentBoundEvent}
   > SetRenderOpacity(self=Get ShotNumberInputAction(), InOpacity=0)
   "On Clicked (CurrentShotButton)"!() -> (OutputDelegate, then, Button:Common Button Base Object Reference) {is=title K2Node_ComponentBoundEvent}
   > n61 = Cast To FortReplaySpectator(Object=GetOwningPlayer())
   > InvokeHotKeyByNumber!(execute, self:Fort Player Controller Spectating Object Reference=@n61, HotKeyNum:int=Get CurrentShot() -> (CurrentShot=0)) -> (then)
   n62 = "On Hovered (CurrentShotButton)"!() -> (OutputDelegate, then, Button:Common Button Base Object Reference) {is=title K2Node_ComponentBoundEvent}
   > Set LastHoveredWidget(LastHoveredWidget=@n62.Button)
   n63 = "On Hovered (SaveShot)"!() -> (OutputDelegate, then, Button:Common Button Base Object Reference) {is=title K2Node_ComponentBoundEvent}
   > Set LastHoveredWidget(LastHoveredWidget=@n63.Button)
   n64 = "On Hovered (SaveShotFull)"!() -> (OutputDelegate, then, Button:Common Button Base Object Reference) {is=title K2Node_ComponentBoundEvent}
   > Set LastHoveredWidget(LastHoveredWidget=@n64.Button)
   n65 = "On Clicked (FPSSettingsButton_Live)"!() -> (OutputDelegate, then, Button:Common Button Base Object Reference) {is=title K2Node_ComponentBoundEvent}
   n66 = GetFPSBlurMultiplier!(execute=[@n65.then, @n73.then], self:Athena Spectator HUDControls Base Object Reference) -> (then, ReturnValue:float)
   > n67 = Branch(Condition="Nearly Equal (Float)"(A=@n66, B=1, ErrorTolerance=0.01))
   > n68 = SetFPSBlurMultiplier!(execute, self:Athena Spectator HUDControls Base Object Reference, InMultiplier:float=2) -> (then)
   UpdateDevFPSBlurButtonVisibility(execute=[@n69.then, @n68.then])
   > UpdateDevBlurButtonVisibility()
   n69 = SetFPSBlurMultiplier(execute=@n67.else, InMultiplier=1)
   n70 = "On Hovered (RecorderButton)"!() -> (OutputDelegate, then, Button:Common Button Base Object Reference) {is=title K2Node_ComponentBoundEvent}
   > Set LastHoveredWidget(LastHoveredWidget=@n70.Button)
   n71 = "On Hovered (FPSSettingsButton)"!() -> (OutputDelegate, then, Button:Common Button Base Object Reference) {is=title K2Node_ComponentBoundEvent}
   > Set LastHoveredWidget(LastHoveredWidget=@n71.Button)
   n72 = "On Clicked (BlurSettingsButton_Live)"!() -> (OutputDelegate, then, Button:Common Button Base Object Reference) {is=title K2Node_ComponentBoundEvent}
   ChangeBlurState(execute=[@n72.then, @n85.then], self:Athena Spectator HUDControls Base Object Reference)
   > UpdateDevBlurButtonVisibility()
   n73 = "On Clicked (FPSSettingsButton)"!() -> (OutputDelegate, then, Button:Common Button Base Object Reference) {is=title K2Node_ComponentBoundEvent}
   n74 = "On Clicked (Faster_Live)"!() -> (OutputDelegate, then, Button:Common Button Base Object Reference) {is=title K2Node_ComponentBoundEvent}
   n75 = IsShippingBuild(self=GetRuntimeOptions())
   n76 = GetIsPlayingReplay!() -> (ReturnValue:bool)
   Branch(execute=[@n216.Pressed, @n74.then], Condition=AND Boolean(A=NOT Boolean(A=@n75), B=NOT Boolean(A=@n76)))
   > n77 = Increment Int(Value=Get SpectatorSpeeds() -> (SpectatorSpeeds:int=0))
   > Set SpectatorSpeeds(SpectatorSpeeds:int="Clamp (Integer)"(Value=@n77, Min=0, Max=3) -> (ReturnValue=0)) -> (Output_Get:int=0)
   > n78 = Switch on Int!(execute:Exec, Selection:int=Get SpectatorSpeeds() -> (SpectatorSpeeds=0)) -> (Default:Exec, 0:Exec, 1:Exec, 2:Exec, 3:Exec)
   ChangeGameSpeed!(execute=@n78.0, self, Command:String=Cheat Slomo 0.1, InText:Text="LOCTEXT(\"0.10x\")") -> (then) {is=title K2Node_CallFunction}
   > Set SpectatorSpeeds(SpectatorSpeeds=0) -> (Output_Get=0)
   ChangeGameSpeed(execute=@n78.3, Command=Cheat Slomo 1, InText="LOCTEXT(\"1.00x\")")
   > Set SpectatorSpeeds(SpectatorSpeeds=3) -> (Output_Get=0)
   ChangeGameSpeed(execute=@n78.1, Command=Cheat Slomo 0.25, InText="LOCTEXT(\"0.25x\")")
   > Set SpectatorSpeeds(SpectatorSpeeds=1) -> (Output_Get=0)
   ChangeGameSpeed(execute=@n78.2, Command=Cheat Slomo 0.5, InText="LOCTEXT(\"0.50x\")")
   > Set SpectatorSpeeds(SpectatorSpeeds=2) -> (Output_Get=0)
   "On Hovered (CameraTypeSpectatorContextMenu)"() {is=title K2Node_ComponentBoundEvent}
   > n79 = Branch(Condition=IsInputType(InputType=Gamepad, IsNot=false))
   > SetVisibility(self=Get CameraTypeSpectatorInputAction(), InVisibility=Visible)
   Set LastHoveredWidget(
    execute=@n79.else
    LastHoveredWidget=Get CameraTypeSpectatorContextMenu() -> (CameraTypeSpectatorContextMenu:WBP UIKit Context Menu Anchor Button Object Reference)
   )
   n80 = "On Clicked (Slower_Live)"!() -> (OutputDelegate, then, Button:Common Button Base Object Reference) {is=title K2Node_ComponentBoundEvent}
   n81 = GetIsPlayingReplay()
   n82 = IsShippingBuild(self=GetRuntimeOptions())
   Branch(execute=[@n215.Pressed, @n80.then], Condition=AND Boolean(A=NOT Boolean(A=@n82), B=NOT Boolean(A=@n81)))
   > n83 = Decrement Int(Value=Get SpectatorSpeeds() -> (SpectatorSpeeds=0))
   > Set SpectatorSpeeds(SpectatorSpeeds="Clamp (Integer)"(Value=@n83, Min=0, Max=3) -> (ReturnValue=0)) -> (Output_Get=0)
   > n84 = Switch on Int(Selection=Get SpectatorSpeeds() -> (SpectatorSpeeds=0))
   ChangeGameSpeed(execute=@n84.3, Command=Cheat Slomo 1, InText="LOCTEXT(\"1.00x\")")
   > Set SpectatorSpeeds(SpectatorSpeeds=3) -> (Output_Get=0)
   ChangeGameSpeed(execute=@n84.1, Command=Cheat Slomo 0.25, InText="LOCTEXT(\"0.25x\")")
   > Set SpectatorSpeeds(SpectatorSpeeds=1) -> (Output_Get=0)
   ChangeGameSpeed(execute=@n84.2, Command=Cheat Slomo 0.5, InText="LOCTEXT(\"0.50x\")")
   > Set SpectatorSpeeds(SpectatorSpeeds=2) -> (Output_Get=0)
   ChangeGameSpeed(execute=@n84.0, Command=Cheat Slomo 0.1, InText="LOCTEXT(\"0.10x\")")
   > Set SpectatorSpeeds(SpectatorSpeeds=0) -> (Output_Get=0)
   n85 = "On Clicked (BlurSettingsButton)"!() -> (OutputDelegate, then, Button:Common Button Base Object Reference) {is=title K2Node_ComponentBoundEvent}
   OnKeybindsChanged() {is=title K2Node_CustomEvent}
   > ResetFocusToPlayPauseIfNeeded()
   > UpdateButtonIconsVisibility()
   > UpdateCameraControls()
   > n86 = Branch(Condition=IsInputType(InputType=Gamepad, IsNot=false))
   n87 = Get CameraTypeContextMenu()
   n88 = Get LastHoveredWidget()
   > n89 = Branch(Condition=NOT Boolean(A="Equal (Object)"(A=@n88, B=@n87)))
   > n90 = SetRenderOpacity(self=Get CameraTypeInputAction(), InOpacity=0)
   SetVisibility(
    execute=[@n90.then, @n89.else]
    self=[
     Get PlayerSelectorSpectatorContextBox() -> (PlayerSelectorSpectatorContextBox:Vertical Box Object Reference)
     Get CamTypeSpectatorContextBox() -> (CamTypeSpectatorContextBox:Vertical Box Object Reference)
    ]
    InVisibility=Collapsed
   )
   > n91 = Branch(Condition=Get bIsPlayingReplay())
   > n92 = SetPadding!(
    execute
    self:User Widget Object Reference=Get ControlsWrapper() -> (ControlsWrapper:Fort HUDElement Wrapper Object Reference)
    InPadding_Left:float=0
    InPadding_Top:float=-100
    InPadding_Right:float=0
    InPadding_Bottom:float=0
   ) -> (then)
   CenterControl(execute=[@n92.then, @n97.then, @n98.then, @n95.then]) {is=title K2Node_CallFunction}
   > Branch(Condition=Get bViewSettingsOpen() -> (bViewSettingsOpen:bool))
   > n93 = Branch(Condition=IsInputType(InputType=Gamepad, IsNot=false))
   > ShowRelevantControls!(execute, self, Config:int=1) -> (then)
   > n94 = Branch(Condition=Get bPanelOpen() -> (bPanelOpen:bool))
   > ShowRelevantControls(Config=1)
   > SetInputMode!(execute, self, InputType:int=0) -> (then)
   ShowRelevantControls(execute=@n94.else, Config=2)
   > SetInputMode(InputType=2)
   ShowRelevantControls(execute=@n93.else, Config=0)
   > UpdateZoomKeybindText()
   > SetInputMode(InputType=1)
   > SetTabCollapsed!(execute, self:Fort Replay Context Object Reference=Get FortReplayContext(), bIsCollapsed:bool=true) -> (then)
   n95 = SetPadding(execute=@n91.else, self=Get ControlsWrapper(), InPadding_Left=0, InPadding_Top=5, InPadding_Right=0, InPadding_Bottom=0)
   n96 = Branch(execute=@n86.else, Condition=IsInputType(InputType=MouseAndKeyboard, IsNot=false))
   > SetVisibility(self=[Get PlayerSelectorSpectatorContextBox(), Get CamTypeSpectatorContextBox()], InVisibility=SelfHitTestInvisible)
   > n97 = SetPadding(self=Get ControlsWrapper(), InPadding_Left=0, InPadding_Top=5, InPadding_Right=0, InPadding_Bottom=0)
   n98 = SetVisibility(execute=@n96.else, self=[Get PlayerSelectorSpectatorContextBox(), Get CamTypeSpectatorContextBox()], InVisibility=SelfHitTestInvisible)
   n99 = OnPauseStateChanged!() -> (OutputDelegate, then, bPaused:bool=~) {is=title K2Node_CustomEvent}
   > SetBrushFromSoftTexture!(
    execute
    self:Image Object Reference=Get Icon!(self:Icon Text Button Replay Object Reference=Get PlayPause() -> (PlayPause:Icon Text Button Replay Object Reference)) -> (Icon:Image Object Reference)
    SoftTexture:Texture 2D Soft Object Reference=Select(
     Option 0:Texture 2D Soft Object Reference=/Game/UI/Foundation/Textures/Icons/Replay/T-Icon-Replay-Pause-L
     Option 1:Texture 2D Soft Object Reference=/Game/UI/Foundation/Textures/Icons/Replay/T-Icon-Replay-Play-L
     Index=@n99.bPaused
    ) -> (ReturnValue:Texture 2D Soft Object Reference)
    bMatchSize:bool=false
   ) -> (then)
   n100 = Get FortReplayContext()
   > n101 = Switch on EHudVisibilityState!(
    execute
    Selection:EHudVisibilityState Enum=Break Fort Replay Playback State(FortReplayPlaybackState=Get PlaybackState(self=@n100)) -> (
     StartTime=~
     EndTime=~
     TimeNow=~
     bIsPaused=~
     PlaybackSpeedMultiplier=~
     bLevelStreaming=~
     bHasRelevancyZone=~
     bHasPlayInterval=~
     PlayIntervalStartTime=~
     PlayIntervalEndTime=~
     bLoopInterval=~
     IntervalLoopBufferTime=~
    ).HUDVisibility
   ) -> (FullyVisible:Exec, FullyHidden:Exec, GameOnly:Exec, ReplayOnly:Exec) {is=title K2Node_SwitchEnum}
   PlaySound2D(
    execute=[@n101.FullyVisible, @n101.ReplayOnly]
    Sound=Select(
     Option 0:Sound Base Object Reference=/Game/Sounds/Fort_UI_Sounds/Replay/DVR_Nav_Stop_Cue
     Option 1:Sound Base Object Reference=/Game/Sounds/Fort_UI_Sounds/Replay/DVR_Nav_Play_Cue
     Index=@n99.bPaused
    ) -> (ReturnValue:Sound Base Object Reference)
   )
   > ResetFocusToPlayPauseIfNeeded()
   n102 = ReapllyCapsuleHalfHeight() {is=title K2Node_CustomEvent}
   > n103 = GetConsoleVariableFloatValue!(execute, VariableName:String=OverridenCapsuleHalfHeight) -> (then, ReturnValue:float)
   > n104 = Set CapsuleHalfHeight(CapsuleHalfHeight:double=@n103) -> (Output_Get:double)
   > Branch(Condition="Not Equal (Float)"(A=@n104))
   > n105 = GetAllActorsOfClass!(execute, ActorClass:Actor Class Reference=/Script/FortniteGame.FortPlayerPawnAthena) -> (then, OutActors:Array of Fort Player Pawn Athena Object References)
   > n106 = For Each Loop(Array:Array of Fort Player Pawn Athena Object References=@n105) -> (Array Element:Fort Player Pawn Athena Object Reference)
   > SetCapsuleHalfHeight!(
    execute
    self:Capsule Collision Object Reference=Get CapsuleComponent!(self:Character Object Reference=@n106) -> (CapsuleComponent:Capsule Collision Object Reference)
    HalfHeight:float=Get CapsuleHalfHeight() -> (CapsuleHalfHeight:double)
    bUpdateOverlaps:bool=true
   ) -> (then)
   n107 = FollowPlayer!() -> (OutputDelegate, then, InteractionNameID:Name) {is=title K2Node_CustomEvent}
   > n108 = Cast To FortPlayerControllerSpectating(Object=GetOwningPlayer())
   n109 = "To String (Name)"(InName=@n107.InteractionNameID)
   > FollowPlayerByName!(
    execute
    self:Fort Player Controller Spectating Object Reference=@n108
    PlayerName:String=Find(
     TargetMap:Map of Integers to Strings=Get PlayerNames() -> (PlayerNames:Map of Integers to Strings)
     Key:"Integer (by ref)"=String To Integer(InString=@n109)
    )
   ) -> (then)
   n110 = SwitchShot!() -> (OutputDelegate, then, InteractionNameID:Name) {is=title K2Node_CustomEvent}
   n111 = "To String (Name)"(InName=@n110.InteractionNameID)
   > n112 = Set CurrentShot(CurrentShot:int=String To Integer(InString=@n111)) -> (Output_Get:int=0)
   > n113 = Cast To FortReplaySpectator(Object=GetOwningPlayer())
   > InvokeHotKeyByNumber(self=@n113, HotKeyNum=@n112)
   > SetText!(
    execute
    self:Icon Text Button Spectator Base Object Reference=Get CurrentShotButton() -> (CurrentShotButton:Icon Text Button Replay Object Reference)
    Text:"Text (by ref)"="To Text (String)"(InString=Append(A="Shot ", B=@n111))
   ) -> (then)
   n114 = WaitForMarkers() {is=title K2Node_CustomEvent}
   > Set ElapsedTime(ElapsedTime:double=float + float(A=Get ElapsedTime() -> (ElapsedTime:double), B=Get TimerSpeed() -> (TimerSpeed:float))) -> (Output_Get:double)
   > n115 = Branch(Condition="float >= float"(A=Get ElapsedTime(), B=10))
   > Clear and Invalidate Timer by Handle(Handle=Get MarkerTimerHandle() -> (MarkerTimerHandle:Timer Handle Structure))
   > Set bSearchingForMarkers(bSearchingForMarkers:bool=false) -> (Output_Get:bool)
   > n116 = GetAllWidgetsOfClass!(execute, WidgetClass:User Widget Class Reference=/Script/FortniteUI.FortReplayTimelineMarker, TopLevelOnly:bool=false) -> (then, FoundWidgets:Array of Fort Replay Timeline Marker Object References)
   > GetUserPlacedMarkers(self=GetPlayerController(PlayerIndex=0), Markers=@n116)
   Branch(
    execute=@n115.else
    Condition=Is Valid Index!(TargetArray:Array of Fort Replay Timeline Marker Object References=GetReplayTimelineMarkersByType(EventType=UserPlaced), IndexToTest:int) -> (ReturnValue:bool)
   )
   > CollectReplayMarkers() {is=title K2Node_CallFunction}
   > Clear and Invalidate Timer by Handle(Handle=Get MarkerTimerHandle())
   > Set bSearchingForMarkers(bSearchingForMarkers=false)
   n117 = OnHotkeyInvoked!() -> (OutputDelegate, then, HotKeyNum:int) {is=title K2Node_CustomEvent}
   > Set CurrentShot(CurrentShot=@n117.HotKeyNum) -> (Output_Get=0)
   n118 = "To Text (String)"(InString=Append(A="Shot ", B="To String (Integer)"(InInt=@n117.HotKeyNum)))
   > n119 = SetText(self=Get CurrentShotButton(), Text=@n118)
   n120 = Cast To FortReplaySpectator(execute=@n172.Completed, Object=GetOwningPlayer())
   n121 = int * int(A=GetCameraShotsPreset!(self:Fort Player Controller Spectating Object Reference=@n120) -> (ReturnValue:int=0), B=5)
   ShowNotification(
    execute=@n119.then
    self=@n120
    NotificationText=Format Text!(
     Format="LOCTEXT(\"{Preset} - {Shot}\")"
     Preset:Text="To Text (Integer)"(Value=@n121, bUseGrouping=false, MinimumIntegralDigits=4, MaximumIntegralDigits=10)
     Shot:Text=@n118
    ) -> (Result)
   )
   n122 = OnHoykeyAssigned!() -> (OutputDelegate, then, HotKeyNum:int) {is=title K2Node_CustomEvent}
   > n123 = Set CurrentShot(CurrentShot=@n122.HotKeyNum) -> (Output_Get=0)
   n124 = "To String (Integer)"(InInt=@n123)
   > SetText(self=Get CurrentShotButton(), Text="To Text (String)"(InString=Append(A="Shot ", B=@n124)))
   InitialWaitForMarkers() {is=title K2Node_CustomEvent}
   > n125 = Branch(Condition=IsShippingBuild(self=GetRuntimeOptions()))
   n126 = Branch(execute=@n125.else, Condition=GetIsPlayingReplay())
   n127 = Branch(execute=[@n152.then, @n126.then], Condition=Get bSearchingForMarkers() -> (bSearchingForMarkers:bool))
   Set bSearchingForMarkers(execute=@n127.else, bSearchingForMarkers=true)
   > Set ElapsedTime(ElapsedTime=0)
   > n128 = Set Timer by Event(Delegate=@n114, Time=Get TimerSpeed(), bLooping=true)
   > Set MarkerTimerHandle(MarkerTimerHandle:Timer Handle Structure=@n128) -> (Output_Get:Timer Handle Structure)
   ForceShowMotionBlur() {is=title K2Node_CustomEvent}
   > n129 = Branch(Condition=IsShippingBuild(self=GetRuntimeOptions()))
   SetVisibility(
    execute=@n129.else
    self=Get BlurSettingsButton_Live() -> (BlurSettingsButton_Live:Icon Text Button Replay Object Reference)
    InVisibility=Visible
   )
   n130 = ChangeCameraSpectator!() -> (OutputDelegate, then, InteractionNameID:Name) {is=title K2Node_CustomEvent}
   > n131 = Cast To FortPlayerControllerSpectating(Object=GetOwningPlayer())
   n132 = "To String (Name)"(InName=@n130.InteractionNameID)
   > SetNewCameraType!(
    execute
    self:Fort Player Controller Spectating Object Reference=@n131
    NewCameraType:ESpectatorCameraType Enum="Get (a copy)"(
     Array:Array of ESpectatorCameraType Enums=Get CamType() -> (CamType:Array of ESpectatorCameraType Enums=ThirdPerson)
     Dimension 1=String To Integer(InString=@n132)
    ) -> (Output:ESpectatorCameraType Enum)
    bReset:bool=false
   ) -> (then)
   n133 = ChangeCamera!() -> (OutputDelegate, then, InteractionNameID:Name) {is=title K2Node_CustomEvent}
   > n134 = Cast To FortPlayerControllerSpectating(Object=GetOwningPlayer())
   n135 = "To String (Name)"(InName=@n133.InteractionNameID)
   > SetNewCameraType(
    self=@n134
    NewCameraType="Get (a copy)"(Array=Get CamType() -> (CamType=ThirdPerson), Dimension 1=String To Integer(InString=@n135))
    bReset=false
   )
   n136 = OnTogglePlayPause() {is=title K2Node_CustomEvent}
   n137 = HandleToggleViewSettings() {is=title K2Node_CustomEvent}
   > n138 = Branch(Condition=Get bViewSettingsOpen())
   > SetIsEnabled(
    self=[
     Get ReplaySettingsWrapper() -> (ReplaySettingsWrapper:Fort HUDElement Wrapper Object Reference)
     Get SpectatorSettingsSize() -> (SpectatorSettingsSize:Size Box Object Reference)
    ]
    bInIsEnabled=true
   )
   > Set bViewSettingsOpen(bViewSettingsOpen:bool=false) -> (Output_Get:bool)
   > UpdateZoomKeybindText()
   > n139 = Branch(Condition=IsInputType(InputType=MouseAndKeyboard, IsNot=false))
   > Reroute Node()
   ShowRelevantControls(execute=@n139.else, Config=0)
   > SetInputMode(InputType=1)
   Set bViewSettingsOpen(execute=@n138.else, bViewSettingsOpen=true)
   > n140 = Branch(Condition=IsInputType(InputType=Gamepad, IsNot=false))
   > ShowRelevantControls(Config=1)
   > SetInputMode(InputType=0)
   Reroute Node(InputPin=@n140.else)
   n141 = FollowPlayer_Spectator!() -> (OutputDelegate, then, InteractionNameID:Name) {is=title K2Node_CustomEvent}
   > n142 = Cast To FortPlayerControllerSpectating(Object=GetOwningPlayer())
   n143 = "To String (Name)"(InName=@n141.InteractionNameID)
   > FollowPlayerByName(self=@n142, PlayerName=Find(TargetMap=Get PlayerNames(), Key=String To Integer(InString=@n143)))
   n144 = HandleOnCollapsedPanel!() -> (OutputDelegate, then, IsCollapsed:bool=~) {is=title K2Node_CustomEvent}
   > Branch(Condition=Get bViewSettingsOpen())
   > n145 = Set bPanelOpen(bPanelOpen:bool=@n144.IsCollapsed) -> (Output_Get:bool)
   > n146 = Branch(Condition=@n145)
   > n147 = Branch(Condition=IsInputType(InputType=Gamepad, IsNot=false))
   > ShowRelevantControls(Config=1)
   > SetInputMode(InputType=0)
   ShowRelevantControls(execute=@n147.else, Config=0)
   > SetInputMode(InputType=1)
   n148 = Branch(execute=@n146.else, Condition=IsInputType(InputType=Gamepad, IsNot=false))
   > ShowRelevantControls(Config=2)
   > SetInputMode(InputType=2)
   ShowRelevantControls(execute=@n148.else, Config=0)
   > SetInputMode(InputType=1)
   n149 = OnReplayPlaybackMultiplierChanged!() -> (OutputDelegate, then, Multiplier:float=~) {is=title K2Node_CustomEvent}
   > "SetText (Text)"(
    self=Get TextTimeDilation() -> (TextTimeDilation:Text Object Reference)
    InText=Format Text!(
     Format="INVTEXT(\"{0}x\")"
     0:Text="To Text (Float)"(Value=@n149.Multiplier, MinimumIntegralDigits=1, MaximumIntegralDigits=1, MinimumFractionalDigits=2, MaximumFractionalDigits=2)
    ) -> (Result)
   )
   > ResetFocusToPlayPauseIfNeeded()
   n150 = Scrubbling() {is=title K2Node_CustomEvent}
   n151 = Add Component by Class!(
    execute:Exec=@n194.Is Not Valid
    self:Actor Object Reference=GetOwningPlayer()
    Class:Actor Component Class Reference=/Script/FortniteGame.FortReplayHidingToolComponent
   ) -> (then:Exec, ReturnValue:Fort Replay Hiding Tool Component Object Reference) {is=title K2Node_AddComponentByClass}
   n152 = ReapplyAllActors(execute=@n150.then, self:Fort Replay Hiding Tool Component Object Reference=@n151)
   n153 = EnhancedInputAction IA_OpenSideBarMenu!() -> (
    Triggered:Exec
    Started:Exec
    Ongoing:Exec
    Canceled:Exec
    Completed:Exec
    ActionValue:bool
    ElapsedSeconds:double=~
    TriggeredSeconds:double=~
    InputAction:Input Action Object Reference=/Game/Input/EnhancedInput/FortPlayerController/IA_OpenSideBarMenu
   ) {is=title K2Node_EnhancedInputAction}
   > Set bSidebarOpen(bSidebarOpen:bool=InvertBool!(bValue:bool=@n153) -> (ReturnValue:bool)) -> (Output_Get:bool)
   n154 = Event PreConstruct()
   n155 = GetOwningPlayer()
   > Set bIsPlayingReplay(bIsPlayingReplay:bool=Select(
     Option 0:bool=Cast To FortReplaySpectatorAthena!(Object=@n155) -> (AsFort Replay Spectator Athena:Fort Replay Spectator Athena Object Reference, bSuccess:bool=~) {is=title K2Node_DynamicCast}.bSuccess
     Option 1:bool=Get DesignTime_IsReplay() -> (DesignTime_IsReplay:bool)
     Index=@n154.IsDesignTime
    ) -> (ReturnValue:bool)) -> (Output_Get:bool)
   Event Construct()
   > InitialWaitForMarkers() {is=title K2Node_CallFunction}
   > ExecuteConsoleCommand(Command=spectator.EnableGameInputForSettingsTab 1, SpecificPlayer=GetOwningPlayer())
   > Bind Event to On Keybinds Changed(
    self:Fort Global UIContext Object Reference=Get FortGlobalUIContext!() -> (ReturnValue:Fort Global UIContext Object Reference) {is=title K2Node_GetBlueprintContextEx}
    Delegate=Create Event()
   ) {is=title K2Node_AddDelegate}
   n156 = Get BattleMapTimelineModeButton() -> (BattleMapTimelineModeButton:Icon Text Button Replay Object Reference)
   > Set BattleMapTimelineTriggeringInputActions(BattleMapTimelineTriggeringInputActions:Data Table Row Handle Structure=Get TriggeringInputAction!(self:Common Button Base Object Reference=@n156) -> (TriggeringInputAction:Data Table Row Handle Structure)) -> (Output_Get:Data Table Row Handle Structure)
   > SetTriggeringInputAction!(
    execute
    self:Common Button Base Object Reference=@n156
    InputActionRow:"Data Table Row Handle Structure (by ref)"=Make Data Table Row Handle!(DataTable:Data Table Object Reference, RowName:Name=None) -> (DataTableRowHandle:Data Table Row Handle Structure) {is=title K2Node_MakeStruct}
   ) -> (then)
   > n157 = Branch(Condition=Get bIsPlayingReplay())
   > n158 = BindReplayContext() {is=title K2Node_CallFunction}
   OnKeybindsChanged(execute=[@n158.then, @n188.else, @n183.then, @n189.then, @n185.then, @n187.then]) {is=title K2Node_CallFunction}
   > Bind Event to On Replay Hud Visibility Changed(self:Fort Replay Context Object Reference=Get FortReplayContext(), Delegate=Create Event()) {is=title K2Node_AddDelegate}
   > n159 = Cast To FortReplaySpectatorAthena!(execute:Exec, Object=GetOwningPlayer()) -> (then:Exec, CastFailed:Exec, AsFort Replay Spectator Athena)
   > n160 = Bind Event to On Toggle Play Pause(self:Fort Replay Spectator Athena Object Reference=@n159, Delegate=Create Event()) {is=title K2Node_AddDelegate}
   UpdateDevBlurButtonVisibility(execute=[@n160.then, @n159.CastFailed])
   > UpdateDevSlomoEnablingButtonVisibility()
   > UpdateDevFPSBlurButtonVisibility()
   > UpdateDevSlomoMultiplierButtonVisibility()
   > UpdateBlurAmount!(execute, self:Athena Spectator HUDControls Base Object Reference, IsReset:bool=false) -> (then)
   > UpdateRackFocusButtonsVisibility()
   > Set bPanelOpen(bPanelOpen=true)
   > Bind Event to On Collapse Panel(self:Fort Replay Context Object Reference=Get FortReplayContext(), Delegate=@n144) {is=title K2Node_AddDelegate}
   > SetHUDVisibilityState!(execute, self:Fort Replay Context Object Reference=Get FortReplayContext(), Visibility:EHudVisibilityState Enum=ReplayOnly) -> (then)
   > SetNavigationRules()
   > Bind Event to On Toggle Replay View Settings(self:Fort Replay Context Object Reference=Get FortReplayContext(), Delegate=@n137) {is=title K2Node_AddDelegate}
   > SetPadding(self=Get ControlsWrapper(), InPadding_Left=0, InPadding_Top=5, InPadding_Right=0, InPadding_Bottom=0)
   > n161 = Branch(Condition=Get bIsPlayingReplay())
   > Set LastHoveredWidget(LastHoveredWidget=Get ViewSettingsButton())
   > n162 = Sequence()
   > n163 = Construct Object from Class(Class=/Script/FortniteUI.UIKitContextMenuActionContext, self=Get CameraTypeContextMenu()) -> (ReturnValue:UIKit Context Menu Action Context Object Reference)
   > n164 = Cast To FortPlayerControllerSpectating(Object=GetOwningPlayer())
   > n165 = For Each Loop(Array:Array of ESpectatorCameraType Enums=Get CameraModeCycle!(self:Fort Player Controller Spectating Object Reference=@n164) -> (CameraModeCycle:Array of ESpectatorCameraType Enums=ThirdPerson)) -> (Array Element:ESpectatorCameraType Enum)
   > Add(TargetArray:Array of ESpectatorCameraType Enums=Get CamType() -> (CamType=ThirdPerson), NewItem:"ESpectatorCameraType Enum (by ref)"=@n165) {is=title K2Node_CallArrayFunction}
   > AddInteraction!(
    execute
    self:UIKit Context Menu Action Context Object Reference=@n163
    InActionInfo_DisplayName:Text=GetCameraNameFromType!(self:Fort Replay Context Object Reference=Get FortReplayContext(), CamType:ESpectatorCameraType Enum=@n165) -> (ReturnValue:Text)
    InActionInfo_IconTexture:Texture 2D Soft Object Reference
    InActionInfo_IconMaterial:Material Instance Soft Object Reference
    InActionInfo_SortPriority:int=1000
    InActionInfo_bIsAvailable:bool=True
    InActionInfo_bIsEnabled:bool=True
    InActionInfo_bIsSelected:bool=False
    InActionInfo_bAddDivider:bool=False
    InActionInfo_bPositionDividerAfter:bool=False
    InActionInfo_bCloseMenuOnSelect:bool=False
    InActionInfo_InteractionNameID:Name=String To Name(InString="To String (Integer)"(InInt=@n165.Array Index))
    InActionInfo_Interaction:Data Table Row Handle Structure=(DataTable=None,RowName="")
    InActionInfo_PressedDelegate:Delegate="(null).None"
    InActionInfo_ReleasedDelegate:Delegate="(null).None"
    InInteractionDelegate:"Delegate (by ref)"=@n133
   ) -> (then)
   > SetContext!(
    execute
    self:WBP UIKit Context Menu Anchor Button Base Object Reference=Get CameraTypeContextMenu()
    InteractionContext_Data:UIKit Context Menu Action Context Object Reference=@n163
   ) -> (then)
   n166 = GetRuntimeOptions()
   Branch(execute=@n165.Completed, Condition=NOT Boolean(A=IsShippingBuild(self=@n166)))
   > n167 = Construct Object from Class(Class=&UIKitContextMenuActionContext, self=Get PlayerSelectorContextMenu())
   > n168 = Cast To FortGameState!(execute, Object=GetGameState!() -> (ReturnValue:Game State Base Object Reference)) -> (then, CastFailed, AsFort Game State:Fort Game State Object Reference) {is=title K2Node_DynamicCast}
   > n169 = For Each Loop(Array:Array of Player State Object References=Get PlayerArray!(self:Game State Base Object Reference=@n168) -> (PlayerArray:Array of Player State Object References)) -> (Array Element:Player State Object Reference)
   > Branch(Condition=NOT Boolean(A=Get bIsABot!(self:Player State Object Reference=@n169) -> (bIsABot:bool)))
   n170 = GetPlayerName!(self:Player State Object Reference=@n169) -> (ReturnValue:String)
   > Branch(Condition=NOT Boolean(A=IsEmpty(InString=@n170)))
   > Add!(execute, TargetMap:Map of Integers to Strings=Get PlayerNames(), Key:"Integer (by ref)"=@n169.Array Index, Value:"String (by ref)"=@n170) -> (then) {is=title K2Node_CallFunction}
   > AddInteraction(
    self=@n167
    InActionInfo_DisplayName="To Text (String)"(InString=@n170)
    InActionInfo_SortPriority=1000
    InActionInfo_bIsAvailable=True
    InActionInfo_bIsEnabled=True
    InActionInfo_bIsSelected=False
    InActionInfo_bAddDivider=False
    InActionInfo_bPositionDividerAfter=False
    InActionInfo_bCloseMenuOnSelect=False
    InActionInfo_InteractionNameID=String To Name(InString="To String (Integer)"(InInt=@n169.Array Index))
    InActionInfo_Interaction=(DataTable=None,RowName="")
    InActionInfo_PressedDelegate="(null).None"
    InActionInfo_ReleasedDelegate="(null).None"
    InInteractionDelegate=@n107
   )
   > SetContext(self=Get PlayerSelectorContextMenu(), InteractionContext_Data=@n167)
   n171 = Construct Object from Class(execute=@n162.then_1, Class=&UIKitContextMenuActionContext, self=Get ShotNumberContextMenu())
   > n172 = For Loop(LastIndex=9)
   n173 = "To String (Integer)"(InInt=@n172)
   > AddInteraction(
    self=@n171
    InActionInfo_DisplayName="To Text (String)"(InString=Append(A="Shot ", B=@n173))
    InActionInfo_SortPriority=1000
    InActionInfo_bIsAvailable=True
    InActionInfo_bIsEnabled=True
    InActionInfo_bIsSelected=False
    InActionInfo_bAddDivider=False
    InActionInfo_bPositionDividerAfter=False
    InActionInfo_bCloseMenuOnSelect=False
    InActionInfo_InteractionNameID=String To Name(InString="To String (Integer)"(InInt=@n172))
    InActionInfo_Interaction=(DataTable=None,RowName="")
    InActionInfo_PressedDelegate="(null).None"
    InActionInfo_ReleasedDelegate="(null).None"
    InInteractionDelegate=@n110
   )
   > SetContext(self=Get ShotNumberContextMenu(), InteractionContext_Data=@n171)
   Set LastHoveredWidget(execute=@n161.else, LastHoveredWidget=Get ViewSettingsButton_Live())
   > n174 = Construct Object from Class(Class=&UIKitContextMenuActionContext, self=Get CameraTypeSpectatorContextMenu())
   > n175 = Cast To FortPlayerControllerSpectating(Object=GetOwningPlayer())
   > n176 = For Each Loop(Array:Array of ESpectatorCameraType Enums=Get CameraModeCycle(self=@n175) -> (CameraModeCycle=ThirdPerson)) -> (Array Element:ESpectatorCameraType Enum)
   n177 = GetRuntimeOptions()
   Branch(execute=@n176.Completed, Condition=NOT Boolean(A=IsShippingBuild(self=@n177)))
   > n178 = Construct Object from Class(Class=&UIKitContextMenuActionContext, self=Get PlayerSelectorSpectatorContextMenu())
   > n179 = Cast To FortGameState(Object=GetGameState())
   > n180 = For Each Loop(Array:Array of Player State Object References=Get PlayerArray(self=@n179)) -> (Array Element:Player State Object Reference)
   > Branch(Condition=NOT Boolean(A=Get bIsABot(self=@n180)))
   n181 = GetPlayerName(self=@n180)
   > Branch(Condition=NOT Boolean(A=IsEmpty(InString=@n181)))
   > Add(TargetMap=Get PlayerNames(), Key=@n180.Array Index, Value=@n181)
   > AddInteraction(
    self=@n178
    InActionInfo_DisplayName="To Text (String)"(InString=@n181)
    InActionInfo_SortPriority=1000
    InActionInfo_bIsAvailable=True
    InActionInfo_bIsEnabled=True
    InActionInfo_bIsSelected=False
    InActionInfo_bAddDivider=False
    InActionInfo_bPositionDividerAfter=False
    InActionInfo_bCloseMenuOnSelect=False
    InActionInfo_InteractionNameID=String To Name(InString="To String (Integer)"(InInt=@n180.Array Index))
    InActionInfo_Interaction=(DataTable=None,RowName="")
    InActionInfo_PressedDelegate="(null).None"
    InActionInfo_ReleasedDelegate="(null).None"
    InInteractionDelegate=@n141
   )
   > SetContext(self=Get PlayerSelectorSpectatorContextMenu(), InteractionContext_Data=@n178)
   Add(execute=@n176.LoopBody, TargetArray=Get CamType() -> (CamType=ThirdPerson), NewItem=@n176) {is=title K2Node_CallArrayFunction}
   > AddInteraction(
    self=@n174
    InActionInfo_DisplayName=GetCameraNameFromType(self=Get FortReplayContext(), CamType=@n176)
    InActionInfo_SortPriority=1000
    InActionInfo_bIsAvailable=True
    InActionInfo_bIsEnabled=True
    InActionInfo_bIsSelected=False
    InActionInfo_bAddDivider=False
    InActionInfo_bPositionDividerAfter=False
    InActionInfo_bCloseMenuOnSelect=False
    InActionInfo_InteractionNameID=String To Name(InString="To String (Integer)"(InInt=@n176.Array Index))
    InActionInfo_Interaction=(DataTable=None,RowName="")
    InActionInfo_PressedDelegate="(null).None"
    InActionInfo_ReleasedDelegate="(null).None"
    InInteractionDelegate=@n130
   )
   > SetContext(self=Get CameraTypeSpectatorContextMenu(), InteractionContext_Data=@n174)
   HandleEnterSpectatorMotionBlur(execute=@n157.else)
   > n182 = Branch(Condition="Equal (Float)"(A=GetGlobalTimeDilation!() -> (ReturnValue:float), B=0.1))
   > "SetText (Text)"(self=Get TextTimeDilation_Live(), InText="LOCTEXT(\"0.10x\")")
   > n183 = Set SpectatorSpeeds(SpectatorSpeeds=0) -> (Output_Get=0)
   n184 = Branch(execute=@n182.else, Condition="Equal (Float)"(A=GetGlobalTimeDilation(), B=0.25))
   > "SetText (Text)"(self=Get TextTimeDilation_Live(), InText="LOCTEXT(\"0.25x\")")
   > n185 = Set SpectatorSpeeds(SpectatorSpeeds=1) -> (Output_Get=0)
   n186 = Branch(execute=@n184.else, Condition="Equal (Float)"(A=GetGlobalTimeDilation(), B=0.5))
   > "SetText (Text)"(self=Get TextTimeDilation_Live(), InText="LOCTEXT(\"0.50x\")")
   > n187 = Set SpectatorSpeeds(SpectatorSpeeds=2) -> (Output_Get=0)
   n188 = Branch(execute=@n186.else, Condition="Equal (Float)"(A=GetGlobalTimeDilation(), B=1))
   > "SetText (Text)"(self=Get TextTimeDilation_Live(), InText="LOCTEXT(\"1.00x\")")
   > n189 = Set SpectatorSpeeds(SpectatorSpeeds=3) -> (Output_Get=0)
   n190 = Event OnDutchControlSchemaEnabledChanged!() -> (OutputDelegate, then, bDutchControlSchemaEnabled:bool) {is=title K2Node_Event}
   > Set IsDutchControlSchemaEnabled(IsDutchControlSchemaEnabled:bool=@n190.bDutchControlSchemaEnabled) -> (Output_Get:bool)
   > UpdateZoomKeybindText()
   > UpdateSpeedKeybindText()
   n191 = Event OnAutoFocusEnabledChanged!() -> (OutputDelegate, then, bAutoFocusEnabled:bool) {is=title K2Node_Event}
   > Set IsAutoFocusEnabled(IsAutoFocusEnabled:bool=@n191.bAutoFocusEnabled) -> (Output_Get:bool)
   > UpdateRackFocusButtonsVisibility()
   Event OnBlurStateChanged!() -> (OutputDelegate, then, NewBlurState:EBlurState Enum=Disabled) {is=title K2Node_Event}
   > UpdateDevBlurButtonVisibility()
   n192 = Event OnSnapZoomEnabledChanged!() -> (OutputDelegate, then, SnapZoomEnabled:bool) {is=title K2Node_Event}
   > Set IsSnapZoomEnabled(IsSnapZoomEnabled:bool=@n192.SnapZoomEnabled) -> (Output_Get:bool)
   > UpdateZoomKeybindText()
   > UpdateSpeedKeybindText()
   Event OnInitialized()
   > n193 = Branch(Condition=IsShippingBuild(self=GetRuntimeOptions()))
   SetVisibility(
    execute=@n193.else
    self=[
     Get SkipOneFrame() -> (SkipOneFrame:Icon Text Button Replay Object Reference)
     Get FreeDroneTargetHidingToolUndo() -> (FreeDroneTargetHidingToolUndo:Icon Text Button Replay Object Reference)
     Get FreeDroneTargetHidingToolSelect() -> (FreeDroneTargetHidingToolSelect:Icon Text Button Replay Object Reference)
     Get FreeDroneTargetHidingToolUndo_Live() -> (FreeDroneTargetHidingToolUndo_Live:Icon Text Button Replay Object Reference)
     Get FreeDroneTargetHidingToolSelect_Live() -> (FreeDroneTargetHidingToolSelect_Live:Icon Text Button Replay Object Reference)
    ]
    InVisibility=Visible
   )
   > SetVisibility(
    self=[
     Get PlayerSelectorSpectatorContextBox()
     Get ToolsHorizontalBox() -> (ToolsHorizontalBox:Horizontal Box Object Reference)
     Get PlayerSelectorContextBox() -> (PlayerSelectorContextBox:Vertical Box Object Reference)
    ]
    InVisibility=SelfHitTestInvisible
   )
   > UpdateDevBlurButtonVisibility()
   > UpdateDevFPSBlurButtonVisibility()
   > SetRenderOpacity(self=[Get PlayerSelectorSpectatorInputAction(), Get PlayerSelectorInputAction()], InOpacity=0)
   > n194 = Is Valid(InputObject=Get HidingToolComponent())
   n195 = Cast To BP_ReplayPC_Athena(execute=[@n222.then, @n194.Is Valid], Object=GetOwningPlayer())
   > Bind Event to On Timeline Scrubbed(self:Fort Player Controller Spectating Object Reference=@n195, Delegate=@n102) {is=title K2Node_AddDelegate}
   > Bind Event to On Timeline Scrubbed(self=@n195, Delegate=@n150)
   n196 = Event OnCameraShotPresetChanged!() -> (OutputDelegate, then, Number:int=0) {is=title K2Node_Event}
   > SetVisibility(
    self=Get MobileCameraShotsControlsHorizontalBox() -> (MobileCameraShotsControlsHorizontalBox:Horizontal Box Object Reference)
    InVisibility=Select(Option 0:ESlateVisibility Enum=Collapsed, Option 1:ESlateVisibility Enum=Visible, Index="integer > integer"(A=@n196.Number, B=-1)) -> (ReturnValue:ESlateVisibility Enum)
   )
   n197 = Event BP_OnSlomoChanged!() -> (OutputDelegate, then, NewTimeDilation:float) {is=title K2Node_Event}
   n198 = IsShippingBuild(self=GetRuntimeOptions())
   n199 = GetIsPlayingReplay()
   > Branch(Condition=AND Boolean(A=NOT Boolean(A=@n198), B=NOT Boolean(A=@n199)))
   > ChangeBlurState()
   > UpdateDevBlurButtonVisibility()
   > n200 = Branch(Condition=OR Boolean(
     A="Nearly Equal (Float)"(A=@n197.NewTimeDilation, B=1, ErrorTolerance=0.01)
     B="Nearly Equal (Float)"(A=@n197.NewTimeDilation, B=0.5, ErrorTolerance=0.01)
    ))
   > SetVisibility(self=Get Alert() -> (Alert:Image Object Reference), InVisibility=Collapsed)
   > Set In Tool Tip Text(In Tool Tip Text:Text) -> (Output_Get:Text)
   > SetToolTipText(self=Get Alert(), InToolTipText=Get In Tool Tip Text() -> (In Tool Tip Text:Text))
   SetVisibility(execute=@n200.else, self=Get Alert(), InVisibility=Visible)
   > Set In Tool Tip Text(In Tool Tip Text="LOCTEXT(\"This warning is regarding your current game speed (or slow-mo) value. For usage here in Spectator, values lower than 0.5 are safe. However, be aware anything beneath 0.5 is not recommended outside of Spectator for capturing gameplay.\")")
   > SetToolTipText(self=Get Alert(), InToolTipText=Get In Tool Tip Text())
   Event Destruct()
   > Unbind Event from On Keybinds Changed(self:Fort Global UIContext Object Reference=Get FortGlobalUIContext(), Delegate=Create Event()) {is=title K2Node_RemoveDelegate}
   > SetLightingToolVisibility(self=Get FortReplayContext(), IsVisible=false)
   n201 = Get HidingToolComponent()
   > n202 = Is Valid(InputObject=@n201)
   n203 = Branch(execute=[@n205.then, @n202.Is Not Valid], Condition=Get bIsPlayingReplay())
   > UnbindReplayContext()
   > n204 = Cast To FortReplaySpectatorAthena(Object=GetOwningPlayer())
   > Unbind all Events from On Toggle Play Pause!(execute:Exec, self:Fort Replay Spectator Athena Object Reference=@n204) -> (then:Exec) {is=title K2Node_ClearDelegate}
   HandleExitSpectatorMotionBlur(execute=@n203.else)
   n205 = UnHideAllActors(execute=@n202.Is Valid, self:Fort Replay Hiding Tool Component Object Reference=@n201)
   Event CenterControls() {is=title K2Node_Event}
   > CenterControl()
   n206 = Event OnFollowedPlayerChanged!() -> (OutputDelegate, then, SpectatorPC:Fort Player Controller Spectating Object Reference, NewFollowedPlayer:Fort Player State Object Reference) {is=title K2Node_Event}
   > n207 = Is Valid(InputObject=@n206.NewFollowedPlayer)
   "SetText (Text)"(
    execute=@n207.Is Not Valid
    self=[
     Get FollowedPlayerText_Live() -> (FollowedPlayerText_Live:Common Text Object Reference)
     Get FollowedPlayerText() -> (FollowedPlayerText:Common Text Object Reference)
    ]
   )
   n208 = "To Text (String)"(InString=GetPlayerName(self=@n206.NewFollowedPlayer))
   "SetText (Text)"(execute=@n207.Is Valid, self=Get FollowedPlayerText(), InText=@n208)
   > "SetText (Text)"(self=Get FollowedPlayerText_Live(), InText=@n208)
   n209 = Event OnCameraTypeChanged!() -> (OutputDelegate, then, SpectatorPC:Fort Player Controller Spectating Object Reference, NewCameraType:ESpectatorCameraType Enum=ThirdPerson) {is=title K2Node_Event}
   > SetText(
    self=[
     Get CurrentCameraButton() -> (CurrentCameraButton:Icon Text Button Replay Object Reference)
     Get CurrentCameraButtonSpectator() -> (CurrentCameraButtonSpectator:Icon Text Button Replay Object Reference)
    ]
    Text=GetCameraNameFromType(self=Get FortReplayContext(), CamType=@n209.NewCameraType)
   )
   > UpdateBattleMapTimelineButton()
   > UpdateCameraControls()
   > n210 = Switch on ESpectatorCameraType!(execute, Selection:ESpectatorCameraType Enum=@n209.NewCameraType) -> (
    ThirdPerson:Exec
    DroneFree:Exec
    Gameplay:Exec
    ReverseShot:Exec
    FollowShot:Exec
    DroneFollow:Exec
    DroneAttach:Exec
    BattleMap:Exec
    ARDrone:Exec
    DroneFreeDynamic:Exec
    Sequencer:Exec
   ) {is=title K2Node_SwitchEnum}
   SetWidthOverride(
    execute=@n210.DroneFreeDynamic
    self=[
     Get CamNameSizeOverride() -> (CamNameSizeOverride:Size Box Object Reference)
     Get CamNameSizeOverride_1() -> (CamNameSizeOverride_1:Size Box Object Reference)
    ]
    InWidthOverride=190
   )
   n211 = SetVisibility(
    execute=[@n210.ARDrone, @n210.BattleMap, @n210.DroneAttach, @n210.FollowShot, @n210.ReverseShot, @n210.DroneFree, @n210.DroneFollow]
    self=Get FreeDroneTeleportToPawn() -> (FreeDroneTeleportToPawn:Icon Text Button Replay Object Reference)
    InVisibility=SelfHitTestInvisible
   )
   SetWidthOverride(execute=[@n211.then, @n212.then], self=[Get CamNameSizeOverride(), Get CamNameSizeOverride_1()], InWidthOverride=160)
   n212 = SetVisibility(execute=[@n210.Gameplay, @n210.ThirdPerson], self=Get FreeDroneTeleportToPawn(), InVisibility=Collapsed)
   Event ResetFocusToPlayPauseIfNeeded() {is=title K2Node_Event}
   n213 = GetCurrentInputType(self=Get CommonInputSubsystem!() -> (ReturnValue:Common Input Object Reference) {is=title K2Node_GetBlueprintContextEx})
   > Branch(Condition=AND Boolean(A="Equal (Enum)"(A=@n213, B=MouseAndKeyboard), B=Get bIsPlayingReplay()))
   > SetFocus(self=Get PlayPause())
   Event HandleBattleMapTimelineModeChanged!() -> (OutputDelegate, then, bIsOnTimelineMode:bool) {is=title K2Node_Event}
   > UpdateBattleMapTimelineButton()
   n214 = Event OnRackFocusEnabledChanged!() -> (OutputDelegate, then, bRackFocusEnabled:bool) {is=title K2Node_Event}
   > Set IsRackFocusEnabled(IsRackFocusEnabled:bool=@n214.bRackFocusEnabled) -> (Output_Get:bool)
   > UpdateZoomKeybindText()
   > UpdateRackFocusButtonsVisibility()
   > UpdateSpeedKeybindText()
   Event OnFreeDroneFocusModeChanged!() -> (OutputDelegate, then, InFreeDroneFocusMode:EFreeDroneFocusMode Enum=Manual) {is=title K2Node_Event}
   > UpdateFreeDroneTargetFocusButtonVisibility()
   Branch(Condition=Get bViewSettingsOpen())
   > Branch(Condition=IsInputType(InputType=MouseAndKeyboard, IsNot=false))
   > ToggleViewSettingsVisibility(self=Get FortReplayContext())
   n215 = Gamepad D-pad Down!() -> (Pressed:Exec, Released:Exec, Key:Key Structure) {is=title K2Node_InputKey, inputKey=Gamepad_DPad_Down}
   n216 = Gamepad D-pad Up!() -> (Pressed:Exec, Released:Exec, Key:Key Structure) {is=title K2Node_InputKey, inputKey=Gamepad_DPad_Up}
   Caps Lock!() -> (Pressed:Exec, Released:Exec, Key:Key Structure) {is=title K2Node_InputKey, inputKey=CapsLock}
   n217 = Get QuickAccessMenu() -> (QuickAccessMenu:Fort Replay Settings Quickview Object Reference)
   > n218 = Branch(Condition="Equal (Enum)"(
     A:ESlateVisibility Enum=GetVisibility!(self:Widget Object Reference=@n217) -> (ReturnValue:ESlateVisibility Enum=Visible)
     B:ESlateVisibility Enum=SelfHitTestInvisible
    ))
   > ToggleQuickAccessSettings!(execute, self, bVisible:bool=false) -> (then)
   ToggleQuickAccessSettings(execute=@n218.else, bVisible=true)
   n219 = Tab!() -> (Pressed:Exec, Released:Exec, Key:Key Structure) {is=title K2Node_InputKey, inputKey=Tab}
   > ToggleQuickAccessSettings(bVisible=true)
   ToggleQuickAccessSettings(execute=@n219.Released, bVisible=false)
   Bind Event to On Hotkey Assigned(execute=@n221.then, self:Fort Player Controller Spectating Object Reference=@n120, Delegate=@n122) {is=title K2Node_AddDelegate}
   n220 = GetGlobalTimeDilation()
   "SetText (Text)"(
    self=Get TextTimeDilation_Live()
    InText=Format Text!(
     Format="LOCTEXT(\"{GameSpeed}{Unit}\")"
     GameSpeed:Text="To Text (Float)"(Value=@n220, bUseGrouping=false, MinimumIntegralDigits=1, MaximumIntegralDigits=2, MinimumFractionalDigits=1, MaximumFractionalDigits=2)
     Unit:Text=Get FormatText() -> (FormatText:Text)
    ) -> (Result)
   )
   "On Context Menu Closed (CameraTypeContextMenu)"() {is=title K2Node_ComponentBoundEvent}
   "On Context Menu Closed (ShotNumberContextMenu)"() {is=title K2Node_ComponentBoundEvent}
   "On Context Menu Closed (PlayerSelectorContextMenu)"() {is=title K2Node_ComponentBoundEvent}
   "On Context Menu Closed (CameraTypeSpectatorContextMenu)"() {is=title K2Node_ComponentBoundEvent}
   "On Context Menu Closed (PlayerSelectorSpectatorContextMenu)"() {is=title K2Node_ComponentBoundEvent}
   n221 = Bind Event to On Hotkey Invoked(execute=@n120.then, self:Fort Player Controller Spectating Object Reference=@n120, Delegate=@n117) {is=title K2Node_AddDelegate}
   n222 = Set HidingToolComponent(execute=@n151.then, HidingToolComponent:Fort Replay Hiding Tool Component Object Reference=@n151) -> (Output_Get:Fort Replay Hiding Tool Component Object Reference)
  ]}
 GetNextFreeShot {nodes [
   GetNextFreeShot() {is=title K2Node_FunctionEntry}
   > n1 = Cast To FortReplaySpectator(Object=GetOwningPlayer())
   > n2 = For Loop(LastIndex=9)
   n3 = GetCameraShotsPreset(self=@n1) -> (ReturnValue=0)
   SetCameraShotsPreset!(
    execute=@n2.Completed
    self:Fort Player Controller Spectating Object Reference=@n1
    Number:int="% (Integer)"(A=int + int(A=@n3), B=10) -> (ReturnValue=0)
   ) -> (then)
   > Return Node!(execute, ShotNumber:int=0)
   n4 = IsHotkeySet!(execute=@n2.LoopBody, self:Fort Player Controller Spectating Object Reference=@n1, HotKeyNum:int=@n2) -> (then, ReturnValue:bool)
   > n5 = Branch(Condition=@n4)
   Return Node(execute=@n5.else, ShotNumber=@n2)
  ]}
 HandleEnterSpectatorMotionBlur {nodes [
   HandleEnterSpectatorMotionBlur() {is=title K2Node_FunctionEntry}
   n1 = IsShippingBuild(self=GetRuntimeOptions())
   n2 = GetIsPlayingReplay()
   > Branch(Condition=AND Boolean(A=NOT Boolean(A=@n1), B=NOT Boolean(A=@n2)))
   > n3 = Branch(Condition="Nearly Equal (Float)"(A=GetGlobalTimeDilation(), B=0.5, ErrorTolerance=0.01))
   > n4 = GetFPSBlurMultiplier()
   > n5 = Branch(Condition="Nearly Equal (Float)"(A=@n4, B=1, ErrorTolerance=0.01))
   > n6 = ExecuteConsoleCommand(Command=r.MotionBlur.Amount 0.5, SpecificPlayer=GetOwningPlayer())
   UpdateDevBlurButtonVisibility(execute=[@n6.then, @n10.then, @n13.then, @n14.then])
   n7 = GetGlobalTimeDilation()
   > n8 = Branch(Condition=OR Boolean(A="Nearly Equal (Float)"(A=@n7, B=1, ErrorTolerance=0.01), B="Nearly Equal (Float)"(A=@n7, B=0.5, ErrorTolerance=0.01)))
   > SetVisibility(self=Get Alert(), InVisibility=Collapsed)
   > Set In Tool Tip Text()
   > SetToolTipText(self=Get Alert(), InToolTipText=Get In Tool Tip Text())
   SetVisibility(execute=@n8.else, self=Get Alert(), InVisibility=Visible)
   n9 = GetGlobalTimeDilation()
   > "SetText (Text)"(
    self=Get TextTimeDilation_Live()
    InText=Format Text(
     Format="LOCTEXT(\"{GameSpeed}{Unit}\")"
     GameSpeed="To Text (Float)"(Value=@n9, bUseGrouping=false, MinimumIntegralDigits=1, MaximumIntegralDigits=2, MinimumFractionalDigits=1, MaximumFractionalDigits=2)
     Unit=Get FormatText()
    )
   )
   > Set In Tool Tip Text(In Tool Tip Text="LOCTEXT(\"This warning is regarding your current game speed (or slow-mo) value. For usage here in Spectator, values lower than 0.5 are safe. However, be aware anything beneath 0.5 is not recommended outside of Spectator for capturing gameplay.\")")
   > SetToolTipText(self=Get Alert(), InToolTipText=Get In Tool Tip Text())
   n10 = ExecuteConsoleCommand(execute=@n5.else, Command=r.MotionBlur.Amount 1.0, SpecificPlayer=GetOwningPlayer())
   n11 = GetFPSBlurMultiplier(execute=@n3.else)
   > n12 = Branch(Condition="Nearly Equal (Float)"(A=@n11, B=1, ErrorTolerance=0.01))
   > n13 = ExecuteConsoleCommand(Command=r.MotionBlur.Amount 0.25, SpecificPlayer=GetOwningPlayer())
   n14 = ExecuteConsoleCommand(execute=@n12.else, Command=r.MotionBlur.Amount 0.5, SpecificPlayer=GetOwningPlayer())
  ]}
 HandleExitSpectatorMotionBlur {nodes [
   HandleExitSpectatorMotionBlur() {is=title K2Node_FunctionEntry}
   n1 = IsShippingBuild(self=GetRuntimeOptions())
   n2 = GetIsPlayingReplay()
   > Branch(Condition=AND Boolean(A=NOT Boolean(A=@n1), B=NOT Boolean(A=@n2)))
   > n3 = Branch(Condition="Nearly Equal (Float)"(A=GetGlobalTimeDilation(), B=0.5, ErrorTolerance=0.01))
   > n4 = GetFPSBlurMultiplier()
   > n5 = Branch(Condition="Nearly Equal (Float)"(A=@n4, B=1, ErrorTolerance=0.01))
   > ExecuteConsoleCommand(Command=r.MotionBlur.Amount 0.1, SpecificPlayer=GetOwningPlayer())
   ExecuteConsoleCommand(execute=@n5.else, Command=r.MotionBlur.Amount 0.2, SpecificPlayer=GetOwningPlayer())
   n6 = GetFPSBlurMultiplier(execute=@n3.else)
   > n7 = Branch(Condition="Nearly Equal (Float)"(A=@n6, B=1, ErrorTolerance=0.01))
   > ExecuteConsoleCommand(Command=r.MotionBlur.Amount -1, SpecificPlayer=GetOwningPlayer())
   ExecuteConsoleCommand(execute=@n7.else, Command=r.MotionBlur.Amount -1, SpecificPlayer=GetOwningPlayer())
   SetFPSBlurMultiplier(InMultiplier=0)
   SetFPSBlurMultiplier(InMultiplier=0)
   SetFPSBlurMultiplier(InMultiplier=0.2)
   SetFPSBlurMultiplier(InMultiplier=0.4)
  ]}
 HandleHUDVisibilityChanged {nodes [
   n1 = HandleHUDVisibilityChanged!() -> (then, IsVisibile:EHudVisibilityState Enum=FullyVisible) {is=title K2Node_FunctionEntry}
   > n2 = Branch(Condition=Get bForceReplayMode() -> (bForceReplayMode:bool))
   > n3 = Branch(Condition="Equal (Enum)"(A:EHudVisibilityState Enum=@n1, B:EHudVisibilityState Enum=FullyVisible))
   > SetHUDVisibilityState(self=Get FortReplayContext(), Visibility=ReplayOnly)
   > Set bSidebarOpen(bSidebarOpen=false)
   > n4 = Set bForceReplayMode(bForceReplayMode:bool=false) -> (Output_Get:bool)
   n5 = Switch on EHudVisibilityState(execute=[@n10.then, @n9.else, @n4.then, @n3.else], Selection=@n1)
   ShowRelevantControls(execute=@n5.FullyHidden, Config=0)
   > Set bViewSettingsOpen(bViewSettingsOpen=false)
   > n6 = Branch(Condition=Get bIsPlayingReplay())
   > n7 = Set LastHoveredWidget(LastHoveredWidget=Get ViewSettingsButton())
   > SetFocus(self=@n7)
   n8 = Set LastHoveredWidget(execute=@n6.else, LastHoveredWidget=Get ViewSettingsButton_Live())
   > SetFocus(self=@n8)
   SetInputMode(execute=[@n5.FullyVisible, @n5.GameOnly, @n5.ReplayOnly], InputType=1)
   n9 = Branch(execute=@n2.else, Condition=Get bSidebarOpen() -> (bSidebarOpen:bool))
   > n10 = Set bForceReplayMode(bForceReplayMode=true)
  ]}
 IsOnBattleMapTimeline {nodes [
   IsOnBattleMapTimeline() {is=title K2Node_FunctionEntry}
   > n1 = Cast To FortPlayerControllerSpectating(Object=GetOwningPlayer())
   > n2 = Cast To BP_BattleMapPawnReplay(Object=GetSpectatorPawn(self=@n1))
   > Return Node!(execute, IsOnBattleMapTimeline:bool=BP_IsOnTimelineMode(self=@n2))
   Return Node(execute=@n2.CastFailed, IsOnBattleMapTimeline=false)
  ]}
 SetInputMode {nodes [
   n1 = SetInputMode!() -> (then, InputType:int) {is=title K2Node_FunctionEntry}
   > Set Loc_Controller(Loc_Controller:Player Controller Object Reference=GetPlayerController(PlayerIndex=0)) -> (Output_Get:Player Controller Object Reference)
   > n2 = Switch on Int!(execute, Selection=@n1) -> (Default, 0, 1, 2)
   Set Input Mode Game And UI!(
    execute=@n2.1
    PlayerController:Player Controller Object Reference=Get Loc_Controller() -> (Loc_Controller:Player Controller Object Reference)
    InWidgetToFocus:Widget Object Reference
    InMouseLockMode:EMouseLockMode Enum=DoNotLock
    bHideCursorDuringCapture:bool=true
    bFlushInput:bool=true
   ) -> (then) {is=title K2Node_CallFunction}
   > ResetIgnoreLookInput(self:Controller Object Reference=Get Loc_Controller())
   > ResetIgnoreMoveInput(self:Controller Object Reference=Get Loc_Controller())
   > SetIsEnabled(self=[Get ReplaySettingsWrapper(), Get SpectatorSettingsSize()], bInIsEnabled=true)
   Set Input Mode UI Only!(
    execute=@n2.0
    PlayerController:Player Controller Object Reference=Get Loc_Controller()
    InWidgetToFocus:Widget Object Reference
    InMouseLockMode:EMouseLockMode Enum=DoNotLock
    bFlushInput:bool=true
   ) -> (then) {is=title K2Node_CallFunction}
   > SetIsEnabled(self=[Get ReplaySettingsWrapper(), Get SpectatorSettingsSize()], bInIsEnabled=false)
   > SetIgnoreLookInput!(execute, self:Controller Object Reference=Get Loc_Controller(), bNewLookInput:bool=true) -> (then)
   > SetIgnoreMoveInput!(execute, self:Controller Object Reference=Get Loc_Controller(), bNewMoveInput:bool=true) -> (then)
   Set Input Mode Game And UI(execute=@n2.2, PlayerController=Get Loc_Controller(), InMouseLockMode=DoNotLock, bHideCursorDuringCapture=true, bFlushInput=true)
   > ResetIgnoreLookInput(self=Get Loc_Controller())
   > ResetIgnoreMoveInput(self=Get Loc_Controller())
   > SetIsEnabled(self=[Get SpectatorSettingsSize(), Get ReplaySettingsWrapper()], bInIsEnabled=false)
  ]}
 SetNavigationRules {nodes [
   SetNavigationRules() {is=title K2Node_FunctionEntry}
   > n1 = Branch(Condition=Get bIsPlayingReplay())
   > SetNavigationRuleExplicit!(
    execute
    self:Widget Object Reference=Get Back() -> (Back:Icon Text Button Replay Object Reference)
    Direction:EUINavigation Enum=Right
    InWidget:Widget Object Reference=Get Forward() -> (Forward:Icon Text Button Replay Object Reference)
   ) -> (then)
   > SetNavigationRuleExplicit(self=Get Forward(), Direction=Left, InWidget=Get Back())
   > SetNavigationRuleExplicit(
    self=Get Button_FocusTimeline() -> (Button_FocusTimeline:Common Button Legacy Object Reference)
    Direction=Left
    InWidget=Get NextPlayerButton() -> (NextPlayerButton:Icon Text Button Replay Object Reference)
   )
   > n2 = Branch(Condition=IsShippingBuild(self=GetRuntimeOptions()))
   > SetNavigationRuleExplicit(
    self=Get PreviousPlayerButton() -> (PreviousPlayerButton:Icon Text Button Replay Object Reference)
    Direction=Right
    InWidget=Get NextPlayerButton()
   )
   > n3 = SetNavigationRuleExplicit(self=Get NextPlayerButton(), Direction=Left, InWidget=Get PreviousPlayerButton())
   SetNavigationRuleExplicit(execute=[@n3.then, @n4.then], self=Get NextPlayerButton(), Direction=Right, InWidget=Get Button_FocusTimeline())
   SetNavigationRuleExplicit(execute=@n2.else, self=Get PreviousPlayerButton(), Direction=Right, InWidget=Get PlayerSelectorContextMenu())
   > SetNavigationRuleExplicit(self=Get PlayerSelectorContextMenu(), Direction=Left, InWidget=Get PreviousPlayerButton())
   > SetNavigationRuleExplicit(self=Get Back(), Direction=Right, InWidget=Get SkipOneFrame())
   > SetNavigationRuleExplicit(self=Get SkipOneFrame(), Direction=Left, InWidget=Get Back())
   > n4 = SetNavigationRuleExplicit(self=Get Forward(), Direction=Left, InWidget=Get SkipOneFrame())
   SetNavigationRuleExplicit(
    execute=@n1.else
    self=Get ViewSettingsButton_Live()
    Direction=Left
    InWidget=Get NextPlayerButton_Live() -> (NextPlayerButton_Live:Icon Text Button Replay Object Reference)
   )
   > SetNavigationRuleExplicit(self=Get NextPlayerButton_Live(), Direction=Right, InWidget=Get ViewSettingsButton_Live())
  ]}
 ShowRelevantControls {nodes [
   n1 = ShowRelevantControls!() -> (then, Config:int) {is=title K2Node_FunctionEntry}
   > n2 = Switch on Int(Selection=@n1)
   SetRenderOpacity(
    execute=@n2.1
    self=[
     Get TimelineOverlay() -> (TimelineOverlay:Overlay Object Reference)
     Get ReplaySettingsWrapper()
     Get SpectatorSettingsSize()
     Get ControlsWrapper()
    ]
    InOpacity=0.5
   )
   > n3 = SetRenderOpacity(self=Get PlayPause(), InOpacity=1)
   n4 = UpdateZoomKeybindText(execute=[@n3.then, @n20.then])
   n5 = Get CameraTypeContextMenu()
   n6 = Get LastHoveredWidget()
   n7 = Branch(execute=[@n4.then, @n21.then], Condition=NOT Boolean(A="Equal (Object)"(A=@n5, B=@n6)))
   > n8 = SetRenderOpacity(self=Get CameraTypeInputAction(), InOpacity=0)
   n9 = Branch(execute=[@n8.then, @n7.else], Condition=IsShippingBuild(self=GetRuntimeOptions()))
   n10 = Get LastHoveredWidget()
   n11 = Get PlayerSelectorContextMenu()
   n12 = Branch(execute=@n9.else, Condition=NOT Boolean(A="Equal (Object)"(A=@n11, B=@n10)))
   > n13 = SetRenderOpacity(self=Get PlayerSelectorInputAction(), InOpacity=0)
   n14 = Get LastHoveredWidget()
   n15 = Get PlayerSelectorSpectatorContextMenu()
   n16 = Branch(execute=[@n12.else, @n13.then], Condition=NOT Boolean(A="Equal (Object)"(A=@n15, B=@n14)))
   > n17 = SetRenderOpacity(self=Get PlayerSelectorSpectatorInputAction(), InOpacity=0)
   n18 = Get LastHoveredWidget()
   n19 = Get ShotNumberContextMenu()
   Branch(execute=[@n17.then, @n16.else], Condition=NOT Boolean(A="Equal (Object)"(A=@n19, B=@n18)))
   > SetRenderOpacity(self=Get ShotNumberInputAction(), InOpacity=0)
   n20 = SetRenderOpacity(
    execute=@n2.0
    self=[Get ControlsWrapper(), Get TimelineOverlay(), Get ReplaySettingsWrapper(), Get SpectatorSettingsSize(), Get PlayPause()]
    InOpacity=1
   )
   SetRenderOpacity(execute=@n2.2, self=[Get ReplaySettingsWrapper(), Get SpectatorSettingsSize(), Get TimelineOverlay()], InOpacity=0.5)
   > SetRenderOpacity(self=Get ControlsWrapper(), InOpacity=1)
   > SetRenderOpacity(self=Get PlayPause(), InOpacity=4)
   > n21 = UpdateKeybindText!(
    execute
    self:Keybinding Replay Object Reference=Get KeybindZoom() -> (KeybindZoom:Keybinding Replay Object Reference)
    NewText:"Text (by ref)"=Get Loc_SwitchTabs() -> (Loc_SwitchTabs:Text)
   ) -> (then)
  ]}
 ToggleQuickAccessSettings {nodes [
   n1 = ToggleQuickAccessSettings!() -> (then, bVisible:bool=~) {is=title K2Node_FunctionEntry}
   > Set Loc_Visible(Loc_Visible:bool=@n1) -> (Output_Get:bool)
   n2 = GetRuntimeOptions()
   > Branch(Condition=NOT Boolean(A=IsShippingBuild(self=@n2)))
   > n3 = Branch(Condition=Get Loc_Visible() -> (Loc_Visible:bool))
   > n4 = Is Valid(InputObject=Get QuickAccessMenu())
   n5 = Create Widget!(
    execute:Exec=@n4.Is Not Valid
    Class:User Widget Class Reference=/Game/UI/Replay/FortReplaySettingsQuickview.FortReplaySettingsQuickview_C
    OwningPlayer:Player Controller Object Reference
   ) -> (then:Exec, ReturnValue:Fort Replay Settings Quickview Object Reference)
   > n6 = Set QuickAccessMenu(QuickAccessMenu:Fort Replay Settings Quickview Object Reference=@n5) -> (Output_Get:Fort Replay Settings Quickview Object Reference)
   > AddToViewport!(execute, self:User Widget Object Reference=@n6, ZOrder:int=0) -> (then)
   > RefreshSettings(self:Fort Replay Settings Quickview Object Reference=Get QuickAccessMenu())
   n7 = Get QuickAccessMenu()
   n8 = Branch(
    execute=[@n10.else, @n4.Is Valid]
    Condition="Equal (Enum)"(A:ESlateVisibility Enum=GetVisibility(self=@n7) -> (ReturnValue=Visible), B:ESlateVisibility Enum=Collapsed)
   )
   > SetVisibility(self=Get QuickAccessMenu(), InVisibility=SelfHitTestInvisible)
   > RefreshSettings(self=Get QuickAccessMenu())
   n9 = Get QuickAccessMenu()
   n10 = Branch(execute=[@n3.else, @n8.else], Condition="Equal (Enum)"(A=GetVisibility(self=@n9) -> (ReturnValue=Visible), B=SelfHitTestInvisible))
   > SetVisibility(self=Get QuickAccessMenu(), InVisibility=Collapsed)
  ]}
 UnbindReplayContext {nodes [
   UnbindReplayContext() {is=title K2Node_FunctionEntry}
   > Unbind Event from On Replay Hud Visibility Changed(self:Fort Replay Context Object Reference=Get FortReplayContext(), Delegate=Create Event()) {is=title K2Node_RemoveDelegate}
  ]}
 UpdateBattleMapTimelineButton {nodes [
   UpdateBattleMapTimelineButton() {is=title K2Node_FunctionEntry}
   > n1 = Cast To FortPlayerControllerSpectating(Object=GetOwningPlayer())
   > n2 = Cast To BP_BattleMapPawnReplay(Object=GetSpectatorPawn(self=@n1))
   > n3 = Set bIsOnBattleMapTimeline(bIsOnBattleMapTimeline:bool=BP_IsOnTimelineMode(self=@n2)) -> (Output_Get:bool)
   n4 = Get BattleMapTimelineModeButton()
   SetTriggeringInputAction(
    execute=[@n3.then, @n6.then]
    self=@n4
    InputActionRow=Select(
     Option 0:"Data Table Row Handle Structure (by ref)"=Make Data Table Row Handle(RowName=None)
     Option 1:"Data Table Row Handle Structure (by ref)"=Get BattleMapTimelineTriggeringInputActions() -> (BattleMapTimelineTriggeringInputActions:Data Table Row Handle Structure)
     Index=Get bIsOnBattleMapTimeline() -> (bIsOnBattleMapTimeline:bool)
    ) -> (ReturnValue:"Data Table Row Handle Structure (by ref)")
   )
   > UpdateButtonIconsVisibility()
   > n5 = Branch(Condition="Equal (Enum)"(
     A:ESpectatorCameraType Enum=Get CurrentCameraType!(self:Fort Player Controller Spectating Object Reference=@n1) -> (CurrentCameraType:ESpectatorCameraType Enum=ThirdPerson)
     B:ESpectatorCameraType Enum=BattleMap
    ))
   Branch(execute=@n5.else, Condition=HasAnyUserFocus!(self:Widget Object Reference=@n4) -> (ReturnValue:bool))
   > CenterControl()
   n6 = Set bIsOnBattleMapTimeline(execute=@n2.CastFailed, bIsOnBattleMapTimeline=false)
  ]}
 UpdateButtonIconsVisibility {nodes [
   UpdateButtonIconsVisibility() {is=title K2Node_FunctionEntry}
   > SetMinDesiredHeight!(
    execute
    self:Size Box Object Reference=Get SpectatorSettingsSize()
    InMinDesiredHeight:float=Select(Option 0:double=0, Option 1:double=82, Index=IsInputType(InputType=Gamepad, IsNot=false)) -> (ReturnValue:double)
   ) -> (then)
  ]}
 UpdateCameraControls {nodes [
   UpdateCameraControls() {is=title K2Node_FunctionEntry}
   > n1 = Cast To FortPlayerControllerSpectating(Object=GetPlayerController(PlayerIndex=0))
   > Set CameraType(CameraType:ESpectatorCameraType Enum=Get CurrentCameraType(self=@n1) -> (CurrentCameraType=ThirdPerson)) -> (Output_Get:ESpectatorCameraType Enum=ThirdPerson)
   > SetActiveWidgetIndex(
    self=Get ControlOverlaySwitcher() -> (ControlOverlaySwitcher:Common Widget Switcher Legacy Object Reference)
    Index=Select!(
     ThirdPerson:int=1
     DroneFree:int=0
     Gameplay:int=2
     ReverseShot:int=2
     FollowShot:int=2
     DroneFollow:int=0
     DroneAttach:int=0
     BattleMap:int=0
     ARDrone:int=0
     DroneFreeDynamic:int=0
     Sequencer:int=0
     Index:ESpectatorCameraType Enum=Get CameraType() -> (CameraType:ESpectatorCameraType Enum=ThirdPerson)
    ) -> (ReturnValue:int)
   )
   n2 = Get CameraType() -> (CameraType=ThirdPerson)
   > SetVisibility(
    self=Get KeybindPanTilt() -> (KeybindPanTilt:Keybinding Replay Object Reference)
    InVisibility=Select!(Option 0:ESlateVisibility Enum=Visible, Option 1:ESlateVisibility Enum=Collapsed, Index:bool="Equal (Enum)"(A=@n2, B=DroneFollow)) -> (ReturnValue:ESlateVisibility Enum)
   )
   > SetIsEnabled(self=[Get ReplaySettingsWrapper(), Get SpectatorSettingsSize()], bInIsEnabled=true)
  ]}
 UpdateDevBlurButtonVisibility {nodes [
   UpdateDevBlurButtonVisibility() {is=title K2Node_FunctionEntry}
   n1 = IsShippingBuild(self=GetRuntimeOptions())
   > SetVisibility(
    self=[Get BlurSettingsButton() -> (BlurSettingsButton:Icon Text Button Replay Object Reference), Get BlurSettingsButton_Live()]
    InVisibility=Select(Option 0=SelfHitTestInvisible, Option 1=Collapsed, Index=@n1)
   )
   > n2 = Branch(Condition=@n1)
   Cast To FortPlayerControllerSpectating(execute=@n2.else, Object=GetPlayerController(PlayerIndex=0))
   > n3 = GetBlurState!(execute, self:Athena Spectator HUDControls Base Object Reference) -> (then, ReturnValue:EBlurState Enum=Disabled)
   > n4 = Switch on EBlurState!(execute, Selection:EBlurState Enum=@n3) -> (Disabled:Exec, PartlyEnabled:Exec, Enabled:Exec) {is=title K2Node_SwitchEnum}
   > Set ButtomImage(ButtomImage:Object Reference=/Game/UI/Foundation/Textures/Icons/Replay/T_Icon_Replay_MotionBlur_Off_C) -> (Output_Get:Object Reference)
   > n5 = Set ButtonColor(ButtonColor:Linear Color Structure=(R=1,G=0,B=0,A=1)) -> (Output_Get:Linear Color Structure)
   SetIcon!(
    execute=[@n5.then, @n7.then, @n6.then]
    self:Icon Text Button Replay Object Reference=[Get BlurSettingsButton(), Get BlurSettingsButton_Live()]
    IconBrush_TintColor_SpecifiedColor:Linear Color Structure=Get ButtonColor() -> (ButtonColor:Linear Color Structure)
    IconBrush_TintColor_ColorUseRule:ESlateColorStylingMode Enum
    IconBrush_DrawAs:ESlateBrushDrawType Enum=Image
    IconBrush_Tiling:ESlateBrushTileType Enum=NoTile
    IconBrush_Mirroring:ESlateBrushMirrorType Enum=NoMirror
    IconBrush_ImageSize_X:double=32
    IconBrush_ImageSize_Y:double=32
    IconBrush_Margin:Margin Structure=(Left=0,Top=0,Right=0,Bottom=0)
    IconBrush_ResourceObject:Object Reference=Get ButtomImage() -> (ButtomImage:Object Reference)
    IconBrush_OutlineSettings:Slate Brush Outline Settings Structure=(CornerRadii=(X=0,Y=0,Z=0,W=0),Color=(SpecifiedColor=(R=0,G=0,B=0,A=0),ColorUseRule=UseColor_Specified),Width=0,RoundingType=HalfHeightRadius,bUseBrushTransparency=False)
   ) -> (then)
   Set ButtomImage(execute=@n4.Enabled, ButtomImage=/Game/UI/Foundation/Textures/Icons/Replay/T_Icon_Replay_MotionBlur_Max)
   > n6 = Set ButtonColor(ButtonColor=(R=0,G=1,B=0,A=1))
   Set ButtomImage(execute=@n4.PartlyEnabled, ButtomImage=/Game/UI/Foundation/Textures/Icons/Replay/T_Icon_Replay_MotionBlur_Half)
   > n7 = Set ButtonColor(ButtonColor=(R=1,G=1,B=0,A=1))
  ]}
 UpdateDevFPSBlurButtonVisibility {nodes [
   UpdateDevFPSBlurButtonVisibility() {is=title K2Node_FunctionEntry}
   n1 = IsShippingBuild(self=GetRuntimeOptions())
   > SetVisibility(
    self=[
     Get FPSSettingsButton_Live() -> (FPSSettingsButton_Live:Icon Text Button Replay Object Reference)
     Get FPSSettingsButton() -> (FPSSettingsButton:Icon Text Button Replay Object Reference)
    ]
    InVisibility=Select(Option 0=SelfHitTestInvisible, Option 1=Collapsed, Index=@n1)
   )
   > n2 = Branch(Condition=@n1)
   n3 = GetFPSBlurMultiplier(execute=@n2.else)
   > n4 = Branch(Condition="Nearly Equal (Float)"(A=@n3, B=1, ErrorTolerance=0.01))
   n5 = Get FPSSettingsButton_Live()
   n6 = Get FPSSettingsButton()
   > SetText(self=[@n5, @n6], Text=Get Loc_60_Text() -> (Loc_60_Text:Text))
   SetText(execute=@n4.else, self=[@n5, @n6], Text=Get Loc_30_Text() -> (Loc_30_Text:Text))
  ]}
 UpdateDevSlomoEnablingButtonVisibility {nodes [
   n1 = NOT Boolean(A=IsShippingBuild(self=GetRuntimeOptions()))
   SetVisibility(InVisibility=Select(Option 0=Collapsed, Option 1=SelfHitTestInvisible, Index=@n1))
   > Branch(Condition=@n1)
   > n2 = BP_GetSlomoTimeDilation!(execute, self:Athena Spectator HUDControls Base Object Reference) -> (then, ReturnValue:float)
   > n3 = Branch(Condition="Nearly Equal (Float)"(A=@n2, B=0.5, ErrorTolerance=0.01))
   > n4 = Set ButtonColor(ButtonColor=(R=0,G=1,B=0,A=1))
   SetColorAndOpacity(execute=[@n5.then, @n4.then], self:User Widget Object Reference, InColorAndOpacity:Linear Color Structure=Get ButtonColor())
   n5 = Set ButtonColor(execute=@n3.else, ButtonColor=(R=1,G=0,B=0,A=1))
   UpdateDevSlomoEnablingButtonVisibility() {is=title K2Node_FunctionEntry}
  ]}
 UpdateDevSlomoMultiplierButtonVisibility {nodes [
   n1 = BP_GetSlomoTimeDilation()
   n2 = GetRuntimeOptions()
   n3 = AND Boolean(A="Nearly Equal (Float)"(A=@n1, B=0.5, ErrorTolerance=0.01), B=NOT Boolean(A=IsShippingBuild(self=@n2)))
   > SetVisibility(InVisibility=Select(Option 0=Collapsed, Option 1=SelfHitTestInvisible, Index=@n3))
   > Branch(Condition=@n3)
   > n4 = GetSlomoBlurMultiplier!(execute, self:Athena Spectator HUDControls Base Object Reference) -> (then, ReturnValue:float)
   > n5 = Branch(Condition="Nearly Equal (Float)"(A=@n4, B=1, ErrorTolerance=0.01))
   > n6 = Set ButtonColor(ButtonColor=(R=1,G=0,B=0,A=1))
   SetColorAndOpacity(execute=[@n6.then, @n7.then], InColorAndOpacity=Get ButtonColor())
   n7 = Set ButtonColor(execute=@n5.else, ButtonColor=(R=0,G=1,B=0,A=1))
   UpdateDevSlomoMultiplierButtonVisibility() {is=title K2Node_FunctionEntry}
  ]}
 UpdateFreeDroneTargetFocusButtonVisibility {nodes [
   n1 = Cast To FortPlayerControllerSpectating(Object=GetOwningPlayer())
   > n2 = GetConsoleVariableBoolValue!(execute, VariableName:String=drone.EnableFocusActorMode) -> (then, ReturnValue:bool)
   > n3 = GetFreeDroneFocusMode!(execute, self:Athena Spectator HUDControls Base Object Reference) -> (then, ReturnValue:EFreeDroneFocusMode Enum=Manual)
   > n4 = Branch(Condition=GetIsPlayingReplay())
   n5 = Get CurrentCameraType(self=@n1) -> (CurrentCameraType=ThirdPerson)
   n6 = Select(
    Option 0=Collapsed
    Option 1=SelfHitTestInvisible
    Index=AND Boolean!(
     A="Equal (Enum)"(A:EFreeDroneFocusMode Enum=@n3, B:EFreeDroneFocusMode Enum=Target)
     B=@n2
     C:bool="Equal (Enum)"(A:ESpectatorCameraType Enum=@n5, B:ESpectatorCameraType Enum=DroneFree)
    ) -> (ReturnValue)
   )
   > SetVisibility(InVisibility=@n6)
   SetVisibility(execute=@n4.else, InVisibility=@n6)
   > n7 = Branch(Condition=IsVisible())
   > SetNavigationRuleExplicit(self=Get NextPlayerButton_Live(), Direction=Right)
   > SetNavigationRuleExplicit(self=Get ViewSettingsButton_Live(), Direction=Left)
   SetNavigationRuleExplicit(execute=@n7.else, self=Get ViewSettingsButton_Live(), Direction=Left, InWidget=Get NextPlayerButton_Live())
   > SetNavigationRuleExplicit(self=Get NextPlayerButton_Live(), Direction=Right, InWidget=Get ViewSettingsButton_Live())
   UpdateFreeDroneTargetFocusButtonVisibility() {is=title K2Node_FunctionEntry}
  ]}
 UpdateRackFocusButtonsVisibility {nodes [
   n1 = GetConsoleVariableBoolValue(VariableName=spectator.EnableRackFocus)
   > n2 = Cast To FortPlayerControllerSpectating(Object=GetPlayerController(PlayerIndex=0))
   n3 = NOT Boolean(A=Get IsAutoFocusEnabled() -> (IsAutoFocusEnabled:bool))
   n4 = Select!(
    ThirdPerson:bool=false
    DroneFree:bool=true
    Gameplay:bool=false
    ReverseShot:bool=false
    FollowShot:bool=false
    DroneFollow:bool=false
    DroneAttach:bool=true
    BattleMap:bool=false
    ARDrone:bool=false
    DroneFreeDynamic:bool=false
    Sequencer:bool=false
    Index:ESpectatorCameraType Enum=Get CurrentCameraType(self=@n2) -> (CurrentCameraType=ThirdPerson)
   ) -> (ReturnValue:bool)
   n5 = Get IsRackFocusEnabled() -> (IsRackFocusEnabled:bool)
   SetVisibility(InVisibility=Select!(
     Option 0:ESlateVisibility Enum=Collapsed
     Option 1:ESlateVisibility Enum=SelfHitTestInvisible
     Index:bool=AND Boolean!(A=@n4, B=@n5, C=@n3, D:bool=@n1) -> (ReturnValue)
    ) -> (ReturnValue:ESlateVisibility Enum))
   UpdateRackFocusButtonsVisibility() {is=title K2Node_FunctionEntry}
  ]}
 UpdateSpeedKeybindText {nodes [
   UpdateSpeedKeybindText() {is=title K2Node_FunctionEntry}
   > n1 = Branch(Condition=Get IsDutchControlSchemaEnabled() -> (IsDutchControlSchemaEnabled:bool))
   > UpdateKeybindText(self=Get KeybindSpeed() -> (KeybindSpeed:Keybinding Replay Object Reference), NewText=Get DutchAngleText() -> (DutchAngleText:Text))
   n2 = Branch(execute=@n1.else, Condition=ShouldUseSnapRackControlScheme!(self:Athena Spectator HUDControls Base Object Reference) -> (ReturnValue:bool))
   > UpdateKeybindText(self=Get KeybindSpeed(), NewText=Get Loc_SnapZoom() -> (Loc_SnapZoom:Text))
   UpdateKeybindText(execute=@n2.else, self=Get KeybindSpeed(), NewText=Get Loc_Speed() -> (Loc_Speed:Text))
  ]}
 UpdateZoomKeybindText {nodes [
   UpdateZoomKeybindText() {is=title K2Node_FunctionEntry}
   > n1 = Branch(Condition=Get IsDutchControlSchemaEnabled())
   > UpdateKeybindText(self=Get KeybindZoom(), NewText=Get TiltSpeedText() -> (TiltSpeedText:Text))
   n2 = Branch(execute=@n1.else, Condition=Get IsRackFocusEnabled())
   > UpdateKeybindText(self=Get KeybindZoom(), NewText=Get Loc_RackControl() -> (Loc_RackControl:Text))
   n3 = Branch(execute=@n2.else, Condition=Get IsDutchControlSchemaEnabled())
   > UpdateKeybindText(self=Get KeybindZoom(), NewText=Get Loc_DutchSpeed() -> (Loc_DutchSpeed:Text))
   n4 = Branch(execute=@n3.else, Condition=Get IsSnapZoomEnabled() -> (IsSnapZoomEnabled:bool))
   > UpdateKeybindText(self=Get KeybindZoom(), NewText=Get Loc_SnapZoom())
   UpdateKeybindText(execute=@n4.else, self=Get KeybindZoom(), NewText=Get Loc_Zoom() -> (Loc_Zoom:Text))
  ]}
]
variables [
 BattleMapTimelineTriggeringInputActions {type=DataTableRowHandle, default {dataTable=null, rowName=None}}
 bForceReplayMode {type=boolean}
 bIsOnBattleMapTimeline {type=boolean}
 bIsPlayingReplay {type=boolean}
 bPanelOpen {type=boolean, default=true}
 bSearchingForMarkers {type=boolean}
 bSidebarOpen {type=boolean}
 bViewSettingsOpen {type=boolean}
 CameraShotsContextMenu {type=UIKitContextMenuActionContext, default=null}
 CamType {type="array<ESpectatorCameraType>", default []}
 CapsuleHalfHeight {type=number}
 CurrentShot {type=integer}
 DesignTime_IsReplay {type=boolean, default=true}
 DutchAngleText {type=string, default=Tilt}
 ElapsedTime {type=number}
 FormatText {type=string, default=x}
 HidingToolComponent {type=FortReplayHidingToolComponent, default=null}
 In Tool Tip Text {type=string, default=""}
 IsAutoFocusEnabled {type=boolean}
 IsDutchControlSchemaEnabled {type=boolean}
 IsRackFocusEnabled {type=boolean}
 IsSnapZoomEnabled {type=boolean}
 LastHoveredWidget {type=UserWidget, default=null}
 MarkerTimerHandle {type=TimerHandle}
 PlayerNames {default {}}
 QuickAccessMenu {type=FortReplaySettingsQuickview_C, default=null}
 SpectatorSpeeds {type=integer}
 TiltSpeedText {type=string, default=Tilt Speed}
 TimerSpeed {type=number, default=0.1}
 TimeToSkip {type=number, default=15}
]
widgets {
 info {
  inheritedWidgetCount=3
  namedSlotCount=5
  parentClass=/Script/FortniteUI.AthenaSpectatorHUDControlsBase
  rootWidgetClass=&VerticalBox
  widgetCount=138
 }
 widgets [
  ContentVB {widgetClassPath=&VerticalBox, bIsVariable=true, widget=ContentVB}
  TimelineOverlay {widgetClassPath=&Overlay, bIsVariable=true, parent=ContentVB, widget=TimelineOverlay, slot {class=VerticalBoxSlot}}
  ReplayTimeline {
   widgetClassPath=/Game/UI/Replay/FortReplayTimeline
   bInherited=true
   parent=TimelineOverlay
   widget=ReplayTimeline
   slot {class=OverlaySlot, properties {horizontalAlignment=HAlign_Fill, verticalAlignment=VAlign_Fill}}
  }
  ReplaySettingsWrapper {
   widgetClassPath=/Game/Athena/HUD/FortHUDElementWrapper
   bIsVariable=true
   parent=ContentVB
   widget=ReplaySettingsWrapper
   properties {elementTag {tagName=Athena.Replay.Live}, hUDElementTag {gameplayTags [{tagName=Athena.Spectator}]}}
   slot {class=VerticalBoxSlot, properties {horizontalAlignment=HAlign_Center, padding=<ltrb 0 0 0 5>}}
  }
  ReplaySettingsHorizontalBox {widgetClassPath=&HorizontalBox, namedSlotHost=ReplaySettingsWrapper, widget=ReplaySettingsHorizontalBox}
  FocusTimeline {
   widgetClassPath=&CommonVisibilityWidgetLegacy
   parent=ReplaySettingsHorizontalBox
   widget=FocusTimeline
   properties {bShowForMouseAndKeyboard=false, bShowForTouch=false, padding=<ltrb 0 0 0 0>}
   slot {class=HorizontalBoxSlot, properties {padding=<ltrb 0 0 10 0>}}
  }
  FocusTimelineOverlay {
   widgetClassPath=&Overlay
   parent=FocusTimeline
   widget=FocusTimelineOverlay
   slot {class=BorderSlot, properties {padding=<ltrb 0 0 0 0>}}
  }
  FocusTimelineBackplate {
   widgetClassPath=&WBP_UIKit_Backplate
   parent=FocusTimelineOverlay
   widget=FocusTimelineBackplate
   properties {
    backplateMaterial=/FortUILibrary/Foundation/Buttons/Materials/M_Timeline_Backplate_Inst
    brightness=Medium
    cornerRadius=Tab
    backplateSize=<0 42>
   }
   slot {class=OverlaySlot, properties {horizontalAlignment=HAlign_Fill}}
  }
  Button_FocusTimeline {
   widgetClassPath=/FortUILibrary/Foundation/Buttons/IconTextButtonReplay
   bInherited=true
   bIsVariable=true
   parent=FocusTimelineOverlay
   widget=Button_FocusTimeline
   properties {
    onButtonBaseClicked=(&self:WidgetTree.Button_FocusTimeline.HandleOnButtonClicked)
    onButtonBaseDoubleClicked=(&self:WidgetTree.Button_FocusTimeline.HandleOnButtonDoubleClicked)
    onButtonBaseHovered=(&self:WidgetTree.Button_FocusTimeline.HandleOnButtonHovered)
    onButtonBaseUnhovered=(&self:WidgetTree.Button_FocusTimeline.HandleOnButtonUnhovered)
    onSelectedChangedBase=(&self:WidgetTree.Button_FocusTimeline.HandleOnSelectedChanged)
    iconBrush {
     bIsDynamicallyLoaded=false
     drawAs=Image
     imageType=NoImage
     mirroring=NoMirror
     resourceName=None
     resourceObject=/Game/UI/Foundation/Textures/Icons/Replay/T-Icon-Replay-Timeline-L
     tiling=NoTile
     imageSize=<32 32>
     margin=<ltrb 0 0 0 0>
     outlineSettings {
      bUseBrushTransparency=false
      roundingType=HalfHeightRadius
      width=0
      color {colorUseRule=UseColor_Specified, specifiedColor=<rgba 0 0 0 0>}
      cornerRadii=<0 0 0 1>
     }
     tintColor {colorUseRule=UseColor_Specified, specifiedColor=<rgba 1 1 1 1>}
     uVRegion {bIsValid=false, max=<0 0>, min=<0 0>}
    }
    pressedSlateSoundOverride {resourceObject=/Game/Sounds/Fort_UI_Sounds/Replay/DVR_Cam_Screenshot_Cue}
   }
   slot {class=OverlaySlot}
  }
  ViewSettings {
   widgetClassPath=&Overlay
   parent=ReplaySettingsHorizontalBox
   widget=ViewSettings
   slot {class=HorizontalBoxSlot, properties {padding=<ltrb 0 0 10 0>}}
  }
  ViewSettingsBackplate {
   widgetClassPath=&WBP_UIKit_Backplate
   parent=ViewSettings
   widget=ViewSettingsBackplate
   properties {
    backplateMaterial=&M_Timeline_Backplate_Inst
    brightness=Medium
    cornerRadius=Tab
    backplateSize=<0 42>
   }
   slot {class=OverlaySlot, properties {horizontalAlignment=HAlign_Fill}}
  }
  ViewSettingsButton {
   widgetClassPath=&IconTextButtonReplay
   bIsVariable=true
   parent=ViewSettings
   widget=ViewSettingsButton
   properties {
    onButtonBaseClicked=(&self:WidgetTree.ViewSettingsButton.HandleOnButtonClicked)
    onButtonBaseDoubleClicked=(&self:WidgetTree.ViewSettingsButton.HandleOnButtonDoubleClicked)
    onButtonBaseHovered=(&self:WidgetTree.ViewSettingsButton.HandleOnButtonHovered)
    onButtonBaseUnhovered=(&self:WidgetTree.ViewSettingsButton.HandleOnButtonUnhovered)
    onSelectedChangedBase=(&self:WidgetTree.ViewSettingsButton.HandleOnSelectedChanged)
    iconBrush {
     bIsDynamicallyLoaded=false
     drawAs=Image
     imageType=NoImage
     mirroring=NoMirror
     resourceName=None
     resourceObject=/Game/UI/Foundation/Textures/Icons/Replay/T-Icon-Replay-CameraSettings-L
     tiling=NoTile
     imageSize=<32 32>
     margin=<ltrb 0 0 0 0>
     outlineSettings {
      bUseBrushTransparency=false
      roundingType=HalfHeightRadius
      width=0
      color {colorUseRule=UseColor_Specified, specifiedColor=<rgba 0 0 0 0>}
      cornerRadii=<0 0 0 1>
     }
     tintColor {colorUseRule=UseColor_Specified, specifiedColor=<rgba 1 1 1 1>}
     uVRegion {bIsValid=false, max=<0 0>, min=<0 0>}
    }
   }
   slot {class=OverlaySlot}
  }
  CamType {
   widgetClassPath=&Overlay
   parent=ReplaySettingsHorizontalBox
   widget=CamType
   properties {bOverride_Cursor=true}
   slot {class=HorizontalBoxSlot, properties {padding=<ltrb 0 0 10 0>}}
  }
  CamTypeBackplate {
   widgetClassPath=&WBP_UIKit_Backplate
   parent=CamType
   widget=CamTypeBackplate
   properties {
    backplateMaterial=&M_Timeline_Backplate_Inst
    brightness=Medium
    cornerRadius=Tab
    backplateSize=<0 42>
   }
   slot {class=OverlaySlot, properties {horizontalAlignment=HAlign_Fill}}
  }
  CamTypeHorizontalBox {
   widgetClassPath=&HorizontalBox
   parent=CamType
   widget=CamTypeHorizontalBox
   properties {bOverride_Cursor=true}
   slot {class=OverlaySlot, properties {horizontalAlignment=HAlign_Center}}
  }
  CamName {widgetClassPath=&VerticalBox, parent=CamTypeHorizontalBox, widget=CamName, slot {class=HorizontalBoxSlot}}
  CamNameSizeOverride {
   widgetClassPath=&SizeBox
   bIsVariable=true
   parent=CamName
   widget=CamNameSizeOverride
   properties {bOverride_MinDesiredWidth=true, maxDesiredWidth=300, minDesiredWidth=160, widthOverride=160}
   slot {class=VerticalBoxSlot}
  }
  CurrentCameraButton {
   widgetClassPath=&IconTextButtonReplay
   bIsVariable=true
   parent=CamName
   widget=CurrentCameraButton
   properties {
    bHideText=false
    bIgnoreAnimations=true
    bIsFocusable=false
    bShouldHideImage=true
    buttonText=Drone Attach
    onButtonBaseClicked=(&self:WidgetTree.CurrentCameraButton.HandleOnButtonClicked)
    onButtonBaseDoubleClicked=(&self:WidgetTree.CurrentCameraButton.HandleOnButtonDoubleClicked)
    onButtonBaseHovered=(&self:WidgetTree.CurrentCameraButton.HandleOnButtonHovered)
    onButtonBaseUnhovered=(&self:WidgetTree.CurrentCameraButton.HandleOnButtonUnhovered)
    onSelectedChangedBase=(&self:WidgetTree.CurrentCameraButton.HandleOnSelectedChanged)
    shouldHideInputAction=true
    iconBrush {
     bIsDynamicallyLoaded=false
     drawAs=Image
     imageType=NoImage
     mirroring=NoMirror
     resourceName=None
     resourceObject=/Game/UI/Foundation/Textures/Icons/Replay/T-Icon-Replay-CurrentCamera-L
     tiling=NoTile
     imageSize=<32 32>
     margin=<ltrb 0 0 0 0>
     outlineSettings {
      bUseBrushTransparency=false
      roundingType=HalfHeightRadius
      width=0
      color {colorUseRule=UseColor_Specified, specifiedColor=<rgba 0 0 0 0>}
      cornerRadii=<0 0 0 1>
     }
     tintColor {colorUseRule=UseColor_Specified, specifiedColor=<rgba 1 1 1 1>}
     uVRegion {bIsValid=false, max=<0 0>, min=<0 0>}
    }
   }
   slot {class=VerticalBoxSlot}
  }
  CamTypeInputAction {
   widgetClassPath=&CommonActionWidget
   parent=CamName
   widget=CamTypeInputAction
   properties {inputActions [{dataTable=/FortUILibrary/Input/ReplayInputActionDataTable, rowName=CycleCameras}]}
   slot {
    class=VerticalBoxSlot
    properties {horizontalAlignment=HAlign_Center, verticalAlignment=VAlign_Bottom, padding=<ltrb 0 5 0 0>, size {sizeRule=Fill, value=1}}
   }
  }
  CamTypeContextBox {widgetClassPath=&VerticalBox, parent=CamTypeHorizontalBox, widget=CamTypeContextBox, slot {class=HorizontalBoxSlot}}
  CameraTypeContextMenu {
   widgetClassPath=/FortUILibrary/UIKit/ContextMenu/WBP_UIKit_ContextMenuAnchorButton
   bIsVariable=true
   parent=CamTypeContextBox
   widget=CameraTypeContextMenu
   properties {
    bIsFocusable=true
    button Background Material=/FortUILibrary/UIKit/Buttons/MaterialInstances/CTAButtons/MI_UIKit_BgQuiet_Fallback
    buttonClass=/FortUILibrary/Foundation/Buttons/IconContextButtonReplay.IconContextButtonReplay_C
    buttonIcon=/FortUILibrary/Textures/Icons/T_UI_Icon_UIKit_ContextMenu
    buttonText=""
    closeMenuOnContextClicked=true
    menuPlacement=MenuPlacement_RightLeftCenter
    menuWidgetClass=/FortUILibrary/Systems/ContextMenu/WBP_UIKit_Context_Menu_Fallback.WBP_UIKit_Context_Menu_Fallback_C
    min Height=42
    min Width=56
    button Padding=<ltrb -9 0 0 0>
    button Text Font {
     bForceMonospaced=false
     bMaterialIsStencil=false
     fontMaterial=null
     fontObject=&HeadingNow
     letterSpacing=48
     monospacedWidth=1
     size=13.5
     skewAmount=0
     typefaceFontName=74 Regular
     outlineSettings {
      bApplyOutlineToDropShadows=false
      bMiteredCorners=false
      bSeparateFillAlpha=false
      outlineMaterial=null
      outlineSize=0
      outlineColor=<rgba 0 0 0 1>
     }
    }
   }
   slot {class=VerticalBoxSlot}
  }
  CameraTypeInputAction {
   widgetClassPath=&CommonActionWidget
   bIsVariable=true
   parent=CamTypeContextBox
   widget=CameraTypeInputAction
   properties {visibility=Hidden, inputActions [{dataTable=&ReplayInputActionDataTable, rowName=ShowAllPlayers}]}
   slot {
    class=VerticalBoxSlot
    properties {horizontalAlignment=HAlign_Center, verticalAlignment=VAlign_Bottom, padding=<ltrb -9 0 0 0>, size {sizeRule=Fill, value=1}}
   }
  }
  Playback {
   widgetClassPath=&Overlay
   parent=ReplaySettingsHorizontalBox
   widget=Playback
   slot {class=HorizontalBoxSlot, properties {padding=<ltrb 0 0 10 0>}}
  }
  PlaybackBackplate {
   widgetClassPath=&WBP_UIKit_Backplate
   parent=Playback
   widget=PlaybackBackplate
   properties {
    backplateMaterial=&M_Timeline_Backplate_Inst
    brightness=Medium
    cornerRadius=Tab
    backplateSize=<0 42>
   }
   slot {class=OverlaySlot, properties {horizontalAlignment=HAlign_Fill}}
  }
  PlaybackHorizontalBox {widgetClassPath=&HorizontalBox, parent=Playback, widget=PlaybackHorizontalBox, slot {class=OverlaySlot}}
  JumptoStart {
   widgetClassPath=&IconTextButtonReplay
   bIsVariable=true
   parent=PlaybackHorizontalBox
   widget=JumptoStart
   properties {
    onButtonBaseClicked=(&self:WidgetTree.JumptoStart.HandleOnButtonClicked)
    onButtonBaseDoubleClicked=(&self:WidgetTree.JumptoStart.HandleOnButtonDoubleClicked)
    onButtonBaseHovered=(&self:WidgetTree.JumptoStart.HandleOnButtonHovered)
    onButtonBaseUnhovered=(&self:WidgetTree.JumptoStart.HandleOnButtonUnhovered)
    onSelectedChangedBase=(&self:WidgetTree.JumptoStart.HandleOnSelectedChanged)
    iconBrush {
     bIsDynamicallyLoaded=false
     drawAs=Image
     imageType=NoImage
     mirroring=NoMirror
     resourceName=None
     resourceObject=/Game/UI/Foundation/Textures/Icons/Replay/T-Icon-Replay-Previous-L
     tiling=NoTile
     imageSize=<32 32>
     margin=<ltrb 0 0 0 0>
     outlineSettings {
      bUseBrushTransparency=false
      roundingType=HalfHeightRadius
      width=0
      color {colorUseRule=UseColor_Specified, specifiedColor=<rgba 0 0 0 0>}
      cornerRadii=<0 0 0 1>
     }
     tintColor {colorUseRule=UseColor_Specified, specifiedColor=<rgba 1 1 1 1>}
     uVRegion {bIsValid=false, max=<0 0>, min=<0 0>}
    }
    pressedSlateSoundOverride {resourceObject=/Game/Sounds/Fort_UI_Sounds/Replay/DVR_Nav_JumpToStart_Cue}
   }
   slot {class=HorizontalBoxSlot}
  }
  Back {
   widgetClassPath=&IconTextButtonReplay
   bIsVariable=true
   parent=PlaybackHorizontalBox
   widget=Back
   properties {
    onButtonBaseClicked=(&self:WidgetTree.Back.HandleOnButtonClicked)
    onButtonBaseDoubleClicked=(&self:WidgetTree.Back.HandleOnButtonDoubleClicked)
    onButtonBaseHovered=(&self:WidgetTree.Back.HandleOnButtonHovered)
    onButtonBaseUnhovered=(&self:WidgetTree.Back.HandleOnButtonUnhovered)
    onSelectedChangedBase=(&self:WidgetTree.Back.HandleOnSelectedChanged)
    iconBrush {
     bIsDynamicallyLoaded=false
     drawAs=Image
     imageType=NoImage
     mirroring=NoMirror
     resourceName=None
     resourceObject=/Game/UI/Foundation/Textures/Icons/Replay/T-Icon-Replay-Rewind-L
     tiling=NoTile
     imageSize=<32 32>
     margin=<ltrb 0 0 0 0>
     outlineSettings {
      bUseBrushTransparency=false
      roundingType=HalfHeightRadius
      width=0
      color {colorUseRule=UseColor_Specified, specifiedColor=<rgba 0 0 0 0>}
      cornerRadii=<0 0 0 1>
     }
     tintColor {colorUseRule=UseColor_Specified, specifiedColor=<rgba 1 1 1 1>}
     uVRegion {bIsValid=false, max=<0 0>, min=<0 0>}
    }
    pressedSlateSoundOverride {resourceObject=/Game/Sounds/Fort_UI_Sounds/Replay/DVR_Nav_QuickBack_Cue}
   }
   slot {class=HorizontalBoxSlot}
  }
  PlayPause {
   widgetClassPath=&IconTextButtonReplay
   bIsVariable=true
   parent=PlaybackHorizontalBox
   widget=PlayPause
   properties {
    onButtonBaseClicked=(&self:WidgetTree.PlayPause.HandleOnButtonClicked)
    onButtonBaseDoubleClicked=(&self:WidgetTree.PlayPause.HandleOnButtonDoubleClicked)
    onButtonBaseHovered=(&self:WidgetTree.PlayPause.HandleOnButtonHovered)
    onButtonBaseUnhovered=(&self:WidgetTree.PlayPause.HandleOnButtonUnhovered)
    onSelectedChangedBase=(&self:WidgetTree.PlayPause.HandleOnSelectedChanged)
    priority=1
    iconBrush {
     bIsDynamicallyLoaded=false
     drawAs=Image
     imageType=NoImage
     mirroring=NoMirror
     resourceName=None
     resourceObject=&T-Icon-Replay-Pause-L
     tiling=NoTile
     imageSize=<32 32>
     margin=<ltrb 0 0 0 0>
     outlineSettings {
      bUseBrushTransparency=false
      roundingType=HalfHeightRadius
      width=0
      color {colorUseRule=UseColor_Specified, specifiedColor=<rgba 0 0 0 0>}
      cornerRadii=<0 0 0 1>
     }
     tintColor {colorUseRule=UseColor_Specified, specifiedColor=<rgba 1 1 1 1>}
     uVRegion {bIsValid=false, max=<0 0>, min=<0 0>}
    }
    pressedSlateSoundOverride {resourceObject=/Game/Sounds/Fort_UI_Sounds/FrontEnd/New_FrontEnd/UI_BlankForOverride_Cue}
    triggeringInputAction {dataTable=&ReplayInputActionDataTable, rowName=PlayPause}
   }
   slot {class=HorizontalBoxSlot}
  }
  SkipOneFrame {
   widgetClassPath=&IconTextButtonReplay
   bIsVariable=true
   parent=PlaybackHorizontalBox
   widget=SkipOneFrame
   properties {
    onButtonBaseClicked=(&self:WidgetTree.SkipOneFrame.HandleOnButtonClicked)
    onButtonBaseDoubleClicked=(&self:WidgetTree.SkipOneFrame.HandleOnButtonDoubleClicked)
    onButtonBaseHovered=(&self:WidgetTree.SkipOneFrame.HandleOnButtonHovered)
    onButtonBaseUnhovered=(&self:WidgetTree.SkipOneFrame.HandleOnButtonUnhovered)
    onSelectedChangedBase=(&self:WidgetTree.SkipOneFrame.HandleOnSelectedChanged)
    visibility=Collapsed
    iconBrush {
     bIsDynamicallyLoaded=false
     drawAs=Image
     imageType=FullColor
     mirroring=NoMirror
     resourceName=None
     resourceObject=/Game/Athena/UI/InGame/Creative/PropertyEditor/Icon_Picker/Textures_V2/128x128/T_UI_IconLibrary_RightArrow_128
     tiling=NoTile
     imageSize=<32 32>
     margin=<ltrb 0 0 0 0>
     outlineSettings {
      bUseBrushTransparency=false
      roundingType=HalfHeightRadius
      width=0
      color {colorUseRule=UseColor_Specified, specifiedColor=<rgba 0 0 0 0>}
      cornerRadii=<0 0 0 1>
     }
     tintColor {colorUseRule=UseColor_Specified, specifiedColor=<rgba 1 1 1 1>}
     uVRegion {bIsValid=false, max=<0 0>, min=<0 0>}
    }
    pressedSlateSoundOverride {resourceObject=/Game/Sounds/Fort_UI_Sounds/Replay/DVR_Nav_QuickFWD_Cue}
   }
   slot {class=HorizontalBoxSlot}
  }
  Forward {
   widgetClassPath=&IconTextButtonReplay
   bIsVariable=true
   parent=PlaybackHorizontalBox
   widget=Forward
   properties {
    onButtonBaseClicked=(&self:WidgetTree.Forward.HandleOnButtonClicked)
    onButtonBaseDoubleClicked=(&self:WidgetTree.Forward.HandleOnButtonDoubleClicked)
    onButtonBaseHovered=(&self:WidgetTree.Forward.HandleOnButtonHovered)
    onButtonBaseUnhovered=(&self:WidgetTree.Forward.HandleOnButtonUnhovered)
    onSelectedChangedBase=(&self:WidgetTree.Forward.HandleOnSelectedChanged)
    iconBrush {
     bIsDynamicallyLoaded=false
     drawAs=Image
     imageType=NoImage
     mirroring=NoMirror
     resourceName=None
     resourceObject=/Game/UI/Foundation/Textures/Icons/Replay/T-Icon-Replay-Forward-L
     tiling=NoTile
     imageSize=<32 32>
     margin=<ltrb 0 0 0 0>
     outlineSettings {
      bUseBrushTransparency=false
      roundingType=HalfHeightRadius
      width=0
      color {colorUseRule=UseColor_Specified, specifiedColor=<rgba 0 0 0 0>}
      cornerRadii=<0 0 0 1>
     }
     tintColor {colorUseRule=UseColor_Specified, specifiedColor=<rgba 1 1 1 1>}
     uVRegion {bIsValid=false, max=<0 0>, min=<0 0>}
    }
    pressedSlateSoundOverride {resourceObject=&DVR_Nav_QuickFWD_Cue}
   }
   slot {class=HorizontalBoxSlot}
  }
  JumpToEnd {
   widgetClassPath=&IconTextButtonReplay
   bIsVariable=true
   parent=PlaybackHorizontalBox
   widget=JumpToEnd
   properties {
    onButtonBaseClicked=(&self:WidgetTree.JumpToEnd.HandleOnButtonClicked)
    onButtonBaseDoubleClicked=(&self:WidgetTree.JumpToEnd.HandleOnButtonDoubleClicked)
    onButtonBaseHovered=(&self:WidgetTree.JumpToEnd.HandleOnButtonHovered)
    onButtonBaseUnhovered=(&self:WidgetTree.JumpToEnd.HandleOnButtonUnhovered)
    onSelectedChangedBase=(&self:WidgetTree.JumpToEnd.HandleOnSelectedChanged)
    iconBrush {
     bIsDynamicallyLoaded=false
     drawAs=Image
     imageType=NoImage
     mirroring=NoMirror
     resourceName=None
     resourceObject=/Game/UI/Foundation/Textures/Icons/Replay/T-Icon-Replay-Next-L
     tiling=NoTile
     imageSize=<32 32>
     margin=<ltrb 0 0 0 0>
     outlineSettings {
      bUseBrushTransparency=false
      roundingType=HalfHeightRadius
      width=0
      color {colorUseRule=UseColor_Specified, specifiedColor=<rgba 0 0 0 0>}
      cornerRadii=<0 0 0 1>
     }
     tintColor {colorUseRule=UseColor_Specified, specifiedColor=<rgba 1 1 1 1>}
     uVRegion {bIsValid=false, max=<0 0>, min=<0 0>}
    }
    pressedSlateSoundOverride {resourceObject=/Game/Sounds/Fort_UI_Sounds/Replay/DVR_Nav_JumpToEnd_Cue}
   }
   slot {class=HorizontalBoxSlot}
  }
  PlaybackSpeedWrapper {
   widgetClassPath=&FortHUDElementWrapper
   parent=ReplaySettingsHorizontalBox
   widget=PlaybackSpeedWrapper
   properties {elementTag {tagName=Athena.Replay.Live}, hUDElementTag {gameplayTags [{tagName=Athena.Spectator}]}}
   slot {class=HorizontalBoxSlot, properties {padding=<ltrb 0 0 10 0>}}
  }
  PlaybackSpeedOverlay {widgetClassPath=&Overlay, namedSlotHost=PlaybackSpeedWrapper, widget=PlaybackSpeedOverlay}
  PlaybackSpeedBackplate {
   widgetClassPath=&WBP_UIKit_Backplate
   parent=PlaybackSpeedOverlay
   widget=PlaybackSpeedBackplate
   properties {
    backplateMaterial=&M_Timeline_Backplate_Inst
    brightness=Medium
    cornerRadius=Tab
    backplateSize=<0 42>
   }
   slot {class=OverlaySlot, properties {horizontalAlignment=HAlign_Fill}}
  }
  HorizontalBox_165 {
   widgetClassPath=&HorizontalBox
   parent=PlaybackSpeedOverlay
   widget=HorizontalBox_165
   slot {class=OverlaySlot, properties {horizontalAlignment=HAlign_Fill, verticalAlignment=VAlign_Fill}}
  }
  Slower {
   widgetClassPath=&IconTextButtonReplay
   bIsVariable=true
   parent=HorizontalBox_165
   widget=Slower
   properties {
    bDisplayInputActionWhenNotInteractable=false
    bIsFocusable=false
    bShouldUseFallbackDefaultInputAction=false
    bUseCustomWidth=true
    customWidth=52
    onButtonBaseClicked=(&self:WidgetTree.Slower.HandleOnButtonClicked)
    onButtonBaseDoubleClicked=(&self:WidgetTree.Slower.HandleOnButtonDoubleClicked)
    onButtonBaseHovered=(&self:WidgetTree.Slower.HandleOnButtonHovered)
    onButtonBaseUnhovered=(&self:WidgetTree.Slower.HandleOnButtonUnhovered)
    onSelectedChangedBase=(&self:WidgetTree.Slower.HandleOnSelectedChanged)
    shouldHideInputAction=true
    iconBrush {
     bIsDynamicallyLoaded=false
     drawAs=Image
     imageType=NoImage
     mirroring=NoMirror
     resourceName=None
     resourceObject=/Game/UI/Foundation/Textures/Icons/Replay/T-Icon-Replay-Minus-L
     tiling=NoTile
     imageSize=<32 32>
     margin=<ltrb 0 0 0 0>
     outlineSettings {
      bUseBrushTransparency=false
      roundingType=HalfHeightRadius
      width=0
      color {colorUseRule=UseColor_Specified, specifiedColor=<rgba 0 0 0 0>}
      cornerRadii=<0 0 0 1>
     }
     tintColor {colorUseRule=UseColor_Specified, specifiedColor=<rgba 1 1 1 1>}
     uVRegion {bIsValid=false, max=<0 0>, min=<0 0>}
    }
    pressedSlateSoundOverride {resourceObject=/Game/Sounds/Fort_UI_Sounds/Replay/DVR_Nav_BWD_2x_Cue}
   }
   slot {class=HorizontalBoxSlot}
  }
  TimeDilationSizeBox {
   widgetClassPath=&SizeBox
   parent=HorizontalBox_165
   widget=TimeDilationSizeBox
   properties {bOverride_WidthOverride=true, widthOverride=65}
   slot {class=HorizontalBoxSlot}
  }
  VBTimeDilation {widgetClassPath=&VerticalBox, parent=TimeDilationSizeBox, widget=VBTimeDilation, slot {class=SizeBoxSlot}}
  TextTimeDilation {
   widgetClassPath=&TextBlock
   bIsVariable=true
   parent=VBTimeDilation
   widget=TextTimeDilation
   properties {
    text=1.00x
    font {
     bForceMonospaced=false
     bMaterialIsStencil=false
     fontMaterial=null
     fontObject=&HeadingNow
     letterSpacing=0
     monospacedWidth=1
     size=13.5
     skewAmount=0
     typefaceFontName=74 Regular
     outlineSettings {
      bApplyOutlineToDropShadows=false
      bMiteredCorners=false
      bSeparateFillAlpha=false
      outlineMaterial=null
      outlineSize=0
      outlineColor=<rgba 0 0 0 1>
     }
    }
   }
   slot {class=VerticalBoxSlot, properties {horizontalAlignment=HAlign_Center, verticalAlignment=VAlign_Center, padding=<ltrb 0 6 0 0>}}
  }
  TimeDilationAction {
   widgetClassPath=&CommonActionWidget
   bIsVariable=true
   parent=VBTimeDilation
   widget=TimeDilationAction
   properties {inputActions [
     {dataTable=&ReplayInputActionDataTable, rowName=IncreasePlaybackSpeed}
     {dataTable=&ReplayInputActionDataTable, rowName=DecreasePlaybackSpeed}
    ]}
   slot {class=VerticalBoxSlot, properties {horizontalAlignment=HAlign_Center, verticalAlignment=VAlign_Bottom, size {sizeRule=Fill, value=1}}}
  }
  Faster {
   widgetClassPath=&IconTextButtonReplay
   bIsVariable=true
   parent=HorizontalBox_165
   widget=Faster
   properties {
    bDisplayInputActionWhenNotInteractable=false
    bIsFocusable=false
    bShouldUseFallbackDefaultInputAction=false
    bUseCustomWidth=true
    customWidth=52
    onButtonBaseClicked=(&self:WidgetTree.Faster.HandleOnButtonClicked)
    onButtonBaseDoubleClicked=(&self:WidgetTree.Faster.HandleOnButtonDoubleClicked)
    onButtonBaseHovered=(&self:WidgetTree.Faster.HandleOnButtonHovered)
    onButtonBaseUnhovered=(&self:WidgetTree.Faster.HandleOnButtonUnhovered)
    onSelectedChangedBase=(&self:WidgetTree.Faster.HandleOnSelectedChanged)
    shouldHideInputAction=true
    iconBrush {
     bIsDynamicallyLoaded=false
     drawAs=Image
     imageType=NoImage
     mirroring=NoMirror
     resourceName=None
     resourceObject=/Game/UI/Foundation/Textures/Icons/Replay/T-Icon-Replay-Plus-L
     tiling=NoTile
     imageSize=<32 32>
     margin=<ltrb 0 0 0 0>
     outlineSettings {
      bUseBrushTransparency=false
      roundingType=HalfHeightRadius
      width=0
      color {colorUseRule=UseColor_Specified, specifiedColor=<rgba 0 0 0 0>}
      cornerRadii=<0 0 0 1>
     }
     tintColor {colorUseRule=UseColor_Specified, specifiedColor=<rgba 1 1 1 1>}
     uVRegion {bIsValid=false, max=<0 0>, min=<0 0>}
    }
    pressedSlateSoundOverride {resourceObject=/Game/Sounds/Fort_UI_Sounds/Replay/DVR_Nav_FWD_2x_Cue}
   }
   slot {class=HorizontalBoxSlot}
  }
  Tools {
   widgetClassPath=&Overlay
   parent=ReplaySettingsHorizontalBox
   widget=Tools
   slot {class=HorizontalBoxSlot, properties {padding=<ltrb 0 0 10 0>}}
  }
  ToolsBackplate {
   widgetClassPath=&WBP_UIKit_Backplate
   parent=Tools
   widget=ToolsBackplate
   properties {
    backplateMaterial=&M_Timeline_Backplate_Inst
    brightness=Medium
    cornerRadius=Tab
    backplateSize=<0 42>
   }
   slot {class=OverlaySlot, properties {horizontalAlignment=HAlign_Fill}}
  }
  ToolsHorizontalBox {
   widgetClassPath=&HorizontalBox
   bIsVariable=true
   parent=Tools
   widget=ToolsHorizontalBox
   properties {visibility=Collapsed}
   slot {class=OverlaySlot}
  }
  MobileCameraShotsControlsHorizontalBox {
   widgetClassPath=&HorizontalBox
   bIsVariable=true
   parent=ToolsHorizontalBox
   widget=MobileCameraShotsControlsHorizontalBox
   properties {visibility=Collapsed}
   slot {class=HorizontalBoxSlot, properties {horizontalAlignment=HAlign_Left, verticalAlignment=VAlign_Top}}
  }
  CurrentShotButton {
   widgetClassPath=&IconTextButtonReplay
   bIsVariable=true
   parent=MobileCameraShotsControlsHorizontalBox
   widget=CurrentShotButton
   properties {
    bHideText=false
    bShouldHideImage=true
    bUseCustomWidth=true
    buttonText=Shot 0
    customWidth=90
    onButtonBaseClicked=()
    onButtonBaseDoubleClicked=()
    onButtonBaseHovered=()
    onButtonBaseUnhovered=()
    onSelectedChangedBase=()
    priority=1
    visibility=Visible
    iconBrush {
     bIsDynamicallyLoaded=false
     drawAs=Image
     imageType=NoImage
     mirroring=NoMirror
     resourceName=None
     resourceObject=&T-Icon-Replay-CurrentCamera-L
     tiling=NoTile
     imageSize=<32 32>
     margin=<ltrb 0 0 0 0>
     outlineSettings {
      bUseBrushTransparency=false
      roundingType=HalfHeightRadius
      width=0
      color {colorUseRule=UseColor_Specified, specifiedColor=<rgba 0 0 0 0>}
      cornerRadii=<0 0 0 1>
     }
     tintColor {colorUseRule=UseColor_Specified, specifiedColor=<rgba 1 1 1 1>}
     uVRegion {bIsValid=false, max=<0 0>, min=<0 0>}
    }
   }
   slot {class=HorizontalBoxSlot}
  }
  ShotNumberContextBox {
   widgetClassPath=&VerticalBox
   parent=MobileCameraShotsControlsHorizontalBox
   widget=ShotNumberContextBox
   properties {visibility=Visible}
   slot {class=HorizontalBoxSlot}
  }
  ShotNumberContextMenu {
   widgetClassPath=&WBP_UIKit_ContextMenuAnchorButton
   bIsVariable=true
   parent=ShotNumberContextBox
   widget=ShotNumberContextMenu
   properties {
    bIsFocusable=true
    button Background Material=&MI_UIKit_BgQuiet_Fallback
    buttonClass=&IconContextButtonReplay_C
    buttonIcon=&T_UI_Icon_UIKit_ContextMenu
    buttonText=""
    closeMenuOnContextClicked=true
    menuWidgetClass=&WBP_UIKit_Context_Menu_Fallback_C
    min Height=42
    min Width=56
    visibility=Visible
    button Padding=<ltrb 0 0 0 0>
    button Text Font {
     bForceMonospaced=false
     bMaterialIsStencil=false
     fontMaterial=null
     fontObject=&HeadingNow
     letterSpacing=48
     monospacedWidth=1
     size=13.5
     skewAmount=0
     typefaceFontName=74 Regular
     outlineSettings {
      bApplyOutlineToDropShadows=false
      bMiteredCorners=false
      bSeparateFillAlpha=false
      outlineMaterial=null
      outlineSize=0
      outlineColor=<rgba 0 0 0 1>
     }
    }
   }
   slot {class=VerticalBoxSlot}
  }
  ShotNumberInputAction {
   widgetClassPath=&CommonActionWidget
   bIsVariable=true
   parent=ShotNumberContextBox
   widget=ShotNumberInputAction
   properties {visibility=Hidden, inputActions [{dataTable=&ReplayInputActionDataTable, rowName=ShowAllPlayers}]}
   slot {class=VerticalBoxSlot, properties {horizontalAlignment=HAlign_Center, verticalAlignment=VAlign_Bottom, size {sizeRule=Fill, value=1}}}
  }
  SaveShot {
   widgetClassPath=&IconTextButtonReplay
   bIsVariable=true
   parent=MobileCameraShotsControlsHorizontalBox
   widget=SaveShot
   properties {
    bUseCustomWidth=true
    customWidth=54
    onButtonBaseClicked=()
    onButtonBaseDoubleClicked=()
    onButtonBaseHovered=()
    onButtonBaseUnhovered=()
    onSelectedChangedBase=()
    priority=1
    visibility=Visible
    iconBrush {
     bIsDynamicallyLoaded=false
     drawAs=Image
     imageType=FullColor
     mirroring=NoMirror
     resourceName=None
     resourceObject=/Game/UI/Foundation/Textures/Icons/Replay/T_Icon_Replay_SaveCamera
     tiling=NoTile
     imageSize=<32 32>
     margin=<ltrb 0 0 0 0>
     outlineSettings {
      bUseBrushTransparency=false
      roundingType=HalfHeightRadius
      width=0
      color {colorUseRule=UseColor_Specified, specifiedColor=<rgba 0 0 0 0>}
      cornerRadii=<0 0 0 1>
     }
     tintColor {colorUseRule=UseColor_Specified, specifiedColor=<rgba 1 1 1 1>}
     uVRegion {bIsValid=false, max=<0 0>, min=<0 0>}
    }
    pressedSlateSoundOverride {resourceObject=&UI_BlankForOverride_Cue}
   }
   slot {class=HorizontalBoxSlot}
  }
  SaveShotFull {
   widgetClassPath=&IconTextButtonReplay
   bIsVariable=true
   parent=MobileCameraShotsControlsHorizontalBox
   widget=SaveShotFull
   properties {
    bUseCustomWidth=true
    customWidth=54
    onButtonBaseClicked=()
    onButtonBaseDoubleClicked=()
    onButtonBaseHovered=()
    onButtonBaseUnhovered=()
    onSelectedChangedBase=()
    priority=1
    visibility=Visible
    iconBrush {
     bIsDynamicallyLoaded=false
     drawAs=Image
     imageType=FullColor
     mirroring=NoMirror
     resourceName=None
     resourceObject=/Game/UI/Foundation/Textures/Icons/Replay/T_Icon_Replay_SaveCameraTime
     tiling=NoTile
     imageSize=<32 32>
     margin=<ltrb 0 0 0 0>
     outlineSettings {
      bUseBrushTransparency=false
      roundingType=HalfHeightRadius
      width=0
      color {colorUseRule=UseColor_Specified, specifiedColor=<rgba 0 0 0 0>}
      cornerRadii=<0 0 0 1>
     }
     tintColor {colorUseRule=UseColor_Specified, specifiedColor=<rgba 1 1 1 1>}
     uVRegion {bIsValid=false, max=<0 0>, min=<0 0>}
    }
    pressedSlateSoundOverride {resourceObject=&UI_BlankForOverride_Cue}
   }
   slot {class=HorizontalBoxSlot}
  }
  FPSButtonSpacer {
   widgetClassPath=&Spacer
   bIsVariable=true
   parent=ToolsHorizontalBox
   widget=FPSButtonSpacer
   properties {size=<8 0>}
   slot {class=HorizontalBoxSlot}
  }
  FPSSettingsButton {
   widgetClassPath=&IconTextButtonReplay
   bIsVariable=true
   parent=ToolsHorizontalBox
   widget=FPSSettingsButton
   properties {
    bHideText=false
    bShouldHideImage=true
    bUseCustomWidth=true
    buttonText=30 FPS
    customWidth=85
    onButtonBaseClicked=()
    onButtonBaseDoubleClicked=()
    onButtonBaseHovered=()
    onButtonBaseUnhovered=()
    onSelectedChangedBase=()
    visibility=Collapsed
    iconBrush {
     bIsDynamicallyLoaded=false
     drawAs=Image
     imageType=FullColor
     mirroring=NoMirror
     resourceName=None
     resourceObject=null
     tiling=NoTile
     imageSize=<32 32>
     margin=<ltrb 0 0 0 0>
     outlineSettings {
      bUseBrushTransparency=false
      roundingType=HalfHeightRadius
      width=0
      color {colorUseRule=UseColor_Specified, specifiedColor=<rgba 0 0 0 0>}
      cornerRadii=<0 0 0 1>
     }
     tintColor {colorUseRule=UseColor_Specified, specifiedColor=<rgba 1 1 1 1>}
     uVRegion {bIsValid=false, max=<0 0>, min=<0 0>}
    }
   }
   slot {class=HorizontalBoxSlot}
  }
  BlurSettingsButton {
   widgetClassPath=&IconTextButtonReplay
   bIsVariable=true
   parent=ToolsHorizontalBox
   widget=BlurSettingsButton
   properties {
    bIgnoreAnimations=true
    bUseCustomWidth=true
    customWidth=54
    onButtonBaseClicked=()
    onButtonBaseDoubleClicked=()
    onButtonBaseHovered=()
    onButtonBaseUnhovered=()
    onSelectedChangedBase=()
    visibility=Collapsed
    iconBrush {
     bIsDynamicallyLoaded=false
     drawAs=Image
     imageType=FullColor
     mirroring=NoMirror
     resourceName=None
     resourceObject=&T_Icon_Replay_MotionBlur_Off_C
     tiling=NoTile
     imageSize=<32 32>
     margin=<ltrb 0 0 0 0>
     outlineSettings {
      bUseBrushTransparency=false
      roundingType=HalfHeightRadius
      width=0
      color {colorUseRule=UseColor_Specified, specifiedColor=<rgba 0 0 0 0>}
      cornerRadii=<0 0 0 1>
     }
     tintColor {colorUseRule=UseColor_Specified, specifiedColor=<rgba 1 1 1 1>}
     uVRegion {bIsValid=false, max=<0 0>, min=<0 0>}
    }
   }
   slot {class=HorizontalBoxSlot}
  }
  KeyframeButton {
   widgetClassPath=&IconTextButtonReplay
   bIsVariable=true
   parent=ToolsHorizontalBox
   widget=KeyframeButton
   properties {
    onButtonBaseClicked=()
    onButtonBaseDoubleClicked=()
    onButtonBaseHovered=()
    onButtonBaseUnhovered=()
    onSelectedChangedBase=()
    visibility=Collapsed
    iconBrush {
     bIsDynamicallyLoaded=false
     drawAs=Image
     imageType=FullColor
     mirroring=NoMirror
     resourceName=None
     resourceObject=/Game/Athena/UI/InGame/Creative/PropertyEditor/Icon_Picker/Textures_V2/128x128/T_UI_IconLibrary_Projector_128
     tiling=NoTile
     imageSize=<32 32>
     margin=<ltrb 0 0 0 0>
     outlineSettings {
      bUseBrushTransparency=false
      roundingType=HalfHeightRadius
      width=0
      color {colorUseRule=UseColor_Specified, specifiedColor=<rgba 0 0 0 0>}
      cornerRadii=<0 0 0 1>
     }
     tintColor {colorUseRule=UseColor_Specified, specifiedColor=<rgba 1 1 1 1>}
     uVRegion {bIsValid=false, max=<0 0>, min=<0 0>}
    }
   }
   slot {class=HorizontalBoxSlot}
  }
  RecorderButton {
   widgetClassPath=&IconTextButtonReplay
   bIsVariable=true
   parent=ToolsHorizontalBox
   widget=RecorderButton
   properties {
    onButtonBaseClicked=()
    onButtonBaseDoubleClicked=()
    onButtonBaseHovered=()
    onButtonBaseUnhovered=()
    onSelectedChangedBase=()
    visibility=Collapsed
    iconBrush {
     bIsDynamicallyLoaded=false
     drawAs=Image
     imageType=FullColor
     mirroring=NoMirror
     resourceName=None
     resourceObject=/Game/Athena/UI/InGame/Creative/PropertyEditor/Icon_Picker/Textures_V2/128x128/T_UI_IconLibrary_FloppyDisc_128
     tiling=NoTile
     imageSize=<32 32>
     margin=<ltrb 0 0 0 0>
     outlineSettings {
      bUseBrushTransparency=false
      roundingType=HalfHeightRadius
      width=0
      color {colorUseRule=UseColor_Specified, specifiedColor=<rgba 0 0 0 0>}
      cornerRadii=<0 0 0 1>
     }
     tintColor {colorUseRule=UseColor_Specified, specifiedColor=<rgba 1 1 1 1>}
     uVRegion {bIsValid=false, max=<0 0>, min=<0 0>}
    }
   }
   slot {class=HorizontalBoxSlot}
  }
  BattleMapTimelineModeWrapper {
   widgetClassPath=&FortHUDElementWrapper
   parent=ToolsHorizontalBox
   widget=BattleMapTimelineModeWrapper
   properties {
    elementTag {tagName=Athena.Replay.HUD}
    hUDElementTag {gameplayTags [
      {tagName=Athena.Replay.CameraMode.ThirdPerson}
      {tagName=Athena.Replay.CameraMode.ReverseShot}
      {tagName=Athena.Replay.CameraMode.GamePlay}
      {tagName=Athena.Replay.CameraMode.FollowShot}
      {tagName=Athena.Replay.CameraMode.DroneFree}
      {tagName=Athena.Replay.CameraMode.DroneFollow}
      {tagName=Athena.Replay.CameraMode.DroneAttach}
      {tagName=Athena.Replay.CameraMode.ARDrone}
      {tagName=Athena.Replay.CameraMode.DroneFreeDynamic}
     ]}
   }
   slot {class=HorizontalBoxSlot}
  }
  BattleMapTimelineModeButton {
   widgetClassPath=&IconTextButtonReplay
   bIsVariable=true
   namedSlotHost=BattleMapTimelineModeWrapper
   widget=BattleMapTimelineModeButton
   properties {
    onButtonBaseClicked=(&self:WidgetTree.BattleMapTimelineModeButton.HandleOnButtonClicked)
    onButtonBaseDoubleClicked=(&self:WidgetTree.BattleMapTimelineModeButton.HandleOnButtonDoubleClicked)
    onButtonBaseHovered=(&self:WidgetTree.BattleMapTimelineModeButton.HandleOnButtonHovered)
    onButtonBaseUnhovered=(&self:WidgetTree.BattleMapTimelineModeButton.HandleOnButtonUnhovered)
    onSelectedChangedBase=(&self:WidgetTree.BattleMapTimelineModeButton.HandleOnSelectedChanged)
    iconBrush {
     bIsDynamicallyLoaded=false
     drawAs=Image
     imageType=NoImage
     mirroring=NoMirror
     resourceName=None
     resourceObject=/Game/Athena/HUD/Spectator/Art/T-Icon-HeroPath
     tiling=NoTile
     imageSize=<32 32>
     margin=<ltrb 0 0 0 0>
     outlineSettings {
      bUseBrushTransparency=false
      roundingType=HalfHeightRadius
      width=0
      color {colorUseRule=UseColor_Specified, specifiedColor=<rgba 0 0 0 0>}
      cornerRadii=<0 0 0 1>
     }
     tintColor {colorUseRule=UseColor_Specified, specifiedColor=<rgba 1 1 1 1>}
     uVRegion {bIsValid=false, max=<0 0>, min=<0 0>}
    }
    pressedSlateSoundOverride {resourceObject=&UI_BlankForOverride_Cue}
    triggeringInputAction {dataTable=&ReplayInputActionDataTable, rowName=BattleMapTimelineStop}
   }
  }
  IconTextButtonReplay_PlayHighlights {
   widgetClassPath=&IconTextButtonReplay
   bInherited=true
   bIsVariable=true
   parent=ToolsHorizontalBox
   widget=IconTextButtonReplay_PlayHighlights
   properties {
    onButtonBaseClicked=(&self:WidgetTree.IconTextButtonReplay_PlayHighlights.HandleOnButtonClicked)
    onButtonBaseDoubleClicked=(&self:WidgetTree.IconTextButtonReplay_PlayHighlights.HandleOnButtonDoubleClicked)
    onButtonBaseHovered=(&self:WidgetTree.IconTextButtonReplay_PlayHighlights.HandleOnButtonHovered)
    onButtonBaseUnhovered=(&self:WidgetTree.IconTextButtonReplay_PlayHighlights.HandleOnButtonUnhovered)
    onSelectedChangedBase=(&self:WidgetTree.IconTextButtonReplay_PlayHighlights.HandleOnSelectedChanged)
    iconBrush {
     bIsDynamicallyLoaded=false
     drawAs=Image
     imageType=NoImage
     mirroring=NoMirror
     resourceName=None
     resourceObject=/Game/Athena/HUD/Spectator/Art/T-Icon-Highlights
     tiling=NoTile
     imageSize=<32 32>
     margin=<ltrb 0 0 0 0>
     outlineSettings {
      bUseBrushTransparency=false
      roundingType=HalfHeightRadius
      width=0
      color {colorUseRule=UseColor_Specified, specifiedColor=<rgba 0 0 0 0>}
      cornerRadii=<0 0 0 1>
     }
     tintColor {colorUseRule=UseColor_Specified, specifiedColor=<rgba 1 1 1 1>}
     uVRegion {bIsValid=false, max=<0 0>, min=<0 0>}
    }
   }
   slot {class=HorizontalBoxSlot}
  }
  FreeDroneTargetHidingToolSelect {
   widgetClassPath=&IconTextButtonReplay
   bIsVariable=true
   parent=ToolsHorizontalBox
   widget=FreeDroneTargetHidingToolSelect
   properties {
    onButtonBaseClicked=()
    onButtonBaseDoubleClicked=()
    onButtonBaseHovered=()
    onButtonBaseUnhovered=()
    onSelectedChangedBase=()
    visibility=Collapsed
    iconBrush {
     bIsDynamicallyLoaded=false
     drawAs=Image
     imageType=FullColor
     mirroring=NoMirror
     resourceName=None
     resourceObject=/Game/UI/Foundation/Textures/Icons/Replay/T_Icon_Replay_HidingTool_M
     tiling=NoTile
     imageSize=<32 32>
     margin=<ltrb 0 0 0 0>
     outlineSettings {
      bUseBrushTransparency=false
      roundingType=HalfHeightRadius
      width=0
      color {colorUseRule=UseColor_Specified, specifiedColor=<rgba 0 0 0 0>}
      cornerRadii=<0 0 0 1>
     }
     tintColor {colorUseRule=UseColor_Specified, specifiedColor=<rgba 1 1 1 1>}
     uVRegion {bIsValid=false, max=<0 0>, min=<0 0>}
    }
   }
   slot {class=HorizontalBoxSlot}
  }
  FreeDroneTargetHidingToolUndo {
   widgetClassPath=&IconTextButtonReplay
   bIsVariable=true
   parent=ToolsHorizontalBox
   widget=FreeDroneTargetHidingToolUndo
   properties {
    onButtonBaseClicked=()
    onButtonBaseDoubleClicked=()
    onButtonBaseHovered=()
    onButtonBaseUnhovered=()
    onSelectedChangedBase=()
    visibility=Collapsed
    iconBrush {
     bIsDynamicallyLoaded=false
     drawAs=Image
     imageType=FullColor
     mirroring=NoMirror
     resourceName=None
     resourceObject=/Game/UI/Foundation/Textures/Icons/Replay/T-Icon-Replay-ToggleUI-L
     tiling=NoTile
     imageSize=<32 32>
     margin=<ltrb 0 0 0 0>
     outlineSettings {
      bUseBrushTransparency=false
      roundingType=HalfHeightRadius
      width=0
      color {colorUseRule=UseColor_Specified, specifiedColor=<rgba 0 0 0 0>}
      cornerRadii=<0 0 0 1>
     }
     tintColor {colorUseRule=UseColor_Specified, specifiedColor=<rgba 1 1 1 1>}
     uVRegion {bIsValid=false, max=<0 0>, min=<0 0>}
    }
   }
   slot {class=HorizontalBoxSlot}
  }
  CameraPlacement {widgetClassPath=&Overlay, parent=ReplaySettingsHorizontalBox, widget=CameraPlacement, slot {class=HorizontalBoxSlot}}
  CameraPlacementBackplate {
   widgetClassPath=&WBP_UIKit_Backplate
   parent=CameraPlacement
   widget=CameraPlacementBackplate
   properties {
    backplateMaterial=&M_Timeline_Backplate_Inst
    brightness=Medium
    cornerRadius=Tab
    backplateSize=<0 42>
   }
   slot {class=OverlaySlot, properties {horizontalAlignment=HAlign_Fill}}
  }
  CameraPlacementHorizontalBox {widgetClassPath=&HorizontalBox, parent=CameraPlacement, widget=CameraPlacementHorizontalBox, slot {class=OverlaySlot}}
  FreeDroneTeleportToPawn {
   widgetClassPath=&IconTextButtonReplay
   bIsVariable=true
   parent=CameraPlacementHorizontalBox
   widget=FreeDroneTeleportToPawn
   properties {
    bUseCustomWidth=true
    customWidth=52
    onButtonBaseClicked=()
    onButtonBaseDoubleClicked=()
    onButtonBaseHovered=()
    onButtonBaseUnhovered=()
    onSelectedChangedBase=()
    iconBrush {
     bIsDynamicallyLoaded=false
     drawAs=Image
     imageType=FullColor
     mirroring=NoMirror
     resourceName=None
     resourceObject=/Game/UI/Foundation/Textures/Icons/Replay/T_Icon_Replay_TeleportToPawn_M
     tiling=NoTile
     imageSize=<32 32>
     margin=<ltrb 0 0 0 0>
     outlineSettings {
      bUseBrushTransparency=false
      roundingType=HalfHeightRadius
      width=0
      color {colorUseRule=UseColor_Specified, specifiedColor=<rgba 0 0 0 0>}
      cornerRadii=<0 0 0 1>
     }
     tintColor {colorUseRule=UseColor_Specified, specifiedColor=<rgba 1 1 1 1>}
     uVRegion {bIsValid=false, max=<0 0>, min=<0 0>}
    }
   }
   slot {class=HorizontalBoxSlot}
  }
  PreviousPlayerButton {
   widgetClassPath=&IconTextButtonReplay
   bIsVariable=true
   parent=CameraPlacementHorizontalBox
   widget=PreviousPlayerButton
   properties {
    onButtonBaseClicked=(&self:WidgetTree.PreviousPlayerButton.HandleOnButtonClicked)
    onButtonBaseDoubleClicked=(&self:WidgetTree.PreviousPlayerButton.HandleOnButtonDoubleClicked)
    onButtonBaseHovered=(&self:WidgetTree.PreviousPlayerButton.HandleOnButtonHovered)
    onButtonBaseUnhovered=(&self:WidgetTree.PreviousPlayerButton.HandleOnButtonUnhovered)
    onSelectedChangedBase=(&self:WidgetTree.PreviousPlayerButton.HandleOnSelectedChanged)
    iconBrush {
     bIsDynamicallyLoaded=false
     drawAs=Image
     imageType=NoImage
     mirroring=NoMirror
     resourceName=None
     resourceObject=/Game/Athena/HUD/Spectator/Art/T-Icon-PlayerSelectLeft
     tiling=NoTile
     imageSize=<32 32>
     margin=<ltrb 0 0 0 0>
     outlineSettings {
      bUseBrushTransparency=false
      roundingType=HalfHeightRadius
      width=0
      color {colorUseRule=UseColor_Specified, specifiedColor=<rgba 0 0 0 0>}
      cornerRadii=<0 0 0 1>
     }
     tintColor {colorUseRule=UseColor_Specified, specifiedColor=<rgba 1 1 1 1>}
     uVRegion {bIsValid=false, max=<0 0>, min=<0 0>}
    }
   }
   slot {class=HorizontalBoxSlot}
  }
  FollowedPlayerText_Box {
   widgetClassPath=&SizeBox
   bIsVariable=true
   parent=CameraPlacementHorizontalBox
   widget=FollowedPlayerText_Box
   properties {bOverride_WidthOverride=true, maxDesiredWidth=300, minDesiredWidth=50, widthOverride=175}
   slot {class=HorizontalBoxSlot}
  }
  HorizontalBox_2 {
   widgetClassPath=&HorizontalBox
   parent=FollowedPlayerText_Box
   widget=HorizontalBox_2
   slot {class=SizeBoxSlot, properties {horizontalAlignment=HAlign_Center, verticalAlignment=VAlign_Top}}
  }
  FollowedPlayerText {
   widgetClassPath=&CommonTextBlock
   bIsVariable=true
   parent=HorizontalBox_2
   widget=FollowedPlayerText
   properties {
    scrollStyle=&TextScrollStyle-Base_C
    text="{FollowedPlayer}"
    font {
     bForceMonospaced=false
     bMaterialIsStencil=false
     fontMaterial=null
     fontObject=&HeadingNow
     letterSpacing=0
     monospacedWidth=1
     size=13.5
     skewAmount=0
     typefaceFontName=74 Regular
     outlineSettings {
      bApplyOutlineToDropShadows=false
      bMiteredCorners=false
      bSeparateFillAlpha=false
      outlineMaterial=null
      outlineSize=0
      outlineColor=<rgba 0 0 0 1>
     }
    }
    shadowOffset=<0 0>
   }
   slot {class=HorizontalBoxSlot, properties {horizontalAlignment=HAlign_Center, padding=<ltrb 5 10.5 5 0>, size {sizeRule=Fill, value=1}}}
  }
  PlayerSelectorContextBox {
   widgetClassPath=&VerticalBox
   bIsVariable=true
   parent=CameraPlacementHorizontalBox
   widget=PlayerSelectorContextBox
   properties {visibility=Collapsed}
   slot {class=HorizontalBoxSlot}
  }
  PlayerSelectorContextMenu {
   widgetClassPath=&WBP_UIKit_ContextMenuAnchorButton
   bIsVariable=true
   parent=PlayerSelectorContextBox
   widget=PlayerSelectorContextMenu
   properties {
    bIsFocusable=true
    button Background Material=&MI_UIKit_BgQuiet_Fallback
    buttonClass=&IconContextButtonReplay_C
    buttonIcon=&T_UI_Icon_UIKit_ContextMenu
    buttonText=""
    closeMenuOnContextClicked=true
    menuPlacement=MenuPlacement_RightLeftCenter
    menuWidgetClass=&WBP_UIKit_Context_Menu_Fallback_C
    min Height=42
    min Width=56
    button Padding=<ltrb 0 0 0 0>
    button Text Font {
     bForceMonospaced=false
     bMaterialIsStencil=false
     fontMaterial=null
     fontObject=&HeadingNow
     letterSpacing=48
     monospacedWidth=1
     size=13.5
     skewAmount=0
     typefaceFontName=74 Regular
     outlineSettings {
      bApplyOutlineToDropShadows=false
      bMiteredCorners=false
      bSeparateFillAlpha=false
      outlineMaterial=null
      outlineSize=0
      outlineColor=<rgba 0 0 0 1>
     }
    }
   }
   slot {class=VerticalBoxSlot}
  }
  PlayerSelectorInputAction {
   widgetClassPath=&CommonActionWidget
   bIsVariable=true
   parent=PlayerSelectorContextBox
   widget=PlayerSelectorInputAction
   properties {visibility=Hidden, inputActions [{dataTable=&ReplayInputActionDataTable, rowName=ShowAllPlayers}]}
   slot {class=VerticalBoxSlot, properties {horizontalAlignment=HAlign_Center, verticalAlignment=VAlign_Bottom, size {sizeRule=Fill, value=1}}}
  }
  NextPlayerButton {
   widgetClassPath=&IconTextButtonReplay
   bIsVariable=true
   parent=CameraPlacementHorizontalBox
   widget=NextPlayerButton
   properties {
    onButtonBaseClicked=(&self:WidgetTree.NextPlayerButton.HandleOnButtonClicked)
    onButtonBaseDoubleClicked=(&self:WidgetTree.NextPlayerButton.HandleOnButtonDoubleClicked)
    onButtonBaseHovered=(&self:WidgetTree.NextPlayerButton.HandleOnButtonHovered)
    onButtonBaseUnhovered=(&self:WidgetTree.NextPlayerButton.HandleOnButtonUnhovered)
    onSelectedChangedBase=(&self:WidgetTree.NextPlayerButton.HandleOnSelectedChanged)
    iconBrush {
     bIsDynamicallyLoaded=false
     drawAs=Image
     imageType=NoImage
     mirroring=NoMirror
     resourceName=None
     resourceObject=/Game/Athena/HUD/Spectator/Art/T-Icon-PlayerSelectRight
     tiling=NoTile
     imageSize=<32 32>
     margin=<ltrb 0 0 0 0>
     outlineSettings {
      bUseBrushTransparency=false
      roundingType=HalfHeightRadius
      width=0
      color {colorUseRule=UseColor_Specified, specifiedColor=<rgba 0 0 0 0>}
      cornerRadii=<0 0 0 1>
     }
     tintColor {colorUseRule=UseColor_Specified, specifiedColor=<rgba 1 1 1 1>}
     uVRegion {bIsValid=false, max=<0 0>, min=<0 0>}
    }
   }
   slot {class=HorizontalBoxSlot}
  }
  VerticalBox_1 {
   widgetClassPath=&VerticalBox
   parent=ContentVB
   widget=VerticalBox_1
   slot {class=VerticalBoxSlot, properties {horizontalAlignment=HAlign_Center, padding=<ltrb 0 0 0 5>}}
  }
  SpectatorSettingsSize {
   widgetClassPath=&SizeBox
   bIsVariable=true
   parent=VerticalBox_1
   widget=SpectatorSettingsSize
   properties {minDesiredHeight=82}
   slot {class=VerticalBoxSlot}
  }
  SpectatorSettingsHorizontalBox {widgetClassPath=&HorizontalBox, parent=SpectatorSettingsSize, widget=SpectatorSettingsHorizontalBox, slot {class=SizeBoxSlot}}
  SpectatorSettingsWrapper {
   widgetClassPath=&FortHUDElementWrapper
   parent=SpectatorSettingsHorizontalBox
   widget=SpectatorSettingsWrapper
   properties {elementTag {tagName=Athena.Replay.Server}, hUDElementTag {gameplayTags [{tagName=Athena.Replay.Client}]}}
   slot {class=HorizontalBoxSlot}
  }
  HorizontalBox_270 {widgetClassPath=&HorizontalBox, namedSlotHost=SpectatorSettingsWrapper, widget=HorizontalBox_270}
  Alert {
   widgetClassPath=&Image
   bIsVariable=true
   parent=HorizontalBox_270
   widget=Alert
   properties {
    visibility=Collapsed
    brush {
     bIsDynamicallyLoaded=false
     drawAs=Image
     imageType=FullColor
     mirroring=NoMirror
     resourceName=None
     resourceObject=/FortUILibrary/Textures/MaterialInstances/MI_UIKit_Icon_AgeGatedAlert
     tiling=NoTile
     imageSize=<32 32>
     margin=<ltrb 0 0 0 0>
     outlineSettings {
      bUseBrushTransparency=false
      roundingType=HalfHeightRadius
      width=0
      color {colorUseRule=UseColor_Specified, specifiedColor=<rgba 0 0 0 0>}
      cornerRadii=<0 0 0 0>
     }
     tintColor {colorUseRule=UseColor_Specified, specifiedColor=<rgba 1 1 0 1>}
     uVRegion {bIsValid=false, max=<0 0>, min=<0 0>}
    }
   }
   slot {class=HorizontalBoxSlot, properties {horizontalAlignment=HAlign_Left, verticalAlignment=VAlign_Top, padding=<ltrb 0 5 12 0>}}
  }
  ViewSettingsLiveOverlay {
   widgetClassPath=&Overlay
   parent=HorizontalBox_270
   widget=ViewSettingsLiveOverlay
   slot {class=HorizontalBoxSlot, properties {padding=<ltrb 0 0 10 0>}}
  }
  ViewSettingsLiveBackplate {
   widgetClassPath=&WBP_UIKit_Backplate
   parent=ViewSettingsLiveOverlay
   widget=ViewSettingsLiveBackplate
   properties {
    backplateMaterial=&M_Timeline_Backplate_Inst
    brightness=Medium
    cornerRadius=Tab
    backplateSize=<0 42>
   }
   slot {class=OverlaySlot, properties {horizontalAlignment=HAlign_Fill}}
  }
  ViewSettingsButton_Live {
   widgetClassPath=&IconTextButtonReplay
   bIsVariable=true
   parent=ViewSettingsLiveOverlay
   widget=ViewSettingsButton_Live
   properties {
    onButtonBaseClicked=(&self:WidgetTree.ViewSettingsButton_Live.HandleOnButtonClicked)
    onButtonBaseDoubleClicked=(&self:WidgetTree.ViewSettingsButton_Live.HandleOnButtonDoubleClicked)
    onButtonBaseHovered=(&self:WidgetTree.ViewSettingsButton_Live.HandleOnButtonHovered)
    onButtonBaseUnhovered=(&self:WidgetTree.ViewSettingsButton_Live.HandleOnButtonUnhovered)
    onSelectedChangedBase=(&self:WidgetTree.ViewSettingsButton_Live.HandleOnSelectedChanged)
    iconBrush {
     bIsDynamicallyLoaded=false
     drawAs=Image
     imageType=NoImage
     mirroring=NoMirror
     resourceName=None
     resourceObject=&T-Icon-Replay-CameraSettings-L
     tiling=NoTile
     imageSize=<32 32>
     margin=<ltrb 0 0 0 0>
     outlineSettings {
      bUseBrushTransparency=false
      roundingType=HalfHeightRadius
      width=0
      color {colorUseRule=UseColor_Specified, specifiedColor=<rgba 0 0 0 0>}
      cornerRadii=<0 0 0 1>
     }
     tintColor {colorUseRule=UseColor_Specified, specifiedColor=<rgba 1 1 1 1>}
     uVRegion {bIsValid=false, max=<0 0>, min=<0 0>}
    }
   }
   slot {class=OverlaySlot, properties {horizontalAlignment=HAlign_Fill, verticalAlignment=VAlign_Fill}}
  }
  CamType_1 {
   widgetClassPath=&Overlay
   parent=HorizontalBox_270
   widget=CamType_1
   properties {bOverride_Cursor=true}
   slot {class=HorizontalBoxSlot, properties {padding=<ltrb 0 0 10 0>}}
  }
  CamTypeBackplate_1 {
   widgetClassPath=&WBP_UIKit_Backplate
   parent=CamType_1
   widget=CamTypeBackplate_1
   properties {
    backplateMaterial=&M_Timeline_Backplate_Inst
    brightness=Medium
    cornerRadius=Tab
    backplateSize=<0 42>
   }
   slot {class=OverlaySlot, properties {horizontalAlignment=HAlign_Fill}}
  }
  CamTypeHorizontalBox_1 {
   widgetClassPath=&HorizontalBox
   parent=CamType_1
   widget=CamTypeHorizontalBox_1
   properties {bOverride_Cursor=true}
   slot {class=OverlaySlot, properties {horizontalAlignment=HAlign_Center}}
  }
  CamName_1 {widgetClassPath=&VerticalBox, parent=CamTypeHorizontalBox_1, widget=CamName_1, slot {class=HorizontalBoxSlot}}
  CamNameSizeOverride_1 {
   widgetClassPath=&SizeBox
   bIsVariable=true
   parent=CamName_1
   widget=CamNameSizeOverride_1
   properties {bOverride_WidthOverride=true, widthOverride=160}
   slot {class=VerticalBoxSlot}
  }
  CurrentCameraButtonSpectator {
   widgetClassPath=&IconTextButtonReplay
   bIsVariable=true
   parent=CamName_1
   widget=CurrentCameraButtonSpectator
   properties {
    bHideText=false
    bIgnoreAnimations=true
    bIsFocusable=false
    bShouldHideImage=true
    buttonText=Drone Attach
    onButtonBaseClicked=()
    onButtonBaseDoubleClicked=()
    onButtonBaseHovered=()
    onButtonBaseUnhovered=()
    onSelectedChangedBase=()
    shouldHideInputAction=true
    iconBrush {
     bIsDynamicallyLoaded=false
     drawAs=Image
     imageType=NoImage
     mirroring=NoMirror
     resourceName=None
     resourceObject=&T-Icon-Replay-CurrentCamera-L
     tiling=NoTile
     imageSize=<32 32>
     margin=<ltrb 0 0 0 0>
     outlineSettings {
      bUseBrushTransparency=false
      roundingType=HalfHeightRadius
      width=0
      color {colorUseRule=UseColor_Specified, specifiedColor=<rgba 0 0 0 0>}
      cornerRadii=<0 0 0 1>
     }
     tintColor {colorUseRule=UseColor_Specified, specifiedColor=<rgba 1 1 1 1>}
     uVRegion {bIsValid=false, max=<0 0>, min=<0 0>}
    }
   }
   slot {class=VerticalBoxSlot}
  }
  CamTypeInputAction_1 {
   widgetClassPath=&CommonActionWidget
   parent=CamName_1
   widget=CamTypeInputAction_1
   properties {inputActions [{dataTable=&ReplayInputActionDataTable, rowName=CycleCameras}]}
   slot {
    class=VerticalBoxSlot
    properties {horizontalAlignment=HAlign_Center, verticalAlignment=VAlign_Bottom, padding=<ltrb 0 5 0 0>, size {sizeRule=Fill, value=1}}
   }
  }
  CamTypeSpectatorContextBox {
   widgetClassPath=&VerticalBox
   bIsVariable=true
   parent=CamTypeHorizontalBox_1
   widget=CamTypeSpectatorContextBox
   slot {class=HorizontalBoxSlot}
  }
  CameraTypeSpectatorContextMenu {
   widgetClassPath=&WBP_UIKit_ContextMenuAnchorButton
   bIsVariable=true
   parent=CamTypeSpectatorContextBox
   widget=CameraTypeSpectatorContextMenu
   properties {
    bIsFocusable=true
    button Background Material=&MI_UIKit_BgQuiet_Fallback
    buttonClass=&IconContextButtonReplay_C
    buttonIcon=&T_UI_Icon_UIKit_ContextMenu
    buttonText=""
    closeMenuOnContextClicked=true
    menuPlacement=MenuPlacement_AboveAnchor
    menuWidgetClass=&WBP_UIKit_Context_Menu_Fallback_C
    min Height=42
    min Width=56
    button Padding=<ltrb -9 0 0 0>
    button Text Font {
     bForceMonospaced=false
     bMaterialIsStencil=false
     fontMaterial=null
     fontObject=&HeadingNow
     letterSpacing=48
     monospacedWidth=1
     size=13.5
     skewAmount=0
     typefaceFontName=74 Regular
     outlineSettings {
      bApplyOutlineToDropShadows=false
      bMiteredCorners=false
      bSeparateFillAlpha=false
      outlineMaterial=null
      outlineSize=0
      outlineColor=<rgba 0 0 0 1>
     }
    }
   }
   slot {class=VerticalBoxSlot}
  }
  CameraTypeSpectatorInputAction {
   widgetClassPath=&CommonActionWidget
   bIsVariable=true
   parent=CamTypeSpectatorContextBox
   widget=CameraTypeSpectatorInputAction
   properties {visibility=Hidden, inputActions [{dataTable=&ReplayInputActionDataTable, rowName=ShowAllPlayers}]}
   slot {
    class=VerticalBoxSlot
    properties {horizontalAlignment=HAlign_Center, verticalAlignment=VAlign_Bottom, padding=<ltrb -9 0 0 0>, size {sizeRule=Fill, value=1}}
   }
  }
  PlaybackSpeedOverlay_Live {
   widgetClassPath=&Overlay
   parent=HorizontalBox_270
   widget=PlaybackSpeedOverlay_Live
   slot {class=HorizontalBoxSlot, properties {padding=<ltrb 0 0 10 0>, size {sizeRule=Fill, value=1}}}
  }
  PlaybackSpeedBackplate_1 {
   widgetClassPath=&WBP_UIKit_Backplate
   parent=PlaybackSpeedOverlay_Live
   widget=PlaybackSpeedBackplate_1
   properties {
    backplateMaterial=&M_Timeline_Backplate_Inst
    brightness=Medium
    cornerRadius=Tab
    backplateSize=<0 42>
   }
   slot {class=OverlaySlot, properties {horizontalAlignment=HAlign_Fill}}
  }
  HorizontalBox {
   widgetClassPath=&HorizontalBox
   parent=PlaybackSpeedOverlay_Live
   widget=HorizontalBox
   slot {class=OverlaySlot, properties {horizontalAlignment=HAlign_Fill, verticalAlignment=VAlign_Fill}}
  }
  Slower_Live {
   widgetClassPath=&IconTextButtonReplay
   bIsVariable=true
   parent=HorizontalBox
   widget=Slower_Live
   properties {
    bDisplayInputActionWhenNotInteractable=false
    bIsFocusable=false
    bShouldUseFallbackDefaultInputAction=false
    bUseCustomWidth=true
    customWidth=52
    onButtonBaseClicked=(&self:WidgetTree.Slower_Live.HandleOnButtonClicked)
    onButtonBaseDoubleClicked=(&self:WidgetTree.Slower_Live.HandleOnButtonDoubleClicked)
    onButtonBaseHovered=(&self:WidgetTree.Slower_Live.HandleOnButtonHovered)
    onButtonBaseUnhovered=(&self:WidgetTree.Slower_Live.HandleOnButtonUnhovered)
    onSelectedChangedBase=(&self:WidgetTree.Slower_Live.HandleOnSelectedChanged)
    shouldHideInputAction=true
    iconBrush {
     bIsDynamicallyLoaded=false
     drawAs=Image
     imageType=NoImage
     mirroring=NoMirror
     resourceName=None
     resourceObject=&T-Icon-Replay-Minus-L
     tiling=NoTile
     imageSize=<32 32>
     margin=<ltrb 0 0 0 0>
     outlineSettings {
      bUseBrushTransparency=false
      roundingType=HalfHeightRadius
      width=0
      color {colorUseRule=UseColor_Specified, specifiedColor=<rgba 0 0 0 0>}
      cornerRadii=<0 0 0 1>
     }
     tintColor {colorUseRule=UseColor_Specified, specifiedColor=<rgba 1 1 1 1>}
     uVRegion {bIsValid=false, max=<0 0>, min=<0 0>}
    }
    pressedSlateSoundOverride {resourceObject=&DVR_Nav_BWD_2x_Cue}
   }
   slot {class=HorizontalBoxSlot}
  }
  TimeDilationSizeBox_1 {
   widgetClassPath=&SizeBox
   parent=HorizontalBox
   widget=TimeDilationSizeBox_1
   properties {bOverride_WidthOverride=true, widthOverride=65}
   slot {class=HorizontalBoxSlot}
  }
  VBTimeDilation_1 {widgetClassPath=&VerticalBox, parent=TimeDilationSizeBox_1, widget=VBTimeDilation_1, slot {class=SizeBoxSlot}}
  TextTimeDilation_Live {
   widgetClassPath=&TextBlock
   bIsVariable=true
   parent=VBTimeDilation_1
   widget=TextTimeDilation_Live
   properties {
    text=1.00x
    font {
     bForceMonospaced=false
     bMaterialIsStencil=false
     fontMaterial=null
     fontObject=&HeadingNow
     letterSpacing=0
     monospacedWidth=1
     size=13.5
     skewAmount=0
     typefaceFontName=74 Regular
     outlineSettings {
      bApplyOutlineToDropShadows=false
      bMiteredCorners=false
      bSeparateFillAlpha=false
      outlineMaterial=null
      outlineSize=0
      outlineColor=<rgba 0 0 0 1>
     }
    }
   }
   slot {class=VerticalBoxSlot, properties {horizontalAlignment=HAlign_Center, verticalAlignment=VAlign_Center, padding=<ltrb 0 6 0 0>}}
  }
  TimeDilationAction_1 {
   widgetClassPath=&CommonActionWidget
   bIsVariable=true
   parent=VBTimeDilation_1
   widget=TimeDilationAction_1
   properties {inputActions [
     {dataTable=&ReplayInputActionDataTable, rowName=IncreasePlaybackSpeed}
     {dataTable=&ReplayInputActionDataTable, rowName=DecreasePlaybackSpeed}
    ]}
   slot {class=VerticalBoxSlot, properties {horizontalAlignment=HAlign_Center, verticalAlignment=VAlign_Bottom, size {sizeRule=Fill, value=1}}}
  }
  Faster_Live {
   widgetClassPath=&IconTextButtonReplay
   bIsVariable=true
   parent=HorizontalBox
   widget=Faster_Live
   properties {
    bDisplayInputActionWhenNotInteractable=false
    bIsFocusable=false
    bShouldUseFallbackDefaultInputAction=false
    bUseCustomWidth=true
    customWidth=52
    onButtonBaseClicked=(&self:WidgetTree.Faster_Live.HandleOnButtonClicked)
    onButtonBaseDoubleClicked=(&self:WidgetTree.Faster_Live.HandleOnButtonDoubleClicked)
    onButtonBaseHovered=(&self:WidgetTree.Faster_Live.HandleOnButtonHovered)
    onButtonBaseUnhovered=(&self:WidgetTree.Faster_Live.HandleOnButtonUnhovered)
    onSelectedChangedBase=(&self:WidgetTree.Faster_Live.HandleOnSelectedChanged)
    shouldHideInputAction=true
    iconBrush {
     bIsDynamicallyLoaded=false
     drawAs=Image
     imageType=NoImage
     mirroring=NoMirror
     resourceName=None
     resourceObject=&T-Icon-Replay-Plus-L
     tiling=NoTile
     imageSize=<32 32>
     margin=<ltrb 0 0 0 0>
     outlineSettings {
      bUseBrushTransparency=false
      roundingType=HalfHeightRadius
      width=0
      color {colorUseRule=UseColor_Specified, specifiedColor=<rgba 0 0 0 0>}
      cornerRadii=<0 0 0 1>
     }
     tintColor {colorUseRule=UseColor_Specified, specifiedColor=<rgba 1 1 1 1>}
     uVRegion {bIsValid=false, max=<0 0>, min=<0 0>}
    }
    pressedSlateSoundOverride {resourceObject=&DVR_Nav_FWD_2x_Cue}
   }
   slot {class=HorizontalBoxSlot}
  }
  ToolsSpectatorOverlay {
   widgetClassPath=&Overlay
   parent=HorizontalBox_270
   widget=ToolsSpectatorOverlay
   slot {class=HorizontalBoxSlot, properties {padding=<ltrb 0 0 10 0>}}
  }
  FocusTimelineBackplate_1 {
   widgetClassPath=&WBP_UIKit_Backplate
   parent=ToolsSpectatorOverlay
   widget=FocusTimelineBackplate_1
   properties {
    backplateMaterial=&M_Timeline_Backplate_Inst
    brightness=Medium
    cornerRadius=Tab
    backplateSize=<0 42>
   }
   slot {class=OverlaySlot, properties {horizontalAlignment=HAlign_Fill}}
  }
  ToolsSpectatorHorizontalBox {widgetClassPath=&HorizontalBox, parent=ToolsSpectatorOverlay, widget=ToolsSpectatorHorizontalBox, slot {class=OverlaySlot}}
  FPSButtonSpacer_Live {
   widgetClassPath=&Spacer
   bIsVariable=true
   parent=ToolsSpectatorHorizontalBox
   widget=FPSButtonSpacer_Live
   properties {size=<8 0>}
   slot {class=HorizontalBoxSlot}
  }
  FPSSettingsButton_Live {
   widgetClassPath=&IconTextButtonReplay
   bIsVariable=true
   parent=ToolsSpectatorHorizontalBox
   widget=FPSSettingsButton_Live
   properties {
    bHideText=false
    bShouldHideImage=true
    bUseCustomWidth=true
    buttonText=30 FPS
    customWidth=85
    onButtonBaseClicked=()
    onButtonBaseDoubleClicked=()
    onButtonBaseHovered=()
    onButtonBaseUnhovered=()
    onSelectedChangedBase=()
    visibility=Collapsed
    iconBrush {
     bIsDynamicallyLoaded=false
     drawAs=Image
     imageType=FullColor
     mirroring=NoMirror
     resourceName=None
     resourceObject=null
     tiling=NoTile
     imageSize=<32 32>
     margin=<ltrb 0 0 0 0>
     outlineSettings {
      bUseBrushTransparency=false
      roundingType=HalfHeightRadius
      width=0
      color {colorUseRule=UseColor_Specified, specifiedColor=<rgba 0 0 0 0>}
      cornerRadii=<0 0 0 1>
     }
     tintColor {colorUseRule=UseColor_Specified, specifiedColor=<rgba 1 1 1 1>}
     uVRegion {bIsValid=false, max=<0 0>, min=<0 0>}
    }
   }
   slot {class=HorizontalBoxSlot}
  }
  BlurSettingsButton_Live {
   widgetClassPath=&IconTextButtonReplay
   bIsVariable=true
   parent=ToolsSpectatorHorizontalBox
   widget=BlurSettingsButton_Live
   properties {
    bIgnoreAnimations=true
    bUseCustomWidth=true
    customWidth=54
    onButtonBaseClicked=()
    onButtonBaseDoubleClicked=()
    onButtonBaseHovered=()
    onButtonBaseUnhovered=()
    onSelectedChangedBase=()
    visibility=Collapsed
    iconBrush {
     bIsDynamicallyLoaded=false
     drawAs=Image
     imageType=FullColor
     mirroring=NoMirror
     resourceName=None
     resourceObject=&T_Icon_Replay_MotionBlur_Off_C
     tiling=NoTile
     imageSize=<32 32>
     margin=<ltrb 0 0 0 0>
     outlineSettings {
      bUseBrushTransparency=false
      roundingType=HalfHeightRadius
      width=0
      color {colorUseRule=UseColor_Specified, specifiedColor=<rgba 0 0 0 0>}
      cornerRadii=<0 0 0 1>
     }
     tintColor {colorUseRule=UseColor_Specified, specifiedColor=<rgba 1 1 1 1>}
     uVRegion {bIsValid=false, max=<0 0>, min=<0 0>}
    }
   }
   slot {class=HorizontalBoxSlot}
  }
  TimecodeButton_Live {
   widgetClassPath=&IconTextButtonReplay
   bIsVariable=true
   parent=ToolsSpectatorHorizontalBox
   widget=TimecodeButton_Live
   properties {
    bDisplayInputActionWhenNotInteractable=false
    onButtonBaseClicked=()
    onButtonBaseDoubleClicked=()
    onButtonBaseHovered=()
    onButtonBaseUnhovered=()
    onSelectedChangedBase=()
    iconBrush {
     bIsDynamicallyLoaded=false
     drawAs=Image
     imageType=FullColor
     mirroring=NoMirror
     resourceName=None
     resourceObject=/Game/Athena/UI/InGame/Creative/PropertyEditor/Icon_Picker/Textures_V2/64x64/T_UI_IconLibrary_Stopwatch_64
     tiling=NoTile
     imageSize=<32 32>
     margin=<ltrb 0 0 0 0>
     outlineSettings {
      bUseBrushTransparency=false
      roundingType=HalfHeightRadius
      width=0
      color {colorUseRule=UseColor_Specified, specifiedColor=<rgba 0 0 0 0>}
      cornerRadii=<0 0 0 1>
     }
     tintColor {colorUseRule=UseColor_Specified, specifiedColor=<rgba 1 1 1 1>}
     uVRegion {bIsValid=false, max=<0 0>, min=<0 0>}
    }
   }
   slot {class=HorizontalBoxSlot}
  }
  SpectatorEjectButton_Live {
   widgetClassPath=&IconTextButtonReplay
   bIsVariable=true
   parent=ToolsSpectatorHorizontalBox
   widget=SpectatorEjectButton_Live
   properties {
    onButtonBaseClicked=()
    onButtonBaseDoubleClicked=()
    onButtonBaseHovered=()
    onButtonBaseUnhovered=()
    onSelectedChangedBase=()
    iconBrush {
     bIsDynamicallyLoaded=false
     drawAs=Image
     imageType=FullColor
     mirroring=NoMirror
     resourceName=None
     resourceObject=/Game/UI/Foundation/Textures/Icons/Replay/T_Icon_Replay_SpectatorEject
     tiling=NoTile
     imageSize=<32 32>
     margin=<ltrb 0 0 0 0>
     outlineSettings {
      bUseBrushTransparency=false
      roundingType=HalfHeightRadius
      width=0
      color {colorUseRule=UseColor_Specified, specifiedColor=<rgba 0 0 0 0>}
      cornerRadii=<0 0 0 1>
     }
     tintColor {colorUseRule=UseColor_Specified, specifiedColor=<rgba 1 1 1 1>}
     uVRegion {bIsValid=false, max=<0 0>, min=<0 0>}
    }
   }
   slot {class=HorizontalBoxSlot}
  }
  FreeDroneTargetHidingToolUndo_Live {
   widgetClassPath=&IconTextButtonReplay
   bIsVariable=true
   parent=ToolsSpectatorHorizontalBox
   widget=FreeDroneTargetHidingToolUndo_Live
   properties {
    onButtonBaseClicked=()
    onButtonBaseDoubleClicked=()
    onButtonBaseHovered=()
    onButtonBaseUnhovered=()
    onSelectedChangedBase=()
    visibility=Collapsed
    iconBrush {
     bIsDynamicallyLoaded=false
     drawAs=Image
     imageType=FullColor
     mirroring=NoMirror
     resourceName=None
     resourceObject=&T-Icon-Replay-ToggleUI-L
     tiling=NoTile
     imageSize=<32 32>
     margin=<ltrb 0 0 0 0>
     outlineSettings {
      bUseBrushTransparency=false
      roundingType=HalfHeightRadius
      width=0
      color {colorUseRule=UseColor_Specified, specifiedColor=<rgba 0 0 0 0>}
      cornerRadii=<0 0 0 1>
     }
     tintColor {colorUseRule=UseColor_Specified, specifiedColor=<rgba 1 1 1 1>}
     uVRegion {bIsValid=false, max=<0 0>, min=<0 0>}
    }
   }
   slot {class=HorizontalBoxSlot}
  }
  FreeDroneTargetHidingToolSelect_Live {
   widgetClassPath=&IconTextButtonReplay
   bIsVariable=true
   parent=ToolsSpectatorHorizontalBox
   widget=FreeDroneTargetHidingToolSelect_Live
   properties {
    onButtonBaseClicked=()
    onButtonBaseDoubleClicked=()
    onButtonBaseHovered=()
    onButtonBaseUnhovered=()
    onSelectedChangedBase=()
    visibility=Collapsed
    iconBrush {
     bIsDynamicallyLoaded=false
     drawAs=Image
     imageType=FullColor
     mirroring=NoMirror
     resourceName=None
     resourceObject=&T_Icon_Replay_HidingTool_M
     tiling=NoTile
     imageSize=<32 32>
     margin=<ltrb 0 0 0 0>
     outlineSettings {
      bUseBrushTransparency=false
      roundingType=HalfHeightRadius
      width=0
      color {colorUseRule=UseColor_Specified, specifiedColor=<rgba 0 0 0 0>}
      cornerRadii=<0 0 0 1>
     }
     tintColor {colorUseRule=UseColor_Specified, specifiedColor=<rgba 1 1 1 1>}
     uVRegion {bIsValid=false, max=<0 0>, min=<0 0>}
    }
   }
   slot {class=HorizontalBoxSlot}
  }
  CameraPlacementOverlay {widgetClassPath=&Overlay, parent=HorizontalBox_270, widget=CameraPlacementOverlay, slot {class=HorizontalBoxSlot}}
  FocusTimelineBackplate_2 {
   widgetClassPath=&WBP_UIKit_Backplate
   parent=CameraPlacementOverlay
   widget=FocusTimelineBackplate_2
   properties {
    backplateMaterial=&M_Timeline_Backplate_Inst
    brightness=Medium
    cornerRadius=Tab
    backplateSize=<0 42>
   }
   slot {class=OverlaySlot, properties {horizontalAlignment=HAlign_Fill}}
  }
  CameraPlacementLiveHorizontalBox {widgetClassPath=&HorizontalBox, parent=CameraPlacementOverlay, widget=CameraPlacementLiveHorizontalBox, slot {class=OverlaySlot}}
  PreviousPlayerButton_Live {
   widgetClassPath=&IconTextButtonReplay
   bIsVariable=true
   parent=CameraPlacementLiveHorizontalBox
   widget=PreviousPlayerButton_Live
   properties {
    onButtonBaseClicked=(&self:WidgetTree.PreviousPlayerButton_Live.HandleOnButtonClicked)
    onButtonBaseDoubleClicked=(&self:WidgetTree.PreviousPlayerButton_Live.HandleOnButtonDoubleClicked)
    onButtonBaseHovered=(&self:WidgetTree.PreviousPlayerButton_Live.HandleOnButtonHovered)
    onButtonBaseUnhovered=(&self:WidgetTree.PreviousPlayerButton_Live.HandleOnButtonUnhovered)
    onSelectedChangedBase=(&self:WidgetTree.PreviousPlayerButton_Live.HandleOnSelectedChanged)
    iconBrush {
     bIsDynamicallyLoaded=false
     drawAs=Image
     imageType=NoImage
     mirroring=NoMirror
     resourceName=None
     resourceObject=&T-Icon-PlayerSelectLeft
     tiling=NoTile
     imageSize=<32 32>
     margin=<ltrb 0 0 0 0>
     outlineSettings {
      bUseBrushTransparency=false
      roundingType=HalfHeightRadius
      width=0
      color {colorUseRule=UseColor_Specified, specifiedColor=<rgba 0 0 0 0>}
      cornerRadii=<0 0 0 1>
     }
     tintColor {colorUseRule=UseColor_Specified, specifiedColor=<rgba 1 1 1 1>}
     uVRegion {bIsValid=false, max=<0 0>, min=<0 0>}
    }
   }
   slot {class=HorizontalBoxSlot}
  }
  SizeBox_4 {
   widgetClassPath=&SizeBox
   parent=CameraPlacementLiveHorizontalBox
   widget=SizeBox_4
   properties {bOverride_WidthOverride=true, maxDesiredWidth=300, minDesiredWidth=50, widthOverride=175}
   slot {class=HorizontalBoxSlot}
  }
  HorizontalBox_3 {
   widgetClassPath=&HorizontalBox
   parent=SizeBox_4
   widget=HorizontalBox_3
   slot {class=SizeBoxSlot, properties {horizontalAlignment=HAlign_Center, verticalAlignment=VAlign_Top}}
  }
  FollowedPlayerText_Live {
   widgetClassPath=&CommonTextBlock
   bIsVariable=true
   parent=HorizontalBox_3
   widget=FollowedPlayerText_Live
   properties {
    scrollStyle=&TextScrollStyle-Base_C
    text="{Followed Player}"
    font {
     bForceMonospaced=false
     bMaterialIsStencil=false
     fontMaterial=null
     fontObject=&HeadingNow
     letterSpacing=0
     monospacedWidth=1
     size=13.5
     skewAmount=0
     typefaceFontName=74 Regular
     outlineSettings {
      bApplyOutlineToDropShadows=false
      bMiteredCorners=false
      bSeparateFillAlpha=false
      outlineMaterial=null
      outlineSize=0
      outlineColor=<rgba 0 0 0 1>
     }
    }
    shadowOffset=<0 0>
   }
   slot {class=HorizontalBoxSlot, properties {horizontalAlignment=HAlign_Center, padding=<ltrb 5 10.5 5 0>, size {sizeRule=Fill, value=1}}}
  }
  PlayerSelectorSpectatorContextBox {
   widgetClassPath=&VerticalBox
   bIsVariable=true
   parent=CameraPlacementLiveHorizontalBox
   widget=PlayerSelectorSpectatorContextBox
   properties {visibility=Collapsed}
   slot {class=HorizontalBoxSlot}
  }
  PlayerSelectorSpectatorContextMenu {
   widgetClassPath=&WBP_UIKit_ContextMenuAnchorButton
   bIsVariable=true
   parent=PlayerSelectorSpectatorContextBox
   widget=PlayerSelectorSpectatorContextMenu
   properties {
    bIsFocusable=true
    button Background Material=&MI_UIKit_BgQuiet_Fallback
    buttonClass=&IconContextButtonReplay_C
    buttonIcon=&T_UI_Icon_UIKit_ContextMenu
    buttonText=""
    closeMenuOnContextClicked=true
    menuPlacement=MenuPlacement_RightLeftCenter
    menuWidgetClass=&WBP_UIKit_Context_Menu_Fallback_C
    min Height=42
    min Width=56
    button Padding=<ltrb 0 0 0 0>
    button Text Font {
     bForceMonospaced=false
     bMaterialIsStencil=false
     fontMaterial=null
     fontObject=&HeadingNow
     letterSpacing=48
     monospacedWidth=1
     size=13.5
     skewAmount=0
     typefaceFontName=74 Regular
     outlineSettings {
      bApplyOutlineToDropShadows=false
      bMiteredCorners=false
      bSeparateFillAlpha=false
      outlineMaterial=null
      outlineSize=0
      outlineColor=<rgba 0 0 0 1>
     }
    }
   }
   slot {class=VerticalBoxSlot}
  }
  PlayerSelectorSpectatorInputAction {
   widgetClassPath=&CommonActionWidget
   bIsVariable=true
   parent=PlayerSelectorSpectatorContextBox
   widget=PlayerSelectorSpectatorInputAction
   properties {visibility=Hidden, inputActions [{dataTable=&ReplayInputActionDataTable, rowName=ShowAllPlayers}]}
   slot {class=VerticalBoxSlot, properties {horizontalAlignment=HAlign_Center, verticalAlignment=VAlign_Bottom, size {sizeRule=Fill, value=1}}}
  }
  NextPlayerButton_Live {
   widgetClassPath=&IconTextButtonReplay
   bIsVariable=true
   parent=CameraPlacementLiveHorizontalBox
   widget=NextPlayerButton_Live
   properties {
    onButtonBaseClicked=(&self:WidgetTree.NextPlayerButton_Live.HandleOnButtonClicked)
    onButtonBaseDoubleClicked=(&self:WidgetTree.NextPlayerButton_Live.HandleOnButtonDoubleClicked)
    onButtonBaseHovered=(&self:WidgetTree.NextPlayerButton_Live.HandleOnButtonHovered)
    onButtonBaseUnhovered=(&self:WidgetTree.NextPlayerButton_Live.HandleOnButtonUnhovered)
    onSelectedChangedBase=(&self:WidgetTree.NextPlayerButton_Live.HandleOnSelectedChanged)
    iconBrush {
     bIsDynamicallyLoaded=false
     drawAs=Image
     imageType=NoImage
     mirroring=NoMirror
     resourceName=None
     resourceObject=&T-Icon-PlayerSelectRight
     tiling=NoTile
     imageSize=<32 32>
     margin=<ltrb 0 0 0 0>
     outlineSettings {
      bUseBrushTransparency=false
      roundingType=HalfHeightRadius
      width=0
      color {colorUseRule=UseColor_Specified, specifiedColor=<rgba 0 0 0 0>}
      cornerRadii=<0 0 0 1>
     }
     tintColor {colorUseRule=UseColor_Specified, specifiedColor=<rgba 1 1 1 1>}
     uVRegion {bIsValid=false, max=<0 0>, min=<0 0>}
    }
   }
   slot {class=HorizontalBoxSlot}
  }
  ControlsWrapper {
   widgetClassPath=&FortHUDElementWrapper
   bIsVariable=true
   parent=ContentVB
   widget=ControlsWrapper
   properties {elementTag {tagName=Athena.Replay.Server}, hUDElementTag {gameplayTags [{tagName=Athena.Replay.Live}]}}
   slot {class=VerticalBoxSlot, properties {padding=<ltrb 0 0 0 10>}}
  }
  ControlsSizeBox {
   widgetClassPath=&SizeBox
   namedSlotHost=ControlsWrapper
   widget=ControlsSizeBox
   properties {heightOverride=53, minDesiredHeight=47}
  }
  ControlOverlaySwitcher {
   widgetClassPath=&CommonWidgetSwitcherLegacy
   bIsVariable=true
   parent=ControlsSizeBox
   widget=ControlOverlaySwitcher
   slot {class=SizeBoxSlot, properties {horizontalAlignment=HAlign_Center}}
  }
  HBoxDroneCams {
   widgetClassPath=&HorizontalBox
   parent=ControlOverlaySwitcher
   widget=HBoxDroneCams
   slot {class=WidgetSwitcherSlot, properties {horizontalAlignment=HAlign_Right}}
  }
  KeybindOrbit {
   widgetClassPath=/FortUILibrary/Foundation/Buttons/KeybindingReplay
   bIsVariable=true
   parent=HBoxDroneCams
   widget=KeybindOrbit
   properties {
    axisScale01=1
    axisScale02=-1
    axisScale03=1
    axisScale04=-1
    bUseImages=true
    bUseSpecificControllerImage=true
    boundActionName01=MoveForward
    boundActionName02=MoveRight
    boundActionName03=MoveForward
    boundActionName04=MoveRight
    controllerBrush=/FortUILibrary/Input/GamepadPS4/Source/T_PS4_LStick
    imagesVisible=4
    keybindsVisible=4
    switcherImage01=/FortUILibrary/Input/Keyboard/MaterialInstances/MI_UI_Input_Keyboard_W
    switcherImage02=/FortUILibrary/Input/Keyboard/MaterialInstances/MI_UI_Input_Keyboard_A
    switcherImage03=/FortUILibrary/Input/Keyboard/MaterialInstances/MI_UI_Input_Keyboard_S
    switcherImage04=/FortUILibrary/Input/Keyboard/MaterialInstances/MI_UI_Input_Keyboard_D
    title=Orbit
   }
   slot {class=HorizontalBoxSlot, properties {verticalAlignment=VAlign_Center, padding=<ltrb 0 0 10 0>}}
  }
  KeybindPanTilt {
   widgetClassPath=&KeybindingReplay
   bIsVariable=true
   parent=HBoxDroneCams
   widget=KeybindPanTilt
   properties {
    bUseImages=true
    bUseSpecificControllerImage=true
    controllerBrush=/FortUILibrary/Input/GamepadPS4/Source/T_PS4_RStick
    imagesVisible=2
    switcherImage01=/FortUILibrary/Input/Mouse/MaterialInstances/MI_UI_Input_Mouse_LeftClick
    switcherImage02=/FortUILibrary/Input/Mouse/MaterialInstances/MI_UI_Input_Mouse_RightClick
    title=Pan/Tilt
   }
   slot {class=HorizontalBoxSlot, properties {verticalAlignment=VAlign_Center, padding=<ltrb 0 0 10 0>}}
  }
  KeybindSpeed {
   widgetClassPath=&KeybindingReplay
   bIsVariable=true
   parent=HBoxDroneCams
   widget=KeybindSpeed
   properties {
    bUseImages=true
    boundActionName01=HoverDrone_DecreaseSpeedOrDutchAngle
    boundActionName02=HoverDrone_IncreaseSpeedOrDutchAngle
    imagesVisible=2
    keybindsVisible=2
    switcherImage01=/FortUILibrary/Input/Keyboard/MaterialInstances/MI_UI_Input_Keyboard_Minus
    switcherImage02=/FortUILibrary/Input/Keyboard/MaterialInstances/MI_UI_Input_Keyboard_Equal
    title=Speed
   }
   slot {class=HorizontalBoxSlot, properties {verticalAlignment=VAlign_Center, padding=<ltrb 0 0 10 0>}}
  }
  KeybindHeight {
   widgetClassPath=&KeybindingReplay
   bIsVariable=true
   parent=HBoxDroneCams
   widget=KeybindHeight
   properties {
    bUseImages=true
    boundActionName01=ReplayUIDecreaseDistance
    boundActionName02=ReplayUIIncreaseDistance
    imagesVisible=2
    keybindsVisible=2
    switcherImage01=/FortUILibrary/Input/Keyboard/MaterialInstances/MI_UI_Input_Keyboard_Q
    switcherImage02=/FortUILibrary/Input/Keyboard/MaterialInstances/MI_UI_Input_Keyboard_E
    title=Height
   }
   slot {class=HorizontalBoxSlot, properties {verticalAlignment=VAlign_Center, padding=<ltrb 0 0 10 0>}}
  }
  KeybindZoom {
   widgetClassPath=&KeybindingReplay
   bIsVariable=true
   parent=HBoxDroneCams
   widget=KeybindZoom
   properties {
    axisScale01=1
    axisScale02=-1
    bUseImages=true
    boundActionName01=SpectorRackOrDutchOrZoomAdjust
    boundActionName02=SpectorRackOrDutchOrZoomAdjust
    imagesVisible=1
    keybindsVisible=2
    switcherImage01=/FortUILibrary/Input/Mouse/MaterialInstances/MI_UI_Input_Mouse_ScrollVertical
    title=Zoom
   }
   slot {class=HorizontalBoxSlot, properties {verticalAlignment=VAlign_Center, padding=<ltrb 0 0 10 0>}}
  }
  KeybindHide {
   widgetClassPath=&KeybindingReplay
   bIsVariable=true
   parent=HBoxDroneCams
   widget=KeybindHide
   properties {
    actionWidgetBoundAction=ToggleVisibility
    bUseImages=true
    bUseSwitcher=false
    imagesVisible=1
    switcherImage01=/FortUILibrary/Input/Keyboard/MaterialInstances/MI_UI_Input_Keyboard_H
    title=Show/Hide UI
   }
   slot {class=HorizontalBoxSlot, properties {verticalAlignment=VAlign_Center}}
  }
  HBoxThirdPerson {
   widgetClassPath=&HorizontalBox
   parent=ControlOverlaySwitcher
   widget=HBoxThirdPerson
   slot {class=WidgetSwitcherSlot, properties {horizontalAlignment=HAlign_Right}}
  }
  KeybindPanTilt_1 {
   widgetClassPath=&KeybindingReplay
   bIsVariable=true
   parent=HBoxThirdPerson
   widget=KeybindPanTilt_1
   properties {
    bUseImages=true
    bUseSpecificControllerImage=true
    controllerBrush=&T_PS4_RStick
    imagesVisible=2
    switcherImage01=&MI_UI_Input_Mouse_LeftClick
    switcherImage02=&MI_UI_Input_Mouse_RightClick
    title=Position
   }
   slot {class=HorizontalBoxSlot, properties {verticalAlignment=VAlign_Center, padding=<ltrb 0 0 10 0>}}
  }
  KeybindTPZoom {
   widgetClassPath=&KeybindingReplay
   bIsVariable=true
   parent=HBoxThirdPerson
   widget=KeybindTPZoom
   properties {
    axisScale01=1
    axisScale02=-1
    bUseImages=true
    boundActionName01=SpectorRackOrDutchOrZoomAdjust
    boundActionName02=SpectorRackOrDutchOrZoomAdjust
    imagesVisible=1
    keybindsVisible=2
    switcherImage01=&MI_UI_Input_Mouse_ScrollVertical
    title=Zoom
   }
   slot {class=HorizontalBoxSlot, properties {verticalAlignment=VAlign_Center, padding=<ltrb 0 0 10 0>}}
  }
  KeybindDistance {
   widgetClassPath=&KeybindingReplay
   bIsVariable=true
   parent=HBoxThirdPerson
   widget=KeybindDistance
   properties {
    bUseImages=true
    boundActionName01=ReplayUIDecreaseDistance
    boundActionName02=ReplayUIIncreaseDistance
    imagesVisible=2
    keybindsVisible=2
    switcherImage01=&MI_UI_Input_Keyboard_Q
    switcherImage02=&MI_UI_Input_Keyboard_E
    title=Distance
   }
   slot {class=HorizontalBoxSlot, properties {verticalAlignment=VAlign_Center, padding=<ltrb 0 0 10 0>}}
  }
  KeybindHide_1 {
   widgetClassPath=&KeybindingReplay
   bIsVariable=true
   parent=HBoxThirdPerson
   widget=KeybindHide_1
   properties {
    actionWidgetBoundAction=ToggleVisibility
    bUseImages=true
    bUseSwitcher=false
    imagesVisible=1
    switcherImage01=&MI_UI_Input_Keyboard_H
    title=Show/Hide UI
   }
   slot {class=HorizontalBoxSlot, properties {verticalAlignment=VAlign_Center}}
  }
  SizeBox {widgetClassPath=&SizeBox, parent=ControlOverlaySwitcher, widget=SizeBox, slot {class=WidgetSwitcherSlot}}
 ]
}
