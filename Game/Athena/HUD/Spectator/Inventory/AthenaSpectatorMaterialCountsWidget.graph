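# Spectator HUD widget, derived from AthenaSpectatorMaterialCountsBase, that temporarily shows gold, wood, stone and metal counts.
# Count increases from its four child count widgets, or entering build mode, reset a display timer and reveal it if hidden.
# Tick runs the timer down outside build mode, then FadeOut collapses it; replay scrubs or followed-player changes collapse it immediately.
# Building-material changes highlight the matching wood, stone or metal border, with a design-time preview in PreConstruct.
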WidgetBlueprint /Game/Athena/HUD/Spectator/Inventory/AthenaSpectatorMaterialCountsWidget
animations [FadeOut {
  binding_count=1
  length_seconds=0.25
  object=&self:FadeOut
  sequence {bindings [AthenaSpectatorMaterialCountsWidget_C_20 {
     id=#1
     possessed_class=AthenaSpectatorMaterialCountsWidget_C
     tracks [{
       type=MovieSceneFloatTrack
       display_name=Render Opacity
       sections [{type=MovieSceneFloatSection, end_frame=5, start_frame=0, channels [None {type=MovieSceneScriptingFloatChannel, keys [[0,1],[5,0]]}]}]
      }]
    }]}
 }]
graphs [EventGraph {nodes [
   n1 = OnBuildCountIncreased() {is=title K2Node_CustomEvent}
   Set RemainingTimeToDisplay(
    execute=[@n1.then, @n2.then]
    RemainingTimeToDisplay:double=TimeToDisplay!(self:Athena Spectator Material Counts Base Object Reference) -> (ReturnValue:float)
   ) -> (Output_Get:double)
   > Branch(Condition=NOT Boolean(A=IsVisible()))
   > SetRenderOpacity(InOpacity=1)
   > SetVisibility(InVisibility=SelfHitTestInvisible)
   Event OnBuildModeChanged() {is=title K2Node_Event}
   > n2 = Branch(Condition=Get bIsInBuildMode() -> (bIsInBuildMode:bool))
   Event Construct()
   > SetVisibility(InVisibility=Collapsed)
   > Bind Event to On Build Count Increased(
    self:Athena Spectator Build Count Widget Object Reference=[
     Get Gold() -> (Gold:Athena Spectator Build Count Widget Object Reference)
     Get Metal() -> (Metal:Athena Spectator Build Count Widget Object Reference)
     Get Wood() -> (Wood:Athena Spectator Build Count Widget Object Reference)
     Get Stone() -> (Stone:Athena Spectator Build Count Widget Object Reference)
    ]
    Delegate=Create Event()
   ) {is=title K2Node_AddDelegate}
   n3 = Event OnReplayScrubComplete() {is=title K2Node_Event}
   n4 = StopAllAnimations(execute=[@n3.then, @n9.then], self:User Widget Object Reference)
   SetVisibility(execute=[@n4.then, @n15.then], InVisibility=Collapsed)
   n5 = Event Tick()
   n6 = Get RemainingTimeToDisplay() -> (RemainingTimeToDisplay:double)
   n7 = Get bIsInBuildMode()
   > Branch(Condition=AND Boolean(A=NOT Boolean(A=@n7), B="float > float"(A=@n6, B=0) {is=title K2Node_CallFunction}))
   > n8 = Set RemainingTimeToDisplay(RemainingTimeToDisplay=float - float(A=@n6, B=@n5.InDeltaTime) {is=title K2Node_CallFunction})
   > Branch(Condition="float <= float"!(A:double=@n8, B:double=0) -> (ReturnValue:bool) {is=title K2Node_CallFunction})
   > PlayAnimation(InAnimation=Get FadeOut() -> (FadeOut:Widget Animation Object Reference))
   n9 = Event OnFollowedPlayerChanged!() -> (OutputDelegate, then, SpectatingPC:Fort Player Controller Spectating Object Reference, FollowedPS:Fort Player State Object Reference) {is=title K2Node_Event}
   n10 = Event OnBuildingMaterialChanged!() -> (OutputDelegate, then, BuildingMaterial:EFortResourceType Enum=Wood) {is=title K2Node_Event}
   n11 = Get NonHighlightedColor() -> (NonHighlightedColor:Linear Color Structure)
   > SetBrushColor(
    self=Get WoodBorder() -> (WoodBorder:Common Border Object Reference)
    InBrushColor=Select!(
     Wood:Linear Color Structure=Get HighlightedColor() -> (HighlightedColor:Linear Color Structure)
     Stone:Linear Color Structure=@n11
     Metal:Linear Color Structure=@n11
     Permanite:Linear Color Structure=@n11
     GoldCurrency:Linear Color Structure
     Ingredient:Linear Color Structure
     None:Linear Color Structure=@n11
     Index:EFortResourceType Enum=@n10.BuildingMaterial
    ) -> (ReturnValue:Linear Color Structure)
   )
   n12 = Get NonHighlightedColor()
   > SetBrushColor(
    self=Get StoneBorder() -> (StoneBorder:Common Border Object Reference)
    InBrushColor=Select(Wood=@n12, Stone=Get HighlightedColor(), Metal=@n12, Permanite=@n12, None=@n12, Index=@n10.BuildingMaterial)
   )
   n13 = Get NonHighlightedColor()
   > SetBrushColor(
    self=Get MetalBorder() -> (MetalBorder:Common Border Object Reference)
    InBrushColor=Select(Wood=@n13, Stone=@n13, Metal=Get HighlightedColor(), Permanite=@n13, None=@n13, Index=@n10.BuildingMaterial)
   )
   n14 = Event PreConstruct()
   > OnBuildingMaterialChanged!(
    execute
    self:Athena Spectator Material Counts Base Object Reference
    BuildingMaterial:EFortResourceType Enum=Select!(
     Option 0:EFortResourceType Enum=None
     Option 1:EFortResourceType Enum=Get DesignTime_BuildingMaterial() -> (DesignTime_BuildingMaterial:EFortResourceType Enum=Wood)
     Index:bool=@n14.IsDesignTime
    ) -> (ReturnValue:EFortResourceType Enum)
   ) -> (then)
   n15 = "Animation Finished (FadeOut)"!() -> (OutputDelegate:Delegate, then:Exec) {is=title K2Node_WidgetAnimationEvent}
  ]}]
variables [
 DesignTime_BuildingMaterial {type=EFortResourceType, default=None}
 HighlightedColor {type=LinearColor, default=<rgba 1 1 1 0.3>}
 NonHighlightedColor {type=LinearColor, default=<rgba 0 0 0 0.298>}
 RemainingTimeToDisplay {type=number}
]
widgets {
 info {
  inheritedWidgetCount=1
  parentClass=/Script/FortniteUI.AthenaSpectatorMaterialCountsBase
  rootWidgetClass=&SizeBox
  widgetCount=11
 }
 widgets [
  SizeBox_0 {
   widgetClassPath=&SizeBox
   widget=SizeBox_0
   properties {bOverride_MinDesiredHeight=true, maxDesiredHeight=148, minDesiredHeight=148}
  }
  MaterialVert {widgetClassPath=&VerticalBox, parent=SizeBox_0, widget=MaterialVert, slot {class=SizeBoxSlot}}
  GoldBorder {
   widgetClassPath=&CommonBorder
   bIsVariable=true
   parent=MaterialVert
   widget=GoldBorder
   properties {
    background {
     bIsDynamicallyLoaded=false
     drawAs=Box
     imageType=NoImage
     mirroring=NoMirror
     resourceName=None
     resourceObject=null
     tiling=NoTile
     imageSize=<32 32>
     margin=<ltrb 0 0 0 0>
     outlineSettings {
      bUseBrushTransparency=false
      roundingType=HalfHeightRadius
      width=0
      color {colorUseRule=UseColor_Specified, specifiedColor=<rgba 0 0 0 0>}
      cornerRadii=<0 0 0 0>
     }
     tintColor {colorUseRule=UseColor_Specified, specifiedColor=<rgba 1 1 1 1>}
     uVRegion {bIsValid=false, max=<0 0>, min=<0 0>}
    }
    brushColor=<rgba 0 0 0 0.298>
    padding=<ltrb 0 0 0 0>
   }
   slot {class=VerticalBoxSlot, properties {padding=<ltrb 0 0 0 6>, size {sizeRule=Fill, value=1}}}
  }
  Gold {
   widgetClassPath=&/AthenaSpectatorBuildCountWidget
   bIsVariable=true
   parent=GoldBorder
   widget=Gold
   properties {buildCountType=Gold, itemDefinition=/Game/Items/ResourcePickups/Athena_WadsItemData}
   slot {class=BorderSlot, properties {padding=<ltrb 0 0 0 0>}}
  }
  WoodBorder {
   widgetClassPath=&CommonBorder
   bIsVariable=true
   parent=MaterialVert
   widget=WoodBorder
   properties {
    background {
     bIsDynamicallyLoaded=false
     drawAs=Box
     imageType=NoImage
     mirroring=NoMirror
     resourceName=None
     resourceObject=null
     tiling=NoTile
     imageSize=<32 32>
     margin=<ltrb 0 0 0 0>
     outlineSettings {
      bUseBrushTransparency=false
      roundingType=HalfHeightRadius
      width=0
      color {colorUseRule=UseColor_Specified, specifiedColor=<rgba 0 0 0 0>}
      cornerRadii=<0 0 0 0>
     }
     tintColor {colorUseRule=UseColor_Specified, specifiedColor=<rgba 1 1 1 1>}
     uVRegion {bIsValid=false, max=<0 0>, min=<0 0>}
    }
    brushColor=<rgba 0 0 0 0.298>
    padding=<ltrb 0 0 0 0>
   }
   slot {class=VerticalBoxSlot, properties {padding=<ltrb 0 0 0 6>, size {sizeRule=Fill, value=1}}}
  }
  Wood {
   widgetClassPath=&AthenaSpectatorBuildCountWidget
   bIsVariable=true
   parent=WoodBorder
   widget=Wood
   properties {buildCountType=Wood}
   slot {class=BorderSlot, properties {padding=<ltrb 0 0 0 0>}}
  }
  StoneBorder {
   widgetClassPath=&CommonBorder
   bIsVariable=true
   parent=MaterialVert
   widget=StoneBorder
   properties {
    background {
     bIsDynamicallyLoaded=false
     drawAs=Box
     imageType=NoImage
     mirroring=NoMirror
     resourceName=None
     resourceObject=null
     tiling=NoTile
     imageSize=<32 32>
     margin=<ltrb 0 0 0 0>
     outlineSettings {
      bUseBrushTransparency=false
      roundingType=HalfHeightRadius
      width=0
      color {colorUseRule=UseColor_Specified, specifiedColor=<rgba 0 0 0 0>}
      cornerRadii=<0 0 0 0>
     }
     tintColor {colorUseRule=UseColor_Specified, specifiedColor=<rgba 1 1 1 1>}
     uVRegion {bIsValid=false, max=<0 0>, min=<0 0>}
    }
    brushColor=<rgba 0 0 0 0.298>
    padding=<ltrb 0 0 0 0>
   }
   slot {class=VerticalBoxSlot, properties {padding=<ltrb 0 0 0 6>, size {sizeRule=Fill, value=1}}}
  }
  Stone {
   widgetClassPath=&AthenaSpectatorBuildCountWidget
   bIsVariable=true
   parent=StoneBorder
   widget=Stone
   properties {buildCountType=Stone}
   slot {class=BorderSlot, properties {padding=<ltrb 0 0 0 0>}}
  }
  MetalBorder {
   widgetClassPath=&CommonBorder
   bIsVariable=true
   parent=MaterialVert
   widget=MetalBorder
   properties {
    background {
     bIsDynamicallyLoaded=false
     drawAs=Box
     imageType=NoImage
     mirroring=NoMirror
     resourceName=None
     resourceObject=null
     tiling=NoTile
     imageSize=<32 32>
     margin=<ltrb 0 0 0 0>
     outlineSettings {
      bUseBrushTransparency=false
      roundingType=HalfHeightRadius
      width=0
      color {colorUseRule=UseColor_Specified, specifiedColor=<rgba 0 0 0 0>}
      cornerRadii=<0 0 0 0>
     }
     tintColor {colorUseRule=UseColor_Specified, specifiedColor=<rgba 1 1 1 1>}
     uVRegion {bIsValid=false, max=<0 0>, min=<0 0>}
    }
    brushColor=<rgba 0 0 0 0.298>
    padding=<ltrb 0 0 0 0>
   }
   slot {class=VerticalBoxSlot, properties {size {sizeRule=Fill, value=1}}}
  }
  Metal {
   widgetClassPath=&AthenaSpectatorBuildCountWidget
   bIsVariable=true
   parent=MetalBorder
   widget=Metal
   properties {buildCountType=Metal}
   slot {class=BorderSlot, properties {padding=<ltrb 0 0 0 0>}}
  }
  TutorialHighlight {widgetClassPath=&TutorialHighlight}
 ]
}
