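# Athena broadcast HUD widget for spectators, live broadcast controllers and replay playback in Battle Royale.
# Construct and game-state events bind replay, broadcast, spectating and game-state delegates, plus stateful event-router listeners for game phase changes.
# Handlers drive camera-based reticle and game-over display, top alert and storm-surge switchers, replay HUD visibility modes and extension slots.
# Playlist and minigame state hide elements; broadcast controllers get a click-to-confirm battle bus start button.
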WidgetBlueprint /FortniteCoreUI/HUD/AthenaBroadcastHUD
graphs [
 BindBroadcastEvents {nodes [
   BindBroadcastEvents() {is=title K2Node_FunctionEntry}
   > n1 = Cast To FortLiveBroadcastController!(execute, Object=GetOwningPlayer()) -> (then, CastFailed, AsFort Live Broadcast Controller:Fort Live Broadcast Controller Object Reference) {is=title K2Node_DynamicCast}
   > Bind Event to On Toggle Off Air Widget Visibilty(self:Fort Live Broadcast Controller Object Reference=@n1, Delegate=Create Event()) {is=title K2Node_AddDelegate}
   > Bind Event to On Toggle Net UIWidget Visibilty(self:Fort Live Broadcast Controller Object Reference=@n1, Delegate=Create Event()) {is=title K2Node_AddDelegate}
  ]}
 BindGameStateEvents {nodes [
   n1 = BindGameStateEvents!() -> (then, GameStateAthena:Fort Game State Athena Object Reference) {is=title K2Node_FunctionEntry}
   > Is Valid(InputObject=@n1)
   > Bind Event to All Winners Announced(self:Fort Game State Athena Object Reference=@n1, Delegate=Create Event()) {is=title K2Node_AddDelegate}
   > Bind Event to Final Countdown Started(self:Fort Game State Athena Object Reference=@n1, Delegate=Create Event()) {is=title K2Node_AddDelegate}
   > Bind Event to Final Countdown Finished(self:Fort Game State Athena Object Reference=@n1, Delegate=Create Event()) {is=title K2Node_AddDelegate}
   > BindGamePhaseUpdated()
   > BindGamePhaseStepUpdated()
  ]}
 BindReplayContextEvents {nodes [
   BindReplayContextEvents() {is=title K2Node_FunctionEntry}
   n1 = Get FortReplayContext!() -> (ReturnValue:Fort Replay Context Object Reference) {is=title K2Node_GetBlueprintContextEx}
   > OnLevelStreamingChanged!(
    execute
    self
    bStreaming:bool=Break Fort Replay Playback State!(FortReplayPlaybackState:"Fort Replay Playback State Structure (by ref)"=Get PlaybackState!(self:Fort Replay Context Object Reference=@n1) -> (PlaybackState:Fort Replay Playback State Structure)) -> (
     StartTime:float=~
     EndTime:float=~
     TimeNow:float=~
     bIsPaused:bool=~
     PlaybackSpeedMultiplier:float=~
     HUDVisibility:EHudVisibilityState Enum
     bLevelStreaming:bool=~
     bHasRelevancyZone:bool=~
     bHasPlayInterval:bool=~
     PlayIntervalStartTime:float=~
     PlayIntervalEndTime:float=~
     bLoopInterval:bool=~
     IntervalLoopBufferTime:float=~
    ) {is=title K2Node_BreakStruct}.bLevelStreaming
   ) -> (then)
   > Bind Event to On Replay Level Streaming Changed(self:Fort Replay Context Object Reference=@n1, Delegate=Create Event()) {is=title K2Node_AddDelegate}
   > Bind Event to On Highlight Playing Event(self:Fort Replay Context Object Reference=@n1, Delegate=Create Event()) {is=title K2Node_AddDelegate}
   > Bind Event to On Sequencer Ready Event(self:Fort Replay Context Object Reference=@n1, Delegate=Create Event()) {is=title K2Node_AddDelegate}
   > Bind Event to On Replay Hud Notification Suppression Changed(self:Fort Replay Context Object Reference=@n1, Delegate=Create Event()) {is=title K2Node_AddDelegate}
   > Bind Event to On Session Id Watermark Visibility Changed(self:Fort Replay Context Object Reference=@n1, Delegate=Create Event()) {is=title K2Node_AddDelegate}
  ]}
 BindSpectatingEvents {nodes [
   BindSpectatingEvents() {is=title K2Node_FunctionEntry}
   > n1 = Cast To FortPlayerControllerSpectating!(execute, Object=GetOwningPlayer()) -> (then, CastFailed, AsFort Player Controller Spectating:Fort Player Controller Spectating Object Reference) {is=title K2Node_DynamicCast}
   > Bind Event to On Player Died(self:Fort Player Controller Spectating Object Reference=@n1, Delegate=Create Event()) {is=title K2Node_AddDelegate}
   > Bind Event to On Camera Type Changed(self:Fort Player Controller Spectating Object Reference=@n1, Delegate=Create Event()) {is=title K2Node_AddDelegate}
   > Bind Event to On Focus Reticle Visibility Changed(self:Fort Player Controller Spectating Object Reference=@n1, Delegate=Create Event()) {is=title K2Node_AddDelegate}
   > HandleCameraChange!(
    execute
    self
    PlayerController:Fort Player Controller Spectating Object Reference=@n1
    CameraType:ESpectatorCameraType Enum=Get CurrentCameraType!(self:Fort Player Controller Spectating Object Reference=@n1) -> (CurrentCameraType:ESpectatorCameraType Enum=ThirdPerson)
   ) -> (then)
  ]}
 CheckIfCreativeOrPlayground {nodes [
   n1 = CheckIfCreativeOrPlayground!() -> (then, GameStateAthena:Fort Game State Athena Object Reference) {is=title K2Node_FunctionEntry}
   > Is Valid(InputObject=@n1)
   n2 = IsCreativeLTMPlaylist!(self:Fort Game State Athena Object Reference=@n1) -> (ReturnValue:bool)
   n3 = IsCreativeOrPlaygroundModePlaylist!(self:Fort Game State Athena Object Reference=@n1) -> (ReturnValue:bool)
   > SetVisibility(
    self=[
     Get StormPhaseSB!(self:Athena Broadcast Storm Status Object Reference=Get AthenaBroadcastStormStatus() -> (AthenaBroadcastStormStatus:Athena Broadcast Storm Status Object Reference)) -> (StormPhaseSB:Size Box Object Reference)
     Get RemainingPlayersAndSquads() -> (RemainingPlayersAndSquads:Vertical Box Object Reference)
     Get KillFeedEntryBox!(self:Athena Broadcast Kill Feed Base Object Reference=Get AthenaBroadcastKillFeedWidget() -> (AthenaBroadcastKillFeedWidget:Athena Broadcast Kill Feed Widget Object Reference)) -> (KillFeedEntryBox:Fort Dynamic Entry Box Object Reference)
     Get MiniMap+StormWatch() -> (MiniMap+StormWatch:Vertical Box Object Reference)
    ]
    InVisibility=Select(Option 0=SelfHitTestInvisible, Option 1=Collapsed, Index=OR Boolean(A=@n3, B=@n2))
   )
  ]}
 EventGraph {nodes [
   n1 = "On Button Clicked (BattleBusButton)"!() -> (OutputDelegate, then, Button:Common Button Legacy Object Reference) {is=title K2Node_ComponentBoundEvent}
   > n2 = Branch(Condition=Get StartBusClicked() -> (StartBusClicked:bool))
   > SetVisibility(self=@n1.Button, InVisibility=Collapsed)
   > n3 = Cast To FortLiveBroadcastController(Object=GetPlayerController!(PlayerIndex:int=0) -> (ReturnValue:Player Controller Object Reference))
   > TryBeginBusPhase(self:Fort Live Broadcast Controller Object Reference=@n3)
   Set Text!(
    execute=@n2.else
    self:Icon Text Button Object Reference=Get BattleBusButton() -> (BattleBusButton:Icon Text Button Object Reference)
    ButtonText:Text="LOCTEXT(\"Confirm Start Bus\")"
   ) -> (then) {is=title K2Node_CallFunction}
   > Set StartBusClicked(StartBusClicked:bool=true) -> (Output_Get:bool)
   n4 = AllWinnersAnnoucedDelegateEvent!() -> (OutputDelegate, then, WinnerInfo_BigNameWinnerName:String, WinnerInfo_WinnerNames:Array of Strings) {is=title K2Node_CustomEvent}
   > Set WinnerInfoNames(WinnerInfoNames:Array of Strings=@n4.WinnerInfo_WinnerNames) -> (Output_Get:Array of Strings)
   > OnAllWinnersAnnounced!(execute, self) -> (then, Displayed:bool)
   OnCountdownStarted!() -> (OutputDelegate, then, TimeRemaining:int) {is=title K2Node_CustomEvent}
   > SetActiveWidget(
    self:Widget Switcher Object Reference=Get WidgetSwitcher_TopMiddleAlerts() -> (WidgetSwitcher_TopMiddleAlerts:Widget Switcher Object Reference)
    Widget=Get AthenaCountdown() -> (AthenaCountdown:Athena Countdown Object Reference)
   )
   > Set ActiveAlertWidget(ActiveAlertWidget:Widget Object Reference=Get AthenaCountdown()) -> (Output_Get:Widget Object Reference)
   UnbindGamePhaseStepUpdated() {is=title K2Node_CustomEvent}
   > n5 = Get!(execute:Exec) -> (then:Exec, else:Exec, GamePhaseStepUpdatedAsyncTask:Async Action Start Listening To Stateful Event Object Reference)
   > StopListeningToStatefulEvent(self:Async Action Start Listening To Stateful Event Object Reference=@n5)
   n6 = HUDVisibilitySound!() -> (OutputDelegate, then, HUDVisibility:EHudVisibilityState Enum=FullyVisible) {is=title K2Node_CustomEvent}
   > PlaySound2D(Sound=Select!(
     FullyVisible:Sound Base Object Reference=/Game/Sounds/Fort_UI_Sounds/Replay/DVR_Nav_UnHide_Cue
     FullyHidden:Sound Base Object Reference=/Game/Sounds/Fort_UI_Sounds/Replay/DVR_Nav_Hide_Cue
     GameOnly:Sound Base Object Reference=&DVR_Nav_UnHide_Cue
     ReplayOnly:Sound Base Object Reference=&DVR_Nav_UnHide_Cue
     Index:EHudVisibilityState Enum=@n6.HUDVisibility
    ) -> (ReturnValue:Sound Base Object Reference))
   OnCountdownFinished() {is=title K2Node_CustomEvent}
   > SetActiveWidget(self=Get WidgetSwitcher_TopMiddleAlerts(), Widget=Get GamePhaseAlert() -> (GamePhaseAlert:Athena Game Phase Change Widget Object Reference))
   > Set ActiveAlertWidget(ActiveAlertWidget=Get GamePhaseAlert())
   BindGamePhaseUpdated() {is=title K2Node_CustomEvent}
   > n7 = StartListeningToStatefulEvent!(
    execute:Exec
    Target:Gameplay Event Router Component Object Reference=GetGlobalEventRouter()
    EventType:Script Struct Object Reference=/Script/FortniteGame.FortGamePhaseUpdatedEvent
    Context:Object Reference
    EventBubblingRule:EEventBubblingRule Enum
   ) -> (
    then:Exec
    AsyncTaskProxy:Async Action Start Listening To Stateful Event Object Reference
    OnEventReceived:Exec
    OnSavedEventStateExists:Exec
    OnEventStateCleared:Exec
    Payload:Fort Game Phase Updated Event Structure
   ) {is=title K2Node_StartListeningToStatefulEventAsyncAction}
   > Set GamePhaseUpdatedAsyncTask(GamePhaseUpdatedAsyncTask:Async Action Start Listening To Stateful Event Object Reference=@n7) -> (Output_Get:Async Action Start Listening To Stateful Event Object Reference)
   GamePhaseChanged!(
    execute=[@n7.OnEventReceived, @n7.OnSavedEventStateExists]
    self
    NewPhase:EAthenaGamePhase Enum=Break Fort Game Phase Updated Event!(FortGamePhaseUpdatedEvent:"Fort Game Phase Updated Event Structure (by ref)"=@n7.Payload) -> (GamePhase:EAthenaGamePhase Enum) {is=title K2Node_BreakStruct}
   ) -> (then)
   UnbindGamePhaseUpdated() {is=title K2Node_CustomEvent}
   > n8 = Get!(execute) -> (then, else, GamePhaseUpdatedAsyncTask:Async Action Start Listening To Stateful Event Object Reference)
   > StopListeningToStatefulEvent(self=@n8)
   OnStartStormSurgeDisplay() {is=title K2Node_CustomEvent}
   > SetActiveWidget(
    self=Get WidgetSwitcher_Phase() -> (WidgetSwitcher_Phase:Common Widget Switcher Legacy Object Reference)
    Widget=Get AthenaStormSurgePhase() -> (AthenaStormSurgePhase:Athena Storm Surge Phase Object Reference)
   )
   OnStopStormSurgeDisplay() {is=title K2Node_CustomEvent}
   > SetActiveWidget(self=Get WidgetSwitcher_Phase(), Widget=Get GamePhaseAlert())
   BindGamePhaseStepUpdated() {is=title K2Node_CustomEvent}
   > n9 = StartListeningToStatefulEvent(Target=GetGlobalEventRouter(), EventType=/Script/FortniteGame.FortGamePhaseStepUpdatedEvent) -> (Payload:Fort Game Phase Step Updated Event Structure)
   > Set GamePhaseStepUpdatedAsyncTask(GamePhaseStepUpdatedAsyncTask:Async Action Start Listening To Stateful Event Object Reference=@n9) -> (Output_Get:Async Action Start Listening To Stateful Event Object Reference)
   OnPhaseStepChanged!(
    execute=[@n9.OnEventReceived, @n9.OnSavedEventStateExists]
    self
    NewPhaseStep:EAthenaGamePhaseStep Enum=Break Fort Game Phase Step Updated Event!(FortGamePhaseStepUpdatedEvent:"Fort Game Phase Step Updated Event Structure (by ref)"=@n9.Payload) -> (GamePhaseStep:EAthenaGamePhaseStep Enum) {is=title K2Node_BreakStruct}
   ) -> (then)
   n10 = OnMinigameStateChanged!() -> (OutputDelegate, then, ChangedMinigame:Fort Minigame Object Reference, MinigameState:EFortMinigameState Enum=PreGame) {is=title K2Node_CustomEvent}
   > SetVisibility(
    self=[
     Get Reticle() -> (Reticle:Reticle Object Reference)
     Get FollowedPlayerInfoWrapper() -> (FollowedPlayerInfoWrapper:Fort HUDElement Wrapper Object Reference)
     Get AthenaBroadcastStormStatus()
    ]
    InVisibility=Select!(
     Option 0:ESlateVisibility Enum=Collapsed
     Option 1:ESlateVisibility Enum=SelfHitTestInvisible
     Index:bool="Equal (Enum)"(A:EFortMinigameState Enum=@n10.MinigameState, B:EFortMinigameState Enum=InProgress)
    ) -> (ReturnValue:ESlateVisibility Enum)
   )
   > SetBattleBusButtonVisibility!(execute, self, IsVisible:bool="Equal (Enum)"(A=@n10.MinigameState, B=PreGame)) -> (then)
   n11 = OnPhaseStepChanged!() -> (OutputDelegate, then, NewPhaseStep:EAthenaGamePhaseStep Enum=None) {is=title K2Node_CustomEvent}
   > n12 = Branch(Condition=OR Boolean(
     A="Equal (Enum)"(A:EAthenaGamePhaseStep Enum=@n11.NewPhaseStep, B:EAthenaGamePhaseStep Enum=Countdown)
     B="Equal (Enum)"(A=@n11.NewPhaseStep, B=FinalCountdown)
    ))
   > SetActiveWidget(self=Get WidgetSwitcher_TopMiddleAlerts(), Widget=Get AthenaCountdown())
   > Set ActiveAlertWidget(ActiveAlertWidget=Get AthenaCountdown())
   SetActiveWidget(execute=@n12.else, self=Get WidgetSwitcher_TopMiddleAlerts(), Widget=Get GamePhaseAlert())
   > Set ActiveAlertWidget(ActiveAlertWidget=Get GamePhaseAlert())
   n13 = Event OnHudVisibilityChanged!() -> (OutputDelegate, then, NewVisibility:EHudVisibilityState Enum=FullyVisible) {is=title K2Node_Event}
   n14 = Select!(
    FullyVisible:ESlateVisibility Enum=SelfHitTestInvisible
    FullyHidden:ESlateVisibility Enum=Collapsed
    GameOnly:ESlateVisibility Enum=Collapsed
    ReplayOnly:ESlateVisibility Enum=SelfHitTestInvisible
    Index:EHudVisibilityState Enum=@n13.NewVisibility
   ) -> (ReturnValue)
   > SetVisibility(self=Get BroadcastHUDControls() -> (BroadcastHUDControls:Athena Spectator HUDControls Base Object Reference), InVisibility=@n14)
   > n15 = Branch(Condition=Get ShouldSuppressVisibilityNotifications() -> (ShouldSuppressVisibilityNotifications:bool))
   SetVisibility(
    execute=[@n18.then, @n15.then]
    self=Get RemainingPlayersAndSquads()
    InVisibility=Select(FullyVisible=SelfHitTestInvisible, FullyHidden=Collapsed, GameOnly=SelfHitTestInvisible, ReplayOnly=Collapsed, Index=@n13.NewVisibility)
   )
   > SetVisibility(
    self=Get VerticalBox_HighlightAnnotationContainer() -> (VerticalBox_HighlightAnnotationContainer:Vertical Box Object Reference)
    InVisibility=Select(FullyVisible=Collapsed, FullyHidden=Collapsed, GameOnly=SelfHitTestInvisible, ReplayOnly=Collapsed, Index=@n13.NewVisibility)
   )
   > SetVisibility(
    self=Get VerticalBox_KillFeedContainer() -> (VerticalBox_KillFeedContainer:Vertical Box Object Reference)
    InVisibility=Get NonHighlightsRelatedUIVisibility() -> (NonHighlightsRelatedUIVisibility:ESlateVisibility Enum=Visible)
   )
   > n16 = Cast To FortGameStateAthena!(execute, Object=GetGameState!() -> (ReturnValue:Game State Base Object Reference)) -> (then, CastFailed, AsFort Game State Athena:Fort Game State Athena Object Reference) {is=title K2Node_DynamicCast}
   > CheckIfCreativeOrPlayground!(execute, self, GameStateAthena:Fort Game State Athena Object Reference=@n16) -> (then) {is=title K2Node_CallFunction}
   SetNotificationVisibility!(
    execute=@n15.else
    self:Fort Replay Notification Object Reference=Get FortReplayNotification() -> (FortReplayNotification:Fort Replay Notification Object Reference)
    Visibility:ESlateVisibility Enum=@n14
   ) -> (then)
   > HUDVisibilitySound!(execute, self, HUDVisibility:EHudVisibilityState Enum=@n13.NewVisibility) -> (then) {is=title K2Node_CallFunction}
   n17 = Get FortReplayNotification()
   > n18 = ShowNotification!(
    execute
    self:Fort Replay Notification Object Reference=@n17
    NotificationText:"Text (by ref)"=GetVisibilityText!(self:Fort Replay Notification Object Reference=@n17, Visibility:EHudVisibilityState Enum=@n13.NewVisibility) -> (Text:Text)
   ) -> (then)
   n19 = Event OnHUDScaleChanged!() -> (OutputDelegate, then, HUDScale:float) {is=title K2Node_Event}
   > SetUserSpecifiedScale!(
    execute
    self:Scale Box Object Reference=Get PersistentHUDScale() -> (PersistentHUDScale:Scale Box Object Reference)
    InUserSpecifiedScale:float=@n19.HUDScale
   ) -> (then)
   n20 = Event SetFullscreenMapVisibility!() -> (OutputDelegate, then, bIsVisible:bool) {is=title K2Node_Event}
   > SetVisibility(
    self="Get [Overlay] Persistent HUD Content"() -> ("[Overlay] Persistent HUD Content":Overlay Object Reference)
    InVisibility=Select!(Option 0:ESlateVisibility Enum=SelfHitTestInvisible, Option 1:ESlateVisibility Enum=Collapsed, Index:bool=@n20.bIsVisible) -> (ReturnValue)
   )
   n21 = Event On Game State Set!() -> (OutputDelegate, then, GameStateAthena:Fort Game State Athena Object Reference) {is=title K2Node_Event}
   > BindGameStateEvents!(execute, self, GameStateAthena:Fort Game State Athena Object Reference=@n21.GameStateAthena) -> (then) {is=title K2Node_CallFunction}
   > CheckIfCreativeOrPlayground(GameStateAthena=@n21.GameStateAthena)
   n22 = Event HandleCursorModeChanged!() -> (OutputDelegate, then, bCursorModeEnabled:bool, ActionName:Name=None, CustomWidget:User Widget Object Reference) {is=title K2Node_Event}
   > Branch(Condition="Not Equal (Boolean)"!(A:bool=Get IsCursorModeEnabled() -> (IsCursorModeEnabled:bool), B:bool=@n22.bCursorModeEnabled) -> (ReturnValue:bool) {is=title K2Node_CallFunction})
   > Set IsCursorModeEnabled(IsCursorModeEnabled:bool=@n22.bCursorModeEnabled) -> (Output_Get:bool)
   n23 = "Equal (Name)"!(A:Name=@n22.ActionName, B:Name=ToggleMainMenu) -> (ReturnValue:bool) {is=title K2Node_CallFunction}
   > SetVisibility(
    self="Get [Overlay] Persistent HUD Content"()
    InVisibility=Select(Option 0=SelfHitTestInvisible, Option 1=Collapsed, Index=AND Boolean(A=@n22.bCursorModeEnabled, B=@n23))
   )
   > Refresh Camera Overlay Visibility()
   n24 = Event OnUnableToPerformAction!() -> (OutputDelegate, then, FailedReason:Gameplay Tag Container Structure, FailureText:"Text (by ref)") {is=title K2Node_Event}
   > Branch(Condition=HasTag!(
     TagContainer:"Gameplay Tag Container Structure (by ref)"=@n24.FailedReason
     Tag:Gameplay Tag Structure=(TagName="Athena.Ability.Failed.Consumable")
     bExactMatch:bool=false
    ) -> (ReturnValue:bool))
   > PlaySound2D(Sound=/Game/Athena/Sounds/Items/Consumable_CantApply_Cue)
   n25 = Event OnExtensionWidgetRemoved!() -> (OutputDelegate, then, ExtensionSlot:EUIExtensionSlot Enum=Invalid, Widget:User Widget Object Reference) {is=title K2Node_Event}
   > n26 = Switch on EUIExtensionSlot!(execute, Selection:EUIExtensionSlot Enum=@n25.ExtensionSlot) -> (
    Invalid:Exec
    Primary:Exec
    Secondary:Exec
    TopRightCorner:Exec
    GameInfoBox:Exec
    UpperCenter:Exec
    CrosshairRight:Exec
    MobileCrosshairRight:Exec
    UnderSquadInfo:Exec
    FullScreenMap:Exec
    BelowRespawnWidget:Exec
    BelowCompass:Exec
    UnderTeammateStatus:Exec
    ControllerBindingCallout:Exec
    AboveStormMessageSlot:Exec
    CustomMinigameCallouts:Exec
    UnderLocalPlayerInfo:Exec
    PlayerHealthbarOverlay:Exec
    UnderMinimapMessageContainer:Exec
    VehicleActionsSpacer:Exec
    Reticle:Exec
    PrioritizedContextualSlot:Exec
    RightOfTeammateStatus:Exec
    RightOfTeammateStatusBars:Exec
    TeammateStatusPortraitOverlay:Exec
    TeammateStatusPortraitUnderlay:Exec
    TeammateStatusPortraitReplacement:Exec
    TeammateStatusPortraitOverlay_IGM:Exec
    TeammateOverHealthbarOverlay:Exec
    TeammateOverShieldbarOverlay:Exec
    MobileHUDBottomRight:Exec
    InventoryScreenTab:Exec
    InventoryEquipSlot:Exec
    UnderneathTeammateStatusList:Exec
    FullScreenMapSquadEliminationsOverlay:Exec
    FullScreenMapStormInfo:Exec
    PrimaryQuickBarSlot:Exec
    MainMenuButtonListInZone:Exec
    MainMenuButtonListFrontEnd:Exec
    DynamicOverlayLobby:Exec
    PostMatchXPDetails_AccoladeList:Exec
    MobileHUDUnderMinimap:Exec
    SidebarPanels:Exec
    SpectatorRightOfPlayerStatus:Exec
    BehindMobileLocalPlayerInfo:Exec
    UnderMobileLocalPlayerInfo:Exec
    AboveMobileLocalPlayerInfo:Exec
    DynamicOverlayFrontEnd:Exec
    NPCStatusWidgetBarsPrefix:Exec
    NPCStatusWidgetAboveBars:Exec
    AbovePlayerIndicatorWidget:Exec
    FullScreenMapTab:Exec
    PostMatchScreenTab:Exec
    AboveBacchusEquippedItemReloadWidget:Exec
    UnderDefaultLobbyDetails:Exec
    DiscoveredAreaTextUnderlay:Exec
    LevelUpMobile:Exec
    ExpMobile:Exec
    InventoryGroupPanel:Exec
    InventoryItemInfoContents:Exec
    DiscoverInLobby:Exec
    InventoryItemInfoInspectPreview:Exec
    AboveQuestsMobileUI:Exec
    PostGamePlacementOverlayVictory:Exec
    TeammateStatusHealthShieldsReplacement:Exec
    GameInfoBelowMap:Exec
    LeftPostGameMobileOverlay:Exec
    LobbyAboveRankedDetails:Exec
    HUDPlayerActionAlertKillPlayerInfoExtension:Exec
    PostGamePlacementOverlayLosing:Exec
    AthenaInteractionIndicator:Exec
    HUDUnderReticle:Exec
    AthenaInventoryGroupMinorMedallionHeader:Exec
    AboveUpperCenterPrioritized:Exec
   ) {is=title K2Node_SwitchEnum}
   RemoveChild(execute=@n26.Primary, self=Get PrimaryContentOverlay() -> (PrimaryContentOverlay:Overlay Object Reference), Content=@n25.Widget)
   Event Construct()
   n27 = Create Event()
   n28 = Get AthenaStormSurgePhase()
   > Bind Event to Start Storm Surge Display(self:Athena Storm Surge Phase Object Reference=@n28, Delegate=@n27) {is=title K2Node_AddDelegate}
   n29 = Create Event()
   > Bind Event to Stop Storm Surge Display(self:Athena Storm Surge Phase Object Reference=@n28, Delegate=@n29) {is=title K2Node_AddDelegate}
   n30 = Create Event()
   n31 = Get VolumePromptWidget() -> (VolumePromptWidget:HUDTag Prompt Widget Object Reference)
   > Assign On Opened or Closed!(execute:Exec, self:Fort HUDTag Prompt Container Widget Object Reference=@n31, Delegate:"Delegate (by ref)"=@n30) -> (then:Exec) {is=title K2Node_AssignDelegate}
   > BindReplayContextEvents() {is=title K2Node_CallFunction}
   > BindBroadcastEvents() {is=title K2Node_CallFunction}
   > BindSpectatingEvents() {is=title K2Node_CallFunction}
   > SetupBattlebusButton()
   > SetIsEnabled(
    self=Get Button_Map!(self:Athena Minimap Container Object Reference=Get Minimap() -> (Minimap:Athena Minimap Container Object Reference)) -> (Button_Map:Button Object Reference)
    bInIsEnabled=false
   )
   > Bind Event to On All Winners Announced Delegate(
    self:Athena HUDContext Object Reference=Get AthenaHUDContext!() -> (ReturnValue:Athena HUDContext Object Reference) {is=title K2Node_GetBlueprintContextEx}
    Delegate=@n4
   ) {is=title K2Node_AddDelegate}
   n32 = Event OnIndicatorModeChanged!() -> (OutputDelegate, then, bIndicatorsEnabled:bool) {is=title K2Node_Event}
   > SetVisibility(
    self="Get [Overlay] Camera Mode Content"() -> ("[Overlay] Camera Mode Content":Overlay Object Reference)
    InVisibility=Select(Option 0=Collapsed, Option 1=SelfHitTestInvisible, Index=@n32.bIndicatorsEnabled)
   )
   Event Destruct()
   > Unbind Event from Start Storm Surge Display(self:Athena Storm Surge Phase Object Reference=@n28, Delegate=@n27) {is=title K2Node_RemoveDelegate}
   > Unbind Event from Stop Storm Surge Display(self:Athena Storm Surge Phase Object Reference=@n28, Delegate=@n29) {is=title K2Node_RemoveDelegate}
   > Unbind Event from On Opened or Closed(self:Fort HUDTag Prompt Container Widget Object Reference=@n31, Delegate=@n30) {is=title K2Node_RemoveDelegate}
   > ReplayHUDCleanup(self:Fort Replay Context Object Reference=Get FortReplayContext())
   > UnbindBroadcastEvents()
   > UnbindSpectatingEvents()
   > UnbindGameStateEvents()
   Event OnPlaylistDataChanged!() -> (OutputDelegate, then, PlaylistData:Fort Playlist Athena Object Reference) {is=title K2Node_Event}
   > n33 = Cast To FortGameStateAthena(Object=GetGameState())
   > CheckIfCreativeOrPlayground(GameStateAthena=@n33)
   n34 = Event OnPlayerTargetingChanged!() -> (OutputDelegate, then, bIsTargeting:bool) {is=title K2Node_Event}
   > n35 = Branch(Condition=@n34.bIsTargeting)
   > n36 = Cast To FortWeaponRanged!(
    execute
    Object=GetEquippedWeapon!(self:Fort Inventory Context Object Reference=Get FortInventoryContext!() -> (ReturnValue:Fort Inventory Context Object Reference) {is=title K2Node_GetBlueprintContextEx}) -> (ReturnValue:Fort Weapon Object Reference)
   ) -> (then, CastFailed, AsFort Weapon Ranged:Fort Weapon Ranged Object Reference) {is=title K2Node_DynamicCast}
   > n37 = Cast To FortPlayerControllerSpectating(Object=GetOwningPlayer())
   n38 = Get CurrentCameraType(self=@n37) -> (CurrentCameraType=ThirdPerson)
   > n39 = Branch(Condition=AND Boolean(
     A="Equal (Enum)"(A:ESpectatorCameraType Enum=@n38, B:ESpectatorCameraType Enum=Gameplay)
     B=IsUsingScope!(self:Fort Weapon Ranged Object Reference=@n36) -> (ReturnValue:bool)
    ))
   n40 = Get RemainingPlayersAndSquads()
   n41 = Get Minimap()
   > SetVisibility(self=[@n40, @n41, "Get [Overlay] Camera Mode Content"()], InVisibility=Hidden)
   SetVisibility(execute=[@n39.else, @n35.else], self=[@n40, @n41], InVisibility=SelfHitTestInvisible)
   > Refresh Camera Overlay Visibility()
   n42 = Event OnPipModeChanged!() -> (OutputDelegate, then, bEnabled:bool) {is=title K2Node_Event}
   > SetVisibility(
    self=Get PipModeSpacer() -> (PipModeSpacer:Spacer Object Reference)
    InVisibility=Select(Option 0=Collapsed, Option 1=SelfHitTestInvisible, Index=@n42.bEnabled)
   )
   n43 = Event OnExtensionWidgetCreated!() -> (OutputDelegate, then, ExtensionSlot:EUIExtensionSlot Enum=Invalid, Widget:User Widget Object Reference) {is=title K2Node_Event}
   > n44 = Switch on EUIExtensionSlot(Selection=@n43.ExtensionSlot)
   n45 = AddChildToOverlay!(execute=@n44.Primary, self:Overlay Object Reference=Get PrimaryContentOverlay(), Content:Widget Object Reference=@n43.Widget) -> (then, ReturnValue:Overlay Slot Object Reference)
   > SetHorizontalAlignment(self=@n45, InHorizontalAlignment=HAlign_Fill)
   > SetVerticalAlignment(self=@n45, InVerticalAlignment=VAlign_Fill)
   n46 = AddChildToOverlay(execute=@n44.Secondary, self=Get SecondaryContentOverlay() -> (SecondaryContentOverlay:Overlay Object Reference), Content=@n43.Widget)
   > SetHorizontalAlignment(self=@n46, InHorizontalAlignment=HAlign_Fill)
   > SetVerticalAlignment(self=@n46, InVerticalAlignment=VAlign_Fill)
   n47 = AddChildToVerticalBox(execute=@n44.BelowCompass, self=Get CompassBox() -> (CompassBox:Vertical Box Object Reference), Content=@n43.Widget)
   > SetHorizontalAlignment(self:Vertical Box Slot Object Reference=@n47, InHorizontalAlignment=HAlign_Fill)
   > SetVerticalAlignment(self:Vertical Box Slot Object Reference=@n47, InVerticalAlignment=VAlign_Fill)
   n48 = AddChildToVerticalBox(execute=@n44.UpperCenter, self=Get TopCenterContentUpper() -> (TopCenterContentUpper:Vertical Box Object Reference), Content=@n43.Widget)
   > SetHorizontalAlignment(self=@n48, InHorizontalAlignment=HAlign_Fill)
   > SetVerticalAlignment(self=@n48, InVerticalAlignment=VAlign_Fill)
   n49 = AddChildToVerticalBox(execute=@n44.TopRightCorner, self=Get TopRightContent() -> (TopRightContent:Vertical Box Object Reference), Content=@n43.Widget)
   > SetHorizontalAlignment(self=@n49, InHorizontalAlignment=HAlign_Fill)
   > SetVerticalAlignment(self=@n49, InVerticalAlignment=VAlign_Fill)
   n50 = AddChildToVerticalBox(execute=@n44.CustomMinigameCallouts, self=Get TopCenterContentUpper(), Content=@n43.Widget)
   > SetHorizontalAlignment(self=@n50, InHorizontalAlignment=HAlign_Fill)
   > SetVerticalAlignment(self=@n50, InVerticalAlignment=VAlign_Fill)
   n51 = Event OnMinigameChanged!() -> (OutputDelegate, then, NewMinigame:Fort Minigame Object Reference) {is=title K2Node_Event}
   > SetMinigameOverride!(
    execute
    self:Athena Broadcast Storm Status Object Reference=Get AthenaBroadcastStormStatus()
    MinigameOverride:Fort Minigame Object Reference=@n51.NewMinigame
   ) -> (then)
   > Is Valid(InputObject=@n51.NewMinigame)
   > Bind Event to State Changed Delegate(self:Fort Minigame Object Reference=@n51.NewMinigame, Delegate=Create Event()) {is=title K2Node_AddDelegate}
   > OnMinigameStateChanged!(
    execute
    self
    ChangedMinigame:Fort Minigame Object Reference=@n51.NewMinigame
    MinigameState:EFortMinigameState Enum=Get CurrentState!(self:Fort Minigame Object Reference=@n51.NewMinigame) -> (CurrentState:EFortMinigameState Enum=PreGame)
   ) -> (then) {is=title K2Node_CallFunction}
  ]}
 GamePhaseChanged {nodes [
   n1 = GamePhaseChanged!() -> (then, NewPhase:EAthenaGamePhase Enum=None) {is=title K2Node_FunctionEntry}
   > Branch(Condition="Equal (Enum)"(A:EAthenaGamePhase Enum=@n1, B:EAthenaGamePhase Enum=Aircraft))
   > SetVisibility(self=Get BattleBusButton(), InVisibility=Collapsed)
  ]}
 HandleCameraChange {nodes [
   n1 = HandleCameraChange!() -> (then, PlayerController:Fort Player Controller Spectating Object Reference, CameraType:ESpectatorCameraType Enum=ThirdPerson) {is=title K2Node_FunctionEntry}
   n2 = Get WinnerInfoNames() -> (WinnerInfoNames:Array of Strings)
   > n3 = Branch(Condition="integer > integer"(A=Length(TargetArray:Array of Strings=@n2), B=0) {is=title K2Node_CallFunction})
   > n4 = OnAllWinnersAnnounced()
   > n5 = Branch(Condition=@n4)
   n6 = Cast To FortPlayerStateAthena!(
    execute=[@n3.else, @n5.else]
    Object=GetFollowedPlayer!(self:Fort Player Controller Spectating Object Reference=@n1) -> (ReturnValue:Fort Player State Object Reference)
   ) -> (then, CastFailed, AsFort Player State Athena:Fort Player State Athena Object Reference) {is=title K2Node_DynamicCast}
   n7 = "Equal (Enum)"(A:ESpectatorCameraType Enum=@n1.CameraType, B:ESpectatorCameraType Enum=Gameplay)
   > SetVisibility(
    self=[Get Reticle(), Get StormSurgeOverlay() -> (StormSurgeOverlay:Overlay Object Reference)]
    InVisibility=Select(Option 0=Collapsed, Option 1=HitTestInvisible, Index=@n7)
   )
   n8 = IsPlayerInGameOver!() -> (bIsPlayerInGameOver:bool, bWasPlayerEliminated:bool)
   > n9 = Branch(Condition=@n8)
   > n10 = Branch(Condition=@n7)
   > n11 = Branch(Condition=@n8.bWasPlayerEliminated)
   > OnPlayerDied!(execute, self, InDeadPlayerState:Fort Player State Object Reference=@n6) -> (then)
   ShowGameOver!(execute=@n11.else, self, bShowEliminated:bool=false) -> (then)
   n12 = GetActiveWidget!(self:Widget Switcher Object Reference=Get GameOverSwitcher() -> (GameOverSwitcher:Widget Switcher Object Reference)) -> (ReturnValue:Widget Object Reference)
   n13 = Get WinnerOverlay_v2() -> (WinnerOverlay_v2:Winner Overlay V 2 Object Reference)
   n14 = Branch(execute=@n10.else, Condition=AND Boolean(A=@n8.bWasPlayerEliminated, B="Equal (Object)"(A=@n13, B=@n12) {is=title K2Node_CallFunction}))
   SetActiveWidgetIndex(execute=[@n14.then, @n9.else], self=Get GameOverSwitcher(), Index=0)
   n15 = Branch(execute=@n14.else, Condition=IsPlayerInGameOver().bWasPlayerEliminated)
   HideGameOver(execute=@n15.else)
  ]}
 HandleFocusReticleVisibilityChanged {nodes [
   n1 = HandleFocusReticleVisibilityChanged!() -> (then, bIsVisible:bool=~) {is=title K2Node_FunctionEntry}
   > SetVisibility(
    self=Get FocusReticle() -> (FocusReticle:Image Object Reference)
    InVisibility=Select(Option 0=Collapsed, Option 1=SelfHitTestInvisible, Index=@n1)
   )
  ]}
 HandleHighlightPlayingEvent {nodes [
   n1 = HandleHighlightPlayingEvent!() -> (then, bIsPlaying:bool=~) {is=title K2Node_FunctionEntry}
   > n2 = Set NonHighlightsRelatedUIVisibility(NonHighlightsRelatedUIVisibility:ESlateVisibility Enum=Select(Option 0=Visible, Option 1=Collapsed, Index=@n1)) -> (Output_Get:ESlateVisibility Enum=Visible)
   > SetVisibility(self=Get VerticalBox_KillFeedContainer(), InVisibility=@n2)
  ]}
 HandleHudNotificationSuppressionChange {nodes [
   n1 = HandleHudNotificationSuppressionChange!() -> (then, bShouldSuppress:bool=~) {is=title K2Node_FunctionEntry}
   > Set ShouldSuppressVisibilityNotifications(ShouldSuppressVisibilityNotifications:bool=@n1) -> (Output_Get:bool)
  ]}
 HandleSequencerPlaybackReadyEvent {nodes [
   n1 = HandleSequencerPlaybackReadyEvent!() -> (then, bIsPlaybackReady:bool=~) {is=title K2Node_FunctionEntry}
   > Set SequencerWantsLoadingScreen(SequencerWantsLoadingScreen:bool=NOT Boolean(A=@n1)) -> (Output_Get:bool)
   > UpdateLoadingVisibility()
  ]}
 HandleSessionIdWatermarkVisibilityChanged {nodes [
   n1 = HandleSessionIdWatermarkVisibilityChanged!() -> (then, IsVisible:bool=~) {is=title K2Node_FunctionEntry}
   > n2 = Cast To FortLiveSpectatorController!(execute, Object=GetOwningPlayer()) -> (then, CastFailed, AsFort Live Spectator Controller:Fort Live Spectator Controller Object Reference) {is=title K2Node_DynamicCast}
   SetVisibility(
    execute=@n2.CastFailed
    self=Get AthenaBroadcastSessionId() -> (AthenaBroadcastSessionId:Athena Broadcast Session Id Object Reference)
    InVisibility=Select(Option 0=Collapsed, Option 1=SelfHitTestInvisible, Index=@n1)
   )
  ]}
 HideGameOver {nodes [
   HideGameOver() {is=title K2Node_FunctionEntry}
   > n1 = Cast To FortPlayerControllerSpectating(Object=GetOwningPlayer())
   n2 = "Equal (Enum)"(A=Get CurrentCameraType(self=@n1) -> (CurrentCameraType=ThirdPerson), B=Gameplay)
   n3 = IsPlayerInGameOver()
   > Branch(Condition=OR Boolean(A=NOT Boolean(A=@n2), B=NOT Boolean(A=@n3)))
   > SetActiveWidgetIndex(self=Get GameOverSwitcher(), Index=0)
  ]}
 IsPlayerInGameOver {nodes [
   IsPlayerInGameOver() {is=title K2Node_FunctionEntry}
   > n1 = Sequence()
   > n2 = Cast To FortPlayerControllerSpectating(Object=GetOwningPlayer())
   > n3 = Cast To FortPlayerStateAthena(Object=GetFollowedPlayer(self=@n2))
   > n4 = Cast To FortGameStateAthena(Object=GetGameState())
   n5 = Get WinningTeam!(self:Fort Game State Athena Object Reference=@n4) -> (WinningTeam:int=0)
   > n6 = Branch(Condition="Equal (Byte)"!(
     A:Byte="To Byte (Integer)"!(InInt:int=@n5) -> (ReturnValue:Byte=0)
     B:Byte=GetActorTeam!(Actor:Actor Object Reference=@n3) -> (ReturnValue:Byte=0)
    ) -> (ReturnValue:bool))
   > Set bLocal_IsPlayerInGameOver(bLocal_IsPlayerInGameOver:bool=true) -> (Output_Get:bool)
   Branch(execute=@n6.else, Condition=IsPlayerDead!(self:Fort Player State Zone Object Reference=@n3) -> (ReturnValue:bool))
   > Set bLocal_IsPlayerInGameOver(bLocal_IsPlayerInGameOver=true)
   > Set bLocal_WasPlayerEliminated(bLocal_WasPlayerEliminated:bool=true) -> (Output_Get:bool)
   Return Node!(
    execute=@n1.then_1
    bIsPlayerInGameOver:bool=Get bLocal_IsPlayerInGameOver() -> (bLocal_IsPlayerInGameOver:bool)
    bWasPlayerEliminated:bool=Get bLocal_WasPlayerEliminated() -> (bLocal_WasPlayerEliminated:bool)
   )
  ]}
 OnAllWinnersAnnounced {nodes [
   OnAllWinnersAnnounced() {is=title K2Node_FunctionEntry}
   > Branch(Condition=Is Array Non Empty!(InputArray:Array of Strings=Get WinnerInfoNames()) -> (Non-Empty:bool=~) {is=title K2Node_MacroInstance})
   > n1 = Cast To FortPlayerControllerSpectating(Object=GetOwningPlayer())
   > n2 = Branch(Condition="Equal (Enum)"(A=Get CurrentCameraType(self=@n1) -> (CurrentCameraType=ThirdPerson), B=Gameplay))
   > n3 = Branch(Condition=IsPlayerInGameOver() {is=title K2Node_CallFunction})
   Return Node!(execute=[@n3.then, @n2.else], Displayed:bool=false)
   ShowGameOver(execute=@n3.else, bShowEliminated=false)
   > Clear and Invalidate Timer by Handle(Handle=Get GameOverTimerHandle() -> (GameOverTimerHandle:Timer Handle Structure))
   > SetVisibility(self=Get WinnerOverlay_v2(), InVisibility=HitTestInvisible)
   > SetVisibility(
    self=[
     Get MiniMap+StormWatch()
     Get AthenaBroadcastPlayersRemaining() -> (AthenaBroadcastPlayersRemaining:Athena Broadcast Players Remaining Object Reference)
     Get WidgetSwitcher_TopMiddleAlerts()
    ]
    InVisibility=Collapsed
   )
  ]}
 OnLevelStreamingChanged {nodes [
   n1 = OnLevelStreamingChanged!() -> (then, bStreaming:bool=~) {is=title K2Node_FunctionEntry}
   > Set LevelStreamingWantsLoadingScreen(LevelStreamingWantsLoadingScreen:bool=@n1) -> (Output_Get:bool)
   > UpdateLoadingVisibility()
  ]}
 OnPlayerDied {nodes [
   n1 = OnPlayerDied!() -> (then, InDeadPlayerState:Fort Player State Object Reference) {is=title K2Node_FunctionEntry}
   > n2 = Cast To FortPlayerStateAthena(Object=@n1)
   > Branch(Condition=IsPlayerDead(self=@n2))
   > n3 = Cast To FortPlayerControllerSpectating(Object=GetOwningPlayer())
   > Branch(Condition="Equal (Enum)"(A=Get CurrentCameraType(self=@n3) -> (CurrentCameraType=ThirdPerson), B=Gameplay))
   > Branch(Condition="Equal (Object)"(A=GetFollowedPlayer(self=@n3), B=@n2))
   > n4 = Branch(Condition=CanRespawnAfterDeath!(self:Fort Player Controller Spectating Object Reference=@n3) -> (ReturnValue:bool))
   ShowGameOver(execute=@n4.else, bShowEliminated=true)
  ]}
 OnToggleFollowedPlayerInfoActionExecuted {nodes [
   OnToggleFollowedPlayerInfoActionExecuted() {is=title K2Node_FunctionEntry}
   n1 = Get FollowedPlayerInfoOverlay() -> (FollowedPlayerInfoOverlay:Overlay Object Reference)
   > SetVisibility(self=@n1, InVisibility=Select(Option 0=SelfHitTestInvisible, Option 1=Collapsed, Index=IsVisible(self=@n1)))
   > Return Node!(execute, bPassthrough:bool=false)
  ]}
 OnToggleRemainingPlayersActionExecuted {nodes [
   OnToggleRemainingPlayersActionExecuted() {is=title K2Node_FunctionEntry}
   n1 = Get AthenaBroadcastPlayersRemaining()
   > SetVisibility(self=@n1, InVisibility=Select(Option 0=SelfHitTestInvisible, Option 1=Collapsed, Index=IsVisible(self=@n1)))
   > Return Node(bPassthrough=false)
  ]}
 OnToggleSquadStatusActionExecuted {nodes [
   OnToggleSquadStatusActionExecuted() {is=title K2Node_FunctionEntry}
   n1 = Get SquadStatus() -> (SquadStatus:Athena Broadcast Squad Status TG Object Reference)
   > SetVisibility(self=@n1, InVisibility=Select(Option 0=SelfHitTestInvisible, Option 1=Collapsed, Index=IsVisible(self=@n1)))
   > Return Node(bPassthrough=false)
  ]}
 OnVolumePromptOpenOrClosed {nodes [
   n1 = OnVolumePromptOpenOrClosed!() -> (then, NewlyOpen:bool=~) {is=title K2Node_FunctionEntry}
   > n2 = Branch(Condition=@n1)
   n3 = GetActiveWidget(self=Get WidgetSwitcher_TopMiddleAlerts())
   > Branch(Condition="Not Equal (Object)"!(A:Object Reference=@n3, B:Object Reference=Get VolumePromptWrapper() -> (VolumePromptWrapper:Fort HUDElement Wrapper Object Reference)) -> (ReturnValue:bool))
   > Set ActiveAlertWidget(ActiveAlertWidget=@n3)
   > SetActiveWidget(self=Get WidgetSwitcher_TopMiddleAlerts(), Widget=Get VolumePromptWrapper())
   SetActiveWidget(execute=@n2.else, self=Get WidgetSwitcher_TopMiddleAlerts(), Widget=Get ActiveAlertWidget() -> (ActiveAlertWidget:Widget Object Reference))
  ]}
 OnWinnerAnnounced {nodes [
   OnWinnerAnnounced() {is=title K2Node_FunctionEntry}
   > n1 = Cast To FortPlayerControllerSpectating(Object=GetOwningPlayer())
   > Branch(Condition="Equal (Enum)"(A=Get CurrentCameraType(self=@n1) -> (CurrentCameraType=ThirdPerson), B=Gameplay))
   n2 = IsPlayerInGameOver()
   > Branch(Condition=AND Boolean(A=@n2, B=NOT Boolean(A=@n2.bWasPlayerEliminated)))
   > ShowGameOver(bShowEliminated=false)
  ]}
 Refresh Camera Overlay Visibility {nodes [
   Refresh Camera Overlay Visibility() {is=title K2Node_FunctionEntry}
   > SetVisibility(
    self="Get [Overlay] Camera Mode Content"()
    InVisibility=Select(Option 0=SelfHitTestInvisible, Option 1=Collapsed, Index=Get IsCursorModeEnabled())
   )
  ]}
 SetBattleBusButtonVisibility {nodes [
   n1 = SetBattleBusButtonVisibility!() -> (then, IsVisible:bool=~) {is=title K2Node_FunctionEntry}
   > n2 = Cast To FortLiveBroadcastController(Object=GetPlayerController(PlayerIndex=0))
   > n3 = CanStartBattleBus!(execute, self:Fort Live Broadcast Controller Object Reference=@n2) -> (then, ReturnValue:bool)
   > SetVisibility(self=Get BattleBusButton(), InVisibility=Select(Option 0=Collapsed, Option 1=Visible, Index=AND Boolean(A=@n1, B=@n3)))
   SetVisibility(execute=@n2.CastFailed, self=Get BattleBusButton(), InVisibility=Collapsed)
  ]}
 SetupBattlebusButton {nodes [
   SetupBattlebusButton() {is=title K2Node_FunctionEntry}
   > SetBattleBusButtonVisibility(IsVisible=true) {is=title K2Node_CallFunction}
  ]}
 ShowGameOver {nodes [
   n1 = ShowGameOver!() -> (then, bShowEliminated:bool=~) {is=title K2Node_FunctionEntry}
   > n2 = Branch(Condition=@n1)
   Clear and Invalidate Timer by Handle(execute=[@n7.then, @n6.then, @n2.then], Handle=Get GameOverTimerHandle())
   > n3 = Set Timer by Event(Delegate=Create Event(), Time=5, bLooping=false)
   > Set GameOverTimerHandle(GameOverTimerHandle:Timer Handle Structure=@n3) -> (Output_Get:Timer Handle Structure)
   > SetVisibility(self=Get Reticle(), InVisibility=Collapsed)
   n4 = Get AthenaBroadcastHUDMenuStatusLayer() -> (AthenaBroadcastHUDMenuStatusLayer:Athena Broadcast HUDMenu Status Layer Object Reference)
   > Is Valid(InputObject=@n4)
   > DeactivateHudMenuWidget(self:Athena Broadcast HUDMenu Status Layer Base Object Reference=@n4)
   n5 = Get WinnerOverlay_v2()
   SetActiveWidget(execute=@n2.else, self=Get GameOverSwitcher(), Widget=@n5)
   > n6 = Branch(Condition=Get bShownVictoryAnim() -> (bShownVictoryAnim:bool))
   Set bShownVictoryAnim(execute=@n6.else, bShownVictoryAnim:bool=true) -> (Output_Get:bool)
   > n7 = PlayVictoryAnim(self:Winner Overlay V 2 Object Reference=@n5)
  ]}
 ToggleNetUIWidgetVisibility {nodes [
   ToggleNetUIWidgetVisibility() {is=title K2Node_FunctionEntry}
   n1 = Get NetUI() -> (NetUI:Net UI Object Reference)
   > SetVisibility(self=@n1, InVisibility=Select(Option 0=SelfHitTestInvisible, Option 1=Collapsed, Index=IsVisible(self=@n1)))
  ]}
 ToggleOffAirWidgetVisibility {nodes [
   ToggleOffAirWidgetVisibility() {is=title K2Node_FunctionEntry}
   n1 = Get OffAirBorder() -> (OffAirBorder:Common Border Object Reference)
   > SetVisibility(self=@n1, InVisibility=Select(Option 0=SelfHitTestInvisible, Option 1=Collapsed, Index=IsVisible(self=@n1)))
  ]}
 UnbindBroadcastEvents {nodes [
   UnbindBroadcastEvents() {is=title K2Node_FunctionEntry}
   > n1 = Cast To FortLiveBroadcastController(Object=GetOwningPlayer())
   > Unbind Event from On Toggle Off Air Widget Visibilty(self:Fort Live Broadcast Controller Object Reference=@n1, Delegate=Create Event()) {is=title K2Node_RemoveDelegate}
  ]}
 UnbindGameStateEvents {nodes [
   UnbindGameStateEvents() {is=title K2Node_FunctionEntry}
   > n1 = Cast To FortGameStateAthena(Object=GetGameState())
   > Unbind Event from All Winners Announced(self:Fort Game State Athena Object Reference=@n1, Delegate=Create Event()) {is=title K2Node_RemoveDelegate}
   > Unbind Event from Final Countdown Started(self:Fort Game State Athena Object Reference=@n1, Delegate=Create Event()) {is=title K2Node_RemoveDelegate}
   > Unbind Event from Final Countdown Finished(self:Fort Game State Athena Object Reference=@n1, Delegate=Create Event()) {is=title K2Node_RemoveDelegate}
   > UnbindGamePhaseUpdated() {is=title K2Node_CallFunction}
   > UnbindGamePhaseStepUpdated() {is=title K2Node_CallFunction}
  ]}
 UnbindSpectatingEvents {nodes [
   UnbindSpectatingEvents() {is=title K2Node_FunctionEntry}
   > n1 = Cast To FortPlayerControllerSpectating(Object=GetOwningPlayer())
   > Unbind Event from On Player Died(self:Fort Player Controller Spectating Object Reference=@n1, Delegate=Create Event()) {is=title K2Node_RemoveDelegate}
   > Bind Event to On Camera Type Changed(self=@n1, Delegate=Create Event())
  ]}
 UpdateLoadingVisibility {nodes [
   UpdateLoadingVisibility() {is=title K2Node_FunctionEntry}
   n1 = Get SequencerWantsLoadingScreen() -> (SequencerWantsLoadingScreen:bool)
   n2 = Get LevelStreamingWantsLoadingScreen() -> (LevelStreamingWantsLoadingScreen:bool)
   > SetVisibility(
    self=Get LoadingBox() -> (LoadingBox:Vertical Box Object Reference)
    InVisibility=Select(Option 0=Collapsed, Option 1=SelfHitTestInvisible, Index=OR Boolean(A=@n1, B=@n2))
   )
  ]}
]
variables [
 ActiveAlertWidget {type=Widget, default=null}
 bShownVictoryAnim {type=boolean}
 FocusedQuickbar {type=EFortQuickBars, default=Primary}
 GameOverTimerHandle {type=TimerHandle}
 GamePhaseStepUpdatedAsyncTask {type=AsyncAction_StartListeningToStatefulEvent, default=null}
 GamePhaseUpdatedAsyncTask {type=AsyncAction_StartListeningToStatefulEvent, default=null}
 IsCursorModeEnabled {type=boolean}
 LevelStreamingWantsLoadingScreen {type=boolean}
 NonHighlightsRelatedUIVisibility {type=ESlateVisibility, default=Visible}
 SequencerWantsLoadingScreen {type=boolean}
 ShouldSuppressVisibilityNotifications {type=boolean}
 StartBusClicked {type=boolean}
 ToggleSpectatorHUDInput {type=DataTableRowHandle, default {dataTable=/FortUILibrary/Input/ReplayInputActionDataTable, rowName=ToggleVisibility}}
 WinnerInfoNames {type="array<string>", default []}
]
widgets {
 info {
  inheritedWidgetCount=15
  namedSlotCount=4
  parentClass=/Script/FortniteUI.AthenaBroadcastHUDBase
  rootWidgetClass=&Overlay
  widgetCount=97
 }
 widgets [
  Overlay_65 {widgetClassPath=&Overlay, widget=Overlay_65}
  FocusReticle {
   widgetClassPath=&Image
   bIsVariable=true
   parent=Overlay_65
   widget=FocusReticle
   properties {
    visibility=Collapsed
    brush {
     bIsDynamicallyLoaded=false
     drawAs=Image
     imageType=FullColor
     mirroring=NoMirror
     resourceName=None
     resourceObject=/Game/Athena/HUD/reticles/reticle_cross_new
     tiling=NoTile
     imageSize=<16 16>
     margin=<ltrb 0 0 0 0>
     outlineSettings {
      bUseBrushTransparency=false
      roundingType=HalfHeightRadius
      width=0
      color {colorUseRule=UseColor_Specified, specifiedColor=<rgba 0 0 0 0>}
      cornerRadii=<0 0 0 0>
     }
     tintColor {colorUseRule=UseColor_Specified, specifiedColor=<rgba 1 1 1 1>}
     uVRegion {bIsValid=false, max=<0 0>, min=<0 0>}
    }
   }
   slot {class=OverlaySlot, properties {horizontalAlignment=HAlign_Center, verticalAlignment=VAlign_Center}}
  }
  AthenaBattleMapTimelineWidget {
   widgetClassPath=/Game/Athena/UI/Spectator/BattleMap/AthenaBattleMapTimelineWidget
   bIsVariable=true
   parent=Overlay_65
   widget=AthenaBattleMapTimelineWidget
   slot {class=OverlaySlot}
  }
  AthenaBattleMapNodeTooltip {
   widgetClassPath=/Game/Athena/UI/Spectator/BattleMap/AthenaBattleMapNodeTooltip
   bIsVariable=true
   parent=Overlay_65
   widget=AthenaBattleMapNodeTooltip
   slot {class=OverlaySlot}
  }
  NameplateLayer {
   widgetClassPath=/FortniteCoreUI/Athena/HUD/Spectator/AthenaSpectatorNameplateLayer
   bInherited=true
   bIsVariable=true
   parent=Overlay_65
   widget=NameplateLayer
   properties {elementTag {tagName=Athena.Replay.HUD}, hUDElementTag {gameplayTags []}}
   slot {class=OverlaySlot, properties {horizontalAlignment=HAlign_Fill, verticalAlignment=VAlign_Fill}}
  }
  "[Overlay] Camera Mode Content" {
   widgetClassPath=&Overlay
   bIsVariable=true
   parent=Overlay_65
   widget="[Overlay] Camera Mode Content"
   slot {class=OverlaySlot, properties {horizontalAlignment=HAlign_Fill, verticalAlignment=VAlign_Fill}}
  }
  AthenaIndicatorLayer {
   widgetClassPath=/Game/Athena/HUD/AthenaIndicatorLayer
   bIsVariable=true
   parent="[Overlay] Camera Mode Content"
   widget=AthenaIndicatorLayer
   slot {class=OverlaySlot, properties {horizontalAlignment=HAlign_Fill, verticalAlignment=VAlign_Fill}}
  }
  Reticle {
   widgetClassPath=/Game/UI/InGame/HUD/Reticle
   bIsVariable=true
   parent="[Overlay] Camera Mode Content"
   widget=Reticle
   properties {hUDElementTag {gameplayTags [{tagName=HUD.Center.Reticle}, {tagName=HUD.Athena.HideOnBus}]}}
   slot {class=OverlaySlot, properties {horizontalAlignment=HAlign_Fill, verticalAlignment=VAlign_Fill}}
  }
  AthenaBroadcastHUDMenuStatusLayer {
   widgetClassPath=/FortniteCoreUI/Athena/HUD/Spectator/AthenaBroadcastHUDMenuStatusLayer
   bIsVariable=true
   parent="[Overlay] Camera Mode Content"
   widget=AthenaBroadcastHUDMenuStatusLayer
   properties {elementTag {tagName=HUD.Athena.IndicatorLayer}}
   slot {class=OverlaySlot, properties {horizontalAlignment=HAlign_Fill, verticalAlignment=VAlign_Fill}}
  }
  PersistentHUD {
   widgetClassPath=&SafeZone
   parent=Overlay_65
   widget=PersistentHUD
   slot {class=OverlaySlot, properties {horizontalAlignment=HAlign_Fill, verticalAlignment=VAlign_Fill}}
  }
  "[Overlay] Persistent HUD Content" {
   widgetClassPath=&Overlay
   bIsVariable=true
   parent=PersistentHUD
   widget="[Overlay] Persistent HUD Content"
   slot {class=SafeZoneSlot}
  }
  PersistentHUDScale {
   widgetClassPath=&ScaleBox
   bIsVariable=true
   parent="[Overlay] Persistent HUD Content"
   widget=PersistentHUDScale
   properties {stretch=UserSpecified}
   slot {class=OverlaySlot, properties {horizontalAlignment=HAlign_Fill, verticalAlignment=VAlign_Fill}}
  }
  ScalableOverlay {
   widgetClassPath=&Overlay
   parent=PersistentHUDScale
   widget=ScalableOverlay
   slot {class=ScaleBoxSlot, properties {horizontalAlignment=HAlign_Fill, verticalAlignment=VAlign_Fill}}
  }
  TopCenterContent {
   widgetClassPath=&VerticalBox
   parent=ScalableOverlay
   widget=TopCenterContent
   slot {class=OverlaySlot, properties {horizontalAlignment=HAlign_Center, padding=<ltrb 30 30 30 30>}}
  }
  TopCenterContentUpper {
   widgetClassPath=&VerticalBox
   bIsVariable=true
   parent=TopCenterContent
   widget=TopCenterContentUpper
   slot {class=VerticalBoxSlot}
  }
  TopCenterContentLower {
   widgetClassPath=&VerticalBox
   bIsVariable=true
   parent=TopCenterContent
   widget=TopCenterContentLower
   slot {class=VerticalBoxSlot}
  }
  TopRightContent {
   widgetClassPath=&VerticalBox
   bIsVariable=true
   parent=ScalableOverlay
   widget=TopRightContent
   slot {class=OverlaySlot, properties {horizontalAlignment=HAlign_Right, padding=<ltrb 30 30 30 30>}}
  }
  RemainingPlayersAndSquads {
   widgetClassPath=&VerticalBox
   bIsVariable=true
   parent=ScalableOverlay
   widget=RemainingPlayersAndSquads
   slot {class=OverlaySlot, properties {padding=<ltrb 30 30 0 0>}}
  }
  FortHUDElementWrapper {
   widgetClassPath=/Game/Athena/HUD/FortHUDElementWrapper
   bIsVariable=true
   parent=RemainingPlayersAndSquads
   widget=FortHUDElementWrapper
   properties {elementTag {tagName=HUD.Athena.PlayersLeft}}
   slot {class=VerticalBoxSlot}
  }
  AthenaBroadcastPlayersRemaining {
   widgetClassPath=/FortniteCoreUI/Athena/HUD/Spectator/MatchStatus/AthenaBroadcastPlayersRemaining
   bIsVariable=true
   namedSlotHost=FortHUDElementWrapper
   widget=AthenaBroadcastPlayersRemaining
  }
  VerticalBox_KillFeedContainer {
   widgetClassPath=&VerticalBox
   bIsVariable=true
   parent=RemainingPlayersAndSquads
   widget=VerticalBox_KillFeedContainer
   slot {class=VerticalBoxSlot, properties {horizontalAlignment=HAlign_Left, verticalAlignment=VAlign_Top}}
  }
  AthenaBroadcastKillFeedWidget {
   widgetClassPath=/FortniteCoreUI/Athena/HUD/Spectator/KillFeed/AthenaBroadcastKillFeedWidget
   bIsVariable=true
   parent=VerticalBox_KillFeedContainer
   widget=AthenaBroadcastKillFeedWidget
   slot {class=VerticalBoxSlot, properties {padding=<ltrb 0 10 0 0>}}
  }
  VerticalBox_HighlightAnnotationContainer {
   widgetClassPath=&VerticalBox
   bIsVariable=true
   parent=RemainingPlayersAndSquads
   widget=VerticalBox_HighlightAnnotationContainer
   slot {class=VerticalBoxSlot, properties {horizontalAlignment=HAlign_Left, verticalAlignment=VAlign_Top, padding=<ltrb 0 20 0 0>}}
  }
  FortReplayHighlightAnnotation {
   widgetClassPath=/FortniteCoreUI/UI/Replay/Highlights/FortReplayHighlightAnnotation
   bIsVariable=true
   parent=VerticalBox_HighlightAnnotationContainer
   widget=FortReplayHighlightAnnotation
   slot {class=VerticalBoxSlot}
  }
  VerticalBoxCenterBottom {
   widgetClassPath=&VerticalBox
   parent=ScalableOverlay
   widget=VerticalBoxCenterBottom
   slot {class=OverlaySlot, properties {horizontalAlignment=HAlign_Fill, verticalAlignment=VAlign_Bottom, padding=<ltrb 0 0 0 40>}}
  }
  Overlay_2 {
   widgetClassPath=&Overlay
   parent=VerticalBoxCenterBottom
   widget=Overlay_2
   slot {class=VerticalBoxSlot, properties {horizontalAlignment=HAlign_Center}}
  }
  BroadcastHUDControls {
   widgetClassPath=/Game/Athena/HUD/Spectator/AthenaBroadcastHUDControls
   bInherited=true
   bIsVariable=true
   parent=Overlay_2
   widget=BroadcastHUDControls
   slot {class=OverlaySlot}
  }
  AthenaHealProgress {
   widgetClassPath=/Game/Athena/HUD/AthenaEquipProgress
   bIsVariable=true
   parent=Overlay_2
   widget=AthenaHealProgress
   properties {hUDElementTag {gameplayTags [
      {tagName=HUD.Athena.HideOnBus}
      {tagName=Athena.HideForCapture}
      {tagName=HUD.LTM.HideForTryHard}
      {tagName=Athena.Creative.HideForLoading}
      {tagName=HUD.Hide.WhileLoading}
     ]}}
   slot {class=OverlaySlot, properties {horizontalAlignment=HAlign_Center, verticalAlignment=VAlign_Bottom}}
  }
  FollowedPlayerInfoWrapper {
   widgetClassPath=&FortHUDElementWrapper
   bIsVariable=true
   parent=VerticalBoxCenterBottom
   widget=FollowedPlayerInfoWrapper
   properties {elementTag {tagName=Athena.Replay.HUD.FollowedPlayerInfo}, hUDElementTag {gameplayTags [{tagName=Athena.Replay.CameraMode.BattleMap}]}}
   slot {class=VerticalBoxSlot, properties {horizontalAlignment=HAlign_Center}}
  }
  FollowedPlayerInfoOverlay {widgetClassPath=&Overlay, bIsVariable=true, namedSlotHost=FollowedPlayerInfoWrapper, widget=FollowedPlayerInfoOverlay}
  AthenaBroadcastPlayerEventScore {
   widgetClassPath=/FortniteCoreUI/HUD/Spectator/AthenaBroadcastPlayerEventScore
   parent=FollowedPlayerInfoOverlay
   widget=AthenaBroadcastPlayerEventScore
   properties {
    elementTag {tagName=Athena.HideForCapture}
    hUDElementTag {gameplayTags [{tagName=HUD.Athena.HideOnBus}]}
    renderTransform {angle=0, scale=<1 1>, shear=<0 0>, translation=<-75 0>}
   }
   slot {class=OverlaySlot, properties {verticalAlignment=VAlign_Center}}
  }
  AthenaBroadcastFollowedPlayerInfo {
   widgetClassPath=/FortniteCoreUI/Athena/HUD/Spectator/AthenaBroadcastFollowedPlayerInfo
   parent=FollowedPlayerInfoOverlay
   widget=AthenaBroadcastFollowedPlayerInfo
   slot {class=OverlaySlot}
  }
  AthenaSpectatorMaterialCountsWidget {
   widgetClassPath=/Game/Athena/HUD/Spectator/Inventory/AthenaSpectatorMaterialCountsWidget
   bIsVariable=true
   parent=FollowedPlayerInfoOverlay
   widget=AthenaSpectatorMaterialCountsWidget
   properties {renderTransform {angle=0, scale=<1 1>, shear=<0 0>, translation=<119 0>}}
   slot {class=OverlaySlot, properties {horizontalAlignment=HAlign_Right, verticalAlignment=VAlign_Bottom}}
  }
  MiniMapStormWatchWrapper {
   widgetClassPath=&FortHUDElementWrapper
   bIsVariable=true
   parent=ScalableOverlay
   widget=MiniMapStormWatchWrapper
   properties {elementTag {tagName=Athena.Replay.HUD.MiniMapStormWatch}, hUDElementTag {gameplayTags [{tagName=Athena.Replay.CameraMode.BattleMap}]}}
   slot {class=OverlaySlot, properties {horizontalAlignment=HAlign_Right, verticalAlignment=VAlign_Bottom, padding=<ltrb 0 0 30 30>}}
  }
  MiniMap+StormWatch {widgetClassPath=&VerticalBox, bIsVariable=true, namedSlotHost=MiniMapStormWatchWrapper, widget=MiniMap+StormWatch}
  StormSurgeOverlay {widgetClassPath=&Overlay, bIsVariable=true, parent=MiniMap+StormWatch, widget=StormSurgeOverlay, slot {class=VerticalBoxSlot}}
  AthenaBroadcastStormSurgeWidget {
   widgetClassPath=/FortniteCoreUI/Athena/HUD/Spectator/AthenaBroadcastStormSurgeWidget
   bIsVariable=true
   parent=StormSurgeOverlay
   widget=AthenaBroadcastStormSurgeWidget
   properties {visibility=Collapsed}
   slot {class=OverlaySlot, properties {horizontalAlignment=HAlign_Fill, verticalAlignment=VAlign_Fill}}
  }
  Overlay_1 {widgetClassPath=&Overlay, parent=MiniMap+StormWatch, widget=Overlay_1, slot {class=VerticalBoxSlot}}
  Minimap {
   widgetClassPath=/Game/Athena/HUD/AthenaMinimapContainer
   bIsVariable=true
   parent=Overlay_1
   widget=Minimap
   properties {elementTag {tagName=Athena.HideForCapture}, hUDElementTag {gameplayTags [{tagName=HUD.TopRight.Minimap}, {tagName=Athena.Crucible}]}}
   slot {class=OverlaySlot, properties {horizontalAlignment=HAlign_Center}}
  }
  AthenaBroadcastStormStatus {
   widgetClassPath=/FortniteCoreUI/Athena/HUD/Spectator/MatchStatus/AthenaBroadcastStormStatus
   bIsVariable=true
   parent=MiniMap+StormWatch
   widget=AthenaBroadcastStormStatus
   slot {class=VerticalBoxSlot, properties {horizontalAlignment=HAlign_Right, padding=<ltrb 0 10 0 10>}}
  }
  GameOverSwitcher {
   widgetClassPath=&WidgetSwitcher
   bIsVariable=true
   parent=ScalableOverlay
   widget=GameOverSwitcher
   slot {class=OverlaySlot, properties {horizontalAlignment=HAlign_Fill, verticalAlignment=VAlign_Fill}}
  }
  SizeBox_0 {widgetClassPath=&SizeBox, parent=GameOverSwitcher, widget=SizeBox_0, slot {class=WidgetSwitcherSlot}}
  WinnerOverlay_v2 {
   widgetClassPath=/Game/Athena/HUD/EndGame/WinnerOverlay_v2
   bIsVariable=true
   parent=GameOverSwitcher
   widget=WinnerOverlay_v2
   slot {class=WidgetSwitcherSlot}
  }
  NetUI {
   widgetClassPath=/Game/Athena/HUD/NetUI
   bIsVariable=true
   parent=ScalableOverlay
   widget=NetUI
   properties {visibility=Collapsed}
   slot {class=OverlaySlot, properties {horizontalAlignment=HAlign_Right, padding=<ltrb 0 30 30 0>}}
  }
  VerticalBox_0 {
   widgetClassPath=&VerticalBox
   parent=ScalableOverlay
   widget=VerticalBox_0
   slot {class=OverlaySlot, properties {verticalAlignment=VAlign_Bottom, padding=<ltrb 30 0 0 30>}}
  }
  FortReplayNotification {
   widgetClassPath=/FortniteCoreUI/UI/Replay/FortReplayNotification
   bIsVariable=true
   parent=VerticalBox_0
   widget=FortReplayNotification
   properties {bIsEnabled=true}
   slot {class=VerticalBoxSlot, properties {horizontalAlignment=HAlign_Left, padding=<ltrb 0 0 0 25>}}
  }
  PipModeSpacer {
   widgetClassPath=&Spacer
   bIsVariable=true
   parent=VerticalBox_0
   widget=PipModeSpacer
   properties {visibility=Collapsed, size=<1 270>}
   slot {class=VerticalBoxSlot}
  }
  SquadStatusSB {
   widgetClassPath=&SizeBox
   parent=VerticalBox_0
   widget=SquadStatusSB
   properties {bOverride_WidthOverride=true, maxDesiredWidth=827.0656, minDesiredWidth=430, widthOverride=430}
   slot {class=VerticalBoxSlot, properties {horizontalAlignment=HAlign_Left}}
  }
  SquadStatus {
   widgetClassPath=/FortniteCoreUI/Athena/HUD/Spectator/SquadStatus/AthenaBroadcastSquadStatusTG
   bIsVariable=true
   parent=SquadStatusSB
   widget=SquadStatus
   slot {class=SizeBoxSlot}
  }
  AthenaBroadcastSessionId {
   widgetClassPath=/FortniteCoreUI/Athena/HUD/Spectator/MatchStatus/AthenaBroadcastSessionId
   bIsVariable=true
   parent=ScalableOverlay
   widget=AthenaBroadcastSessionId
   properties {elementTag {tagName=HUD.Athena.SessionId}}
   slot {class=OverlaySlot, properties {horizontalAlignment=HAlign_Right, verticalAlignment=VAlign_Bottom, padding=<ltrb 0 0 30 10>}}
  }
  ScaleBox_SecondaryContent {
   widgetClassPath=&ScaleBox
   bIsVariable=true
   parent=ScalableOverlay
   widget=ScaleBox_SecondaryContent
   properties {stretch=UserSpecified}
   slot {class=OverlaySlot, properties {horizontalAlignment=HAlign_Fill, verticalAlignment=VAlign_Fill}}
  }
  SecondaryContentOverlay {
   widgetClassPath=&Overlay
   bIsVariable=true
   parent=ScaleBox_SecondaryContent
   widget=SecondaryContentOverlay
   slot {class=ScaleBoxSlot, properties {horizontalAlignment=HAlign_Fill, verticalAlignment=VAlign_Fill}}
  }
  WidgetSwitcher_TopMiddleAlerts {
   widgetClassPath=&WidgetSwitcher
   bIsVariable=true
   parent="[Overlay] Persistent HUD Content"
   widget=WidgetSwitcher_TopMiddleAlerts
   slot {class=OverlaySlot, properties {horizontalAlignment=HAlign_Center, padding=<ltrb 0 200 0 0>}}
  }
  WidgetSwitcher_Phase {
   widgetClassPath=&CommonWidgetSwitcherLegacy
   bIsVariable=true
   parent=WidgetSwitcher_TopMiddleAlerts
   widget=WidgetSwitcher_Phase
   properties {transitionCurveType=QuadInOut}
   slot {class=WidgetSwitcherSlot}
  }
  GamePhaseAlert {
   widgetClassPath=/Game/Athena/HUD/Phase/AthenaGamePhaseChangeWidget
   bIsVariable=true
   parent=WidgetSwitcher_Phase
   widget=GamePhaseAlert
   properties {visibility=Collapsed, hUDElementTag {gameplayTags [{tagName=HUD.Athena.GamePhase}, {tagName=Athena.HideForCapture}]}}
   slot {class=WidgetSwitcherSlot}
  }
  AthenaStormSurgePhase {
   widgetClassPath=/Game/Athena/HUD/AthenaStormSurgePhase
   bIsVariable=true
   parent=WidgetSwitcher_Phase
   widget=AthenaStormSurgePhase
   properties {elementTag {tagName=Athena.HideForCapture}}
   slot {class=WidgetSwitcherSlot}
  }
  AthenaCountdown {
   widgetClassPath=/Game/Athena/HUD/AthenaCountdown
   bIsVariable=true
   parent=WidgetSwitcher_TopMiddleAlerts
   widget=AthenaCountdown
   properties {elementTag {tagName=Athena.HideForCapture}, renderTransform {angle=0, scale=<1 1>, shear=<0 0>, translation=<0 -76>}}
   slot {class=WidgetSwitcherSlot, properties {verticalAlignment=VAlign_Bottom}}
  }
  VolumePromptWrapper {
   widgetClassPath=&FortHUDElementWrapper
   bIsVariable=true
   parent=WidgetSwitcher_TopMiddleAlerts
   widget=VolumePromptWrapper
   properties {
    elementTag {tagName=Athena.Replay.HUD}
    hUDElementTag {gameplayTags [{tagName=Athena.Replay.CameraMode.DroneFree}, {tagName=Athena.Replay.CameraMode.BattleMap}, {tagName=Athena.Replay.CameraMode.ARDrone}]}
   }
   slot {class=WidgetSwitcherSlot}
  }
  VolumePromptWidget {
   widgetClassPath=/Game/UI/Foundation/HUDTagPrompt/HUDTagPromptWidget
   bIsVariable=true
   namedSlotHost=VolumePromptWrapper
   widget=VolumePromptWidget
  }
  Overlay_0 {
   widgetClassPath=&Overlay
   parent="[Overlay] Persistent HUD Content"
   widget=Overlay_0
   slot {class=OverlaySlot, properties {padding=<ltrb 25 96 0 0>}}
  }
  "[Vertical Box] Top Left Content" {
   widgetClassPath=&VerticalBox
   bIsVariable=true
   parent=Overlay_0
   widget="[Vertical Box] Top Left Content"
   slot {class=OverlaySlot}
  }
  SizeBox_1 {
   widgetClassPath=&SizeBox
   parent="[Vertical Box] Top Left Content"
   widget=SizeBox_1
   properties {
    bOverride_MaxDesiredHeight=true
    bOverride_MinDesiredHeight=true
    bOverride_MinDesiredWidth=true
    maxDesiredHeight=630
    minDesiredHeight=410
    minDesiredWidth=650
   }
   slot {class=VerticalBoxSlot}
  }
  Switcher_TopLeft {
   widgetClassPath=&CommonWidgetSwitcherLegacy
   bInherited=true
   bIsVariable=true
   parent=SizeBox_1
   widget=Switcher_TopLeft
   slot {class=SizeBoxSlot}
  }
  TopLeft_Empty {
   widgetClassPath=&Spacer
   bInherited=true
   parent=Switcher_TopLeft
   widget=TopLeft_Empty
   properties {size=<0 0>}
   slot {class=WidgetSwitcherSlot}
  }
  TopLeft_ViewSettings {
   widgetClassPath=/FortniteCoreUI/UI/Broadcast/FortBroadcastViewSettings
   bInherited=true
   parent=Switcher_TopLeft
   widget=TopLeft_ViewSettings
   slot {class=WidgetSwitcherSlot}
  }
  TopLeft_ActionFeed {
   widgetClassPath=/FortniteCoreUI/Athena/UI/Spectator/ActionFeed/AthenaReplayActionFeedScreen
   bInherited=true
   parent=Switcher_TopLeft
   widget=TopLeft_ActionFeed
   slot {class=WidgetSwitcherSlot}
  }
  TopLeft_GoToTime {
   widgetClassPath=/FortniteCoreUI/UI/Replay/GotoTimeDialog
   bInherited=true
   parent=Switcher_TopLeft
   widget=TopLeft_GoToTime
   slot {class=WidgetSwitcherSlot}
  }
  TopLeft_Highlights {
   widgetClassPath=/FortniteCoreUI/UI/Replay/Highlights/FortReplayHighlightsWindow
   bInherited=true
   parent=Switcher_TopLeft
   widget=TopLeft_Highlights
   slot {class=WidgetSwitcherSlot}
  }
  LoadingBox {
   widgetClassPath=&VerticalBox
   bIsVariable=true
   parent="[Overlay] Persistent HUD Content"
   widget=LoadingBox
   slot {class=OverlaySlot, properties {horizontalAlignment=HAlign_Center, verticalAlignment=VAlign_Center}}
  }
  TextBlock_0 {
   widgetClassPath=&TextBlock
   parent=LoadingBox
   widget=TextBlock_0
   properties {
    text=LOADING
    font {
     bForceMonospaced=false
     bMaterialIsStencil=false
     fontMaterial=null
     fontObject=&Burbank
     letterSpacing=0
     monospacedWidth=1
     size=34
     skewAmount=0
     typefaceFontName=Burbank-Black
     outlineSettings {
      bApplyOutlineToDropShadows=false
      bMiteredCorners=false
      bSeparateFillAlpha=false
      outlineMaterial=null
      outlineSize=0
      outlineColor=<rgba 0 0 0 1>
     }
    }
    shadowColorAndOpacity=<rgba 0 0 0 0.822>
    shadowOffset=<2.5373 5.8361>
   }
   slot {class=VerticalBoxSlot, properties {horizontalAlignment=HAlign_Center, verticalAlignment=VAlign_Center}}
  }
  Image_0 {
   widgetClassPath=&Image
   parent=LoadingBox
   widget=Image_0
   properties {brush {
     bIsDynamicallyLoaded=false
     drawAs=Image
     imageType=FullColor
     mirroring=NoMirror
     resourceName=None
     resourceObject=/FortUILibrary/UIKit/Throbber/MI_UIKit_Monochrome_Throbber
     tiling=NoTile
     imageSize=<64 64>
     margin=<ltrb 0 0 0 0>
     outlineSettings {
      bUseBrushTransparency=false
      roundingType=HalfHeightRadius
      width=0
      color {colorUseRule=UseColor_Specified, specifiedColor=<rgba 0 0 0 0>}
      cornerRadii=<0 0 0 0>
     }
     tintColor {colorUseRule=UseColor_Specified, specifiedColor=<rgba 1 1 1 1>}
     uVRegion {bIsValid=false, max=<0 0>, min=<0 0>}
    }}
   slot {class=VerticalBoxSlot, properties {horizontalAlignment=HAlign_Center, verticalAlignment=VAlign_Center}}
  }
  VerticalBox_4 {
   widgetClassPath=&VerticalBox
   parent="[Overlay] Persistent HUD Content"
   widget=VerticalBox_4
   slot {class=OverlaySlot, properties {verticalAlignment=VAlign_Center, padding=<ltrb 30 30 30 30>}}
  }
  BattleBusButton {
   widgetClassPath=&IconTextButton
   bIsVariable=true
   parent=VerticalBox_4
   widget=BattleBusButton
   properties {
    buttonText=Start Bus
    onButtonBaseClicked=(&self:WidgetTree.BattleBusButton.HandleOnButtonClicked)
    onButtonBaseDoubleClicked=(&self:WidgetTree.BattleBusButton.HandleOnButtonDoubleClicked)
    onButtonBaseHovered=(&self:WidgetTree.BattleBusButton.HandleOnButtonHovered)
    onButtonBaseUnhovered=(&self:WidgetTree.BattleBusButton.HandleOnButtonUnhovered)
    onSelectedChangedBase=(&self:WidgetTree.BattleBusButton.HandleOnSelectedChanged)
   }
   slot {class=VerticalBoxSlot, properties {padding=<ltrb 0 0 0 5>}}
  }
  AthenaPlayerActionAlert {
   widgetClassPath=/Game/Athena/HUD/AthenaPlayerActionAlert
   parent="[Overlay] Persistent HUD Content"
   widget=AthenaPlayerActionAlert
   properties {elementTag {tagName=Athena.HideForCapture}, hUDElementTag {gameplayTags [{tagName=HUD.Athena.PlayerActions}]}}
   slot {class=OverlaySlot, properties {horizontalAlignment=HAlign_Center, verticalAlignment=VAlign_Bottom, padding=<ltrb 0 0 0 300>}}
  }
  AthenaStandaloneTimer {
   widgetClassPath=/Game/Athena/HUD/AthenaStandaloneTimer
   parent="[Overlay] Persistent HUD Content"
   widget=AthenaStandaloneTimer
   properties {elementTag {tagName=Athena.HideForCapture}}
   slot {class=OverlaySlot, properties {horizontalAlignment=HAlign_Center, padding=<ltrb 0 90 0 0>}}
  }
  ScaleBox_1 {
   widgetClassPath=&ScaleBox
   parent="[Overlay] Persistent HUD Content"
   widget=ScaleBox_1
   properties {stretch=UserSpecified}
   slot {class=OverlaySlot, properties {horizontalAlignment=HAlign_Fill, verticalAlignment=VAlign_Fill}}
  }
  EventMatchInfoSwitcher {
   widgetClassPath=&CommonWidgetSwitcherLegacy
   bInherited=true
   bIsVariable=true
   parent=ScaleBox_1
   widget=EventMatchInfoSwitcher
   properties {bOutroPanelBelow=false, bWidgetActivationEnabled=false, transitionDuration=0}
   slot {class=ScaleBoxSlot, properties {horizontalAlignment=HAlign_Fill, verticalAlignment=VAlign_Fill}}
  }
  EventMatchInfoNone {
   widgetClassPath=&SizeBox
   bIsVariable=true
   parent=EventMatchInfoSwitcher
   widget=EventMatchInfoNone
   slot {class=WidgetSwitcherSlot}
  }
  AthenaEventMatchStatusScreen {
   widgetClassPath=/FortniteCoreUI/Athena/UI/Spectator/EventMatchInfo/AthenaEventMatchStatusScreen
   bIsVariable=true
   parent=EventMatchInfoSwitcher
   widget=AthenaEventMatchStatusScreen
   slot {class=WidgetSwitcherSlot}
  }
  AthenaEventMatchScoreboardScreen {
   widgetClassPath=/FortniteCoreUI/Athena/UI/Spectator/EventMatchInfo/AthenaEventMatchScoreboardScreen
   bIsVariable=true
   parent=EventMatchInfoSwitcher
   widget=AthenaEventMatchScoreboardScreen
   slot {class=WidgetSwitcherSlot}
  }
  AthenaRespawnTimer {
   widgetClassPath=/Game/Athena/HUD/AthenaRespawnTimer
   parent="[Overlay] Persistent HUD Content"
   widget=AthenaRespawnTimer
   properties {elementTag {tagName=Athena.HideForCapture}}
   slot {class=OverlaySlot, properties {horizontalAlignment=HAlign_Fill, verticalAlignment=VAlign_Fill, padding=<ltrb 0 600 0 0>}}
  }
  VerticalBox {
   widgetClassPath=&VerticalBox
   parent="[Overlay] Persistent HUD Content"
   widget=VerticalBox
   slot {class=OverlaySlot, properties {horizontalAlignment=HAlign_Right}}
  }
  Spacer_110 {widgetClassPath=&Spacer, parent=VerticalBox, widget=Spacer_110, properties {size=<1 128>}, slot {class=VerticalBoxSlot}}
  AthenaHUDMenu {
   widgetClassPath=&/WBP_AthenaHUDMenu
   bInherited=true
   bIsVariable=true
   parent=Overlay_65
   widget=AthenaHUDMenu
   slot {class=OverlaySlot, properties {horizontalAlignment=HAlign_Fill, verticalAlignment=VAlign_Fill}}
  }
  AthenaBroadcastMapPanel {
   widgetClassPath=/FortniteCoreUI/HUD/Spectator/AthenaBroadcastMapPanel
   bInherited=true
   bIsVariable=true
   parent=Overlay_65
   widget=AthenaBroadcastMapPanel
   properties {visibility=Collapsed}
   slot {class=OverlaySlot, properties {horizontalAlignment=HAlign_Fill, verticalAlignment=VAlign_Fill}}
  }
  OffAirBorder {
   widgetClassPath=&CommonBorder
   bIsVariable=true
   parent=Overlay_65
   widget=OffAirBorder
   properties {
    visibility=Hidden
    background {
     bIsDynamicallyLoaded=false
     drawAs=Image
     imageType=NoImage
     mirroring=NoMirror
     resourceName=None
     resourceObject=/FortUILibrary/Styles/Border/Materials/M-Border-1pxOutline
     tiling=NoTile
     imageSize=<1280 720>
     margin=<ltrb 1 1 1 1>
     outlineSettings {
      bUseBrushTransparency=false
      roundingType=HalfHeightRadius
      width=0
      color {colorUseRule=UseColor_Specified, specifiedColor=<rgba 0 0 0 0>}
      cornerRadii=<0 0 0 0>
     }
     tintColor {colorUseRule=UseColor_Specified, specifiedColor=<rgba 1 1 1 1>}
     uVRegion {bIsValid=false, max=<0 0>, min=<0 0>}
    }
    brushColor=<rgba 1 0 0 1>
   }
   slot {class=OverlaySlot, properties {horizontalAlignment=HAlign_Fill, verticalAlignment=VAlign_Fill}}
  }
  ScaleBox_PrimaryContent {
   widgetClassPath=&ScaleBox
   bIsVariable=true
   parent=Overlay_65
   widget=ScaleBox_PrimaryContent
   properties {stretch=UserSpecified}
   slot {class=OverlaySlot, properties {horizontalAlignment=HAlign_Fill, verticalAlignment=VAlign_Fill}}
  }
  PrimaryContentOverlay {
   widgetClassPath=&Overlay
   bIsVariable=true
   parent=ScaleBox_PrimaryContent
   widget=PrimaryContentOverlay
   slot {class=ScaleBoxSlot, properties {horizontalAlignment=HAlign_Fill, verticalAlignment=VAlign_Fill}}
  }
  FortShotTitleScreen {
   widgetClassPath=/FortniteCoreUI/UI/Replay/FortShotTitleScreen
   bIsVariable=true
   parent=Overlay_65
   widget=FortShotTitleScreen
   slot {class=OverlaySlot}
  }
  CompassBox {
   widgetClassPath=&VerticalBox
   bIsVariable=true
   parent=Overlay_65
   widget=CompassBox
   properties {visibility=HitTestInvisible}
   slot {class=OverlaySlot, properties {horizontalAlignment=HAlign_Center}}
  }
  PaddingUntilAPermanentWidget {
   widgetClassPath=&Spacer
   parent=CompassBox
   widget=PaddingUntilAPermanentWidget
   properties {size=<1 50>}
   slot {class=VerticalBoxSlot}
  }
  WidgetStack_SpectatorPrimaryContent {
   widgetClassPath=/Script/CommonUI.CommonActivatableWidgetStack
   bInherited=true
   bIsVariable=true
   parent=Overlay_65
   widget=WidgetStack_SpectatorPrimaryContent
   properties {transitionDuration=0}
   slot {class=OverlaySlot, properties {horizontalAlignment=HAlign_Fill, verticalAlignment=VAlign_Fill}}
  }
  SafeZone_TextChat {
   widgetClassPath=&SafeZone
   parent=Overlay_65
   widget=SafeZone_TextChat
   slot {class=OverlaySlot, properties {horizontalAlignment=HAlign_Fill, verticalAlignment=VAlign_Bottom, padding=<ltrb 20 20 20 225>}}
  }
  WBP_TextChat_InputListener {
   widgetClassPath=/FortniteCoreUI/TextChat/UI/WBP_TextChat_InputListener
   bIsVariable=true
   parent=SafeZone_TextChat
   widget=WBP_TextChat_InputListener
   properties {bAllowLookWhileActive=false}
   slot {class=SafeZoneSlot}
  }
  MapScreenContainer {widgetClassPath=/Script/FortniteUI.AthenaMapScreenContainer, bInherited=true, widget=null}
  TitleBar {widgetClassPath=&CommonActivatableWidget, bInherited=true, widget=null}
  WidgetStack_PrimaryContent {widgetClassPath=&CommonActivatableWidgetStack, bInherited=true, widget=null}
 ]
}
