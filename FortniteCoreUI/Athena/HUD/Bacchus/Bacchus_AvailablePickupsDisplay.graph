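# Bacchus HUD widget that lays out a list of available pickups on top of the native BacchusAvailablePickupsDisplay class.
# A right-aligned toggle button, styled as a minus and inherited from the parent, sits above a fixed-size box holding a scroll box with a hidden scrollbar.
# Inside it, an inherited dynamic entry box wraps a capped number of Button_AvailablePickup entries in last-to-first order.
# The event graph holds only unconnected stub events, leaving behavior to the parent class.
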WidgetBlueprint /FortniteCoreUI/Athena/HUD/Bacchus/Bacchus_AvailablePickupsDisplay
graphs [EventGraph {nodes [
   Event PreConstruct()
   Event Construct()
   Event Tick()
  ]}]
variables []
widgets {
 info {
  inheritedWidgetCount=5
  parentClass=/Script/FortniteUI.BacchusAvailablePickupsDisplay
  rootWidgetClass=&VerticalBox
  widgetCount=9
 }
 widgets [
  VerticalBox_1 {widgetClassPath=&VerticalBox, widget=VerticalBox_1}
  Button_ToggleVisibility {
   widgetClassPath=&IconTextButton
   bInherited=true
   bIsVariable=true
   parent=VerticalBox_1
   widget=Button_ToggleVisibility
   properties {
    bSelectable=true
    bToggleable=true
    buttonText=""
    minHeight=58
    onButtonBaseClicked=(&self:WidgetTree.Button_ToggleVisibility.HandleOnButtonClicked)
    onButtonBaseDoubleClicked=(&self:WidgetTree.Button_ToggleVisibility.HandleOnButtonDoubleClicked)
    onButtonBaseHovered=(&self:WidgetTree.Button_ToggleVisibility.HandleOnButtonHovered)
    onButtonBaseUnhovered=(&self:WidgetTree.Button_ToggleVisibility.HandleOnButtonUnhovered)
    onSelectedChangedBase=(&self:WidgetTree.Button_ToggleVisibility.HandleOnSelectedChanged)
    style=/FortUILibrary/Foundation/Buttons/ButtonStyles/ButtonStyle_Minus.ButtonStyle_Minus_C
   }
   slot {class=VerticalBoxSlot, properties {horizontalAlignment=HAlign_Right, padding=<ltrb 32 0 0 0>}}
  }
  SizeBox_70 {
   widgetClassPath=&SizeBox
   parent=VerticalBox_1
   widget=SizeBox_70
   properties {
    bOverride_HeightOverride=true
    bOverride_WidthOverride=true
    heightOverride=490
    maxDesiredHeight=420
    maxDesiredWidth=300
    widthOverride=290
   }
   slot {class=VerticalBoxSlot, properties {size {sizeRule=Automatic, value=0}}}
  }
  VerticalBox_0 {widgetClassPath=&VerticalBox, parent=SizeBox_70, widget=VerticalBox_0, slot {class=SizeBoxSlot}}
  CommonHierarchicalScrollBox_195 {
   widgetClassPath=/Script/CommonUI.CommonHierarchicalScrollBox
   parent=VerticalBox_0
   widget=CommonHierarchicalScrollBox_195
   properties {scrollBarVisibility=Collapsed, scrollbarThickness=<0 9>}
   slot {class=VerticalBoxSlot, properties {size {sizeRule=Fill, value=1}}}
  }
  EntryBox_PickupButtons {
   widgetClassPath=/Script/FortniteUI.FortDynamicEntryBox
   bInherited=true
   bIsVariable=true
   parent=CommonHierarchicalScrollBox_195
   widget=EntryBox_PickupButtons
   properties {
    entryBoxType=Wrap
    entryWidgetClass=&/Button_AvailablePickup.Button_AvailablePickup_C
    maximumEntries=25
    numDesignerPreviewEntries=8
    patternDirection=LastToFirst
   }
   slot {class=ScrollBoxSlot}
  }
  Image_Background {widgetClassPath=&Image, bInherited=true, widget=null}
  ScaleBox_HUDWidgetScale {widgetClassPath=&ScaleBox, bInherited=true, widget=null}
  TutorialHighlight {widgetClassPath=&TutorialHighlight}
 ]
}
