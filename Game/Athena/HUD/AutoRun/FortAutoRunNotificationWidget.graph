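# HUD notification widget that shows a "Character Auto Running!" banner with a run icon when auto-run toggles.
# It extends native FortAutoRunNotificationWidget, starts collapsed on Construct, and plays Intro or Outro from OnAutoRunEnabled's bool.
# Intro pops in the icon, slides the clipped banner and text in, and stretches and fades in the background.
# Outro retracts banner and background, then shrinks the icon; the graph itself never makes the widget visible again.
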WidgetBlueprint /Game/Athena/HUD/AutoRun/FortAutoRunNotificationWidget
animations [
 Intro {
  binding_count=4
  length_seconds=0.85
  object=&self:Intro
  sequence {bindings [
    Image_Background {
     id=#1
     possessed_class=FortLazyImage
     tracks [
      {
       type=MovieScene2DTransformTrack
       sections [{
         type=MovieScene2DTransformSection
         channels [
          Scale.X {type=MovieSceneScriptingFloatChannel, keys [[8,0],[14.0003,1]]}
          Scale.Y {type=MovieSceneScriptingFloatChannel}
         ]
        }]
      }
      {
       type=MovieSceneFloatTrack
       display_name=Render Opacity
       sections [{type=MovieSceneFloatSection, end_frame=14, channels [None {type=MovieSceneScriptingFloatChannel, keys [[8,0],[14.0003,1]]}]}]
      }
     ]
    }
    Border_0 {
     id=#2
     possessed_class=Border
     tracks [{
       type=MovieScene2DTransformTrack
       sections [{
         type=MovieScene2DTransformSection
         channels [
          Translation.X {type=MovieSceneScriptingFloatChannel, keys [[7,600],[14,0]]}
          Translation.Y {type=MovieSceneScriptingFloatChannel}
         ]
        }]
      }]
    }
    Border_icon {
     id=#3
     possessed_class=Border
     tracks [{
       type=MovieScene2DTransformTrack
       sections [{
         type=MovieScene2DTransformSection
         channels [
          Translation.X {type=MovieSceneScriptingFloatChannel}
          Translation.Y {type=MovieSceneScriptingFloatChannel}
          Angle {type=MovieSceneScriptingFloatChannel}
          Scale.X {type=MovieSceneScriptingFloatChannel, keys [[0,0],[7,1]]}
          Scale.Y {type=MovieSceneScriptingFloatChannel, keys [[0,0],[7,1]]}
          Shear.X {type=MovieSceneScriptingFloatChannel}
          Shear.Y {type=MovieSceneScriptingFloatChannel}
         ]
        }]
      }]
    }
    Text {
     id=#4
     possessed_class=CommonTextBlock
     tracks [{
       type=MovieScene2DTransformTrack
       sections [{
         type=MovieScene2DTransformSection
         channels [
          Translation.X {type=MovieSceneScriptingFloatChannel, keys [[10,128],[17,0]]}
          Translation.Y {type=MovieSceneScriptingFloatChannel}
          Angle {type=MovieSceneScriptingFloatChannel}
          Scale.X {type=MovieSceneScriptingFloatChannel}
          Scale.Y {type=MovieSceneScriptingFloatChannel}
          Shear.X {type=MovieSceneScriptingFloatChannel}
          Shear.Y {type=MovieSceneScriptingFloatChannel}
         ]
        }]
      }]
    }
   ]}
 }
 Outro {
  binding_count=3
  length_seconds=0.45
  object=&self:Outro
  sequence {bindings [
    Border_0 {
     id=#5
     possessed_class=Border
     tracks [{
       type=MovieScene2DTransformTrack
       sections [{
         type=MovieScene2DTransformSection
         channels [
          Translation.X {type=MovieSceneScriptingFloatChannel, keys [[0,0],[5,600]]}
          Translation.Y {type=MovieSceneScriptingFloatChannel}
          Angle {type=MovieSceneScriptingFloatChannel}
          Scale.X {type=MovieSceneScriptingFloatChannel}
          Scale.Y {type=MovieSceneScriptingFloatChannel}
          Shear.X {type=MovieSceneScriptingFloatChannel}
          Shear.Y {type=MovieSceneScriptingFloatChannel}
         ]
        }]
      }]
    }
    Border_icon {
     id=#6
     possessed_class=Border
     tracks [{
       type=MovieScene2DTransformTrack
       sections [{
         type=MovieScene2DTransformSection
         channels [
          Translation.X {type=MovieSceneScriptingFloatChannel}
          Translation.Y {type=MovieSceneScriptingFloatChannel}
          Angle {type=MovieSceneScriptingFloatChannel}
          Scale.X {type=MovieSceneScriptingFloatChannel, keys [[6,1],[9,0]]}
          Scale.Y {type=MovieSceneScriptingFloatChannel, keys [[6,1],[9,0]]}
          Shear.X {type=MovieSceneScriptingFloatChannel}
          Shear.Y {type=MovieSceneScriptingFloatChannel}
         ]
        }]
      }]
    }
    Image_Background {
     id=#7
     possessed_class=FortLazyImage
     tracks [
      {
       type=MovieScene2DTransformTrack
       sections [{
         type=MovieScene2DTransformSection
         channels [
          Scale.X {type=MovieSceneScriptingFloatChannel, keys [[0,1],[5,0]]}
          Scale.Y {type=MovieSceneScriptingFloatChannel}
         ]
        }]
      }
      {
       type=MovieSceneFloatTrack
       display_name=Render Opacity
       sections [{type=MovieSceneFloatSection, channels [None {type=MovieSceneScriptingFloatChannel, keys [[0,1],[5,0]]}]}]
      }
     ]
    }
   ]}
 }
]
graphs [EventGraph {nodes [
   n1 = Event OnAutoRunEnabled!() -> (OutputDelegate, then, bAutoRunEnabled:bool) {is=title K2Node_Event}
   > n2 = Branch(Condition=@n1.bAutoRunEnabled)
   > PlayAnimation(InAnimation=Get Intro())
   PlayAnimation(execute=@n2.else, InAnimation=Get Outro() -> (Outro:Widget Animation Object Reference))
   Event Construct()
   > Parent: Construct()
   > SetVisibility(InVisibility=Collapsed)
  ]}]
variables []
widgets {
 info {inheritedWidgetCount=0, parentClass=/Script/FortniteUI.FortAutoRunNotificationWidget, rootWidgetClass=&Overlay, widgetCount=8}
 widgets [
  Overlay_0 {widgetClassPath=&Overlay, bIsVariable=true, widget=Overlay_0}
  Image_Background {
   widgetClassPath=&FortLazyImage
   bIsVariable=true
   parent=Overlay_0
   widget=Image_Background
   properties {
    colorAndOpacity=<rgba 0.0009 0.0018 0.0545 0.33>
    renderTransform {angle=0, scale=<1 1>, shear=<0 0>, translation=<10 10>}
    renderTransformPivot=<1 1>
   }
   slot {class=OverlaySlot, properties {horizontalAlignment=HAlign_Fill, verticalAlignment=VAlign_Fill}}
  }
  HorizontalBox_0 {
   widgetClassPath=&HorizontalBox
   parent=Overlay_0
   widget=HorizontalBox_0
   slot {class=OverlaySlot, properties {horizontalAlignment=HAlign_Center, verticalAlignment=VAlign_Bottom}}
  }
  Overlay_1 {
   widgetClassPath=&Overlay
   parent=HorizontalBox_0
   widget=Overlay_1
   properties {clipping=ClipToBounds}
   slot {class=HorizontalBoxSlot}
  }
  Border_0 {
   widgetClassPath=&Border
   parent=Overlay_1
   widget=Border_0
   properties {verticalAlignment=VAlign_Center, padding=<ltrb 16 2 16 2>, renderTransformPivot=<1 0.5>}
   slot {class=OverlaySlot, properties {verticalAlignment=VAlign_Fill}}
  }
  Text {
   widgetClassPath=&CommonTextBlock
   bIsVariable=true
   parent=Border_0
   widget=Text
   properties {
    text="Character Auto Running!"
    visibility=HitTestInvisible
    colorAndOpacity {colorUseRule=UseColor_Specified, specifiedColor=<rgba 0.0009 0.0018 0.0545 1>}
    font {
     bForceMonospaced=false
     bMaterialIsStencil=false
     fontMaterial=null
     fontObject=&BurbankSmall
     letterSpacing=0
     monospacedWidth=1
     size=24
     skewAmount=0
     typefaceFontName=Bold
     outlineSettings {
      bApplyOutlineToDropShadows=false
      bMiteredCorners=false
      bSeparateFillAlpha=false
      outlineMaterial=null
      outlineSize=0
      outlineColor=<rgba 0 0 0 1>
     }
    }
    shadowOffset=<0 0>
   }
   slot {class=BorderSlot, properties {verticalAlignment=VAlign_Center, padding=<ltrb 16 2 16 2>}}
  }
  Border_icon {
   widgetClassPath=&Border
   parent=HorizontalBox_0
   widget=Border_icon
   properties {
    background {
     bIsDynamicallyLoaded=false
     drawAs=Image
     imageType=NoImage
     mirroring=NoMirror
     resourceName=None
     resourceObject=null
     tiling=NoTile
     imageSize=<32 32>
     margin=<ltrb 0 0 0 0>
     outlineSettings {
      bUseBrushTransparency=false
      roundingType=HalfHeightRadius
      width=0
      color {colorUseRule=UseColor_Specified, specifiedColor=<rgba 0 0 0 0>}
      cornerRadii=<0 0 0 0>
     }
     tintColor {colorUseRule=UseColor_Specified, specifiedColor=<rgba 0.0331 0.402 1 1>}
     uVRegion {bIsValid=false, max=<0 0>, min=<0 0>}
    }
    renderTransformPivot=<1 0.5>
   }
   slot {class=HorizontalBoxSlot, properties {horizontalAlignment=HAlign_Center, verticalAlignment=VAlign_Top}}
  }
  Image_Run {
   widgetClassPath=&Image
   bIsVariable=true
   parent=Border_icon
   widget=Image_Run
   properties {
    visibility=HitTestInvisible
    brush {
     bIsDynamicallyLoaded=false
     drawAs=Image
     imageType=NoImage
     mirroring=NoMirror
     resourceName=None
     resourceObject=&/M_AutoRun
     tiling=NoTile
     imageSize=<64 64>
     margin=<ltrb 0 0 0 0>
     outlineSettings {
      bUseBrushTransparency=false
      roundingType=HalfHeightRadius
      width=0
      color {colorUseRule=UseColor_Specified, specifiedColor=<rgba 0 0 0 0>}
      cornerRadii=<0 0 0 0>
     }
     tintColor {colorUseRule=UseColor_Specified, specifiedColor=<rgba 1 1 1 1>}
     uVRegion {bIsValid=false, max=<0 0>, min=<0 0>}
    }
   }
   slot {class=BorderSlot}
  }
 ]
}
