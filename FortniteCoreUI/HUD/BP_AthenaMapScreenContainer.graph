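# Tabbed full-screen map container widget with top tab bar, profile and battle pass XP, optional island collab icon, bottom action bar and damage vignette.
# Parent-class events drive tab sizing and visibility, creative scoreboard toggling and vignette animation.
# Activation plays sounds, loads the collab icon from a data registry, and updates XP visibility.
# Creative photo mode deactivates the screen, then reopens the My Island tab on exit.
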WidgetBlueprint /FortniteCoreUI/HUD/BP_AthenaMapScreenContainer
animations [DamageTaken {
  binding_count=1
  length_seconds=0.5
  object=&self:DamageTaken
  sequence {bindings [Image_DamageVignette {
     id=#1
     possessed_class=Image
     tracks [
      {
       type=MovieSceneFloatTrack
       display_name=Render Opacity
       sections [{type=MovieSceneFloatSection, channels [None {type=MovieSceneScriptingFloatChannel, keys [[0,1],[10,0]]}]}]
      }
      {
       type=MovieSceneWidgetMaterialTrack
       sections [{
         type=MovieSceneParameterSection
         channels [
          "Density (S)" {type=MovieSceneScriptingFloatChannel, keys [[0,4.5738],[3,5.0718],[10,4.5738]]}
          "Radius (S)" {type=MovieSceneScriptingFloatChannel, keys [[0,0.9905],[3,0.7085],[10,0.9905]]}
         ]
        }]
      }
     ]
    }]}
 }]
graphs [
 EventGraph {nodes [
   n1 = Event OnShowOrHideCreativeScoreboard!() -> (OutputDelegate, then, bShowCreativeScoreboard:bool) {is=title K2Node_Event}
   ShowOrHideScoreboardUI!(
    execute
    self
    ScoreboardWidget:Object Reference=Get CreativeScoreboardWidget() -> (CreativeScoreboardWidget:Common User Widget Weak object)
    bShow:bool=@n1.bShowCreativeScoreboard
   ) -> (then)
   > n2 = Get!(execute:Exec) -> (then:Exec, else:Exec, CreativeCareerScoreboardWidget:Common User Widget Weak object)
   > ShowOrHideScoreboardUI(ScoreboardWidget=@n2, bShow=@n1.bShowCreativeScoreboard)
   n3 = OnInputMethodChanged_Event!() -> (OutputDelegate, then, bNewInputType:ECommonInputType Enum) {is=title K2Node_CustomEvent}
   OnUpdateActionBar(execute=[@n9.then, @n3.then], self:Athena Map Screen Container Object Reference)
   n4 = Event PreConstruct()
   > SetInputAction!(
    execute
    self:Common Action Widget Object Reference=Get NextTabAction() -> (NextTabAction:Common Action Widget Object Reference)
    InputActionRow:Data Table Row Handle Structure=Get NextTabInputAction() -> (NextTabInputAction:Data Table Row Handle Structure)
   ) -> (then)
   > SetInputAction(
    self=Get PreviousTabAction() -> (PreviousTabAction:Common Action Widget Object Reference)
    InputActionRow=Get PreviousTabInputAction() -> (PreviousTabInputAction:Data Table Row Handle Structure)
   )
   > Branch(Condition=@n4.IsDesignTime)
   > ClearChildren(self=Get Panel_TabButtonsContainer() -> (Panel_TabButtonsContainer:Panel Widget Object Reference))
   > For Loop(FirstIndex=1, LastIndex=Get DesignTimeTabCount() -> (DesignTimeTabCount:int=0))
   > n5 = Create Widget!(
    execute:Exec
    Class:User Widget Class Reference=/Game/Athena/HUD/BP_SkewedTabButton.BP_SkewedTabButton_C
    OwningPlayer:Player Controller Object Reference
    Style:Common Button Style Class Reference=/FortUILibrary/Foundation/Buttons/ButtonStyles/Athena/Athena_ButtonStyle_CollectionTab.Athena_ButtonStyle_CollectionTab_C
    bApplyAlphaOnDisable:bool=True
    bLocked:bool=False
    bSelectable:bool=True
    bShouldSelectUponReceivingFocus:bool=False
    bInteractableWhenSelected:bool=False
    bToggleable:bool=False
    bTriggerClickedAfterSelection:bool=False
    bDisplayInputActionWhenNotInteractable:bool=True
    bHideInputActionWithKeyboard:bool=False
    bShouldUseFallbackDefaultInputAction:bool=True
    bRequiresHold:bool=False
    HoldData:Common UIHold Data Class Reference
    ClickMethod:EButtonClickMethod Enum=DownAndUp
    InputPriority:int=0
    TriggeringInputAction:Data Table Row Handle Structure
   ) -> (then:Exec, ReturnValue:BP Skewed Tab Button Object Reference)
   > AddChild(self=Get Panel_TabButtonsContainer(), Content=@n5)
   Event OnDamageReceived() {is=title K2Node_Event}
   > PlayAnimation(InAnimation=Get DamageTaken() -> (DamageTaken:Widget Animation Object Reference))
   n6 = Event On Show Tabs Set!() -> (OutputDelegate, then, bShowTabs:bool) {is=title K2Node_Event}
   > n7 = Set ShowTabs(ShowTabs:bool=OR Boolean(A=@n6.bShowTabs, B=Get TabsAlwaysVisible() -> (TabsAlwaysVisible:bool))) -> (Output_Get:bool)
   > SetVisibility(
    self=Get HBox_Tabs() -> (HBox_Tabs:Horizontal Box Object Reference)
    InVisibility=Select(Option 0=Collapsed, Option 1=SelfHitTestInvisible, Index=@n7)
   )
   > UpdateTabBackgroundSize()
   Event Construct()
   > n8 = Sequence()
   > UpdateTabBackgroundSize()
   > Set BR Default Tabs() {is=title K2Node_CallFunction}
   > SetCompactUIStyle() {is=title K2Node_CallFunction}
   > n9 = Assign On Input Method Changed!(
    execute:Exec
    self:Common Input Object Reference=Get CommonInputSubsystem!(PlayerController:Player Controller Object Reference=GetOwningPlayer()) -> (ReturnValue:Common Input Object Reference) {is=title K2Node_GetSubsystemFromPC}
    Delegate:"Delegate (by ref)"=@n3
   ) -> (then:Exec) {is=title K2Node_AssignDelegate}
   Broadcast Event!(
    execute=@n8.then_1
    self:Gameplay Event Router Component Object Reference=GetGlobalEventRouter()
    EventData:BPS Set Text Chat Active State Structure=Make BPS Set Text Chat Active State!(bSetActive_5_#2:bool=False) -> (BPS_SetTextChatActiveState:BPS Set Text Chat Active State Structure) {is=title K2Node_MakeStruct}
    EventContext:Object Reference=GetOwningPlayer()
   ) -> (then, ReturnValue:bool)
   Event On Activated()
   > PlaySound2D(Sound=/Game/Sounds/Fort_UI_Sounds/Gameplay/Menus/MSS_ui_map_open)
   > SetupCollabIcon()
   > UpdatePlayerXpVisibility()
   n10 = Get FortFrontEndContext() -> (ReturnValue:Fort Front End Context Object Reference) {is=title K2Node_GetSubsystem}
   > n11 = Is Valid(InputObject=@n10)
   > n12 = SetFrontEndCamera!(
    execute
    self:Fort Front End Context Object Reference=@n10
    NewState:EFrontEndCamera Enum=HeroSelect
    TransitionParams:View Target Transition Params Structure=(BlendTime=0,BlendFunction=VTBlend_Cubic,BlendExp=2,bLockOutgoing=False)
   ) -> (then)
   n13 = Cast To FortPlayerControllerGameplay!(execute=[@n12.then, @n11.Is Not Valid], Object=GetOwningPlayer()) -> (then, CastFailed, AsFort Player Controller Gameplay:Fort Player Controller Gameplay Object Reference) {is=title K2Node_DynamicCast}
   > Bind Event to On Creative Photo Mode Activated(self:Fort Player Controller Gameplay Object Reference=@n13, Delegate=Create Event()) {is=title K2Node_AddDelegate}
   Event On Deactivated()
   > PlaySound2D(Sound=/Game/Sounds/Fort_UI_Sounds/Gameplay/Menus/MSS_ui_map_close)
   > n14 = Cast To FortPlayerControllerGameplay(Object=GetOwningPlayer())
   > Unbind Event from On Creative Photo Mode Activated(self:Fort Player Controller Gameplay Object Reference=@n14, Delegate=Create Event()) {is=title K2Node_RemoveDelegate}
   Event Destruct()
   > n15 = Sequence()
   > n16 = Cast To FortPlayerControllerGameplay(Object=GetOwningPlayer())
   > Unbind Event from On Creative Photo Mode Deactivated(self:Fort Player Controller Gameplay Object Reference=@n16, Delegate=Create Event()) {is=title K2Node_RemoveDelegate}
   Branch(execute=@n15.then_1, Condition=Get IsAthenaHUD() -> (IsAthenaHUD:bool))
   > Broadcast Event(
    self=GetGlobalEventRouter()
    EventData=Make BPS Set Text Chat Active State(bSetActive_5_#2=true)
    EventContext=GetOwningPlayer()
   )
   Event OnUpdateActionBar() {is=title K2Node_Event}
   > SetVisibility(
    self=Get AthenaHUDBottomBar() -> (AthenaHUDBottomBar:WBP Athena Bottom Bar Widget Object Reference)
    InVisibility=Select(
     Option 0=Collapsed
     Option 1=SelfHitTestInvisible
     Index=ShouldShowActionBarForCurrentTab!(self:Athena Map Screen Container Object Reference) -> (ReturnValue:bool)
    )
   )
   > UpdatePlayerXpVisibility()
  ]}
 HandleCreativePhotoModeActivated {nodes [
   HandleCreativePhotoModeActivated() {is=title K2Node_FunctionEntry}
   > DeactivateWidget()
   > n1 = Cast To FortPlayerControllerGameplay(Object=GetOwningPlayer())
   > Bind Event to On Creative Photo Mode Deactivated(self:Fort Player Controller Gameplay Object Reference=@n1, Delegate=Create Event()) {is=title K2Node_AddDelegate}
  ]}
 HandleCreativePhotoModeDeactivated {nodes [
   HandleCreativePhotoModeDeactivated() {is=title K2Node_FunctionEntry}
   > n1 = Cast To FortPlayerControllerGameplay(Object=GetOwningPlayer())
   > Unbind Event from On Creative Photo Mode Deactivated(self=@n1, Delegate=Create Event())
   n2 = GetMiniMapComponent!(self:Fort Player Controller Object Reference=@n1) -> (ReturnValue:Fort Controller Component Minimap Object Reference)
   > Is Valid(InputObject=@n2)
   > OpenFullScreenMapTab!(execute, self:Fort Controller Component Minimap Object Reference=@n2, TabId:Name=GetMyIslandTabID!() -> (ReturnValue:Name=None)) -> (then)
  ]}
 Set BR Default Tabs {nodes [
   Set BR Default Tabs() {is=title K2Node_FunctionEntry}
   > Branch(Condition=AND Boolean(A=Get IsAthenaHUD(), B=UseNewMapScreen!(self:Athena Map Screen Container Object Reference) -> (ReturnValue:bool)))
   > Clear(TargetArray:Array of User Widget Soft Class References=Get DefaultTabs() -> (DefaultTabs:Array of User Widget Soft Class References))
   > Set DefaultTabs(DefaultTabs:Array of User Widget Soft Class References=Make Array("[0]":User Widget Soft Class Reference=Get NewBRMapScreen() -> (NewBRMapScreen:User Widget Soft Class Reference)) -> (Array:Array of User Widget Soft Class References)) -> (Output_Get:Array of User Widget Soft Class References)
  ]}
 SetCompactUIStyle {nodes [
   SetCompactUIStyle() {is=title K2Node_FunctionEntry}
   n1 = GetOwningPlayer()
   > Branch(Condition=AND Boolean(
     A=IsUsingCompactUI()
     B=IsUsingInputType!(InputType:ECommonInputType Enum=Touch, PlayerController:Player Controller Object Reference=@n1) -> (ReturnValue:bool)
    ))
   > SetOpacity!(execute, self:Image Object Reference=Get Image_DamageVignette() -> (Image_DamageVignette:Image Object Reference), InOpacity:float=0.3) -> (then)
   > SetRenderTranslation!(
    execute
    self=[
     Get S18_PlayerBattlePassXpAndReward_310() -> (S18_PlayerBattlePassXpAndReward_310:S18 Player Battle Pass Xp And Reward Object Reference)
     Get Overlay_SocialIcon() -> (Overlay_SocialIcon:Overlay Object Reference)
    ]
    Translation_X:double=0
    Translation_Y:double=12
   ) -> (then)
  ]}
 SetupCollabIcon {nodes [
   SetupCollabIcon() {is=title K2Node_FunctionEntry}
   > Set bCollabIsland(bCollabIsland:bool=false) -> (Output_Get:bool)
   > n1 = Sequence()
   > n2 = GetGameFeatureSetNameForContext!(execute, ContextActor:Actor Object Reference=GetOwningPlayer()) -> (then, ReturnValue:String)
   > n3 = Branch(Condition=IsEmpty(InString=@n2))
   n4 = GetCreativeIslandUIOverridesRegistryName!() -> (ReturnValue:Name=None)
   n5 = Get Data Registry Item!(
    execute=@n3.else
    ItemId:Data Registry Id Structure=Make Data Registry Id!(
     RegistryType:Data Registry Type Structure=Make Data Registry Type!(Name:Name=@n4) -> (DataRegistryType:Data Registry Type Structure) {is=title K2Node_MakeStruct}
     ItemName:Name=GetCreativeIslandUIOverridesDefaultRowName!() -> (ReturnValue:Name=None)
    ) -> (DataRegistryId:Data Registry Id Structure) {is=title K2Node_MakeStruct}
    OutItem:Fort Creative Island UIMetadata Table Row Structure=Get IslandUIMetadataEntry() -> (IslandUIMetadataEntry:Fort Creative Island UIMetadata Table Row Structure)
   ) -> (then, ReturnValue:bool) {is=title K2Node_CallFunction}
   > Set bCollabIsland(bCollabIsland=@n5)
   n6 = Break Fort Creative Island UIMetadata Table Row!(FortCreativeIslandUIMetadataTableRow:"Fort Creative Island UIMetadata Table Row Structure (by ref)"=Get IslandUIMetadataEntry()) -> (
    CreativeTagID:Name
    PortalUrl:String
    PortalQrImageUrl:String
    PublishUrl:String
    CollabIcon:Texture 2D Soft Object Reference
    CreativeTagLabelOverrides:Map of Names to Texts
    InfoButtonText:Text
    InfoButtonIcon:Texture 2D Soft Object Reference
    InfoButtonWidgetClass:Common Activatable Widget Soft Class Reference
   ) {is=title K2Node_BreakStruct}
   n7 = AND Boolean(A=IsValidSoftObjectReference(SoftObjectReference=@n6.CollabIcon), B=Get bCollabIsland() -> (bCollabIsland:bool))
   SetBrushFromSoftTexture!(
    execute=@n1.then_1
    self:Image Object Reference=Get Image_CollabIcon() -> (Image_CollabIcon:Image Object Reference)
    SoftTexture:Texture 2D Soft Object Reference=Select(Option 0:Texture 2D Soft Object Reference, Option 1:Texture 2D Soft Object Reference=@n6.CollabIcon, Index=@n7) -> (ReturnValue:Texture 2D Soft Object Reference)
    bMatchSize:bool=false
   ) -> (then)
   > SetVisibility(
    self=Get SizeBox_CollabIcon() -> (SizeBox_CollabIcon:Size Box Object Reference)
    InVisibility=Select(Option 0:ESlateVisibility Enum=Collapsed, Option 1:ESlateVisibility Enum=SelfHitTestInvisible, Index=@n7) -> (ReturnValue:ESlateVisibility Enum)
   )
  ]}
 ShowOrHideScoreboardUI {nodes [
   n1 = ShowOrHideScoreboardUI!() -> (then, ScoreboardWidget:Object Reference, bShow:bool=~) {is=title K2Node_FunctionEntry}
   > n2 = Cast To FortHideShowInterface!(execute, Object=@n1) -> (then, CastFailed, AsFort Hide Show Interface:Fort Hide Show Interface Interface) {is=title K2Node_DynamicCast}
   > n3 = Branch(Condition=@n1.bShow)
   > DoShowUI(self:Fort Hide Show Interface Interface=@n2)
   DoHideUI(execute=@n3.else, self:Fort Hide Show Interface Interface=@n2)
  ]}
 UpdatePlayerXpVisibility {nodes [
   UpdatePlayerXpVisibility() {is=title K2Node_FunctionEntry}
   > Set bShouldHideXp(bShouldHideXp:bool=false) -> (Output_Get:bool)
   > n1 = Sequence()
   > n2 = Cast To FortPlayerControllerGameplay(Object=GetOwningPlayer())
   n3 = ShouldAlwaysShowXpBar!(self:Athena Map Screen Container Object Reference) -> (ReturnValue:bool)
   > Set bShouldHideXp(bShouldHideXp=AND Boolean(A=IsCreativeModeEnabled!(self:Fort Player Controller Gameplay Object Reference=@n2) -> (ReturnValue:bool), B=NOT Boolean(A=@n3)))
   SetVisibility(
    execute=@n1.then_1
    self=Get S18_PlayerBattlePassXpAndReward_310()
    InVisibility=Select(Option 0=SelfHitTestInvisible, Option 1=Collapsed, Index=Get bShouldHideXp() -> (bShouldHideXp:bool))
   )
   > n4 = Branch(Condition=Get bShouldHideXp())
   n5 = Get S18_PlayerBattlePassXpAndReward_310()
   > DeactivateWidget(self=@n5)
   ActivateWidget(execute=@n4.else, self=@n5)
  ]}
 UpdateTabBackgroundSize {nodes [
   UpdateTabBackgroundSize() {is=title K2Node_FunctionEntry}
   > n1 = Branch(Condition=Get ShowTabs() -> (ShowTabs:bool))
   > SetHeightOverride(self=Get SizeBox_Tabs() -> (SizeBox_Tabs:Size Box Object Reference), InHeightOverride=SelectFloat(A=100, B=70, bPickA=IsMobileGame()))
   > SetVisibility(
    self=[Get TabBackground() -> (TabBackground:Image Object Reference), Get Overlay_TopBar() -> (Overlay_TopBar:Overlay Object Reference)]
    InVisibility=SelfHitTestInvisible
   )
   > SetHeightOverride(
    self=Get SizeBox_Mobile() -> (SizeBox_Mobile:Size Box Object Reference)
    InHeightOverride=Select(Option 0:float=70, Option 1:float=100, Index=IsMobileGame()) -> (ReturnValue:float)
   )
   SetVisibility(
    execute=@n1.else
    self=[Get Overlay_TopBar(), Get TabBackground()]
    InVisibility=Select(Option 0:ESlateVisibility Enum=SelfHitTestInvisible, Option 1:ESlateVisibility Enum=Collapsed, Index=IsMobileGame()) -> (ReturnValue:ESlateVisibility Enum)
   )
  ]}
]
variables [
 DesignTimeTabCount {type=integer, default=3}
 IsAthenaHUD {type=boolean}
 IslandUIMetadataEntry {
  type=FortCreativeIslandUIMetadataTableRow
  default {
   collabIcon=null
   creativeTagId=None
   infoButtonIcon=null
   infoButtonText=""
   infoButtonWidgetClass=null
   portalQrImageUrl=""
   portalUrl=""
   publishUrl=""
   creativeTagLabelOverrides {}
  }
 }
 ShowTabs {type=boolean}
 TabsAlwaysVisible {type=boolean}
]
widgets {
 info {inheritedWidgetCount=6, parentClass=/Script/FortniteUI.AthenaMapScreenContainer, rootWidgetClass=&Overlay, widgetCount=31}
 widgets [
  Overlay_HitTestable {widgetClassPath=&Overlay, widget=Overlay_HitTestable, properties {visibility=Visible}}
  VerticalBox_0 {
   widgetClassPath=&VerticalBox
   parent=Overlay_HitTestable
   widget=VerticalBox_0
   slot {class=OverlaySlot, properties {horizontalAlignment=HAlign_Fill, verticalAlignment=VAlign_Fill}}
  }
  Overlay_TopBar {widgetClassPath=&Overlay, bIsVariable=true, parent=VerticalBox_0, widget=Overlay_TopBar, slot {class=VerticalBoxSlot}}
  TabBackground {
   widgetClassPath=&Image
   bIsVariable=true
   parent=Overlay_TopBar
   widget=TabBackground
   properties {
    visibility=SelfHitTestInvisible
    brush {
     bIsDynamicallyLoaded=false
     drawAs=Image
     imageType=NoImage
     mirroring=NoMirror
     resourceName=None
     resourceObject=null
     tiling=NoTile
     imageSize=<32 32>
     margin=<ltrb 0 0 0 0>
     outlineSettings {
      bUseBrushTransparency=false
      roundingType=HalfHeightRadius
      width=0
      color {colorUseRule=UseColor_Specified, specifiedColor=<rgba 0 0 0 0>}
      cornerRadii=<0 0 0 0>
     }
     tintColor {colorUseRule=UseColor_Specified, specifiedColor=<rgba 0.0019 0.0043 0.075 1>}
     uVRegion {bIsValid=false, max=<0 0>, min=<0 0>}
    }
   }
   slot {class=OverlaySlot, properties {horizontalAlignment=HAlign_Fill, verticalAlignment=VAlign_Fill}}
  }
  SafeZone_TabButtons {
   widgetClassPath=&SafeZone
   parent=Overlay_TopBar
   widget=SafeZone_TabButtons
   properties {padBottom=false, padRight=false}
   slot {class=OverlaySlot, properties {horizontalAlignment=HAlign_Fill, verticalAlignment=VAlign_Fill, padding=<ltrb 16 0 0 0>}}
  }
  SizeBox_Tabs {
   widgetClassPath=&SizeBox
   bIsVariable=true
   parent=SafeZone_TabButtons
   widget=SizeBox_Tabs
   properties {bOverride_HeightOverride=true, heightOverride=70}
   slot {class=SafeZoneSlot}
  }
  HBox_Tabs {widgetClassPath=&HorizontalBox, bIsVariable=true, parent=SizeBox_Tabs, widget=HBox_Tabs, slot {class=SizeBoxSlot}}
  PreviousTabVisibilityWidget {
   widgetClassPath=&CommonVisibilityWidgetLegacy
   parent=HBox_Tabs
   widget=PreviousTabVisibilityWidget
   properties {bShowForMouseAndKeyboard=false, verticalAlignment=VAlign_Center, padding=<ltrb 0 0 0 0>}
   slot {class=HorizontalBoxSlot, properties {verticalAlignment=VAlign_Center, padding=<ltrb 0 0 10 0>}}
  }
  PreviousTabAction {
   widgetClassPath=&CommonActionWidget
   bIsVariable=true
   parent=PreviousTabVisibilityWidget
   widget=PreviousTabAction
   slot {class=BorderSlot, properties {verticalAlignment=VAlign_Center, padding=<ltrb 0 0 0 0>}}
  }
  SizeBox_Mobile {
   widgetClassPath=&SizeBox
   bIsVariable=true
   parent=HBox_Tabs
   widget=SizeBox_Mobile
   properties {bOverride_HeightOverride=true, heightOverride=70}
   slot {class=HorizontalBoxSlot, properties {verticalAlignment=VAlign_Top}}
  }
  Border {
   widgetClassPath=&Border
   parent=SizeBox_Mobile
   widget=Border
   properties {
    background {
     bIsDynamicallyLoaded=false
     drawAs=NoDrawType
     imageType=NoImage
     mirroring=NoMirror
     resourceName=None
     resourceObject=null
     tiling=NoTile
     imageSize=<32 32>
     margin=<ltrb 0.5 0 0.5 0>
     outlineSettings {
      bUseBrushTransparency=false
      roundingType=HalfHeightRadius
      width=0
      color {colorUseRule=UseColor_Specified, specifiedColor=<rgba 0 0 0 0>}
      cornerRadii=<0 0 0 0>
     }
     tintColor {colorUseRule=UseColor_Specified, specifiedColor=<rgba 1 1 1 1>}
     uVRegion {bIsValid=false, max=<0 0>, min=<0 0>}
    }
    padding=<ltrb 0 0 0 0>
   }
   slot {class=SizeBoxSlot}
  }
  Panel_TabButtonsContainer {
   widgetClassPath=&HorizontalBox
   bInherited=true
   parent=Border
   widget=Panel_TabButtonsContainer
   slot {class=BorderSlot, properties {padding=<ltrb 0 0 0 0>}}
  }
  NextTabVisibilityWidget {
   widgetClassPath=&CommonVisibilityWidgetLegacy
   parent=HBox_Tabs
   widget=NextTabVisibilityWidget
   properties {bShowForMouseAndKeyboard=false, verticalAlignment=VAlign_Center, padding=<ltrb 0 0 0 0>}
   slot {class=HorizontalBoxSlot, properties {verticalAlignment=VAlign_Center, padding=<ltrb 10 0 0 0>}}
  }
  NextTabAction {
   widgetClassPath=&CommonActionWidget
   bIsVariable=true
   parent=NextTabVisibilityWidget
   widget=NextTabAction
   slot {class=BorderSlot, properties {verticalAlignment=VAlign_Center, padding=<ltrb 0 0 0 0>}}
  }
  VisibilitySwitcher_Content {
   widgetClassPath=&CommonVisibilitySwitcher
   bInherited=true
   parent=VerticalBox_0
   widget=VisibilitySwitcher_Content
   properties {clipping=ClipToBounds}
   slot {class=VerticalBoxSlot, properties {size {sizeRule=Fill, value=1}}}
  }
  Content_CreativeScoreboardContainer {
   widgetClassPath=&SizeBox
   parent=Overlay_HitTestable
   widget=Content_CreativeScoreboardContainer
   slot {class=OverlaySlot, properties {horizontalAlignment=HAlign_Fill, verticalAlignment=VAlign_Fill}}
  }
  Image_DamageVignette {
   widgetClassPath=&Image
   bIsVariable=true
   parent=Overlay_HitTestable
   widget=Image_DamageVignette
   properties {
    renderOpacity=0
    visibility=HitTestInvisible
    brush {
     bIsDynamicallyLoaded=false
     drawAs=Image
     imageType=FullColor
     mirroring=NoMirror
     resourceName=None
     resourceObject=/Game/Athena/HUD/Materials/M_UI_DamageVignette_Inst
     tiling=NoTile
     imageSize=<32 32>
     margin=<ltrb 0 0 0 0>
     outlineSettings {
      bUseBrushTransparency=false
      roundingType=HalfHeightRadius
      width=0
      color {colorUseRule=UseColor_Specified, specifiedColor=<rgba 0 0 0 0>}
      cornerRadii=<0 0 0 0>
     }
     tintColor {colorUseRule=UseColor_Specified, specifiedColor=<rgba 1 0 0.0177 1>}
     uVRegion {bIsValid=false, max=<0 0>, min=<0 0>}
    }
   }
   slot {class=OverlaySlot, properties {horizontalAlignment=HAlign_Fill, verticalAlignment=VAlign_Fill}}
  }
  SafeZone_RightSideWidgets {
   widgetClassPath=&SafeZone
   parent=Overlay_HitTestable
   widget=SafeZone_RightSideWidgets
   slot {class=OverlaySlot, properties {horizontalAlignment=HAlign_Right}}
  }
  HBox_RightSideWidgets {widgetClassPath=&HorizontalBox, parent=SafeZone_RightSideWidgets, widget=HBox_RightSideWidgets, slot {class=SafeZoneSlot}}
  Overlay_PlayerProfile {
   widgetClassPath=&Overlay
   bIsVariable=true
   parent=HBox_RightSideWidgets
   widget=Overlay_PlayerProfile
   slot {class=HorizontalBoxSlot, properties {horizontalAlignment=HAlign_Center, verticalAlignment=VAlign_Top, padding=<ltrb 16 10 24 16>}}
  }
  Panel_PlayerProfile {
   widgetClassPath=&HorizontalBox
   bInherited=true
   parent=Overlay_PlayerProfile
   widget=Panel_PlayerProfile
   slot {class=OverlaySlot}
  }
  Overlay_SocialIcon {widgetClassPath=&Overlay, bIsVariable=true, parent=Panel_PlayerProfile, widget=Overlay_SocialIcon, slot {class=HorizontalBoxSlot}}
  SocialAvatarIcon {
   widgetClassPath=/Script/FortniteUI.FortSocialAvatarIcon
   bInherited=true
   bIsVariable=true
   parent=Overlay_SocialIcon
   widget=SocialAvatarIcon
   properties {
    avatarMaterial=/Game/UI/SocialPanel/Art/Materials/UI_UserPortrait_Buffet_Small
    materialTextureParamName=SmallPreview
    brush {
     bIsDynamicallyLoaded=false
     drawAs=Image
     imageType=NoImage
     mirroring=NoMirror
     resourceName=None
     resourceObject=&UI_UserPortrait_Buffet_Small
     tiling=NoTile
     imageSize=<55 55>
     margin=<ltrb 0 0 0 0>
     outlineSettings {
      bUseBrushTransparency=false
      roundingType=HalfHeightRadius
      width=0
      color {colorUseRule=UseColor_Specified, specifiedColor=<rgba 0 0 0 0>}
      cornerRadii=<0 0 0 0>
     }
     tintColor {colorUseRule=UseColor_Specified, specifiedColor=<rgba 1 1 1 1>}
     uVRegion {bIsValid=false, max=<0 0>, min=<0 0>}
    }
   }
   slot {class=OverlaySlot}
  }
  OnlineStatusIndicator {
   widgetClassPath=&Image
   bIsVariable=true
   parent=Overlay_SocialIcon
   widget=OnlineStatusIndicator
   properties {
    visibility=Collapsed
    brush {
     bIsDynamicallyLoaded=false
     drawAs=Image
     imageType=NoImage
     mirroring=NoMirror
     resourceName=None
     resourceObject=/Game/UI/Friends_UI/Social/circle
     tiling=NoTile
     imageSize=<12 12>
     margin=<ltrb 0 0 0 0>
     outlineSettings {
      bUseBrushTransparency=false
      roundingType=HalfHeightRadius
      width=0
      color {colorUseRule=UseColor_Specified, specifiedColor=<rgba 0 0 0 0>}
      cornerRadii=<0 0 0 0>
     }
     tintColor {colorUseRule=UseColor_Specified, specifiedColor=<rgba 1 1 1 1>}
     uVRegion {bIsValid=false, max=<0 0>, min=<0 0>}
    }
    colorAndOpacity=<rgba 0.1946 0.9823 0.1812 1>
   }
   slot {class=OverlaySlot, properties {horizontalAlignment=HAlign_Right, verticalAlignment=VAlign_Bottom, padding=<ltrb 3 3 3 3>}}
  }
  S18_PlayerBattlePassXpAndReward_310 {
   widgetClassPath=/FortniteCoreUI/UI/Progression/S18_PlayerBattlePassXpAndReward
   bIsVariable=true
   parent=Panel_PlayerProfile
   widget=S18_PlayerBattlePassXpAndReward_310
   slot {class=HorizontalBoxSlot, properties {verticalAlignment=VAlign_Center, padding=<ltrb 6 0 0 0>}}
  }
  SizeBox_CollabIcon {
   widgetClassPath=&SizeBox
   bIsVariable=true
   parent=HBox_RightSideWidgets
   widget=SizeBox_CollabIcon
   properties {bOverride_HeightOverride=true, bOverride_WidthOverride=true, heightOverride=70, visibility=Collapsed, widthOverride=70}
   slot {class=HorizontalBoxSlot, properties {horizontalAlignment=HAlign_Center, verticalAlignment=VAlign_Top}}
  }
  Image_CollabIcon {
   widgetClassPath=&Image
   bIsVariable=true
   parent=SizeBox_CollabIcon
   widget=Image_CollabIcon
   properties {
    visibility=SelfHitTestInvisible
    brush {
     bIsDynamicallyLoaded=false
     drawAs=Image
     imageType=FullColor
     mirroring=NoMirror
     resourceName=None
     resourceObject=null
     tiling=NoTile
     imageSize=<128 128>
     margin=<ltrb 0 0 0 0>
     outlineSettings {
      bUseBrushTransparency=false
      roundingType=HalfHeightRadius
      width=0
      color {colorUseRule=UseColor_Specified, specifiedColor=<rgba 0 0 0 0>}
      cornerRadii=<0 0 0 0>
     }
     tintColor {colorUseRule=UseColor_Specified, specifiedColor=<rgba 1 1 1 1>}
     uVRegion {bIsValid=false, max=<0 0>, min=<0 0>}
    }
   }
   slot {class=SizeBoxSlot}
  }
  HardwareVisibilityBorder_TouchClose {
   widgetClassPath=&CommonHardwareVisibilityBorder
   parent=HBox_RightSideWidgets
   widget=HardwareVisibilityBorder_TouchClose
   properties {visibilityQuery {
     autoDescription=" ANY( Input.Touch )"
     tokenStreamVersion=0
     userDescription=""
     queryTokenStream [0,1,1,1,0]
     tagDictionary [{tagName=Input.Touch}]
    }}
   slot {class=HorizontalBoxSlot, properties {padding=<ltrb 24 25 25 0>}}
  }
  Button_TouchClose {
   widgetClassPath=/FortUILibrary/Foundation/Buttons/CloseButton
   bInherited=true
   bIsVariable=true
   parent=HardwareVisibilityBorder_TouchClose
   widget=Button_TouchClose
   properties {
    onButtonBaseClicked=(&self:WidgetTree.Button_TouchClose.HandleOnButtonClicked)
    onButtonBaseDoubleClicked=(&self:WidgetTree.Button_TouchClose.HandleOnButtonDoubleClicked)
    onButtonBaseHovered=(&self:WidgetTree.Button_TouchClose.HandleOnButtonHovered)
    onButtonBaseUnhovered=(&self:WidgetTree.Button_TouchClose.HandleOnButtonUnhovered)
    onSelectedChangedBase=(&self:WidgetTree.Button_TouchClose.HandleOnSelectedChanged)
   }
   slot {class=BorderSlot, properties {padding=<ltrb 0 0 0 0>}}
  }
  Widget_AutoRunNotification {
   widgetClassPath=/Game/Athena/HUD/AutoRun/FortAutoRunNotificationWidget
   bInherited=true
   bIsVariable=true
   parent=Overlay_HitTestable
   widget=Widget_AutoRunNotification
   properties {visibility=Collapsed}
   slot {class=OverlaySlot, properties {horizontalAlignment=HAlign_Right, padding=<ltrb 0 140 0 0>}}
  }
  AthenaHUDBottomBar {
   widgetClassPath=/Game/UI/Frontend/BottomBar/WBP_AthenaBottomBarWidget
   bIsVariable=true
   parent=Overlay_HitTestable
   widget=AthenaHUDBottomBar
   slot {class=OverlaySlot, properties {horizontalAlignment=HAlign_Right, verticalAlignment=VAlign_Bottom}}
  }
 ]
}
