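# Root mobile (Bacchus) HUD widget for Athena matches, derived from AthenaMobileHUD.
# Layers persistent HUD with touch controls and extension slots, reticle, HUD menu and map container, mostly inside tagged element wrappers.
# HUD-context and controller delegates drive game-over, spectate and revive transitions, tag-based hiding for bus and inventory, and indicator visibility.
# Also opens the Creative options modal asynchronously and removes the event slot for dynamic-touch-HUD playlists.
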WidgetBlueprint /FortniteCoreUI/HUD/AthenaHUD_Bacchus
graphs [
 CheckHUDElementVisibility {nodes [
   n1 = CheckHUDElementVisibility!() -> (
    then
    HiddenHUDElementTags:"Gameplay Tag Container Structure (by ref)"
    HUDElementTagToCheck:Gameplay Tag Structure
    HUDElement:"Widget Object Reference (by ref)"
   ) {is=title K2Node_FunctionEntry}
   > n2 = Branch(Condition=HasAnyTags!(
     TagContainer:"Gameplay Tag Container Structure (by ref)"=MakeGameplayTagContainerFromTag!(SingleTag:Gameplay Tag Structure=@n1.HUDElementTagToCheck) -> (ReturnValue:Gameplay Tag Container Structure)
     OtherContainer:"Gameplay Tag Container Structure (by ref)"=@n1
     bExactMatch:bool=false
    ) -> (ReturnValue:bool))
   n3 = IsVisible(self=@n1.HUDElement)
   > Branch(Condition=@n3)
   > SetVisibility(self=@n1.HUDElement, InVisibility=Hidden)
   n4 = Branch(execute=@n2.else, Condition=@n3)
   SetVisibility(execute=@n4.else, self=@n1.HUDElement, InVisibility=SelfHitTestInvisible)
  ]}
 EventGraph {nodes [
   n1 = HandleLocalPawnSet() {is=title K2Node_CustomEvent}
   > OnPawnSet()
   OnPawnSet() {is=title K2Node_CustomEvent}
   > n2 = Cast To FortPlayerControllerAthena(Object=GetOwningPlayer())
   > n3 = Branch(Condition=GetIsSpectating!(self:Fort Player Controller Athena Object Reference=@n2) -> (ReturnValue:bool))
   n4 = SetZoneCompleted!(execute=@n3.else, self:Athena HUDBase Object Reference, bValue:bool=false) -> (then)
   EnterCameraMode(
    execute=[@n4.then, @n17.then]
    self:Fort HUDContext Object Reference=Get FortHUDContext!() -> (ReturnValue:Fort HUDContext Object Reference) {is=title K2Node_GetBlueprintContextEx}
   )
   n5 = GetOwningPlayer()
   > n6 = Cast To FortPlayerControllerAthena(Object=@n5)
   n7 = MakeGameplayTagContainerFromTag(SingleTag=(TagName="HUD.Athena.HideOnSpectate"))
   > SetHUDElementVisibilityWithContext!(
    execute
    self:Fort Player Controller Object Reference=@n6
    HUDElementTags:"Gameplay Tag Container Structure (by ref)"=@n7
    bHideElements:bool=false
    ReasonName:Name=None
    LogContext:String
   ) -> (then)
   > SetHUDElementVisibilityWithContext(
    self=@n6
    HUDElementTags=MakeGameplayTagContainerFromTag(SingleTag=(TagName="HUD.Athena.HideForAbility"))
    bHideElements=false
    ReasonName=None
   )
   > SetHUDElementVisibilityWithContext(
    self=@n6
    HUDElementTags=MakeGameplayTagContainerFromTag(SingleTag=(TagName="HUD.Athena.HideSecondaryForAbility"))
    bHideElements=false
    ReasonName=None
   )
   n8 = HandleRevived() {is=title K2Node_CustomEvent}
   > OnRevived()
   n9 = On Placement Changed() {is=title K2Node_CustomEvent}
   > n10 = Branch(Condition=IsSpatialGameplayExperience!() -> (ReturnValue:bool))
   > n11 = Cast To FortPlayerControllerAthena(Object=GetOwningPlayer())
   n12 = GetMinigame!(ContextActor:Actor Object Reference=@n11) -> (ReturnValue:Fort Minigame Object Reference)
   > Is Valid(InputObject=@n12)
   n13 = GetMinigameSettingsMachine!(ContextActor:Actor Object Reference=@n12) -> (ReturnValue:Fort Minigame Settings Building Object Reference)
   > Is Valid(InputObject=@n13)
   n14 = GetPostGameSettings!(self:Fort Minigame Settings Building Object Reference=@n13) -> (ReturnValue:"Minigame Post Game Settings Structure (by ref)")
   > n15 = Branch(Condition="Not Equal (Enum)"!(
     A:EMinigameGameEndPostGameType Enum=Break Minigame Post Game Settings!(MinigamePostGameSettings:"Minigame Post Game Settings Structure (by ref)"=@n14) -> (
      PostGameType:EMinigameGameEndPostGameType Enum
      PostGameDelay:float=~
      bCustomShowScoreboard:bool=~
      VictoryAnimationSettings:Minigame Post Game Animation Settings Structure
      DefeatAnimationSettings:Minigame Post Game Animation Settings Structure
      TieAnimationSettings:Minigame Post Game Animation Settings Structure
     ) {is=title K2Node_BreakStruct}
     B:EMinigameGameEndPostGameType Enum=Classic
    ) -> (ReturnValue:bool))
   OnDisplayGameOver(execute=[@n15.then, @n10.else])
   > SetHorizontalAlignment(
    self=SlotAsOverlaySlot(Widget=Get TopMidVerticalBox() -> (TopMidVerticalBox:Vertical Box Object Reference))
    InHorizontalAlignment=HAlign_Left
   )
   OnRevived() {is=title K2Node_CustomEvent}
   > n16 = Branch(Condition=Get PostGameMenuStarted() -> (PostGameMenuStarted:bool))
   SetZoneCompleted(execute=@n16.else, bValue=false)
   > n17 = RecoverFromPostGame(self:Athena HUDMenu Object Reference=Get AthenaHudMenu() -> (AthenaHudMenu:Athena HUDMenu Object Reference))
   n18 = OnUnableToPerformAction!() -> (OutputDelegate, then, FailedReason:Gameplay Tag Container Structure, FailureText:"Text (by ref)") {is=title K2Node_CustomEvent}
   > Branch(Condition=HasTag!(
     TagContainer:"Gameplay Tag Container Structure (by ref)"=@n18.FailedReason
     Tag:Gameplay Tag Structure=(TagName="Athena.Ability.Failed.Consumable")
     bExactMatch:bool=false
    ) -> (ReturnValue:bool))
   > PlaySound2D(Sound=/Game/Athena/Sounds/Items/Consumable_CantApply_Cue)
   HandleShowCreativeGlobalOptions() {is=title K2Node_CustomEvent}
   > n19 = CreateWidgetAsync!(
    execute:Exec
    UserWidgetSoftClass:User Widget Soft Class Reference=/Game/Athena/HUD/CreativeHUDGlobalOptions_Modal.CreativeHUDGlobalOptions_Modal_C
    OwningPlayer:Player Controller Object Reference=GetOwningPlayer()
    bSuspendInputUntilComplete:bool=true
   ) -> (then:Exec, OnComplete:Exec, OnFailure:Exec, UserWidget:User Widget Object Reference) {is=title K2Node_AsyncAction}
   n20 = Cast To CommonActivatablePanelLegacy!(execute=@n19.OnComplete, Object=@n19) -> (then, CastFailed, AsCommon Activatable Panel Legacy:Common Activatable Panel Legacy Object Reference) {is=title K2Node_DynamicCast}
   > TryPushWidgetAsActivatablePanel!(
    execute
    PushWidget:User Widget Object Reference=@n20
    Header:EActivatePanelOption Enum=Hide
    Footer:EActivatePanelOption Enum=Hide
    ChatWidget:EActivatePanelOption Enum=Hide
   ) -> (then)
   > EnterCursorMode!(
    execute
    self:Fort HUDContext Object Reference=Get FortHUDContext()
    ActionName:Name=Show Quick Options
    CursorModeWidget:User Widget Object Reference=@n20
   ) -> (then)
   > ForwardOnCreativeQuickMenuDisplayedChanged!(
    execute
    self:Creative HUDLContext Object Reference=Get CreativeHUDLContext!() -> (ReturnValue:Creative HUDLContext Object Reference) {is=title K2Node_GetBlueprintContextEx}
    bIsDisplayed:bool=true
   ) -> (then)
   n21 = On Player Died!() -> (OutputDelegate, then, DeathReport:"Fort Player Death Report Structure (by ref)") {is=title K2Node_CustomEvent}
   > PrepareToShowEndGameUI()
   n22 = HandleInputMethodChanged!() -> (OutputDelegate, then, InputType:ECommonInputType Enum) {is=title K2Node_CustomEvent}
   > SetVisibility(InVisibility=Select(Option 0=Hidden, Option 1=SelfHitTestInvisible, Index="Equal (Enum)"(A=@n22.InputType, B=Gamepad)))
   Event PrepareToShowEndGameUI() {is=title K2Node_Event}
   > n23 = Branch(Condition=GetZoneCompleted!(self:Athena HUDBase Object Reference) -> (ReturnValue:bool))
   SetZoneCompleted(execute=@n23.else, bValue=true)
   > EnterCursorMode(self=Get FortHUDContext(), ActionName=None)
   > n24 = Cast To FortPlayerControllerAthena(Object=@n5)
   > SetHUDElementVisibilityWithContext(self=@n24, HUDElementTags=@n7, bHideElements=true, ReasonName=None)
   n25 = Event HandleCursorModeChanged!() -> (OutputDelegate, then, bCursorModeEnabled:bool, ActionName:Name=None, CustomWidget:User Widget Object Reference) {is=title K2Node_Event}
   > n26 = Branch(Condition=GetZoneCompleted())
   > n27 = Branch(Condition=@n25.bCursorModeEnabled)
   SetInputMode!(
    execute=[@n26.else, @n27.then]
    self:Fort Global UIContext Object Reference=Get FortGlobalUIContext!() -> (ReturnValue:Fort Global UIContext Object Reference) {is=title K2Node_GetBlueprintContextEx}
    InMode:EFortInputMode Enum=Select(Option 0:EFortInputMode Enum=InGame, Option 1:EFortInputMode Enum=InGameCursor, Index=@n25.bCursorModeEnabled) -> (ReturnValue:EFortInputMode Enum)
   ) -> (then)
   > UpdateCursorModeVisibility!(execute, self, CursorMode:Name=@n25.ActionName, InCursorModeEnabled:bool=@n25.bCursorModeEnabled) -> (then)
   EnterCursorMode(execute=@n27.else, self=Get FortHUDContext(), ActionName=None)
   Event Construct()
   > Initialize!(
    execute
    self:Common Tag Visibility Widget Object Reference=Get TagVisibilityWidget_Controls() -> (TagVisibilityWidget_Controls:Common Tag Visibility Widget Object Reference)
    OwningLocalPlayer:Local Player Object Reference=GetOwningLocalPlayer!(self:Widget Object Reference) -> (ReturnValue:Local Player Object Reference)
    OwningPlayerController:Player Controller Object Reference=GetPlayerController!(PlayerIndex:int=0) -> (ReturnValue:Player Controller Object Reference)
   ) -> (then)
   n28 = Create Event()
   n29 = Get FortHUDContext()
   > Bind Event to On Unable to Perform Action(self:Fort HUDContext Object Reference=@n29, Delegate=@n28) {is=title K2Node_AddDelegate}
   > Prepare Game Over()
   > n30 = Cast To FortPlayerControllerAthena(Object=GetOwningPlayer())
   n31 = Create Event()
   > Assign On Aircraft State Change!(execute:Exec, self:Fort Player Controller Athena Object Reference=@n30, Delegate:"Delegate (by ref)"=@n31) -> (then:Exec) {is=title K2Node_AssignDelegate}
   > HandleAircraftModeChanged()
   n32 = Create Event()
   n33 = Get FortHUDContext()
   > Bind Event to On Indicator Mode Changed(self:Fort HUDContext Object Reference=@n33, Delegate=@n32) {is=title K2Node_AddDelegate}
   n34 = Create Event()
   n35 = Get FortGlobalUIContext()
   > Bind Event to On Keybinds Changed(self:Fort Global UIContext Object Reference=@n35, Delegate=@n34) {is=title K2Node_AddDelegate}
   > HandleKeybindsChanged()
   n36 = Create Event()
   n37 = Get FortHUDContext()
   > Bind Event to On HUDElement Visibility Changed(self:Fort HUDContext Object Reference=@n37, Delegate=@n36) {is=title K2Node_AddDelegate}
   > SetFortPickerData!(
    execute
    self:Fort Picker Context Object Reference=Get FortPickerContext!(PlayerController:Player Controller Object Reference=GetOwningPlayer()) -> (ReturnValue:Fort Picker Context Object Reference) {is=title K2Node_GetSubsystemFromPC}
    Data:Fort Picker Data Object Reference=/Game/UI/InGame/HUD/FortPickers
   ) -> (then)
   > n38 = Cast To FortPlayerControllerGameplay!(execute, Object=GetOwningPlayer()) -> (then, CastFailed, AsFort Player Controller Gameplay:Fort Player Controller Gameplay Object Reference) {is=title K2Node_DynamicCast}
   n39 = Create Event()
   > Bind Event to On Show Creative Global Options(self:Fort Player Controller Gameplay Object Reference=@n38, Delegate=@n39) {is=title K2Node_AddDelegate}
   > Branch(Condition=Get bEnableHUDLayoutToolV2!(self:Fort Runtime Options Object Reference=GetRuntimeOptions()) -> (bEnableHUDLayoutToolV2:bool))
   > AddLooseGameplayTags!(
    execute
    Actor:Actor Object Reference=GetOwningPlayer()
    GameplayTags:"Gameplay Tag Container Structure (by ref)"=MakeGameplayTagContainerFromTag(SingleTag=(TagName="MobileHUD.TouchControlsActive"))
    bShouldReplicate:bool=false
   ) -> (then, ReturnValue:bool)
   n40 = Event OnExtensionWidgetRemoved!() -> (OutputDelegate, then, ExtensionSlot:EUIExtensionSlot Enum=Invalid, Widget:User Widget Object Reference) {is=title K2Node_Event}
   > n41 = Switch on EUIExtensionSlot!(execute, Selection:EUIExtensionSlot Enum=@n40.ExtensionSlot) -> (
    Invalid:Exec
    Primary:Exec
    Secondary:Exec
    TopRightCorner:Exec
    GameInfoBox:Exec
    UpperCenter:Exec
    CrosshairRight:Exec
    MobileCrosshairRight:Exec
    UnderSquadInfo:Exec
    FullScreenMap:Exec
    BelowRespawnWidget:Exec
    BelowCompass:Exec
    UnderTeammateStatus:Exec
    ControllerBindingCallout:Exec
    AboveStormMessageSlot:Exec
    CustomMinigameCallouts:Exec
    UnderLocalPlayerInfo:Exec
    PlayerHealthbarOverlay:Exec
    UnderMinimapMessageContainer:Exec
    VehicleActionsSpacer:Exec
    Reticle:Exec
    PrioritizedContextualSlot:Exec
    RightOfTeammateStatus:Exec
    RightOfTeammateStatusBars:Exec
    TeammateStatusPortraitOverlay:Exec
    TeammateStatusPortraitUnderlay:Exec
    TeammateStatusPortraitReplacement:Exec
    TeammateStatusPortraitOverlay_IGM:Exec
    TeammateOverHealthbarOverlay:Exec
    TeammateOverShieldbarOverlay:Exec
    MobileHUDBottomRight:Exec
    InventoryScreenTab:Exec
    InventoryEquipSlot:Exec
    UnderneathTeammateStatusList:Exec
    FullScreenMapSquadEliminationsOverlay:Exec
    FullScreenMapStormInfo:Exec
    PrimaryQuickBarSlot:Exec
    MainMenuButtonListInZone:Exec
    MainMenuButtonListFrontEnd:Exec
    DynamicOverlayLobby:Exec
    PostMatchXPDetails_AccoladeList:Exec
    MobileHUDUnderMinimap:Exec
    SidebarPanels:Exec
    SpectatorRightOfPlayerStatus:Exec
    BehindMobileLocalPlayerInfo:Exec
    UnderMobileLocalPlayerInfo:Exec
    AboveMobileLocalPlayerInfo:Exec
    DynamicOverlayFrontEnd:Exec
    NPCStatusWidgetBarsPrefix:Exec
    NPCStatusWidgetAboveBars:Exec
    AbovePlayerIndicatorWidget:Exec
    FullScreenMapTab:Exec
    PostMatchScreenTab:Exec
    AboveBacchusEquippedItemReloadWidget:Exec
    UnderDefaultLobbyDetails:Exec
    DiscoveredAreaTextUnderlay:Exec
    LevelUpMobile:Exec
    ExpMobile:Exec
    InventoryGroupPanel:Exec
    InventoryItemInfoContents:Exec
    DiscoverInLobby:Exec
    InventoryItemInfoInspectPreview:Exec
    AboveQuestsMobileUI:Exec
    PostGamePlacementOverlayVictory:Exec
    TeammateStatusHealthShieldsReplacement:Exec
    GameInfoBelowMap:Exec
    LeftPostGameMobileOverlay:Exec
    LobbyAboveRankedDetails:Exec
    HUDPlayerActionAlertKillPlayerInfoExtension:Exec
    PostGamePlacementOverlayLosing:Exec
    AthenaInteractionIndicator:Exec
    HUDUnderReticle:Exec
    AthenaInventoryGroupMinorMedallionHeader:Exec
    AboveUpperCenterPrioritized:Exec
   ) {is=title K2Node_SwitchEnum}
   RemoveChild(execute=@n41.Primary, self=Get ExtensionLayer() -> (ExtensionLayer:Overlay Object Reference), Content=@n40.Widget)
   Event Destruct()
   > Unbind Event from On Unable to Perform Action(self:Fort HUDContext Object Reference=@n29, Delegate=@n28) {is=title K2Node_RemoveDelegate}
   > n42 = Cast To FortPlayerControllerAthena(Object=GetOwningPlayer())
   > Unbind Event from On Aircraft State Change(self:Fort Player Controller Athena Object Reference=@n42, Delegate=@n31) {is=title K2Node_RemoveDelegate}
   > Unbind Event from On Indicator Mode Changed(self:Fort HUDContext Object Reference=@n33, Delegate=@n32) {is=title K2Node_RemoveDelegate}
   > Unbind Event from On Keybinds Changed(self:Fort Global UIContext Object Reference=@n35, Delegate=@n34) {is=title K2Node_RemoveDelegate}
   > Unbind Event from On HUDElement Visibility Changed(self:Fort HUDContext Object Reference=@n37, Delegate=@n36) {is=title K2Node_RemoveDelegate}
   > n43 = Cast To FortPlayerControllerGameplay(Object=GetOwningPlayer())
   > Unbind Event from On Show Creative Global Options(self:Fort Player Controller Gameplay Object Reference=@n43, Delegate=@n39) {is=title K2Node_RemoveDelegate}
   > RemoveLooseGameplayTags!(
    execute
    Actor:Actor Object Reference=GetOwningPlayer()
    GameplayTags:"Gameplay Tag Container Structure (by ref)"=MakeGameplayTagContainerFromTag(SingleTag=(TagName="MobileHUD.TouchControlsActive"))
    bShouldReplicate:bool=false
   ) -> (then, ReturnValue:bool)
   Event OnDisplayGameOver() {is=title K2Node_Event}
   > SetZoneCompleted(bValue=true)
   > StartPostGameMenu(self:Athena HUDContext Object Reference=Get AthenaHUDContext!() -> (ReturnValue:Athena HUDContext Object Reference) {is=title K2Node_GetBlueprintContextEx})
   > Set PostGameMenuStarted(PostGameMenuStarted:bool=true) -> (Output_Get:bool)
   Event OnInitialized()
   > Set PadTop!(
    execute
    PadTop:bool=Get bEnableTopSafeZone!(self:Fort Runtime Options Object Reference=GetRuntimeOptions()) -> (bEnableTopSafeZone:bool)
    self:Safe Zone Object Reference=Get SafeZone_PersistentHUD() -> (SafeZone_PersistentHUD:Safe Zone Object Reference)
   ) -> (then, Output_Get:bool)
   > n44 = PlaylistDataReadyAsync!(execute:Exec) -> (
    then:Exec
    OnReady:Exec
    GameState:Fort Game State Athena Object Reference
    Playlist:Fort Playlist Object Reference
    PlaylistContextTags:"Gameplay Tag Container Structure (by ref)"
   ) {is=title K2Node_AsyncAction}
   Branch(
    execute=@n44.OnReady
    Condition=IsDynamicTouchHUDActiveForPlaylist!(Playlist:Fort Playlist Object Reference=@n44.Playlist) -> (ReturnValue:bool)
   )
   > RemoveFromParent(self=Get ExtSlot_Event() -> (ExtSlot_Event:Fort Playlist UIExtension Slot Object Reference))
   K2Node_Event_985 = Prepare Game Over() {is=title K2Node_CustomEvent}
   > n45 = Sequence!(execute) -> (then_0, then_1, then_2:Exec)
   > Bind Event to On Local Player Died(self:Fort HUDContext Object Reference=Get FortHUDContext(), Delegate=@n21) {is=title K2Node_AddDelegate}
   > Bind Event to On Local Player Place Changed(self:Fort HUDContext Object Reference=Get FortHUDContext(), Delegate=@n9) {is=title K2Node_AddDelegate}
   n46 = Get FortHUDContext()
   Bind Event to On Pawn Set(execute=@n45.then_2, self:Fort HUDContext Object Reference=@n46, Delegate=@n1) {is=title K2Node_AddDelegate}
   > Bind Event to On Local Player Revived(self:Fort HUDContext Object Reference=@n46, Delegate=@n8) {is=title K2Node_AddDelegate}
   n47 = Create Event()
   n48 = Get CommonInputSubsystem()
   Bind Event to On Input Method Changed(self=@n48, Delegate=@n47)
   n49 = GetCurrentInputType(self=Get CommonInputSubsystem!() -> (ReturnValue:Common Input Object Reference) {is=title K2Node_GetBlueprintContextEx})
   SetVisibility(InVisibility=Select(Option 0:ESlateVisibility Enum=Hidden, Option 1:ESlateVisibility Enum=SelfHitTestInvisible, Index="Equal (Enum)"(A=@n49, B=Gamepad)) -> (ReturnValue:ESlateVisibility Enum))
   Unbind Event from On Input Method Changed(self=@n48, Delegate=@n47)
  ]}
 HandleAircraftModeChanged {nodes [
   HandleAircraftModeChanged() {is=title K2Node_FunctionEntry}
   > n1 = Cast To FortPlayerControllerAthena(Object=GetOwningPlayer())
   > SetHUDElementVisibilityWithContext(
    self=@n1
    HUDElementTags=MakeGameplayTagContainerFromTag(SingleTag=(TagName="HUD.Athena.HideOnBus"))
    bHideElements=IsInAircraft!(self:Fort Player Controller Object Reference=@n1) -> (ReturnValue:bool)
    ReasonName=None
   )
  ]}
 HandleFocusChat {nodes [
   HandleFocusChat() {is=title K2Node_FunctionEntry}
   > ToggleChat!(execute, self, Show:bool=true) -> (then)
  ]}
 HandleIndicatorModeChanged {nodes [
   n1 = HandleIndicatorModeChanged!() -> (then, InidicatorsEnabled:bool=~) {is=title K2Node_FunctionEntry}
   n2 = Select(Option 0=Hidden, Option 1=HitTestInvisible, Index=@n1)
   > SetVisibility(self=Get CameraMode() -> (CameraMode:Overlay Object Reference), InVisibility=@n2)
   > SetVisibility(self=Get CameraMode_Reticle() -> (CameraMode_Reticle:Overlay Object Reference), InVisibility=@n2)
  ]}
 HandleKeybindsChanged {nodes [HandleKeybindsChanged() {is=title K2Node_FunctionEntry}]}
 IsClientNotifiedOfWin {nodes [
   IsClientNotifiedOfWin() {is=title K2Node_FunctionEntry}
   > n1 = Cast To FortPlayerControllerAthena(Object=GetOwningPlayer())
   > Return Node!(
    execute
    bClientWon:bool=IsClientNotifiedOfWin!(self:Fort Player Controller Athena Object Reference=@n1) -> (ReturnValue:bool) {is=title K2Node_CallFunction}
    bTeamWon:bool=IsClientNotifiedOfTeamWin!(self:Fort Player Controller Athena Object Reference=@n1) -> (ReturnValue:bool)
   )
   Return Node(execute=@n1.CastFailed, bClientWon=false, bTeamWon=false)
  ]}
 IsPickAxeEquipped {nodes [
   IsPickAxeEquipped() {is=title K2Node_FunctionEntry}
   n1 = GetQuickbarFocus!(self:Fort Inventory Context Object Reference=Get FortInventoryContext!() -> (ReturnValue:Fort Inventory Context Object Reference) {is=title K2Node_GetBlueprintContextEx}) -> (OutQuickBar:EFortQuickBars Enum=Primary, OutSlot:int=0, OutSecondarySlot:int=0, OutPreviousFocusedSlot:int=0)
   > Return Node!(
    execute
    PickAxeEquipped:bool=AND Boolean(
     A="Equal (Enum)"(A:EFortQuickBars Enum=@n1, B:EFortQuickBars Enum=Primary)
     B="Equal (Integer)"(A=@n1.OutSlot, B=0) {is=title K2Node_CallFunction}
    )
   )
  ]}
 OnExtensionWidgetCreated {nodes [
   n1 = OnExtensionWidgetCreated!() -> (then, ExtensionSlot:EUIExtensionSlot Enum=Invalid, Widget:User Widget Object Reference) {is=title K2Node_FunctionEntry}
   > n2 = Switch on EUIExtensionSlot(Selection=@n1)
   Call On Playlist Extension Widget Created Not Supported!(
    execute:Exec=[
     @n2.BehindMobileLocalPlayerInfo
     @n2.ControllerBindingCallout
     @n2.GameInfoBox
     @n2.MobileHUDBottomRight
     @n2.TopRightCorner
     @n2.UnderLocalPlayerInfo
     @n2.UnderMobileLocalPlayerInfo
     @n2.UnderSquadInfo
    ]
    self:Athena HUDContext Object Reference=Get AthenaHUDContext()
    ExtensionSlot:EUIExtensionSlot Enum=Get ExtensionSlot() -> (ExtensionSlot:EUIExtensionSlot Enum=Invalid)
    Widget:User Widget Object Reference=Get Widget() -> (Widget:User Widget Object Reference)
   ) -> (then:Exec) {is=title K2Node_CallDelegate}
  ]}
 OnHUDElementVisibilityChanged {nodes [OnHUDElementVisibilityChanged!() -> (then, HiddenHUDElementTags:"Gameplay Tag Container Structure (by ref)") {is=title K2Node_FunctionEntry}]}
 SetQuickbarSizes {nodes [
   SetQuickbarSizes() {is=title K2Node_FunctionEntry}
   > Return Node!(execute)
  ]}
 ToggleChat {nodes [ToggleChat!() -> (then, Show:bool=~) {is=title K2Node_FunctionEntry}]}
 ToggleTopLevelMenu {nodes [SetVisibility(InVisibility=Select(Option 0=Collapsed, Option 1=SelfHitTestInvisible, Index=ToggleTopLevelMenu!() -> (then, Show:bool=~) {is=title K2Node_FunctionEntry}))]}
 UpdateCursorModeVisibility {nodes [
   n1 = UpdateCursorModeVisibility!() -> (then, CursorMode:Name, InCursorModeEnabled:bool=~) {is=title K2Node_FunctionEntry}
   n2 = IsClientNotifiedOfWin!() -> (bClientWon:bool, bTeamWon:bool)
   > Set ZoneCompleteOrNotifiedOfWin(ZoneCompleteOrNotifiedOfWin:bool=OR Boolean!(A=GetZoneCompleted(), B=@n2, C:bool=@n2.bTeamWon) -> (ReturnValue)) -> (Output_Get:bool)
   > SetIndicatorsAllowed!(
    execute
    self:Fort HUDContext Object Reference=Get FortHUDContext()
    bIndicatorsAllowed:bool=NOT Boolean(A=Get ZoneCompleteOrNotifiedOfWin() -> (ZoneCompleteOrNotifiedOfWin:bool))
   ) -> (then)
   > n3 = Cast To FortPlayerControllerGameplay(Object=GetOwningPlayer())
   n4 = "Equal (Name)"!(A:Name=@n1, B:Name=ToggleMgmtMenu) -> (ReturnValue:bool) {is=title K2Node_CallFunction}
   n5 = "Equal (Name)"(A=@n1, B=ToggleInventory)
   > SetHUDElementVisibilityWithContext(
    self=@n3
    HUDElementTags=MakeGameplayTagContainerFromTag(SingleTag=(TagName="HUD.Athena.Hide.WithInventory"))
    bHideElements=AND Boolean(A=OR Boolean!(A=@n5, B=@n4) -> (ReturnValue), B=Get InCursorModeEnabled() -> (InCursorModeEnabled:bool))
    ReasonName=None
   )
  ]}
]
variables [
 Active Alert Widget {type=Widget, default=null}
 CursorModeEnabled {type=boolean}
 FocusedQuickbar {type=EFortQuickBars, default=Primary}
 GameInfoBoxDefaultHUDTags {
  type=GameplayTagContainer
  default {gameplayTags [
    {tagName=Athena.HideForCapture}
    {tagName=Athena.Creative.HideForLoading}
    {tagName=HUD.Hide.WhileLoading}
    {tagName=HUD.Athena.HideForPostGame}
   ]}
 }
 PostGameMenuStarted {type=boolean}
 UnfocusedQuickbarScale {type=Vector2D, default=<0.7 0.7>}
 ZoneCompleteOrNotifiedOfWin {type=boolean}
]
widgets {
 info {
  inheritedWidgetCount=4
  namedSlotCount=4
  parentClass=/Script/FortniteUI.AthenaMobileHUD
  rootWidgetClass=&Overlay
  widgetCount=23
 }
 widgets [
  Overlay_65 {widgetClassPath=&Overlay, widget=Overlay_65}
  CameraMode {
   widgetClassPath=&Overlay
   bIsVariable=true
   parent=Overlay_65
   widget=CameraMode
   slot {class=OverlaySlot, properties {horizontalAlignment=HAlign_Fill, verticalAlignment=VAlign_Fill}}
  }
  FortHUDElementWrapper_PersistentHUD {
   widgetClassPath=/Game/Athena/HUD/FortHUDElementWrapper
   bIsVariable=true
   parent=Overlay_65
   widget=FortHUDElementWrapper_PersistentHUD
   properties {elementTag {tagName=Athena.HUD.PersistentHUDWrapper}, hUDElementTag {gameplayTags [{tagName=Ability.DeckBed.Dent.Active}]}}
   slot {class=OverlaySlot, properties {horizontalAlignment=HAlign_Fill, verticalAlignment=VAlign_Fill}}
  }
  SafeZone_PersistentHUD {
   widgetClassPath=&SafeZone
   bIsVariable=true
   namedSlotHost=FortHUDElementWrapper_PersistentHUD
   widget=SafeZone_PersistentHUD
   properties {padTop=false}
  }
  "[Overlay] Persistent HUD Content" {
   widgetClassPath=&Overlay
   bIsVariable=true
   parent=SafeZone_PersistentHUD
   widget="[Overlay] Persistent HUD Content"
   slot {class=SafeZoneSlot}
  }
  CommonHud {
   widgetClassPath=&CanvasPanel
   bIsVariable=true
   parent="[Overlay] Persistent HUD Content"
   widget=CommonHud
   properties {visibility=HitTestInvisible}
   slot {class=OverlaySlot, properties {horizontalAlignment=HAlign_Fill, verticalAlignment=VAlign_Fill}}
  }
  TagVisibilityWidget_Controls {
   widgetClassPath=/Script/FortniteUI.CommonTagVisibilityWidget
   bIsVariable=true
   parent="[Overlay] Persistent HUD Content"
   widget=TagVisibilityWidget_Controls
   properties {visibility=SelfHitTestInvisible, hUDElementTag {gameplayTags [{tagName=HUD.Bacchus.TouchControls}]}}
   slot {class=OverlaySlot, properties {horizontalAlignment=HAlign_Fill, verticalAlignment=VAlign_Fill}}
  }
  Overlay_Controls {widgetClassPath=&Overlay, bIsVariable=true, parent=TagVisibilityWidget_Controls, widget=Overlay_Controls, slot {class=PanelSlot}}
  Bacchus_AvailablePickupsDisplay {
   widgetClassPath=/FortniteCoreUI/Athena/HUD/Bacchus/Bacchus_AvailablePickupsDisplay
   bIsVariable=true
   parent=Overlay_Controls
   widget=Bacchus_AvailablePickupsDisplay
   properties {inLockedBus=Hide, inUnlockedBus=Hide, isFreeFalling=Hide, visibility=Visible}
   slot {class=OverlaySlot, properties {horizontalAlignment=HAlign_Center, verticalAlignment=VAlign_Center, padding=<ltrb 150 0 0 0>}}
  }
  TopMidVerticalBox {
   widgetClassPath=&VerticalBox
   bIsVariable=true
   parent="[Overlay] Persistent HUD Content"
   widget=TopMidVerticalBox
   slot {class=OverlaySlot, properties {horizontalAlignment=HAlign_Center, padding=<ltrb 0 70 0 0>}}
  }
  ExtSlot_Event {
   widgetClassPath=/Script/FortniteUI.FortPlaylistUIExtensionSlot
   bIsVariable=true
   parent=TopMidVerticalBox
   widget=ExtSlot_Event
   properties {
    entryHorizontalAlignment=HAlign_Center
    entryVerticalAlignment=VAlign_Center
    entryWidgetClass=&CommonUserWidget
    numDesignerPreviewEntries=1
    extensionSlots [AboveQuestsMobileUI]
   }
   slot {class=VerticalBoxSlot, properties {horizontalAlignment=HAlign_Center, verticalAlignment=VAlign_Center}}
  }
  HUDPlaylistExtensionSlots {
   widgetClassPath=/FortniteCoreUI/Athena/HUD/Bacchus/CustomizableHUD/HUDPlaylistExtensionSlots
   bIsVariable=true
   parent=TopMidVerticalBox
   widget=HUDPlaylistExtensionSlots
   slot {class=VerticalBoxSlot, properties {horizontalAlignment=HAlign_Center, verticalAlignment=VAlign_Center}}
  }
  ExtensionLayer {
   widgetClassPath=&Overlay
   bIsVariable=true
   parent="[Overlay] Persistent HUD Content"
   widget=ExtensionLayer
   slot {class=OverlaySlot, properties {horizontalAlignment=HAlign_Fill, verticalAlignment=VAlign_Fill}}
  }
  CameraMode_Reticle {
   widgetClassPath=&Overlay
   bIsVariable=true
   parent=Overlay_65
   widget=CameraMode_Reticle
   slot {class=OverlaySlot, properties {horizontalAlignment=HAlign_Fill, verticalAlignment=VAlign_Fill}}
  }
  FortHUDElementWrapper_Reticle {
   widgetClassPath=&FortHUDElementWrapper
   bIsVariable=true
   parent=CameraMode_Reticle
   widget=FortHUDElementWrapper_Reticle
   properties {elementTag {tagName=Athena.HUD.ReticleWrapper}, hUDElementTag {gameplayTags [{tagName=Ability.DeckBed.Dent.Active}]}}
   slot {class=OverlaySlot, properties {horizontalAlignment=HAlign_Fill, verticalAlignment=VAlign_Fill}}
  }
  Overlay_Reticle {widgetClassPath=&Overlay, bIsVariable=true, namedSlotHost=FortHUDElementWrapper_Reticle, widget=Overlay_Reticle}
  OverlayReticleContent {
   widgetClassPath=&Overlay
   bIsVariable=true
   parent=Overlay_Reticle
   widget=OverlayReticleContent
   slot {class=OverlaySlot, properties {horizontalAlignment=HAlign_Fill, verticalAlignment=VAlign_Fill}}
  }
  FortHUDElementWrapper_AthenaHUDMenu {
   widgetClassPath=&FortHUDElementWrapper
   bIsVariable=true
   parent=Overlay_65
   widget=FortHUDElementWrapper_AthenaHUDMenu
   properties {elementTag {tagName=Athena.HUD.MenuWrapper}, hUDElementTag {gameplayTags [{tagName=Ability.DeckBed.Dent.Active}]}}
   slot {class=OverlaySlot, properties {horizontalAlignment=HAlign_Fill, verticalAlignment=VAlign_Fill}}
  }
  AthenaHUDMenu {
   widgetClassPath=&/WBP_AthenaHUDMenu
   bInherited=true
   bIsVariable=true
   namedSlotHost=FortHUDElementWrapper_AthenaHUDMenu
   widget=AthenaHUDMenu
  }
  FortHUDElementWrapper_MapScreenContainer {
   widgetClassPath=&FortHUDElementWrapper
   bIsVariable=true
   parent=Overlay_65
   widget=FortHUDElementWrapper_MapScreenContainer
   properties {elementTag {tagName=Athena.HUD.MapScreenContainerWrapper}, hUDElementTag {gameplayTags [{tagName=Ability.DeckBed.Dent.Active}]}}
   slot {class=OverlaySlot, properties {horizontalAlignment=HAlign_Fill, verticalAlignment=VAlign_Fill}}
  }
  MapScreenContainer {
   widgetClassPath=&/BP_AthenaMapScreenContainer
   bInherited=true
   bIsVariable=true
   namedSlotHost=FortHUDElementWrapper_MapScreenContainer
   widget=MapScreenContainer
   properties {visibility=Collapsed}
  }
  TitleBar {widgetClassPath=&CommonActivatableWidget, bInherited=true, widget=null}
  WidgetStack_PrimaryContent {widgetClassPath=/Script/CommonUI.CommonActivatableWidgetStack, bInherited=true, widget=null}
 ]
}
