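# AthenaHUD is the root Battle Royale HUD widget on AthenaHUDBase, layering the menu, map, tutorial and title bar.
# Construct binds HUD context and controller delegates for death, placement, revive, pawn and aircraft changes.
# These drive the zone-completed flag, death spectator cursor mode, post-game menu and recovery on revive.
# Cursor mode sets input mode and indicators, while gameplay tags hide elements for bus, spectating and inventory.
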WidgetBlueprint /FortniteCoreUI/HUD/AthenaHUD
graphs [
 AllowCursorModeWhenDead {nodes [
   AllowCursorModeWhenDead() {is=title K2Node_FunctionEntry}
   > n1 = Cast To FortPlayerController(Object=GetOwningPlayer())
   > Return Node!(execute, Allow:bool=GetForceAllowCameraMode!(self:Fort Player Controller Object Reference=@n1) -> (ReturnValue:bool))
   Return Node(execute=@n1.CastFailed, Allow=false)
  ]}
 EventGraph {nodes [
   n1 = On Placement Changed() {is=title K2Node_CustomEvent}
   n2 = Get FortHUDContext!() -> (ReturnValue:Fort HUDContext Object Reference) {is=title K2Node_GetBlueprintContextEx}
   > Branch(Condition="integer > integer"(A=GetLocalPlayerPlacement!(self:Fort HUDContext Object Reference=@n2) -> (ReturnValue:int=0)))
   > n3 = Branch(Condition=NOT Boolean(A=IsSpatialGameplayExperience!() -> (ReturnValue:bool)))
   OnDisplayGameOver(execute=[@n3.then, @n8.then])
   n4 = Cast To FortPlayerControllerAthena(execute=@n3.else, Object=GetOwningPlayer())
   n5 = GetMinigame!(ContextActor:Actor Object Reference=@n4) -> (ReturnValue:Fort Minigame Object Reference)
   > Is Valid(InputObject=@n5)
   n6 = GetMinigameSettingsMachine!(ContextActor:Actor Object Reference=@n5) -> (ReturnValue:Fort Minigame Settings Building Object Reference)
   > Is Valid(InputObject=@n6)
   n7 = GetPostGameSettings!(self:Fort Minigame Settings Building Object Reference=@n6) -> (ReturnValue:"Minigame Post Game Settings Structure (by ref)")
   > n8 = Branch(Condition="Not Equal (Enum)"!(
     A:EMinigameGameEndPostGameType Enum=Break Minigame Post Game Settings!(MinigamePostGameSettings:"Minigame Post Game Settings Structure (by ref)"=@n7) -> (
      PostGameType:EMinigameGameEndPostGameType Enum
      PostGameDelay:float=~
      bCustomShowScoreboard:bool=~
      VictoryAnimationSettings:Minigame Post Game Animation Settings Structure
      DefeatAnimationSettings:Minigame Post Game Animation Settings Structure
      TieAnimationSettings:Minigame Post Game Animation Settings Structure
     ) {is=title K2Node_BreakStruct}
     B:EMinigameGameEndPostGameType Enum=Classic
    ) -> (ReturnValue:bool))
   n9 = HandleRevived() {is=title K2Node_CustomEvent}
   > OnRevived()
   OnPawnSet() {is=title K2Node_CustomEvent}
   > n10 = Cast To FortPlayerControllerAthena(Object=GetOwningPlayer())
   > n11 = Branch(Condition=GetIsSpectating!(self:Fort Player Controller Athena Object Reference=@n10) -> (ReturnValue:bool))
   n12 = SetZoneCompleted!(execute=@n11.else, self:Athena HUDBase Object Reference, bValue:bool=false) -> (then)
   EnterCameraMode(execute=[@n12.then, @n17.then], self:Fort HUDContext Object Reference=Get FortHUDContext())
   n13 = GetOwningPlayer()
   > n14 = Cast To FortPlayerControllerAthena(Object=@n13)
   n15 = MakeGameplayTagContainerFromTag!(SingleTag:Gameplay Tag Structure=(TagName="HUD.Athena.HideOnSpectate")) -> (ReturnValue:Gameplay Tag Container Structure)
   > SetHUDElementVisibilityWithContext!(
    execute
    self:Fort Player Controller Object Reference=@n14
    HUDElementTags:"Gameplay Tag Container Structure (by ref)"=@n15
    bHideElements:bool=false
    ReasonName:Name=None
    LogContext:String
   ) -> (then)
   OnRevived() {is=title K2Node_CustomEvent}
   > n16 = Branch(Condition=Get PostGameMenuStarted() -> (PostGameMenuStarted:bool))
   SetZoneCompleted(execute=@n16.else, bValue=false)
   > n17 = RecoverFromPostGameMenu(self:Athena HUDContext Object Reference=Get AthenaHUDContext!() -> (ReturnValue:Athena HUDContext Object Reference) {is=title K2Node_GetBlueprintContextEx})
   n18 = OnUnableToPerformAction!() -> (OutputDelegate, then, FailedReason:Gameplay Tag Container Structure, FailureText:"Text (by ref)") {is=title K2Node_CustomEvent}
   > Branch(Condition=HasTag!(
     TagContainer:"Gameplay Tag Container Structure (by ref)"=@n18.FailedReason
     Tag:Gameplay Tag Structure=(TagName="Athena.Ability.Failed.Consumable")
     bExactMatch:bool=false
    ) -> (ReturnValue:bool))
   > PlaySound2D(Sound=/Game/Athena/Sounds/Items/Consumable_CantApply_Cue)
   n19 = HandleLocalPawnSet() {is=title K2Node_CustomEvent}
   > OnPawnSet() {is=title K2Node_CallFunction}
   n20 = On Player Died!() -> (OutputDelegate, then, DeathReport:"Fort Player Death Report Structure (by ref)") {is=title K2Node_CustomEvent}
   > PrepareToShowEndGameUI()
   > n21 = GetAllWidgetsOfClass!(execute, WidgetClass:User Widget Class Reference=/Game/Athena/HUD/AthenaCountdown.AthenaCountdown_C, TopLevelOnly:bool=true) -> (then, FoundWidgets:Array of Athena Countdown Object References)
   > n22 = For Each Loop(Array:Array of Athena Countdown Object References=@n21) -> (Array Element:Athena Countdown Object Reference)
   > SetRenderTranslation(self=@n22)
   n23 = HandleInputMethodChanged!() -> (OutputDelegate, then, InputType:ECommonInputType Enum) {is=title K2Node_CustomEvent}
   > SetVisibility(InVisibility=Select(Option 0=Hidden, Option 1=SelfHitTestInvisible, Index="Equal (Enum)"(A=@n23.InputType, B=Gamepad)))
   n24 = Event HandleCursorModeChanged!() -> (OutputDelegate, then, bCursorModeEnabled:bool, ActionName:Name=None, CustomWidget:User Widget Object Reference) {is=title K2Node_Event}
   n25 = AllowCursorModeWhenDead!() -> (Allow:bool) {is=title K2Node_CallFunction}
   n26 = IsClientNotifiedOfWin!(self) -> (bClientWon:bool, bTeamWon:bool)
   n27 = GetZoneCompleted!(self:Athena HUDBase Object Reference) -> (ReturnValue:bool)
   > n28 = Branch(Condition=AND Boolean(A=OR Boolean!(A=@n27, B=@n26, C:bool=@n26.bTeamWon) -> (ReturnValue), B=NOT Boolean(A=@n25)))
   > n29 = Branch(Condition=@n24.bCursorModeEnabled)
   n30 = Set CursorModeEnabled(execute=[@n28.else, @n29.then], CursorModeEnabled:bool=@n24.bCursorModeEnabled) -> (Output_Get:bool)
   > SetInputMode!(
    execute
    self:Fort Global UIContext Object Reference=Get FortGlobalUIContext!() -> (ReturnValue:Fort Global UIContext Object Reference) {is=title K2Node_GetBlueprintContextEx}
    InMode:EFortInputMode Enum=Select(Option 0:EFortInputMode Enum=InGame, Option 1:EFortInputMode Enum=InGameCursor, Index=@n30) -> (ReturnValue:EFortInputMode Enum)
   ) -> (then)
   > Branch(Condition="Not Equal (Name)"!(A:Name=@n24.ActionName, B:Name=ToggleFocusLostTitleBar) -> (ReturnValue:bool) {is=title K2Node_CallFunction})
   > UpdateCursorModeVisibility!(execute, self, CursorMode:Name=@n24.ActionName) -> (then)
   EnterCursorMode!(
    execute=@n29.else
    self:Fort HUDContext Object Reference=Get FortHUDContext()
    ActionName:Name=Show Death Spectator Menu
    CursorModeWidget:User Widget Object Reference
   ) -> (then)
   > UpdateCursorModeVisibility(CursorMode=None)
   Event Construct()
   n31 = Create Event()
   n32 = Get FortHUDContext()
   > Bind Event to On Unable to Perform Action(self:Fort HUDContext Object Reference=@n32, Delegate=@n31) {is=title K2Node_AddDelegate}
   > Prepare Game Over()
   > n33 = Cast To FortPlayerControllerAthena(Object=GetOwningPlayer())
   n34 = Create Event()
   > Assign On Aircraft State Change!(execute:Exec, self:Fort Player Controller Athena Object Reference=@n33, Delegate:"Delegate (by ref)"=@n34) -> (then:Exec) {is=title K2Node_AssignDelegate}
   > HandleAircraftModeChanged()
   > SetFortPickerData!(
    execute
    self:Fort Picker Context Object Reference=Get FortPickerContext!(PlayerController:Player Controller Object Reference=GetOwningPlayer()) -> (ReturnValue:Fort Picker Context Object Reference) {is=title K2Node_GetSubsystemFromPC}
    Data:Fort Picker Data Object Reference=/Game/UI/InGame/HUD/FortPickers
   ) -> (then)
   > n35 = Branch(Condition=IsTouchInputPlatform!() -> (ReturnValue:bool))
   n36 = PlaylistDataReadyAsync!(execute:Exec=@n35.else) -> (
    then:Exec
    OnReady:Exec
    GameState:Fort Game State Athena Object Reference
    Playlist:Fort Playlist Object Reference
    PlaylistContextTags:"Gameplay Tag Container Structure (by ref)"
   ) {is=title K2Node_AsyncAction}
   Handle Offset Translation on Gameplay Tags!(
    Exec:Exec=@n36.OnReady
    PlaylistGameplayTags:Array of Gameplay Tag Structures=BreakGameplayTagContainer!(GameplayTagContainer:"Gameplay Tag Container Structure (by ref)"=@n36.PlaylistContextTags) -> (GameplayTags:Array of Gameplay Tag Structures)
    PlaylistOffsetMap:Map of Gameplay Tag Structures to Vector 2D Structures=Get PlaylistOffsetMap() -> (PlaylistOffsetMap:Map of Gameplay Tag Structures to Vector 2D Structures)
   ) -> (OnOffsetFound:Exec, OnCompleted:Exec, TranslationOffset:Vector 2D Structure, FoundTag:Gameplay Tag Structure) {is=title K2Node_MacroInstance}
   Event Destruct()
   > Unbind Event from On Unable to Perform Action(self:Fort HUDContext Object Reference=@n32, Delegate=@n31) {is=title K2Node_RemoveDelegate}
   > n37 = Cast To FortPlayerControllerAthena(Object=GetOwningPlayer())
   > n38 = Unbind Event from On Aircraft State Change(self:Fort Player Controller Athena Object Reference=@n37, Delegate=@n34) {is=title K2Node_RemoveDelegate}
   n39 = Get FortHUDContext()
   Unbind all Events from On Local Player Died!(execute:Exec=[@n38.then, @n37.CastFailed], self:Fort HUDContext Object Reference=@n39) -> (then:Exec) {is=title K2Node_ClearDelegate}
   > Unbind all Events from On Pawn Set!(execute:Exec, self:Fort HUDContext Object Reference=@n39) -> (then:Exec) {is=title K2Node_ClearDelegate}
   > Unbind all Events from On Local Player Revived!(execute:Exec, self:Fort HUDContext Object Reference=@n39) -> (then:Exec) {is=title K2Node_ClearDelegate}
   > Unbind all Events from On Local Player Place Changed!(execute:Exec, self:Fort HUDContext Object Reference=@n39) -> (then:Exec) {is=title K2Node_ClearDelegate}
   > Unbind all Events from On Force Start Post Game Flow!(execute:Exec, self:Athena HUDContext Object Reference=Get AthenaHUDContext()) -> (then:Exec) {is=title K2Node_ClearDelegate}
   n40 = Get FortGlobalUIContext() -> (ReturnValue:Fort Global UIContext Object Reference) {is=title K2Node_GetSubsystem}
   > Unbind all Events from On Exit Vehicle!(execute:Exec, self:Fort Global UIContext Object Reference=@n40) -> (then:Exec) {is=title K2Node_ClearDelegate}
   > Unbind all Events from On Enter Vehicle Driver!(execute:Exec, self:Fort Global UIContext Object Reference=@n40) -> (then:Exec) {is=title K2Node_ClearDelegate}
   > Unbind all Events from On Enter Vehicle Passenger!(execute:Exec, self:Fort Global UIContext Object Reference=@n40) -> (then:Exec) {is=title K2Node_ClearDelegate}
   Event OnDisplayGameOver() {is=title K2Node_Event}
   > SetZoneCompleted(bValue=true)
   > StartPostGameMenu(self:Athena HUDContext Object Reference=Get AthenaHUDContext())
   > Set PostGameMenuStarted(PostGameMenuStarted:bool=true) -> (Output_Get:bool)
   Event PrepareToShowEndGameUI() {is=title K2Node_Event}
   > n41 = Branch(Condition=GetZoneCompleted())
   SetZoneCompleted(execute=@n41.else, bValue=true)
   > EnterCursorMode(self=Get FortHUDContext(), ActionName=Show Death Spectator Menu)
   > n42 = Cast To FortPlayerControllerAthena(Object=@n13)
   > SetHUDElementVisibilityWithContext(self=@n42, HUDElementTags=@n15, bHideElements=true, ReasonName=None)
   K2Node_Event_985 = Prepare Game Over() {is=title K2Node_CustomEvent}
   > n43 = Sequence()
   n44 = Get FortHUDContext()
   > Bind Event to On Local Player Died(self:Fort HUDContext Object Reference=@n44, Delegate=@n20) {is=title K2Node_AddDelegate}
   > Bind Event to On Local Player Place Changed(self:Fort HUDContext Object Reference=@n44, Delegate=@n1) {is=title K2Node_AddDelegate}
   > Bind Event to On Force Start Post Game Flow(self:Athena HUDContext Object Reference=Get AthenaHUDContext(), Delegate=@n1) {is=title K2Node_AddDelegate}
   n45 = Get FortHUDContext()
   Bind Event to On Pawn Set(execute=@n43.then_1, self:Fort HUDContext Object Reference=@n45, Delegate=@n19) {is=title K2Node_AddDelegate}
   > Bind Event to On Local Player Revived(self:Fort HUDContext Object Reference=@n45, Delegate=@n9) {is=title K2Node_AddDelegate}
   n46 = Create Event()
   n47 = Get CommonInputSubsystem()
   Bind Event to On Input Method Changed(self=@n47, Delegate=@n46)
   n48 = GetCurrentInputType(self=Get CommonInputSubsystem())
   SetVisibility(InVisibility=Select(Option 0:ESlateVisibility Enum=Hidden, Option 1:ESlateVisibility Enum=SelfHitTestInvisible, Index="Equal (Enum)"(A=@n48, B=Gamepad)) -> (ReturnValue:ESlateVisibility Enum))
   Unbind Event from On Input Method Changed(self=@n47, Delegate=@n46)
  ]}
 HandleAircraftModeChanged {nodes [
   HandleAircraftModeChanged() {is=title K2Node_FunctionEntry}
   > n1 = Cast To FortPlayerControllerAthena(Object=GetOwningPlayer())
   > SetHUDElementVisibilityWithContext(
    self=@n1
    HUDElementTags=MakeGameplayTagContainerFromTag(SingleTag=(TagName="HUD.Athena.HideOnBus"))
    bHideElements=IsInAircraft!(self:Fort Player Controller Object Reference=@n1) -> (ReturnValue:bool)
    ReasonName=None
   )
  ]}
 IsClientNotifiedOfWin {nodes [
   IsClientNotifiedOfWin() {is=title K2Node_FunctionEntry}
   > n1 = Cast To FortPlayerControllerAthena(Object=GetOwningPlayer())
   > Return Node!(
    execute
    bClientWon:bool=IsClientNotifiedOfWin!(self:Fort Player Controller Athena Object Reference=@n1) -> (ReturnValue:bool) {is=title K2Node_CallFunction}
    bTeamWon:bool=IsClientNotifiedOfTeamWin!(self:Fort Player Controller Athena Object Reference=@n1) -> (ReturnValue:bool)
   )
   Return Node(execute=@n1.CastFailed, bClientWon=false, bTeamWon=false)
  ]}
 OnKeyDown {nodes [
   OnKeyDown!() -> (then, MyGeometry:Geometry Structure, InKeyEvent:Key Event Structure) {is=title K2Node_FunctionEntry}
   > Return Node!(execute, ReturnValue:Event Reply Structure=Unhandled())
  ]}
 UpdateCursorModeVisibility {nodes [
   n1 = UpdateCursorModeVisibility!() -> (then, CursorMode:Name) {is=title K2Node_FunctionEntry}
   n2 = IsClientNotifiedOfWin!(self:Self Object Reference) -> (bClientWon:bool, bTeamWon:bool)
   > Set ZoneCompleteOrNotifiedOfWin(ZoneCompleteOrNotifiedOfWin:bool=OR Boolean(A=GetZoneCompleted(), B=@n2, C=@n2.bTeamWon)) -> (Output_Get:bool)
   > SetIndicatorsAllowed!(
    execute
    self:Fort HUDContext Object Reference=Get FortHUDContext()
    bIndicatorsAllowed:bool=NOT Boolean(A=Get ZoneCompleteOrNotifiedOfWin() -> (ZoneCompleteOrNotifiedOfWin:bool))
   ) -> (then)
   > n3 = Cast To FortPlayerControllerGameplay!(execute, Object=GetOwningPlayer()) -> (then, CastFailed, AsFort Player Controller Gameplay:Fort Player Controller Gameplay Object Reference) {is=title K2Node_DynamicCast}
   n4 = "Equal (Name)"!(A:Name=@n1, B:Name=ToggleMgmtMenu) -> (ReturnValue:bool) {is=title K2Node_CallFunction}
   n5 = "Equal (Name)"(A=@n1, B=ToggleInventory)
   > SetHUDElementVisibilityWithContext(
    self=@n3
    HUDElementTags=MakeGameplayTagContainerFromTag(SingleTag=(TagName="HUD.Athena.Hide.WithInventory"))
    bHideElements=AND Boolean(A=OR Boolean!(A=@n5, B=@n4) -> (ReturnValue), B=Get CursorModeEnabled() -> (CursorModeEnabled:bool))
    ReasonName=None
   )
  ]}
]
variables [
 CursorModeEnabled {type=boolean}
 GameInfoBoxDefaultHUDTags {
  type=GameplayTagContainer
  default {gameplayTags [
    {tagName=Athena.HideForCapture}
    {tagName=HUD.Athena.Hide.WithInventory}
    {tagName=HUD.Athena.Hide.WithMap}
    {tagName=HUD.LTM.HideForTryHard}
    {tagName=HUD.Event.HideForSpecialEvent}
    {tagName=Athena.Creative.HideForLoading}
    {tagName=HUD.Hide.WhileLoading}
   ]}
 }
 PlaylistOffsetMap {default {"(TagName=\"Athena.Playlist.HasTopUI\")"=<0 71>}}
 PostGameMenuStarted {type=boolean}
]
widgets {
 info {inheritedWidgetCount=4, parentClass=/Script/FortniteUI.AthenaHUDBase, rootWidgetClass=&Overlay, widgetCount=8}
 widgets [
  Overlay_Main {widgetClassPath=&Overlay, bIsVariable=true, widget=Overlay_Main}
  AthenaHUDMenu {
   widgetClassPath=&/WBP_AthenaHUDMenu
   bInherited=true
   parent=Overlay_Main
   widget=AthenaHUDMenu
   slot {class=OverlaySlot, properties {horizontalAlignment=HAlign_Fill, verticalAlignment=VAlign_Fill}}
  }
  MapScreenContainer {
   widgetClassPath=&/BP_AthenaMapScreenContainer
   bInherited=true
   parent=Overlay_Main
   widget=MapScreenContainer
   properties {visibility=Collapsed}
   slot {class=OverlaySlot, properties {horizontalAlignment=HAlign_Fill, verticalAlignment=VAlign_Fill}}
  }
  TutorialHUDContainer {
   widgetClassPath=/Game/Athena/HUD/Tutorial/AthenaTutorialHUDContainer
   parent=Overlay_Main
   widget=TutorialHUDContainer
   slot {class=OverlaySlot, properties {horizontalAlignment=HAlign_Fill, verticalAlignment=VAlign_Fill}}
  }
  TopBar_SizeBox {
   widgetClassPath=&SizeBox
   parent=Overlay_Main
   widget=TopBar_SizeBox
   properties {bOverride_HeightOverride=true, heightOverride=24}
   slot {class=OverlaySlot, properties {horizontalAlignment=HAlign_Fill}}
  }
  TopTitlebarArea {
   widgetClassPath=/Script/UMG.WindowTitleBarArea
   parent=TopBar_SizeBox
   widget=TopTitlebarArea
   slot {class=SizeBoxSlot, properties {padding=<ltrb 0 0 120 0>}}
  }
  TitleBar {widgetClassPath=&CommonActivatableWidget, bInherited=true, widget=null}
  WidgetStack_PrimaryContent {widgetClassPath=/Script/CommonUI.CommonActivatableWidgetStack, bInherited=true, widget=null}
 ]
}
