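# AthenaScoreAlert is a HUD widget that shows game mode alerts such as score targets, near-victory and respawn warnings.
# Game state and HUD message events feed TryToDisplayMessage, which filters by type and view target, then handles queuing.
# Display sets text, sound and desktop or mobile placement, plays intro and exclamation animations, then times the outro, which pulls the next queued message.
# Creative mode changes queuing, sound, exclamation and timing rules.
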WidgetBlueprint /FortniteCoreUI/HUD/AthenaScoreAlert
animations [
 Intro {
  binding_count=1
  display_label=""
  length_seconds=0.25
  object=&self:Intro
  sequence {bindings [Feedback {
     id=#1
     display_name=OverlayFeedback
     possessed_class=HorizontalBox
     tracks [
      {
       type=MovieScene2DTransformTrack
       sections [{
         type=MovieScene2DTransformSection
         end_frame=5
         start_frame=0
         channels [
          Translation.X {type=MovieSceneScriptingFloatChannel}
          Translation.Y {type=MovieSceneScriptingFloatChannel}
          Angle {type=MovieSceneScriptingFloatChannel}
          Scale.X {type=MovieSceneScriptingFloatChannel, keys [[0,1],[3,1.282],[3.5983,1],[4.2237,1]]}
          Scale.Y {type=MovieSceneScriptingFloatChannel, keys [[0,0],[3,1.282],[3.5983,1],[4.2237,1]]}
          Shear.X {type=MovieSceneScriptingFloatChannel}
          Shear.Y {type=MovieSceneScriptingFloatChannel}
         ]
        }]
      }
      {
       type=MovieSceneFloatTrack
       display_name=Render Opacity
       sections [{type=MovieSceneFloatSection, end_frame=4, start_frame=0, channels [None {type=MovieSceneScriptingFloatChannel, keys [[0,0],[2.7623,1]]}]}]
      }
     ]
    }]}
 }
 Outro {
  binding_count=1
  display_label=""
  length_seconds=0.3397
  object=&self:Outro
  sequence {
   bindings [Feedback {
     id=#2
     display_name=OverlayFeedback
     possessed_class=HorizontalBox
     tracks [
      {
       type=MovieSceneFloatTrack
       display_name=Render Opacity
       sections [{type=MovieSceneFloatSection, end_frame=40, start_frame=0, channels [None {type=MovieSceneScriptingFloatChannel, keys [[0,1],[40.756,0]]}]}]
      }
      {
       type=MovieScene2DTransformTrack
       sections [{
         type=MovieScene2DTransformSection
         end_frame=40
         start_frame=0
         channels [
          Translation.X {type=MovieSceneScriptingFloatChannel}
          Translation.Y {type=MovieSceneScriptingFloatChannel}
          Angle {type=MovieSceneScriptingFloatChannel}
          Scale.X {type=MovieSceneScriptingFloatChannel, const [1,2]}
          Scale.Y {type=MovieSceneScriptingFloatChannel, keys [[0,1],[9,1.2783],[11,1],[40.756,0]]}
          Shear.X {type=MovieSceneScriptingFloatChannel}
          Shear.Y {type=MovieSceneScriptingFloatChannel}
         ]
        }]
      }
     ]
    }]
   display_rate {numerator=120}
  }
 }
 Exclamation {
  binding_count=1
  display_label=""
  length_seconds=1.25
  object=&self:Exclamation
  sequence {
   bindings [ExclamationTextBlock {
     id=#3
     possessed_class=TextBlock
     tracks [
      {
       type=MovieSceneFloatTrack
       display_name=Render Opacity
       sections [{
         type=MovieSceneFloatSection
         end_frame=149
         start_frame=0
         channels [None {type=MovieSceneScriptingFloatChannel, keys [[0,0],[15.774,1],[123.702,1],[149.426,0]]}]
        }]
      }
      {
       type=MovieScene2DTransformTrack
       sections [{
         type=MovieScene2DTransformSection
         end_frame=150
         start_frame=0
         channels [
          Translation.X {type=MovieSceneScriptingFloatChannel}
          Translation.Y {type=MovieSceneScriptingFloatChannel, keys [[0,130],[19,90],[124,90],[150,130]]}
          Angle {type=MovieSceneScriptingFloatChannel}
          Scale.X {
           type=MovieSceneScriptingFloatChannel
           keys [[0,1],[19,2.1864],[29,0.7511],[36,1.201],[43,0.8469],[53.844,1],[117.032,1],[127,1.1192],[149.584,1],[149.986,0]]
          }
          Scale.Y {
           type=MovieSceneScriptingFloatChannel
           keys [[0,1],[19,2.9101],[29,0.6527],[36,1.2787],[43,0.7639],[53.844,1],[117.032,1],[127,1.3143],[146,0.918],[149.986,0]]
          }
          Shear.X {type=MovieSceneScriptingFloatChannel}
          Shear.Y {type=MovieSceneScriptingFloatChannel}
         ]
        }]
      }
     ]
    }]
   display_rate {numerator=120}
  }
 }
]
graphs [
 BindToGameStateEvents {nodes [
   n1 = BindToGameStateEvents!() -> (then, GameState:Fort Game State Object Reference) {is=title K2Node_FunctionEntry}
   > n2 = Cast To FortGameStateAthena!(execute, Object=@n1) -> (then, CastFailed, AsFort Game State Athena:Fort Game State Athena Object Reference) {is=title K2Node_DynamicCast}
   > Bind Event to Game Mode Message Request(self:Fort Game State Athena Object Reference=@n2, Delegate=Create Event()) {is=title K2Node_AddDelegate}
   > Bind Event to Hide Game Mode Message Request(self:Fort Game State Athena Object Reference=@n2, Delegate=Create Event()) {is=title K2Node_AddDelegate}
  ]}
 BroadcastHUDMessageToggled {nodes [
   BroadcastHUDMessageToggled!() -> (then, bShown:bool=~) {is=title K2Node_FunctionEntry}
   > n1 = Cast To FortPlayerControllerAthena(Object=GetOwningPlayer())
   > BroadcastOnHUDMessageToggled!(
    execute
    self:Fort Player Controller Athena Object Reference=@n1
    HUDMessageData:"HUDMessage Data Structure (by ref)"=Get MostRecentDisplayedMessage() -> (MostRecentDisplayedMessage:HUDMessage Data Structure)
    bShowMessage:bool=Get bShown() -> (bShown:bool)
   ) -> (then)
  ]}
 ConvertToHUDMessage {nodes [
   n1 = ConvertToHUDMessage!() -> (
    then
    Placement:EHUDMessagePlacement Enum=BottomCenter
    Text Transform Policy:ETextTransformPolicy Enum=ToUpper
    AthenaGameMessageData:"Athena Game Message Data Structure (by ref)"
   ) {is=title K2Node_FunctionEntry}
   n2 = Break Athena Game Message Data!(AthenaGameMessageData:"Athena Game Message Data Structure (by ref)"=@n1.AthenaGameMessageData) -> (
    MsgType:EAthenaGameMsgType Enum
    MsgText:Text
    MsgSound:Sound Cue Object Reference
    MsgMediaSource:Media Source Object Reference
    MsgDelay:float=~
    bIsTeamBased:bool=~
    TeamIndex:int
    DisplayTime:float=~
    MessageFlags:EAthenaGameMessageDataFlags Enum
    TargetPlayerController:Player Controller Object Reference
   ) {is=title K2Node_BreakStruct}
   > Return Node!(
    execute
    HUDMessageData:HUDMessage Data Structure=Make HUDMessage Data!(
     Placement:EHUDMessagePlacement Enum=@n1
     TextTransformPolicy:ETextTransformPolicy Enum=@n1.Text Transform Policy
     DisableChannel:int=0
     MsgType:EAthenaGameMsgType Enum=@n2
     MsgText:Text=@n2.MsgText
     MsgSound:Sound Cue Object Reference=@n2.MsgSound
     MsgMediaSource:Media Source Object Reference
     MsgDelay:float=@n2.MsgDelay
     bIsTeamBased:bool=@n2.bIsTeamBased
     TeamIndex:int=@n2.TeamIndex
     DisplayTime:float=@n2.DisplayTime
     MessageFlags:EAthenaGameMessageDataFlags Enum=None
     TargetPlayerController:Player Controller Object Reference=@n2.TargetPlayerController
    ) -> (HUDMessageData:HUDMessage Data Structure) {is=title K2Node_MakeStruct}
   )
  ]}
 DisplayMessage {nodes [
   DisplayMessage() {is=title K2Node_FunctionEntry}
   > n1 = Sequence()
   > Set MostRecentDisplayedMessage(MostRecentDisplayedMessage:HUDMessage Data Structure=Get CurrentMessage() -> (CurrentMessage:HUDMessage Data Structure)) -> (Output_Get:HUDMessage Data Structure)
   > SetVisibility(InVisibility=HitTestInvisible)
   > n2 = ShouldPlayIntroAnimations!(execute, self) -> (then, bShouldPlay:bool)
   > n3 = Branch(Condition=@n2)
   > n4 = PlayIntroAnimations()
   n5 = Get CurrentMessage()
   UpdateContentTranslation!(
    execute=[@n4.then, @n3.else]
    self
    bShiftContentDown:bool="Not Equal (Enum)"!(
     A:EAthenaGameMsgType Enum=Break HUDMessage Data!(HUDMessageData:"HUDMessage Data Structure (by ref)"=@n5) -> (MsgType:EAthenaGameMsgType Enum, MsgMediaSource:Media Source Object Reference, MessageFlags:EAthenaGameMessageDataFlags Enum) {is=title K2Node_BreakStruct}
     B:EAthenaGameMsgType Enum=DefaultIntro
    ) -> (ReturnValue:bool)
   ) -> (then)
   > SetText(
    self=Get MessageTextBlock() -> (MessageTextBlock:Rich Text Block Object Reference)
    InText=Break HUDMessage Data!(HUDMessageData=Get CurrentMessage()) -> (MsgText:Text, MsgMediaSource, MessageFlags)
   )
   > PlaySoundFromType!(execute, self, MessageData:HUDMessage Data Structure=Get CurrentMessage()) -> (then)
   > SetPlacement()
   > StartHideTimer()
   BroadcastHUDMessageToggled!(execute=@n1.then_1, self, bShown:bool=true) -> (then) {is=title K2Node_CallFunction}
  ]}
 DoesTeamMatchViewTargetsTeam {nodes [
   n1 = DoesTeamMatchViewTargetsTeam!() -> (then, Team:int) {is=title K2Node_FunctionEntry}
   n2 = GetOwningPlayer()
   > n3 = Is Valid(InputObject=@n2)
   Return Node!(execute=@n3.Is Not Valid, bDoesMatchViewTargetsTeam:bool=false)
   n4 = GetViewTarget!(self:Controller Object Reference=@n2) -> (ReturnValue:Actor Object Reference)
   > Return Node(bDoesMatchViewTargetsTeam="Equal (Byte)"!(
     A:Byte=GetActorTeam!(Actor:Actor Object Reference=@n4) -> (ReturnValue:Byte=0)
     B:Byte="To Byte (Integer)"!(InInt:int=@n1) -> (ReturnValue:Byte=0)
    ) -> (ReturnValue:bool))
  ]}
 DoesViewTargetMatchTriggeringPlayer {nodes [
   DoesViewTargetMatchTriggeringPlayer!() -> (then, TriggeringPlayer:Player State Object Reference) {is=title K2Node_FunctionEntry}
   > n1 = Sequence()
   > Get!(execute:Exec) -> (then:Exec, else:Exec, TriggeringPlayer:Player State Object Reference)
   n2 = GetOwningPlayer()
   > Is Valid(InputObject=@n2)
   > n3 = Cast To Pawn!(execute, Object=GetViewTarget(self=@n2)) -> (then, CastFailed, AsPawn:Pawn Object Reference) {is=title K2Node_DynamicCast}
   > n4 = Get!(execute, self:Pawn Object Reference=@n3) -> (then, else, PlayerState:Player State Object Reference)
   > Set bReturnDoesMatch(bReturnDoesMatch:bool="Equal (Object)"(A=@n4, B=Get TriggeringPlayer() -> (TriggeringPlayer:Player State Object Reference)) {is=title K2Node_CallFunction}) -> (Output_Get:bool)
   Return Node!(execute=@n1.then_1, bDoesMatch:bool=Get bReturnDoesMatch() -> (bReturnDoesMatch:bool))
  ]}
 DoesViewTargetsTeamMatchTriggeringPlayersTeam {nodes [
   DoesViewTargetsTeamMatchTriggeringPlayersTeam!() -> (then, TriggeringPlayer:Player State Object Reference) {is=title K2Node_FunctionEntry}
   > n1 = Sequence()
   > Get!(execute) -> (then, else, TriggeringPlayer:Player State Object Reference)
   n2 = GetOwningPlayer()
   > Is Valid(InputObject=@n2)
   n3 = Get TriggeringPlayer()
   n4 = GetViewTarget(self=@n2)
   > Set bReturnDoesMatch(bReturnDoesMatch="Equal (Byte)"(A=GetActorTeam(Actor=@n4) -> (ReturnValue=0), B=GetActorTeam(Actor=@n3) -> (ReturnValue=0)))
   Return Node(execute=@n1.then_1, bDoesMatch=Get bReturnDoesMatch())
  ]}
 DoesWidgetHandleThisEnum {nodes [
   n1 = DoesWidgetHandleThisEnum!() -> (then, MessageType:EAthenaGameMsgType Enum=None) {is=title K2Node_FunctionEntry}
   > Return Node!(
    execute
    bIsValid:bool=OR Boolean!(
     A="Equal (Enum)"(A:EAthenaGameMsgType Enum=@n1, B:EAthenaGameMsgType Enum=DefaultIntro)
     B="Equal (Enum)"(A=@n1, B=DefaultMessage)
     C:bool="Equal (Enum)"(A=@n1, B=DefaultCriticalMessage)
     D:bool="Equal (Enum)"(A=@n1, B=RespawnTurningOffWarning)
     E:bool="Equal (Enum)"(A=@n1, B=RespawnOffWarning)
    ) -> (ReturnValue)
   )
  ]}
 EventGraph {nodes [
   n1 = Event PreConstruct()
   > Branch(Condition=@n1.IsDesignTime)
   > SetText(
    self=Get MessageTextBlock()
    InText=Select!(
     Option 0:Text="LOCTEXT(\"<intro>Score</> <introscore>1,000</><intro> points to win!</>\")"
     Option 1:Text="LOCTEXT(\"<enemy>An enemy is </><score>100</><enemy> points from victory!</>\")"
     Option 2:Text="LOCTEXT(\"<ally>Your team is </><score>100</><ally> points from victory!</>\")"
     Option 3:Text="LOCTEXT(\"<RespawnAlertB>NO RESPAWNS</><RespawnAlert>AFTER THE NEXT STORM CYCLE!</>\")"
     Option 4:Text="LOCTEXT(\"<RespawnAlertB>RESPAWNING IS NOW</> <RespawnAlertBLrg>OFF!</>\")"
     Index:int=Get TestMessageType() -> (TestMessageType:int=0)
    ) -> (ReturnValue:Text)
   )
   > UpdateContentTranslation(bShiftContentDown="Not Equal (Integer)"!(A:int=Get TestMessageType() -> (TestMessageType=0), B:int=0) -> (ReturnValue:bool) {is=title K2Node_CallFunction})
   n2 = Break HUDMessage Data!(HUDMessageData=Get CurrentMessage()) -> (
    Placement:EHUDMessagePlacement Enum
    PlayerState:Player State Object Reference
    MessageTargets:int
    TextTransformPolicy:ETextTransformPolicy Enum
    DisableChannel:int
    MsgType:EAthenaGameMsgType Enum
    MsgText
    MsgSound:Sound Cue Object Reference
    MsgMediaSource
    MsgDelay:float=~
    bIsTeamBased:bool=~
    TeamIndex:int
    DisplayTime:float=~
    MessageFlags
    TargetPlayerController:Player Controller Object Reference
   )
   > Set CurrentMessage(CurrentMessage:HUDMessage Data Structure=Make HUDMessage Data!(
     Placement=Get TestMessagePlacement() -> (TestMessagePlacement:EHUDMessagePlacement Enum=None)
     PlayerState:Player State Object Reference=@n2.PlayerState
     MessageTargets:int=@n2.MessageTargets
     TextTransformPolicy=@n2.TextTransformPolicy
     DisableChannel=@n2.DisableChannel
     MsgType=@n2.MsgType
     MsgText=@n2.MsgText
     MsgSound=@n2.MsgSound
     MsgMediaSource=@n2.MsgMediaSource
     MsgDelay=@n2.MsgDelay
     bIsTeamBased=@n2.bIsTeamBased
     TeamIndex=@n2.TeamIndex
     DisplayTime=@n2.DisplayTime
     MessageFlags=@n2.MessageFlags
     TargetPlayerController=@n2.TargetPlayerController
    ) -> (HUDMessageData)) -> (Output_Get:HUDMessage Data Structure)
   > Reroute Node()
   Event On Minigame Ended() {is=title K2Node_Event}
   > OnHideGameModeMessageRequest!(execute, self, MessageType:EAthenaGameMsgType Enum=DefaultIntro, bInstantHide:bool=true, bClearQueue:bool=true, IncomingMessageChannel:int=0) -> (then)
   n3 = Event On GameState Initialized!() -> (OutputDelegate, then, GameState:Fort Game State Object Reference) {is=title K2Node_Event}
   > BindToGameStateEvents!(execute, self, GameState:Fort Game State Object Reference=@n3.GameState) -> (then) {is=title K2Node_CallFunction}
   > Is Creative!(execute:Exec) -> (True:Exec, False:Exec, Unknown:Exec) {is=title K2Node_MacroInstance}
   > RemoveGameplayTag!(
    execute
    TagContainer:"Gameplay Tag Container Structure (by ref)"=Get HUDElementTag() -> (HUDElementTag:Gameplay Tag Container Structure)
    Tag:Gameplay Tag Structure=(TagName="HUD.Athena.HideOnSpectate")
   ) -> (then, ReturnValue:bool)
   Event Construct()
   > SetVisibility(InVisibility=Collapsed)
   > FillInitialPositionData()
   n4 = GetOwningPlayer()
   > Branch(Condition=AND Boolean(A=IsMobileGame(), B=IsUsingTouch(PlayerController=@n4)))
   > n5 = PlaylistDataReadyAsync!(execute:Exec) -> (
    then:Exec
    OnReady:Exec
    GameState:Fort Game State Athena Object Reference
    Playlist:Fort Playlist Object Reference
    PlaylistContextTags:"Gameplay Tag Container Structure (by ref)"
   ) {is=title K2Node_AsyncAction}
   Branch(execute=@n5.OnReady)
   > SetPadding!(
    execute
    self:Vertical Box Slot Object Reference=SlotAsVerticalBoxSlot(Widget=Get OverlayFeedback() -> (OverlayFeedback:Overlay Object Reference))
    InPadding_Left:float=0
    InPadding_Top:float=250
    InPadding_Right:float=0
    InPadding_Bottom:float=0
   ) -> (then)
   n6 = Event On Show HUD Message!() -> (OutputDelegate, then, MessageData:HUDMessage Data Structure) {is=title K2Node_Event}
   > TryToDisplayMessage!(execute, self, CurrentMessage:HUDMessage Data Structure=@n6.MessageData) -> (then)
   SetPlacement()
   HasTag!(
    TagContainer:"Gameplay Tag Container Structure (by ref)"=@n5.PlaylistContextTags
    Tag:Gameplay Tag Structure=(TagName="Athena.Playlist.TeamRumble")
    bExactMatch:bool=false
   ) -> (ReturnValue:bool)
   SetPadding(
    self:Overlay Slot Object Reference=SlotAsOverlaySlot(Widget=Get ExclamationTextBlock() -> (ExclamationTextBlock:Text Object Reference))
    InPadding_Left=0
    InPadding_Top=-10
    InPadding_Right=0
    InPadding_Bottom=0
   )
  ]}
 FillInitialPositionData {nodes [
   FillInitialPositionData() {is=title K2Node_FunctionEntry}
   > n1 = Get Parent Slot!(execute:Exec) -> (Is Valid, Slot:Overlay Slot Object Reference, Is Not Valid) {is=title K2Node_MacroInstance}
   Log HUD Message Warning!(
    execute=@n1.Is Not Valid
    InString:String="AthenaScoreAlert::FillInitialPositionData() Expecting this widget's parent' slot to be an Overlay slot. It must have changed. Please update."
   ) -> (then)
   n2 = Select!(
    Option 0:"EWinConditionParentType Enum (by ref)"=Literal enum EWinConditionParentType!(Enum:EWinConditionParentType Enum=Mobile) -> (ReturnValue:EWinConditionParentType Enum) {is=title K2Node_EnumLiteral}
    Option 1:"EWinConditionParentType Enum (by ref)"=Literal enum EWinConditionParentType(Enum=Desktop)
    Index:bool=IsDesktop!() -> (bReturnValue:bool)
   ) -> (ReturnValue:"EWinConditionParentType Enum (by ref)")
   n3 = Find!(
    TargetMap:Map of EWinConditionParentType Enums to Win Condition Placement Data Map Wrapper Structures=Get Placements() -> (Placements:Map of EWinConditionParentType Enums to Win Condition Placement Data Map Wrapper Structures=None)
    Key:"EWinConditionParentType Enum (by ref)"=@n2
   ) -> (Value_Data:Map of EHUDMessagePlacement Enums to Win Condition Placement Data Structures, ReturnValue)
   n4 = Branch(execute=@n1.Is Valid, Condition=@n3.ReturnValue)
   > n5 = Set TempData(TempData:Map of EHUDMessagePlacement Enums to Win Condition Placement Data Structures=@n3) -> (Output_Get:Map of EHUDMessagePlacement Enums to Win Condition Placement Data Structures=None)
   > Add!(
    execute
    TargetMap:Map of EHUDMessagePlacement Enums to Win Condition Placement Data Structures=@n5
    Key:"EHUDMessagePlacement Enum (by ref)"=BottomCenter
    Value_Padding:Margin Structure=Get Padding!(self:Overlay Slot Object Reference=@n1) -> (Padding:Margin Structure)
    Value_HorizontalAlignment:EHorizontalAlignment Enum=Get HorizontalAlignment!(self:Overlay Slot Object Reference=@n1) -> (HorizontalAlignment:EHorizontalAlignment Enum=HAlign_Fill)
    Value_VerticalAlignment:EVerticalAlignment Enum=Get VerticalAlignment!(self:Overlay Slot Object Reference=@n1) -> (VerticalAlignment:EVerticalAlignment Enum=VAlign_Fill)
   ) -> (then) {is=title K2Node_CallFunction}
   > Add!(
    execute
    TargetMap:Map of EWinConditionParentType Enums to Win Condition Placement Data Map Wrapper Structures=Get Placements() -> (Placements=None)
    Key:"EWinConditionParentType Enum (by ref)"=@n2
    Value:"Win Condition Placement Data Map Wrapper Structure (by ref)"=Make Win Condition Placement Data Map Wrapper!(Data:Map of EHUDMessagePlacement Enums to Win Condition Placement Data Structures=@n5) -> (WinConditionPlacementDataMapWrapper:Win Condition Placement Data Map Wrapper Structure) {is=title K2Node_MakeStruct}
   ) -> (then)
   Log HUD Message Warning(execute=@n4.else, InString="AthenaScoreAlert::FillInitialPositionData() Could not find a data type in Placements")
  ]}
 GetAllPlacements {nodes [
   GetAllPlacements() {is=title K2Node_FunctionEntry}
   n1 = Find(
    TargetMap=Get Placements() -> (Placements=None)
    Key=Select(Option 0=Literal enum EWinConditionParentType(Enum=Mobile), Option 1=Literal enum EWinConditionParentType(Enum=Desktop), Index=IsDesktop())
   )
   > n2 = Branch(Condition=@n1.ReturnValue)
   > Return Node!(execute, bSuccess:bool=true, ReturnValue:Map of EHUDMessagePlacement Enums to Win Condition Placement Data Structures=@n1)
   Log HUD Message Warning(execute=@n2.else, InString="AthenaScoreAlert::GetAllPlacements() Could not find a data type in Placements")
   > Return Node(bSuccess=false, ReturnValue=None)
  ]}
 GetParentSlot {nodes [
   Inputs()
   n1 = GetParent!(self:Widget Object Reference=Self-Reference()) -> (ReturnValue:Panel Widget Object Reference)
   > n2 = Is Valid(InputObject=@n1)
   n3 = SlotAsOverlaySlot(Widget=@n1)
   > n4 = Is Valid(InputObject=@n3)
   > Outputs!(Is Valid:Exec, Slot:Overlay Slot Object Reference=@n3, Is Not Valid:Exec=[@n4.Is Not Valid, @n2.Is Not Valid])
  ]}
 GetPlacementData {nodes [
   GetPlacementData() {is=title K2Node_FunctionEntry}
   > n1 = Sequence()
   > n2 = GetAllPlacements!(execute, self) -> (then, bSuccess:bool, ReturnValue:Map of EHUDMessagePlacement Enums to Win Condition Placement Data Structures=None) {is=title K2Node_CallFunction}
   > Branch(Condition=@n2)
   n3 = Find!(
    TargetMap:Map of EHUDMessagePlacement Enums to Win Condition Placement Data Structures=@n2.ReturnValue
    Key:"EHUDMessagePlacement Enum (by ref)"=Break HUDMessage Data!(HUDMessageData=Get CurrentMessage()) -> (Placement, MsgMediaSource, MessageFlags)
   ) -> (Value:Win Condition Placement Data Structure, ReturnValue)
   > n4 = Branch(Condition=@n3.ReturnValue)
   > Return Node!(execute, bSuccess=true, PlacementData:Win Condition Placement Data Structure=@n3)
   Log HUD Message Warning(execute=@n4.else, InString="AthenaScoreAlert::GetPlacementData() Could not find placement in AllPlacementData")
   Return Node(execute=@n1.then_1, bSuccess=false)
  ]}
 GetRichTextStyleFromMessage {nodes [
   n1 = GetRichTextStyleFromMessage!() -> (then, Message:"Text (by ref)") {is=title K2Node_FunctionEntry}
   n2 = "To String (Text)"(InText=@n1)
   n3 = int - int(A=FindSubstring(SearchIn=@n2, Substring=">"), B=1) {is=title K2Node_CallFunction}
   n4 = int + int(A=FindSubstring(SearchIn=@n2, Substring="<"), B=1) -> (ReturnValue=0) {is=title K2Node_CommutativeAssociativeBinaryOperator}
   > Return Node!(
    execute
    Style:Name=String To Name(InString=GetSubstring!(SourceString:String=@n2, StartIndex:int=@n4, Length:int=@n3) -> (ReturnValue:String))
   )
  ]}
 GetShowTimeInterval {nodes [
   n1 = GetShowTimeInterval!() -> (then, MessageData:"HUDMessage Data Structure (by ref)") {is=title K2Node_FunctionEntry}
   n2 = Break HUDMessage Data!(HUDMessageData=@n1) -> (MsgMediaSource, DisplayTime:float=~, MessageFlags)
   > n3 = Branch(Condition="Nearly Equal (Float)"(A=@n2.DisplayTime, B=0, ErrorTolerance=0))
   n4 = Get MessageData() -> (MessageData:"HUDMessage Data Structure (by ref)")
   n5 = Get DefaultShowTime() -> (DefaultShowTime:double)
   > Return Node!(
    execute
    ShowTimeInterval:double=Select!(
     None:double=0
     DefaultIntro:double=Get IntroShowTime() -> (IntroShowTime:double)
     DefaultMessage:double=@n5
     DefaultCriticalMessage:double=0
     CommIntro:double=0
     CommMessage:double=0
     CommCriticalMessage:double=0
     CornerIntro:double=0
     CornerMessage:double=0
     CornerCriticalMessage:double=0
     RespawnTurningOffWarning:double=@n5
     RespawnOffWarning:double=@n5
     CenterMessage:double=0
     CenterImportantMessage:double=0
     CenterCriticalMessage:double=0
     Index:EAthenaGameMsgType Enum=Break HUDMessage Data!(HUDMessageData=@n4) -> (MsgType:EAthenaGameMsgType Enum, MsgMediaSource, MessageFlags)
    ) -> (ReturnValue:double)
   )
   Return Node(execute=@n3.else, ShowTimeInterval=@n2.DisplayTime)
  ]}
 HUDMessageToText {nodes [
   n1 = HUDMessageToText!() -> (
    then
    HUDMessage_Placement:EHUDMessagePlacement Enum
    HUDMessage_PlayerState:Player State Object Reference
    HUDMessage_MessageTargets:int
    HUDMessage_TextTransformPolicy:ETextTransformPolicy Enum
    HUDMessage_DisableChannel:int
    HUDMessage_MsgType:EAthenaGameMsgType Enum
    HUDMessage_MsgText:Text
    HUDMessage_MsgSound:Sound Cue Object Reference
    HUDMessage_MsgMediaSource:Media Source Object Reference
    HUDMessage_MsgDelay:float=~
    HUDMessage_bIsTeamBased:bool=~
    HUDMessage_TeamIndex:int
    HUDMessage_DisplayTime:float=~
    HUDMessage_MessageFlags:EAthenaGameMessageDataFlags Enum
    HUDMessage_TargetPlayerController:Player Controller Object Reference
   ) {is=title K2Node_FunctionEntry}
   > n2 = Is Valid(InputObject=@n1.HUDMessage_MsgSound)
   n3 = Set MsgSoundText(execute=@n2.Is Not Valid, MsgSoundText:Text="LOCTEXT(\"nullptr\")") -> (Output_Get:Text)
   n4 = Enum to String!(Enumerator:EHUDMessagePlacement Enum=@n1) -> (ReturnValue:String)
   n5 = Enum to String(Enumerator:EAthenaGameMsgType Enum=@n1.HUDMessage_MsgType)
   Return Node!(
    execute=[@n3.then, @n6.then]
    Text:Text=Format Text!(
     Format="LOCTEXT(\"Placement: {P}, MsgType: {MTy}, MsgText: {MTe}, MsgSound: {MS}, MsgDelay: {MD}, IsTeamBased: {ITB}, TeamIndex: {TI}, DisplayTime: {DT}\")"
     P:Text="To Text (String)"(InString=@n4)
     MTy:Text="To Text (String)"(InString=@n5)
     MTe:Text=@n1.HUDMessage_MsgText
     MS:Text=Get MsgSoundText() -> (MsgSoundText:Text)
     MD:float=@n1.HUDMessage_MsgDelay
     ITB:Text="To Text (Boolean)"!(InBool:bool=@n1.HUDMessage_bIsTeamBased) -> (ReturnValue:Text)
     TI:int=@n1.HUDMessage_TeamIndex
     DT:float=@n1.HUDMessage_DisplayTime
    ) -> (Result)
   )
   n6 = Set MsgSoundText(execute=@n2.Is Valid, MsgSoundText="To Text (String)"(InString=GetDisplayName(Object=@n1.HUDMessage_MsgSound)))
  ]}
 HandleMessageQueuing {nodes [
   HandleMessageQueuing() {is=title K2Node_FunctionEntry}
   > n1 = Branch(Condition=IsVisible())
   n2 = Get CurrentMessage()
   > n3 = Branch(Condition="Equal (Enum)"(A=Break HUDMessage Data(HUDMessageData=@n2), B=DefaultCriticalMessage))
   > n4 = Is Creative()
   Return Node!(execute=[@n4.False, @n4.Unknown], bMessageWasQueued:bool=false)
   > Clear(TargetArray:Array of HUDMessage Data Structures=Get QueuedMessages() -> (QueuedMessages:Array of HUDMessage Data Structures))
   > Add(TargetArray:Array of HUDMessage Data Structures=Get QueuedMessages(), NewItem:"HUDMessage Data Structure (by ref)"=Get CurrentMessage()) {is=title K2Node_CallArrayFunction}
   > OnHideGameModeMessageRequest(MessageType=Break HUDMessage Data(HUDMessageData=Get CurrentMessage()), bInstantHide=false, bClearQueue=false, IncomingMessageChannel=0)
   > Return Node(bMessageWasQueued=true)
   Add(execute=@n3.else, TargetArray=Get QueuedMessages(), NewItem=Get CurrentMessage())
   > Return Node(bMessageWasQueued=true)
   Return Node(execute=@n1.else, bMessageWasQueued=false)
  ]}
 IsCreative {nodes [
   Inputs()
   > n1 = Cast To FortGameStateAthena(Object=GetGameState!() -> (ReturnValue:Game State Base Object Reference))
   > n2 = Branch(Condition=IsCreativeOrPlaygroundModePlaylist!(self:Fort Game State Athena Object Reference=@n1) -> (ReturnValue:bool))
   > Outputs!(True:Exec, False:Exec=@n2.else, Unknown:Exec=@n1.CastFailed)
  ]}
 IsDesktop {nodes [
   IsDesktop() {is=title K2Node_FunctionEntry}
   > n1 = FindParentWidgetOfType!(execute, StartingWidget:Widget Object Reference=Self-Reference(), Type:Widget Class Reference=&/AthenaHUD.AthenaHUD_C) -> (then, ReturnValue:Athena HUD Object Reference)
   > Return Node!(execute, bReturnValue:bool="Not Equal (Object)"!(A:Object Reference=@n1, B:Object Reference) -> (ReturnValue:bool))
  ]}
 IsMessageValidForViewTarget {nodes [
   IsMessageValidForViewTarget!() -> (then, Message:HUDMessage Data Structure) {is=title K2Node_FunctionEntry}
   > n1 = Sequence!(execute) -> (then_0, then_1, then_2:Exec, then_3:Exec, then_4:Exec)
   n2 = Break HUDMessage Data!(HUDMessageData=Get Message() -> (Message:HUDMessage Data Structure)) -> (MsgMediaSource, bIsTeamBased:bool=~, TeamIndex:int, MessageFlags)
   > Branch(Condition=@n2.bIsTeamBased)
   > n3 = DoesTeamMatchViewTargetsTeam!(execute, self, Team:int=@n2.TeamIndex) -> (then, bDoesMatchViewTargetsTeam:bool) {is=title K2Node_CallFunction}
   > Return Node!(execute, bIsValid:bool=@n3)
   n4 = Get Message()
   Branch(
    execute=@n1.then_1
    Condition=IsTeamAll!(InTeamValue:int=Break HUDMessage Data!(HUDMessageData=@n4) -> (MessageTargets:int, MsgMediaSource, MessageFlags)) -> (ReturnValue:bool)
   )
   > Return Node(bIsValid=true)
   n5 = Break HUDMessage Data!(HUDMessageData=Get Message()) -> (PlayerState:Player State Object Reference, MessageTargets, MsgMediaSource, MessageFlags)
   Branch(execute=@n1.then_2, Condition=IsTeamTriggeringPlayer!(InTeamValue:int=@n5.MessageTargets) -> (ReturnValue:bool))
   > n6 = DoesViewTargetMatchTriggeringPlayer!(execute, self, TriggeringPlayer:Player State Object Reference=@n5) -> (then, bDoesMatch:bool) {is=title K2Node_CallFunction}
   > Return Node(bIsValid=@n6)
   n7 = Break HUDMessage Data!(HUDMessageData=Get Message()) -> (MessageTargets, MsgMediaSource, MessageFlags)
   Branch(
    execute=@n1.then_3
    Condition=OR Boolean!(A=IsTeamFriendlies!(InTeamValue:int=@n7) -> (ReturnValue:bool), B=IsTeamEnemies!(InTeamValue:int=@n7) -> (ReturnValue:bool)) -> (ReturnValue)
   )
   > n8 = DoesViewTargetsTeamMatchTriggeringPlayersTeam!(
    execute
    self
    TriggeringPlayer:Player State Object Reference=Break HUDMessage Data!(HUDMessageData=Get Message()) -> (PlayerState:Player State Object Reference, MsgMediaSource, MessageFlags)
   ) -> (then, bDoesMatch:bool) {is=title K2Node_CallFunction}
   n9 = Break HUDMessage Data!(HUDMessageData=Get Message()) -> (MessageTargets:int, MsgMediaSource, MessageFlags)
   > Return Node(bIsValid=Select!(Option 0:bool=IsTeamEnemies(InTeamValue=@n9), Option 1:bool=IsTeamFriendlies(InTeamValue=@n9), Index:bool=@n8) -> (ReturnValue:bool))
   Return Node(execute=@n1.then_4, bIsValid=false)
  ]}
 IsMessageValidForWidget {nodes [
   n1 = IsMessageValidForWidget!() -> (then, Message:HUDMessage Data Structure) {is=title K2Node_FunctionEntry}
   > n2 = Branch(Condition=DoesWidgetHandleThisEnum!(
     self
     MessageType:EAthenaGameMsgType Enum=Break HUDMessage Data!(HUDMessageData=@n1) -> (MsgType:EAthenaGameMsgType Enum, MsgMediaSource, MessageFlags)
    ) -> (bIsValid:bool) {is=title K2Node_CallFunction})
   > Return Node(bIsValid=IsMessageValidForViewTarget!(self, Message:HUDMessage Data Structure=Get Message()) -> (bIsValid:bool) {is=title K2Node_CallFunction})
   Return Node(execute=@n2.else, bIsValid=false)
  ]}
 OnGameModeMessageRequest {nodes [
   n1 = OnGameModeMessageRequest!() -> (then, MessageData:Athena Game Message Data Structure) {is=title K2Node_FunctionEntry}
   > TryToDisplayMessage(CurrentMessage=ConvertToHUDMessage!(
     Placement:EHUDMessagePlacement Enum=BottomCenter
     Text Transform Policy:ETextTransformPolicy Enum=ToUpper
     AthenaGameMessageData:"Athena Game Message Data Structure (by ref)"=@n1
    ) -> (HUDMessageData:HUDMessage Data Structure) {is=title K2Node_CallFunction})
  ]}
 OnHideGameModeMessageRequest {nodes [
   OnHideGameModeMessageRequest!() -> (then, MessageType:EAthenaGameMsgType Enum=None, bInstantHide:bool=~, bClearQueue:bool=~, IncomingMessageChannel:int) {is=title K2Node_FunctionEntry}
   > Branch(Condition=DoesWidgetHandleThisEnum(MessageType=Get MessageType() -> (MessageType:EAthenaGameMsgType Enum=None)))
   n1 = Get IncomingMessageChannel() -> (IncomingMessageChannel:int=0)
   n2 = Break HUDMessage Data!(HUDMessageData=Get MostRecentDisplayedMessage()) -> (DisableChannel:int, MsgMediaSource, MessageFlags)
   > Branch(Condition=OR Boolean(
     A="integer <= integer"!(A:int=@n1, B:int=0) -> (ReturnValue:bool) {is=title K2Node_CallFunction}
     B="Equal (Integer)"(A=@n1, B=@n2) {is=title K2Node_CallFunction}
    ))
   > StopAnimation(InAnimation=Get Exclamation() -> (Exclamation:Widget Animation Object Reference))
   > StopAnimation(InAnimation=Get Intro())
   > Clear and Invalidate Timer by Handle(Handle=Get ShowTimeTimerHandle() -> (ShowTimeTimerHandle:Timer Handle Structure))
   > n3 = Branch(Condition=Get bClearQueue() -> (bClearQueue:bool))
   > n4 = Clear(TargetArray=Get QueuedMessages())
   BroadcastHUDMessageToggled(execute=[@n4.then, @n3.else], bShown=false)
   n5 = IsVisible()
   > n6 = Branch(Condition=OR Boolean(A=NOT Boolean(A=@n5), B=Get bInstantHide() -> (bInstantHide:bool)))
   > StopAnimation(InAnimation=Get Outro() -> (Outro:Widget Animation Object Reference))
   > Clear and Invalidate Timer by Handle(Handle=Get OutroTimerHandle() -> (OutroTimerHandle:Timer Handle Structure))
   > OnOutroAnimationFinished()
   PlayAnimation(execute=@n6.else, InAnimation=Get Outro())
   > n7 = Set Timer by Event(Delegate=Create Event(), Time=GetEndTime(self=Get Outro()), bLooping=false)
   > Set OutroTimerHandle(OutroTimerHandle:Timer Handle Structure=@n7) -> (Output_Get:Timer Handle Structure)
  ]}
 OnOutroAnimationFinished {nodes [
   OnOutroAnimationFinished() {is=title K2Node_FunctionEntry}
   > SetVisibility(InVisibility=Collapsed)
   n1 = Get QueuedMessages()
   > n2 = Branch(Condition="integer > integer"(A=Length(TargetArray:Array of HUDMessage Data Structures=@n1), B=0) {is=title K2Node_CallFunction})
   > Set NextMessage(NextMessage:HUDMessage Data Structure="Get (a ref)"(Array:Array of HUDMessage Data Structures=Get QueuedMessages(), Dimension 1=0) -> (Output:"HUDMessage Data Structure (by ref)")) -> (Output_Get:HUDMessage Data Structure)
   > Remove Index!(execute, TargetArray:Array of HUDMessage Data Structures=Get QueuedMessages(), IndexToRemove:int=0) -> (then)
   > TryToDisplayMessage(CurrentMessage=Get NextMessage() -> (NextMessage:HUDMessage Data Structure))
   SetHasActiveContent!(execute=@n2.else, self:Fort Prioritized Widget Interface, bNewActive:bool=false) -> (then)
  ]}
 OnShowTimeExpired {nodes [
   OnShowTimeExpired() {is=title K2Node_FunctionEntry}
   > OnHideGameModeMessageRequest(
    MessageType=Break HUDMessage Data!(HUDMessageData=Get CurrentMessage()) -> (MsgType:EAthenaGameMsgType Enum, MsgMediaSource, MessageFlags)
    bInstantHide=false
    bClearQueue=false
    IncomingMessageChannel=0
   ) {is=title K2Node_CallFunction}
  ]}
 PlayIntroAnimations {nodes [
   PlayIntroAnimations() {is=title K2Node_FunctionEntry}
   > PlayAnimation(InAnimation=Get Intro())
   > Branch(Condition=ShouldShowExclamation!(self) -> (bShouldShow:bool))
   > PlayAnimation(InAnimation=Get Exclamation())
   > Branch(Condition=ShouldColorizeExclamationTextBlock!(self) -> (bShouldColorize:bool))
   > UpdateExclamationTextBlockColors!(
    execute
    self
    Message:"Text (by ref)"=Break HUDMessage Data!(HUDMessageData=Get CurrentMessage()) -> (MsgText:Text, MsgMediaSource, MessageFlags)
   ) -> (then)
  ]}
 PlaySoundFromType {nodes [
   PlaySoundFromType!() -> (then, MessageData:HUDMessage Data Structure) {is=title K2Node_FunctionEntry}
   > n1 = Is Creative()
   n2 = Break HUDMessage Data!(HUDMessageData=Get MessageData() -> (MessageData:HUDMessage Data Structure)) -> (MsgSound:Sound Cue Object Reference, MsgMediaSource, MessageFlags)
   n3 = Set SoundCue(execute=[@n5.Is Valid, @n1.True], SoundCue:Sound Cue Object Reference=@n2) -> (Output_Get:Sound Cue Object Reference)
   n4 = Get!(execute=[@n7.then, @n3.then]) -> (then, else, SoundCue:Sound Cue Object Reference)
   > PlaySound2D(Sound=@n4)
   n5 = Is Valid(exec=[@n1.False, @n1.Unknown], InputObject=@n2)
   n6 = Get MessageData()
   n7 = Set SoundCue(
    execute=@n5.Is Not Valid
    SoundCue=Select!(
     None:Sound Cue Object Reference
     DefaultIntro:Sound Cue Object Reference=/Game/Athena/Sounds/LTMs/Score/ScoreLTM_ImportantMessage_Cue
     DefaultMessage:Sound Cue Object Reference=&ScoreLTM_ImportantMessage_Cue
     DefaultCriticalMessage:Sound Cue Object Reference
     CommIntro:Sound Cue Object Reference
     CommMessage:Sound Cue Object Reference
     CommCriticalMessage:Sound Cue Object Reference
     CornerIntro:Sound Cue Object Reference
     CornerMessage:Sound Cue Object Reference
     CornerCriticalMessage:Sound Cue Object Reference
     RespawnTurningOffWarning:Sound Cue Object Reference
     RespawnOffWarning:Sound Cue Object Reference
     CenterMessage:Sound Cue Object Reference
     CenterImportantMessage:Sound Cue Object Reference
     CenterCriticalMessage:Sound Cue Object Reference
     Index:EAthenaGameMsgType Enum=Break HUDMessage Data!(HUDMessageData=@n6) -> (MsgType:EAthenaGameMsgType Enum, MsgMediaSource, MessageFlags)
    ) -> (ReturnValue:Sound Cue Object Reference)
   )
  ]}
 SetPlacement {nodes [
   SetPlacement() {is=title K2Node_FunctionEntry}
   > n1 = GetPlacementData!(execute, self) -> (then, bSuccess:bool, PlacementData:Win Condition Placement Data Structure) {is=title K2Node_CallFunction}
   > Branch(Condition=@n1)
   > Set PlacementData(PlacementData:Win Condition Placement Data Structure=@n1.PlacementData) -> (Output_Get:Win Condition Placement Data Structure)
   n2 = Get!() -> (
    PlacementData_Padding:Margin Structure
    PlacementData_HorizontalAlignment:EHorizontalAlignment Enum
    PlacementData_VerticalAlignment:EVerticalAlignment Enum
   )
   n3 = SlotAsOverlaySlot(Widget=Get Content() -> (Content:Vertical Box Object Reference))
   > SetPadding!(execute, self=@n3, InPadding:Margin Structure=@n2) -> (then)
   > SetHorizontalAlignment(self=@n3, InHorizontalAlignment=@n2.PlacementData_HorizontalAlignment)
   > SetVerticalAlignment(self=@n3, InVerticalAlignment=@n2.PlacementData_VerticalAlignment)
  ]}
 ShouldColorizeExclamationTextBlock {nodes [
   ShouldColorizeExclamationTextBlock() {is=title K2Node_FunctionEntry}
   > n1 = Is Creative()
   Return Node!(execute=[@n1.False, @n1.Unknown], bShouldColorize:bool=false)
   > Return Node(bShouldColorize=true)
  ]}
 ShouldPlayIntroAnimations {nodes [
   ShouldPlayIntroAnimations() {is=title K2Node_FunctionEntry}
   > n1 = Is Creative()
   > Return Node!(execute, bShouldPlay:bool=true)
   n2 = Get CurrentMessage()
   n3 = Branch(execute=[@n1.False, @n1.Unknown], Condition="Equal (Enum)"(A=Break HUDMessage Data(HUDMessageData=@n2), B=DefaultCriticalMessage))
   n4 = Get PreviousMessage() -> (PreviousMessage:HUDMessage Data Structure)
   n5 = Get CurrentMessage()
   > n6 = Set ReturnValue(ReturnValue:bool="Not Equal (Enum)"(A=Break HUDMessage Data(HUDMessageData=@n4), B=Break HUDMessage Data(HUDMessageData=@n5))) -> (Output_Get:bool)
   Set PreviousMessage(execute=[@n6.then, @n7.then], PreviousMessage:HUDMessage Data Structure=Get CurrentMessage()) -> (Output_Get:HUDMessage Data Structure)
   > Return Node(bShouldPlay=Get ReturnValue() -> (ReturnValue:bool))
   n7 = Set ReturnValue(execute=@n3.else, ReturnValue=true)
  ]}
 ShouldShowExclamation {nodes [
   ShouldShowExclamation() {is=title K2Node_FunctionEntry}
   > n1 = Is Creative()
   Return Node!(execute=[@n1.False, @n1.Unknown], bShouldShow:bool=true)
   n2 = Break HUDMessage Data(HUDMessageData=Get CurrentMessage())
   n3 = "To Integer (Byte)"!(InByte:Byte=Literal enum EAthenaGameMessageDataFlags!(Enum:EAthenaGameMessageDataFlags Enum=DisableCreativeExclamation) -> (ReturnValue:EAthenaGameMessageDataFlags Enum) {is=title K2Node_EnumLiteral}) -> (ReturnValue:int=0)
   n4 = Bitwise AND(A:int="To Integer (Byte)"(InByte=@n2.MessageFlags) -> (ReturnValue=0), B:int=@n3) -> (ReturnValue:int=0) {is=title K2Node_CommutativeAssociativeBinaryOperator}
   > Return Node(bShouldShow=AND Boolean(A="Equal (Enum)"(A=@n2, B=DefaultCriticalMessage), B="Not Equal (Integer)"(A=@n4, B=@n3) {is=title K2Node_PromotableOperator}))
  ]}
 StartHideTimer {nodes [
   StartHideTimer() {is=title K2Node_FunctionEntry}
   > n1 = Is Creative()
   n2 = Get CurrentMessage()
   n3 = Branch(execute=[@n1.False, @n1.Unknown], Condition="Not Equal (Enum)"(A=Break HUDMessage Data(HUDMessageData=@n2), B=DefaultCriticalMessage))
   n4 = Set Timer by Event(
    execute=[@n3.then, @n1.True]
    Delegate=Create Event()
    Time=GetShowTimeInterval!(self, MessageData:"HUDMessage Data Structure (by ref)"=Get CurrentMessage()) -> (ShowTimeInterval:double) {is=title K2Node_CallFunction}
    bLooping=false
   )
   > Set ShowTimeTimerHandle(ShowTimeTimerHandle:Timer Handle Structure=@n4) -> (Output_Get:Timer Handle Structure)
  ]}
 TryToDisplayMessage {nodes [
   n1 = TryToDisplayMessage!() -> (then, CurrentMessage:HUDMessage Data Structure) {is=title K2Node_FunctionEntry}
   > Branch(Condition=IsMessageValidForWidget!(self, Message:HUDMessage Data Structure=@n1) -> (bIsValid:bool) {is=title K2Node_CallFunction})
   > SetHasActiveContent(bNewActive=true)
   > n2 = Set CurrentMessage(CurrentMessage=@n1)
   > SetTextTransformPolicy!(
    execute
    self:Rich Text Block Object Reference=Get MessageTextBlock()
    InTransformPolicy:ETextTransformPolicy Enum=Break HUDMessage Data!(HUDMessageData=@n2) -> (TextTransformPolicy:ETextTransformPolicy Enum, MsgMediaSource, MessageFlags)
   ) -> (then)
   > n3 = HandleMessageQueuing!(execute, self) -> (then, bMessageWasQueued:bool) {is=title K2Node_CallFunction}
   > n4 = Branch(Condition=@n3)
   DisplayMessage(execute=@n4.else) {is=title K2Node_CallFunction}
  ]}
 UpdateContentTranslation {nodes [
   n1 = UpdateContentTranslation!() -> (then, bShiftContentDown:bool=~) {is=title K2Node_FunctionEntry}
   n2 = Get ContentOffset() -> (ContentOffset:int=0)
   > SetRenderTranslation(
    self=Get Content()
    Translation=Select!(Option 0:Vector 2D Structure, Option 1_X:double=0, Option 1_Y:double="To Float (Integer)"(InInt=@n2), Index:bool=@n1) -> (ReturnValue:Vector 2D Structure)
   )
  ]}
 UpdateExclamationTextBlockColors {nodes [
   n1 = UpdateExclamationTextBlockColors!() -> (then, Message:"Text (by ref)") {is=title K2Node_FunctionEntry}
   > n2 = Get Data Table Row AthenaScoreAlertStyles!(
    execute:Exec
    DataTable:Data Table Object Reference=/Game/Athena/HUD/TextStyles/AthenaScoreAlertStyles
    RowName:Name=GetRichTextStyleFromMessage!(self, Message:"Text (by ref)"=@n1) -> (Style:Name=None) {is=title K2Node_CallFunction}
   ) -> (then:Exec, RowNotFound:Exec, ReturnValue:Rich Text Style Row Structure) {is=title K2Node_GetDataTableRow}
   > SetColorAndOpacity(
    self=Get ExclamationTextBlock()
    InColorAndOpacity=GetTextFillColor!(self:Athena Win Condition Msg Base Object Reference, Row:"Rich Text Style Row Structure (by ref)"=@n2) -> (ReturnValue:Slate Color Structure)
   )
   n3 = Break Slate Font Info!(SlateFontInfo:"Slate Font Info Structure (by ref)"=Get Font!(self:Text Object Reference=Get ExclamationTextBlock()) -> (Font:Slate Font Info Structure)) -> (
    FontObject:Object Reference
    FontMaterial:Object Reference
    OutlineSettings:Font Outline Settings Structure
    TypefaceFontName:Name
    Size:float=~
    LetterSpacing:int
    SkewAmount:float=~
    bForceMonospaced:bool=~
    bMaterialIsStencil:bool=~
    MonospacedWidth:float=~
   ) {is=title K2Node_BreakStruct}
   > SetFont(
    self=Get ExclamationTextBlock()
    InFontInfo=MakeSlateFontInfo!(
     FontObject:Object Reference=@n3
     FontMaterial:Object Reference=@n3.FontMaterial
     OutlineSettings:Font Outline Settings Structure=(OutlineSize=0,bMiteredCorners=False,bSeparateFillAlpha=False,bApplyOutlineToDropShadows=False,OutlineMaterial=None,OutlineColor=(R=0,G=0,B=0,A=1))
     TypefaceFontName:Name=@n3.TypefaceFontName
     Size:float=@n3.Size
     LetterSpacing:int=0
     SkewAmount:float=0
     bForceMonospaced:bool=false
     bMaterialIsStencil:bool=false
     MonospacedWidth:float=1
    ) -> (ReturnValue:Slate Font Info Structure)
   )
  ]}
]
variables [
 ContentOffset {type=integer, default=45}
 CurrentMessage {
  type=HUDMessageData
  default {
   bIsTeamBased=false
   disableChannel=0
   displayTime=5
   messageFlags=None
   messageTargets=0
   msgDelay=0
   msgMediaSource=null
   msgSound=null
   msgText=""
   msgType=None
   placement=None
   playerState=null
   teamIndex=0
   textTransformPolicy=None
  }
 }
 DefaultShowTime {type=number, default=6}
 IntroShowTime {type=number, default=15}
 MostRecentDisplayedMessage {
  type=HUDMessageData
  default {
   bIsTeamBased=false
   disableChannel=0
   displayTime=5
   messageFlags=None
   messageTargets=0
   msgDelay=0
   msgMediaSource=null
   msgSound=null
   msgText=""
   msgType=None
   placement=None
   playerState=null
   teamIndex=0
   textTransformPolicy=None
  }
 }
 NextMessage {
  type=HUDMessageData
  default {
   bIsTeamBased=false
   disableChannel=0
   displayTime=5
   messageFlags=None
   messageTargets=0
   msgDelay=0
   msgMediaSource=null
   msgSound=null
   msgText=""
   msgType=None
   placement=None
   playerState=null
   teamIndex=0
   textTransformPolicy=None
  }
 }
 OutroTimerHandle {type=TimerHandle}
 Placements {default {
   desktop {data {
     centerRight {horizontalAlignment=HAlign_Right, verticalAlignment=VAlign_Center, padding=<ltrb 0 0 50 0>}
     topCenter {horizontalAlignment=HAlign_Center, verticalAlignment=VAlign_Top, padding=<ltrb 0 200 0 0>}
    }}
   mobile {data {
     centerRight {horizontalAlignment=HAlign_Right, verticalAlignment=VAlign_Center, padding=<ltrb 0 0 400 0>}
     topCenter {horizontalAlignment=HAlign_Center, verticalAlignment=VAlign_Top, padding=<ltrb 0 150 0 0>}
    }}
  }}
 PreviousMessage {
  type=HUDMessageData
  default {
   bIsTeamBased=false
   disableChannel=0
   displayTime=5
   messageFlags=None
   messageTargets=0
   msgDelay=0
   msgMediaSource=null
   msgSound=null
   msgText=""
   msgType=None
   placement=None
   playerState=null
   teamIndex=0
   textTransformPolicy=None
  }
 }
 QueuedMessages {type="array<HUDMessageData>", default []}
 ShowTimeTimerHandle {type=TimerHandle}
 TestMessagePlacement {type=EHUDMessagePlacement, default=BottomCenter}
 TestMessageType {type=integer, default=2}
 TextStyleDataAsset {type=boolean}
]
widgets {
 info {inheritedWidgetCount=1, parentClass=/Script/FortniteUI.AthenaWinConditionMsgBase, rootWidgetClass=&Overlay, widgetCount=7}
 widgets [
  Overlay_0 {widgetClassPath=&Overlay, widget=Overlay_0}
  Content {
   widgetClassPath=&VerticalBox
   bIsVariable=true
   parent=Overlay_0
   widget=Content
   slot {class=OverlaySlot, properties {horizontalAlignment=HAlign_Center, verticalAlignment=VAlign_Center}}
  }
  OverlayFeedback {
   widgetClassPath=&Overlay
   bIsVariable=true
   parent=Content
   widget=OverlayFeedback
   slot {class=VerticalBoxSlot, properties {horizontalAlignment=HAlign_Center}}
  }
  ExclamationTextBlock {
   widgetClassPath=&TextBlock
   bIsVariable=true
   parent=OverlayFeedback
   widget=ExclamationTextBlock
   properties {
    renderOpacity=0
    text="!"
    colorAndOpacity {colorUseRule=UseColor_Specified, specifiedColor=<rgba 1 0.679 0.3121 1>}
    font {
     bForceMonospaced=false
     bMaterialIsStencil=false
     fontMaterial=/FortUILibrary/Materials/Foundation/M_Font_Horizon_Dynamic
     fontObject=&Burbank
     letterSpacing=0
     monospacedWidth=1
     size=72
     skewAmount=0
     typefaceFontName=Burbank-Black
     outlineSettings {
      bApplyOutlineToDropShadows=false
      bMiteredCorners=false
      bSeparateFillAlpha=false
      outlineMaterial=&M_Font_Horizon_Dynamic
      outlineSize=2
      outlineColor=<rgba 0.0677 0.0214 0.0126 1>
     }
    }
    renderTransformPivot=<0.5 0.7834>
   }
   slot {class=OverlaySlot, properties {horizontalAlignment=HAlign_Center, padding=<ltrb -200 -200 -200 -200>}}
  }
  SizeBox_0 {
   widgetClassPath=&SizeBox
   parent=OverlayFeedback
   widget=SizeBox_0
   properties {maxDesiredWidth=720, widthOverride=720}
   slot {class=OverlaySlot, properties {horizontalAlignment=HAlign_Center}}
  }
  MessageTextBlock {
   widgetClassPath=/Script/UMG.RichTextBlock
   bIsVariable=true
   parent=SizeBox_0
   widget=MessageTextBlock
   properties {
    justification=Center
    lineHeightPercentage=0.85
    text="<RespawnAlert>RESPAWNING</> IS NOW <enemy>OFF!</>"
    textStyleSet=&AthenaScoreAlertStyles
    textTransformPolicy=ToUpper
    decoratorClasses [/Game/Athena/HUD/TextStyles/AthenaScoreAlertDecorator.AthenaScoreAlertDecorator_C]
    renderTransform {angle=0, scale=<1 1>, shear=<-4 0>, translation=<0 0>}
   }
   slot {class=SizeBoxSlot, properties {horizontalAlignment=HAlign_Center, verticalAlignment=VAlign_Top}}
  }
  TutorialHighlight {widgetClassPath=&TutorialHighlight}
 ]
}
