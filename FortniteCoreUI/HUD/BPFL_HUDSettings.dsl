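; Blueprint function library that reports whether the new HUD, new quickbar and itemization quickbar are enabled.
; New HUD and quickbar checks read scalable-float defaults from ExperienceSettingsComponent in Creative/Playground and VK edit or play playlists, and BattleRoyalGameUI console variables otherwise.
; The itemization check queries the player controller's InventorySwitcher component, defaulting to false when either is invalid.
; In those Creative and VK playlists, an enabled itemization quickbar forces the new quickbar off.
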(fn IsItemizationQuickbarEnabled
  (bind _worldcontext (Variables|Get__WorldContext))
  (bind _returnvalue (Game|GetPlayerController _worldcontext))
  (bind _returnvalue_2 (Actor|GetComponentbyClass _returnvalue "/Script/FortniteGame.FortControllerComponent_InventorySwitcher"))
  (Utilities|IsValid _returnvalue
    (:Is%20Valid
      (Utilities|IsValid _returnvalue_2
        (:Is%20Valid
          (return (InventorySwitcher|IsItemizationCustomUIRequired _returnvalue_2)))
        (:Is%20Not%20Valid
          _)))
    (:Is%20Not%20Valid
      _))
  (return false))

(fn IsNewHUDEnabled
  (params (WorldContextObject :type Object%20Reference))
  (if (Math|Boolean|ORBoolean (Class|FortGameStateAthena|IsCreativeorPlaygroundModePlaylist (Fortnite|Athena|GameState|GetGameStateAthena WorldContextObject)) (Class|FortGameStateAthena|IsVKEditPlaylist (Fortnite|Athena|GameState|GetGameStateAthena WorldContextObject)) (Class|FortGameStateAthena|IsVKPlayPlaylist (Fortnite|Athena|GameState|GetGameStateAthena WorldContextObject)))
    (return (ScalableFloat|GetValueasBoolwithContext WorldContextObject (Class|GetClassDefaults "/FortniteCoreUI/Settings/ExperienceSettingsComponent.ExperienceSettingsComponent_C")))
    (else
      (bind _returnvalue (Development|GetConsoleVariableBoolValue "BattleRoyalGameUI.NewHUD.Enabled"))
      (return _returnvalue))))

(fn IsNewQuickbarEnabled
  (params (WorldContextObject :type Object%20Reference))
  (if (Math|Boolean|ORBoolean (Class|FortGameStateAthena|IsCreativeorPlaygroundModePlaylist (Fortnite|Athena|GameState|GetGameStateAthena WorldContextObject)) (Class|FortGameStateAthena|IsVKEditPlaylist (Fortnite|Athena|GameState|GetGameStateAthena WorldContextObject)) (Class|FortGameStateAthena|IsVKPlayPlaylist (Fortnite|Athena|GameState|GetGameStateAthena WorldContextObject)))
    (bind _isenabled (CallFunction|IsItemizationQuickbarEnabled WorldContextObject))
    (if _isenabled
      (return false)
      (else
        (return (ScalableFloat|GetValueasBoolwithContext WorldContextObject (Class|GetClassDefaults "/FortniteCoreUI/Settings/ExperienceSettingsComponent.ExperienceSettingsComponent_C")))))
    (else
      (bind _returnvalue (Development|GetConsoleVariableBoolValue "BattleRoyalGameUI.NewQuickbar.Enabled"))
      (return _returnvalue))))
