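# Sound cue for the Creative Move Tool gadget's highlight sound, variant 03.
# A single wave player feeds an Enveloper that fades volume out over a very short window while pitch follows a rising cubic curve, with slight random pitch and volume variation.
# It plays quietly in the Creative gadgets sound class, using the Move Tool attenuation asset.
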SoundCue /Game/Sounds/Creative/Gadgets/MoveTool/MoveTool_Highlight_03_Cue
root=Enveloper_0
nodes [
 {
  id=Enveloper_0
  children [WavePlayer_0]
  properties {
   pitchMax=1.05
   pitchMin=0.95
   volumeMax=0.9
   volumeMin=0.8
   pitchCurve {editorCurveData {keys [{arriveTangent=0.8408, interpMode=RCIM_Cubic, leaveTangent=0.8408, tangentMode=RCTM_User, time=0}, <at 0.11 0.7>]}}
   volumeCurve {editorCurveData {keys [{interpMode=RCIM_Cubic, tangentMode=RCTM_User, time=0, value=1}, <at 0.11 0>]}}
  }
 }
 {id=WavePlayer_0, properties {soundWaveAssetPtr=&/MoveTool_Highlight_03}}
]
properties {
 attenuationSettings=/Game/Sounds/Creative/Attenuations/Fort_Gadget_MoveTool_Att
 maxDistance=4000
 soundClassObject=&Fort_Gadgets_Creative
 volumeMultiplier=0.3
}
